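; Blueprint interface for the Fortnite conversation UI, named for conversation brief elements.
; It declares a single function, SetSelectedVisuals, which takes an IsSelected boolean and, by its name, switches an element's selected look.
; The file holds only the signature, with no logic of its own.
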(fn SetSelectedVisuals
  (params (IsSelected :type Boolean)))
