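# Conversation UI widget derived from FortBasicBrief that displays a choice's text on a translucent dark panel.
# A width-capped SizeBox holds an Overlay and padded Border, then a sheared inner Overlay containing a CommonRichTextBlock with a bold Burbank style override and text wrapping.
# OnChoiceTextReceived writes the incoming text into that block.
# Construct and SetSelectedVisuals are empty, so this Blueprint adds no visual change on selection.
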WidgetBlueprint /FortniteConversationUI/Conversation/FortBasicBrief
graphs [EventGraph {nodes [
   n1 = Event OnChoiceTextReceived!() -> (OutputDelegate, then, Text:"Text (by ref)") {is=title K2Node_Event}
   > SetText(self=Get Text!() -> (Text:Common Rich Text Block Object Reference), InText=@n1.Text)
   Event Construct()
   Event SetSelectedVisuals!() -> (OutputDelegate, then, IsSelected:bool) {is=title K2Node_Event}
  ]}]
variables []
widgets {
 info {inheritedWidgetCount=0, parentClass=/Script/FortniteConversationUI.FortBasicBrief, rootWidgetClass=&SizeBox, widgetCount=5}
 widgets [
  SizeBox_36 {
   widgetClassPath=&SizeBox
   widget=SizeBox_36
   properties {bOverride_MaxDesiredWidth=true, heightOverride=350, maxDesiredWidth=900, widthOverride=550}
  }
  Overlay_4 {widgetClassPath=&Overlay, parent=SizeBox_36, widget=Overlay_4, slot {class=SizeBoxSlot}}
  Border_168 {
   widgetClassPath=&Border
   parent=Overlay_4
   widget=Border_168
   properties {brushColor=<rgba 0.0052 0.0052 0.0056 0.6>, padding=<ltrb 18 10 18 10>}
   slot {class=OverlaySlot}
  }
  Overlay_0 {
   widgetClassPath=&Overlay
   parent=Border_168
   widget=Overlay_0
   properties {renderTransform {angle=0, scale=<1 1>, shear=<-8 0>, translation=<0 0>}}
   slot {class=BorderSlot, properties {padding=<ltrb 18 10 18 10>}}
  }
  Text {
   widgetClassPath=&CommonRichTextBlock
   bIsVariable=true
   parent=Overlay_0
   widget=Text
   properties {
    bOverrideDefaultStyle=true
    text=Back
    wrapTextAt=900
    defaultTextStyleOverride {
     overflowPolicy=Clip
     transformPolicy=None
     colorAndOpacity {colorUseRule=UseColor_Specified, specifiedColor=<rgba 1 1 1 1>}
     font {
      bForceMonospaced=false
      bMaterialIsStencil=false
      fontMaterial=null
      fontObject=&BurbankSmall
      letterSpacing=0
      monospacedWidth=1
      size=20
      skewAmount=0
      typefaceFontName=Bold
      outlineSettings {
       bApplyOutlineToDropShadows=false
       bMiteredCorners=false
       bSeparateFillAlpha=false
       outlineMaterial=null
       outlineSize=0
       outlineColor=<rgba 0 0 0 1>
      }
     }
     highlightColor {colorUseRule=UseColor_Specified, specifiedColor=<rgba 0 0 0 1>}
     highlightShape {
      bIsDynamicallyLoaded=false
      drawAs=Image
      imageType=NoImage
      mirroring=NoMirror
      resourceName=None
      resourceObject=null
      tiling=NoTile
      imageSize=<32 32>
      margin=<ltrb 0 0 0 0>
      outlineSettings {
       bUseBrushTransparency=false
       roundingType=HalfHeightRadius
       width=0
       color {colorUseRule=UseColor_Specified, specifiedColor=<rgba 0 0 0 0>}
       cornerRadii=<0 0 0 0>
      }
      tintColor {colorUseRule=UseColor_Specified, specifiedColor=<rgba 1 1 1 1>}
      uVRegion {bIsValid=false, max=<0 0>, min=<0 0>}
     }
     selectedBackgroundColor {colorUseRule=UseColor_Foreground, specifiedColor=<rgba 1 0 1 1>}
     shadowColorAndOpacity=<rgba 0 0 0 1>
     shadowOffset=<0 0>
     strikeBrush {
      bIsDynamicallyLoaded=false
      drawAs=Image
      imageType=NoImage
      mirroring=NoMirror
      resourceName=None
      resourceObject=null
      tiling=NoTile
      imageSize=<32 32>
      margin=<ltrb 0 0 0 0>
      outlineSettings {
       bUseBrushTransparency=false
       roundingType=HalfHeightRadius
       width=0
       color {colorUseRule=UseColor_Specified, specifiedColor=<rgba 0 0 0 0>}
       cornerRadii=<0 0 0 0>
      }
      tintColor {colorUseRule=UseColor_Specified, specifiedColor=<rgba 1 1 1 1>}
      uVRegion {bIsValid=false, max=<0 0>, min=<0 0>}
     }
     underlineBrush {
      bIsDynamicallyLoaded=false
      drawAs=Image
      imageType=NoImage
      mirroring=NoMirror
      resourceName=None
      resourceObject=null
      tiling=NoTile
      imageSize=<32 32>
      margin=<ltrb 0 0 0 0>
      outlineSettings {
       bUseBrushTransparency=false
       roundingType=HalfHeightRadius
       width=0
       color {colorUseRule=UseColor_Specified, specifiedColor=<rgba 0 0 0 0>}
       cornerRadii=<0 0 0 0>
      }
      tintColor {colorUseRule=UseColor_Specified, specifiedColor=<rgba 1 1 1 1>}
      uVRegion {bIsValid=false, max=<0 0>, min=<0 0>}
     }
    }
   }
   slot {class=OverlaySlot}
  }
 ]
}
