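# FortChatBrief is a conversation speech-bubble widget derived from FortBasicBrief, with a message, tail and offset drop shadow.
# Events set the message text and switch bubble, tail and text colors by selection state, while SetTailOrientation aligns and mirrors both tail images left or right.
# The outer container is sheared and the bubble art counter-sheared, leaving only the text slanted.
# A short NewMessage animation pops the Background scale, though these graphs never play it.
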WidgetBlueprint /FortniteConversationUI/Conversation/FortChatBrief
animations [NewMessage {
  binding_count=1
  length_seconds=0.3087
  object=&self:NewMessage
  sequence {bindings [Background {
     id=#1
     display_name=Overlay_BubbleContainer
     possessed_class=Image
     tracks [{
       type=MovieScene2DTransformTrack
       sections [{
         type=MovieScene2DTransformSection
         channels [
          Translation.X {type=MovieSceneScriptingFloatChannel}
          Translation.Y {type=MovieSceneScriptingFloatChannel}
          Angle {type=MovieSceneScriptingFloatChannel}
          Scale.X {type=MovieSceneScriptingFloatChannel, keys [[0,1],[1.048,1.05],[6.174,1]]}
          Scale.Y {type=MovieSceneScriptingFloatChannel, keys [[0,1],[1.048,1.1],[6.174,1]]}
          Shear.X {type=MovieSceneScriptingFloatChannel}
          Shear.Y {type=MovieSceneScriptingFloatChannel}
         ]
        }]
      }]
    }]}
 }]
graphs [
 EventGraph {nodes [
   n1 = Event SetSelectedVisuals!() -> (OutputDelegate, then, IsSelected:bool) {is=title K2Node_Event}
   > SetColorAndOpacity(
    self:Image Object Reference=[Get Background() -> (Background:Image Object Reference), Get Image_Tail() -> (Image_Tail:Image Object Reference)]
    InColorAndOpacity:Linear Color Structure=Select(Option 0:Linear Color Structure=(R=0.0052,G=0.0052,B=0.0056,A=1), Option 1:Linear Color Structure=(R=1,G=1,B=1,A=1), Index=@n1.IsSelected) -> (ReturnValue:Linear Color Structure)
   )
   > SetColorAndOpacity!(
    execute
    self:Text Object Reference=Get Message() -> (Message:Common Text Object Reference)
    InColorAndOpacity_SpecifiedColor:Linear Color Structure=Select(Option 0=(R=0.3372,G=0.3372,B=0.3372,A=1), Option 1=(R=0.0052,G=0.0052,B=0.0056,A=1), Index=@n1.IsSelected)
    InColorAndOpacity_ColorUseRule:ESlateColorStylingMode Enum
   ) -> (then)
   n2 = Event OnChoiceTextReceived!() -> (OutputDelegate, then, Text:"Text (by ref)") {is=title K2Node_Event}
   > "SetText (Text)"(self=Get Message(), InText=@n2.Text)
   Event PreConstruct()
   Event Construct()
  ]}
 SetTailOrientation {nodes [
   n1 = SetTailOrientation!() -> (then, isPointingToRight:bool=~) {is=title K2Node_FunctionEntry}
   n2 = SlotAsOverlaySlot(Widget=Get Image_Tail_Shadow() -> (Image_Tail_Shadow:Image Object Reference))
   n3 = SlotAsOverlaySlot(Widget=Get Image_Tail())
   > SetPadding!(
    execute
    self:Overlay Slot Object Reference=[@n2, @n3]
    InPadding_Left:float=Select(Option 0:float=40, Option 1:float=0, Index=@n1) -> (ReturnValue:float)
    InPadding_Top:float=0
    InPadding_Right:float=Select(Option 0=0, Option 1=40, Index=@n1)
    InPadding_Bottom:float=0
   ) -> (then)
   > SetHorizontalAlignment(
    self=[@n3, @n2]
    InHorizontalAlignment=Select(Option 0:EHorizontalAlignment Enum=HAlign_Left, Option 1:EHorizontalAlignment Enum=HAlign_Right, Index=@n1) -> (ReturnValue:EHorizontalAlignment Enum)
   )
   > SetRenderScale!(
    execute
    self:Widget Object Reference=[Get Image_Tail(), Get Image_Tail_Shadow()]
    Scale:Vector 2D Structure=Select(Option 0:Vector 2D Structure=(X=-1,Y=1), Option 1:Vector 2D Structure=(X=1,Y=1), Index=@n1) -> (ReturnValue:Vector 2D Structure)
   ) -> (then)
  ]}
]
variables []
widgets {
 info {
  inheritedWidgetCount=0
  parentClass=&/FortBasicBrief.FortBasicBrief_C
  rootWidgetClass=&SizeBox
  widgetCount=15
 }
 widgets [
  SizeBox_0 {widgetClassPath=&SizeBox, widget=SizeBox_0, properties {widthOverride=650}}
  Overlay_DialogContainer {
   widgetClassPath=&Overlay
   parent=SizeBox_0
   widget=Overlay_DialogContainer
   properties {renderTransform {angle=0, scale=<1 1>, shear=<-8 0>, translation=<0 0>}}
   slot {class=SizeBoxSlot}
  }
  Overlay_BubbleContainer {
   widgetClassPath=&Overlay
   parent=Overlay_DialogContainer
   widget=Overlay_BubbleContainer
   slot {class=OverlaySlot, properties {horizontalAlignment=HAlign_Fill, verticalAlignment=VAlign_Fill}}
  }
  Overlay_47 {
   widgetClassPath=&Overlay
   parent=Overlay_BubbleContainer
   widget=Overlay_47
   properties {renderTransform {angle=0, scale=<1 1>, shear=<8 0>, translation=<0 0>}}
   slot {class=OverlaySlot, properties {horizontalAlignment=HAlign_Fill, verticalAlignment=VAlign_Fill}}
  }
  Overlay {
   widgetClassPath=&Overlay
   parent=Overlay_47
   widget=Overlay
   slot {class=OverlaySlot, properties {horizontalAlignment=HAlign_Fill, verticalAlignment=VAlign_Fill}}
  }
  Border_Shadow {
   widgetClassPath=&Border
   parent=Overlay
   widget=Border_Shadow
   properties {
    background {
     bIsDynamicallyLoaded=false
     drawAs=NoDrawType
     imageType=NoImage
     mirroring=NoMirror
     resourceName=None
     resourceObject=null
     tiling=NoTile
     imageSize=<32 32>
     margin=<ltrb 0 0 0 0>
     outlineSettings {
      bUseBrushTransparency=false
      roundingType=HalfHeightRadius
      width=0
      color {colorUseRule=UseColor_Specified, specifiedColor=<rgba 0 0 0 0>}
      cornerRadii=<0 0 0 0>
     }
     tintColor {colorUseRule=UseColor_Specified, specifiedColor=<rgba 1 1 1 1>}
     uVRegion {bIsValid=false, max=<0 0>, min=<0 0>}
    }
    contentColorAndOpacity=<rgba 0.0052 0.0052 0.0056 0.4>
    padding=<ltrb 0 0 0 0>
    renderTransform {angle=0, scale=<1 1>, shear=<0 0>, translation=<3 6>}
   }
   slot {class=OverlaySlot, properties {horizontalAlignment=HAlign_Fill, verticalAlignment=VAlign_Fill}}
  }
  Overlay_1 {
   widgetClassPath=&Overlay
   parent=Border_Shadow
   widget=Overlay_1
   properties {bOverride_Cursor=true}
   slot {class=BorderSlot, properties {padding=<ltrb 0 0 0 0>}}
  }
  Background_Shadow {
   widgetClassPath=&Image
   parent=Overlay_1
   widget=Background_Shadow
   properties {
    visibility=SelfHitTestInvisible
    brush {
     bIsDynamicallyLoaded=false
     drawAs=Box
     imageType=NoImage
     mirroring=NoMirror
     resourceName=None
     resourceObject=/Game/Athena/HUD/Materials/M_UI_MessageBG_Inst
     tiling=NoTile
     imageSize=<64 64>
     margin=<ltrb 0.4 0 0.4 0>
     outlineSettings {
      bUseBrushTransparency=false
      roundingType=HalfHeightRadius
      width=0
      color {colorUseRule=UseColor_Specified, specifiedColor=<rgba 0 0 0 0>}
      cornerRadii=<0 0 0 0>
     }
     tintColor {colorUseRule=UseColor_Specified, specifiedColor=<rgba 1 1 1 1>}
     uVRegion {bIsValid=false, max=<0 0>, min=<0 0>}
    }
    renderTransformPivot=<0.5 0.9>
   }
   slot {class=OverlaySlot, properties {horizontalAlignment=HAlign_Fill, verticalAlignment=VAlign_Fill}}
  }
  Image_Tail_Shadow {
   widgetClassPath=&Image
   bIsVariable=true
   parent=Overlay_1
   widget=Image_Tail_Shadow
   properties {
    visibility=SelfHitTestInvisible
    brush {
     bIsDynamicallyLoaded=false
     drawAs=Image
     imageType=FullColor
     mirroring=NoMirror
     resourceName=None
     resourceObject=/Game/Athena/UI/Frontend/Parties/Art/T_UI_LightningBanner_Tail_2
     tiling=NoTile
     imageSize=<32 28>
     margin=<ltrb 0 0 0 0>
     outlineSettings {
      bUseBrushTransparency=false
      roundingType=HalfHeightRadius
      width=0
      color {colorUseRule=UseColor_Specified, specifiedColor=<rgba 0 0 0 0>}
      cornerRadii=<0 0 0 0>
     }
     tintColor {colorUseRule=UseColor_Specified, specifiedColor=<rgba 1 1 1 1>}
     uVRegion {bIsValid=false, max=<0 0>, min=<0 0>}
    }
    renderTransform {angle=0, scale=<1 1>, shear=<0 0>, translation=<0 28>}
    renderTransformPivot=<0.5 1>
   }
   slot {class=OverlaySlot, properties {horizontalAlignment=HAlign_Right, verticalAlignment=VAlign_Bottom, padding=<ltrb 0 0 40 0>}}
  }
  Overlay_2 {
   widgetClassPath=&Overlay
   parent=Overlay
   widget=Overlay_2
   properties {bOverride_Cursor=true}
   slot {class=OverlaySlot, properties {horizontalAlignment=HAlign_Fill, verticalAlignment=VAlign_Fill}}
  }
  Background {
   widgetClassPath=&Image
   bIsVariable=true
   parent=Overlay_2
   widget=Background
   properties {
    brush {
     bIsDynamicallyLoaded=false
     drawAs=Box
     imageType=NoImage
     mirroring=NoMirror
     resourceName=None
     resourceObject=&M_UI_MessageBG_Inst
     tiling=NoTile
     imageSize=<64 64>
     margin=<ltrb 0.4 0 0.4 0>
     outlineSettings {
      bUseBrushTransparency=false
      roundingType=HalfHeightRadius
      width=0
      color {colorUseRule=UseColor_Specified, specifiedColor=<rgba 0 0 0 0>}
      cornerRadii=<0 0 0 0>
     }
     tintColor {colorUseRule=UseColor_Specified, specifiedColor=<rgba 1 1 1 1>}
     uVRegion {bIsValid=false, max=<0 0>, min=<0 0>}
    }
    colorAndOpacity=<rgba 0.0052 0.0052 0.0056 1>
    renderTransformPivot=<0.5 0.9>
   }
   slot {class=OverlaySlot, properties {horizontalAlignment=HAlign_Fill, verticalAlignment=VAlign_Fill}}
  }
  Image_Tail {
   widgetClassPath=&Image
   bIsVariable=true
   parent=Overlay_2
   widget=Image_Tail
   properties {
    visibility=SelfHitTestInvisible
    brush {
     bIsDynamicallyLoaded=false
     drawAs=Image
     imageType=FullColor
     mirroring=NoMirror
     resourceName=None
     resourceObject=&T_UI_LightningBanner_Tail_2
     tiling=NoTile
     imageSize=<32 28>
     margin=<ltrb 0 0 0 0>
     outlineSettings {
      bUseBrushTransparency=false
      roundingType=HalfHeightRadius
      width=0
      color {colorUseRule=UseColor_Specified, specifiedColor=<rgba 0 0 0 0>}
      cornerRadii=<0 0 0 0>
     }
     tintColor {colorUseRule=UseColor_Specified, specifiedColor=<rgba 1 1 1 1>}
     uVRegion {bIsValid=false, max=<0 0>, min=<0 0>}
    }
    colorAndOpacity=<rgba 0.0052 0.0052 0.0056 1>
    renderTransform {angle=0, scale=<1 1>, shear=<0 0>, translation=<0 28>}
    renderTransformPivot=<0.5 1>
   }
   slot {class=OverlaySlot, properties {horizontalAlignment=HAlign_Right, verticalAlignment=VAlign_Bottom, padding=<ltrb 0 0 40 0>}}
  }
  ScaleBox_0 {
   widgetClassPath=&ScaleBox
   parent=Overlay_BubbleContainer
   widget=ScaleBox_0
   properties {stretch=UserSpecified}
   slot {class=OverlaySlot, properties {horizontalAlignment=HAlign_Center, verticalAlignment=VAlign_Center, padding=<ltrb 20 10 20 10>}}
  }
  Overlay_0 {widgetClassPath=&Overlay, parent=ScaleBox_0, widget=Overlay_0, slot {class=ScaleBoxSlot}}
  Message {
   widgetClassPath=&CommonTextBlock
   bIsVariable=true
   parent=Overlay_0
   widget=Message
   properties {
    justification=Center
    text="Lorem Ipsum Dolor Amit Sequitor  Ipsum Dolor Amit Sequitor"
    wrapTextAt=462.6352
    colorAndOpacity {colorUseRule=UseColor_Specified, specifiedColor=<rgba 0.3372 0.3372 0.3372 1>}
    font {
     bForceMonospaced=false
     bMaterialIsStencil=false
     fontMaterial=null
     fontObject=&BurbankSmall
     letterSpacing=0
     monospacedWidth=1
     size=20
     skewAmount=0
     typefaceFontName=Bold
     outlineSettings {
      bApplyOutlineToDropShadows=false
      bMiteredCorners=false
      bSeparateFillAlpha=false
      outlineMaterial=null
      outlineSize=0
      outlineColor=<rgba 0 0 0 1>
     }
    }
    shadowOffset=<0 0>
   }
   slot {class=OverlaySlot, properties {horizontalAlignment=HAlign_Fill, verticalAlignment=VAlign_Fill}}
  }
 ]
}
