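# Translucent UI material that reveals text progressively over time, line by line across three rows, for the conversation UI.
# UV1 becomes a reading-order gradient: horizontal position is offset by which vertical third it occupies.
# A timestamp transition driven by RevealTimestamp and RevealSpeed sweeps a SmoothStep along that gradient, with EdgeSoftness widening the edge, to mask texture alpha.
# A debug readout of timestamp and time is mixed into color and alpha.
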Material /FortniteConversationUI/Conversation/Materials/M_UI_TextReveal
graph {
 properties {blendMode=BLEND_Translucent, materialDomain=MD_UI, numCustomizedUVs=2, pixelDepthOffsetMode=PDOM_Legacy, usedShadingModels=Default Lit}
 nodes [
  n1 = ScalarParameter() {parameterName=RevealTimestamp}
  n2 = Time()
  n3 = MaterialFunctionCall(R=@n1, G=@n2) {materialFunction=&AppendMany}
  n4 = MaterialFunctionCall!(Vector2=@n3.RG, MaximumNumberOfDigits, UVs, DebugTextLocation RG_UpperRight BA_LowerLeft, Component Spacing) {materialFunction=/Engine/Functions/Engine_MaterialFunctions02/Utility/DebugFloat2Values}
  n5 = TextureSampleParameter2D(UVs=TextureCoordinate()) {parameterName=Texture, texture=/Game/UI/GreyBox/white}
  n6 = Divide() {constB=3}
  n7 = TextureCoordinate() {coordinateIndex=1}
  n8 = Divide(A=ComponentMask(None=@n7) {r=true}) {constB=3}
  n9 = Add(A=@n8, B=@n6)
  n10 = Add(A=@n9, B=@n6)
  n11 = ComponentMask(None=@n7) {g=true}
  n12 = If(A=@n11, B=Multiply(A=@n6) {constB=2}, "A > B"=@n10, "A == B"=@n10, "A < B"=@n9)
  n13 = ScalarParameter() {defaultValue=0.2, parameterName=EdgeSoftness}
  n14 = MaterialFunctionCall(
   TimestampA=@n1
   TransitionLength_In=ScalarParameter() {defaultValue=3, parameterName=RevealSpeed}
   Transition In Only?=StaticBool() {value=true}
  ) {materialFunction=&MF_Timestamp_InOut_01}
  n15 = MaterialFunctionCall(
   Alpha=If(A=@n11, B=@n6, "A > B"=@n12, "A == B"=@n12, "A < B"=@n8)
   Min=@n14.SmoothStep
   Max=Add(A=@n14.SmoothStep, B=Multiply(A=@n14.SmoothStep, B=@n13))
  ) {materialFunction=&SmoothStep}
  Add(A=@n4.GreyScaleOutput, B=Multiply(A=@n5.A, B=OneMinus(None=@n15.Result)))
  n16 = VectorParameter() {parameterName=Offset, defaultValue=<rgba 0 -0.0948 0 1>}
  Add(A=TextureCoordinate(), B=AppendVector(A=@n16.R, B=@n16.R))
  Fmod(A=Multiply(A=@n2, B=ScalarParameter() {defaultValue=1, parameterName=speed}), B=Constant() {r=1})
  LinearInterpolate(A=@n5.RGB, B=Constant3Vector() {constant=<rgba 0.5927 0 1 0>}, Alpha=@n4.GreyScaleOutput)
  MaterialFunctionCall!(Vector3=@n3.RGB, MaximumNumberOfDigits, UVs, DebugTextLocation RG_UpperRight BA_LowerLeft, Component Spacing) {materialFunction=/Engine/Functions/Engine_MaterialFunctions02/Utility/DebugFloat3Values}
  ScalarParameter() {defaultValue=1, parameterName=Reveal}
 ]
}
