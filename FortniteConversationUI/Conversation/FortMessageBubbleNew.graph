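# Conversation message bubble widget that shows NPC dialogue and preview text on a skewed background with a pointer tail.
# Main messages get an uppercased, rich-text NPC name prefix, while previews show raw text; both play NewMessage at half speed.
# That animation widens the clipped text border and pops the background scale.
# The text sits in a TextReveal-material retainer with retained rendering off; RevealTimestamp and debug-print nodes are unconnected, and shown/hidden events are empty.
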WidgetBlueprint /FortniteConversationUI/Conversation/FortMessageBubbleNew
animations [NewMessage {
  binding_count=2
  length_seconds=0.3087
  object=&self:NewMessage
  sequence {bindings [
    Border_0 {
     id=#1
     possessed_class=Border
     tracks [{
       type=MovieSceneFloatVectorTrack
       display_name=Desired Size Scale
       sections [{
         type=MovieSceneFloatVectorSection
         channels [
          Vector.X {type=MovieSceneScriptingFloatChannel, keys [[0,0],[3.231,1]]}
          Vector.Y {type=MovieSceneScriptingFloatChannel}
         ]
        }]
      }]
    }
    Background {
     id=#2
     display_name=Overlay_BubbleContainer
     possessed_class=Image
     tracks [{
       type=MovieScene2DTransformTrack
       sections [{
         type=MovieScene2DTransformSection
         channels [
          Translation.X {type=MovieSceneScriptingFloatChannel}
          Translation.Y {type=MovieSceneScriptingFloatChannel}
          Angle {type=MovieSceneScriptingFloatChannel}
          Scale.X {type=MovieSceneScriptingFloatChannel, keys [[0,1],[1.048,1.05],[6.174,1]]}
          Scale.Y {type=MovieSceneScriptingFloatChannel, keys [[0,1],[1.048,1.1],[6.174,1]]}
          Shear.X {type=MovieSceneScriptingFloatChannel}
          Shear.Y {type=MovieSceneScriptingFloatChannel}
         ]
        }]
      }]
    }
   ]}
 }]
graphs [EventGraph {nodes [
   n1 = Get Message() -> (Message:Common Rich Text Block Object Reference)
   PrintString!(
    execute
    InString="To String (Text)"(InText=GetText!(self:Rich Text Block Object Reference=@n1) -> (ReturnValue:Text))
    bPrintToScreen
    bPrintToLog
    TextColor
    Duration
    Key
   ) -> (then)
   n2 = Get Message_ForSize() -> (Message_ForSize:Common Rich Text Block Object Reference)
   > PrintString(InString="To String (Text)"(InText=GetText(self=@n2)))
   n3 = Event OnMainMessageSet!() -> (OutputDelegate, then, NPCName:"Text (by ref)", MessageBody:"Text (by ref)") {is=title K2Node_Event}
   > n4 = SetText(
    self=Get Message()
    InText=Format Text!(
     Format="LOCTEXT(\"<text color=\\\"FFFFFFFF\\\" fontface=\\\"black\\\"> {NPCName}:</>  : {MessageBody}\")"
     NPCName:Text=TextToUpper!(InText:"Text (by ref)"=@n3.NPCName) -> (ReturnValue:Text)
     MessageBody:"Text (by ref)"=@n3.MessageBody
    ) -> (Result)
   )
   PlayAnimation(execute=[@n4.then, @n6.then], InAnimation=Get NewMessage() -> (NewMessage:Widget Animation Object Reference), PlaybackSpeed=0.5)
   n5 = Event OnPreviewMessageSet!() -> (OutputDelegate, then, PreviewText:"Text (by ref)") {is=title K2Node_Event}
   > n6 = SetText(self=Get Message(), InText=@n5.PreviewText)
   GetServerWorldTimeSeconds!(self:Game State Base Object Reference=GetGameStateZone!(self:Fort HUDContext Object Reference=Get FortHUDContext!() -> (ReturnValue:Fort HUDContext Object Reference) {is=title K2Node_GetBlueprintContextEx}) -> (ReturnValue:Fort Game State Zone Object Reference)) -> (ReturnValue:double)
   SetScalarParameterValue(
    self=GetEffectMaterial!(self:Retainer Box Object Reference=Get RetainerBox_TextReveal() -> (RetainerBox_TextReveal:Retainer Box Object Reference)) -> (ReturnValue:Material Instance Dynamic Object Reference)
    ParameterName=RevealTimestamp
    Value=GetWorldDeltaSeconds!() -> (ReturnValue:double)
   )
   GetGameTimeInSeconds()
   GetRealTimeSeconds()
   Event OnMessageShown() {is=title K2Node_Event}
   Event OnMessageHidden() {is=title K2Node_Event}
  ]}]
variables []
widgets {
 info {
  inheritedWidgetCount=1
  parentClass=/Script/FortniteConversationUI.FortConversationMessageWidget
  rootWidgetClass=&SizeBox
  widgetCount=13
 }
 widgets [
  SizeBox_0 {widgetClassPath=&SizeBox, widget=SizeBox_0, properties {widthOverride=650}}
  Overlay_DialogContainer {
   widgetClassPath=&Overlay
   parent=SizeBox_0
   widget=Overlay_DialogContainer
   properties {renderTransform {angle=0, scale=<1 1>, shear=<-8 0>, translation=<0 0>}}
   slot {class=SizeBoxSlot}
  }
  Overlay_BubbleContainer {
   widgetClassPath=&Overlay
   parent=Overlay_DialogContainer
   widget=Overlay_BubbleContainer
   slot {class=OverlaySlot, properties {horizontalAlignment=HAlign_Fill, verticalAlignment=VAlign_Fill}}
  }
  Background {
   widgetClassPath=&Image
   bIsVariable=true
   parent=Overlay_BubbleContainer
   widget=Background
   properties {
    brush {
     bIsDynamicallyLoaded=false
     drawAs=Box
     imageType=NoImage
     mirroring=NoMirror
     resourceName=None
     resourceObject=/Game/Athena/HUD/Materials/M_UI_MessageBG_Inst
     tiling=NoTile
     imageSize=<64 64>
     margin=<ltrb 0.4 0 0.4 0>
     outlineSettings {
      bUseBrushTransparency=false
      roundingType=HalfHeightRadius
      width=0
      color {colorUseRule=UseColor_Specified, specifiedColor=<rgba 0 0 0 0>}
      cornerRadii=<0 0 0 0>
     }
     tintColor {colorUseRule=UseColor_Specified, specifiedColor=<rgba 1 1 1 1>}
     uVRegion {bIsValid=false, max=<0 0>, min=<0 0>}
    }
    colorAndOpacity=<rgba 0.0052 0.0052 0.0056 0.9>
    renderTransformPivot=<0.5 0.9>
   }
   slot {class=OverlaySlot, properties {horizontalAlignment=HAlign_Fill, verticalAlignment=VAlign_Fill}}
  }
  Overlay_0 {
   widgetClassPath=&Overlay
   parent=Overlay_BubbleContainer
   widget=Overlay_0
   slot {class=OverlaySlot, properties {horizontalAlignment=HAlign_Center, verticalAlignment=VAlign_Center, padding=<ltrb 20 10 20 14>}}
  }
  Border_0 {
   widgetClassPath=&Border
   parent=Overlay_0
   widget=Border_0
   properties {
    clipping=ClipToBounds
    horizontalAlignment=HAlign_Left
    background {
     bIsDynamicallyLoaded=false
     drawAs=NoDrawType
     imageType=NoImage
     mirroring=NoMirror
     resourceName=None
     resourceObject=null
     tiling=NoTile
     imageSize=<32 32>
     margin=<ltrb 0 0 0 0>
     outlineSettings {
      bUseBrushTransparency=false
      roundingType=HalfHeightRadius
      width=0
      color {colorUseRule=UseColor_Specified, specifiedColor=<rgba 0 0 0 0>}
      cornerRadii=<0 0 0 0>
     }
     tintColor {colorUseRule=UseColor_Specified, specifiedColor=<rgba 1 1 1 1>}
     uVRegion {bIsValid=false, max=<0 0>, min=<0 0>}
    }
    padding=<ltrb 0 0 0 0>
   }
   slot {class=OverlaySlot}
  }
  Overlay_1 {
   widgetClassPath=&Overlay
   parent=Border_0
   widget=Overlay_1
   slot {class=BorderSlot, properties {horizontalAlignment=HAlign_Left, padding=<ltrb 0 0 0 0>}}
  }
  RetainerBox_TextReveal {
   widgetClassPath=/Script/UMG.RetainerBox
   bIsVariable=true
   parent=Overlay_1
   widget=RetainerBox_TextReveal
   properties {bRetainRender=false, effectMaterial=/FortniteConversationUI/Conversation/Materials/M_UI_TextReveal}
   slot {class=OverlaySlot, properties {padding=<ltrb 0 0 0 5>}}
  }
  Message {
   widgetClassPath=&CommonRichTextBlock
   bIsVariable=true
   parent=RetainerBox_TextReveal
   widget=Message
   properties {
    bOverrideDefaultStyle=true
    justification=Center
    lineHeightPercentage=0.9
    text="<name>Farmer Joe:</> This text is <b>emphasized</>. This is text about an <enemy>enemy</> and a <ally>friendly</>."
    textStyleSet=&/Conversation_RichTextStyleSet
    wrapTextAt=500
    defaultTextStyleOverride {
     overflowPolicy=Clip
     transformPolicy=None
     colorAndOpacity {colorUseRule=UseColor_Specified, specifiedColor=<rgba 1 1 1 1>}
     font {
      bForceMonospaced=false
      bMaterialIsStencil=false
      fontMaterial=null
      fontObject=&BurbankSmall
      letterSpacing=0
      monospacedWidth=1
      size=20
      skewAmount=0
      typefaceFontName=Bold
      outlineSettings {
       bApplyOutlineToDropShadows=false
       bMiteredCorners=false
       bSeparateFillAlpha=false
       outlineMaterial=null
       outlineSize=0
       outlineColor=<rgba 0 0 0 1>
      }
     }
     highlightColor {colorUseRule=UseColor_Specified, specifiedColor=<rgba 0 0 0 1>}
     highlightShape {
      bIsDynamicallyLoaded=false
      drawAs=Image
      imageType=NoImage
      mirroring=NoMirror
      resourceName=None
      resourceObject=null
      tiling=NoTile
      imageSize=<32 32>
      margin=<ltrb 0 0 0 0>
      outlineSettings {
       bUseBrushTransparency=false
       roundingType=HalfHeightRadius
       width=0
       color {colorUseRule=UseColor_Specified, specifiedColor=<rgba 0 0 0 0>}
       cornerRadii=<0 0 0 0>
      }
      tintColor {colorUseRule=UseColor_Specified, specifiedColor=<rgba 1 1 1 1>}
      uVRegion {bIsValid=false, max=<0 0>, min=<0 0>}
     }
     selectedBackgroundColor {colorUseRule=UseColor_Foreground, specifiedColor=<rgba 1 0 1 1>}
     shadowColorAndOpacity=<rgba 0 0 0 1>
     shadowOffset=<0 0>
     strikeBrush {
      bIsDynamicallyLoaded=false
      drawAs=Image
      imageType=NoImage
      mirroring=NoMirror
      resourceName=None
      resourceObject=null
      tiling=NoTile
      imageSize=<32 32>
      margin=<ltrb 0 0 0 0>
      outlineSettings {
       bUseBrushTransparency=false
       roundingType=HalfHeightRadius
       width=0
       color {colorUseRule=UseColor_Specified, specifiedColor=<rgba 0 0 0 0>}
       cornerRadii=<0 0 0 0>
      }
      tintColor {colorUseRule=UseColor_Specified, specifiedColor=<rgba 1 1 1 1>}
      uVRegion {bIsValid=false, max=<0 0>, min=<0 0>}
     }
     underlineBrush {
      bIsDynamicallyLoaded=false
      drawAs=Image
      imageType=NoImage
      mirroring=NoMirror
      resourceName=None
      resourceObject=null
      tiling=NoTile
      imageSize=<32 32>
      margin=<ltrb 0 0 0 0>
      outlineSettings {
       bUseBrushTransparency=false
       roundingType=HalfHeightRadius
       width=0
       color {colorUseRule=UseColor_Specified, specifiedColor=<rgba 0 0 0 0>}
       cornerRadii=<0 0 0 0>
      }
      tintColor {colorUseRule=UseColor_Specified, specifiedColor=<rgba 1 1 1 1>}
      uVRegion {bIsValid=false, max=<0 0>, min=<0 0>}
     }
    }
   }
   slot {class=PanelSlot}
  }
  HorizontalBox_Tail {
   widgetClassPath=&HorizontalBox
   parent=Overlay_DialogContainer
   widget=HorizontalBox_Tail
   slot {class=OverlaySlot, properties {horizontalAlignment=HAlign_Fill, verticalAlignment=VAlign_Bottom}}
  }
  Image_Tail {
   widgetClassPath=&Image
   bIsVariable=true
   parent=HorizontalBox_Tail
   widget=Image_Tail
   properties {
    brush {
     bIsDynamicallyLoaded=false
     drawAs=Image
     imageType=NoImage
     mirroring=NoMirror
     resourceName=None
     resourceObject=/FortniteCoreUI/Frontend/Art/M_UI_PanelPointer_MOTDCornerWedge
     tiling=NoTile
     imageSize=<32 40>
     margin=<ltrb 0 0 0 0>
     outlineSettings {
      bUseBrushTransparency=false
      roundingType=HalfHeightRadius
      width=0
      color {colorUseRule=UseColor_Specified, specifiedColor=<rgba 0 0 0 0>}
      cornerRadii=<0 0 0 0>
     }
     tintColor {colorUseRule=UseColor_Specified, specifiedColor=<rgba 1 1 1 1>}
     uVRegion {bIsValid=false, max=<0 0>, min=<0 0>}
    }
    colorAndOpacity=<rgba 0.0052 0.0052 0.0056 0.9>
    renderTransform {angle=0, scale=<-1 -1>, shear=<0 0>, translation=<0 0>}
    renderTransformPivot=<0.5 1>
   }
   slot {class=HorizontalBoxSlot, properties {horizontalAlignment=HAlign_Right, size {sizeRule=Fill, value=1}}}
  }
  Spacer_74 {
   widgetClassPath=&Spacer
   parent=HorizontalBox_Tail
   widget=Spacer_74
   slot {class=HorizontalBoxSlot, properties {size {sizeRule=Fill, value=0.25}}}
  }
  TutorialHighlight {widgetClassPath=&TutorialHighlight}
 ]
}
