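# Sound cue for conversation text in the Fortnite conversation UI, currently empty with no root or sound nodes, so it references no audio.
# It carries only playback settings: the non-critical in-game UI sound class, conversation attenuation and a low volume multiplier.
# A concurrency override allows one instance at a time, with a new play stopping the oldest.
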SoundCue /FortniteConversationUI/Sounds/Conversation_Text_Cue
root=null
nodes []
properties {
 attenuationSettings=&/Conversation_Att
 bOverrideConcurrency=true
 maxDistance=512
 soundClassObject=&Fort_UI_InGame_NonCritical
 volumeMultiplier=0.1
 concurrencyOverrides {maxCount=1, resolutionRule=StopOldest}
}
