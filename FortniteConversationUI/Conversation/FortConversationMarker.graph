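# Widget Blueprint for the marker over a conversation NPC, switching between a preview icon, a message bubble with a collapsed hold prompt, and a nameplate with keybind and arrow.
# Interaction and visibility-change events control the nameplate keybind, which HideInteractionKey forces collapsed.
# A custom event swaps the NPC icon texture.
# Sound logic is incomplete: the indicated-actor playback has no sound assigned, and the text cue playback is unconnected.
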WidgetBlueprint /FortniteConversationUI/Conversation/FortConversationMarker
graphs [EventGraph {nodes [
   n1 = "On Visibility Changed (Keybind_Nameplate)"!() -> (OutputDelegate, then, InVisibility:ESlateVisibility Enum=Visible) {is=title K2Node_ComponentBoundEvent}
   n2 = OR Boolean(
    A="Equal (Enum)"(A:ESlateVisibility Enum=@n1.InVisibility, B:ESlateVisibility Enum=Collapsed)
    B="Equal (Enum)"(A=@n1.InVisibility, B=Hidden)
   )
   > Branch(Condition=AND Boolean(A=NOT Boolean(A=@n2), B=Get HideInteractionKey() -> (HideInteractionKey:bool)))
   > SetVisibility(self=Get Keybind_Nameplate() -> (Keybind_Nameplate:Fort Keybind Widget Object Reference), InVisibility=Collapsed)
   n3 = SetConversationIcon!() -> (OutputDelegate, then, Texture:Texture 2D Object Reference) {is=title K2Node_CustomEvent}
   > SetBrushFromTexture!(
    execute
    self:Image Object Reference=Get NPCIcon() -> (NPCIcon:Image Object Reference)
    Texture:Texture 2D Object Reference=@n3.Texture
    bMatchSize:bool=false
   ) -> (then)
   n4 = Event OnSetIndicatedActor!() -> (OutputDelegate, then, NewIndicatorActor:Actor Object Reference) {is=title K2Node_Event}
   > Is Valid(InputObject=@n4.NewIndicatorActor)
   > PlaySoundAtLocation!(
    execute
    Sound:Sound Base Object Reference
    Location:Vector=Get Actor Location(self=@n4.NewIndicatorActor)
    Rotation:Rotator="0, 0, 0"
    VolumeMultiplier:float=1
    PitchMultiplier:float=1
    StartTime:float=0
    AttenuationSettings:Sound Attenuation Object Reference
    ConcurrencySettings:Sound Concurrency Object Reference
    OwningActor:Actor Object Reference
    InitialParams:Initial Active Sound Params Object Reference
   ) -> (then)
   n5 = Event OnCanInteract!() -> (OutputDelegate, then, bCanInteract:bool) {is=title K2Node_Event}
   > n6 = Branch(Condition=Get HideInteractionKey())
   > SetVisibility(self=Get Keybind_Nameplate(), InVisibility=Collapsed)
   SetVisibility(execute=@n6.else, self=Get Keybind_Nameplate(), InVisibility=Select(Option 0=Collapsed, Option 1=HitTestInvisible, Index=@n5.bCanInteract))
   PlaySoundAtLocation(
    Sound=/FortniteConversationUI/Sounds/Conversation_Text_Cue
    Location=vector + vector(A:Vector=Get Actor Location(), B:Vector="0, 0,75") -> (ReturnValue:Vector) {is=title K2Node_CommutativeAssociativeBinaryOperator}
    Rotation="0, 0, 0"
    VolumeMultiplier=1
    PitchMultiplier=1
    StartTime=0
   )
   Event OnMessageShownOverActor!() -> (OutputDelegate, then, OverActor:Actor Object Reference) {is=title K2Node_Event}
  ]}]
variables [HideInteractionKey {type=boolean}]
widgets {
 info {
  inheritedWidgetCount=7
  parentClass=/Script/FortniteConversationUI.FortConversationMarker
  rootWidgetClass=&CommonVisibilitySwitcher
  widgetCount=19
 }
 widgets [
  Switcher_States {widgetClassPath=&CommonVisibilitySwitcher, bInherited=true, widget=Switcher_States}
  Overlay_PreviewState {
   widgetClassPath=&Overlay
   bInherited=true
   parent=Switcher_States
   widget=Overlay_PreviewState
   slot {class=CommonVisibilitySwitcherSlot}
  }
  VerticalBox_2 {
   widgetClassPath=&VerticalBox
   parent=Overlay_PreviewState
   widget=VerticalBox_2
   slot {class=OverlaySlot, properties {horizontalAlignment=HAlign_Center}}
  }
  SizeBox_0 {
   widgetClassPath=&SizeBox
   parent=VerticalBox_2
   widget=SizeBox_0
   properties {bOverride_HeightOverride=true, bOverride_WidthOverride=true, heightOverride=40, widthOverride=40}
   slot {class=VerticalBoxSlot, properties {horizontalAlignment=HAlign_Center, verticalAlignment=VAlign_Center}}
  }
  NPCIcon {
   widgetClassPath=&Image
   bInherited=true
   bIsVariable=true
   parent=SizeBox_0
   widget=NPCIcon
   properties {brush {
     bIsDynamicallyLoaded=false
     drawAs=Image
     imageType=NoImage
     mirroring=NoMirror
     resourceName=None
     resourceObject=/FortniteConversationUI/Conversation/Materials/T_UI_NPC_SpeechIcon
     tiling=NoTile
     imageSize=<64 64>
     margin=<ltrb 0 0 0 0>
     outlineSettings {
      bUseBrushTransparency=false
      roundingType=HalfHeightRadius
      width=0
      color {colorUseRule=UseColor_Specified, specifiedColor=<rgba 0 0 0 0>}
      cornerRadii=<0 0 0 0>
     }
     tintColor {colorUseRule=UseColor_Specified, specifiedColor=<rgba 1 1 1 1>}
     uVRegion {bIsValid=false, max=<0 0>, min=<0 0>}
    }}
   slot {class=SizeBoxSlot}
  }
  VerticalBox_0 {
   widgetClassPath=&VerticalBox
   parent=Switcher_States
   widget=VerticalBox_0
   slot {class=CommonVisibilitySwitcherSlot, properties {horizontalAlignment=HAlign_Center, verticalAlignment=VAlign_Center}}
  }
  Callout {
   widgetClassPath=&Overlay
   bIsVariable=true
   parent=VerticalBox_0
   widget=Callout
   properties {visibility=Collapsed}
   slot {class=VerticalBoxSlot, properties {horizontalAlignment=HAlign_Center, padding=<ltrb 0 0 0 10>}}
  }
  HorizontalBox_38 {widgetClassPath=&HorizontalBox, parent=Callout, widget=HorizontalBox_38, slot {class=OverlaySlot}}
  CommonTextBlock_74 {
   widgetClassPath=&CommonTextBlock
   parent=HorizontalBox_38
   widget=CommonTextBlock_74
   properties {
    text=Hold
    font {
     bForceMonospaced=false
     bMaterialIsStencil=false
     fontMaterial=null
     fontObject=&BurbankSmall
     letterSpacing=0
     monospacedWidth=1
     size=16
     skewAmount=0
     typefaceFontName=Bold
     outlineSettings {
      bApplyOutlineToDropShadows=false
      bMiteredCorners=false
      bSeparateFillAlpha=false
      outlineMaterial=null
      outlineSize=0
      outlineColor=<rgba 0 0 0 1>
     }
    }
    shadowOffset=<0 0>
   }
   slot {class=HorizontalBoxSlot, properties {verticalAlignment=VAlign_Center, padding=<ltrb 0 0 10 0>}}
  }
  KeybindWidget {
   widgetClassPath=/FortUILibrary/Input/KeybindWidget
   bIsVariable=true
   parent=HorizontalBox_38
   widget=KeybindWidget
   properties {boundAction=Target}
   slot {class=HorizontalBoxSlot}
  }
  CommonTextBlock {
   widgetClassPath=&CommonTextBlock
   parent=HorizontalBox_38
   widget=CommonTextBlock
   properties {
    text=To Interact
    font {
     bForceMonospaced=false
     bMaterialIsStencil=false
     fontMaterial=null
     fontObject=&BurbankSmall
     letterSpacing=0
     monospacedWidth=1
     size=16
     skewAmount=0
     typefaceFontName=Bold
     outlineSettings {
      bApplyOutlineToDropShadows=false
      bMiteredCorners=false
      bSeparateFillAlpha=false
      outlineMaterial=null
      outlineSize=0
      outlineColor=<rgba 0 0 0 1>
     }
    }
    shadowOffset=<0 0>
   }
   slot {class=HorizontalBoxSlot, properties {verticalAlignment=VAlign_Center, padding=<ltrb 10 0 0 0>}}
  }
  ConversationMessage_Main {
   widgetClassPath=&/FortMessageBubbleNew
   bInherited=true
   bIsVariable=true
   parent=VerticalBox_0
   widget=ConversationMessage_Main
   properties {elementTag {tagName=Athena.HideForCapture}}
   slot {class=VerticalBoxSlot, properties {horizontalAlignment=HAlign_Center, padding=<ltrb 0 0 0 30>}}
  }
  VerticalBox_1 {widgetClassPath=&VerticalBox, parent=Switcher_States, widget=VerticalBox_1, slot {class=CommonVisibilitySwitcherSlot}}
  HorizontalBox {widgetClassPath=&HorizontalBox, parent=VerticalBox_1, widget=HorizontalBox, slot {class=VerticalBoxSlot}}
  Keybind_Nameplate {
   widgetClassPath=&KeybindWidget
   bInherited=true
   bIsVariable=true
   parent=HorizontalBox
   widget=Keybind_Nameplate
   properties {boundAction=Use}
   slot {class=HorizontalBoxSlot}
  }
  Text_ParticipantName {
   widgetClassPath=&CommonTextBlock
   bInherited=true
   parent=HorizontalBox
   widget=Text_ParticipantName
   properties {
    textTransformPolicy=ToUpper
    font {
     bForceMonospaced=false
     bMaterialIsStencil=false
     fontMaterial=null
     fontObject=&Burbank
     letterSpacing=0
     monospacedWidth=1
     size=18
     skewAmount=0
     typefaceFontName=Burbank-PlayerPower
     outlineSettings {
      bApplyOutlineToDropShadows=false
      bMiteredCorners=false
      bSeparateFillAlpha=false
      outlineMaterial=&M-FontShadow
      outlineSize=3
      outlineColor=<rgba 0 0 0 0.8>
     }
    }
    shadowOffset=<0 0>
   }
   slot {class=HorizontalBoxSlot, properties {verticalAlignment=VAlign_Center}}
  }
  CommonVisualAttachment_39 {
   widgetClassPath=&CommonVisualAttachment
   parent=VerticalBox_1
   widget=CommonVisualAttachment_39
   properties {contentAnchor=<0.5 0.7463>}
   slot {class=VerticalBoxSlot, properties {horizontalAlignment=HAlign_Center, verticalAlignment=VAlign_Center}}
  }
  Image_4 {
   widgetClassPath=&Image
   bIsVariable=true
   parent=CommonVisualAttachment_39
   widget=Image_4
   properties {brush {
     bIsDynamicallyLoaded=false
     drawAs=Image
     imageType=NoImage
     mirroring=NoMirror
     resourceName=None
     resourceObject=/Game/Athena/HUD/LTM/Mash/Materials/M_UI_InteractionArrow
     tiling=NoTile
     imageSize=<128 128>
     margin=<ltrb 0 0 0 0>
     outlineSettings {
      bUseBrushTransparency=false
      roundingType=HalfHeightRadius
      width=0
      color {colorUseRule=UseColor_Specified, specifiedColor=<rgba 0 0 0 0>}
      cornerRadii=<0 0 0 0>
     }
     tintColor {colorUseRule=UseColor_Specified, specifiedColor=<rgba 1 1 1 1>}
     uVRegion {bIsValid=false, max=<0 0>, min=<0 0>}
    }}
   slot {class=SizeBoxSlot}
  }
  TutorialHighlight {widgetClassPath=&TutorialHighlight}
 ]
}
