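# Widget Blueprint for a radial NPC conversation option, showing either a display-asset icon or an item icon with rarity-colored glow.
# Setup events choose the switcher page, with special icon source and scale for ammo.
# Selection plays sounds, with device overrides, and drives hover animations and availability/lock tinting, skipping visual hover feedback for spectators.
# Outside mobile, radial index offsets the scale pivot and lock badge.
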WidgetBlueprint /FortniteConversationUI/Conversation/FortConversationOption
animations [
 Hover {
  binding_count=1
  length_seconds=0.25
  object=&self:Hover
  sequence {bindings [Overlay_SizeBox {
     id=#1
     display_name=SizeBox_0
     possessed_class=CommonTextBlock
     tracks [{
       type=MovieScene2DTransformTrack
       sections [{
         type=MovieScene2DTransformSection
         channels [
          Scale.X {type=MovieSceneScriptingFloatChannel, keys [[0,1],[0.9813,1.1641],[5,1]]}
          Scale.Y {type=MovieSceneScriptingFloatChannel, keys [[0,1],[0.9813,1.1641],[5,1]]}
         ]
        }]
      }]
    }]}
 }
 ItemHover {
  binding_count=2
  length_seconds=0.1
  object=&self:ItemHover
  sequence {bindings [
    Image_RarityGlow {
     id=#2
     possessed_class=Image
     tracks [{
       type=MovieSceneFloatTrack
       display_name=Render Opacity
       sections [{type=MovieSceneFloatSection, channels [None {type=MovieSceneScriptingFloatChannel, keys [[0,0],[2,1]]}]}]
      }]
    }
    Image_ItemDefIcon {
     id=#3
     possessed_class=FortLazyImage
     tracks [{
       type=MovieSceneWidgetMaterialTrack
       sections [{type=MovieSceneParameterSection, channels [FillIcon {type=MovieSceneScriptingFloatChannel, keys [[0,1],[2,0]]}]}]
      }]
    }
   ]}
 }
 HoverScale {
  binding_count=1
  length_seconds=0.2332
  object=&self:HoverScale
  sequence {bindings [Overlay_Main {
     id=#4
     display_name=Overlay_Icon
     possessed_class=Overlay
     tracks [{
       type=MovieScene2DTransformTrack
       sections [{
         type=MovieScene2DTransformSection
         channels [
          Translation.X {type=MovieSceneScriptingFloatChannel}
          Translation.Y {type=MovieSceneScriptingFloatChannel}
          Angle {type=MovieSceneScriptingFloatChannel}
          Scale.X {type=MovieSceneScriptingFloatChannel, keys [[0,1],[4.6637,1.5]]}
          Scale.Y {type=MovieSceneScriptingFloatChannel, keys [[0,1],[4.6637,1.5]]}
          Shear.X {type=MovieSceneScriptingFloatChannel}
          Shear.Y {type=MovieSceneScriptingFloatChannel}
         ]
        }]
      }]
    }]}
 }
]
graphs [
 CalcVerticalOffsetForLockStuff {nodes [
   n1 = CalcVerticalOffsetForLockStuff!() -> (then, Index:int, Length:int) {is=title K2Node_FunctionEntry}
   n2 = int - int(A=@n1, B=int / int(A=@n1.Length, B=2))
   n3 = Select(Option 0:int=1, Option 1:int=0, Index:int="% (Integer)"(A=@n1, B=2) -> (ReturnValue=0)) -> (ReturnValue:int)
   > n4 = Set LocOffsetCoef(LocOffsetCoef:int=SelectInt(A=int + int(A=@n3, B=@n2), B="Absolute (Integer)"!(A:int=@n2) -> (ReturnValue:int=0), bPickA="integer >= integer"(A=@n2))) -> (Output_Get:int=0)
   n5 = int * int(A=@n4, B=8)
   > Return Node!(execute, Result:double=float - float(A=16, B="To Float (Integer)"(InInt=@n5)))
  ]}
 EventGraph {nodes [
   EventResetState() {is=title K2Node_CustomEvent}
   > StopAllAnimations(self:User Widget Object Reference)
   n1 = Get Hover() -> (Hover:Widget Animation Object Reference)
   > PlayAnimation(InAnimation=@n1, StartAtTime=GetEndTime(self=@n1), PlayMode=Reverse)
   n2 = Get ItemHover() -> (ItemHover:Widget Animation Object Reference)
   > PlayAnimation(InAnimation=@n2, StartAtTime=GetEndTime(self=@n2), PlayMode=Reverse)
   n3 = Get HoverScale() -> (HoverScale:Widget Animation Object Reference)
   > PlayAnimation(InAnimation=@n3, StartAtTime=GetEndTime(self=@n3), PlayMode=Reverse)
   n4 = OnHoveredStateChanged!() -> (OutputDelegate, then, bIsChoiceHovered:bool=~) {is=title K2Node_CustomEvent}
   > n5 = Branch(Condition=Get IsSpectating() -> (IsSpectating:bool))
   UpdateVisualState(execute=@n5.else)
   > n6 = Branch(Condition=@n4.bIsChoiceHovered)
   n7 = Get ItemHover()
   > PlayAnimationForward(InAnimation=@n7, PlaybackSpeed=2)
   n8 = Get HoverScale()
   > PlayAnimationForward(InAnimation=@n8)
   PlayAnimationReverse(execute=@n6.else, InAnimation=@n7)
   > PlayAnimationReverse(InAnimation=@n8)
   n9 = Event OnSetupPivotFromRadialInformation!() -> (OutputDelegate, then, NumElements:int=0, ItemIndex:int=0) {is=title K2Node_Event}
   > SetupPivotFromRadialInfo!(execute, self, Num Elements:int=@n9.NumElements, Item Index:int=@n9.ItemIndex) -> (then)
   Event OnOptionConfirmed() {is=title K2Node_Event}
   > n10 = Branch(Condition=Get ConversationDeviceOverride() -> (ConversationDeviceOverride:bool))
   > n11 = PlaySound2D(Sound=Get OptionConfirmedSoundOverride() -> (OptionConfirmedSoundOverride:Sound Base Object Reference))
   PushSoundMixModifier!(
    execute=[@n13.then, @n11.then]
    InSoundMixModifier:Sound Mix Object Reference=/Game/Athena/Sounds/NPCs/Conversation/Conversation_Button_Suppress_MixMod
   ) -> (then)
   n12 = Get OnOptionConfirmedSound() -> (OnOptionConfirmedSound:Sound Base Object Reference)
   n13 = PlaySound2D(
    execute=@n10.else
    Sound=Select(
     Option 0:Sound Base Object Reference=Get OnClickSound() -> (OnClickSound:Sound Base Object Reference)
     Option 1:Sound Base Object Reference=@n12
     Index:bool=IsValid(Object=@n12)
    ) -> (ReturnValue:Sound Base Object Reference)
   )
   Event On Selected()
   > Set IsSelected(IsSelected:bool=true) -> (Output_Get:bool)
   > OnHoveredStateChanged!(execute, self, bIsChoiceHovered:bool=true) -> (then) {is=title K2Node_CallFunction}
   > n14 = Branch(Condition=Get ConversationDeviceOverride())
   > PlaySound2D(Sound=Get OptionHoveredSoundOverride() -> (OptionHoveredSoundOverride:Sound Base Object Reference))
   n15 = Get OnOptionHoveredSound() -> (OnOptionHoveredSound:Sound Base Object Reference)
   PlaySound2D(
    execute=@n14.else
    Sound=Select(Option 0=Get OnHoverSound() -> (OnHoverSound:Sound Base Object Reference), Option 1=@n15, Index=IsValid(Object=@n15))
   )
   Event On Deselected()
   > n16 = Branch(Condition=Get IsSelected() -> (IsSelected:bool))
   > Set IsSelected(IsSelected=false)
   > OnHoveredStateChanged(bIsChoiceHovered=false)
   EventResetState(execute=@n16.else) {is=title K2Node_CallFunction}
   Event OnAvailabilityChanged!() -> (OutputDelegate, then, bIsAvailable:bool) {is=title K2Node_Event}
   > UpdateVisualState()
   Event Construct()
   > EventResetState()
   n17 = GetOwningPlayer()
   > n18 = Cast To FortPlayerControllerAthena(Object=@n17)
   > Set IsSpectating(IsSpectating:bool=GetIsSpectating!(self:Fort Player Controller Athena Object Reference=@n18) -> (ReturnValue:bool)) -> (Output_Get:bool)
   Set IsSpectating(
    execute=@n18.CastFailed
    IsSpectating=Cast To FortPlayerControllerSpectating!(Object=@n17) -> (AsFort Player Controller Spectating:Fort Player Controller Spectating Object Reference, bSuccess:bool=~) {is=title K2Node_DynamicCast}.bSuccess
   )
   n19 = Event OnSetupFromItemDef!() -> (OutputDelegate, then, Item:Fort Item Object Reference) {is=title K2Node_Event}
   > n20 = Set ItemToRepresent(ItemToRepresent:Fort Item Object Reference=@n19.Item) -> (Output_Get:Fort Item Object Reference)
   > SetItemToRepresent!(
    execute
    self:Fort Item Widget NUI Object Reference=Get ItemWidget() -> (ItemWidget:Fort Multi Size Item Card Object Reference)
    ItemToRepresent:Fort Item Object Reference=@n20
   ) -> (then)
   > SetActiveWidget(
    self=Get Switcher_DisplayAsset() -> (Switcher_DisplayAsset:Common Visibility Switcher Object Reference)
    Widget=Get Overlay_ItemDefIcon() -> (Overlay_ItemDefIcon:Overlay Object Reference)
   )
   n21 = Get ItemToRepresent() -> (ItemToRepresent:Fort Item Object Reference)
   > n22 = Branch(Condition="Equal (Enum)"(A:EFortItemType Enum=GetType!(self:Fort Item Object Reference=@n21) -> (ReturnValue:EFortItemType Enum=WorldItem), B:EFortItemType Enum=Ammo))
   n23 = GetItemDefinition!(self:Fort Item Object Reference=Get ItemToRepresent()) -> (ReturnValue:Fort Item Definition Object Reference)
   > n24 = GetItemData!(execute, self:Item Definition Base Object Reference=@n23) -> (Found:Exec, NotFound:Exec, ItemData:"Fortnite UI Large Icon Item Data Structure (by ref)")
   n25 = Get Image_ItemDefIcon() -> (Image_ItemDefIcon:Fort Lazy Image Object Reference)
   > n26 = SetBrushFromSoftTexture!(
    execute
    self:Image Object Reference=@n25
    SoftTexture:Texture 2D Soft Object Reference=Break Fortnite UI Large Icon Item Data!(FortItemComponentData_LargeIcon:"Fortnite UI Large Icon Item Data Structure (by ref)"=@n24) -> (LargeIcon:Texture 2D Soft Object Reference) {is=title K2Node_BreakStruct}
    bMatchSize:bool=false
   ) -> (then)
   n27 = GetRarity!(self:Fort Item Object Reference=Get ItemToRepresent()) -> (ReturnValue:EFortRarity Enum=Common)
   n28 = GetRarityData!(execute=[@n33.then, @n26.then], Rarity:EFortRarity Enum=@n27) -> (then, ReturnValue:Fort Rarity Item Data Structure)
   > n29 = GetDynamicMaterial(self=Get Image_RarityGlow() -> (Image_RarityGlow:Image Object Reference))
   n30 = Break Fort Rarity Item Data!(FortRarityItemData:"Fort Rarity Item Data Structure (by ref)"=@n28) -> (
    Name:Text
    Color1:Linear Color Structure
    Color2:Linear Color Structure
    Color3:Linear Color Structure
    Color4:Linear Color Structure
    Color5:Linear Color Structure
    Radius:float=~
    Falloff:float=~
    Brightness:float=~
    Roughness:float=~
    Glow:float=~
   ) {is=title K2Node_BreakStruct}
   > SetVectorParameterValue(self=@n29, ParameterName=Color1, Value=@n30.Color1)
   > SetVectorParameterValue(self=@n29, ParameterName=Color2, Value=@n30.Color3)
   > SetVisibility(
    self=Get Image_RarityGlow()
    InVisibility=Select(
     Option 0:ESlateVisibility Enum=Hidden
     Option 1:ESlateVisibility Enum=HitTestInvisible
     Index="Byte > Byte"!(A:Byte=@n27, B:Byte=0) -> (ReturnValue:bool)
    ) -> (ReturnValue:ESlateVisibility Enum)
   )
   n31 = GetType(self=Get ItemToRepresent()) -> (ReturnValue=WorldItem)
   > SetRenderScale!(
    execute
    self:Widget Object Reference=Get Overlay_ItemDefIcon()
    Scale:Vector 2D Structure=Select(Option 0:Vector 2D Structure=(X=1,Y=1), Option 1:Vector 2D Structure=(X=0.7,Y=0.7), Index="Equal (Enum)"(A=@n31, B=Ammo)) -> (ReturnValue:Vector 2D Structure)
   ) -> (then)
   n32 = GetItemData(execute=@n22.else, self=@n23) -> (ItemData:"UI Icon Item Data Structure (by ref)")
   > n33 = SetBrushFromSoftTexture(
    self=@n25
    SoftTexture=Break UI Icon Item Data!(ItemComponentData_Icon:"UI Icon Item Data Structure (by ref)"=@n32) -> (Icon:Texture 2D Soft Object Reference) {is=title K2Node_BreakStruct}
    bMatchSize=false
   )
   n34 = Event OnSetupFromDisplayAsset!() -> (OutputDelegate, then, DisplayAsset:Object Reference) {is=title K2Node_Event}
   > SetBrushFromLazyDisplayAsset!(
    execute
    self:Common Lazy Image Object Reference=Get Image_DisplayAssetIcon() -> (Image_DisplayAssetIcon:Fort Lazy Image Object Reference)
    LazyObject:"Soft Object Reference (by ref)"=Make Soft Reference!(Input:Object Reference=@n34.DisplayAsset) -> (Output:Soft Object Reference)
    bMatchTextureSize:bool=false
   ) -> (then)
   > SetActiveWidget(self=Get Switcher_DisplayAsset(), Widget=Get Overlay_DisplayAssetIcon() -> (Overlay_DisplayAssetIcon:Overlay Object Reference))
  ]}
 SetupPivotFromRadialInfo {nodes [
   SetupPivotFromRadialInfo!() -> (then, Num Elements:int, Item Index:int) {is=title K2Node_FunctionEntry}
   > Branch(Condition=NOT Boolean(A=IsMobileGame()))
   > n1 = Sequence()
   n2 = Get Item Index() -> (Item Index:int=0)
   n3 = Get Num Elements() -> (Num Elements:int=0)
   n4 = "To Float (Integer)"(InInt=@n2)
   n5 = float - float(A:int=@n3, B:int=1)
   SetRenderTranslation(
    execute=@n1.then_1
    self=Get Overlay_LockIcon() -> (Overlay_LockIcon:Overlay Object Reference)
    Translation=MakeVector2D(
     X=MapRangeClamped(Value=@n4, InRangeA=0, InRangeB=@n5, OutRangeA=20, OutRangeB=-20)
     Y=CalcVerticalOffsetForLockStuff!(Index:int=@n2, Length:int=@n3) -> (Result:double) {is=title K2Node_CallFunction}
    )
   )
   n6 = int + int(A=Get Item Index() -> (Item Index=0), B=1) -> (ReturnValue=0) {is=title K2Node_CommutativeAssociativeBinaryOperator}
   n7 = Get Num Elements() -> (Num Elements=0)
   n8 = float - float(B=float / float(A="To Float (Integer)"(InInt=@n6), B="To Float (Integer)"(InInt=@n7)) {is=title K2Node_CallFunction}) {is=title K2Node_CallFunction}
   SetRenderTransformPivot!(
    execute=@n1.then_0
    self:Widget Object Reference=Get Overlay_Main() -> (Overlay_Main:Overlay Object Reference)
    Pivot:Vector 2D Structure=MakeVector2D(X=Lerp(A=0.15, B=0.85, Alpha=@n8), Y=0.8)
   ) -> (then)
   IsMobilePlatform!() -> (ReturnValue:bool)
  ]}
 UpdateVisualState {nodes [
   UpdateVisualState() {is=title K2Node_FunctionEntry}
   > SetBrushTintColor!(
    execute
    self:Image Object Reference=Get Image_DisplayAssetIcon()
    TintColor:Slate Color Structure=Select!(
     Option 0_SpecifiedColor:Linear Color Structure=(R=0.4941,G=0.4941,B=0.4941,A=0.7)
     Option 0_ColorUseRule:ESlateColorStylingMode Enum
     Option 1_SpecifiedColor:Linear Color Structure=(R=1,G=1,B=1,A=1)
     Option 1_ColorUseRule:ESlateColorStylingMode Enum
     Index=IsAvailable!(self:Fort Conversation Option Object Reference) -> (ReturnValue:bool)
    ) -> (ReturnValue:Slate Color Structure)
   ) -> (then)
   n1 = IsLocked!(self:Fort Conversation Option Object Reference) -> (ReturnValue:bool)
   n2 = Select!(
    Option 0:Linear Color Structure=(R=1,G=0.0561,B=0.0648,A=1)
    Option 1:Linear Color Structure=(R=0.0103,G=0.0103,B=0.0103,A=1)
    Index=IsAvailable()
   ) -> (ReturnValue:Linear Color Structure)
   > SetColorAndOpacity(
    self:Image Object Reference=Get Image_DisplayAssetIcon()
    InColorAndOpacity:Linear Color Structure=Select(Option 0=(R=1,G=1,B=1,A=1), Option 1=Select(Option 0=@n2, Option 1=(R=0.4948,G=0.4948,B=0.4948,A=1), Index=@n1), Index=Get IsSelected())
   )
   n3 = IsAvailable()
   > SetContentColorAndOpacity!(
    execute
    self:Border Object Reference=Get CommonBorder_ItemAvailability() -> (CommonBorder_ItemAvailability:Common Border Object Reference)
    InContentColorAndOpacity:Linear Color Structure=Select(
     Option 0=Select(Option 0=(R=0.4969,G=0.4969,B=0.4969,A=0.702), Option 1=(R=1,G=1,B=1,A=1), Index=@n3)
     Option 1=Select(Option 0=(R=1,G=0.0561,B=0.0648,A=1), Option 1=(R=1,G=1,B=1,A=1), Index=@n3)
     Index=Get IsSelected()
    )
   ) -> (then)
   > SetVisibility(
    self=Get Overlay_LockIcon()
    InVisibility=Select(Option 0:ESlateVisibility Enum=Collapsed, Option 1:ESlateVisibility Enum=SelfHitTestInvisible, Index=IsLocked()) -> (ReturnValue:ESlateVisibility Enum)
   )
  ]}
]
variables [
 ConversationDeviceOverride {type=boolean}
 Is Choice Hovered {type=boolean}
 IsSelected {type=boolean}
 IsSpectating {type=boolean}
 ItemToRepresent {type=FortItem, default=null}
 OnClickSound {type=SoundBase, default=/Game/Athena/Sounds/NPCs/Conversation/Conversation_Click_Cue}
 OnHoverSound {type=SoundBase, default=/Game/Athena/Sounds/NPCs/Conversation/Conversation_Hover_Cue}
 OptionConfirmedSoundOverride {type=SoundBase, default=null}
 OptionHoveredSoundOverride {type=SoundBase, default=null}
]
widgets {
 info {
  inheritedWidgetCount=9
  parentClass=/Script/FortniteConversationUI.FortConversationOption
  rootWidgetClass=&SizeBox
  widgetCount=25
 }
 widgets [
  Overlay_SizeBox {
   widgetClassPath=&SizeBox
   bInherited=true
   widget=Overlay_SizeBox
   properties {bOverride_HeightOverride=true, bOverride_WidthOverride=true, heightOverride=70, maxDesiredWidth=500, widthOverride=70}
  }
  Overlay_EquippedItem {
   widgetClassPath=&Overlay
   bInherited=true
   bIsVariable=true
   parent=Overlay_SizeBox
   widget=Overlay_EquippedItem
   slot {class=SizeBoxSlot, properties {horizontalAlignment=HAlign_Center, verticalAlignment=VAlign_Center, padding=<ltrb 0 8 0 8>}}
  }
  Overlay_Main {
   widgetClassPath=&Overlay
   bIsVariable=true
   parent=Overlay_EquippedItem
   widget=Overlay_Main
   slot {class=OverlaySlot, properties {horizontalAlignment=HAlign_Center, verticalAlignment=VAlign_Center}}
  }
  Switcher_DisplayAsset {
   widgetClassPath=&CommonVisibilitySwitcher
   bInherited=true
   parent=Overlay_Main
   widget=Switcher_DisplayAsset
   slot {class=OverlaySlot, properties {horizontalAlignment=HAlign_Fill, verticalAlignment=VAlign_Fill}}
  }
  Overlay_DisplayAssetIcon {
   widgetClassPath=&Overlay
   bIsVariable=true
   parent=Switcher_DisplayAsset
   widget=Overlay_DisplayAssetIcon
   slot {class=CommonVisibilitySwitcherSlot}
  }
  ScaleBox_0 {widgetClassPath=&ScaleBox, parent=Overlay_DisplayAssetIcon, widget=ScaleBox_0, slot {class=OverlaySlot}}
  Image_DisplayAssetIcon {
   widgetClassPath=&FortLazyImage
   bIsVariable=true
   parent=ScaleBox_0
   widget=Image_DisplayAssetIcon
   properties {brush {
     bIsDynamicallyLoaded=false
     drawAs=Image
     imageType=NoImage
     mirroring=NoMirror
     resourceName=None
     resourceObject=/Game/Athena/HUD/Interaction/T_UI_NPC_statement
     tiling=NoTile
     imageSize=<64 64>
     margin=<ltrb 0 0 0 0>
     outlineSettings {
      bUseBrushTransparency=false
      roundingType=HalfHeightRadius
      width=0
      color {colorUseRule=UseColor_Specified, specifiedColor=<rgba 0 0 0 0>}
      cornerRadii=<0 0 0 0>
     }
     tintColor {colorUseRule=UseColor_Specified, specifiedColor=<rgba 1 1 1 1>}
     uVRegion {bIsValid=false, max=<0 0>, min=<0 0>}
    }}
   slot {class=ScaleBoxSlot}
  }
  Overlay_ItemDefIcon {
   widgetClassPath=&Overlay
   bIsVariable=true
   parent=Switcher_DisplayAsset
   widget=Overlay_ItemDefIcon
   slot {class=CommonVisibilitySwitcherSlot}
  }
  SizeBox_0 {
   widgetClassPath=&SizeBox
   parent=Overlay_ItemDefIcon
   widget=SizeBox_0
   properties {bOverride_HeightOverride=true, bOverride_WidthOverride=true, heightOverride=64, widthOverride=64}
   slot {class=OverlaySlot, properties {horizontalAlignment=HAlign_Center, verticalAlignment=VAlign_Center}}
  }
  ScaleBox_2 {
   widgetClassPath=&ScaleBox
   parent=SizeBox_0
   widget=ScaleBox_2
   slot {class=SizeBoxSlot, properties {horizontalAlignment=HAlign_Center, verticalAlignment=VAlign_Center, padding=<ltrb -16 -16 -16 -16>}}
  }
  Overlay_0 {widgetClassPath=&Overlay, parent=ScaleBox_2, widget=Overlay_0, slot {class=ScaleBoxSlot}}
  ScaleBox_1 {
   widgetClassPath=&ScaleBox
   parent=Overlay_0
   widget=ScaleBox_1
   properties {stretch=UserSpecified, userSpecifiedScale=0.4762}
   slot {class=OverlaySlot, properties {horizontalAlignment=HAlign_Fill, verticalAlignment=VAlign_Fill}}
  }
  ItemWidget {
   widgetClassPath=/Script/FortniteUI.FortMultiSizeItemCard
   bIsVariable=true
   parent=ScaleBox_1
   widget=ItemWidget
   properties {
    bPItemCardSize=L
    designTimePreviewItemType=/Game/Athena/Items/Weapons/WID_Assault_Suppressed_SpyTech_Athena_VR_Ore_T03
    shouldCollapseBorderPadding=true
    shouldSuppressItemTypeIcon=true
    shouldSuppressTierMeter=true
    visibility=Collapsed
   }
   slot {class=ScaleBoxSlot}
  }
  Image_RarityGlow {
   widgetClassPath=&Image
   bIsVariable=true
   parent=Overlay_0
   widget=Image_RarityGlow
   properties {
    visibility=Hidden
    brush {
     bIsDynamicallyLoaded=false
     drawAs=Image
     imageType=NoImage
     mirroring=NoMirror
     resourceName=None
     resourceObject=/Game/Athena/HUD/Interaction/M_UI_RotatingBurst
     tiling=NoTile
     imageSize=<180 180>
     margin=<ltrb 0 0 0 0>
     outlineSettings {
      bUseBrushTransparency=false
      roundingType=HalfHeightRadius
      width=0
      color {colorUseRule=UseColor_Specified, specifiedColor=<rgba 0 0 0 0>}
      cornerRadii=<0 0 0 0>
     }
     tintColor {colorUseRule=UseColor_Specified, specifiedColor=<rgba 1 1 1 1>}
     uVRegion {bIsValid=false, max=<0 0>, min=<0 0>}
    }
   }
   slot {class=OverlaySlot, properties {horizontalAlignment=HAlign_Center, verticalAlignment=VAlign_Center, padding=<ltrb -200 -200 -200 -200>}}
  }
  CommonBorder_ItemAvailability {
   widgetClassPath=&CommonBorder
   bIsVariable=true
   parent=Overlay_0
   widget=CommonBorder_ItemAvailability
   properties {padding=<ltrb 0 0 0 0>}
   slot {class=OverlaySlot, properties {horizontalAlignment=HAlign_Center, verticalAlignment=VAlign_Center}}
  }
  Image_ItemDefIcon {
   widgetClassPath=&FortLazyImage
   bIsVariable=true
   parent=CommonBorder_ItemAvailability
   widget=Image_ItemDefIcon
   properties {
    bApplyDropShadow=true
    materialTextureParamName=ItemIcon
    brush {
     bIsDynamicallyLoaded=false
     drawAs=Image
     imageType=NoImage
     mirroring=NoMirror
     resourceName=None
     resourceObject=/FortniteConversationUI/Conversation/Materials/M_UI_Conversation_ItemIcon_Inst
     tiling=NoTile
     imageSize=<90 90>
     margin=<ltrb 0 0 0 0>
     outlineSettings {
      bUseBrushTransparency=false
      roundingType=HalfHeightRadius
      width=0
      color {colorUseRule=UseColor_Specified, specifiedColor=<rgba 0 0 0 0>}
      cornerRadii=<0 0 0 0>
     }
     tintColor {colorUseRule=UseColor_Specified, specifiedColor=<rgba 1 1 1 1>}
     uVRegion {bIsValid=false, max=<0 0>, min=<0 0>}
    }
    renderTransform {angle=0, scale=<1 1>, shear=<0 0>, translation=<4 -2>}
    shadowColorAndOpacity=<rgba 0 0 0 0.4>
   }
   slot {class=BorderSlot, properties {padding=<ltrb 0 0 0 0>}}
  }
  Overlay_LockIcon {
   widgetClassPath=&Overlay
   bIsVariable=true
   parent=Overlay_Main
   widget=Overlay_LockIcon
   properties {renderTransform {angle=0, scale=<1 1>, shear=<0 0>, translation=<0 16>}, renderTransformPivot=<0 1>}
   slot {class=OverlaySlot, properties {horizontalAlignment=HAlign_Center, verticalAlignment=VAlign_Bottom}}
  }
  Image_305 {
   widgetClassPath=&Image
   parent=Overlay_LockIcon
   widget=Image_305
   properties {
    brush {
     bIsDynamicallyLoaded=false
     drawAs=Image
     imageType=NoImage
     mirroring=NoMirror
     resourceName=None
     resourceObject=null
     tiling=NoTile
     imageSize=<25 25>
     margin=<ltrb 0 0 0 0>
     outlineSettings {
      bUseBrushTransparency=false
      roundingType=HalfHeightRadius
      width=0
      color {colorUseRule=UseColor_Specified, specifiedColor=<rgba 0 0 0 0>}
      cornerRadii=<0 0 0 0>
     }
     tintColor {colorUseRule=UseColor_Specified, specifiedColor=<rgba 1 1 1 1>}
     uVRegion {bIsValid=false, max=<0 0>, min=<0 0>}
    }
    colorAndOpacity=<rgba 1 0 0.2988 1>
    renderTransform {angle=0, scale=<1 1>, shear=<-9 0>, translation=<0 0>}
   }
   slot {class=OverlaySlot, properties {horizontalAlignment=HAlign_Fill, verticalAlignment=VAlign_Fill}}
  }
  Image_LockIcon {
   widgetClassPath=&Image
   parent=Overlay_LockIcon
   widget=Image_LockIcon
   properties {brush {
     bIsDynamicallyLoaded=false
     drawAs=Image
     imageType=NoImage
     mirroring=NoMirror
     resourceName=None
     resourceObject=/Game/UI/Foundation/Textures/Icons/Locks/T-Icon-Lock-64
     tiling=NoTile
     imageSize=<20 20>
     margin=<ltrb 0 0 0 0>
     outlineSettings {
      bUseBrushTransparency=false
      roundingType=HalfHeightRadius
      width=0
      color {colorUseRule=UseColor_Specified, specifiedColor=<rgba 0 0 0 0>}
      cornerRadii=<0 0 0 0>
     }
     tintColor {colorUseRule=UseColor_Specified, specifiedColor=<rgba 1 1 1 1>}
     uVRegion {bIsValid=false, max=<0 0>, min=<0 0>}
    }}
   slot {class=OverlaySlot, properties {horizontalAlignment=HAlign_Center, verticalAlignment=VAlign_Center}}
  }
  HorizontalBox_EntryName {widgetClassPath=&HorizontalBox, bInherited=true, widget=null}
  Text_Center {widgetClassPath=&CommonTextBlock, bInherited=true, widget=null}
  HorizontalBox_CenterText {widgetClassPath=&HorizontalBox, bInherited=true, widget=null}
  Image_Icon {widgetClassPath=/Script/CommonUI.CommonLazyImage, bInherited=true, widget=null}
  Text_Main {widgetClassPath=&CommonTextBlock, bInherited=true, widget=null}
  InputActionWidget {widgetClassPath=&CommonActionWidget, bInherited=true, widget=null}
 ]
}
