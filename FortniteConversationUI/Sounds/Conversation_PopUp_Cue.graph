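# Sound cue for a conversation pop-up UI sound in the in-game conversation system.
# A single wave player feeds Conversation_PopUp_01 through a modulator with pitch pinned to a constant value.
# It plays quietly in the non-critical in-game UI sound class, and a concurrency override allows one instance at a time, blocking new plays while one is active.
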SoundCue /FortniteConversationUI/Sounds/Conversation_PopUp_Cue
root=Modulator_1
nodes [
 {id=Modulator_1, children [WavePlayer_0], properties {pitchMax=1, pitchMin=1}}
 {id=WavePlayer_0, properties {soundWaveAssetPtr=&/Conversation_PopUp_01}}
]
properties {
 bOverrideConcurrency=true
 maxDistance=2097152
 soundClassObject=&Fort_UI_InGame_NonCritical
 volumeMultiplier=0.15
 concurrencyOverrides {maxCount=1, resolutionRule=PreventNew}
}
