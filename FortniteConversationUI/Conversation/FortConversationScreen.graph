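# Widget Blueprint for the conversation screen, a five-slot half-wheel radial menu of dialog options with a details brief and confirm keybind.
# Native parent FortConversationScreen events drive intro, option and focus animations, availability tint and error sound.
# Construct plays pop-up audio, hides the screen for spectators and binds option updates; activation pushes a sound mix, popped on deactivation.
# Custom events cache options and later clear their conversation device override.
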WidgetBlueprint /FortniteConversationUI/Conversation/FortConversationScreen
animations [
 NewOptions {
  binding_count=2
  length_seconds=0.3808
  object=&self:NewOptions
  sequence {bindings [
    Overlay_Wheel {
     id=#1
     possessed_class=Overlay
     tracks [{
       type=MovieScene2DTransformTrack
       sections [{
         type=MovieScene2DTransformSection
         channels [
          Translation.X {type=MovieSceneScriptingFloatChannel}
          Translation.Y {type=MovieSceneScriptingFloatChannel}
          Angle {type=MovieSceneScriptingFloatChannel}
          Scale.X {type=MovieSceneScriptingFloatChannel, keys [[0,0.8711],[4.615,1]]}
          Scale.Y {type=MovieSceneScriptingFloatChannel, keys [[0,0.8711],[4.615,1]]}
          Shear.X {type=MovieSceneScriptingFloatChannel}
          Shear.Y {type=MovieSceneScriptingFloatChannel}
         ]
        }]
      }]
    }
    RadialMenu_DialogOptions {
     id=#2
     possessed_class=FortSlottedRadialMenu_C
     tracks [{
       type=MovieScene2DTransformTrack
       sections [{
         type=MovieScene2DTransformSection
         channels [
          Translation.X {type=MovieSceneScriptingFloatChannel}
          Translation.Y {type=MovieSceneScriptingFloatChannel}
          Angle {type=MovieSceneScriptingFloatChannel}
          Scale.X {type=MovieSceneScriptingFloatChannel, keys [[0,0.8211],[7.615,1]]}
          Scale.Y {type=MovieSceneScriptingFloatChannel, keys [[0,0.8211],[7.615,1]]}
          Shear.X {type=MovieSceneScriptingFloatChannel}
          Shear.Y {type=MovieSceneScriptingFloatChannel}
         ]
        }]
      }]
    }
   ]}
 }
 OptionFocused {
  binding_count=1
  length_seconds=0.3
  object=&self:OptionFocused
  sequence {bindings [Image_RadialHighlight {
     id=#3
     possessed_class=Image
     tracks [{
       type=MovieScene2DTransformTrack
       sections [{
         type=MovieScene2DTransformSection
         channels [
          Translation.X {type=MovieSceneScriptingFloatChannel}
          Translation.Y {type=MovieSceneScriptingFloatChannel}
          Angle {type=MovieSceneScriptingFloatChannel}
          Scale.X {type=MovieSceneScriptingFloatChannel, keys [[0,0.9463],[6,1.05]]}
          Scale.Y {type=MovieSceneScriptingFloatChannel, keys [[0,0.9463],[6,1.05]]}
          Shear.X {type=MovieSceneScriptingFloatChannel}
          Shear.Y {type=MovieSceneScriptingFloatChannel}
         ]
        }]
      }]
    }]}
 }
 OptionRight {
  binding_count=1
  length_seconds=0.35
  object=&self:OptionRight
  sequence {bindings [Overlay_Details {
     id=#4
     possessed_class=Overlay
     tracks [{
       type=MovieScene2DTransformTrack
       sections [{
         type=MovieScene2DTransformSection
         channels [
          Translation.X {type=MovieSceneScriptingFloatChannel, keys [[0,-40],[7,0]]}
          Translation.Y {type=MovieSceneScriptingFloatChannel}
          Angle {type=MovieSceneScriptingFloatChannel}
          Scale.X {type=MovieSceneScriptingFloatChannel}
          Scale.Y {type=MovieSceneScriptingFloatChannel}
          Shear.X {type=MovieSceneScriptingFloatChannel}
          Shear.Y {type=MovieSceneScriptingFloatChannel}
         ]
        }]
      }]
    }]}
 }
 OptionIntro {
  binding_count=1
  length_seconds=0.2
  object=&self:OptionIntro
  sequence {bindings [Overlay_Details {
     id=#5
     possessed_class=Overlay
     tracks [{
       type=MovieSceneFloatTrack
       display_name=Render Opacity
       sections [{type=MovieSceneFloatSection, channels [None {type=MovieSceneScriptingFloatChannel, keys [[0,0],[4,1]]}]}]
      }]
    }]}
 }
 OptionLeft {
  binding_count=1
  length_seconds=0.35
  object=&self:OptionLeft
  sequence {bindings [Overlay_Details {
     id=#4
     possessed_class=Overlay
     tracks [{
       type=MovieScene2DTransformTrack
       sections [{
         type=MovieScene2DTransformSection
         channels [
          Translation.X {type=MovieSceneScriptingFloatChannel, keys [[0,40],[7,0]]}
          Translation.Y {type=MovieSceneScriptingFloatChannel}
          Angle {type=MovieSceneScriptingFloatChannel}
          Scale.X {type=MovieSceneScriptingFloatChannel}
          Scale.Y {type=MovieSceneScriptingFloatChannel}
          Shear.X {type=MovieSceneScriptingFloatChannel}
          Shear.Y {type=MovieSceneScriptingFloatChannel}
         ]
        }]
      }]
    }]}
 }
 Intro {
  binding_count=1
  length_seconds=1
  object=&self:Intro
  sequence {bindings [Image_Background {
     id=#6
     possessed_class=Image
     tracks [{
       type=MovieSceneFloatTrack
       display_name=Render Opacity
       sections [{type=MovieSceneFloatSection, end_frame=20, start_frame=0, channels [None {type=MovieSceneScriptingFloatChannel, keys [[0,0],[20,1]]}]}]
      }]
    }]}
 }
]
graphs [
 EventGraph {nodes [
   Play Intro Audio() {is=title K2Node_CustomEvent}
   > PlaySound2D(Sound=/FortniteConversationUI/Sounds/Conversation_PopUp_Cue)
   n1 = CacheOption!() -> (OutputDelegate, then, FortConversationOption:Fort Conversation Option Object Reference) {is=title K2Node_CustomEvent}
   n2 = Get CachedFort Conversation Option() -> (CachedFort Conversation Option:Array of Fort Conversation Option Object References)
   > Add Unique(
    TargetArray:Array of Fort Conversation Option Object References=@n2
    NewItem:"Fort Conversation Option Object Reference (by ref)"=@n1.FortConversationOption
   )
   SetCreativeAudioOverride() {is=title K2Node_CustomEvent}
   > n3 = For Each Loop(Array:Array of Fort Conversation Option Object References=@n2) -> (Array Element:Fort Conversation Option Object Reference)
   Clear(execute=@n3.Completed, TargetArray:Array of Fort Conversation Option Object References=Get CachedFort Conversation Option())
   Is Valid(exec=@n3.LoopBody, InputObject=@n3)
   > n4 = Cast To FortConversationOption!(execute, Object=@n3) -> (then, CastFailed, AsFort Conversation Option:Fort Conversation Option Object Reference) {is=title K2Node_DynamicCast}
   > Set ConversationDeviceOverride!(execute, ConversationDeviceOverride:bool=false, self:Fort Conversation Option Object Reference=@n4) -> (then, Output_Get:bool)
   Event OnConversationStarted() {is=title K2Node_Event}
   > PlayAnimation(InAnimation=Get Intro())
   Event OnOptionsPopulated() {is=title K2Node_Event}
   > PlayAnimation(InAnimation=Get NewOptions() -> (NewOptions:Widget Animation Object Reference))
   n5 = Event OnSelectedItemAvailabilityChanged!() -> (OutputDelegate, then, bIsAvailable:bool) {is=title K2Node_Event}
   > SetContentColorAndOpacity!(
    execute
    self:Border Object Reference=Get KeybindBorder() -> (KeybindBorder:Border Object Reference)
    InContentColorAndOpacity:Linear Color Structure=Select(Option 0:Linear Color Structure=(R=0.0273,G=0.0273,B=0.0273,A=1), Option 1:Linear Color Structure=(R=1,G=1,B=1,A=1), Index=@n5.bIsAvailable) -> (ReturnValue:Linear Color Structure)
   ) -> (then)
   Event OnItemSelected!() -> (OutputDelegate, then, OriginalIndex:int=0) {is=title K2Node_Event}
   > n6 = Sequence()
   n7 = Get OptionFocused() -> (OptionFocused:Widget Animation Object Reference)
   > StopAnimation(InAnimation=@n7)
   > PlayAnimation(InAnimation=@n7)
   StopAnimation(execute=@n6.then_1, InAnimation=Get OptionRight() -> (OptionRight:Widget Animation Object Reference))
   > StopAnimation(InAnimation=Get OptionLeft() -> (OptionLeft:Widget Animation Object Reference))
   > n8 = Branch(Condition=ShouldBlockOptionIntroAnim!(self:Fort Conversation Screen Object Reference) -> (ReturnValue:bool))
   PlayAnimation(execute=@n8.else, InAnimation=Get OptionIntro() -> (OptionIntro:Widget Animation Object Reference))
   Event Construct()
   > Play Intro Audio() {is=title K2Node_CallFunction}
   n9 = GetOwningPlayer()
   > n10 = Cast To FortPlayerControllerAthena(Object=@n9)
   > n11 = Set IsSpectating(IsSpectating:bool=GetIsSpectating!(self:Fort Player Controller Athena Object Reference=@n10) -> (ReturnValue:bool)) -> (Output_Get:bool)
   SetSpectatorMode!(execute=[@n11.then, @n12.then], self, IsSpectating:bool=Get IsSpectating() -> (IsSpectating:bool)) -> (then)
   > Bind Event to On Conversation Option Updated(self:Fort Conversation Screen Object Reference, Delegate=Create Event()) {is=title K2Node_AddDelegate}
   n12 = Set IsSpectating(
    execute=@n10.CastFailed
    IsSpectating=Cast To FortPlayerControllerSpectating!(Object=@n9) -> (AsFort Player Controller Spectating:Fort Player Controller Spectating Object Reference, bSuccess:bool=~) {is=title K2Node_DynamicCast}.bSuccess
   )
   Event On Activated()
   > PushSoundMixModifier!(execute, InSoundMixModifier:Sound Mix Object Reference=/FortniteConversationUI/Sounds/Conversation_Suppress_MixMod) -> (then)
   Event On Deactivated()
   > PopSoundMixModifier!(execute, InSoundMixModifier:Sound Mix Object Reference=&Conversation_Suppress_MixMod) -> (then)
   Event OnConversationOptionUnavailable!() -> (OutputDelegate, then, SelectedEntry:Fort Radial Slot Object Reference) {is=title K2Node_Event}
   > PlaySound2D(Sound=/Game/Athena/Sounds/NPCs/Conversation/Conversation_Error_Cue)
  ]}
 SetSpectatorMode {nodes [
   n1 = SetSpectatorMode!() -> (then, IsSpectating:bool=~) {is=title K2Node_FunctionEntry}
   > SetVisibility(
    self=Get Overlay_0() -> (Overlay_0:Overlay Object Reference)
    InVisibility=Select(Option 0:ESlateVisibility Enum=Visible, Option 1:ESlateVisibility Enum=Hidden, Index=@n1) -> (ReturnValue:ESlateVisibility Enum)
   )
  ]}
]
variables [
 CachedFort Conversation Option {type="array<FortConversationOption>", default []}
 IsSpectating {type=boolean}
]
widgets {
 info {
  inheritedWidgetCount=10
  parentClass=/Script/FortniteConversationUI.FortConversationScreen
  rootWidgetClass=&Overlay
  widgetCount=23
 }
 widgets [
  Overlay_0 {widgetClassPath=&Overlay, bIsVariable=true, widget=Overlay_0, properties {visibility=Visible}}
  Image_Background {
   widgetClassPath=&Image
   bInherited=true
   parent=Overlay_0
   widget=Image_Background
   properties {brush {
     bIsDynamicallyLoaded=false
     drawAs=Image
     imageType=NoImage
     mirroring=NoMirror
     resourceName=None
     resourceObject=/FortniteConversationUI/Conversation/Materials/M_UI_ConversationBG_Inst
     tiling=NoTile
     imageSize=<32 32>
     margin=<ltrb 0 0 0 0>
     outlineSettings {
      bUseBrushTransparency=false
      roundingType=HalfHeightRadius
      width=0
      color {colorUseRule=UseColor_Specified, specifiedColor=<rgba 0 0 0 0>}
      cornerRadii=<0 0 0 0>
     }
     tintColor {colorUseRule=UseColor_Specified, specifiedColor=<rgba 1 1 1 1>}
     uVRegion {bIsValid=false, max=<0 0>, min=<0 0>}
    }}
   slot {class=OverlaySlot, properties {horizontalAlignment=HAlign_Fill, verticalAlignment=VAlign_Fill}}
  }
  VerticalBox_1 {
   widgetClassPath=&VerticalBox
   parent=Overlay_0
   widget=VerticalBox_1
   slot {class=OverlaySlot, properties {horizontalAlignment=HAlign_Fill, verticalAlignment=VAlign_Bottom, padding=<ltrb 0 0 0 240>}}
  }
  Spacer_LetoCounterOffset {
   widgetClassPath=&Spacer
   parent=VerticalBox_1
   widget=Spacer_LetoCounterOffset
   properties {size=<1 400>}
   slot {class=VerticalBoxSlot, properties {size {sizeRule=Fill, value=1}}}
  }
  Overlay_Details {
   widgetClassPath=&Overlay
   bInherited=true
   parent=VerticalBox_1
   widget=Overlay_Details
   slot {class=VerticalBoxSlot, properties {verticalAlignment=VAlign_Bottom, padding=<ltrb 0 0 0 10>, size {sizeRule=Fill, value=1}}}
  }
  VerticalBox_0 {
   widgetClassPath=&VerticalBox
   parent=Overlay_Details
   widget=VerticalBox_0
   slot {class=OverlaySlot, properties {horizontalAlignment=HAlign_Fill}}
  }
  DetailsBox {
   widgetClassPath=&SizeBox
   bIsVariable=true
   parent=VerticalBox_0
   widget=DetailsBox
   slot {class=VerticalBoxSlot, properties {size {sizeRule=Fill, value=1}}}
  }
  Switcher_Details {
   widgetClassPath=&CommonVisibilitySwitcher
   bInherited=true
   parent=DetailsBox
   widget=Switcher_Details
   properties {activeWidgetIndex=1}
   slot {class=SizeBoxSlot}
  }
  Spacer_94 {
   widgetClassPath=&Spacer
   parent=Switcher_Details
   widget=Spacer_94
   slot {class=CommonVisibilitySwitcherSlot, properties {horizontalAlignment=HAlign_Center}}
  }
  BasicBrief_Main {
   widgetClassPath=&/FortBasicBrief
   bInherited=true
   parent=Switcher_Details
   widget=BasicBrief_Main
   slot {class=CommonVisibilitySwitcherSlot, properties {horizontalAlignment=HAlign_Center}}
  }
  Overlay_ConversationContainer {
   widgetClassPath=&Overlay
   parent=Overlay_0
   widget=Overlay_ConversationContainer
   slot {class=OverlaySlot, properties {horizontalAlignment=HAlign_Fill, verticalAlignment=VAlign_Bottom, padding=<ltrb 0 0 0 -90>}}
  }
  Overlay_WheelContainer {
   widgetClassPath=&Overlay
   bInherited=true
   parent=Overlay_ConversationContainer
   widget=Overlay_WheelContainer
   properties {renderTransform {angle=0, scale=<1 1>, shear=<0 0>, translation=<0 99>}}
   slot {class=OverlaySlot, properties {horizontalAlignment=HAlign_Center}}
  }
  SizeBox_RadialMenu {
   widgetClassPath=&SizeBox
   bInherited=true
   parent=Overlay_WheelContainer
   widget=SizeBox_RadialMenu
   properties {bOverride_HeightOverride=true, bOverride_WidthOverride=true, heightOverride=400, widthOverride=400}
   slot {class=OverlaySlot, properties {horizontalAlignment=HAlign_Center, verticalAlignment=VAlign_Bottom}}
  }
  Overlay_Wheel {widgetClassPath=&Overlay, parent=SizeBox_RadialMenu, widget=Overlay_Wheel, slot {class=SizeBoxSlot}}
  Image_LoadingSpinner {
   widgetClassPath=&Image
   bInherited=true
   parent=Overlay_Wheel
   widget=Image_LoadingSpinner
   properties {
    visibility=Hidden
    brush {
     bIsDynamicallyLoaded=false
     drawAs=Image
     imageType=NoImage
     mirroring=NoMirror
     resourceName=None
     resourceObject=/FortUILibrary/UIKit/Throbber/MI_UIKit_Monochrome_Throbber
     tiling=NoTile
     imageSize=<128 128>
     margin=<ltrb 0 0 0 0>
     outlineSettings {
      bUseBrushTransparency=false
      roundingType=HalfHeightRadius
      width=0
      color {colorUseRule=UseColor_Specified, specifiedColor=<rgba 0 0 0 0>}
      cornerRadii=<0 0 0 0>
     }
     tintColor {colorUseRule=UseColor_Specified, specifiedColor=<rgba 1 1 1 1>}
     uVRegion {bIsValid=false, max=<0 0>, min=<0 0>}
    }
    colorAndOpacity=<rgba 1 1 1 0.4>
   }
   slot {class=OverlaySlot, properties {horizontalAlignment=HAlign_Center, verticalAlignment=VAlign_Center, padding=<ltrb 0 0 0 30>}}
  }
  Image_Radial_BG {
   widgetClassPath=&Image
   parent=Overlay_Wheel
   widget=Image_Radial_BG
   properties {
    visibility=Collapsed
    brush {
     bIsDynamicallyLoaded=false
     drawAs=Image
     imageType=NoImage
     mirroring=NoMirror
     resourceName=None
     resourceObject=/Game/UI/Items/Materials/MI-AthenaPieShadow_Half
     tiling=NoTile
     imageSize=<128 128>
     margin=<ltrb 0 0 0 0>
     outlineSettings {
      bUseBrushTransparency=false
      roundingType=HalfHeightRadius
      width=0
      color {colorUseRule=UseColor_Specified, specifiedColor=<rgba 0 0 0 0>}
      cornerRadii=<0 0 0 0>
     }
     tintColor {colorUseRule=UseColor_Specified, specifiedColor=<rgba 1 1 1 1>}
     uVRegion {bIsValid=false, max=<0 0>, min=<0 0>}
    }
    colorAndOpacity=<rgba 0.0052 0.0052 0.0056 0.4>
    renderTransform {angle=0, scale=<1 1>, shear=<0 0>, translation=<10 10>}
   }
   slot {class=OverlaySlot, properties {horizontalAlignment=HAlign_Fill, verticalAlignment=VAlign_Fill}}
  }
  Image_Radial {
   widgetClassPath=&Image
   parent=Overlay_Wheel
   widget=Image_Radial
   properties {
    visibility=HitTestInvisible
    brush {
     bIsDynamicallyLoaded=false
     drawAs=Image
     imageType=NoImage
     mirroring=NoMirror
     resourceName=None
     resourceObject=/Game/UI/Items/Materials/MI-AthenaHalfPieMenu_Dark
     tiling=NoTile
     imageSize=<128 128>
     margin=<ltrb 0 0 0 0>
     outlineSettings {
      bUseBrushTransparency=false
      roundingType=HalfHeightRadius
      width=0
      color {colorUseRule=UseColor_Specified, specifiedColor=<rgba 0 0 0 0>}
      cornerRadii=<0 0 0 0>
     }
     tintColor {colorUseRule=UseColor_Specified, specifiedColor=<rgba 1 1 1 1>}
     uVRegion {bIsValid=false, max=<0 0>, min=<0 0>}
    }
    colorAndOpacity=<rgba 1 1 1 0.75>
   }
   slot {class=OverlaySlot, properties {horizontalAlignment=HAlign_Fill, verticalAlignment=VAlign_Fill}}
  }
  HighlightBox {
   widgetClassPath=&SizeBox
   bIsVariable=true
   parent=Overlay_Wheel
   widget=HighlightBox
   slot {class=OverlaySlot, properties {horizontalAlignment=HAlign_Fill, verticalAlignment=VAlign_Fill}}
  }
  Image_RadialHighlight {
   widgetClassPath=&Image
   bInherited=true
   parent=HighlightBox
   widget=Image_RadialHighlight
   properties {
    visibility=Hidden
    brush {
     bIsDynamicallyLoaded=false
     drawAs=Image
     imageType=NoImage
     mirroring=NoMirror
     resourceName=None
     resourceObject=/Game/UI/Items/Materials/MI-AthenaPieMenu_Selection_FiveSlots
     tiling=NoTile
     imageSize=<32 32>
     margin=<ltrb 0 0 0 0>
     outlineSettings {
      bUseBrushTransparency=false
      roundingType=HalfHeightRadius
      width=0
      color {colorUseRule=UseColor_Specified, specifiedColor=<rgba 0 0 0 0>}
      cornerRadii=<0 0 0 0>
     }
     tintColor {colorUseRule=UseColor_Specified, specifiedColor=<rgba 1 1 1 1>}
     uVRegion {bIsValid=false, max=<0 0>, min=<0 0>}
    }
   }
   slot {class=SizeBoxSlot}
  }
  RadialMenu_DialogOptions {
   widgetClassPath=/Game/Athena/HUD/ContextMenus/FortSlottedRadialMenu
   bInherited=true
   bIsVariable=true
   parent=Overlay_Wheel
   widget=RadialMenu_DialogOptions
   properties {
    bUseAsManySlotsAsItems=true
    entryWidgetClass=&/FortConversationOption.FortConversationOption_C
    joystickDeadZone=0.4
    mouseDeadZoneRadiusPercentage=0.3
    navigation=&self:WidgetTree.RadialMenu_DialogOptions.WidgetNavigation_0
    numSlotsSupported=5
    orderingMode=Custom
    visibility=Visible
    customOrdering [N,NNE,NE,ENE,E]
    radialBoxSettings {angleBetweenItems=0, bClockwiseOrder=false, bDistributeItemsEvenly=true, marginSize=0, sectorCentralAngle=180, startingAngle=0}
   }
   slot {class=OverlaySlot, properties {horizontalAlignment=HAlign_Fill, verticalAlignment=VAlign_Fill, padding=<ltrb 0 2 2 2>}}
  }
  KeybindBox {
   widgetClassPath=&SizeBox
   bIsVariable=true
   parent=Overlay_WheelContainer
   widget=KeybindBox
   slot {class=OverlaySlot, properties {horizontalAlignment=HAlign_Center, verticalAlignment=VAlign_Center, padding=<ltrb 0 30 0 0>}}
  }
  KeybindBorder {
   widgetClassPath=&Border
   bIsVariable=true
   parent=KeybindBox
   widget=KeybindBorder
   properties {
    background {
     bIsDynamicallyLoaded=false
     drawAs=Image
     imageType=NoImage
     mirroring=NoMirror
     resourceName=None
     resourceObject=null
     tiling=NoTile
     imageSize=<24 24>
     margin=<ltrb 0 0 0 0>
     outlineSettings {
      bUseBrushTransparency=false
      roundingType=HalfHeightRadius
      width=0
      color {colorUseRule=UseColor_Specified, specifiedColor=<rgba 0 0 0 0>}
      cornerRadii=<0 0 0 0>
     }
     tintColor {colorUseRule=UseColor_Specified, specifiedColor=<rgba 1 1 1 1>}
     uVRegion {bIsValid=false, max=<0 0>, min=<0 0>}
    }
    brushColor=<rgba 1 1 1 0>
    padding=<ltrb 0 0 0 0>
    renderTransform {angle=0, scale=<1 1>, shear=<0 0>, translation=<0 -59>}
   }
   slot {class=SizeBoxSlot}
  }
  Keybind_Confirm {
   widgetClassPath=/FortUILibrary/Input/KeybindWidgetDark
   bInherited=true
   parent=KeybindBorder
   widget=Keybind_Confirm
   properties {boundAction=Use}
   slot {class=BorderSlot, properties {padding=<ltrb 0 0 0 0>}}
  }
 ]
}
