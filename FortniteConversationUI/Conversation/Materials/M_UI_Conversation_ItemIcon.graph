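# M_UI_Conversation_ItemIcon is a translucent UI material for item icons in the Fortnite conversation interface.
# It samples the ItemIcon texture parameter, which defaults to a suppressed auto weapon icon.
# A linear interpolation blends that texture's RGB toward solid white, with the FillIcon scalar parameter as alpha, so the icon color can be whited out.
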Material /FortniteConversationUI/Conversation/Materials/M_UI_Conversation_ItemIcon
graph {
 properties {blendMode=BLEND_Translucent, materialDomain=MD_UI, pixelDepthOffsetMode=PDOM_Legacy, usedShadingModels=Default Lit}
 nodes [LinearInterpolate(
   A=TextureSampleParameter2D() {parameterName=ItemIcon, texture=/Game/UI/Foundation/Textures/Icons/Weapons/Items/T-Icon-Weapons-SK-AutoHigh-Suppressed}.RGB
   B=Constant() {r=1}
   Alpha=ScalarParameter() {parameterName=FillIcon}
  )]
}
