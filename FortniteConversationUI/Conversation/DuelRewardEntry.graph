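# DuelRewardEntry is a rotating-item tile button widget that displays a duel reward item in the conversation UI.
# On Construct it passes its Item to SetupFromItemDef, which hands it to the inherited multi-size item card, scaled down inside a small tile.
# A chain for lazy icon, rarity-colored burst glow and smaller ammo scale remains in the graph but is unconnected to execution; the icon border and glow default to collapsed.
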WidgetBlueprint /FortniteConversationUI/Conversation/DuelRewardEntry
graphs [EventGraph {nodes [
   K2Node_Event_3 = SetupFromItemDef!() -> (OutputDelegate, then, Item:Fort Item Object Reference) {is=title K2Node_CustomEvent}
   n1 = GetType!(self:Fort Item Object Reference=@K2Node_Event_3.Item) -> (ReturnValue:EFortItemType Enum=WorldItem)
   SetBrushFromSoftTexture!(
    execute
    self:Image Object Reference=Get LazyImage_ItemIcon() -> (LazyImage_ItemIcon:Fort Lazy Image Object Reference)
    SoftTexture:Texture 2D Soft Object Reference=Select(
     Option 0:Texture 2D Soft Object Reference=GetSmallPreviewImage!(self:Fort Item Object Reference=@K2Node_Event_3.Item) -> (ReturnValue:Texture 2D Soft Object Reference)
     Option 1:Texture 2D Soft Object Reference=GetLargePreviewImage!(self:Fort Item Object Reference=@K2Node_Event_3.Item) -> (ReturnValue:Texture 2D Soft Object Reference)
     Index="Equal (Enum)"(A:EFortItemType Enum=@n1, B:EFortItemType Enum=Ammo)
    ) -> (ReturnValue:Texture 2D Soft Object Reference)
    bMatchSize:bool=false
   ) -> (then)
   n2 = GetRarity!(self:Fort Item Object Reference=@K2Node_Event_3.Item) -> (ReturnValue:EFortRarity Enum=Common)
   > n3 = GetRarityData!(execute, Rarity:EFortRarity Enum=@n2) -> (then, ReturnValue:Fort Rarity Item Data Structure)
   > n4 = GetDynamicMaterial(self=Get Image_RarityGlow() -> (Image_RarityGlow:Image Object Reference))
   n5 = Break Fort Rarity Item Data!(FortRarityItemData:"Fort Rarity Item Data Structure (by ref)"=@n3) -> (
    Name:Text
    Color1:Linear Color Structure
    Color2:Linear Color Structure
    Color3:Linear Color Structure
    Color4:Linear Color Structure
    Color5:Linear Color Structure
    Radius:float=~
    Falloff:float=~
    Brightness:float=~
    Roughness:float=~
    Glow:float=~
   ) {is=title K2Node_BreakStruct}
   > SetVectorParameterValue(self=@n4, ParameterName=Color1, Value=@n5.Color1)
   > SetVectorParameterValue(self=@n4, ParameterName=Color2, Value=@n5.Color3)
   > SetVisibility(
    self=Get Image_RarityGlow()
    InVisibility=Select(
     Option 0:ESlateVisibility Enum=Hidden
     Option 1:ESlateVisibility Enum=HitTestInvisible
     Index="Byte > Byte"!(A:Byte=@n2, B:Byte=0) -> (ReturnValue:bool)
    ) -> (ReturnValue:ESlateVisibility Enum)
   )
   > SetRenderScale!(
    execute
    self:Widget Object Reference=Get ItemDef() -> (ItemDef:Overlay Object Reference)
    Scale:Vector 2D Structure=Select(Option 0:Vector 2D Structure=(X=1,Y=1), Option 1:Vector 2D Structure=(X=0.7,Y=0.7), Index="Equal (Enum)"(A=@n1, B=Ammo)) -> (ReturnValue:Vector 2D Structure)
   ) -> (then)
   Event Construct()
   > SetupFromItemDef!(execute, self, Item:Fort Item Object Reference=Get Item() -> (Item:Fort Item Weak object)) -> (then) {is=title K2Node_CallFunction}
   SetItemToRepresent!(
    execute=@K2Node_Event_3.then
    self:Fort Item Widget NUI Object Reference=Get ItemWidget() -> (ItemWidget:Fort Multi Size Item Card Object Reference)
    ItemToRepresent:Fort Item Object Reference=@K2Node_Event_3.Item
   ) -> (then)
  ]}]
variables []
widgets {
 info {inheritedWidgetCount=3, parentClass=/Script/FortniteUI.FortRotatingItemTileButton, rootWidgetClass=&Overlay, widgetCount=11}
 widgets [
  ItemDef {widgetClassPath=&Overlay, bIsVariable=true, widget=ItemDef}
  SizeBox_0 {
   widgetClassPath=&SizeBox
   parent=ItemDef
   widget=SizeBox_0
   properties {bOverride_HeightOverride=true, bOverride_WidthOverride=true, heightOverride=64, widthOverride=64}
   slot {class=OverlaySlot, properties {horizontalAlignment=HAlign_Center, verticalAlignment=VAlign_Center}}
  }
  ScaleBox_2 {
   widgetClassPath=&ScaleBox
   parent=SizeBox_0
   widget=ScaleBox_2
   slot {class=SizeBoxSlot, properties {horizontalAlignment=HAlign_Center, verticalAlignment=VAlign_Center, padding=<ltrb -16 -16 -16 -16>}}
  }
  Overlay_0 {widgetClassPath=&Overlay, parent=ScaleBox_2, widget=Overlay_0, slot {class=ScaleBoxSlot}}
  ScaleBox_1 {
   widgetClassPath=&ScaleBox
   parent=Overlay_0
   widget=ScaleBox_1
   properties {stretch=UserSpecified, userSpecifiedScale=0.4762}
   slot {class=OverlaySlot, properties {horizontalAlignment=HAlign_Fill, verticalAlignment=VAlign_Fill}}
  }
  ItemWidget {
   widgetClassPath=/Script/FortniteUI.FortMultiSizeItemCard
   bInherited=true
   bIsVariable=true
   parent=ScaleBox_1
   widget=ItemWidget
   properties {
    bPItemCardSize=M
    designTimePreviewItemType=/Game/Athena/Items/Weapons/WID_Assault_Suppressed_SpyTech_Athena_VR_Ore_T03
    shouldCollapseBorderPadding=true
    shouldSuppressItemTypeIcon=true
    shouldSuppressTierMeter=true
   }
   slot {class=ScaleBoxSlot}
  }
  Image_RarityGlow {
   widgetClassPath=&Image
   bIsVariable=true
   parent=Overlay_0
   widget=Image_RarityGlow
   properties {
    visibility=Collapsed
    brush {
     bIsDynamicallyLoaded=false
     drawAs=Image
     imageType=NoImage
     mirroring=NoMirror
     resourceName=None
     resourceObject=/Game/Athena/HUD/Interaction/M_UI_RotatingBurst
     tiling=NoTile
     imageSize=<100 100>
     margin=<ltrb 0 0 0 0>
     outlineSettings {
      bUseBrushTransparency=false
      roundingType=HalfHeightRadius
      width=0
      color {colorUseRule=UseColor_Specified, specifiedColor=<rgba 0 0 0 0>}
      cornerRadii=<0 0 0 0>
     }
     tintColor {colorUseRule=UseColor_Specified, specifiedColor=<rgba 1 1 1 1>}
     uVRegion {bIsValid=false, max=<0 0>, min=<0 0>}
    }
   }
   slot {class=OverlaySlot, properties {horizontalAlignment=HAlign_Center, verticalAlignment=VAlign_Center, padding=<ltrb -100 -100 -100 -100>}}
  }
  CommonBorder_ItemAvailability {
   widgetClassPath=&CommonBorder
   bIsVariable=true
   parent=Overlay_0
   widget=CommonBorder_ItemAvailability
   properties {visibility=Collapsed, padding=<ltrb 0 0 0 0>}
   slot {class=OverlaySlot, properties {horizontalAlignment=HAlign_Center, verticalAlignment=VAlign_Center}}
  }
  LazyImage_ItemIcon {
   widgetClassPath=&FortLazyImage
   bIsVariable=true
   parent=CommonBorder_ItemAvailability
   widget=LazyImage_ItemIcon
   properties {
    bApplyDropShadow=true
    materialTextureParamName=ItemIcon
    brush {
     bIsDynamicallyLoaded=false
     drawAs=Image
     imageType=NoImage
     mirroring=NoMirror
     resourceName=None
     resourceObject=/FortniteConversationUI/Conversation/Materials/M_UI_Conversation_ItemIcon
     tiling=NoTile
     imageSize=<90 90>
     margin=<ltrb 0 0 0 0>
     outlineSettings {
      bUseBrushTransparency=false
      roundingType=HalfHeightRadius
      width=0
      color {colorUseRule=UseColor_Specified, specifiedColor=<rgba 0 0 0 0>}
      cornerRadii=<0 0 0 0>
     }
     tintColor {colorUseRule=UseColor_Specified, specifiedColor=<rgba 1 1 1 1>}
     uVRegion {bIsValid=false, max=<0 0>, min=<0 0>}
    }
    renderTransform {angle=0, scale=<1 1>, shear=<0 0>, translation=<4 -2>}
    shadowColorAndOpacity=<rgba 0 0 0 0.4>
   }
   slot {class=BorderSlot, properties {padding=<ltrb 0 0 0 0>}}
  }
  ItemWidgetSTW {widgetClassPath=/Script/UMG.Widget, bInherited=true, widget=null}
  InputActionWidget {widgetClassPath=&CommonActionWidget, bInherited=true, widget=null}
 ]
}
