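# Duel briefing panel for the conversation UI, built on native FortDuelBrief, showing a "Duel" title, a rich-text challenge objective and the rewards.
# The DisplayLootItems event feeds its items into a horizontal, non-interactive tile view of DuelRewardEntry widgets.
# The same event collapses the rewards section when the list is empty and switches its heading between "Reward" and "Rewards" by item count.
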WidgetBlueprint /FortniteConversationUI/Conversation/FortDuelBrief
graphs [EventGraph {nodes [
   n1 = Event DisplayLootItems!() -> (OutputDelegate, then, Items:Array of Fort Item Object References) {is=title K2Node_Event}
   > Set List Items!(
    execute
    self:List View Object Reference=Get TileView() -> (TileView:Common Tile View Object Reference)
    InListItems:Array of Object References=@n1.Items
   ) -> (then) {is=title K2Node_CallFunction}
   n2 = Length(TargetArray:Array of Fort Item Object References=@n1.Items)
   > SetVisibility(
    self=Get VerticalBox_Rewards() -> (VerticalBox_Rewards:Vertical Box Object Reference)
    InVisibility=Select(
     Option 0=HitTestInvisible
     Option 1=Collapsed
     Index="integer < integer"!(A:int=@n2, B:int=1) -> (ReturnValue:bool) {is=title K2Node_CallFunction}
    )
   )
   > "SetText (Text)"(
    self=Get Text_RewardsTitle() -> (Text_RewardsTitle:Common Text Object Reference)
    InText=Select(
     Option 0:Text="LOCTEXT(\"Reward\")"
     Option 1:Text="LOCTEXT(\"Rewards\")"
     Index="integer > integer"(A=@n2, B=1) {is=title K2Node_CallFunction}
    ) -> (ReturnValue:Text)
   )
  ]}]
variables []
widgets {
 info {inheritedWidgetCount=3, parentClass=/Script/FortniteConversationUI.FortDuelBrief, rootWidgetClass=&SizeBox, widgetCount=11}
 widgets [
  SizeBox_36 {
   widgetClassPath=&SizeBox
   widget=SizeBox_36
   properties {bOverride_MaxDesiredWidth=true, heightOverride=350, maxDesiredWidth=900, widthOverride=550}
  }
  VerticalBox_1 {widgetClassPath=&VerticalBox, parent=SizeBox_36, widget=VerticalBox_1, slot {class=SizeBoxSlot}}
  Border_168 {
   widgetClassPath=&Border
   parent=VerticalBox_1
   widget=Border_168
   properties {brushColor=<rgba 0 0 0 0.446>, padding=<ltrb 18 10 18 10>}
   slot {class=VerticalBoxSlot, properties {verticalAlignment=VAlign_Top}}
  }
  Overlay_0 {widgetClassPath=&Overlay, parent=Border_168, widget=Overlay_0, slot {class=BorderSlot, properties {padding=<ltrb 18 10 18 10>}}}
  Text_Title {
   widgetClassPath=&CommonTextBlock
   parent=Overlay_0
   widget=Text_Title
   properties {
    text=Duel
    textTransformPolicy=ToUpper
    font {
     bForceMonospaced=false
     bMaterialIsStencil=false
     fontMaterial=null
     fontObject=&Burbank
     letterSpacing=0
     monospacedWidth=1
     size=18
     skewAmount=0
     typefaceFontName=BurbankRegular-Black
     outlineSettings {
      bApplyOutlineToDropShadows=false
      bMiteredCorners=false
      bSeparateFillAlpha=false
      outlineMaterial=null
      outlineSize=2
      outlineColor=<rgba 0.0052 0.0052 0.0056 0.6>
     }
    }
    renderTransform {angle=0, scale=<1 1>, shear=<-8 0>, translation=<0 -30>}
    shadowOffset=<0 0>
   }
   slot {class=OverlaySlot}
  }
  VerticalBox_0 {
   widgetClassPath=&VerticalBox
   parent=Overlay_0
   widget=VerticalBox_0
   slot {class=OverlaySlot, properties {verticalAlignment=VAlign_Center}}
  }
  Text_Objective {
   widgetClassPath=&CommonRichTextBlock
   bInherited=true
   bIsVariable=true
   parent=VerticalBox_0
   widget=Text_Objective
   properties {
    bOverrideDefaultStyle=true
    text="Challenge <Name>{Name}</>"
    textStyleSet=&/Conversation_RichTextStyleSet
    wrapTextAt=900
    defaultTextStyleOverride {
     overflowPolicy=Clip
     transformPolicy=None
     colorAndOpacity {colorUseRule=UseColor_Specified, specifiedColor=<rgba 1 1 1 1>}
     font {
      bForceMonospaced=false
      bMaterialIsStencil=false
      fontMaterial=null
      fontObject=&BurbankSmall
      letterSpacing=0
      monospacedWidth=1
      size=20
      skewAmount=0
      typefaceFontName=Bold
      outlineSettings {
       bApplyOutlineToDropShadows=false
       bMiteredCorners=false
       bSeparateFillAlpha=false
       outlineMaterial=null
       outlineSize=0
       outlineColor=<rgba 0 0 0 1>
      }
     }
     highlightColor {colorUseRule=UseColor_Specified, specifiedColor=<rgba 0 0 0 1>}
     highlightShape {
      bIsDynamicallyLoaded=false
      drawAs=Image
      imageType=NoImage
      mirroring=NoMirror
      resourceName=None
      resourceObject=null
      tiling=NoTile
      imageSize=<32 32>
      margin=<ltrb 0 0 0 0>
      outlineSettings {
       bUseBrushTransparency=false
       roundingType=HalfHeightRadius
       width=0
       color {colorUseRule=UseColor_Specified, specifiedColor=<rgba 0 0 0 0>}
       cornerRadii=<0 0 0 0>
      }
      tintColor {colorUseRule=UseColor_Specified, specifiedColor=<rgba 1 1 1 1>}
      uVRegion {bIsValid=false, max=<0 0>, min=<0 0>}
     }
     selectedBackgroundColor {colorUseRule=UseColor_Foreground, specifiedColor=<rgba 1 0 1 1>}
     shadowColorAndOpacity=<rgba 0 0 0 1>
     shadowOffset=<0 0>
     strikeBrush {
      bIsDynamicallyLoaded=false
      drawAs=Image
      imageType=NoImage
      mirroring=NoMirror
      resourceName=None
      resourceObject=null
      tiling=NoTile
      imageSize=<32 32>
      margin=<ltrb 0 0 0 0>
      outlineSettings {
       bUseBrushTransparency=false
       roundingType=HalfHeightRadius
       width=0
       color {colorUseRule=UseColor_Specified, specifiedColor=<rgba 0 0 0 0>}
       cornerRadii=<0 0 0 0>
      }
      tintColor {colorUseRule=UseColor_Specified, specifiedColor=<rgba 1 1 1 1>}
      uVRegion {bIsValid=false, max=<0 0>, min=<0 0>}
     }
     underlineBrush {
      bIsDynamicallyLoaded=false
      drawAs=Image
      imageType=NoImage
      mirroring=NoMirror
      resourceName=None
      resourceObject=null
      tiling=NoTile
      imageSize=<32 32>
      margin=<ltrb 0 0 0 0>
      outlineSettings {
       bUseBrushTransparency=false
       roundingType=HalfHeightRadius
       width=0
       color {colorUseRule=UseColor_Specified, specifiedColor=<rgba 0 0 0 0>}
       cornerRadii=<0 0 0 0>
      }
      tintColor {colorUseRule=UseColor_Specified, specifiedColor=<rgba 1 1 1 1>}
      uVRegion {bIsValid=false, max=<0 0>, min=<0 0>}
     }
    }
    renderTransform {angle=0, scale=<1 1>, shear=<-8 0>, translation=<0 0>}
   }
   slot {class=VerticalBoxSlot}
  }
  VerticalBox_Rewards {
   widgetClassPath=&VerticalBox
   bIsVariable=true
   parent=VerticalBox_0
   widget=VerticalBox_Rewards
   slot {class=VerticalBoxSlot, properties {padding=<ltrb 0 10 0 0>}}
  }
  Text_RewardsTitle {
   widgetClassPath=&CommonTextBlock
   bIsVariable=true
   parent=VerticalBox_Rewards
   widget=Text_RewardsTitle
   properties {
    text=Rewards
    textTransformPolicy=ToUpper
    font {
     bForceMonospaced=false
     bMaterialIsStencil=false
     fontMaterial=null
     fontObject=&Burbank
     letterSpacing=0
     monospacedWidth=1
     size=18
     skewAmount=0
     typefaceFontName=BurbankRegular-Black
     outlineSettings {
      bApplyOutlineToDropShadows=false
      bMiteredCorners=false
      bSeparateFillAlpha=false
      outlineMaterial=null
      outlineSize=2
      outlineColor=<rgba 0.0052 0.0052 0.0056 0.6>
     }
    }
    renderTransform {angle=0, scale=<1 1>, shear=<-8 0>, translation=<0 0>}
    shadowOffset=<0 0>
   }
   slot {class=VerticalBoxSlot}
  }
  TileView {
   widgetClassPath=/Script/CommonUI.CommonTileView
   bInherited=true
   bIsVariable=true
   parent=VerticalBox_Rewards
   widget=TileView
   properties {
    bIsFocusable=false
    clipping=Inherit
    consumeMouseWheel=Never
    entryHeight=64
    entryWidgetClass=&/DuelRewardEntry.DuelRewardEntry_C
    entryWidth=64
    horizontalEntrySpacing=8
    orientation=Orient_Horizontal
    selectionMode=None
    tileAlignment=LeftAligned
    verticalEntrySpacing=8
    visibility=HitTestInvisible
   }
   slot {class=VerticalBoxSlot, properties {horizontalAlignment=HAlign_Left}}
  }
  TransactionStrip_Main {widgetClassPath=/Script/FortniteConversationUI.FortTransactionStrip, bInherited=true, widget=null}
 ]
}
