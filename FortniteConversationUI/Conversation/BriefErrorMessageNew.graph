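# Widget Blueprint for a brief error banner in the Fortnite conversation UI, showing an alert icon beside uppercase text on a translucent dark border.
# SetErrorMessage replaces the displayed text.
# PreConstruct applies the wrap width, rebuilds the text font with the FontSize variable, and tints the border's content with ContentColor, red by default.
# The SetSelectedVisuals event is present but has no logic.
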WidgetBlueprint /FortniteConversationUI/Conversation/BriefErrorMessageNew
graphs [EventGraph {nodes [
   n1 = SetErrorMessage!() -> (OutputDelegate, then, ErrorText:Text) {is=title K2Node_CustomEvent}
   > "SetText (Text)"(self=Get ErrorText() -> (ErrorText:Common Text Object Reference), InText=@n1.ErrorText)
   n2 = Event PreConstruct()
   > Parent: PreConstruct(IsDesignTime=@n2.IsDesignTime)
   n3 = Get ErrorText()
   > Set WrapTextAt!(execute, WrapTextAt:float=Get Wrap Text At() -> (Wrap Text At:float), self:Text Layout Widget Object Reference=@n3) -> (then, Output_Get:float)
   n4 = Break Slate Font Info!(SlateFontInfo:"Slate Font Info Structure (by ref)"=Get Font!(self:Text Object Reference=@n3) -> (Font:Slate Font Info Structure)) -> (
    FontObject:Object Reference
    FontMaterial:Object Reference
    OutlineSettings:Font Outline Settings Structure
    TypefaceFontName:Name
    Size:float=~
    LetterSpacing:int
    SkewAmount:float=~
    bForceMonospaced:bool=~
    bMaterialIsStencil:bool=~
    MonospacedWidth:float=~
   ) {is=title K2Node_BreakStruct}
   > SetFont(
    self=@n3
    InFontInfo=MakeSlateFontInfo!(
     FontObject:Object Reference=@n4
     FontMaterial:Object Reference=@n4.FontMaterial
     OutlineSettings:Font Outline Settings Structure=@n4.OutlineSettings
     TypefaceFontName:Name=@n4.TypefaceFontName
     Size:float=Get FontSize() -> (FontSize:float)
     LetterSpacing:int=@n4.LetterSpacing
     SkewAmount:float=@n4.SkewAmount
     bForceMonospaced:bool=@n4.bForceMonospaced
     bMaterialIsStencil:bool=false
     MonospacedWidth:float=@n4.MonospacedWidth
    ) -> (ReturnValue:Slate Font Info Structure)
   )
   > SetContentColorAndOpacity!(
    execute
    self:Border Object Reference=Get Border_Outer() -> (Border_Outer:Border Object Reference)
    InContentColorAndOpacity:Linear Color Structure=Get ContentColor() -> (ContentColor:Linear Color Structure)
   ) -> (then)
   Event SetSelectedVisuals!() -> (OutputDelegate, then, IsSelected:bool) {is=title K2Node_Event}
  ]}]
variables [
 ContentColor {type=LinearColor, default=<rgba 1 0.0561 0.0648 1>}
 FontSize {type=number, default=26.67}
 Wrap Text At {type=number}
]
widgets {
 info {inheritedWidgetCount=0, parentClass=&UserWidget, rootWidgetClass=&Border, widgetCount=4}
 widgets [
  Border_Outer {
   widgetClassPath=&Border
   bIsVariable=true
   widget=Border_Outer
   properties {verticalAlignment=VAlign_Center, brushColor=<rgba 0 0 0 0.446>, contentColorAndOpacity=<rgba 1 0.0561 0.0648 1>, padding=<ltrb 30 10 30 10>}
  }
  HorizontalBox_1 {
   widgetClassPath=&HorizontalBox
   parent=Border_Outer
   widget=HorizontalBox_1
   slot {class=BorderSlot, properties {verticalAlignment=VAlign_Center, padding=<ltrb 30 10 30 10>}}
  }
  SubDescriptionAlertImage {
   widgetClassPath=&Image
   bIsVariable=true
   parent=HorizontalBox_1
   widget=SubDescriptionAlertImage
   properties {
    visibility=HitTestInvisible
    brush {
     bIsDynamicallyLoaded=false
     drawAs=Image
     imageType=NoImage
     mirroring=NoMirror
     resourceName=None
     resourceObject=/Game/UI/Foundation/Textures/Icons/HUD/T-Icon-Alert2-32
     tiling=NoTile
     imageSize=<25 25>
     margin=<ltrb 0 0 0 0>
     outlineSettings {
      bUseBrushTransparency=false
      roundingType=HalfHeightRadius
      width=0
      color {colorUseRule=UseColor_Specified, specifiedColor=<rgba 0 0 0 0>}
      cornerRadii=<0 0 0 0>
     }
     tintColor {colorUseRule=UseColor_Specified, specifiedColor=<rgba 1 1 1 1>}
     uVRegion {bIsValid=false, max=<0 0>, min=<0 0>}
    }
   }
   slot {class=HorizontalBoxSlot, properties {horizontalAlignment=HAlign_Center, verticalAlignment=VAlign_Center, padding=<ltrb 4 0 8 0>}}
  }
  ErrorText {
   widgetClassPath=&CommonTextBlock
   bIsVariable=true
   parent=HorizontalBox_1
   widget=ErrorText
   properties {
    scrollStyle=&TextScrollStyle-Base_C
    text=You can't do that right now
    textTransformPolicy=ToUpper
    font {
     bForceMonospaced=false
     bMaterialIsStencil=false
     fontMaterial=null
     fontObject=&BurbankSmall
     letterSpacing=0
     monospacedWidth=1
     size=20
     skewAmount=0.2
     typefaceFontName=Bold
     outlineSettings {
      bApplyOutlineToDropShadows=false
      bMiteredCorners=false
      bSeparateFillAlpha=false
      outlineMaterial=null
      outlineSize=0
      outlineColor=<rgba 0 0 0 1>
     }
    }
    shadowOffset=<0 0>
   }
   slot {class=HorizontalBoxSlot, properties {verticalAlignment=VAlign_Center, size {sizeRule=Fill, value=1}}}
  }
 ]
}
