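# Transaction strip widget for the conversation UI, showing a service's cost or reward, or its locked requirement.
# Transaction info switches between locked and available states; available shows FREE or a cost colored by affordability, with owned currency hidden under a rollback CVar.
# Preconstruct applies cost/reward, border and text settings, plus optional donation-fund styling.
# Functions handle selection colors, reward fields and a donation pulse animation on currency and fund outline.
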WidgetBlueprint /FortniteConversationUI/Conversation/FortTransactionStripNew
animations [Anim_Donated {
  binding_count=3
  length_seconds=0.55
  object=&self:Anim_Donated
  sequence {bindings [
    OwnedCurrency {
     id=#1
     possessed_class=CommonTextBlock
     tracks [{
       type=MovieScene2DTransformTrack
       sections [{
         type=MovieScene2DTransformSection
         channels [
          Translation.X {type=MovieSceneScriptingFloatChannel, const [0,1]}
          Translation.Y {type=MovieSceneScriptingFloatChannel, const [0,1]}
          Angle {type=MovieSceneScriptingFloatChannel, const [0,1]}
          Scale.X {type=MovieSceneScriptingFloatChannel, keys [[0,1],[1,1.2],[4,1]]}
          Scale.Y {type=MovieSceneScriptingFloatChannel, keys [[0,1],[1,1.2],[4,1]]}
          Shear.X {type=MovieSceneScriptingFloatChannel, const [-8,1]}
          Shear.Y {type=MovieSceneScriptingFloatChannel, const [0,1]}
         ]
        }]
      }]
    }
    Overlay_1 {
     id=#2
     possessed_class=Overlay
     tracks [{
       type=MovieScene2DTransformTrack
       sections [{
         type=MovieScene2DTransformSection
         channels [
          Scale.X {type=MovieSceneScriptingFloatChannel, keys [[0,1],[1,1.0325],[5,1]]}
          Scale.Y {type=MovieSceneScriptingFloatChannel, keys [[0,1],[1,1.0325],[5,1]]}
         ]
        }]
      }]
    }
    Image_FundOutline {
     id=#3
     possessed_class=Image
     tracks [{
       type=MovieSceneWidgetMaterialTrack
       sections [{
         type=MovieSceneParameterSection
         channels [
          FillColorR {type=MovieSceneScriptingFloatChannel, const [0,5]}
          FillColorG {type=MovieSceneScriptingFloatChannel, keys [[-1,0.0417],[0,0.0417],[1,0.75],[2,0.75],[10,0.0417]]}
          FillColorB {type=MovieSceneScriptingFloatChannel, keys [[-1,0.5],[0,0.5],[1,0.25],[2,0.25],[10,0.5]]}
          FillColorA {type=MovieSceneScriptingFloatChannel, const [1,2]}
          RingColorR {type=MovieSceneScriptingFloatChannel, const [0,5]}
          RingColorG {type=MovieSceneScriptingFloatChannel, keys [[-1,0.0417],[0,0.0417],[1,0.75],[2,0.75],[10,0.0417]]}
          RingColorB {type=MovieSceneScriptingFloatChannel, keys [[-1,0.5],[0,0.5],[1,0.25],[2,0.25],[10,0.5]]}
          RingColorA {type=MovieSceneScriptingFloatChannel, const [1,2]}
         ]
        }]
      }]
    }
   ]}
 }]
graphs [
 DonationAnim {nodes [
   DonationAnim() {is=title K2Node_FunctionEntry}
   > PlayAnimation(InAnimation=Get Anim_Donated() -> (Anim_Donated:Widget Animation Object Reference))
  ]}
 EventGraph {nodes [
   UpdateCurrencyIcon() {is=title K2Node_CustomEvent}
   n1 = Get CachedGlobalCurrency() -> (CachedGlobalCurrency:Fort World Item Definition Weak object)
   > Is Valid(InputObject=@n1)
   > SetBrushFromSoftTexture!(
    execute
    self:Image Object Reference=Select(
     Option 0:Image Object Reference=Get Image_CurrencyIcon() -> (Image_CurrencyIcon:Image Object Reference)
     Option 1:Image Object Reference=Get Image_Money() -> (Image_Money:Image Object Reference)
     Index=Get isReward() -> (isReward:bool)
    ) -> (ReturnValue:Image Object Reference)
    SoftTexture:Texture 2D Soft Object Reference=GetSmallPreviewImage!(self:Fort Item Definition Object Reference=@n1) -> (ReturnValue:Texture 2D Soft Object Reference)
    bMatchSize:bool=false
   ) -> (then)
   UpdatedInfo() {is=title K2Node_CustomEvent}
   > "SetText (Text)"(
    self=Get Text_CostDescription() -> (Text_CostDescription:Common Text Object Reference)
    InText=Get TransactionString() -> (TransactionString:Text)
   )
   > SetActiveWidget(
    self:Widget Switcher Object Reference=Get Switcher_Type() -> (Switcher_Type:Common Widget Switcher Legacy Object Reference)
    Widget=Select(
     Option 0:Widget Object Reference=Get HorizontalBox_Cost() -> (HorizontalBox_Cost:Horizontal Box Object Reference)
     Option 1:Widget Object Reference=Get HorizontalBox_Reward() -> (HorizontalBox_Reward:Horizontal Box Object Reference)
     Index=Get isReward()
    ) -> (ReturnValue:Widget Object Reference)
   )
   > SetShowBorder!(execute, self, ShowBorder:bool=Get ShowBorder() -> (ShowBorder:bool)) -> (then)
   > SetCostReward!(execute, self, EnableCostReward:bool=Get Enable Cost Reward() -> (Enable Cost Reward:bool)) -> (then)
   > SetTextCostSettings!(
    execute
    self
    Text_Cost_OutlineSettings:Font Outline Settings Structure=Get Text_Cost_OutlineSettings() -> (Text_Cost_OutlineSettings:Font Outline Settings Structure)
    Text_Cost_Color:Slate Color Structure=Get Text_Cost_Color() -> (Text_Cost_Color:Slate Color Structure)
   ) -> (then)
   > n2 = Branch(Condition=Get bFundingVersion() -> (bFundingVersion:bool))
   > SetVisibility(self=Get FundStyling() -> (FundStyling:Overlay Object Reference), InVisibility=HitTestInvisible)
   n3 = Get Border_OwnedGold() -> (Border_OwnedGold:Border Object Reference)
   > SetBrushColor(self=@n3, InBrushColor=(R=0,G=0,B=0,A=0))
   > SetPadding!(execute, self:Border Object Reference=@n3, InPadding:Margin Structure=Make Margin(Left=24, Top=2, Right=24, Bottom=2)) -> (then)
   > SetContentColorAndOpacity!(
    execute
    self:Border Object Reference=Get DescriptionTint() -> (DescriptionTint:Border Object Reference)
    InContentColorAndOpacity:Linear Color Structure=(R=0,G=0.6667,B=1,A=1)
   ) -> (then)
   SetContentColorAndOpacity(execute=@n2.else, self=Get DescriptionTint(), InContentColorAndOpacity=(R=1,G=1,B=1,A=1))
   Event PreConstruct()
   > UpdatedInfo() {is=title K2Node_CallFunction}
   Event Construct()
   > SetVisibility(self=Get Image_CurrencyIcon_Free() -> (Image_CurrencyIcon_Free:Image Object Reference), InVisibility=Collapsed)
   n4 = Get Anim_Donated()
   > PlayAnimation(InAnimation=@n4, StartAtTime=GetEndTime(self=@n4), PlayMode=Reverse)
   n5 = Event OnSetTransactionCostVisibility!() -> (OutputDelegate, then, bVisible:bool) {is=title K2Node_Event}
   > n6 = Branch(Condition=@n5.bVisible)
   n7 = Get HorizontalBox_Cost()
   > SetVisibility(self=@n7, InVisibility=HitTestInvisible)
   SetVisibility(execute=@n6.else, self=@n7, InVisibility=Collapsed)
   n8 = Event OnTransactionInfoReceived!() -> (OutputDelegate, then, TransactionInfo:"Fort Service Transaction Info Structure (by ref)") {is=title K2Node_Event}
   n9 = Break Fort Service Transaction Info!(FortServiceTransactionInfo:"Fort Service Transaction Info Structure (by ref)"=@n8.TransactionInfo) -> (Cost:int, OwnedCurrency:int, CostMultiplier:float=~, bIsLocked:bool=~, UnlockLevel:int, LockedText:Text) {is=title K2Node_BreakStruct}
   > n10 = Branch(Condition=@n9.bIsLocked)
   > PopulateLocked!(execute, self, UnlockLevel:int=@n9.UnlockLevel, Locked Text:Text=@n9.LockedText) -> (then)
   PopulateAvailable!(execute=@n10.else, self, Cost:int=@n9, OwnedCurrency:int=@n9.OwnedCurrency) -> (then)
   UpdatedInfo()
  ]}
 PopulateAvailable {nodes [
   n1 = PopulateAvailable!() -> (then, Cost:int, OwnedCurrency:int) {is=title K2Node_FunctionEntry}
   > n2 = Branch(Condition="integer <= integer"(A=@n1))
   > "SetText (Text)"(self=Get Text_CostDescription(), InText="LOCTEXT(\"FREE\")")
   > SetVisibility(self=Get Image_CurrencyIcon_Free(), InVisibility=HitTestInvisible)
   > n3 = SetVisibility(self=Get Switcher_Type(), InVisibility=Collapsed)
   UpdateCurrencyIcon(execute=[@n13.then, @n3.then]) {is=title K2Node_CallFunction}
   "SetText (Text)"(execute=@n2.else, self=Get Text_CostDescription(), InText=Get TransactionString())
   > SetActiveWidgetIndex(self=Get Switcher_State() -> (Switcher_State:Common Widget Switcher Legacy Object Reference), Index=0)
   > SetVisibility(self=Get Image_CurrencyIcon_Free(), InVisibility=Collapsed)
   > SetVisibility(self=Get Switcher_Type(), InVisibility=HitTestInvisible)
   > n4 = GetConsoleVariableBoolValue!(execute, VariableName:String=Fort.Rollback.NewUIConversationWheelEnabled) -> (then, ReturnValue:bool)
   > "SetText (Text)"(self=Get OwnedCurrency() -> (OwnedCurrency:Common Text Object Reference), InText="To Text (Integer)"(Value=@n1.OwnedCurrency))
   > SetVisibility(
    self=Get OwnedCurrency()
    InVisibility=Select(Option 0:ESlateVisibility Enum=SelfHitTestInvisible, Option 1:ESlateVisibility Enum=Collapsed, Index=@n4) -> (ReturnValue:ESlateVisibility Enum)
   )
   n5 = "To Text (Integer)"(Value=@n1)
   > "SetText (Text)"(
    self=Get Text_Cost() -> (Text_Cost:Common Text Object Reference)
    InText=Select(Option 0:Text=Format Text!(Format="LOCTEXT(\"/{gold}\")", gold:Text=@n5) -> (Result), Option 1:Text=@n5, Index=@n4) -> (ReturnValue:Text)
   )
   n6 = "integer <= integer"!(A:int=@n1, B:int=@n1.OwnedCurrency) -> (ReturnValue:bool) {is=title K2Node_CallFunction}
   n7 = Make Slate Color(SpecifiedColor=Get RedColorText() -> (RedColorText:Linear Color Structure))
   n8 = Select(
    Option 0:Slate Color Structure=Get Text_Cost_Color()
    Option 1:Slate Color Structure=Get Text_Cost_Color_Selected() -> (Text_Cost_Color_Selected:Slate Color Structure)
    Index=Get IsSelected() -> (IsSelected:bool)
   ) -> (ReturnValue:Slate Color Structure)
   > SetColorAndOpacity(
    self=Get Text_Cost()
    InColorAndOpacity=Select(
     Option 0=Get Text_Cost_Color()
     Option 1=Select(Option 0=@n7, Option 1=@n8, Index=@n6)
     Index=Get Text_Cost_IsInsufficientFundsRed() -> (Text_Cost_IsInsufficientFundsRed:bool)
    )
   )
   n9 = Get RedColorText()
   n10 = Get IsSelected()
   n11 = Get Text_Cost_Color_Selected()
   n12 = Get Text_Cost_Color()
   > n13 = SetColorAndOpacity(
    self=Get OwnedCurrency()
    InColorAndOpacity=Select(Option 0=Make Slate Color(SpecifiedColor=@n9), Option 1=Select(Option 0=@n12, Option 1=@n11, Index=@n10), Index=@n6)
   )
  ]}
 PopulateLocked {nodes [
   n1 = PopulateLocked!() -> (then, UnlockLevel:int, Locked Text:Text) {is=title K2Node_FunctionEntry}
   > SetActiveWidgetIndex(self=Get Switcher_State(), Index=1)
   > "SetText (Text)"(
    self=Get Text_UnlockLevel() -> (Text_UnlockLevel:Common Text Object Reference)
    InText=Format Text!(Format=@n1.Locked Text, 0:int=@n1) -> (Result)
   )
  ]}
 SetCostReward {nodes [
   n1 = SetCostReward!() -> (then, EnableCostReward:bool=~) {is=title K2Node_FunctionEntry}
   > SetVisibility(
    self=Get Switcher_Type()
    InVisibility=Select(Option 0:ESlateVisibility Enum=Hidden, Option 1:ESlateVisibility Enum=HitTestInvisible, Index=@n1) -> (ReturnValue:ESlateVisibility Enum)
   )
  ]}
 SetSelectedVisuals {nodes [
   n1 = SetSelectedVisuals!() -> (then, IsSelected:bool=~) {is=title K2Node_FunctionEntry}
   > Set IsSelected(IsSelected:bool=@n1) -> (Output_Get:bool)
   > SetBrushColor(
    self=Get Border_OwnedGold()
    InBrushColor=Select(Option 0:Linear Color Structure=(R=0.0273,G=0.0273,B=0.0273,A=1), Option 1:Linear Color Structure=(R=1,G=1,B=1,A=1), Index=@n1) -> (ReturnValue:Linear Color Structure)
   )
   > SetContentColorAndOpacity(self=Get DescriptionTint(), InContentColorAndOpacity=Select(Option 0=(R=1,G=1,B=1,A=1), Option 1=(R=0,G=0,B=0,A=1), Index=@n1))
   > SetColorAndOpacity!(
    execute
    self=[
     Get Text_MoneyReward() -> (Text_MoneyReward:Common Text Object Reference)
     Get Text_XpReward() -> (Text_XpReward:Common Text Object Reference)
     Get Text_CustomReward() -> (Text_CustomReward:Common Text Object Reference)
    ]
    InColorAndOpacity_SpecifiedColor:Linear Color Structure=Select(Option 0=(R=1,G=1,B=1,A=1), Option 1=(R=0,G=0,B=0,A=1), Index=@n1)
    InColorAndOpacity_ColorUseRule:ESlateColorStylingMode Enum
   ) -> (then)
   > SetColorAndOpacity!(
    execute
    self:Image Object Reference=[Get Image_56() -> (Image_56:Image Object Reference), Get Image_TokenReward() -> (Image_TokenReward:Image Object Reference)]
    InColorAndOpacity:Linear Color Structure=Select(Option 0=(R=1,G=1,B=1,A=1), Option 1=(R=0,G=0,B=0,A=1), Index=@n1)
   ) -> (then)
   n2 = Get OwnedCurrency()
   > n3 = Branch(Condition="Equal (LinearColor)"(
     A:Linear Color Structure=Get!(self:Text Object Reference=@n2) -> (ColorAndOpacity_SpecifiedColor:Linear Color Structure, ColorAndOpacity_ColorUseRule:ESlateColorStylingMode Enum)
     B:Linear Color Structure=Get RedColorText()
    ) -> (ReturnValue:bool=~) {is=title K2Node_PromotableOperator})
   SetColorAndOpacity(
    execute=@n3.else
    self:Text Object Reference=[Get OwnedCurrency(), Get Text_Cost()]
    InColorAndOpacity:Slate Color Structure=Select(Option 0:Slate Color Structure=Get Text_Cost_Color(), Option 1:Slate Color Structure=Get Text_Cost_Color_Selected(), Index=Get IsSelected()) -> (ReturnValue:Slate Color Structure)
   )
  ]}
 SetShowBorder {nodes [
   n1 = SetShowBorder!() -> (then, ShowBorder:bool=~) {is=title K2Node_FunctionEntry}
   > Set ShowBorder(ShowBorder:bool=@n1) -> (Output_Get:bool)
   n2 = Get Border_OwnedGold()
   n3 = Get!(self:Border Object Reference=@n2) -> (
    Background_TintColor:Slate Color Structure
    Background_DrawAs:ESlateBrushDrawType Enum
    Background_Tiling:ESlateBrushTileType Enum
    Background_Mirroring:ESlateBrushMirrorType Enum
    Background_ImageSize:Vector 2D Structure
    Background_Margin:Margin Structure
    Background_ResourceObject:Object Reference
    Background_OutlineSettings:Slate Brush Outline Settings Structure
   )
   > SetBrush!(
    execute
    self:Border Object Reference=@n2
    InBrush_TintColor:Slate Color Structure=@n3
    InBrush_DrawAs:ESlateBrushDrawType Enum=Select(Option 0:ESlateBrushDrawType Enum=NoDrawType, Option 1:ESlateBrushDrawType Enum=Image, Index=Get ShowBorder()) -> (ReturnValue:ESlateBrushDrawType Enum)
    InBrush_Tiling:ESlateBrushTileType Enum=@n3.Background_Tiling
    InBrush_Mirroring:ESlateBrushMirrorType Enum=@n3.Background_Mirroring
    InBrush_ImageSize:Vector 2D Structure=@n3.Background_ImageSize
    InBrush_Margin:Margin Structure=@n3.Background_Margin
    InBrush_ResourceObject:Object Reference=@n3.Background_ResourceObject
    InBrush_OutlineSettings:Slate Brush Outline Settings Structure=@n3.Background_OutlineSettings
   ) -> (then)
  ]}
 SetTextCostIsRedInsufficientFunds {nodes [
   n1 = SetTextCostIsRedInsufficientFunds!() -> (then, Text_Cost_IsInsufficientFundsRed:bool=~) {is=title K2Node_FunctionEntry}
   > Set Text_Cost_IsInsufficientFundsRed(Text_Cost_IsInsufficientFundsRed:bool=@n1) -> (Output_Get:bool)
  ]}
 SetTextCostSettings {nodes [
   n1 = SetTextCostSettings!() -> (then, Text_Cost_OutlineSettings:Font Outline Settings Structure, Text_Cost_Color:Slate Color Structure) {is=title K2Node_FunctionEntry}
   > Set Text_Cost_OutlineSettings(Text_Cost_OutlineSettings:Font Outline Settings Structure=@n1) -> (Output_Get:Font Outline Settings Structure)
   > Set Text_Cost_Color(Text_Cost_Color:Slate Color Structure=@n1.Text_Cost_Color) -> (Output_Get:Slate Color Structure)
   n2 = Get Text_Cost()
   n3 = Get!(self:Text Object Reference=@n2) -> (
    Font_FontObject:Object Reference
    Font_FontMaterial:Object Reference
    Font_OutlineSettings:Font Outline Settings Structure
    Font_TypefaceFontName:Name
    Font_Size:float=~
    Font_LetterSpacing:int
    Font_SkewAmount:float=~
    Font_bForceMonospaced:bool=~
    Font_bMaterialIsStencil:bool=~
    Font_MonospacedWidth:float=~
   )
   > SetFont(
    self=@n2
    InFontInfo=MakeSlateFontInfo!(
     FontObject:Object Reference=@n3
     FontMaterial:Object Reference=@n3.Font_FontMaterial
     OutlineSettings:Font Outline Settings Structure=Get Text_Cost_OutlineSettings()
     TypefaceFontName:Name=@n3.Font_TypefaceFontName
     Size:float=@n3.Font_Size
     LetterSpacing:int=@n3.Font_LetterSpacing
     SkewAmount:float=@n3.Font_SkewAmount
     bForceMonospaced:bool=@n3.Font_bForceMonospaced
     bMaterialIsStencil:bool=@n3.Font_bMaterialIsStencil
     MonospacedWidth:float=@n3.Font_MonospacedWidth
    ) -> (ReturnValue:Slate Font Info Structure)
   )
   > SetColorAndOpacity(self=Get Text_Cost(), InColorAndOpacity=Get Text_Cost_Color())
  ]}
 SetupTransactionStripRewardData {nodes [
   n1 = SetupTransactionStripRewardData!() -> (
    then
    MoneyIcon:Texture 2D Soft Object Reference
    MoneyName:Text
    RewardMoneyFieldVisibility:ESlateVisibility Enum=Visible
    RewardXpFieldVisibility:ESlateVisibility Enum=Visible
   ) {is=title K2Node_FunctionEntry}
   > SetBrushFromSoftTexture(self=Get Image_Money(), SoftTexture=@n1, bMatchSize=true)
   > "SetText (Text)"(self=Get Text_MoneyReward(), InText=Get MoneyName() -> (MoneyName:Text))
   > SetVisibility(
    self=Get Reward_Money() -> (Reward_Money:Size Box Object Reference)
    InVisibility=Get RewardMoneyFieldVisibility() -> (RewardMoneyFieldVisibility:ESlateVisibility Enum=Visible)
   )
   > SetVisibility(
    self=Get Reward_XP() -> (Reward_XP:Size Box Object Reference)
    InVisibility=Get RewardXpFieldVisibility() -> (RewardXpFieldVisibility:ESlateVisibility Enum=Visible)
   )
  ]}
]
variables [
 bFundingVersion {type=boolean}
 Enable Cost Reward {type=boolean, default=true}
 isReward {type=boolean}
 IsSelected {type=boolean}
 RedColorText {type=LinearColor, default=<rgba 1 0.0545 0.0666 1>}
 ShowBorder {type=boolean, default=true}
 Text_Cost_Color {type=SlateColor, default {colorUseRule=UseColor_Specified, specifiedColor=<rgba 1 1 1 1>}}
 Text_Cost_Color_Selected {type=SlateColor, default {colorUseRule=UseColor_Specified, specifiedColor=<rgba 0.0052 0.0052 0.0056 1>}}
 Text_Cost_IsInsufficientFundsRed {type=boolean, default=true}
 Text_Cost_OutlineSettings {
  type=FontOutlineSettings
  default {
   bApplyOutlineToDropShadows=false
   bMiteredCorners=false
   bSeparateFillAlpha=false
   outlineMaterial=null
   outlineSize=0
   outlineColor=<rgba 0 0 0 1>
  }
 }
 TransactionString {type=string, default=Cost}
]
widgets {
 info {inheritedWidgetCount=1, parentClass=/Script/FortniteConversationUI.FortTransactionStrip, rootWidgetClass=&Overlay, widgetCount=38}
 widgets [
  Overlay_1 {widgetClassPath=&Overlay, widget=Overlay_1}
  OV_TransactionContent {
   widgetClassPath=&Overlay
   parent=Overlay_1
   widget=OV_TransactionContent
   slot {class=OverlaySlot, properties {horizontalAlignment=HAlign_Fill, verticalAlignment=VAlign_Fill}}
  }
  FundStyling {
   widgetClassPath=&Overlay
   bIsVariable=true
   parent=OV_TransactionContent
   widget=FundStyling
   properties {visibility=Collapsed}
   slot {class=OverlaySlot, properties {horizontalAlignment=HAlign_Fill, verticalAlignment=VAlign_Fill}}
  }
  Image_FundShadow {
   widgetClassPath=&Image
   bIsVariable=true
   parent=FundStyling
   widget=Image_FundShadow
   properties {
    renderOpacity=0.5
    visibility=SelfHitTestInvisible
    brush {
     bIsDynamicallyLoaded=false
     drawAs=Box
     imageType=NoImage
     mirroring=NoMirror
     resourceName=None
     resourceObject=/FortniteConversationUI/Conversation/Materials/M_UI_DonationFund_BadgeCircle_Default_Secondary
     tiling=NoTile
     imageSize=<72 72>
     margin=<ltrb 0.5 0.5 0.5 0.5>
     outlineSettings {
      bUseBrushTransparency=false
      roundingType=HalfHeightRadius
      width=0
      color {colorUseRule=UseColor_Specified, specifiedColor=<rgba 0 0 0 0>}
      cornerRadii=<0 0 0 0>
     }
     tintColor {colorUseRule=UseColor_Specified, specifiedColor=<rgba 1 1 1 1>}
     uVRegion {bIsValid=false, max=<0 0>, min=<0 0>}
    }
    renderTransform {angle=0, scale=<1.05 1.45>, shear=<0 0>, translation=<0 0>}
   }
   slot {class=OverlaySlot, properties {horizontalAlignment=HAlign_Fill, verticalAlignment=VAlign_Fill}}
  }
  Image_FundOutline {
   widgetClassPath=&Image
   bIsVariable=true
   parent=FundStyling
   widget=Image_FundOutline
   properties {
    visibility=SelfHitTestInvisible
    brush {
     bIsDynamicallyLoaded=false
     drawAs=Box
     imageType=NoImage
     mirroring=NoMirror
     resourceName=None
     resourceObject=/FortniteConversationUI/Conversation/Materials/M_UI_DonationFund_StripBacking_Default
     tiling=NoTile
     imageSize=<64 64>
     margin=<ltrb 0.5 0.5 0.5 0.5>
     outlineSettings {
      bUseBrushTransparency=false
      roundingType=HalfHeightRadius
      width=0
      color {colorUseRule=UseColor_Specified, specifiedColor=<rgba 0 0 0 0>}
      cornerRadii=<0 0 0 0>
     }
     tintColor {colorUseRule=UseColor_Specified, specifiedColor=<rgba 1 1 1 1>}
     uVRegion {bIsValid=false, max=<0 0>, min=<0 0>}
    }
   }
   slot {class=OverlaySlot, properties {horizontalAlignment=HAlign_Fill, verticalAlignment=VAlign_Center}}
  }
  SizeBox_0 {
   widgetClassPath=&SizeBox
   parent=OV_TransactionContent
   widget=SizeBox_0
   properties {bOverride_HeightOverride=true, heightOverride=70}
   slot {class=OverlaySlot, properties {horizontalAlignment=HAlign_Fill, verticalAlignment=VAlign_Fill}}
  }
  Border_OwnedGold {
   widgetClassPath=&Border
   bIsVariable=true
   parent=SizeBox_0
   widget=Border_OwnedGold
   properties {brushColor=<rgba 0.0273 0.0273 0.0273 1>, padding=<ltrb 0 0 0 0>}
   slot {class=SizeBoxSlot}
  }
  Switcher_State {
   widgetClassPath=&CommonWidgetSwitcherLegacy
   bIsVariable=true
   parent=Border_OwnedGold
   widget=Switcher_State
   properties {transitionDuration=0}
   slot {class=BorderSlot, properties {padding=<ltrb 0 0 0 0>}}
  }
  HB_Available {
   widgetClassPath=&HorizontalBox
   parent=Switcher_State
   widget=HB_Available
   slot {class=WidgetSwitcherSlot, properties {verticalAlignment=VAlign_Center, padding=<ltrb 20 2 20 2>}}
  }
  Image_CurrencyIcon_Free {
   widgetClassPath=&Image
   bIsVariable=true
   parent=HB_Available
   widget=Image_CurrencyIcon_Free
   properties {
    visibility=Collapsed
    brush {
     bIsDynamicallyLoaded=false
     drawAs=Image
     imageType=NoImage
     mirroring=NoMirror
     resourceName=None
     resourceObject=/Game/UI/Foundation/Textures/Icons/Athena/T-T-Icon-BR-GoldBars-UI-Icon
     tiling=NoTile
     imageSize=<42 42>
     margin=<ltrb 0 0 0 0>
     outlineSettings {
      bUseBrushTransparency=false
      roundingType=HalfHeightRadius
      width=0
      color {colorUseRule=UseColor_Specified, specifiedColor=<rgba 0 0 0 0>}
      cornerRadii=<0 0 0 0>
     }
     tintColor {colorUseRule=UseColor_Specified, specifiedColor=<rgba 1 1 1 1>}
     uVRegion {bIsValid=false, max=<0 0>, min=<0 0>}
    }
   }
   slot {class=HorizontalBoxSlot, properties {horizontalAlignment=HAlign_Center, verticalAlignment=VAlign_Center, padding=<ltrb 3 0 3 0>}}
  }
  DescriptionTint {
   widgetClassPath=&Border
   bIsVariable=true
   parent=HB_Available
   widget=DescriptionTint
   properties {
    background {
     bIsDynamicallyLoaded=false
     drawAs=NoDrawType
     imageType=NoImage
     mirroring=NoMirror
     resourceName=None
     resourceObject=null
     tiling=NoTile
     imageSize=<32 32>
     margin=<ltrb 0 0 0 0>
     outlineSettings {
      bUseBrushTransparency=false
      roundingType=HalfHeightRadius
      width=0
      color {colorUseRule=UseColor_Specified, specifiedColor=<rgba 0 0 0 0>}
      cornerRadii=<0 0 0 0>
     }
     tintColor {colorUseRule=UseColor_Specified, specifiedColor=<rgba 1 1 1 1>}
     uVRegion {bIsValid=false, max=<0 0>, min=<0 0>}
    }
    padding=<ltrb 0 0 0 0>
   }
   slot {
    class=HorizontalBoxSlot
    properties {horizontalAlignment=HAlign_Left, verticalAlignment=VAlign_Center, padding=<ltrb 0 0 10 0>, size {sizeRule=Fill, value=1}}
   }
  }
  Text_CostDescription {
   widgetClassPath=&CommonTextBlock
   bInherited=true
   bIsVariable=true
   parent=DescriptionTint
   widget=Text_CostDescription
   properties {
    textTransformPolicy=ToUpper
    font {
     bForceMonospaced=false
     bMaterialIsStencil=false
     fontMaterial=null
     fontObject=&BurbankSmall
     letterSpacing=0
     monospacedWidth=1
     size=20
     skewAmount=0
     typefaceFontName=Black
     outlineSettings {
      bApplyOutlineToDropShadows=false
      bMiteredCorners=false
      bSeparateFillAlpha=false
      outlineMaterial=null
      outlineSize=0
      outlineColor=<rgba 0 0 0 1>
     }
    }
    margin=<ltrb 0 5 0 5>
    renderTransform {angle=0, scale=<1 1>, shear=<-8 0>, translation=<0 0>}
    shadowOffset=<0 0>
   }
   slot {class=BorderSlot, properties {padding=<ltrb 0 0 0 0>}}
  }
  Switcher_Type {
   widgetClassPath=&CommonWidgetSwitcherLegacy
   bIsVariable=true
   parent=HB_Available
   widget=Switcher_Type
   properties {activeWidgetIndex=1, transitionDuration=0}
   slot {class=HorizontalBoxSlot, properties {verticalAlignment=VAlign_Center}}
  }
  HorizontalBox_Cost {
   widgetClassPath=&HorizontalBox
   bIsVariable=true
   parent=Switcher_Type
   widget=HorizontalBox_Cost
   slot {class=WidgetSwitcherSlot}
  }
  Image_CurrencyIcon {
   widgetClassPath=&Image
   bIsVariable=true
   parent=HorizontalBox_Cost
   widget=Image_CurrencyIcon
   properties {brush {
     bIsDynamicallyLoaded=false
     drawAs=Image
     imageType=NoImage
     mirroring=NoMirror
     resourceName=None
     resourceObject=&T-T-Icon-BR-GoldBars-UI-Icon
     tiling=NoTile
     imageSize=<42 42>
     margin=<ltrb 0 0 0 0>
     outlineSettings {
      bUseBrushTransparency=false
      roundingType=HalfHeightRadius
      width=0
      color {colorUseRule=UseColor_Specified, specifiedColor=<rgba 0 0 0 0>}
      cornerRadii=<0 0 0 0>
     }
     tintColor {colorUseRule=UseColor_Specified, specifiedColor=<rgba 1 1 1 1>}
     uVRegion {bIsValid=false, max=<0 0>, min=<0 0>}
    }}
   slot {class=HorizontalBoxSlot, properties {horizontalAlignment=HAlign_Center, verticalAlignment=VAlign_Center, padding=<ltrb 3 0 3 0>}}
  }
  OwnedCurrency {
   widgetClassPath=&CommonTextBlock
   bIsVariable=true
   parent=HorizontalBox_Cost
   widget=OwnedCurrency
   properties {
    text=500
    font {
     bForceMonospaced=false
     bMaterialIsStencil=false
     fontMaterial=null
     fontObject=&BurbankSmall
     letterSpacing=0
     monospacedWidth=1
     size=24
     skewAmount=0
     typefaceFontName=Black
     outlineSettings {
      bApplyOutlineToDropShadows=false
      bMiteredCorners=false
      bSeparateFillAlpha=false
      outlineMaterial=null
      outlineSize=0
      outlineColor=<rgba 0.0052 0.0052 0.0056 0.3>
     }
    }
    renderTransform {angle=0, scale=<1 1>, shear=<-8 0>, translation=<0 0>}
    shadowOffset=<0 0>
   }
   slot {class=HorizontalBoxSlot, properties {verticalAlignment=VAlign_Center, padding=<ltrb 10 0 3 0>}}
  }
  Text_Cost {
   widgetClassPath=&CommonTextBlock
   bIsVariable=true
   parent=HorizontalBox_Cost
   widget=Text_Cost
   properties {
    text=/ 500
    textTransformPolicy=ToUpper
    wrapTextAt=900
    font {
     bForceMonospaced=false
     bMaterialIsStencil=false
     fontMaterial=null
     fontObject=&Burbank
     letterSpacing=0
     monospacedWidth=1
     size=24
     skewAmount=0
     typefaceFontName=Burbank-PlayerPower
     outlineSettings {
      bApplyOutlineToDropShadows=false
      bMiteredCorners=false
      bSeparateFillAlpha=false
      outlineMaterial=null
      outlineSize=0
      outlineColor=<rgba 0 0 0 1>
     }
    }
    renderTransform {angle=0, scale=<1 1>, shear=<-8 0>, translation=<0 0>}
    shadowOffset=<0 0>
   }
   slot {class=HorizontalBoxSlot, properties {verticalAlignment=VAlign_Center}}
  }
  HorizontalBox_Reward {
   widgetClassPath=&HorizontalBox
   bIsVariable=true
   parent=Switcher_Type
   widget=HorizontalBox_Reward
   slot {class=WidgetSwitcherSlot}
  }
  Reward_Money {
   widgetClassPath=&SizeBox
   bIsVariable=true
   parent=HorizontalBox_Reward
   widget=Reward_Money
   slot {class=HorizontalBoxSlot, properties {padding=<ltrb 15 0 0 0>}}
  }
  Border_MoneyRewardColor {
   widgetClassPath=&CommonBorder
   bIsVariable=true
   parent=Reward_Money
   widget=Border_MoneyRewardColor
   properties {padding=<ltrb 0 0 0 0>}
   slot {class=SizeBoxSlot}
  }
  HorizontalBox_1 {
   widgetClassPath=&HorizontalBox
   parent=Border_MoneyRewardColor
   widget=HorizontalBox_1
   slot {class=BorderSlot, properties {padding=<ltrb 0 0 0 0>}}
  }
  Image_Money {
   widgetClassPath=&Image
   bIsVariable=true
   parent=HorizontalBox_1
   widget=Image_Money
   properties {brush {
     bIsDynamicallyLoaded=false
     drawAs=Image
     imageType=NoImage
     mirroring=NoMirror
     resourceName=None
     resourceObject=&T-T-Icon-BR-GoldBars-UI-Icon
     tiling=NoTile
     imageSize=<42 42>
     margin=<ltrb 0 0 0 0>
     outlineSettings {
      bUseBrushTransparency=false
      roundingType=HalfHeightRadius
      width=0
      color {colorUseRule=UseColor_Specified, specifiedColor=<rgba 0 0 0 0>}
      cornerRadii=<0 0 0 0>
     }
     tintColor {colorUseRule=UseColor_Specified, specifiedColor=<rgba 1 1 1 1>}
     uVRegion {bIsValid=false, max=<0 0>, min=<0 0>}
    }}
   slot {class=HorizontalBoxSlot, properties {horizontalAlignment=HAlign_Center, verticalAlignment=VAlign_Center, padding=<ltrb 3 0 3 0>}}
  }
  Text_MoneyReward {
   widgetClassPath=&CommonTextBlock
   bIsVariable=true
   parent=HorizontalBox_1
   widget=Text_MoneyReward
   properties {
    text=500
    font {
     bForceMonospaced=false
     bMaterialIsStencil=false
     fontMaterial=null
     fontObject=&Burbank
     letterSpacing=0
     monospacedWidth=1
     size=24
     skewAmount=0
     typefaceFontName=Burbank-PlayerPower
     outlineSettings {
      bApplyOutlineToDropShadows=false
      bMiteredCorners=false
      bSeparateFillAlpha=false
      outlineMaterial=null
      outlineSize=2
      outlineColor=<rgba 0.0052 0.0052 0.0056 0.3>
     }
    }
    renderTransform {angle=0, scale=<1 1>, shear=<-8 0>, translation=<0 0>}
    shadowOffset=<0 0>
   }
   slot {class=HorizontalBoxSlot}
  }
  Reward_XP {
   widgetClassPath=&SizeBox
   bIsVariable=true
   parent=HorizontalBox_Reward
   widget=Reward_XP
   slot {class=HorizontalBoxSlot, properties {padding=<ltrb 15 0 0 0>}}
  }
  Border_RewardColor {
   widgetClassPath=&Border
   bIsVariable=true
   parent=Reward_XP
   widget=Border_RewardColor
   properties {
    background {
     bIsDynamicallyLoaded=false
     drawAs=NoDrawType
     imageType=NoImage
     mirroring=NoMirror
     resourceName=None
     resourceObject=null
     tiling=NoTile
     imageSize=<32 32>
     margin=<ltrb 0 0 0 0>
     outlineSettings {
      bUseBrushTransparency=false
      roundingType=HalfHeightRadius
      width=0
      color {colorUseRule=UseColor_Specified, specifiedColor=<rgba 0 0 0 0>}
      cornerRadii=<0 0 0 0>
     }
     tintColor {colorUseRule=UseColor_Specified, specifiedColor=<rgba 1 1 1 1>}
     uVRegion {bIsValid=false, max=<0 0>, min=<0 0>}
    }
    padding=<ltrb 0 0 0 0>
   }
   slot {class=SizeBoxSlot}
  }
  HorizontalBox_3 {
   widgetClassPath=&HorizontalBox
   parent=Border_RewardColor
   widget=HorizontalBox_3
   slot {class=BorderSlot, properties {padding=<ltrb 0 0 0 0>}}
  }
  Image_56 {
   widgetClassPath=&Image
   bIsVariable=true
   parent=HorizontalBox_3
   widget=Image_56
   properties {brush {
     bIsDynamicallyLoaded=false
     drawAs=Image
     imageType=NoImage
     mirroring=NoMirror
     resourceName=None
     resourceObject=/FortniteCoreUI/Athena/HUD/Quests/Art/M_UI_TabCategory_XP_White
     tiling=NoTile
     imageSize=<48 48>
     margin=<ltrb 0 0 0 0>
     outlineSettings {
      bUseBrushTransparency=false
      roundingType=HalfHeightRadius
      width=0
      color {colorUseRule=UseColor_Specified, specifiedColor=<rgba 0 0 0 0>}
      cornerRadii=<0 0 0 0>
     }
     tintColor {colorUseRule=UseColor_Specified, specifiedColor=<rgba 1 1 1 1>}
     uVRegion {bIsValid=false, max=<0 0>, min=<0 0>}
    }}
   slot {class=HorizontalBoxSlot, properties {horizontalAlignment=HAlign_Center, verticalAlignment=VAlign_Center, padding=<ltrb 0 -32 0 -32>}}
  }
  Text_XpReward {
   widgetClassPath=&CommonTextBlock
   bIsVariable=true
   parent=HorizontalBox_3
   widget=Text_XpReward
   properties {
    text=50k
    font {
     bForceMonospaced=false
     bMaterialIsStencil=false
     fontMaterial=null
     fontObject=&Burbank
     letterSpacing=0
     monospacedWidth=1
     size=24
     skewAmount=0
     typefaceFontName=Burbank-PlayerPower
     outlineSettings {
      bApplyOutlineToDropShadows=false
      bMiteredCorners=false
      bSeparateFillAlpha=false
      outlineMaterial=null
      outlineSize=2
      outlineColor=<rgba 0.0052 0.0052 0.0056 0.3>
     }
    }
    renderTransform {angle=0, scale=<1 1>, shear=<-8 0>, translation=<0 0>}
    shadowOffset=<0 0>
   }
   slot {class=HorizontalBoxSlot}
  }
  Reward_Custom {
   widgetClassPath=&SizeBox
   bIsVariable=true
   parent=HorizontalBox_Reward
   widget=Reward_Custom
   slot {class=HorizontalBoxSlot, properties {padding=<ltrb 15 0 0 0>}}
  }
  Border_RewardColor_1 {
   widgetClassPath=&Border
   bIsVariable=true
   parent=Reward_Custom
   widget=Border_RewardColor_1
   properties {
    background {
     bIsDynamicallyLoaded=false
     drawAs=NoDrawType
     imageType=NoImage
     mirroring=NoMirror
     resourceName=None
     resourceObject=null
     tiling=NoTile
     imageSize=<32 32>
     margin=<ltrb 0 0 0 0>
     outlineSettings {
      bUseBrushTransparency=false
      roundingType=HalfHeightRadius
      width=0
      color {colorUseRule=UseColor_Specified, specifiedColor=<rgba 0 0 0 0>}
      cornerRadii=<0 0 0 0>
     }
     tintColor {colorUseRule=UseColor_Specified, specifiedColor=<rgba 1 1 1 1>}
     uVRegion {bIsValid=false, max=<0 0>, min=<0 0>}
    }
    padding=<ltrb 0 0 0 0>
   }
   slot {class=SizeBoxSlot}
  }
  HorizontalBox {
   widgetClassPath=&HorizontalBox
   parent=Border_RewardColor_1
   widget=HorizontalBox
   slot {class=BorderSlot, properties {padding=<ltrb 0 0 0 0>}}
  }
  Image_TokenReward {
   widgetClassPath=&Image
   bIsVariable=true
   parent=HorizontalBox
   widget=Image_TokenReward
   properties {brush {
     bIsDynamicallyLoaded=false
     drawAs=Image
     imageType=FullColor
     mirroring=NoMirror
     resourceName=None
     resourceObject=/Game/Athena/UI/InGame/Creative/PropertyEditor/Icon_Picker/Textures/128x128/T_UI_IconLibrary_car_128
     tiling=NoTile
     imageSize=<48 48>
     margin=<ltrb 0 0 0 0>
     outlineSettings {
      bUseBrushTransparency=false
      roundingType=HalfHeightRadius
      width=0
      color {colorUseRule=UseColor_Specified, specifiedColor=<rgba 0 0 0 0>}
      cornerRadii=<0 0 0 0>
     }
     tintColor {colorUseRule=UseColor_Specified, specifiedColor=<rgba 1 1 1 1>}
     uVRegion {bIsValid=false, max=<0 0>, min=<0 0>}
    }}
   slot {class=HorizontalBoxSlot, properties {horizontalAlignment=HAlign_Center, verticalAlignment=VAlign_Center, padding=<ltrb 0 -32 5 -32>}}
  }
  Text_CustomReward {
   widgetClassPath=&CommonTextBlock
   bIsVariable=true
   parent=HorizontalBox
   widget=Text_CustomReward
   properties {
    text=1
    font {
     bForceMonospaced=false
     bMaterialIsStencil=false
     fontMaterial=null
     fontObject=&Burbank
     letterSpacing=0
     monospacedWidth=1
     size=24
     skewAmount=0
     typefaceFontName=Burbank-PlayerPower
     outlineSettings {
      bApplyOutlineToDropShadows=false
      bMiteredCorners=false
      bSeparateFillAlpha=false
      outlineMaterial=null
      outlineSize=2
      outlineColor=<rgba 0.0052 0.0052 0.0056 0.3>
     }
    }
    renderTransform {angle=0, scale=<1 1>, shear=<-8 0>, translation=<0 0>}
    shadowOffset=<0 0>
   }
   slot {class=HorizontalBoxSlot}
  }
  HB_Locked {
   widgetClassPath=&HorizontalBox
   parent=Switcher_State
   widget=HB_Locked
   slot {class=WidgetSwitcherSlot, properties {horizontalAlignment=HAlign_Center, verticalAlignment=VAlign_Center, padding=<ltrb 20 2 20 2>}}
  }
  LockStuff {
   widgetClassPath=&Overlay
   parent=HB_Locked
   widget=LockStuff
   properties {renderTransformPivot=<0 1>}
   slot {class=HorizontalBoxSlot, properties {horizontalAlignment=HAlign_Center, verticalAlignment=VAlign_Center}}
  }
  Image_305 {
   widgetClassPath=&Image
   bIsVariable=true
   parent=LockStuff
   widget=Image_305
   properties {
    brush {
     bIsDynamicallyLoaded=false
     drawAs=Image
     imageType=NoImage
     mirroring=NoMirror
     resourceName=None
     resourceObject=null
     tiling=NoTile
     imageSize=<28 28>
     margin=<ltrb 0 0 0 0>
     outlineSettings {
      bUseBrushTransparency=false
      roundingType=HalfHeightRadius
      width=0
      color {colorUseRule=UseColor_Specified, specifiedColor=<rgba 0 0 0 0>}
      cornerRadii=<0 0 0 0>
     }
     tintColor {colorUseRule=UseColor_Specified, specifiedColor=<rgba 1 1 1 1>}
     uVRegion {bIsValid=false, max=<0 0>, min=<0 0>}
    }
    colorAndOpacity=<rgba 1 0 0.2988 1>
    renderTransform {angle=0, scale=<1 1>, shear=<-9 0>, translation=<0 0>}
   }
   slot {class=OverlaySlot, properties {horizontalAlignment=HAlign_Fill, verticalAlignment=VAlign_Fill}}
  }
  ImgLockIcon {
   widgetClassPath=&Image
   parent=LockStuff
   widget=ImgLockIcon
   properties {
    visibility=SelfHitTestInvisible
    brush {
     bIsDynamicallyLoaded=false
     drawAs=Image
     imageType=NoImage
     mirroring=NoMirror
     resourceName=None
     resourceObject=/Game/UI/Foundation/Textures/Icons/Locks/T-Icon-Lock-64
     tiling=NoTile
     imageSize=<22 22>
     margin=<ltrb 0 0 0 0>
     outlineSettings {
      bUseBrushTransparency=false
      roundingType=HalfHeightRadius
      width=0
      color {colorUseRule=UseColor_Specified, specifiedColor=<rgba 0 0 0 0>}
      cornerRadii=<0 0 0 0>
     }
     tintColor {colorUseRule=UseColor_Specified, specifiedColor=<rgba 1 1 1 1>}
     uVRegion {bIsValid=false, max=<0 0>, min=<0 0>}
    }
    renderTransform {angle=0, scale=<1 1>, shear=<0 0>, translation=<0 -1>}
   }
   slot {class=OverlaySlot, properties {horizontalAlignment=HAlign_Center, verticalAlignment=VAlign_Center}}
  }
  Text_UnlockLevel {
   widgetClassPath=&CommonTextBlock
   bIsVariable=true
   parent=HB_Locked
   widget=Text_UnlockLevel
   properties {
    lineHeightPercentage=1.28
    mobileFontSizeMultiplier=1.2
    text=Requires Rep 8
    colorAndOpacity {colorUseRule=UseColor_Specified, specifiedColor=<rgba 1 0 0.3005 1>}
    font {
     bForceMonospaced=false
     bMaterialIsStencil=false
     fontMaterial=null
     fontObject=&BurbankSmall
     letterSpacing=0
     monospacedWidth=1
     size=15.75
     skewAmount=0.14
     typefaceFontName=Bold
     outlineSettings {
      bApplyOutlineToDropShadows=false
      bMiteredCorners=false
      bSeparateFillAlpha=false
      outlineMaterial=null
      outlineSize=0
      outlineColor=<rgba 0 0 0 1>
     }
    }
    renderTransform {angle=0, scale=<1 1>, shear=<0 0>, translation=<0 5>}
    shadowOffset=<0 0>
    strikeBrush {
     bIsDynamicallyLoaded=false
     drawAs=NoDrawType
     imageType=NoImage
     mirroring=NoMirror
     resourceName=None
     resourceObject=null
     tiling=NoTile
     imageSize=<32 32>
     margin=<ltrb 0 0 0 0>
     outlineSettings {
      bUseBrushTransparency=false
      roundingType=HalfHeightRadius
      width=0
      color {colorUseRule=UseColor_Specified, specifiedColor=<rgba 0 0 0 0>}
      cornerRadii=<0 0 0 0>
     }
     tintColor {colorUseRule=UseColor_Specified, specifiedColor=<rgba 1 1 1 1>}
     uVRegion {bIsValid=false, max=<0 0>, min=<0 0>}
    }
   }
   slot {class=HorizontalBoxSlot, properties {verticalAlignment=VAlign_Center, padding=<ltrb 4 0 0 0>}}
  }
 ]
}
