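# FortHireBrief is a conversation UI widget for hiring an NPC, showing the NPC's name, specialization icon and a Recruit transaction strip.
# It subclasses the native FortHireBrief, handling parent events to pass selection state to the strip and set the icon texture.
# At the max NPC count it shows an error line, choosing between a cannot-hire message and a warning that the first follower will be dismissed.
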WidgetBlueprint /FortniteConversationUI/Conversation/FortHireBrief
graphs [EventGraph {nodes [
   n1 = Event SetSelectedVisuals!() -> (OutputDelegate, then, IsSelected:bool) {is=title K2Node_Event}
   > n2 = Cast To FortTransactionStripNew!(execute, Object=Get TransactionStrip_Main() -> (TransactionStrip_Main:Fort Transaction Strip Object Reference)) -> (then, CastFailed, AsFort Transaction Strip New:Fort Transaction Strip New Object Reference) {is=title K2Node_DynamicCast}
   > SetSelectedVisuals!(execute, self:Fort Transaction Strip New Object Reference=@n2, IsSelected:bool=@n1.IsSelected) -> (then)
   n3 = Event HandleNPCDismissal!() -> (OutputDelegate, then, bAtMaxNPCNum:bool, bWillDismissNPCAtHire:bool) {is=title K2Node_Event}
   > SetVisibility(
    self=Get BriefErrorMessage() -> (BriefErrorMessage:Brief Error Message New Object Reference)
    InVisibility=Select(Option 0=Collapsed, Option 1=HitTestInvisible, Index=@n3.bAtMaxNPCNum)
   )
   > SetErrorMessage!(
    execute
    self:Brief Error Message New Object Reference=Get BriefErrorMessage()
    ErrorText:Text=Select(
     Option 0:Text="LOCTEXT(\"Cannot hire: Dismiss current follower in Follower Command Menu\")"
     Option 1:Text="LOCTEXT(\"Warning: Will dismiss first follower\")"
     Index=@n3.bWillDismissNPCAtHire
    ) -> (ReturnValue:Text)
   ) -> (then)
   n4 = Event SetSpecializationIcon!() -> (OutputDelegate, then, ResourceObject:Object Reference) {is=title K2Node_Event}
   > SetBrushResourceObject!(
    execute
    self:Image Object Reference=Get Image_SpecialistIcon() -> (Image_SpecialistIcon:Image Object Reference)
    ResourceObject:Object Reference=@n4.ResourceObject
   ) -> (then)
  ]}]
variables []
widgets {
 info {inheritedWidgetCount=2, parentClass=/Script/FortniteConversationUI.FortHireBrief, rootWidgetClass=&Overlay, widgetCount=17}
 widgets [
  Overlay_1 {widgetClassPath=&Overlay, widget=Overlay_1, properties {bOverride_Cursor=true}}
  VB_MainContent {widgetClassPath=&VerticalBox, parent=Overlay_1, widget=VB_MainContent, slot {class=OverlaySlot}}
  Overlay_4 {widgetClassPath=&Overlay, parent=VB_MainContent, widget=Overlay_4, slot {class=VerticalBoxSlot}}
  Border_168 {
   widgetClassPath=&Border
   parent=Overlay_4
   widget=Border_168
   properties {brushColor=<rgba 0 0 0 0.446>, padding=<ltrb 18 10 18 10>}
   slot {class=OverlaySlot, properties {horizontalAlignment=HAlign_Fill}}
  }
  Overlay_0 {widgetClassPath=&Overlay, parent=Border_168, widget=Overlay_0, slot {class=BorderSlot, properties {padding=<ltrb 18 10 18 10>}}}
  Text_Title {
   widgetClassPath=&CommonTextBlock
   parent=Overlay_0
   widget=Text_Title
   properties {
    text=Hire
    textTransformPolicy=ToUpper
    font {
     bForceMonospaced=false
     bMaterialIsStencil=false
     fontMaterial=null
     fontObject=&Burbank
     letterSpacing=0
     monospacedWidth=1
     size=18
     skewAmount=0
     typefaceFontName=BurbankRegular-Black
     outlineSettings {
      bApplyOutlineToDropShadows=false
      bMiteredCorners=false
      bSeparateFillAlpha=false
      outlineMaterial=null
      outlineSize=2
      outlineColor=<rgba 0.0052 0.0052 0.0056 0.6>
     }
    }
    renderTransform {angle=0, scale=<1 1>, shear=<-8 0>, translation=<0 -30>}
    shadowOffset=<0 0>
   }
   slot {class=OverlaySlot, properties {verticalAlignment=VAlign_Fill}}
  }
  Spacer_140 {widgetClassPath=&Spacer, parent=Overlay_0, widget=Spacer_140, properties {size=<1 48>}, slot {class=OverlaySlot}}
  SizeBox_0 {
   widgetClassPath=&SizeBox
   parent=Overlay_0
   widget=SizeBox_0
   properties {bOverride_MaxDesiredWidth=true, bOverride_MinDesiredWidth=true, maxDesiredWidth=512, minDesiredWidth=320}
   slot {class=OverlaySlot, properties {verticalAlignment=VAlign_Center}}
  }
  HorizontalBox_0 {
   widgetClassPath=&HorizontalBox
   parent=SizeBox_0
   widget=HorizontalBox_0
   properties {renderTransform {angle=0, scale=<1 1>, shear=<-8 0>, translation=<0 0>}}
   slot {class=SizeBoxSlot}
  }
  Text_Objective {
   widgetClassPath=&CommonRichTextBlock
   bInherited=true
   bIsVariable=true
   parent=HorizontalBox_0
   widget=Text_Objective
   properties {
    bOverrideDefaultStyle=true
    text="Hire <Name>{Name}</>"
    textStyleSet=&/Conversation_RichTextStyleSet
    wrapTextAt=512
    defaultTextStyleOverride {
     overflowPolicy=Clip
     transformPolicy=None
     colorAndOpacity {colorUseRule=UseColor_Specified, specifiedColor=<rgba 1 1 1 1>}
     font {
      bForceMonospaced=false
      bMaterialIsStencil=false
      fontMaterial=null
      fontObject=&BurbankSmall
      letterSpacing=0
      monospacedWidth=1
      size=20
      skewAmount=0
      typefaceFontName=Bold
      outlineSettings {
       bApplyOutlineToDropShadows=false
       bMiteredCorners=false
       bSeparateFillAlpha=false
       outlineMaterial=null
       outlineSize=0
       outlineColor=<rgba 0 0 0 1>
      }
     }
     highlightColor {colorUseRule=UseColor_Specified, specifiedColor=<rgba 0 0 0 1>}
     highlightShape {
      bIsDynamicallyLoaded=false
      drawAs=Image
      imageType=NoImage
      mirroring=NoMirror
      resourceName=None
      resourceObject=null
      tiling=NoTile
      imageSize=<32 32>
      margin=<ltrb 0 0 0 0>
      outlineSettings {
       bUseBrushTransparency=false
       roundingType=HalfHeightRadius
       width=0
       color {colorUseRule=UseColor_Specified, specifiedColor=<rgba 0 0 0 0>}
       cornerRadii=<0 0 0 0>
      }
      tintColor {colorUseRule=UseColor_Specified, specifiedColor=<rgba 1 1 1 1>}
      uVRegion {bIsValid=false, max=<0 0>, min=<0 0>}
     }
     selectedBackgroundColor {colorUseRule=UseColor_Foreground, specifiedColor=<rgba 1 0 1 1>}
     shadowColorAndOpacity=<rgba 0 0 0 1>
     shadowOffset=<0 0>
     strikeBrush {
      bIsDynamicallyLoaded=false
      drawAs=Image
      imageType=NoImage
      mirroring=NoMirror
      resourceName=None
      resourceObject=null
      tiling=NoTile
      imageSize=<32 32>
      margin=<ltrb 0 0 0 0>
      outlineSettings {
       bUseBrushTransparency=false
       roundingType=HalfHeightRadius
       width=0
       color {colorUseRule=UseColor_Specified, specifiedColor=<rgba 0 0 0 0>}
       cornerRadii=<0 0 0 0>
      }
      tintColor {colorUseRule=UseColor_Specified, specifiedColor=<rgba 1 1 1 1>}
      uVRegion {bIsValid=false, max=<0 0>, min=<0 0>}
     }
     underlineBrush {
      bIsDynamicallyLoaded=false
      drawAs=Image
      imageType=NoImage
      mirroring=NoMirror
      resourceName=None
      resourceObject=null
      tiling=NoTile
      imageSize=<32 32>
      margin=<ltrb 0 0 0 0>
      outlineSettings {
       bUseBrushTransparency=false
       roundingType=HalfHeightRadius
       width=0
       color {colorUseRule=UseColor_Specified, specifiedColor=<rgba 0 0 0 0>}
       cornerRadii=<0 0 0 0>
      }
      tintColor {colorUseRule=UseColor_Specified, specifiedColor=<rgba 1 1 1 1>}
      uVRegion {bIsValid=false, max=<0 0>, min=<0 0>}
     }
    }
   }
   slot {class=HorizontalBoxSlot, properties {verticalAlignment=VAlign_Center}}
  }
  Overlay_77 {
   widgetClassPath=&Overlay
   parent=HorizontalBox_0
   widget=Overlay_77
   slot {class=HorizontalBoxSlot, properties {verticalAlignment=VAlign_Center, padding=<ltrb 8 0 0 0>}}
  }
  Image_IconBG {
   widgetClassPath=&Image
   bIsVariable=true
   parent=Overlay_77
   widget=Image_IconBG
   properties {colorAndOpacity=<rgba 0.0078 0.0353 0.1647 1>}
   slot {class=OverlaySlot}
  }
  SizeBox_1 {
   widgetClassPath=&SizeBox
   parent=Overlay_77
   widget=SizeBox_1
   properties {bOverride_HeightOverride=true, bOverride_WidthOverride=true, heightOverride=32, widthOverride=32}
   slot {class=OverlaySlot, properties {horizontalAlignment=HAlign_Center, verticalAlignment=VAlign_Center}}
  }
  Image_SpecialistIcon {
   widgetClassPath=&Image
   bIsVariable=true
   parent=SizeBox_1
   widget=Image_SpecialistIcon
   properties {brush {
     bIsDynamicallyLoaded=false
     drawAs=Image
     imageType=FullColor
     mirroring=NoMirror
     resourceName=None
     resourceObject=/FortniteCoreUI/HUD/MiniMap/Services/Textures/SpecialistIcons/T_UI_Specialist_Scout
     tiling=NoTile
     imageSize=<32 32>
     margin=<ltrb 0 0 0 0>
     outlineSettings {
      bUseBrushTransparency=false
      roundingType=HalfHeightRadius
      width=0
      color {colorUseRule=UseColor_Specified, specifiedColor=<rgba 0 0 0 0>}
      cornerRadii=<0 0 0 0>
     }
     tintColor {colorUseRule=UseColor_Specified, specifiedColor=<rgba 1 1 1 1>}
     uVRegion {bIsValid=false, max=<0 0>, min=<0 0>}
    }}
   slot {class=SizeBoxSlot}
  }
  Image_RarityRibbon {
   widgetClassPath=&Image
   bIsVariable=true
   parent=Overlay_4
   widget=Image_RarityRibbon
   properties {
    visibility=Collapsed
    brush {
     bIsDynamicallyLoaded=false
     drawAs=Image
     imageType=NoImage
     mirroring=NoMirror
     resourceName=None
     resourceObject=null
     tiling=NoTile
     imageSize=<6 32>
     margin=<ltrb 0 0 0 0>
     outlineSettings {
      bUseBrushTransparency=false
      roundingType=HalfHeightRadius
      width=0
      color {colorUseRule=UseColor_Specified, specifiedColor=<rgba 0 0 0 0>}
      cornerRadii=<0 0 0 0>
     }
     tintColor {colorUseRule=UseColor_Specified, specifiedColor=<rgba 1 1 1 1>}
     uVRegion {bIsValid=false, max=<0 0>, min=<0 0>}
    }
   }
   slot {class=OverlaySlot, properties {verticalAlignment=VAlign_Fill}}
  }
  TransactionStrip_Main {
   widgetClassPath=&/FortTransactionStripNew
   bInherited=true
   bIsVariable=true
   parent=VB_MainContent
   widget=TransactionStrip_Main
   properties {transactionString=Recruit}
   slot {class=VerticalBoxSlot}
  }
  BriefErrorMessage {
   widgetClassPath=&/BriefErrorMessageNew
   bIsVariable=true
   parent=VB_MainContent
   widget=BriefErrorMessage
   properties {fontSize=16}
   slot {class=VerticalBoxSlot}
  }
 ]
}
