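# Fund brief entry widget for one item in a multi-item funding conversation, showing its icon, name, funding progress and transaction strip.
# PreConstruct pushes crop and shadow flips into materials; OnProgressUpdated feeds the blueprint-style percent text and progress bar.
# Selection plays icon and left/right content animations forward or reversed, swapping the blueprint look for detailed funding UI.
# A player-controller donation delegate triggers donation animations while selected.
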WidgetBlueprint /FortniteConversationUI/Conversation/FortMultiItemFundBriefEntry
animations [
 Anim_Selected {
  binding_count=1
  length_seconds=0.2
  object=&self:Anim_Selected
  sequence {
   bindings [LazyImage_Icon {
     id=#1
     possessed_class=FortLazyImage
     tracks [{
       type=MovieSceneWidgetMaterialTrack
       sections [{
         type=MovieSceneParameterSection
         channels [
          Animate {type=MovieSceneScriptingFloatChannel, keys [[0,0],[12,1]]}
          AnimateRotation {type=MovieSceneScriptingFloatChannel, keys [[0,0],[12.001,1]]}
          AnimateCrop {type=MovieSceneScriptingFloatChannel, keys [[0,0],[12,1]]}
          AnimateVisual {type=MovieSceneScriptingFloatChannel, keys [[0,0],[6,1]]}
          Spotlight_Intensity {type=MovieSceneScriptingFloatChannel, keys [[0,0],[12,11]]}
          AnimateDepthFade {type=MovieSceneScriptingFloatChannel, keys [[6,0],[12,0.75]]}
          AnimateShadow {type=MovieSceneScriptingFloatChannel, keys [[0,1],[12,0]]}
         ]
        }]
      }]
    }]
   display_rate {numerator=60}
  }
 }
 Anim_Selected_Content {
  binding_count=8
  length_seconds=0.3167
  object=&self:Anim_Selected_Content
  sequence {
   bindings [
    BlueprintPercentText {
     id=#2
     possessed_class=HorizontalBox
     tracks [{
       type=MovieSceneFloatTrack
       display_name=Render Opacity
       sections [{type=MovieSceneFloatSection, channels [None {type=MovieSceneScriptingFloatChannel, keys [[0,1],[3,0]]}]}]
      }]
    }
    Text_TitleBar {
     id=#3
     possessed_class=CommonTextBlock
     tracks [
      {
       type=MovieSceneWidgetMaterialTrack
       display_name=Font.Outline Settings.OutlineMaterial
       sections [{type=MovieSceneParameterSection, channels [AnimateOff {type=MovieSceneScriptingFloatChannel, keys [[0,0],[3,1]]}]}]
      }
      {
       type=MovieScene2DTransformTrack
       sections [{
         type=MovieScene2DTransformSection
         channels [
          Translation.X {type=MovieSceneScriptingFloatChannel, const [0,1]}
          Translation.Y {type=MovieSceneScriptingFloatChannel, keys [[3,48],[5,0]]}
          Angle {type=MovieSceneScriptingFloatChannel, const [0,1]}
          Scale.X {type=MovieSceneScriptingFloatChannel, keys [[3,0.6],[5,1]]}
          Scale.Y {type=MovieSceneScriptingFloatChannel, keys [[3,0.6],[5,1]]}
          Shear.X {type=MovieSceneScriptingFloatChannel, const [-12,1]}
          Shear.Y {type=MovieSceneScriptingFloatChannel, const [0,1]}
         ]
        }]
      }
      {
       type=MovieSceneWidgetMaterialTrack
       display_name=Font.FontMaterial
       sections [{type=MovieSceneParameterSection, channels [AnimatetoFill {type=MovieSceneScriptingFloatChannel, keys [[0,0],[3,0],[4,1]]}]}]
      }
      {
       type=MovieSceneFloatTrack
       display_name=Render Opacity
       sections [{type=MovieSceneFloatSection, channels [None {type=MovieSceneScriptingFloatChannel, keys [[0,1],[3,0],[4,0],[15,1]]}]}]
      }
     ]
    }
    FundingPercent_Overlay {
     id=#4
     possessed_class=Overlay
     tracks [
      {
       type=MovieSceneFloatTrack
       display_name=Render Opacity
       sections [{type=MovieSceneFloatSection, channels [None {type=MovieSceneScriptingFloatChannel, keys [[10,0],[13,1]]}]}]
      }
      {
       type=MovieScene2DTransformTrack
       sections [{
         type=MovieScene2DTransformSection
         channels [
          Scale.X {type=MovieSceneScriptingFloatChannel, keys [[10,1.25],[13,0.95],[19,1]]}
          Scale.Y {type=MovieSceneScriptingFloatChannel, keys [[10,1.25],[13,0.95],[19,1]]}
         ]
        }]
      }
     ]
    }
    BottomStrip {
     id=#5
     possessed_class=Image
     tracks [
      {
       type=MovieSceneFloatTrack
       display_name=Render Opacity
       sections [{type=MovieSceneFloatSection, channels [None {type=MovieSceneScriptingFloatChannel, keys [[0,0],[12,1]]}]}]
      }
      {
       type=MovieScene2DTransformTrack
       sections [{
         type=MovieScene2DTransformSection
         channels [
          Scale.X {type=MovieSceneScriptingFloatChannel, keys [[0,0],[12,1]]}
          Scale.Y {type=MovieSceneScriptingFloatChannel, const [1,1]}
         ]
        }]
      }
     ]
    }
    Ruler {
     id=#6
     possessed_class=Image
     tracks [{
       type=MovieSceneWidgetMaterialTrack
       sections [{type=MovieSceneParameterSection, channels [AnimateOff {type=MovieSceneScriptingFloatChannel, keys [[0,0],[3,1]]}]}]
      }]
    }
    BlueprintProgressBarSizeBox {
     id=#7
     possessed_class=SizeBox
     tracks [{
       type=MovieSceneFloatTrack
       display_name=Render Opacity
       sections [{type=MovieSceneFloatSection, channels [None {type=MovieSceneScriptingFloatChannel, keys [[0,1],[3,0]]}]}]
      }]
    }
    FundStrip {
     id=#8
     possessed_class=SizeBox
     tracks [
      {
       type=MovieScene2DTransformTrack
       sections [{
         type=MovieScene2DTransformSection
         channels [
          Angle {type=MovieSceneScriptingFloatChannel, const [0,1]}
          Scale.X {type=MovieSceneScriptingFloatChannel, keys [[3,0.75],[9,1.05],[15.001,1]]}
          Scale.Y {type=MovieSceneScriptingFloatChannel, keys [[3,0.875],[9,1.05],[15.001,1]]}
          Shear.X {type=MovieSceneScriptingFloatChannel, const [0,1]}
          Shear.Y {type=MovieSceneScriptingFloatChannel, const [0,1]}
         ]
        }]
      }
      {
       type=MovieSceneFloatTrack
       display_name=Render Opacity
       sections [{type=MovieSceneFloatSection, channels [None {type=MovieSceneScriptingFloatChannel, keys [[3,0],[6,1]]}]}]
      }
     ]
    }
    ProgressInfo {
     id=#9
     possessed_class=VerticalBox
     tracks [
      {
       type=MovieScene2DTransformTrack
       sections [{
         type=MovieScene2DTransformSection
         channels [
          Translation.X {type=MovieSceneScriptingFloatChannel, const [0,1]}
          Translation.Y {type=MovieSceneScriptingFloatChannel, keys [[0,24],[12,0]]}
          Angle {type=MovieSceneScriptingFloatChannel, const [0,1]}
          Scale.X {type=MovieSceneScriptingFloatChannel, const [1,1]}
          Scale.Y {type=MovieSceneScriptingFloatChannel, const [1,1]}
          Shear.X {type=MovieSceneScriptingFloatChannel, const [0,1]}
          Shear.Y {type=MovieSceneScriptingFloatChannel, const [0,1]}
         ]
        }]
      }
      {
       type=MovieSceneFloatTrack
       display_name=Render Opacity
       sections [{type=MovieSceneFloatSection, end_frame=6, start_frame=0, channels [None {type=MovieSceneScriptingFloatChannel, keys [[0,0],[6,1]]}]}]
      }
     ]
    }
   ]
   display_rate {numerator=60}
  }
 }
 Anim_Donated {
  binding_count=3
  length_seconds=0.75
  object=&self:Anim_Donated
  sequence {
   bindings [
    Border_0 {
     id=#10
     possessed_class=Border
     tracks [{
       type=MovieSceneWidgetMaterialTrack
       sections [{
         type=MovieSceneParameterSection
         channels [
          BackingAlpha {type=MovieSceneScriptingFloatChannel, keys [[0,0.25],[6,0.75],[30,0.333]]}
          OffsetShine {type=MovieSceneScriptingFloatChannel, keys [[0,0],[45,1.475]]}
         ]
        }]
      }]
    }
    LazyImage_Icon {
     id=#11
     possessed_class=FortLazyImage
     tracks [{
       type=MovieSceneWidgetMaterialTrack
       sections [{
         type=MovieSceneParameterSection
         channels [
          Spotlight_Intensity {type=MovieSceneScriptingFloatChannel, keys [[0,16],[6,18],[30,11]]}
          Gradient_Size {type=MovieSceneScriptingFloatChannel, keys [[0,66],[6,85],[24,66]]}
         ]
        }]
      }]
    }
    ProgressInfo {
     id=#12
     possessed_class=VerticalBox
     tracks [{
       type=MovieScene2DTransformTrack
       sections [{
         type=MovieScene2DTransformSection
         channels [
          Scale.X {type=MovieSceneScriptingFloatChannel, keys [[0.001,1],[3,0.95],[8,1.15],[20,1]]}
          Scale.Y {type=MovieSceneScriptingFloatChannel, keys [[0.001,1],[3,0.95],[8,1.15],[20,1]]}
         ]
        }]
      }]
    }
   ]
   display_rate {numerator=60}
  }
 }
 Anim_Selected_Content_Right {
  binding_count=8
  length_seconds=0.5
  object=&self:Anim_Selected_Content_Right
  sequence {
   bindings [
    BlueprintPercentText {
     id=#2
     possessed_class=HorizontalBox
     tracks [{
       type=MovieSceneFloatTrack
       display_name=Render Opacity
       sections [{type=MovieSceneFloatSection, channels [None {type=MovieSceneScriptingFloatChannel, keys [[0,1],[3,0]]}]}]
      }]
    }
    Text_TitleBar {
     id=#3
     possessed_class=CommonTextBlock
     tracks [
      {
       type=MovieSceneWidgetMaterialTrack
       display_name=Font.Outline Settings.OutlineMaterial
       sections [{type=MovieSceneParameterSection, channels [AnimateOff {type=MovieSceneScriptingFloatChannel, keys [[0,0],[2,1]]}]}]
      }
      {
       type=MovieScene2DTransformTrack
       sections [{
         type=MovieScene2DTransformSection
         channels [
          Translation.X {type=MovieSceneScriptingFloatChannel, const [0,1]}
          Translation.Y {type=MovieSceneScriptingFloatChannel, keys [[2,48],[3,0]]}
          Angle {type=MovieSceneScriptingFloatChannel, const [0,1]}
          Scale.X {type=MovieSceneScriptingFloatChannel, keys [[2,0.75],[3,1]]}
          Scale.Y {type=MovieSceneScriptingFloatChannel, keys [[2,0.75],[3,1]]}
          Shear.X {type=MovieSceneScriptingFloatChannel, const [-12,1]}
          Shear.Y {type=MovieSceneScriptingFloatChannel, const [0,1]}
         ]
        }]
      }
      {
       type=MovieSceneWidgetMaterialTrack
       display_name=Font.FontMaterial
       sections [{type=MovieSceneParameterSection, channels [AnimatetoFill {type=MovieSceneScriptingFloatChannel, keys [[0,0],[3,1]]}]}]
      }
      {
       type=MovieSceneFloatTrack
       display_name=Render Opacity
       sections [{type=MovieSceneFloatSection, channels [None {type=MovieSceneScriptingFloatChannel, keys [[0,1],[2,0],[3,0],[30,1]]}]}]
      }
     ]
    }
    FundingPercent_Overlay {
     id=#4
     possessed_class=Overlay
     tracks [
      {
       type=MovieSceneFloatTrack
       display_name=Render Opacity
       sections [{type=MovieSceneFloatSection, channels [None {type=MovieSceneScriptingFloatChannel, keys [[10,0],[13,1]]}]}]
      }
      {
       type=MovieScene2DTransformTrack
       sections [{
         type=MovieScene2DTransformSection
         channels [
          Scale.X {type=MovieSceneScriptingFloatChannel, keys [[10,1.25],[13,0.95],[18,1]]}
          Scale.Y {type=MovieSceneScriptingFloatChannel, keys [[10,1.25],[13,0.95],[18,1]]}
         ]
        }]
      }
     ]
    }
    BottomStrip {
     id=#5
     possessed_class=Image
     tracks [
      {
       type=MovieSceneFloatTrack
       display_name=Render Opacity
       sections [{type=MovieSceneFloatSection, channels [None {type=MovieSceneScriptingFloatChannel, keys [[0,0],[18.001,1]]}]}]
      }
      {
       type=MovieScene2DTransformTrack
       sections [{
         type=MovieScene2DTransformSection
         channels [
          Scale.X {type=MovieSceneScriptingFloatChannel, keys [[0,0],[12,1]]}
          Scale.Y {type=MovieSceneScriptingFloatChannel, const [1,1]}
         ]
        }]
      }
     ]
    }
    Ruler {
     id=#6
     possessed_class=Image
     tracks [{
       type=MovieSceneWidgetMaterialTrack
       sections [{type=MovieSceneParameterSection, channels [AnimateOff {type=MovieSceneScriptingFloatChannel, keys [[0,0],[3,1]]}]}]
      }]
    }
    BlueprintProgressBarSizeBox {
     id=#7
     possessed_class=SizeBox
     tracks [{
       type=MovieSceneFloatTrack
       display_name=Render Opacity
       sections [{type=MovieSceneFloatSection, channels [None {type=MovieSceneScriptingFloatChannel, keys [[0,1],[3,0]]}]}]
      }]
    }
    FundStrip {
     id=#8
     possessed_class=SizeBox
     tracks [
      {
       type=MovieSceneFloatTrack
       display_name=Render Opacity
       sections [{type=MovieSceneFloatSection, channels [None {type=MovieSceneScriptingFloatChannel, keys [[2,0],[8,1]]}]}]
      }
      {
       type=MovieScene2DTransformTrack
       sections [{
         type=MovieScene2DTransformSection
         channels [
          Translation.X {type=MovieSceneScriptingFloatChannel, keys [[2,-319.9999],[17,0]]}
          Scale.X {type=MovieSceneScriptingFloatChannel, keys [[2,0.9],[17,1]]}
          Scale.Y {type=MovieSceneScriptingFloatChannel, keys [[2,0.9],[17,1]]}
         ]
        }]
      }
     ]
    }
    ProgressInfo {
     id=#9
     possessed_class=VerticalBox
     tracks [
      {
       type=MovieScene2DTransformTrack
       sections [{
         type=MovieScene2DTransformSection
         channels [
          Translation.X {type=MovieSceneScriptingFloatChannel, const [0,1]}
          Translation.Y {type=MovieSceneScriptingFloatChannel, keys [[0,24],[12,0]]}
          Angle {type=MovieSceneScriptingFloatChannel, const [0,1]}
          Scale.X {type=MovieSceneScriptingFloatChannel, const [1,1]}
          Scale.Y {type=MovieSceneScriptingFloatChannel, const [1,1]}
          Shear.X {type=MovieSceneScriptingFloatChannel, const [0,1]}
          Shear.Y {type=MovieSceneScriptingFloatChannel, const [0,1]}
         ]
        }]
      }
      {
       type=MovieSceneFloatTrack
       display_name=Render Opacity
       sections [{type=MovieSceneFloatSection, end_frame=6, start_frame=0, channels [None {type=MovieSceneScriptingFloatChannel, keys [[0,0],[6,1]]}]}]
      }
     ]
    }
   ]
   display_rate {numerator=60}
  }
 }
 Anim_Selected_Content_Left {
  binding_count=8
  length_seconds=0.5
  object=&self:Anim_Selected_Content_Left
  sequence {
   bindings [
    BlueprintPercentText {
     id=#2
     possessed_class=HorizontalBox
     tracks [{
       type=MovieSceneFloatTrack
       display_name=Render Opacity
       sections [{type=MovieSceneFloatSection, channels [None {type=MovieSceneScriptingFloatChannel, keys [[0,1],[3,0]]}]}]
      }]
    }
    Text_TitleBar {
     id=#3
     possessed_class=CommonTextBlock
     tracks [
      {
       type=MovieSceneWidgetMaterialTrack
       display_name=Font.Outline Settings.OutlineMaterial
       sections [{type=MovieSceneParameterSection, channels [AnimateOff {type=MovieSceneScriptingFloatChannel, keys [[0,0],[2,1]]}]}]
      }
      {
       type=MovieScene2DTransformTrack
       sections [{
         type=MovieScene2DTransformSection
         channels [
          Translation.X {type=MovieSceneScriptingFloatChannel, const [0,1]}
          Translation.Y {type=MovieSceneScriptingFloatChannel, keys [[2,48],[3,0]]}
          Angle {type=MovieSceneScriptingFloatChannel, const [0,1]}
          Scale.X {type=MovieSceneScriptingFloatChannel, keys [[2,0.75],[3,1]]}
          Scale.Y {type=MovieSceneScriptingFloatChannel, keys [[2,0.75],[3,1]]}
          Shear.X {type=MovieSceneScriptingFloatChannel, const [-12,1]}
          Shear.Y {type=MovieSceneScriptingFloatChannel, const [0,1]}
         ]
        }]
      }
      {
       type=MovieSceneWidgetMaterialTrack
       display_name=Font.FontMaterial
       sections [{type=MovieSceneParameterSection, channels [AnimatetoFill {type=MovieSceneScriptingFloatChannel, keys [[0,0],[3,1]]}]}]
      }
      {
       type=MovieSceneFloatTrack
       display_name=Render Opacity
       sections [{type=MovieSceneFloatSection, channels [None {type=MovieSceneScriptingFloatChannel, keys [[0,1],[2,0],[3,0],[30,1]]}]}]
      }
     ]
    }
    FundingPercent_Overlay {
     id=#4
     possessed_class=Overlay
     tracks [
      {
       type=MovieSceneFloatTrack
       display_name=Render Opacity
       sections [{type=MovieSceneFloatSection, channels [None {type=MovieSceneScriptingFloatChannel, keys [[10,0],[13,1]]}]}]
      }
      {
       type=MovieScene2DTransformTrack
       sections [{
         type=MovieScene2DTransformSection
         channels [
          Scale.X {type=MovieSceneScriptingFloatChannel, keys [[10,1.25],[13,0.95],[18,1]]}
          Scale.Y {type=MovieSceneScriptingFloatChannel, keys [[10,1.25],[13,0.95],[18,1]]}
         ]
        }]
      }
     ]
    }
    BottomStrip {
     id=#5
     possessed_class=Image
     tracks [
      {
       type=MovieSceneFloatTrack
       display_name=Render Opacity
       sections [{type=MovieSceneFloatSection, channels [None {type=MovieSceneScriptingFloatChannel, keys [[0,0],[18,1]]}]}]
      }
      {
       type=MovieScene2DTransformTrack
       sections [{
         type=MovieScene2DTransformSection
         channels [
          Scale.X {type=MovieSceneScriptingFloatChannel, keys [[0,0],[12,1]]}
          Scale.Y {type=MovieSceneScriptingFloatChannel, const [1,1]}
         ]
        }]
      }
     ]
    }
    Ruler {
     id=#6
     possessed_class=Image
     tracks [{
       type=MovieSceneWidgetMaterialTrack
       sections [{type=MovieSceneParameterSection, channels [AnimateOff {type=MovieSceneScriptingFloatChannel, keys [[0,0],[3,1]]}]}]
      }]
    }
    BlueprintProgressBarSizeBox {
     id=#7
     possessed_class=SizeBox
     tracks [{
       type=MovieSceneFloatTrack
       display_name=Render Opacity
       sections [{type=MovieSceneFloatSection, channels [None {type=MovieSceneScriptingFloatChannel, keys [[0,1],[3,0]]}]}]
      }]
    }
    FundStrip {
     id=#8
     possessed_class=SizeBox
     tracks [
      {
       type=MovieSceneFloatTrack
       display_name=Render Opacity
       sections [{type=MovieSceneFloatSection, channels [None {type=MovieSceneScriptingFloatChannel, keys [[2,0],[8,1]]}]}]
      }
      {
       type=MovieScene2DTransformTrack
       sections [{
         type=MovieScene2DTransformSection
         channels [
          Translation.X {type=MovieSceneScriptingFloatChannel, keys [[2,319.9999],[17,0]]}
          Scale.X {type=MovieSceneScriptingFloatChannel, keys [[2,0.9],[17,1]]}
          Scale.Y {type=MovieSceneScriptingFloatChannel, keys [[2,0.9],[17,1]]}
         ]
        }]
      }
     ]
    }
    ProgressInfo {
     id=#9
     possessed_class=VerticalBox
     tracks [
      {
       type=MovieScene2DTransformTrack
       sections [{
         type=MovieScene2DTransformSection
         channels [
          Translation.X {type=MovieSceneScriptingFloatChannel, const [0,1]}
          Translation.Y {type=MovieSceneScriptingFloatChannel, keys [[0,24],[12,0]]}
          Angle {type=MovieSceneScriptingFloatChannel, const [0,1]}
          Scale.X {type=MovieSceneScriptingFloatChannel, const [1,1]}
          Scale.Y {type=MovieSceneScriptingFloatChannel, const [1,1]}
          Shear.X {type=MovieSceneScriptingFloatChannel, const [0,1]}
          Shear.Y {type=MovieSceneScriptingFloatChannel, const [0,1]}
         ]
        }]
      }
      {
       type=MovieSceneFloatTrack
       display_name=Render Opacity
       sections [{type=MovieSceneFloatSection, end_frame=6, start_frame=0, channels [None {type=MovieSceneScriptingFloatChannel, keys [[0,0],[6,1]]}]}]
      }
     ]
    }
   ]
   display_rate {numerator=60}
  }
 }
]
graphs [
 EventGraph {nodes [
   Event Construct()
   > SetMaterialTextureParamName!(
    execute
    self:Common Lazy Image Object Reference=Get LazyImage_Icon() -> (LazyImage_Icon:Fort Lazy Image Object Reference)
    TextureParamName:Name=IconImage
   ) -> (then)
   > n1 = Cast To Athena_PlayerController!(execute, Object=GetOwningPlayer()) -> (then, CastFailed, AsAthena Player Controller:Athena Player Controller Object Reference) {is=title K2Node_DynamicCast}
   > Bind Event to On Client Donation Success Delegate(self:Fort Player Controller Athena Object Reference=@n1, Delegate=Create Event()) {is=title K2Node_AddDelegate}
   n2 = Event OnProgressUpdated!() -> (OutputDelegate, then, CurrentFundingFraction:float) {is=title K2Node_Event}
   > "SetText (Text)"(
    self=Get Text_ProgressPercent_2() -> (Text_ProgressPercent_2:Common Text Object Reference)
    InText=As Percent!(
     Value:float=@n2.CurrentFundingFraction
     RoundingMode:ERoundingMode Enum=ToPositiveInfinity
     bAlwaysSign:bool=false
     bUseGrouping:bool=true
     MinimumIntegralDigits:int=1
     MaximumIntegralDigits:int=324
     MinimumFractionalDigits:int=0
     MaximumFractionalDigits:int=2
    ) -> (ReturnValue:Text)
   )
   > n3 = GetDynamicMaterial(self=Get BlueprintProgressBar() -> (BlueprintProgressBar:Image Object Reference))
   > SetScalarParameterValue(self=@n3, ParameterName=Progress, Value=@n2.CurrentFundingFraction)
   Event PreConstruct()
   > n4 = GetDynamicMaterial(self=Get LazyImage_Icon())
   > SetScalarParameterValue(self=@n4, ParameterName=FlipCrop, Value=Get CropDirection() -> (CropDirection:double))
   > SetScalarParameterValue(self=@n4, ParameterName=OffsetCrop, Value=Get CropOffset() -> (CropOffset:double))
   > SetScalarParameterValue(self=@n4, ParameterName=FlipShadow, Value=Get FlipShadow() -> (FlipShadow:double))
   > n5 = GetDynamicMaterial(self=Get BottomStrip() -> (BottomStrip:Image Object Reference))
   > SetScalarParameterValue(self=@n5, ParameterName=FlipCrop, Value=Get CropDirection())
   > SetScalarParameterValue(self=@n5, ParameterName=OffsetCrop, Value=Get CropOffset())
   > Set IsLeftSideSelection(IsLeftSideSelection:bool=Get IsLeftSideSelection() -> (IsLeftSideSelection:bool)) -> (Output_Get:bool)
   Event Destruct()
   > n6 = Cast To Athena_PlayerController(Object=GetOwningPlayer())
   > Unbind Event from On Client Donation Success Delegate(self:Fort Player Controller Athena Object Reference=@n6, Delegate=Create Event()) {is=title K2Node_RemoveDelegate}
  ]}
 OnDonateSuccess {nodes [
   OnDonateSuccess() {is=title K2Node_FunctionEntry}
   > Branch(Condition=Get bSelected() -> (bSelected:bool))
   > PlayAnimation(InAnimation=Get Anim_Donated() -> (Anim_Donated:Widget Animation Object Reference))
   > n1 = Cast To FortTransactionStripNew!(execute, Object=Get TransactionStrip_Main() -> (TransactionStrip_Main:Fort Transaction Strip Object Reference)) -> (then, CastFailed, AsFort Transaction Strip New:Fort Transaction Strip New Object Reference) {is=title K2Node_DynamicCast}
   > DonationAnim(self:Fort Transaction Strip New Object Reference=@n1)
  ]}
 ResetAnimStates {nodes [
   ResetAnimStates() {is=title K2Node_FunctionEntry}
   n1 = Get Anim_Selected() -> (Anim_Selected:Widget Animation Object Reference)
   > PlayAnimation(InAnimation=@n1, StartAtTime=GetEndTime(self=@n1), PlayMode=Reverse)
   n2 = Get Anim_Selected_Content() -> (Anim_Selected_Content:Widget Animation Object Reference)
   > PlayAnimation(InAnimation=@n2, StartAtTime=GetEndTime(self=@n2), PlayMode=Reverse)
  ]}
 SetIntroState {nodes [
   n1 = SetIntroState!() -> (then, Selected:bool=~) {is=title K2Node_FunctionEntry}
   > Branch(Condition="Not Equal (Boolean)"!(A:bool=@n1, B:bool=Get bSelected()) -> (ReturnValue:bool) {is=title K2Node_CallFunction})
   > n2 = Set bSelected(bSelected:bool=false) -> (Output_Get:bool)
   > n3 = Branch(Condition=@n2)
   n4 = Get Anim_Selected()
   > PlayAnimation(InAnimation=@n4, StartAtTime=GetEndTime(self=@n4), PlaybackSpeed=10)
   n5 = Get Anim_Selected()
   PlayAnimation(execute=@n3.else, InAnimation=@n5, StartAtTime=GetEndTime(self=@n5), PlayMode=Reverse, PlaybackSpeed=10)
  ]}
 SetSelectedVisuals {nodes [
   n1 = SetSelectedVisuals!() -> (then, InSelected:bool=~) {is=title K2Node_FunctionEntry}
   > n2 = Set bSelected(bSelected=@n1)
   > n3 = Branch(Condition=@n2)
   > PlayAnimationForward(InAnimation=Get Anim_Selected())
   > PlayAnimationForward(InAnimation=Select(
     Option 0:Widget Animation Object Reference=Get Anim_Selected_Content_Right() -> (Anim_Selected_Content_Right:Widget Animation Object Reference)
     Option 1:Widget Animation Object Reference=Get Anim_Selected_Content_Left() -> (Anim_Selected_Content_Left:Widget Animation Object Reference)
     Index=Get IsLeftSideSelection()
    ) -> (ReturnValue:Widget Animation Object Reference))
   n4 = Select(Option 0=Get Anim_Selected_Content_Right(), Option 1=Get Anim_Selected_Content_Left(), Index=Get IsLeftSideSelection())
   PlayAnimation(execute=@n3.else, InAnimation=@n4, StartAtTime=GetEndTime(self=@n4), PlayMode=Reverse)
   > PlayAnimationReverse(InAnimation=Get Anim_Selected(), PlaybackSpeed=2)
  ]}
]
variables [
 bSelected {type=boolean, default=true}
 CropDirection {type=number}
 CropOffset {type=number}
 DebugItem {type=FortItem, default=null}
 FlipShadow {type=number}
 IsLeftSideSelection {type=boolean}
 OffsetShadow {type=number}
]
widgets {
 info {inheritedWidgetCount=6, parentClass=/Script/FortniteConversationUI.FortItemFundBrief, rootWidgetClass=&ScaleBox, widgetCount=32}
 widgets [
  ItemScaleBox {widgetClassPath=&ScaleBox, bIsVariable=true, widget=ItemScaleBox, properties {stretch=UserSpecified}}
  DisplayOverlay {
   widgetClassPath=&Overlay
   bIsVariable=true
   parent=ItemScaleBox
   widget=DisplayOverlay
   slot {class=ScaleBoxSlot, properties {horizontalAlignment=HAlign_Fill, verticalAlignment=VAlign_Fill}}
  }
  LazyImage_Icon {
   widgetClassPath=&FortLazyImage
   bInherited=true
   bIsVariable=true
   parent=DisplayOverlay
   widget=LazyImage_Icon
   properties {
    clipping=ClipToBoundsWithoutIntersecting
    brush {
     bIsDynamicallyLoaded=false
     drawAs=Image
     imageType=NoImage
     mirroring=NoMirror
     resourceName=None
     resourceObject=/FortniteConversationUI/Conversation/Materials/M_UI_DonationFundTile_Default
     tiling=NoTile
     imageSize=<512 512>
     margin=<ltrb 0 0 0 0>
     outlineSettings {
      bUseBrushTransparency=false
      roundingType=HalfHeightRadius
      width=0
      color {colorUseRule=UseColor_Specified, specifiedColor=<rgba 0 0 0 0>}
      cornerRadii=<0 0 0 0>
     }
     tintColor {colorUseRule=UseColor_Specified, specifiedColor=<rgba 1 1 1 1>}
     uVRegion {bIsValid=false, max=<0 0>, min=<0 0>}
    }
   }
   slot {class=OverlaySlot, properties {horizontalAlignment=HAlign_Fill, verticalAlignment=VAlign_Fill}}
  }
  Ruler {
   widgetClassPath=&Image
   bIsVariable=true
   parent=DisplayOverlay
   widget=Ruler
   properties {
    renderOpacity=0
    visibility=SelfHitTestInvisible
    brush {
     bIsDynamicallyLoaded=false
     drawAs=Image
     imageType=NoImage
     mirroring=NoMirror
     resourceName=None
     resourceObject=/FortniteConversationUI/Conversation/Materials/M_UI_FundBlueprint_Ruler
     tiling=NoTile
     imageSize=<256 16>
     margin=<ltrb 0 0 0 0>
     outlineSettings {
      bUseBrushTransparency=false
      roundingType=HalfHeightRadius
      width=0
      color {colorUseRule=UseColor_Specified, specifiedColor=<rgba 0 0 0 0>}
      cornerRadii=<0 0 0 0>
     }
     tintColor {colorUseRule=UseColor_Specified, specifiedColor=<rgba 1 1 1 1>}
     uVRegion {bIsValid=false, max=<0 0>, min=<0 0>}
    }
    renderTransform {angle=0, scale=<0.9 0.9>, shear=<0 0>, translation=<0 0>}
    renderTransformPivot=<0 1>
   }
   slot {class=OverlaySlot, properties {verticalAlignment=VAlign_Bottom, padding=<ltrb 64 0 0 80>}}
  }
  BlueprintProgressBarSizeBox {
   widgetClassPath=&SizeBox
   parent=DisplayOverlay
   widget=BlueprintProgressBarSizeBox
   properties {bOverride_WidthOverride=true, widthOverride=512, renderTransform {angle=0, scale=<1 1>, shear=<-8 0>, translation=<0 0>}}
   slot {class=OverlaySlot, properties {horizontalAlignment=HAlign_Center, verticalAlignment=VAlign_Bottom, padding=<ltrb 0 0 0 24>}}
  }
  BlueprintProgressBar {
   widgetClassPath=&Image
   bIsVariable=true
   parent=BlueprintProgressBarSizeBox
   widget=BlueprintProgressBar
   properties {
    visibility=SelfHitTestInvisible
    brush {
     bIsDynamicallyLoaded=false
     drawAs=Box
     imageType=NoImage
     mirroring=NoMirror
     resourceName=None
     resourceObject=/FortniteConversationUI/Conversation/Materials/M_UI_FundBlueprint_ProgressBar
     tiling=NoTile
     imageSize=<16 16>
     margin=<ltrb 0.5 0.5 0.5 0.5>
     outlineSettings {
      bUseBrushTransparency=false
      roundingType=HalfHeightRadius
      width=0
      color {colorUseRule=UseColor_Specified, specifiedColor=<rgba 0 0 0 0>}
      cornerRadii=<0 0 0 0>
     }
     tintColor {colorUseRule=UseColor_Specified, specifiedColor=<rgba 1 1 1 1>}
     uVRegion {bIsValid=false, max=<0 0>, min=<0 0>}
    }
   }
   slot {class=SizeBoxSlot}
  }
  Content {
   widgetClassPath=&VerticalBox
   parent=DisplayOverlay
   widget=Content
   slot {class=OverlaySlot, properties {horizontalAlignment=HAlign_Fill, verticalAlignment=VAlign_Fill}}
  }
  Overlay_0 {
   widgetClassPath=&Overlay
   parent=Content
   widget=Overlay_0
   slot {class=VerticalBoxSlot, properties {size {sizeRule=Fill, value=1}}}
  }
  Overlay_2 {
   widgetClassPath=&Overlay
   parent=Overlay_0
   widget=Overlay_2
   properties {clipping=ClipToBoundsWithoutIntersecting}
   slot {class=OverlaySlot, properties {horizontalAlignment=HAlign_Center, verticalAlignment=VAlign_Bottom, padding=<ltrb 0 0 0 72>}}
  }
  NamePercent {
   widgetClassPath=&SizeBox
   parent=Overlay_2
   widget=NamePercent
   properties {bOverride_WidthOverride=true, clipping=ClipToBoundsWithoutIntersecting, widthOverride=704}
   slot {class=OverlaySlot, properties {horizontalAlignment=HAlign_Fill, verticalAlignment=VAlign_Fill}}
  }
  ScaleBox_1 {
   widgetClassPath=&ScaleBox
   parent=NamePercent
   widget=ScaleBox_1
   properties {clipping=ClipToBoundsWithoutIntersecting, stretch=ScaleToFitX, stretchDirection=DownOnly}
   slot {class=SizeBoxSlot, properties {horizontalAlignment=HAlign_Center}}
  }
  Overlay_3 {
   widgetClassPath=&Overlay
   parent=ScaleBox_1
   widget=Overlay_3
   properties {clipping=ClipToBoundsWithoutIntersecting}
   slot {class=ScaleBoxSlot}
  }
  Text_TitleBar {
   widgetClassPath=&CommonTextBlock
   bInherited=true
   bIsVariable=true
   parent=Overlay_3
   widget=Text_TitleBar
   properties {
    clipping=ClipToBoundsWithoutIntersecting
    justification=Center
    lineHeightPercentage=0.9
    text=Weapon Name
    textTransformPolicy=ToUpper
    font {
     bForceMonospaced=false
     bMaterialIsStencil=false
     fontMaterial=/FortniteConversationUI/Conversation/Materials/M_Fund_UI_BlueprintText_Default
     fontObject=&Burbank
     letterSpacing=0
     monospacedWidth=1
     size=40
     skewAmount=0
     typefaceFontName=BurbankRegular-Black
     outlineSettings {
      bApplyOutlineToDropShadows=true
      bMiteredCorners=false
      bSeparateFillAlpha=true
      outlineMaterial=/FortniteConversationUI/Conversation/Materials/M_UI_FundBlueprint_Texture_Font
      outlineSize=2
      outlineColor=<rgba 1 1 1 1>
     }
    }
    margin=<ltrb 0 5 0 10>
    renderTransform {angle=0, scale=<0.75 0.75>, shear=<-12 0>, translation=<0 40>}
    shadowOffset=<0 0>
   }
   slot {class=OverlaySlot, properties {horizontalAlignment=HAlign_Fill, verticalAlignment=VAlign_Fill}}
  }
  BlueprintPercentText {
   widgetClassPath=&HorizontalBox
   parent=Overlay_2
   widget=BlueprintPercentText
   properties {renderTransform {angle=-2, scale=<1 1>, shear=<0 0>, translation=<0 -40>}}
   slot {class=OverlaySlot, properties {horizontalAlignment=HAlign_Right, verticalAlignment=VAlign_Bottom, padding=<ltrb 0 0 40 0>}}
  }
  Text_ProgressPercent_2 {
   widgetClassPath=&CommonTextBlock
   bIsVariable=true
   parent=BlueprintPercentText
   widget=Text_ProgressPercent_2
   properties {
    text=99%
    font {
     bForceMonospaced=false
     bMaterialIsStencil=false
     fontMaterial=&M_UI_FundBlueprint_Texture_Font
     fontObject=&Burbank
     letterSpacing=0
     monospacedWidth=1
     size=24
     skewAmount=0
     typefaceFontName=BurbankRegular-Black
     outlineSettings {
      bApplyOutlineToDropShadows=true
      bMiteredCorners=false
      bSeparateFillAlpha=true
      outlineMaterial=null
      outlineSize=0
      outlineColor=<rgba 0 0 0 0>
     }
    }
    shadowOffset=<0 0>
   }
   slot {class=HorizontalBoxSlot, properties {horizontalAlignment=HAlign_Right, verticalAlignment=VAlign_Bottom}}
  }
  Text_ProgressPercent_4 {
   widgetClassPath=&CommonTextBlock
   bIsVariable=true
   parent=BlueprintPercentText
   widget=Text_ProgressPercent_4
   properties {
    text="FUNDED!"
    font {
     bForceMonospaced=false
     bMaterialIsStencil=false
     fontMaterial=&M_UI_FundBlueprint_Texture_Font
     fontObject=&BurbankSmall
     letterSpacing=0
     monospacedWidth=1
     size=18
     skewAmount=0
     typefaceFontName=Black
     outlineSettings {
      bApplyOutlineToDropShadows=true
      bMiteredCorners=false
      bSeparateFillAlpha=true
      outlineMaterial=null
      outlineSize=0
      outlineColor=<rgba 0 0 0 0>
     }
    }
    shadowOffset=<0 0>
   }
   slot {class=HorizontalBoxSlot, properties {verticalAlignment=VAlign_Center, padding=<ltrb 5 0 0 0>}}
  }
  FundingPercent_Overlay {
   widgetClassPath=&Overlay
   bIsVariable=true
   parent=Overlay_2
   widget=FundingPercent_Overlay
   properties {clipping=ClipToBoundsWithoutIntersecting, renderTransform {angle=0, scale=<1 1>, shear=<0 0>, translation=<8 -36>}}
   slot {class=OverlaySlot, properties {horizontalAlignment=HAlign_Right, padding=<ltrb 0 -8 40 0>}}
  }
  Image_93 {
   widgetClassPath=&Image
   bIsVariable=true
   parent=FundingPercent_Overlay
   widget=Image_93
   properties {
    visibility=SelfHitTestInvisible
    brush {
     bIsDynamicallyLoaded=false
     drawAs=Image
     imageType=NoImage
     mirroring=NoMirror
     resourceName=None
     resourceObject=/FortUILibrary/Foundation/Art_ToolKit/AngledBoxes/M_UI_FundCenter_PercentBacking_Default
     tiling=NoTile
     imageSize=<32 32>
     margin=<ltrb 0 0 0 0>
     outlineSettings {
      bUseBrushTransparency=false
      roundingType=HalfHeightRadius
      width=0
      color {colorUseRule=UseColor_Specified, specifiedColor=<rgba 0 0 0 0>}
      cornerRadii=<0 0 0 0>
     }
     tintColor {colorUseRule=UseColor_Specified, specifiedColor=<rgba 1 1 1 1>}
     uVRegion {bIsValid=false, max=<0 0>, min=<0 0>}
    }
   }
   slot {class=OverlaySlot, properties {horizontalAlignment=HAlign_Fill, verticalAlignment=VAlign_Fill}}
  }
  HorizontalBox_1 {
   widgetClassPath=&HorizontalBox
   parent=FundingPercent_Overlay
   widget=HorizontalBox_1
   properties {renderTransform {angle=0, scale=<1 1>, shear=<-16 0>, translation=<0 0>}}
   slot {class=OverlaySlot, properties {horizontalAlignment=HAlign_Center, verticalAlignment=VAlign_Center, padding=<ltrb 18 10 18 10>}}
  }
  Text_ProgressPercent {
   widgetClassPath=&CommonTextBlock
   bInherited=true
   bIsVariable=true
   parent=HorizontalBox_1
   widget=Text_ProgressPercent
   properties {
    text=99%
    font {
     bForceMonospaced=false
     bMaterialIsStencil=false
     fontMaterial=null
     fontObject=&Burbank
     letterSpacing=0
     monospacedWidth=1
     size=16
     skewAmount=0
     typefaceFontName=BurbankRegular-Black
     outlineSettings {
      bApplyOutlineToDropShadows=true
      bMiteredCorners=false
      bSeparateFillAlpha=true
      outlineMaterial=null
      outlineSize=0
      outlineColor=<rgba 0 0 0 0>
     }
    }
    shadowOffset=<0 0>
   }
   slot {class=HorizontalBoxSlot}
  }
  Text_ProgressPercent_1 {
   widgetClassPath=&CommonTextBlock
   bIsVariable=true
   parent=HorizontalBox_1
   widget=Text_ProgressPercent_1
   properties {
    text="FUNDED!"
    font {
     bForceMonospaced=false
     bMaterialIsStencil=false
     fontMaterial=null
     fontObject=&BurbankSmall
     letterSpacing=0
     monospacedWidth=1
     size=16
     skewAmount=0
     typefaceFontName=Black
     outlineSettings {
      bApplyOutlineToDropShadows=true
      bMiteredCorners=false
      bSeparateFillAlpha=true
      outlineMaterial=null
      outlineSize=0
      outlineColor=<rgba 0 0 0 0>
     }
    }
    shadowOffset=<0 0>
   }
   slot {class=HorizontalBoxSlot, properties {verticalAlignment=VAlign_Center, padding=<ltrb 6 0 0 0>}}
  }
  ProgressInfo {
   widgetClassPath=&VerticalBox
   parent=Overlay_0
   widget=ProgressInfo
   properties {renderTransform {angle=0, scale=<1 1>, shear=<-4 0>, translation=<0 0>}}
   slot {class=OverlaySlot, properties {horizontalAlignment=HAlign_Center, verticalAlignment=VAlign_Bottom, padding=<ltrb 0 0 0 44>}}
  }
  Border_0 {
   widgetClassPath=&Border
   parent=ProgressInfo
   widget=Border_0
   properties {
    background {
     bIsDynamicallyLoaded=false
     drawAs=Box
     imageType=NoImage
     mirroring=NoMirror
     resourceName=None
     resourceObject=/FortniteConversationUI/Conversation/Materials/M_UI_MultiFundProgress_BarBacking_Default
     tiling=NoTile
     imageSize=<32 32>
     margin=<ltrb 0.5 0 0.5 0>
     outlineSettings {
      bUseBrushTransparency=false
      roundingType=HalfHeightRadius
      width=0
      color {colorUseRule=UseColor_Specified, specifiedColor=<rgba 0 0 0 0>}
      cornerRadii=<0 0 0 0>
     }
     tintColor {colorUseRule=UseColor_Specified, specifiedColor=<rgba 1 1 1 1>}
     uVRegion {bIsValid=false, max=<0 0>, min=<0 0>}
    }
    padding=<ltrb 10 10 12 10>
   }
   slot {class=VerticalBoxSlot, properties {horizontalAlignment=HAlign_Center, verticalAlignment=VAlign_Bottom}}
  }
  ProgressBarSize {
   widgetClassPath=&SizeBox
   parent=Border_0
   widget=ProgressBarSize
   properties {bOverride_HeightOverride=true, bOverride_WidthOverride=true, heightOverride=16, widthOverride=512}
   slot {class=BorderSlot, properties {padding=<ltrb 10 10 12 10>}}
  }
  Progress_Funding {
   widgetClassPath=/Game/Athena/HUD/SimpleMaterialProgressBar
   bInherited=true
   bIsVariable=true
   parent=ProgressBarSize
   widget=Progress_Funding
   properties {
    bOverrideDefaultColorA=true
    bOverrideDefaultColorB=true
    bOverrideDefaultColorBackground=true
    bOverrideDefaultFillEdgeSoftness=true
    bOverrideDefaultOutlineScale=true
    bOverrideDefaultSegments=true
    bUseStroke=false
    designTime_Progress=0.75
    fillEdgeSoftness=0.0005
    segmentEdge=0.0002
    segments=1
    cachedColorA=<rgba 0 1 1 1>
    cachedColorB=<rgba 1 1 1 1>
    cachedColorBackground=<rgba 0 0.3125 0.75 1>
    renderTransform {angle=0, scale=<1 1>, shear=<-8 0>, translation=<0 0>}
   }
   slot {class=SizeBoxSlot}
  }
  HBox_ProgressText {
   widgetClassPath=&HorizontalBox
   bIsVariable=true
   parent=ProgressInfo
   widget=HBox_ProgressText
   properties {visibility=Collapsed}
   slot {class=VerticalBoxSlot, properties {padding=<ltrb -5 -20 0 0>}}
  }
  Image_CurrencyIcon {
   widgetClassPath=&Image
   bIsVariable=true
   parent=HBox_ProgressText
   widget=Image_CurrencyIcon
   properties {brush {
     bIsDynamicallyLoaded=false
     drawAs=Image
     imageType=NoImage
     mirroring=NoMirror
     resourceName=None
     resourceObject=/Game/UI/Foundation/Textures/Icons/Athena/T-T-Icon-BR-GoldBars-UI-Icon
     tiling=NoTile
     imageSize=<42 42>
     margin=<ltrb 0 0 0 0>
     outlineSettings {
      bUseBrushTransparency=false
      roundingType=HalfHeightRadius
      width=0
      color {colorUseRule=UseColor_Specified, specifiedColor=<rgba 0 0 0 0>}
      cornerRadii=<0 0 0 0>
     }
     tintColor {colorUseRule=UseColor_Specified, specifiedColor=<rgba 1 1 1 1>}
     uVRegion {bIsValid=false, max=<0 0>, min=<0 0>}
    }}
   slot {class=HorizontalBoxSlot, properties {horizontalAlignment=HAlign_Center, verticalAlignment=VAlign_Center, padding=<ltrb 3 0 3 0>}}
  }
  Text_ProgressCount {
   widgetClassPath=&CommonTextBlock
   bInherited=true
   bIsVariable=true
   parent=HBox_ProgressText
   widget=Text_ProgressCount
   properties {
    text="500,000 / 1,000,000"
    font {
     bForceMonospaced=false
     bMaterialIsStencil=false
     fontMaterial=null
     fontObject=&Burbank
     letterSpacing=0
     monospacedWidth=1
     size=16
     skewAmount=0
     typefaceFontName=BurbankRegular-Black
     outlineSettings {
      bApplyOutlineToDropShadows=true
      bMiteredCorners=false
      bSeparateFillAlpha=true
      outlineMaterial=null
      outlineSize=2
      outlineColor=<rgba 0 0 0 1>
     }
    }
    renderTransform {angle=0, scale=<1 1>, shear=<-8 0>, translation=<0 0>}
    shadowOffset=<0 0>
   }
   slot {class=HorizontalBoxSlot, properties {verticalAlignment=VAlign_Center}}
  }
  Button_BriefEntry {
   widgetClassPath=/Game/Athena/UI/Matchmaking/SimpleCommonButton
   bIsVariable=true
   parent=Overlay_0
   widget=Button_BriefEntry
   properties {
    onButtonBaseClicked=(&self:WidgetTree.Button_BriefEntry.HandleOnButtonClicked)
    onButtonBaseDoubleClicked=(&self:WidgetTree.Button_BriefEntry.HandleOnButtonDoubleClicked)
    onButtonBaseHovered=(&self:WidgetTree.Button_BriefEntry.HandleOnButtonHovered)
    onButtonBaseUnhovered=(&self:WidgetTree.Button_BriefEntry.HandleOnButtonUnhovered)
    onSelectedChangedBase=(&self:WidgetTree.Button_BriefEntry.HandleOnSelectedChanged)
    renderOpacity=0
    style=null
   }
   slot {class=OverlaySlot, properties {horizontalAlignment=HAlign_Fill, verticalAlignment=VAlign_Fill}}
  }
  BottomStrip {
   widgetClassPath=&Image
   bIsVariable=true
   parent=DisplayOverlay
   widget=BottomStrip
   properties {
    renderOpacity=0
    visibility=SelfHitTestInvisible
    brush {
     bIsDynamicallyLoaded=false
     drawAs=Image
     imageType=NoImage
     mirroring=NoMirror
     resourceName=None
     resourceObject=/FortniteConversationUI/Conversation/Materials/M_UI_DonationFundTile_Background_Inst
     tiling=NoTile
     imageSize=<32 16>
     margin=<ltrb 0 0 0 0>
     outlineSettings {
      bUseBrushTransparency=false
      roundingType=HalfHeightRadius
      width=0
      color {colorUseRule=UseColor_Specified, specifiedColor=<rgba 0 0 0 0>}
      cornerRadii=<0 0 0 0>
     }
     tintColor {colorUseRule=UseColor_Specified, specifiedColor=<rgba 1 1 1 1>}
     uVRegion {bIsValid=false, max=<0 0>, min=<0 0>}
    }
    renderTransformPivot=<0.5 1>
   }
   slot {class=OverlaySlot, properties {horizontalAlignment=HAlign_Fill, verticalAlignment=VAlign_Fill}}
  }
  FundStrip {
   widgetClassPath=&SizeBox
   parent=DisplayOverlay
   widget=FundStrip
   properties {
    bOverride_MinDesiredWidth=true
    minDesiredWidth=320
    widthOverride=320
    renderTransform {angle=0, scale=<1 1>, shear=<0 0>, translation=<0 36>}
   }
   slot {class=OverlaySlot, properties {horizontalAlignment=HAlign_Center, verticalAlignment=VAlign_Bottom}}
  }
  TransactionStrip_Main {
   widgetClassPath=&/FortTransactionStripNew
   bInherited=true
   bIsVariable=true
   parent=FundStrip
   widget=TransactionStrip_Main
   properties {bFundingVersion=true, transactionString=Buy}
   slot {class=SizeBoxSlot, properties {verticalAlignment=VAlign_Center}}
  }
 ]
}
