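# NPC dialogue widget for conversations, showing a portrait and speech bubble that also reacts to successful donations.
# OnItemNameUpdated extracts the quote-delimited item name, which keys the DonationResponses map.
# On donation success, it shows that item's line, falling back to random single- or multi-funded text, in place of the normal message.
# NewItemUpdate restores the normal message; both switches replay the Text_Update bubble animation.
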WidgetBlueprint /FortniteConversationUI/Conversation/NPCTextDisplay
animations [
 Anim_TextUpdate {
  binding_count=2
  length_seconds=0.5
  object=&self:Anim_TextUpdate
  sequence {bindings [
    Text_Message {
     id=#1
     possessed_class=CommonRichTextBlock
     tracks [{
       type=MovieSceneFloatTrack
       display_name=Render Opacity
       sections [{type=MovieSceneFloatSection, end_frame=10, start_frame=0, channels [None {type=MovieSceneScriptingFloatChannel, keys [[0,0],[10,1]]}]}]
      }]
    }
    NPCTextDisplay_C_19 {
     id=#2
     possessed_class=NPCTextDisplay_C
     tracks [{
       type=MovieScene2DTransformTrack
       sections [{
         type=MovieScene2DTransformSection
         end_frame=5
         start_frame=0
         channels [
          Translation.X {type=MovieSceneScriptingFloatChannel, keys [[0,25],[5,0]]}
          Translation.Y {type=MovieSceneScriptingFloatChannel, const [0,1]}
          Angle {type=MovieSceneScriptingFloatChannel, const [0,1]}
          Scale.X {type=MovieSceneScriptingFloatChannel, const [1,1]}
          Scale.Y {type=MovieSceneScriptingFloatChannel, const [1,1]}
          Shear.X {type=MovieSceneScriptingFloatChannel, const [0,1]}
          Shear.Y {type=MovieSceneScriptingFloatChannel, const [0,1]}
         ]
        }]
      }]
    }
   ]}
 }
 Text_Update {
  binding_count=2
  length_seconds=0.4834
  object=&self:Text_Update
  sequence {
   bindings [
    SizeBox_0 {
     id=#3
     possessed_class=SizeBox
     tracks [
      {
       type=MovieSceneFloatTrack
       display_name=Render Opacity
       sections [{type=MovieSceneFloatSection, channels [None {type=MovieSceneScriptingFloatChannel, keys [[0,0],[20,0],[22,1]]}]}]
      }
      {
       type=MovieScene2DTransformTrack
       sections [{
         type=MovieScene2DTransformSection
         channels [
          Translation.X {type=MovieSceneScriptingFloatChannel, keys [[20,-8],[26,0]]}
          Translation.Y {type=MovieSceneScriptingFloatChannel, keys [[20,8],[26,0]]}
          Angle {type=MovieSceneScriptingFloatChannel, const [0,1]}
          Scale.X {type=MovieSceneScriptingFloatChannel, keys [[20,0.9],[26,1]]}
          Scale.Y {type=MovieSceneScriptingFloatChannel, keys [[20,0.9],[26,1]]}
          Shear.X {type=MovieSceneScriptingFloatChannel, const [0,1]}
          Shear.Y {type=MovieSceneScriptingFloatChannel, const [0,1]}
         ]
        }]
      }
     ]
    }
    ImageContainer {
     id=#4
     possessed_class=Image
     tracks [{
       type=MovieScene2DTransformTrack
       sections [{
         type=MovieScene2DTransformSection
         channels [
          Scale.X {type=MovieSceneScriptingFloatChannel, keys [[0,1],[20,1],[25,1.0125],[29,1]]}
          Scale.Y {type=MovieSceneScriptingFloatChannel, keys [[0,1],[20,1],[25,1.0125],[29,1]]}
         ]
        }]
      }]
    }
   ]
   display_rate {numerator=60}
  }
 }
 Anim_DonationSuccess {
  binding_count=1
  length_seconds=3
  object=&self:Anim_DonationSuccess
  sequence {bindings [Funded_Text_Message {
     id=#5
     possessed_class=CommonRichTextBlock
     tracks [{
       type=MovieSceneFloatTrack
       display_name=Render Opacity
       sections [{type=MovieSceneFloatSection, end_frame=60, start_frame=0, channels [None {type=MovieSceneScriptingFloatChannel, const [1,2]}]}]
      }]
    }]}
 }
]
graphs [
 EventGraph {nodes [
   Event PreConstruct()
   > SetMaterialTextureParamName!(
    execute
    self:Common Lazy Image Object Reference=Get LazyImage_NPCImage() -> (LazyImage_NPCImage:Fort Lazy Image Object Reference)
    TextureParamName:Name=NPC-Portrait
   ) -> (then)
   Event Construct()
   > n1 = Cast To FortPlayerControllerAthena(Object=GetOwningPlayer())
   > Bind Event to On Client Donation Success Delegate(self:Fort Player Controller Athena Object Reference=@n1, Delegate=Create Event()) {is=title K2Node_AddDelegate}
   Event Destruct()
   > n2 = Cast To FortPlayerControllerAthena(Object=GetOwningPlayer())
   > Unbind Event from On Client Donation Success Delegate(self:Fort Player Controller Athena Object Reference=@n2, Delegate=Create Event()) {is=title K2Node_RemoveDelegate}
   > UnbindFromAnimationFinished!(
    execute
    self:User Widget Object Reference
    Animation:Widget Animation Object Reference=Get Anim_DonationSuccess() -> (Anim_DonationSuccess:Widget Animation Object Reference)
    Delegate:Delegate=Create Event()
   ) -> (then)
   n3 = Event OnItemNameUpdated!() -> (OutputDelegate, then, ItemName:String) {is=title K2Node_Event}
   > Set SelectedItemName(SelectedItemName:String=Split(SourceString=Split(SourceString=@n3.ItemName, InStr="\"", SearchDir=FromEnd), InStr="\"", SearchDir=FromEnd).RightS) -> (Output_Get:String)
   Event Tick()
   Event OnTextUpdated() {is=title K2Node_Event}
  ]}
 GetItemSpecificDonationResponse {nodes [
   GetItemSpecificDonationResponse() {is=title K2Node_FunctionEntry}
   n1 = Find(
    TargetMap:Map of Strings to Texts=Get DonationResponses() -> (DonationResponses:Map of Strings to Texts)
    Key=Get SelectedItemName() -> (SelectedItemName:String)
   ) -> (Value:Text)
   > Return Node!(execute, ResponseFound:bool=@n1.ReturnValue, ResponseText:Text=@n1)
   SetText(self=Get Funded_Text_Message() -> (Funded_Text_Message:Common Rich Text Block Object Reference))
  ]}
 GetRandomDonationText {nodes [
   GetRandomDonationText() {is=title K2Node_FunctionEntry}
   > n1 = Sequence()
   > Set TextArrayToUse(TextArrayToUse:Array of Texts=Select(
     Option 0:Array of Texts=Get SingleFundedText() -> (SingleFundedText:Array of Texts)
     Option 1:Array of Texts=Get MultiFundedText() -> (MultiFundedText:Array of Texts)
     Index=Get bIsMultiVersion() -> (bIsMultiVersion:bool)
    ) -> (ReturnValue:Array of Texts)) -> (Output_Get:Array of Texts)
   n2 = Length(TargetArray:Array of Texts=Get TextArrayToUse() -> (TextArrayToUse:Array of Texts))
   > n3 = Set ConfirmTextToShow(ConfirmTextToShow:Text="Get (a ref)"(Array:Array of Texts=Get TextArrayToUse(), Dimension 1=RandomInteger!(Max:int=@n2) -> (ReturnValue:int=0)) -> (Output:"Text (by ref)")) -> (Output_Get:Text)
   > Return Node!(execute, ResponseText=@n3)
   SetText(execute=@n1.then_1, self=Get Funded_Text_Message(), InText=Get ConfirmTextToShow() -> (ConfirmTextToShow:Text))
  ]}
 NewItemUpdate {nodes [
   NewItemUpdate() {is=title K2Node_FunctionEntry}
   > SetVisibility(self=Get Text_Message() -> (Text_Message:Common Rich Text Block Object Reference), InVisibility=HitTestInvisible)
   > SetVisibility(self=Get Funded_Text_Message(), InVisibility=Hidden)
   > PlayTextUpdate()
  ]}
 OnDonationSuccess {nodes [
   OnDonationSuccess() {is=title K2Node_FunctionEntry}
   > n1 = GetItemSpecificDonationResponse!(execute, self) -> (then, ResponseFound:bool, ResponseText:Text) {is=title K2Node_CallFunction}
   > n2 = Branch(Condition=@n1)
   > n3 = Set ResponseToShow(ResponseToShow:Text=@n1.ResponseText) -> (Output_Get:Text)
   SetText(execute=[@n3.then, @n5.then], self=Get Funded_Text_Message(), InText=Get ResponseToShow() -> (ResponseToShow:Text))
   > SetVisibility(self=Get Text_Message(), InVisibility=Hidden)
   > SetVisibility(self=Get Funded_Text_Message(), InVisibility=HitTestInvisible)
   > PlayTextUpdate()
   > StopAnimation(InAnimation=Get Anim_DonationSuccess())
   > PlayAnimation(InAnimation=Get Anim_DonationSuccess())
   > BindToAnimationFinished(Animation=Get Anim_DonationSuccess(), Delegate=Create Event())
   n4 = GetRandomDonationText!(execute=@n2.else, self) -> (then, ResponseText:Text) {is=title K2Node_CallFunction}
   > n5 = Set ResponseToShow(ResponseToShow=@n4)
   SetText(
    self=Get Funded_Text_Message()
    InText=Format Text!(Format="LOCTEXT(\"ItemNameString I have is: \\r\\n{0}\")", 0:String=Get SelectedItemName()) -> (Result)
   )
  ]}
 OnDonationSuccessAnimFinished {nodes [
   OnDonationSuccessAnimFinished() {is=title K2Node_FunctionEntry}
   > UnbindFromAnimationFinished(Animation=Get Anim_DonationSuccess(), Delegate=Create Event())
  ]}
 PlayTextUpdate {nodes [
   PlayTextUpdate() {is=title K2Node_FunctionEntry}
   n1 = Get Text_Update() -> (Text_Update:Widget Animation Object Reference)
   > StopAnimation(InAnimation=@n1)
   > PlayAnimation(InAnimation=@n1)
  ]}
]
variables [
 bIsMultiVersion {type=boolean}
 DonationResponses {
  type="map<string>"
  default {
   0ricent=Thanks for voting.
   Armored Battle Bus="I'd like to thank our top <WarEffort>donor</>: the <hot>bus driver</>."
   Balloons="The noble Balloon <hot>floats</> closer toward victory..."
   Boogie Bomb="The Island's ready for a <hot>tactical dance-off</>."
   Choppa="We'll see you in the <hot>cockpit</> someday soon."
   Combat Assault Rifle="High <hot>recoil</>, high reward."
   Heavy Turret="We're a few Bars closer to a brand new <hot>battle station</>."
   Light Turret="We're a few Bars closer to a brand new <hot>battle station</>."
   MK-Seven Assault Rifle=The Loopers have spoken...and they want a red dot sight.
   Rift-to-Go="Soon we'll be making moves from the safety of the <hot>skies</>."
   Shield Bubble="Hard to deny the usefulness of a <hot>portable shield</>."
   ashurrii="A sweat-free selection!"
   benmac="We at <hot>No Sweat Insurance</> appreciate your opinion... and your Bars."
   charlee-brown=What a sweat-free and stunning addition to the Tilted Towers skyline.
   destructo="What a responsible choice!"
   justafrog=No Sweat thanks you for voting responsibly.
   mrstruelegend="Nice choice. You didn't even break a sweat!"
   ninjah_dj="A great choice. Thanks for the <hot>Bars</>."
   xdark="A great option for the new Tilted. On an unrelated note: <hot>I'm rich!</>"
  }
 }
 MultiFundedText {
  type="array<string>"
  default ["I'm sensing this is a Looper <hot>favorite.</>", "<hot>Nice choice.</> The people agree with you.", I should try this one out myself...]
 }
 SelectedItemName {type=string, default=""}
 SingleFundedText {
  type="array<string>"
  default ["Every little bit helps. <hot>Thanks, soldier.</>", "I <hot>appreciate</> the decisiveness, soldier. We'll <hot>get it done.</>"]
 }
]
widgets {
 info {inheritedWidgetCount=2, parentClass=/Script/FortniteConversationUI.FortNPCTextDisplay, rootWidgetClass=&SizeBox, widgetCount=26}
 widgets [
  SizeBox_1 {widgetClassPath=&SizeBox, widget=SizeBox_1, properties {bOverride_MinDesiredHeight=true, minDesiredHeight=160}}
  HorizontalBox_0 {widgetClassPath=&HorizontalBox, parent=SizeBox_1, widget=HorizontalBox_0, slot {class=SizeBoxSlot}}
  Overlay_1 {
   widgetClassPath=&Overlay
   parent=HorizontalBox_0
   widget=Overlay_1
   properties {renderTransform {angle=-1.5, scale=<1 1>, shear=<0 0>, translation=<0 0>}}
   slot {class=HorizontalBoxSlot, properties {horizontalAlignment=HAlign_Right, verticalAlignment=VAlign_Center, padding=<ltrb 16 0 0 0>}}
  }
  Image_136 {
   widgetClassPath=&Image
   bIsVariable=true
   parent=Overlay_1
   widget=Image_136
   properties {
    visibility=SelfHitTestInvisible
    brush {
     bIsDynamicallyLoaded=false
     drawAs=Image
     imageType=NoImage
     mirroring=NoMirror
     resourceName=None
     resourceObject=/FortUILibrary/Foundation/Art_ToolKit/AngledBoxes/M_UI_FundCenter_PortraitBacking
     tiling=NoTile
     imageSize=<32 32>
     margin=<ltrb 0 0 0 0>
     outlineSettings {
      bUseBrushTransparency=false
      roundingType=HalfHeightRadius
      width=0
      color {colorUseRule=UseColor_Specified, specifiedColor=<rgba 0 0 0 0>}
      cornerRadii=<0 0 0 0>
     }
     tintColor {colorUseRule=UseColor_Specified, specifiedColor=<rgba 1 1 1 1>}
     uVRegion {bIsValid=false, max=<0 0>, min=<0 0>}
    }
   }
   slot {class=OverlaySlot, properties {horizontalAlignment=HAlign_Fill, verticalAlignment=VAlign_Fill}}
  }
  LazyImage_NPCImage {
   widgetClassPath=&FortLazyImage
   bInherited=true
   parent=Overlay_1
   widget=LazyImage_NPCImage
   properties {
    materialTextureParamName=Texture
    brush {
     bIsDynamicallyLoaded=false
     drawAs=Image
     imageType=FullColor
     mirroring=NoMirror
     resourceName=None
     resourceObject=/FortniteConversationUI/Conversation/Materials/M_UI_Figment_Donation_NPC_Portrait
     tiling=NoTile
     imageSize=<96 96>
     margin=<ltrb 0 0 0 0>
     outlineSettings {
      bUseBrushTransparency=false
      roundingType=HalfHeightRadius
      width=0
      color {colorUseRule=UseColor_Specified, specifiedColor=<rgba 0 0 0 0>}
      cornerRadii=<0 0 0 0>
     }
     tintColor {colorUseRule=UseColor_Specified, specifiedColor=<rgba 1 1 1 1>}
     uVRegion {bIsValid=false, max=<0 0>, min=<0 0>}
    }
    shadowColorAndOpacity=<rgba 0.0052 0.0052 0.0056 0.5>
   }
   slot {class=OverlaySlot, properties {padding=<ltrb 8 10 10 8>}}
  }
  Overlay_2 {
   widgetClassPath=&Overlay
   parent=HorizontalBox_0
   widget=Overlay_2
   slot {class=HorizontalBoxSlot, properties {verticalAlignment=VAlign_Top, padding=<ltrb 12 24 0 0>}}
  }
  Overlay_BubbleContainer {
   widgetClassPath=&Overlay
   parent=Overlay_2
   widget=Overlay_BubbleContainer
   properties {visibility=Collapsed}
   slot {class=OverlaySlot, properties {verticalAlignment=VAlign_Center}}
  }
  Overlay_47 {
   widgetClassPath=&Overlay
   parent=Overlay_BubbleContainer
   widget=Overlay_47
   properties {renderTransform {angle=0, scale=<1 1>, shear=<8 0>, translation=<0 0>}}
   slot {class=OverlaySlot, properties {horizontalAlignment=HAlign_Fill, verticalAlignment=VAlign_Fill}}
  }
  Border_Shadow {
   widgetClassPath=&Border
   parent=Overlay_47
   widget=Border_Shadow
   properties {
    background {
     bIsDynamicallyLoaded=false
     drawAs=NoDrawType
     imageType=NoImage
     mirroring=NoMirror
     resourceName=None
     resourceObject=null
     tiling=NoTile
     imageSize=<32 32>
     margin=<ltrb 0 0 0 0>
     outlineSettings {
      bUseBrushTransparency=false
      roundingType=HalfHeightRadius
      width=0
      color {colorUseRule=UseColor_Specified, specifiedColor=<rgba 0 0 0 0>}
      cornerRadii=<0 0 0 0>
     }
     tintColor {colorUseRule=UseColor_Specified, specifiedColor=<rgba 1 1 1 1>}
     uVRegion {bIsValid=false, max=<0 0>, min=<0 0>}
    }
    contentColorAndOpacity=<rgba 0.0052 0.0052 0.0056 0.4>
    padding=<ltrb 0 0 0 0>
    renderTransform {angle=0, scale=<1 1>, shear=<0 0>, translation=<8 8>}
   }
   slot {class=OverlaySlot, properties {horizontalAlignment=HAlign_Fill, verticalAlignment=VAlign_Fill}}
  }
  Overlay_Shadow {
   widgetClassPath=&Overlay
   parent=Border_Shadow
   widget=Overlay_Shadow
   slot {class=BorderSlot, properties {padding=<ltrb 0 0 0 0>}}
  }
  Background_Shadow {
   widgetClassPath=&Image
   parent=Overlay_Shadow
   widget=Background_Shadow
   properties {
    brush {
     bIsDynamicallyLoaded=false
     drawAs=Box
     imageType=NoImage
     mirroring=NoMirror
     resourceName=None
     resourceObject=/Game/Athena/HUD/Materials/M_UI_MessageBG_Inst
     tiling=NoTile
     imageSize=<64 64>
     margin=<ltrb 0.4 0 0.4 0>
     outlineSettings {
      bUseBrushTransparency=false
      roundingType=HalfHeightRadius
      width=0
      color {colorUseRule=UseColor_Specified, specifiedColor=<rgba 0 0 0 0>}
      cornerRadii=<0 0 0 0>
     }
     tintColor {colorUseRule=UseColor_Specified, specifiedColor=<rgba 1 1 1 1>}
     uVRegion {bIsValid=false, max=<0 0>, min=<0 0>}
    }
    renderTransformPivot=<0.5 0.9>
   }
   slot {class=OverlaySlot, properties {horizontalAlignment=HAlign_Fill, verticalAlignment=VAlign_Fill}}
  }
  HorizontalBox_Tail_1 {
   widgetClassPath=&HorizontalBox
   parent=Overlay_Shadow
   widget=HorizontalBox_Tail_1
   slot {class=OverlaySlot, properties {horizontalAlignment=HAlign_Fill, verticalAlignment=VAlign_Center, padding=<ltrb 0 -10 0 -1>}}
  }
  Image_TailShadow {
   widgetClassPath=&Image
   parent=HorizontalBox_Tail_1
   widget=Image_TailShadow
   properties {
    brush {
     bIsDynamicallyLoaded=false
     drawAs=Image
     imageType=NoImage
     mirroring=NoMirror
     resourceName=None
     resourceObject=/Game/Athena/UI/Frontend/Parties/Art/M_UI_LightningBanner_Normal_Tail_White
     tiling=NoTile
     imageSize=<64 80>
     margin=<ltrb 0 0 0 0>
     outlineSettings {
      bUseBrushTransparency=false
      roundingType=HalfHeightRadius
      width=0
      color {colorUseRule=UseColor_Specified, specifiedColor=<rgba 0 0 0 0>}
      cornerRadii=<0 0 0 0>
     }
     tintColor {colorUseRule=UseColor_Specified, specifiedColor=<rgba 1 1 1 1>}
     uVRegion {bIsValid=false, max=<0 0>, min=<0 0>}
    }
    renderTransform {angle=-90, scale=<1 1>, shear=<0 0>, translation=<80 -30>}
    renderTransformPivot=<0.5 1>
   }
   slot {class=HorizontalBoxSlot, properties {horizontalAlignment=HAlign_Right, size {sizeRule=Fill, value=1}}}
  }
  HorizontalBox_Tail {
   widgetClassPath=&HorizontalBox
   parent=Overlay_47
   widget=HorizontalBox_Tail
   slot {class=OverlaySlot, properties {horizontalAlignment=HAlign_Fill, verticalAlignment=VAlign_Center, padding=<ltrb 0 -10 0 -10>}}
  }
  Image_Tail {
   widgetClassPath=&Image
   bIsVariable=true
   parent=HorizontalBox_Tail
   widget=Image_Tail
   properties {
    brush {
     bIsDynamicallyLoaded=false
     drawAs=Image
     imageType=NoImage
     mirroring=NoMirror
     resourceName=None
     resourceObject=&M_UI_LightningBanner_Normal_Tail_White
     tiling=NoTile
     imageSize=<64 80>
     margin=<ltrb 0 0 0 0>
     outlineSettings {
      bUseBrushTransparency=false
      roundingType=HalfHeightRadius
      width=0
      color {colorUseRule=UseColor_Specified, specifiedColor=<rgba 0 0 0 0>}
      cornerRadii=<0 0 0 0>
     }
     tintColor {colorUseRule=UseColor_Specified, specifiedColor=<rgba 1 1 1 1>}
     uVRegion {bIsValid=false, max=<0 0>, min=<0 0>}
    }
    renderTransform {angle=-90, scale=<1 1>, shear=<0 0>, translation=<80 -35>}
    renderTransformPivot=<0.5 1>
   }
   slot {class=HorizontalBoxSlot, properties {horizontalAlignment=HAlign_Right, size {sizeRule=Fill, value=1}}}
  }
  Background {
   widgetClassPath=&Image
   bIsVariable=true
   parent=Overlay_47
   widget=Background
   properties {
    brush {
     bIsDynamicallyLoaded=false
     drawAs=Box
     imageType=NoImage
     mirroring=NoMirror
     resourceName=None
     resourceObject=&M_UI_MessageBG_Inst
     tiling=NoTile
     imageSize=<64 64>
     margin=<ltrb 0.4 0 0.4 0>
     outlineSettings {
      bUseBrushTransparency=false
      roundingType=HalfHeightRadius
      width=0
      color {colorUseRule=UseColor_Specified, specifiedColor=<rgba 0 0 0 0>}
      cornerRadii=<0 0 0 0>
     }
     tintColor {colorUseRule=UseColor_Specified, specifiedColor=<rgba 1 1 1 1>}
     uVRegion {bIsValid=false, max=<0 0>, min=<0 0>}
    }
    renderTransformPivot=<0.5 0.9>
   }
   slot {class=OverlaySlot, properties {horizontalAlignment=HAlign_Fill, verticalAlignment=VAlign_Fill, padding=<ltrb 0 -5 0 -5>}}
  }
  ScaleBox_0 {
   widgetClassPath=&ScaleBox
   parent=Overlay_BubbleContainer
   widget=ScaleBox_0
   properties {stretch=UserSpecified}
   slot {class=OverlaySlot, properties {horizontalAlignment=HAlign_Center, verticalAlignment=VAlign_Center, padding=<ltrb 20 0 20 0>}}
  }
  Overlay_0 {widgetClassPath=&Overlay, parent=ScaleBox_0, widget=Overlay_0, slot {class=ScaleBoxSlot}}
  SizeBox_0 {
   widgetClassPath=&SizeBox
   parent=Overlay_2
   widget=SizeBox_0
   properties {
    bOverride_MaxDesiredWidth=true
    bOverride_MinDesiredHeight=true
    bOverride_MinDesiredWidth=true
    maxDesiredHeight=100
    maxDesiredWidth=500
    minDesiredHeight=80
    minDesiredWidth=200
    widthOverride=672
    renderTransformPivot=<0 0.5>
   }
   slot {class=OverlaySlot, properties {horizontalAlignment=HAlign_Fill, padding=<ltrb 20 4 0 0>}}
  }
  Overlay_4 {widgetClassPath=&Overlay, parent=SizeBox_0, widget=Overlay_4, slot {class=SizeBoxSlot}}
  SizeBoxContainerSizeBox {
   widgetClassPath=&SizeBox
   parent=Overlay_4
   widget=SizeBoxContainerSizeBox
   properties {heightOverride=180, widthOverride=500}
   slot {class=OverlaySlot, properties {horizontalAlignment=HAlign_Fill, verticalAlignment=VAlign_Fill}}
  }
  ImageContainer {
   widgetClassPath=&Image
   bIsVariable=true
   parent=SizeBoxContainerSizeBox
   widget=ImageContainer
   properties {
    brush {
     bIsDynamicallyLoaded=false
     drawAs=Box
     imageType=NoImage
     mirroring=NoMirror
     resourceName=None
     resourceObject=/FortniteConversationUI/Conversation/Materials/M_UI_NPCText_Container_Default
     tiling=NoTile
     imageSize=<64 64>
     margin=<ltrb 0.5 0.5 0.5 0.5>
     outlineSettings {
      bUseBrushTransparency=false
      roundingType=HalfHeightRadius
      width=0
      color {colorUseRule=UseColor_Specified, specifiedColor=<rgba 0 0 0 0>}
      cornerRadii=<0 0 0 0>
     }
     tintColor {colorUseRule=UseColor_Specified, specifiedColor=<rgba 1 1 1 1>}
     uVRegion {bIsValid=false, max=<0 0>, min=<0 0>}
    }
    renderTransformPivot=<0 0.5>
   }
   slot {class=SizeBoxSlot}
  }
  ImageArrow {
   widgetClassPath=&Image
   bIsVariable=true
   parent=Overlay_4
   widget=ImageArrow
   properties {
    brush {
     bIsDynamicallyLoaded=false
     drawAs=Image
     imageType=NoImage
     mirroring=NoMirror
     resourceName=None
     resourceObject=/FortniteCoreUI/Athena/HUD/Quests/Art/M_UI_Quest_Container_Arrow_Overlay
     tiling=NoTile
     imageSize=<32 32>
     margin=<ltrb 0 0 0 0>
     outlineSettings {
      bUseBrushTransparency=false
      roundingType=HalfHeightRadius
      width=0
      color {colorUseRule=UseColor_Specified, specifiedColor=<rgba 0 0 0 0>}
      cornerRadii=<0 0 0 0>
     }
     tintColor {colorUseRule=UseColor_Specified, specifiedColor=<rgba 1 1 1 1>}
     uVRegion {bIsValid=false, max=<0 0>, min=<0 0>}
    }
    colorAndOpacity=<rgba 0 0.2232 0.6654 1>
    renderTransformPivot=<1 0.5>
   }
   slot {class=OverlaySlot, properties {padding=<ltrb -22 12 0 0>}}
  }
  Overlay_5 {
   widgetClassPath=&Overlay
   parent=Overlay_4
   widget=Overlay_5
   slot {class=OverlaySlot, properties {horizontalAlignment=HAlign_Fill, verticalAlignment=VAlign_Fill, padding=<ltrb 32 16 18 16>}}
  }
  Text_Message {
   widgetClassPath=&CommonRichTextBlock
   bInherited=true
   bIsVariable=true
   parent=Overlay_5
   widget=Text_Message
   properties {
    bOverrideDefaultStyle=true
    textStyleSet=&/Conversation_RichTextStyleSet
    wrapTextAt=448
    defaultTextStyleOverride {
     overflowPolicy=Clip
     transformPolicy=None
     colorAndOpacity {colorUseRule=UseColor_Specified, specifiedColor=<rgba 1 1 1 1>}
     font {
      bForceMonospaced=false
      bMaterialIsStencil=false
      fontMaterial=null
      fontObject=&BurbankSmall
      letterSpacing=0
      monospacedWidth=1
      size=15
      skewAmount=0
      typefaceFontName=Bold
      outlineSettings {
       bApplyOutlineToDropShadows=false
       bMiteredCorners=false
       bSeparateFillAlpha=false
       outlineMaterial=null
       outlineSize=0
       outlineColor=<rgba 0 0 0 1>
      }
     }
     highlightColor {colorUseRule=UseColor_Specified, specifiedColor=<rgba 0 0 0 1>}
     highlightShape {
      bIsDynamicallyLoaded=false
      drawAs=Image
      imageType=NoImage
      mirroring=NoMirror
      resourceName=None
      resourceObject=null
      tiling=NoTile
      imageSize=<32 32>
      margin=<ltrb 0 0 0 0>
      outlineSettings {
       bUseBrushTransparency=false
       roundingType=HalfHeightRadius
       width=0
       color {colorUseRule=UseColor_Specified, specifiedColor=<rgba 0 0 0 0>}
       cornerRadii=<0 0 0 0>
      }
      tintColor {colorUseRule=UseColor_Specified, specifiedColor=<rgba 1 1 1 1>}
      uVRegion {bIsValid=false, max=<0 0>, min=<0 0>}
     }
     selectedBackgroundColor {colorUseRule=UseColor_Foreground, specifiedColor=<rgba 1 0 1 1>}
     shadowColorAndOpacity=<rgba 0 0 0 1>
     shadowOffset=<0 0>
     strikeBrush {
      bIsDynamicallyLoaded=false
      drawAs=Image
      imageType=NoImage
      mirroring=NoMirror
      resourceName=None
      resourceObject=null
      tiling=NoTile
      imageSize=<32 32>
      margin=<ltrb 0 0 0 0>
      outlineSettings {
       bUseBrushTransparency=false
       roundingType=HalfHeightRadius
       width=0
       color {colorUseRule=UseColor_Specified, specifiedColor=<rgba 0 0 0 0>}
       cornerRadii=<0 0 0 0>
      }
      tintColor {colorUseRule=UseColor_Specified, specifiedColor=<rgba 1 1 1 1>}
      uVRegion {bIsValid=false, max=<0 0>, min=<0 0>}
     }
     underlineBrush {
      bIsDynamicallyLoaded=false
      drawAs=Image
      imageType=NoImage
      mirroring=NoMirror
      resourceName=None
      resourceObject=null
      tiling=NoTile
      imageSize=<32 32>
      margin=<ltrb 0 0 0 0>
      outlineSettings {
       bUseBrushTransparency=false
       roundingType=HalfHeightRadius
       width=0
       color {colorUseRule=UseColor_Specified, specifiedColor=<rgba 0 0 0 0>}
       cornerRadii=<0 0 0 0>
      }
      tintColor {colorUseRule=UseColor_Specified, specifiedColor=<rgba 1 1 1 1>}
      uVRegion {bIsValid=false, max=<0 0>, min=<0 0>}
     }
    }
   }
   slot {class=OverlaySlot, properties {horizontalAlignment=HAlign_Fill, verticalAlignment=VAlign_Fill}}
  }
  Funded_Text_Message {
   widgetClassPath=&CommonRichTextBlock
   bIsVariable=true
   parent=Overlay_5
   widget=Funded_Text_Message
   properties {
    bOverrideDefaultStyle=true
    textStyleSet=&Conversation_RichTextStyleSet
    wrapTextAt=448
    defaultTextStyleOverride {
     overflowPolicy=Clip
     transformPolicy=None
     colorAndOpacity {colorUseRule=UseColor_Specified, specifiedColor=<rgba 1 1 1 1>}
     font {
      bForceMonospaced=false
      bMaterialIsStencil=false
      fontMaterial=null
      fontObject=&BurbankSmall
      letterSpacing=0
      monospacedWidth=1
      size=15
      skewAmount=0
      typefaceFontName=Bold
      outlineSettings {
       bApplyOutlineToDropShadows=false
       bMiteredCorners=false
       bSeparateFillAlpha=false
       outlineMaterial=null
       outlineSize=0
       outlineColor=<rgba 0 0 0 1>
      }
     }
     highlightColor {colorUseRule=UseColor_Specified, specifiedColor=<rgba 0 0 0 1>}
     highlightShape {
      bIsDynamicallyLoaded=false
      drawAs=Image
      imageType=NoImage
      mirroring=NoMirror
      resourceName=None
      resourceObject=null
      tiling=NoTile
      imageSize=<32 32>
      margin=<ltrb 0 0 0 0>
      outlineSettings {
       bUseBrushTransparency=false
       roundingType=HalfHeightRadius
       width=0
       color {colorUseRule=UseColor_Specified, specifiedColor=<rgba 0 0 0 0>}
       cornerRadii=<0 0 0 0>
      }
      tintColor {colorUseRule=UseColor_Specified, specifiedColor=<rgba 1 1 1 1>}
      uVRegion {bIsValid=false, max=<0 0>, min=<0 0>}
     }
     selectedBackgroundColor {colorUseRule=UseColor_Foreground, specifiedColor=<rgba 1 0 1 1>}
     shadowColorAndOpacity=<rgba 0 0 0 1>
     shadowOffset=<0 0>
     strikeBrush {
      bIsDynamicallyLoaded=false
      drawAs=Image
      imageType=NoImage
      mirroring=NoMirror
      resourceName=None
      resourceObject=null
      tiling=NoTile
      imageSize=<32 32>
      margin=<ltrb 0 0 0 0>
      outlineSettings {
       bUseBrushTransparency=false
       roundingType=HalfHeightRadius
       width=0
       color {colorUseRule=UseColor_Specified, specifiedColor=<rgba 0 0 0 0>}
       cornerRadii=<0 0 0 0>
      }
      tintColor {colorUseRule=UseColor_Specified, specifiedColor=<rgba 1 1 1 1>}
      uVRegion {bIsValid=false, max=<0 0>, min=<0 0>}
     }
     underlineBrush {
      bIsDynamicallyLoaded=false
      drawAs=Image
      imageType=NoImage
      mirroring=NoMirror
      resourceName=None
      resourceObject=null
      tiling=NoTile
      imageSize=<32 32>
      margin=<ltrb 0 0 0 0>
      outlineSettings {
       bUseBrushTransparency=false
       roundingType=HalfHeightRadius
       width=0
       color {colorUseRule=UseColor_Specified, specifiedColor=<rgba 0 0 0 0>}
       cornerRadii=<0 0 0 0>
      }
      tintColor {colorUseRule=UseColor_Specified, specifiedColor=<rgba 1 1 1 1>}
      uVRegion {bIsValid=false, max=<0 0>, min=<0 0>}
     }
    }
   }
   slot {class=OverlaySlot, properties {horizontalAlignment=HAlign_Fill, verticalAlignment=VAlign_Fill}}
  }
 ]
}
