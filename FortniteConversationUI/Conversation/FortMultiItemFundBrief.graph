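# Widget for the NPC conversation donation screen, showing two item fund entries side by side with a central arrow badge and NPC text.
# Construct plays Intro, binds to the player controller's donation-success delegate and refreshes the NPC text.
# Selection changes are filtered against a cached index, then flip the arrow, set both entries' highlights using an inverted index (1 - SelectedIndex), and refresh the text.
# FundedAnim plays the check-and-burst Funded animation.
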WidgetBlueprint /FortniteConversationUI/Conversation/FortMultiItemFundBrief
animations [
 Anim_SelectLeft {
  binding_count=1
  length_seconds=0.4
  object=&self:Anim_SelectLeft
  sequence {
   bindings [Image_Arrow {
     id=#1
     possessed_class=Image
     tracks [{
       type=MovieScene2DTransformTrack
       sections [{
         type=MovieScene2DTransformSection
         channels [
          Scale.X {type=MovieSceneScriptingFloatChannel, thinned_from=7, keys [[0,1],[3,1],[8,1.05],[16,-1.05],[21,-1],[24,-1]]}
          Scale.Y {type=MovieSceneScriptingFloatChannel, const [1,3]}
         ]
        }]
      }]
    }]
   display_rate {numerator=60}
  }
 }
 Anim_SelectRight {
  binding_count=1
  length_seconds=0.4
  object=&self:Anim_SelectRight
  sequence {
   bindings [Image_Arrow {
     id=#2
     possessed_class=Image
     tracks [{
       type=MovieScene2DTransformTrack
       sections [{
         type=MovieScene2DTransformSection
         channels [
          Scale.X {type=MovieSceneScriptingFloatChannel, thinned_from=7, keys [[0,-1],[3,-1],[8,-1.05],[16,1.05],[21,1],[24,1]]}
          Scale.Y {type=MovieSceneScriptingFloatChannel, const [1,3]}
         ]
        }]
      }]
    }]
   display_rate {numerator=60}
  }
 }
 Anim_SelectRight_Bounce {
  binding_count=4
  length_seconds=0.4
  object=&self:Anim_SelectRight_Bounce
  sequence {
   bindings [
    Image_Shadow {
     id=#3
     possessed_class=Image
     tracks [
      {
       type=MovieScene2DTransformTrack
       sections [{
         type=MovieScene2DTransformSection
         end_frame=24
         start_frame=0
         channels [
          Translation.X {type=MovieSceneScriptingFloatChannel, const [0,1]}
          Translation.Y {type=MovieSceneScriptingFloatChannel, const [0,1]}
          Angle {type=MovieSceneScriptingFloatChannel, const [0,1]}
          Scale.X {type=MovieSceneScriptingFloatChannel, keys [[0,1],[3,0.95],[6,1.05],[12,1]]}
          Scale.Y {type=MovieSceneScriptingFloatChannel, keys [[0,1],[3,0.95],[6,1.05],[12,1]]}
          Shear.X {type=MovieSceneScriptingFloatChannel, const [0,1]}
          Shear.Y {type=MovieSceneScriptingFloatChannel, const [0,1]}
         ]
        }]
      }
      {
       type=MovieSceneWidgetMaterialTrack
       sections [{
         type=MovieSceneParameterSection
         end_frame=24
         start_frame=0
         channels [
          ShadowAmount {type=MovieSceneScriptingFloatChannel, keys [[0,0.333],[3,1],[6,0.25],[12,0.333]]}
          ShadowScale {type=MovieSceneScriptingFloatChannel, keys [[0,1],[3,0.9],[6,1.125],[12,1]]}
         ]
        }]
      }
     ]
    }
    Image_Circle {
     id=#4
     possessed_class=Image
     tracks [{
       type=MovieScene2DTransformTrack
       sections [{
         type=MovieScene2DTransformSection
         end_frame=24
         start_frame=0
         channels [
          Scale.X {type=MovieSceneScriptingFloatChannel, keys [[2,1],[5,0.95],[8,1.05],[14,1]]}
          Scale.Y {type=MovieSceneScriptingFloatChannel, keys [[2,1],[5,0.95],[8,1.05],[14,1]]}
         ]
        }]
      }]
    }
    Image_Arrow {
     id=#2
     possessed_class=Image
     tracks [{
       type=MovieScene2DTransformTrack
       sections [{
         type=MovieScene2DTransformSection
         channels [
          Scale.X {type=MovieSceneScriptingFloatChannel, thinned_from=7, keys [[0,-1],[3,-1],[8,-1.05],[16,1.05],[21,1],[24,1]]}
          Scale.Y {type=MovieSceneScriptingFloatChannel, const [1,3]}
         ]
        }]
      }]
    }
    CommonTextBlock_82 {
     id=#5
     possessed_class=CommonTextBlock
     tracks [{
       type=MovieScene2DTransformTrack
       sections [{
         type=MovieScene2DTransformSection
         end_frame=24
         start_frame=0
         channels [
          Translation.X {type=MovieSceneScriptingFloatChannel}
          Translation.Y {type=MovieSceneScriptingFloatChannel}
          Angle {type=MovieSceneScriptingFloatChannel}
          Scale.X {type=MovieSceneScriptingFloatChannel, keys [[0,1],[3,0.95],[6,1.05],[12,1]]}
          Scale.Y {type=MovieSceneScriptingFloatChannel, keys [[0,1],[3,0.95],[6,1.05],[12,1]]}
          Shear.X {type=MovieSceneScriptingFloatChannel}
          Shear.Y {type=MovieSceneScriptingFloatChannel}
         ]
        }]
      }]
    }
   ]
   display_rate {numerator=60}
  }
 }
 Anim_SelectLeft_Bounce {
  binding_count=4
  length_seconds=0.4
  object=&self:Anim_SelectLeft_Bounce
  sequence {
   bindings [
    Image_Arrow {
     id=#1
     possessed_class=Image
     tracks [{
       type=MovieScene2DTransformTrack
       sections [{
         type=MovieScene2DTransformSection
         channels [
          Scale.X {type=MovieSceneScriptingFloatChannel, thinned_from=7, keys [[0,1],[3,1],[8,1.05],[16,-1.05],[21,-1],[24,-1]]}
          Scale.Y {type=MovieSceneScriptingFloatChannel, const [1,3]}
         ]
        }]
      }]
    }
    Image_Circle {
     id=#6
     possessed_class=Image
     tracks [{
       type=MovieScene2DTransformTrack
       sections [{
         type=MovieScene2DTransformSection
         end_frame=24
         start_frame=0
         channels [
          Scale.X {type=MovieSceneScriptingFloatChannel, keys [[2,1],[5,0.95],[8,1.05],[14,1]]}
          Scale.Y {type=MovieSceneScriptingFloatChannel, keys [[2,1],[5,0.95],[8,1.05],[14,1]]}
         ]
        }]
      }]
    }
    CommonTextBlock_82 {
     id=#7
     possessed_class=CommonTextBlock
     tracks [{
       type=MovieScene2DTransformTrack
       sections [{
         type=MovieScene2DTransformSection
         end_frame=24
         start_frame=0
         channels [
          Translation.X {type=MovieSceneScriptingFloatChannel}
          Translation.Y {type=MovieSceneScriptingFloatChannel}
          Angle {type=MovieSceneScriptingFloatChannel}
          Scale.X {type=MovieSceneScriptingFloatChannel, keys [[0,1],[3,0.95],[6,1.05],[12,1]]}
          Scale.Y {type=MovieSceneScriptingFloatChannel, keys [[0,1],[3,0.95],[6,1.05],[12,1]]}
          Shear.X {type=MovieSceneScriptingFloatChannel}
          Shear.Y {type=MovieSceneScriptingFloatChannel}
         ]
        }]
      }]
    }
    Image_Shadow {
     id=#8
     possessed_class=Image
     tracks [
      {
       type=MovieScene2DTransformTrack
       sections [{
         type=MovieScene2DTransformSection
         end_frame=24
         start_frame=0
         channels [
          Translation.X {type=MovieSceneScriptingFloatChannel, const [0,1]}
          Translation.Y {type=MovieSceneScriptingFloatChannel, const [0,1]}
          Angle {type=MovieSceneScriptingFloatChannel, const [0,1]}
          Scale.X {type=MovieSceneScriptingFloatChannel, keys [[0,1],[3,0.95],[6,1.05],[12,1]]}
          Scale.Y {type=MovieSceneScriptingFloatChannel, keys [[0,1],[3,0.95],[6,1.05],[12,1]]}
          Shear.X {type=MovieSceneScriptingFloatChannel, const [0,1]}
          Shear.Y {type=MovieSceneScriptingFloatChannel, const [0,1]}
         ]
        }]
      }
      {
       type=MovieSceneWidgetMaterialTrack
       sections [{
         type=MovieSceneParameterSection
         end_frame=24
         start_frame=0
         channels [
          ShadowAmount {type=MovieSceneScriptingFloatChannel, keys [[0,0.333],[3,1],[6,0.25],[12,0.333]]}
          ShadowScale {type=MovieSceneScriptingFloatChannel, keys [[0,1],[3,0.9],[6,1.125],[12,1]]}
         ]
        }]
      }
     ]
    }
   ]
   display_rate {numerator=60}
  }
 }
 Intro {
  binding_count=6
  length_seconds=0.55
  object=&self:Intro
  sequence {bindings [
    Image_Circle {
     id=#9
     possessed_class=Image
     tracks [{
       type=MovieScene2DTransformTrack
       sections [{
         type=MovieScene2DTransformSection
         channels [
          Translation.X {type=MovieSceneScriptingFloatChannel, const [0,1]}
          Translation.Y {type=MovieSceneScriptingFloatChannel, const [0,1]}
          Angle {type=MovieSceneScriptingFloatChannel, const [0,1]}
          Scale.X {type=MovieSceneScriptingFloatChannel, keys [[1,1],[4,1.1],[7,0.95],[9.0003,1]]}
          Scale.Y {type=MovieSceneScriptingFloatChannel, keys [[1,1],[4,1.1],[7,0.95],[9.0003,1]]}
          Shear.X {type=MovieSceneScriptingFloatChannel, const [0,1]}
          Shear.Y {type=MovieSceneScriptingFloatChannel, const [0,1]}
         ]
        }]
      }]
    }
    Overlay_15 {
     id=#10
     possessed_class=Overlay
     tracks [
      {
       type=MovieSceneFloatTrack
       display_name=Render Opacity
       sections [{type=MovieSceneFloatSection, channels [None {type=MovieSceneScriptingFloatChannel, keys [[0,0],[2,1]]}]}]
      }
      {
       type=MovieScene2DTransformTrack
       sections [{
         type=MovieScene2DTransformSection
         channels [
          Translation.X {type=MovieSceneScriptingFloatChannel, const [0,1]}
          Translation.Y {type=MovieSceneScriptingFloatChannel, keys [[0,128],[5,0]]}
          Angle {type=MovieSceneScriptingFloatChannel, const [0,1]}
          Scale.X {type=MovieSceneScriptingFloatChannel, const [1,1]}
          Scale.Y {type=MovieSceneScriptingFloatChannel, const [1,1]}
          Shear.X {type=MovieSceneScriptingFloatChannel, const [0,1]}
          Shear.Y {type=MovieSceneScriptingFloatChannel, const [0,1]}
         ]
        }]
      }
     ]
    }
    GradientBacking {
     id=#11
     possessed_class=Image
     tracks [{
       type=MovieSceneFloatTrack
       display_name=Render Opacity
       sections [{type=MovieSceneFloatSection, channels [None {type=MovieSceneScriptingFloatChannel, keys [[0,0],[10,1]]}]}]
      }]
    }
    Darken {
     id=#12
     possessed_class=Image
     tracks [{
       type=MovieSceneFloatTrack
       display_name=Render Opacity
       sections [{type=MovieSceneFloatSection, channels [None {type=MovieSceneScriptingFloatChannel, keys [[3,0],[7,1]]}]}]
      }]
    }
    GlobalEffort {
     id=#13
     possessed_class=Overlay
     tracks [
      {
       type=MovieScene2DTransformTrack
       sections [{
         type=MovieScene2DTransformSection
         channels [
          Translation.X {type=MovieSceneScriptingFloatChannel}
          Translation.Y {type=MovieSceneScriptingFloatChannel}
          Angle {type=MovieSceneScriptingFloatChannel}
          Scale.X {type=MovieSceneScriptingFloatChannel, keys [[6,1.25],[8,0.95],[11,1]]}
          Scale.Y {type=MovieSceneScriptingFloatChannel, keys [[6,1.25],[8,0.95],[11,1]]}
          Shear.X {type=MovieSceneScriptingFloatChannel}
          Shear.Y {type=MovieSceneScriptingFloatChannel}
         ]
        }]
      }
      {
       type=MovieSceneFloatTrack
       display_name=Render Opacity
       sections [{type=MovieSceneFloatSection, channels [None {type=MovieSceneScriptingFloatChannel, keys [[6,0],[8,1]]}]}]
      }
     ]
    }
    Overlay_1 {
     id=#14
     possessed_class=Overlay
     tracks [
      {
       type=MovieSceneFloatTrack
       display_name=Render Opacity
       sections [{type=MovieSceneFloatSection, channels [None {type=MovieSceneScriptingFloatChannel, keys [[0,1],[1,0],[4,0],[6,1]]}]}]
      }
      {
       type=MovieScene2DTransformTrack
       sections [{
         type=MovieScene2DTransformSection
         channels [
          Translation.X {type=MovieSceneScriptingFloatChannel, const [0,1]}
          Translation.Y {type=MovieSceneScriptingFloatChannel, const [16,1]}
          Angle {type=MovieSceneScriptingFloatChannel, const [0,1]}
          Scale.X {type=MovieSceneScriptingFloatChannel, keys [[1,1.5],[4,1.25],[6,0.95],[9,1]]}
          Scale.Y {type=MovieSceneScriptingFloatChannel, keys [[1,1.5],[4,1.25],[6,0.95],[9,1]]}
          Shear.X {type=MovieSceneScriptingFloatChannel, const [0,1]}
          Shear.Y {type=MovieSceneScriptingFloatChannel, const [0,1]}
         ]
        }]
      }
     ]
    }
   ]}
 }
 Funded {
  binding_count=8
  length_seconds=1.1
  object=&self:Funded
  sequence {
   bindings [
    Image_Circle {
     id=#15
     possessed_class=Image
     tracks [{
       type=MovieScene2DTransformTrack
       sections [{
         type=MovieScene2DTransformSection
         channels [
          Scale.X {type=MovieSceneScriptingFloatChannel, keys [[5.999,1],[11,0.85],[23,1]]}
          Scale.Y {type=MovieSceneScriptingFloatChannel, keys [[6,1],[11,0.85],[23,1]]}
         ]
        }]
      }]
    }
    Image_Arrow {
     id=#16
     display_name=AnimBoundFundedBorder
     possessed_class=Image
     tracks [
      {
       type=MovieSceneFloatTrack
       display_name=Render Opacity
       sections [{type=MovieSceneFloatSection, channels [None {type=MovieSceneScriptingFloatChannel, keys [[0,1],[3,1],[9,0],[48,0],[54,1]]}]}]
      }
      {
       type=MovieScene2DTransformTrack
       sections [{
         type=MovieScene2DTransformSection
         channels [
          Scale.X {type=MovieSceneScriptingFloatChannel, keys [[0,1],[3,1.125],[9,0],[48,0],[54,1.125],[66,1]]}
          Scale.Y {type=MovieSceneScriptingFloatChannel, keys [[0,1],[3,1.125],[9,0],[48,0],[54,1.125],[66,1]]}
         ]
        }]
      }
     ]
    }
    Image_229 {
     id=#17
     possessed_class=Image
     tracks [{
       type=MovieScene2DTransformTrack
       sections [{type=MovieScene2DTransformSection, channels [Angle {type=MovieSceneScriptingFloatChannel, keys [[0,-8],[27,352]]}]}]
      }]
    }
    Image_Check {
     id=#18
     possessed_class=Image
     tracks [
      {
       type=MovieSceneFloatTrack
       display_name=Render Opacity
       sections [{type=MovieSceneFloatSection, channels [None {type=MovieSceneScriptingFloatChannel, keys [[3.001,0],[9,0],[15,1],[42,1],[48,0]]}]}]
      }
      {
       type=MovieScene2DTransformTrack
       sections [{
         type=MovieScene2DTransformSection
         channels [
          Scale.X {type=MovieSceneScriptingFloatChannel, keys [[3,1],[9,1],[15,1.25],[27,1],[36,1],[42,1.125],[48,0]]}
          Scale.Y {type=MovieSceneScriptingFloatChannel, keys [[3,1],[9,1],[15,1.25],[27,1],[36,1],[42,1.125],[48,0]]}
         ]
        }]
      }
     ]
    }
    CommonTextBlock_82 {
     id=#19
     possessed_class=CommonTextBlock
     tracks [{
       type=MovieScene2DTransformTrack
       sections [{
         type=MovieScene2DTransformSection
         channels [
          Translation.X {type=MovieSceneScriptingFloatChannel, const [0,1]}
          Translation.Y {type=MovieSceneScriptingFloatChannel, keys [[6,88],[9,80],[15,92],[27,88]]}
          Scale.X {type=MovieSceneScriptingFloatChannel, keys [[6,1],[9,0.95],[15,1.05],[27,1]]}
          Scale.Y {type=MovieSceneScriptingFloatChannel, keys [[6,1],[9,0.95],[15,1.05],[27,1]]}
         ]
        }]
      }]
    }
    Image_Shadow {
     id=#20
     possessed_class=Image
     tracks [
      {
       type=MovieScene2DTransformTrack
       sections [{
         type=MovieScene2DTransformSection
         channels [
          Scale.X {type=MovieSceneScriptingFloatChannel, keys [[6,1],[9,0.9],[15,1.1],[27,1]]}
          Scale.Y {type=MovieSceneScriptingFloatChannel, keys [[6,1],[9,0.9],[15,1.1],[27,1]]}
         ]
        }]
      }
      {
       type=MovieSceneWidgetMaterialTrack
       sections [{type=MovieSceneParameterSection, channels [DonateColorTransition {type=MovieSceneScriptingFloatChannel, keys [[15,0],[18,0.666],[42,0]]}]}]
      }
     ]
    }
    Image {
     id=#21
     possessed_class=Image
     tracks [{
       type=MovieScene2DTransformTrack
       sections [{type=MovieScene2DTransformSection, channels [Angle {type=MovieSceneScriptingFloatChannel, keys [[0,8],[27,-352]]}]}]
      }]
    }
    Image_Burst {
     id=#22
     possessed_class=Image
     tracks [
      {
       type=MovieSceneFloatTrack
       display_name=Render Opacity
       sections [{type=MovieSceneFloatSection, channels [None {type=MovieSceneScriptingFloatChannel, keys [[0,0],[3,1],[24,0]]}]}]
      }
      {
       type=MovieScene2DTransformTrack
       sections [{
         type=MovieScene2DTransformSection
         channels [
          Scale.X {type=MovieSceneScriptingFloatChannel, keys [[0,0.5],[33,1.5]]}
          Scale.Y {type=MovieSceneScriptingFloatChannel, keys [[0,0.5],[33,1.5]]}
         ]
        }]
      }
     ]
    }
   ]
   display_rate {numerator=60}
  }
 }
]
graphs [
 EventGraph {nodes [
   n1 = Event UpdateSelectedVisuals!() -> (OutputDelegate, then, SelectedIndex:int=0) {is=title K2Node_Event}
   > Branch(Condition="Not Equal (Integer)"!(A:int=@n1.SelectedIndex, B:int=Get CachedSelectedIndex() -> (CachedSelectedIndex:int=0)) -> (ReturnValue:bool) {is=title K2Node_CallFunction})
   > n2 = Set CachedSelectedIndex(CachedSelectedIndex:int=@n1.SelectedIndex) -> (Output_Get:int=0)
   > UpdateSelectedState!(execute, self, SelectedIndex:int=@n2) -> (then)
   > n3 = Cast To NPCTextDisplay!(execute, Object=Get NPCTextDisplay() -> (NPCTextDisplay:Fort NPCText Display Object Reference)) -> (then, CastFailed, AsNPCText Display:NPCText Display Object Reference) {is=title K2Node_DynamicCast}
   > NewItemUpdate(self:NPCText Display Object Reference=@n3)
   Event Construct()
   > PlayAnimation(InAnimation=Get Intro())
   > n4 = Cast To Athena_PlayerController!(execute, Object=GetOwningPlayer()) -> (then, CastFailed, AsAthena Player Controller:Athena Player Controller Object Reference) {is=title K2Node_DynamicCast}
   > Bind Event to On Client Donation Success Delegate(self:Fort Player Controller Athena Object Reference=@n4, Delegate=Create Event()) {is=title K2Node_AddDelegate}
   > n5 = Cast To NPCTextDisplay(Object=Get NPCTextDisplay())
   > NewItemUpdate(self=@n5)
   n6 = Event PreConstruct()
   > Branch(Condition=@n6.IsDesignTime)
  ]}
 FundedAnim {nodes [
   FundedAnim() {is=title K2Node_FunctionEntry}
   > PlayAnimation(InAnimation=Get Funded() -> (Funded:Widget Animation Object Reference))
  ]}
 UpdateSelectedState {nodes [
   n1 = UpdateSelectedState!() -> (then, SelectedIndex:int) {is=title K2Node_FunctionEntry}
   > PlayAnimation(InAnimation=Select(
     Option 0:Widget Animation Object Reference=Get Anim_SelectRight() -> (Anim_SelectRight:Widget Animation Object Reference)
     Option 1:Widget Animation Object Reference=Get Anim_SelectLeft() -> (Anim_SelectLeft:Widget Animation Object Reference)
     Index:int=@n1
    ) -> (ReturnValue:Widget Animation Object Reference))
   > n2 = Cast To FortMultiItemFundBriefEntry!(execute, Object=Get Item_Entry() -> (Item_Entry:Fort Item Fund Brief Object Reference)) -> (then, CastFailed, AsFort Multi Item Fund Brief Entry:Fort Multi Item Fund Brief Entry Object Reference) {is=title K2Node_DynamicCast}
   n3 = int - int(A=1, B=Get SelectedIndex() -> (SelectedIndex:int=0)) {is=title K2Node_CallFunction}
   > n4 = SetSelectedVisuals!(
    execute
    self:Fort Multi Item Fund Brief Entry Object Reference=@n2
    InSelected:bool="Equal (Integer)"(A=@n3, B=0) {is=title K2Node_CallFunction}
   ) -> (then)
   n5 = Cast To FortMultiItemFundBriefEntry(execute=[@n4.then, @n6.then], Object=Get Item_EntrySecondary() -> (Item_EntrySecondary:Fort Item Fund Brief Object Reference))
   > SetSelectedVisuals(self=@n5, InSelected="Equal (Integer)"(A=@n3, B=1))
   PrintString!(
    execute=@n5.CastFailed
    InString=UpdateSelectedVisuals - WarEffort Item Entry Secondary didn't cast to MultiEntry
    bPrintToScreen
    bPrintToLog
    TextColor
    Duration
    Key
   ) -> (then)
   n6 = PrintString(execute=@n2.CastFailed, InString=UpdateSelectedVisuals - WarEffort Item Entry didn't cast to MultiEntry)
  ]}
]
variables [
 CachedSelectedIndex {type=integer, default=-1}
 DebugItem {type=FortItem, default=null}
]
widgets {
 info {
  inheritedWidgetCount=4
  parentClass=/Script/FortniteConversationUI.FortMultiItemFundBrief
  rootWidgetClass=&Overlay
  widgetCount=32
 }
 widgets [
  Overlay_13 {widgetClassPath=&Overlay, widget=Overlay_13}
  GradientBacking {
   widgetClassPath=&Image
   bIsVariable=true
   parent=Overlay_13
   widget=GradientBacking
   properties {
    visibility=SelfHitTestInvisible
    brush {
     bIsDynamicallyLoaded=false
     drawAs=Image
     imageType=NoImage
     mirroring=NoMirror
     resourceName=None
     resourceObject=/FortniteConversationUI/Conversation/Materials/M_UI_Donation_FadedBG
     tiling=NoTile
     imageSize=<32 448>
     margin=<ltrb 0 0 0 0>
     outlineSettings {
      bUseBrushTransparency=false
      roundingType=HalfHeightRadius
      width=0
      color {colorUseRule=UseColor_Specified, specifiedColor=<rgba 0 0 0 0>}
      cornerRadii=<0 0 0 0>
     }
     tintColor {colorUseRule=UseColor_Specified, specifiedColor=<rgba 1 1 1 1>}
     uVRegion {bIsValid=false, max=<0 0>, min=<0 0>}
    }
    renderTransform {angle=0, scale=<2 4>, shear=<0 0>, translation=<0 0>}
   }
   slot {class=OverlaySlot, properties {horizontalAlignment=HAlign_Fill, verticalAlignment=VAlign_Center, padding=<ltrb 0 0 0 20>}}
  }
  NPCTextDisplay {
   widgetClassPath=&/NPCTextDisplay
   bInherited=true
   bIsVariable=true
   parent=Overlay_13
   widget=NPCTextDisplay
   properties {bIsMultiVersion=true}
   slot {class=OverlaySlot, properties {verticalAlignment=VAlign_Center, padding=<ltrb 64 0 0 384>}}
  }
  Overlay_15 {
   widgetClassPath=&Overlay
   parent=Overlay_13
   widget=Overlay_15
   slot {class=OverlaySlot, properties {horizontalAlignment=HAlign_Fill, verticalAlignment=VAlign_Center, padding=<ltrb 0 0 0 80>}}
  }
  SizeBox_1 {
   widgetClassPath=&SizeBox
   parent=Overlay_15
   widget=SizeBox_1
   properties {bOverride_HeightOverride=true, bOverride_WidthOverride=true, heightOverride=448, widthOverride=5000}
   slot {class=OverlaySlot, properties {horizontalAlignment=HAlign_Fill, verticalAlignment=VAlign_Center}}
  }
  Overlay_0 {widgetClassPath=&Overlay, parent=SizeBox_1, widget=Overlay_0, slot {class=SizeBoxSlot}}
  Darken {
   widgetClassPath=&Image
   bIsVariable=true
   parent=Overlay_0
   widget=Darken
   properties {colorAndOpacity=<rgba 0 0 0 0>}
   slot {class=OverlaySlot, properties {horizontalAlignment=HAlign_Fill, verticalAlignment=VAlign_Fill, padding=<ltrb 0 56 0 0>}}
  }
  HorizontalBox_22 {
   widgetClassPath=&HorizontalBox
   parent=Overlay_0
   widget=HorizontalBox_22
   slot {class=OverlaySlot, properties {horizontalAlignment=HAlign_Fill, verticalAlignment=VAlign_Fill}}
  }
  Item_Entry {
   widgetClassPath=&/FortMultiItemFundBriefEntry
   bInherited=true
   bIsVariable=true
   parent=HorizontalBox_22
   widget=Item_Entry
   properties {bSelected=false, cropOffset=-0.0295, flipShadow=1, isLeftSideSelection=true}
   slot {class=HorizontalBoxSlot, properties {size {sizeRule=Fill, value=1}}}
  }
  Spacer_92 {
   widgetClassPath=&Spacer
   parent=HorizontalBox_22
   widget=Spacer_92
   slot {class=HorizontalBoxSlot, properties {size {sizeRule=Fill, value=0.925}}}
  }
  HorizontalBox_136 {
   widgetClassPath=&HorizontalBox
   parent=Overlay_0
   widget=HorizontalBox_136
   slot {class=OverlaySlot, properties {horizontalAlignment=HAlign_Fill, verticalAlignment=VAlign_Fill}}
  }
  Spacer {
   widgetClassPath=&Spacer
   parent=HorizontalBox_136
   widget=Spacer
   slot {class=HorizontalBoxSlot, properties {size {sizeRule=Fill, value=0.975}}}
  }
  Item_EntrySecondary {
   widgetClassPath=&FortMultiItemFundBriefEntry
   bInherited=true
   bIsVariable=true
   parent=HorizontalBox_136
   widget=Item_EntrySecondary
   properties {bSelected=false, cropDirection=1, cropOffset=-0.0192}
   slot {class=HorizontalBoxSlot, properties {size {sizeRule=Fill, value=1}}}
  }
  GlobalEffort {
   widgetClassPath=&Overlay
   parent=Overlay_15
   widget=GlobalEffort
   properties {visibility=Collapsed}
   slot {class=OverlaySlot, properties {horizontalAlignment=HAlign_Center, padding=<ltrb 8 32 0 0>}}
  }
  Image_1 {
   widgetClassPath=&Image
   bIsVariable=true
   parent=GlobalEffort
   widget=Image_1
   properties {
    visibility=SelfHitTestInvisible
    brush {
     bIsDynamicallyLoaded=false
     drawAs=Box
     imageType=NoImage
     mirroring=NoMirror
     resourceName=None
     resourceObject=/FortUILibrary/Foundation/Art_ToolKit/AngledBoxes/M_UI_FundCenter_AnnouncementBacking_Shadow
     tiling=NoTile
     imageSize=<128 32>
     margin=<ltrb 0.5 0 0.5 0>
     outlineSettings {
      bUseBrushTransparency=false
      roundingType=HalfHeightRadius
      width=0
      color {colorUseRule=UseColor_Specified, specifiedColor=<rgba 0 0 0 0>}
      cornerRadii=<0 0 0 0>
     }
     tintColor {colorUseRule=UseColor_Specified, specifiedColor=<rgba 1 1 1 1>}
     uVRegion {bIsValid=false, max=<0 0>, min=<0 0>}
    }
    renderTransform {angle=0, scale=<1.125 1.5>, shear=<-8 0>, translation=<0 8>}
   }
   slot {class=OverlaySlot, properties {horizontalAlignment=HAlign_Fill, verticalAlignment=VAlign_Fill}}
  }
  Image_93 {
   widgetClassPath=&Image
   bIsVariable=true
   parent=GlobalEffort
   widget=Image_93
   properties {
    brush {
     bIsDynamicallyLoaded=false
     drawAs=Box
     imageType=NoImage
     mirroring=NoMirror
     resourceName=None
     resourceObject=/FortUILibrary/Foundation/Art_ToolKit/AngledBoxes/M_UI_FundCenter_AnnouncementBacking
     tiling=NoTile
     imageSize=<128 32>
     margin=<ltrb 0.5 0 0.5 0>
     outlineSettings {
      bUseBrushTransparency=false
      roundingType=HalfHeightRadius
      width=0
      color {colorUseRule=UseColor_Specified, specifiedColor=<rgba 0 0 0 0>}
      cornerRadii=<0 0 0 0>
     }
     tintColor {colorUseRule=UseColor_Specified, specifiedColor=<rgba 1 1 1 1>}
     uVRegion {bIsValid=false, max=<0 0>, min=<0 0>}
    }
    renderTransform {angle=0, scale=<1 1>, shear=<-8 0>, translation=<0 0>}
   }
   slot {class=OverlaySlot, properties {horizontalAlignment=HAlign_Fill, verticalAlignment=VAlign_Fill}}
  }
  HorizontalBox_0 {
   widgetClassPath=&HorizontalBox
   parent=GlobalEffort
   widget=HorizontalBox_0
   slot {class=OverlaySlot, properties {horizontalAlignment=HAlign_Center, verticalAlignment=VAlign_Center, padding=<ltrb 18 16 20 16>}}
  }
  Image_229 {
   widgetClassPath=&Image
   bIsVariable=true
   parent=HorizontalBox_0
   widget=Image_229
   properties {
    renderOpacity=0.9
    brush {
     bIsDynamicallyLoaded=false
     drawAs=Image
     imageType=NoImage
     mirroring=NoMirror
     resourceName=None
     resourceObject=/Game/UI/Foundation/Textures/Icons/Stars/Stars-Full-64
     tiling=NoTile
     imageSize=<24 24>
     margin=<ltrb 0 0 0 0>
     outlineSettings {
      bUseBrushTransparency=false
      roundingType=HalfHeightRadius
      width=0
      color {colorUseRule=UseColor_Specified, specifiedColor=<rgba 0 0 0 0>}
      cornerRadii=<0 0 0 0>
     }
     tintColor {colorUseRule=UseColor_Specified, specifiedColor=<rgba 1 1 1 1>}
     uVRegion {bIsValid=false, max=<0 0>, min=<0 0>}
    }
    colorAndOpacity=<rgba 0.6939 0.0343 0.0242 1>
    renderTransform {angle=-8, scale=<1 1>, shear=<0 0>, translation=<0 0>}
   }
   slot {class=HorizontalBoxSlot, properties {horizontalAlignment=HAlign_Center, verticalAlignment=VAlign_Center}}
  }
  SizeBox_0 {
   widgetClassPath=&SizeBox
   parent=HorizontalBox_0
   widget=SizeBox_0
   properties {bOverride_MaxDesiredWidth=true, clipping=OnDemand, maxDesiredWidth=400, minDesiredWidth=400, widthOverride=400}
   slot {class=HorizontalBoxSlot, properties {verticalAlignment=VAlign_Center}}
  }
  ScaleBox_0 {
   widgetClassPath=&ScaleBox
   parent=SizeBox_0
   widget=ScaleBox_0
   properties {clipping=OnDemand, stretch=ScaleToFitX, stretchDirection=DownOnly}
   slot {class=SizeBoxSlot}
  }
  Text_TitleBar {
   widgetClassPath=&CommonTextBlock
   bIsVariable=true
   parent=ScaleBox_0
   widget=Text_TitleBar
   properties {
    clipping=OnDemand
    justification=Center
    lineHeightPercentage=0.9
    text="Contribute to the global effort!"
    textTransformPolicy=ToUpper
    colorAndOpacity {colorUseRule=UseColor_Specified, specifiedColor=<rgba 0.1195 0.1144 0.0545 1>}
    font {
     bForceMonospaced=false
     bMaterialIsStencil=false
     fontMaterial=null
     fontObject=&Burbank
     letterSpacing=0
     monospacedWidth=1
     size=15
     skewAmount=0
     typefaceFontName=BurbankRegular-Black
     outlineSettings {
      bApplyOutlineToDropShadows=true
      bMiteredCorners=false
      bSeparateFillAlpha=false
      outlineMaterial=null
      outlineSize=0
      outlineColor=<rgba 0.0144 0.0194 0.0065 0>
     }
    }
    margin=<ltrb 8 -2 8 0>
    renderTransform {angle=0, scale=<1 1>, shear=<-12 0>, translation=<0 0>}
    shadowOffset=<0 0>
   }
   slot {class=ScaleBoxSlot}
  }
  Image {
   widgetClassPath=&Image
   bIsVariable=true
   parent=HorizontalBox_0
   widget=Image
   properties {
    renderOpacity=0.9
    brush {
     bIsDynamicallyLoaded=false
     drawAs=Image
     imageType=NoImage
     mirroring=NoMirror
     resourceName=None
     resourceObject=&Stars-Full-64
     tiling=NoTile
     imageSize=<24 24>
     margin=<ltrb 0 0 0 0>
     outlineSettings {
      bUseBrushTransparency=false
      roundingType=HalfHeightRadius
      width=0
      color {colorUseRule=UseColor_Specified, specifiedColor=<rgba 0 0 0 0>}
      cornerRadii=<0 0 0 0>
     }
     tintColor {colorUseRule=UseColor_Specified, specifiedColor=<rgba 1 1 1 1>}
     uVRegion {bIsValid=false, max=<0 0>, min=<0 0>}
    }
    colorAndOpacity=<rgba 0.6939 0.0343 0.0242 1>
    renderTransform {angle=8, scale=<1 1>, shear=<0 0>, translation=<0 0>}
   }
   slot {class=HorizontalBoxSlot, properties {horizontalAlignment=HAlign_Center, verticalAlignment=VAlign_Center}}
  }
  Overlay_1 {
   widgetClassPath=&Overlay
   parent=Overlay_15
   widget=Overlay_1
   properties {renderTransform {angle=0, scale=<1 1>, shear=<0 0>, translation=<0 16>}}
   slot {class=OverlaySlot, properties {horizontalAlignment=HAlign_Center, verticalAlignment=VAlign_Center}}
  }
  Overlay_2 {
   widgetClassPath=&Overlay
   parent=Overlay_1
   widget=Overlay_2
   slot {class=OverlaySlot, properties {horizontalAlignment=HAlign_Center, verticalAlignment=VAlign_Center}}
  }
  Image_Shadow {
   widgetClassPath=&Image
   bIsVariable=true
   parent=Overlay_2
   widget=Image_Shadow
   properties {
    renderOpacity=0.5
    brush {
     bIsDynamicallyLoaded=false
     drawAs=Image
     imageType=NoImage
     mirroring=NoMirror
     resourceName=None
     resourceObject=/FortniteConversationUI/Conversation/Materials/M_UI_DonationFund_BadgeCircle_Default_Secondary
     tiling=NoTile
     imageSize=<168 168>
     margin=<ltrb 0 0 0 0>
     outlineSettings {
      bUseBrushTransparency=false
      roundingType=HalfHeightRadius
      width=0
      color {colorUseRule=UseColor_Specified, specifiedColor=<rgba 0 0 0 0>}
      cornerRadii=<0 0 0 0>
     }
     tintColor {colorUseRule=UseColor_Specified, specifiedColor=<rgba 1 1 1 1>}
     uVRegion {bIsValid=false, max=<0 0>, min=<0 0>}
    }
   }
   slot {class=OverlaySlot, properties {horizontalAlignment=HAlign_Fill, verticalAlignment=VAlign_Fill}}
  }
  Image_Burst {
   widgetClassPath=&Image
   bIsVariable=true
   parent=Overlay_2
   widget=Image_Burst
   properties {
    brush {
     bIsDynamicallyLoaded=false
     drawAs=Image
     imageType=NoImage
     mirroring=NoMirror
     resourceName=None
     resourceObject=/FortniteConversationUI/Conversation/Materials/M_UI_DonationFund_StripBacking_GreenBurst
     tiling=NoTile
     imageSize=<32 32>
     margin=<ltrb 0 0 0 0>
     outlineSettings {
      bUseBrushTransparency=false
      roundingType=HalfHeightRadius
      width=0
      color {colorUseRule=UseColor_Specified, specifiedColor=<rgba 0 0 0 0>}
      cornerRadii=<0 0 0 0>
     }
     tintColor {colorUseRule=UseColor_Specified, specifiedColor=<rgba 1 1 1 1>}
     uVRegion {bIsValid=false, max=<0 0>, min=<0 0>}
    }
    renderTransform {angle=0, scale=<0.5 0.5>, shear=<0 0>, translation=<0 0>}
   }
   slot {class=OverlaySlot, properties {horizontalAlignment=HAlign_Fill, verticalAlignment=VAlign_Fill}}
  }
  Image_Circle {
   widgetClassPath=&Image
   bIsVariable=true
   parent=Overlay_2
   widget=Image_Circle
   properties {brush {
     bIsDynamicallyLoaded=false
     drawAs=Image
     imageType=NoImage
     mirroring=NoMirror
     resourceName=None
     resourceObject=/FortniteConversationUI/Conversation/Materials/M_UI_DonationFund_BadgeCircle_Default
     tiling=NoTile
     imageSize=<112 112>
     margin=<ltrb 0 0 0 0>
     outlineSettings {
      bUseBrushTransparency=false
      roundingType=HalfHeightRadius
      width=0
      color {colorUseRule=UseColor_Specified, specifiedColor=<rgba 0 0 0 0>}
      cornerRadii=<0 0 0 0>
     }
     tintColor {colorUseRule=UseColor_Specified, specifiedColor=<rgba 1 1 1 1>}
     uVRegion {bIsValid=false, max=<0 0>, min=<0 0>}
    }}
   slot {class=OverlaySlot, properties {horizontalAlignment=HAlign_Center, verticalAlignment=VAlign_Center}}
  }
  AnimBoundFundedBorder {
   widgetClassPath=&Border
   parent=Overlay_2
   widget=AnimBoundFundedBorder
   properties {
    background {
     bIsDynamicallyLoaded=false
     drawAs=NoDrawType
     imageType=NoImage
     mirroring=NoMirror
     resourceName=None
     resourceObject=null
     tiling=NoTile
     imageSize=<32 32>
     margin=<ltrb 0 0 0 0>
     outlineSettings {
      bUseBrushTransparency=false
      roundingType=HalfHeightRadius
      width=0
      color {colorUseRule=UseColor_Specified, specifiedColor=<rgba 0 0 0 0>}
      cornerRadii=<0 0 0 0>
     }
     tintColor {colorUseRule=UseColor_Specified, specifiedColor=<rgba 1 1 1 1>}
     uVRegion {bIsValid=false, max=<0 0>, min=<0 0>}
    }
    padding=<ltrb 0 0 0 0>
   }
   slot {class=OverlaySlot, properties {horizontalAlignment=HAlign_Center, verticalAlignment=VAlign_Center}}
  }
  Image_Arrow {
   widgetClassPath=&Image
   bIsVariable=true
   parent=AnimBoundFundedBorder
   widget=Image_Arrow
   properties {
    brush {
     bIsDynamicallyLoaded=false
     drawAs=Image
     imageType=NoImage
     mirroring=NoMirror
     resourceName=None
     resourceObject=/FortniteCoreUI/UI/Login/Art/M_Arrow_DF
     tiling=NoTile
     imageSize=<90 90>
     margin=<ltrb 0 0 0 0>
     outlineSettings {
      bUseBrushTransparency=false
      roundingType=HalfHeightRadius
      width=0
      color {colorUseRule=UseColor_Specified, specifiedColor=<rgba 0 0 0 0>}
      cornerRadii=<0 0 0 0>
     }
     tintColor {colorUseRule=UseColor_Specified, specifiedColor=<rgba 1 1 1 1>}
     uVRegion {bIsValid=false, max=<0 0>, min=<0 0>}
    }
    colorAndOpacity=<rgba 0 0.9131 1 1>
   }
   slot {class=BorderSlot, properties {padding=<ltrb 0 0 0 0>}}
  }
  Image_Check {
   widgetClassPath=&Image
   bIsVariable=true
   parent=Overlay_2
   widget=Image_Check
   properties {
    renderOpacity=0
    visibility=SelfHitTestInvisible
    brush {
     bIsDynamicallyLoaded=false
     drawAs=Image
     imageType=NoImage
     mirroring=NoMirror
     resourceName=None
     resourceObject=/Game/UI/Foundation/Textures/Icons/Checks/T-Icon-Check-64
     tiling=NoTile
     imageSize=<64 64>
     margin=<ltrb 0 0 0 0>
     outlineSettings {
      bUseBrushTransparency=false
      roundingType=HalfHeightRadius
      width=0
      color {colorUseRule=UseColor_Specified, specifiedColor=<rgba 0 0 0 0>}
      cornerRadii=<0 0 0 0>
     }
     tintColor {colorUseRule=UseColor_Specified, specifiedColor=<rgba 1 1 1 1>}
     uVRegion {bIsValid=false, max=<0 0>, min=<0 0>}
    }
    colorAndOpacity=<rgba 0 0.7529 0.2502 1>
   }
   slot {class=OverlaySlot, properties {horizontalAlignment=HAlign_Center, verticalAlignment=VAlign_Center}}
  }
  CommonTextBlock_82 {
   widgetClassPath=&CommonTextBlock
   parent=Overlay_1
   widget=CommonTextBlock_82
   properties {
    justification=Center
    text=OR
    visibility=Collapsed
    colorAndOpacity {colorUseRule=UseColor_Specified, specifiedColor=<rgba 0.0144 0.0194 0.0065 1>}
    font {
     bForceMonospaced=false
     bMaterialIsStencil=false
     fontMaterial=null
     fontObject=&Burbank
     letterSpacing=0
     monospacedWidth=1
     size=32
     skewAmount=0
     typefaceFontName=BurbankRegular-Black
     outlineSettings {
      bApplyOutlineToDropShadows=false
      bMiteredCorners=false
      bSeparateFillAlpha=false
      outlineMaterial=null
      outlineSize=4
      outlineColor=<rgba 1 0.9375 0.625 1>
     }
    }
    renderTransform {angle=0, scale=<1 1>, shear=<-12 0>, translation=<0 88>}
    shadowOffset=<0 0>
   }
   slot {class=OverlaySlot, properties {horizontalAlignment=HAlign_Center, verticalAlignment=VAlign_Center}}
  }
  Item_EntryTertiary {widgetClassPath=/Script/FortniteConversationUI.FortItemFundBrief, bInherited=true, widget=null}
 ]
}
