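# B_VoteInfoEntry is a conversation UI widget that shows one vote entry as a row with a name, a material progress bar and a percentage.
# SetData stores the progress for GetCurrentProgress, drives the bar, and sets the name and the percent text, which rounds up.
# SetWinner rebuilds the value text's font to use the Black typeface for a winner and Medium otherwise.
# The event graph holds only unconnected stub events.
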WidgetBlueprint /FortniteConversationUI/Conversation/B_VoteInfoEntry
graphs [
 EventGraph {nodes [
   Event PreConstruct()
   Event Construct()
   Event Tick()
  ]}
 GetCurrentProgress {nodes [
   GetCurrentProgress() {is=title K2Node_FunctionEntry}
   > Return Node!(execute, CurrentProgress:double=Get CurrentProgress() -> (CurrentProgress:double))
  ]}
 SetData {nodes [
   n1 = SetData!() -> (then, Progress:double=~, Name:Text) {is=title K2Node_FunctionEntry}
   > n2 = Set CurrentProgress(CurrentProgress:double=@n1) -> (Output_Get:double)
   > SetProgress!(
    execute
    self:Fort Simple Material Progress Bar Object Reference=Get SimpleMaterialProgressBar() -> (SimpleMaterialProgressBar:Simple Material Progress Bar Object Reference)
    Progress:float=@n2
   ) -> (then)
   > "SetText (Text)"(
    self=Get EntryValue() -> (EntryValue:Common Text Object Reference)
    InText=As Percent!(
     Value:float=@n2
     RoundingMode:ERoundingMode Enum=ToPositiveInfinity
     bAlwaysSign:bool=false
     bUseGrouping:bool=true
     MinimumIntegralDigits:int=1
     MaximumIntegralDigits:int=324
     MinimumFractionalDigits:int=0
     MaximumFractionalDigits:int=2
    ) -> (ReturnValue:Text)
   )
   > "SetText (Text)"(self=Get EntryName() -> (EntryName:Common Text Object Reference), InText=@n1.Name)
  ]}
 SetWinner {nodes [
   n1 = SetWinner!() -> (then, IsWinner:bool=~) {is=title K2Node_FunctionEntry}
   n2 = Get!(self:Text Object Reference=Get EntryValue()) -> (
    Font_FontObject:Object Reference
    Font_FontMaterial:Object Reference
    Font_OutlineSettings:Font Outline Settings Structure
    Font_TypefaceFontName:Name
    Font_Size:float=~
    Font_LetterSpacing:int
    Font_SkewAmount:float=~
    Font_bForceMonospaced:bool=~
    Font_bMaterialIsStencil:bool=~
    Font_MonospacedWidth:float=~
   )
   > SetFont(
    self=Get EntryValue()
    InFontInfo=MakeSlateFontInfo!(
     FontObject:Object Reference=@n2
     FontMaterial:Object Reference=@n2.Font_FontMaterial
     OutlineSettings:Font Outline Settings Structure=@n2.Font_OutlineSettings
     TypefaceFontName:Name=Select(Option 0:Name=Medium, Option 1:Name=Black, Index=@n1) -> (ReturnValue:Name)
     Size:float=@n2.Font_Size
     LetterSpacing:int=@n2.Font_LetterSpacing
     SkewAmount:float=@n2.Font_SkewAmount
     bForceMonospaced:bool=false
     bMaterialIsStencil:bool=false
     MonospacedWidth:float=1
    ) -> (ReturnValue:Slate Font Info Structure)
   )
  ]}
]
variables [CurrentProgress {type=number}]
widgets {
 info {inheritedWidgetCount=0, parentClass=&UserWidget, rootWidgetClass=&HorizontalBox, widgetCount=8}
 widgets [
  HorizontalBox_0 {widgetClassPath=&HorizontalBox, widget=HorizontalBox_0}
  EntryName {
   widgetClassPath=&CommonTextBlock
   bIsVariable=true
   parent=HorizontalBox_0
   widget=EntryName
   properties {
    text=Text Block
    font {
     bForceMonospaced=false
     bMaterialIsStencil=false
     fontMaterial=null
     fontObject=&BurbankSmall
     letterSpacing=0
     monospacedWidth=1
     size=18
     skewAmount=0
     typefaceFontName=Bold
     outlineSettings {
      bApplyOutlineToDropShadows=false
      bMiteredCorners=false
      bSeparateFillAlpha=false
      outlineMaterial=null
      outlineSize=0
      outlineColor=<rgba 0 0 0 1>
     }
    }
    shadowOffset=<0 0>
   }
   slot {class=HorizontalBoxSlot, properties {horizontalAlignment=HAlign_Left}}
  }
  Spacer_160 {
   widgetClassPath=&Spacer
   parent=HorizontalBox_0
   widget=Spacer_160
   properties {size=<24 1>}
   slot {class=HorizontalBoxSlot, properties {horizontalAlignment=HAlign_Center}}
  }
  SizeBox_0 {
   widgetClassPath=&SizeBox
   parent=HorizontalBox_0
   widget=SizeBox_0
   properties {bOverride_HeightOverride=true, bOverride_WidthOverride=true, heightOverride=12, widthOverride=192}
   slot {class=HorizontalBoxSlot, properties {verticalAlignment=VAlign_Center, size {sizeRule=Fill, value=1}}}
  }
  SimpleMaterialProgressBar {
   widgetClassPath=/Game/Athena/HUD/SimpleMaterialProgressBar
   bIsVariable=true
   parent=SizeBox_0
   widget=SimpleMaterialProgressBar
   properties {
    bOverrideDefaultColorA=true
    bOverrideDefaultColorB=true
    bOverrideDefaultGlowEdge=true
    bOverrideDefaultSegments=true
    glowEdge=-0.25
    segments=1
    cachedColorA=<rgba 0 0.1131 1 1>
    cachedColorB=<rgba 0 0.1131 1 1>
    cachedColorBackground=<rgba 0 0 0 0.25>
   }
   slot {class=SizeBoxSlot}
  }
  Spacer {
   widgetClassPath=&Spacer
   parent=HorizontalBox_0
   widget=Spacer
   properties {size=<12 1>}
   slot {class=HorizontalBoxSlot, properties {horizontalAlignment=HAlign_Center}}
  }
  SizeBox_1 {
   widgetClassPath=&SizeBox
   parent=HorizontalBox_0
   widget=SizeBox_1
   properties {bOverride_MinDesiredWidth=true, minDesiredWidth=90}
   slot {class=HorizontalBoxSlot, properties {horizontalAlignment=HAlign_Right}}
  }
  EntryValue {
   widgetClassPath=&CommonTextBlock
   bIsVariable=true
   parent=SizeBox_1
   widget=EntryValue
   properties {
    text=99.99%
    font {
     bForceMonospaced=false
     bMaterialIsStencil=false
     fontMaterial=null
     fontObject=&BurbankSmall
     letterSpacing=0
     monospacedWidth=1
     size=18
     skewAmount=0
     typefaceFontName=Black
     outlineSettings {
      bApplyOutlineToDropShadows=false
      bMiteredCorners=false
      bSeparateFillAlpha=false
      outlineMaterial=null
      outlineSize=0
      outlineColor=<rgba 0 0 0 1>
     }
    }
    shadowOffset=<0 0>
   }
   slot {class=SizeBoxSlot, properties {horizontalAlignment=HAlign_Left}}
  }
 ]
}
