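# Style B tile for one item in a multi-item vote/funding conversation screen, built on FortItemFundBrief with a lazy-loaded thumbnail, spinner and progress percent.
# Initialization fills creator and event-time data, then broadcasts completion; progress updates refresh the percent text.
# Selection changes play thumbnail-material and content animations forward or reversed.
# The player controller's donation-success delegate triggers a shine animation and transaction-strip effect on the selected tile only.
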WidgetBlueprint /FortniteConversationUI/Conversation/FortMultiItemVoteBriefEntry_StyleB
animations [
 Anim_Selected {
  binding_count=1
  length_seconds=0.2
  object=&self:Anim_Selected
  sequence {
   bindings [LazyImage_Icon {
     id=#1
     possessed_class=FortLazyImage
     tracks [{
       type=MovieSceneWidgetMaterialTrack
       sections [{
         type=MovieSceneParameterSection
         end_frame=12
         start_frame=0
         channels [
          Grid Opacity {type=MovieSceneScriptingFloatChannel, thinned_from=3, keys [[0,0.15],[12.001,0]]}
          Overlay Opacity {type=MovieSceneScriptingFloatChannel, keys [[0,0.85],[12.001,0]]}
          Overlay Edge Opacity {type=MovieSceneScriptingFloatChannel, keys [[0,0.85],[12.001,0]]}
          Gradient Offset {type=MovieSceneScriptingFloatChannel, const [0,2]}
          Gradient Scale {type=MovieSceneScriptingFloatChannel, keys [[0,0.5],[12.001,0.25]]}
          Depth Opacity {type=MovieSceneScriptingFloatChannel, keys [[0,1],[12,0.9]]}
          Gradient Color BR {type=MovieSceneScriptingFloatChannel, const [0,2]}
          Gradient Color BG {type=MovieSceneScriptingFloatChannel, keys [[0,0.0212],[12,0.552]]}
          Gradient Color BB {type=MovieSceneScriptingFloatChannel, keys [[0,0.2502],[12,1]]}
          Gradient Color BA {type=MovieSceneScriptingFloatChannel, const [1,2]}
         ]
        }]
      }]
    }]
   display_rate {numerator=60}
  }
 }
 Anim_Selected_Content {
  binding_count=9
  length_seconds=0.3167
  object=&self:Anim_Selected_Content
  sequence {
   bindings [
    BlueprintPercentText {
     id=#2
     possessed_class=HorizontalBox
     tracks [{
       type=MovieSceneFloatTrack
       display_name=Render Opacity
       sections [{type=MovieSceneFloatSection, channels [None {type=MovieSceneScriptingFloatChannel, keys [[0,1],[3,0]]}]}]
      }]
    }
    Text_TitleBar {
     id=#3
     possessed_class=CommonTextBlock
     tracks [
      {
       type=MovieSceneWidgetMaterialTrack
       display_name=Font.Outline Settings.OutlineMaterial
       sections [{type=MovieSceneParameterSection, channels [AnimateOff {type=MovieSceneScriptingFloatChannel, keys [[0,0],[3,1]]}]}]
      }
      {
       type=MovieScene2DTransformTrack
       sections [{
         type=MovieScene2DTransformSection
         channels [
          Translation.X {type=MovieSceneScriptingFloatChannel, const [0,1]}
          Translation.Y {type=MovieSceneScriptingFloatChannel, keys [[0,45],[3,45],[5,25]]}
          Angle {type=MovieSceneScriptingFloatChannel, const [0,1]}
          Scale.X {type=MovieSceneScriptingFloatChannel, keys [[3,0.75],[5,1]]}
          Scale.Y {type=MovieSceneScriptingFloatChannel, keys [[3,0.75],[5,1]]}
          Shear.X {type=MovieSceneScriptingFloatChannel, const [-12,1]}
          Shear.Y {type=MovieSceneScriptingFloatChannel, const [0,1]}
         ]
        }]
      }
      {
       type=MovieSceneWidgetMaterialTrack
       display_name=Font.FontMaterial
       sections [{type=MovieSceneParameterSection, channels [AnimatetoFill {type=MovieSceneScriptingFloatChannel, keys [[0,0],[3,0],[4,1],[15,1]]}]}]
      }
      {
       type=MovieSceneFloatTrack
       display_name=Render Opacity
       sections [{type=MovieSceneFloatSection, channels [None {type=MovieSceneScriptingFloatChannel, keys [[0,1],[3,0],[4,0],[15,1]]}]}]
      }
      {
       type=MovieSceneColorTrack
       display_name=Color and Opacity
       sections [{
         type=MovieSceneColorSection
         end_frame=19
         start_frame=0
         channels [
          Color.R {type=MovieSceneScriptingFloatChannel, keys [[0,0],[19,1]]}
          Color.G {type=MovieSceneScriptingFloatChannel, keys [[0,0.0118],[19,1]]}
          Color.B {type=MovieSceneScriptingFloatChannel, keys [[0,0.0667],[19,1]]}
          Color.A {type=MovieSceneScriptingFloatChannel, const [1,2]}
         ]
        }]
      }
     ]
    }
    FundingPercent_Overlay {
     id=#4
     possessed_class=Overlay
     tracks [
      {
       type=MovieSceneFloatTrack
       display_name=Render Opacity
       sections [{type=MovieSceneFloatSection, channels [None {type=MovieSceneScriptingFloatChannel, keys [[10,0],[13,1]]}]}]
      }
      {
       type=MovieScene2DTransformTrack
       sections [{
         type=MovieScene2DTransformSection
         channels [
          Scale.X {type=MovieSceneScriptingFloatChannel, keys [[10,1.25],[13,0.95],[19,1]]}
          Scale.Y {type=MovieSceneScriptingFloatChannel, keys [[10,1.25],[13,0.95],[19,1]]}
         ]
        }]
      }
     ]
    }
    BottomStrip {
     id=#5
     possessed_class=Image
     tracks [
      {
       type=MovieSceneFloatTrack
       display_name=Render Opacity
       sections [{type=MovieSceneFloatSection, channels [None {type=MovieSceneScriptingFloatChannel, keys [[0,0],[12,1]]}]}]
      }
      {
       type=MovieScene2DTransformTrack
       sections [{
         type=MovieScene2DTransformSection
         channels [
          Scale.X {type=MovieSceneScriptingFloatChannel, keys [[0,0],[12,1]]}
          Scale.Y {type=MovieSceneScriptingFloatChannel, const [1,1]}
         ]
        }]
      }
     ]
    }
    FortMultiItemVoteBriefEntry_C_97 {
     id=#6
     possessed_class=FortMultiItemVoteBriefEntry_C
     tracks [{
       type=MovieScene2DTransformTrack
       sections [{
         type=MovieScene2DTransformSection
         channels [
          Scale.X {type=MovieSceneScriptingFloatChannel, keys [[3,0.9],[10,1.1]]}
          Scale.Y {type=MovieSceneScriptingFloatChannel, keys [[3,0.9],[10,1.1]]}
         ]
        }]
      }]
    }
    Ruler {
     id=#7
     possessed_class=Image
     tracks [{
       type=MovieSceneWidgetMaterialTrack
       sections [{type=MovieSceneParameterSection, channels [AnimateOff {type=MovieSceneScriptingFloatChannel, keys [[0,0],[3,1]]}]}]
      }]
    }
    FundStrip {
     id=#8
     possessed_class=SizeBox
     tracks [
      {
       type=MovieScene2DTransformTrack
       sections [{
         type=MovieScene2DTransformSection
         channels [
          Angle {type=MovieSceneScriptingFloatChannel, const [0,1]}
          Scale.X {type=MovieSceneScriptingFloatChannel, keys [[3,0.75],[9,1.05],[15.001,1]]}
          Scale.Y {type=MovieSceneScriptingFloatChannel, keys [[3,0.875],[9,1.05],[15.001,1]]}
          Shear.X {type=MovieSceneScriptingFloatChannel, const [0,1]}
          Shear.Y {type=MovieSceneScriptingFloatChannel, const [0,1]}
         ]
        }]
      }
      {
       type=MovieSceneFloatTrack
       display_name=Render Opacity
       sections [{type=MovieSceneFloatSection, channels [None {type=MovieSceneScriptingFloatChannel, keys [[3,0],[6,1]]}]}]
      }
     ]
    }
    ProgressInfo {
     id=#9
     possessed_class=VerticalBox
     tracks [
      {
       type=MovieScene2DTransformTrack
       sections [{
         type=MovieScene2DTransformSection
         channels [
          Translation.X {type=MovieSceneScriptingFloatChannel, const [0,1]}
          Translation.Y {type=MovieSceneScriptingFloatChannel, keys [[0,24],[12,0]]}
          Angle {type=MovieSceneScriptingFloatChannel, const [0,1]}
          Scale.X {type=MovieSceneScriptingFloatChannel, const [1,1]}
          Scale.Y {type=MovieSceneScriptingFloatChannel, const [1,1]}
          Shear.X {type=MovieSceneScriptingFloatChannel, const [0,1]}
          Shear.Y {type=MovieSceneScriptingFloatChannel, const [0,1]}
         ]
        }]
      }
      {
       type=MovieSceneFloatTrack
       display_name=Render Opacity
       sections [{type=MovieSceneFloatSection, end_frame=6, start_frame=0, channels [None {type=MovieSceneScriptingFloatChannel, keys [[0,0],[6,1]]}]}]
      }
     ]
    }
    Text_SecondLine_L {
     id=#10
     display_name=Text_TitleBar_1
     possessed_class=CommonTextBlock
     tracks [
      {
       type=MovieSceneWidgetMaterialTrack
       display_name=Font.Outline Settings.OutlineMaterial
       sections [{type=MovieSceneParameterSection, channels [AnimateOff {type=MovieSceneScriptingFloatChannel, keys [[0,0.25],[3,1]]}]}]
      }
      {
       type=MovieScene2DTransformTrack
       sections [{
         type=MovieScene2DTransformSection
         end_frame=19
         start_frame=0
         channels [
          Translation.X {type=MovieSceneScriptingFloatChannel, const [0,2]}
          Translation.Y {type=MovieSceneScriptingFloatChannel, keys [[0,25],[3,25],[5,0]]}
          Angle {type=MovieSceneScriptingFloatChannel, const [0,2]}
          Scale.X {type=MovieSceneScriptingFloatChannel, keys [[0,0.75],[3,0.75],[4,1]]}
          Scale.Y {type=MovieSceneScriptingFloatChannel, keys [[0,0.75],[3,0.75],[4,1]]}
          Shear.X {type=MovieSceneScriptingFloatChannel, const [-12,2]}
          Shear.Y {type=MovieSceneScriptingFloatChannel, const [0,2]}
         ]
        }]
      }
      {
       type=MovieSceneWidgetMaterialTrack
       display_name=Font.FontMaterial
       sections [{
         type=MovieSceneParameterSection
         channels [
          AnimatetoFill {type=MovieSceneScriptingFloatChannel, keys [[0,0],[3,0],[4,1]]}
          FillColorR {type=MovieSceneScriptingFloatChannel, keys [[0,0.025],[19,1]]}
          FillColorG {type=MovieSceneScriptingFloatChannel, keys [[0,0.8375],[19,1]]}
          FillColorB {type=MovieSceneScriptingFloatChannel, const [1,2]}
          FillColorA {type=MovieSceneScriptingFloatChannel, const [1,2]}
         ]
        }]
      }
      {
       type=MovieSceneFloatTrack
       display_name=Render Opacity
       sections [{type=MovieSceneFloatSection, channels [None {type=MovieSceneScriptingFloatChannel, keys [[0,1],[3,0],[4,0],[15,1]]}]}]
      }
      {
       type=MovieSceneColorTrack
       display_name=Color and Opacity
       sections [{
         type=MovieSceneColorSection
         end_frame=19
         start_frame=0
         channels [
          Color.R {type=MovieSceneScriptingFloatChannel, keys [[0,0],[19,1]]}
          Color.G {type=MovieSceneScriptingFloatChannel, keys [[0,0.0118],[19,1]]}
          Color.B {type=MovieSceneScriptingFloatChannel, keys [[0,0.0667],[19,1]]}
          Color.A {type=MovieSceneScriptingFloatChannel, const [1,2]}
         ]
        }]
      }
     ]
    }
   ]
   display_rate {numerator=60}
  }
 }
 Anim_Donated {
  binding_count=3
  length_seconds=0.75
  object=&self:Anim_Donated
  sequence {
   bindings [
    Border_0 {
     id=#11
     possessed_class=Border
     tracks [{
       type=MovieSceneWidgetMaterialTrack
       sections [{
         type=MovieSceneParameterSection
         channels [
          BackingAlpha {type=MovieSceneScriptingFloatChannel, keys [[0,0.25],[6,0.75],[30,0.333]]}
          OffsetShine {type=MovieSceneScriptingFloatChannel, keys [[0,0],[45,1.475]]}
         ]
        }]
      }]
    }
    ProgressInfo {
     id=#12
     possessed_class=VerticalBox
     tracks [{
       type=MovieScene2DTransformTrack
       sections [{
         type=MovieScene2DTransformSection
         channels [
          Scale.X {type=MovieSceneScriptingFloatChannel, keys [[0.001,1],[3,0.95],[8,1.15],[20,1]]}
          Scale.Y {type=MovieSceneScriptingFloatChannel, keys [[0.001,1],[3,0.95],[8,1.15],[20,1]]}
         ]
        }]
      }]
    }
    LazyImage_Icon {
     id=#13
     possessed_class=FortLazyImage
     tracks [{
       type=MovieSceneWidgetMaterialTrack
       sections [{
         type=MovieSceneParameterSection
         end_frame=45
         start_frame=0
         channels [
          Overlay Opacity {type=MovieSceneScriptingFloatChannel, keys [[0.001,0],[6,0.25],[24,0]]}
          Overlay Edge Opacity {type=MovieSceneScriptingFloatChannel, keys [[0.001,0],[6,1],[30,0]]}
          Gradient Scale {type=MovieSceneScriptingFloatChannel, keys [[0.001,0.25],[6,1],[24,0.25]]}
          Depth Opacity {type=MovieSceneScriptingFloatChannel, keys [[0,0.9],[6,1],[24,0.9]]}
          Gradient Color AR {type=MovieSceneScriptingFloatChannel, const [0,3]}
          Gradient Color AG {type=MovieSceneScriptingFloatChannel, keys [[0,0.0196],[6,0.549],[24,0.0194]]}
          Gradient Color AB {type=MovieSceneScriptingFloatChannel, keys [[0,0.251],[6,1],[24,0.2502]]}
          Gradient Color AA {type=MovieSceneScriptingFloatChannel, const [1,3]}
          Depth Gradient ColorR {type=MovieSceneScriptingFloatChannel, const [0,3]}
          Depth Gradient ColorG {type=MovieSceneScriptingFloatChannel, keys [[0,0.0423],[6,0.552],[24,0.0423]]}
          Depth Gradient ColorB {type=MovieSceneScriptingFloatChannel, keys [[0,0.5029],[6,1],[24,0.5029]]}
          Depth Gradient ColorA {type=MovieSceneScriptingFloatChannel, const [1,3]}
         ]
        }]
      }]
    }
   ]
   display_rate {numerator=60}
  }
 }
 Anim_Selected_Content_Right {
  binding_count=9
  length_seconds=0.5
  object=&self:Anim_Selected_Content_Right
  sequence {
   bindings [
    BlueprintPercentText {
     id=#2
     possessed_class=HorizontalBox
     tracks [{
       type=MovieSceneFloatTrack
       display_name=Render Opacity
       sections [{type=MovieSceneFloatSection, channels [None {type=MovieSceneScriptingFloatChannel, keys [[0,1],[3,0]]}]}]
      }]
    }
    Text_TitleBar {
     id=#3
     possessed_class=CommonTextBlock
     tracks [
      {
       type=MovieSceneWidgetMaterialTrack
       display_name=Font.Outline Settings.OutlineMaterial
       sections [{type=MovieSceneParameterSection, channels [AnimateOff {type=MovieSceneScriptingFloatChannel, keys [[0,0],[2,1]]}]}]
      }
      {
       type=MovieScene2DTransformTrack
       sections [{
         type=MovieScene2DTransformSection
         channels [
          Translation.X {type=MovieSceneScriptingFloatChannel, const [0,1]}
          Translation.Y {type=MovieSceneScriptingFloatChannel, keys [[0,45],[2,45],[3,25]]}
          Angle {type=MovieSceneScriptingFloatChannel, const [0,1]}
          Scale.X {type=MovieSceneScriptingFloatChannel, keys [[2,0.75],[3,1]]}
          Scale.Y {type=MovieSceneScriptingFloatChannel, keys [[2,0.75],[3,1]]}
          Shear.X {type=MovieSceneScriptingFloatChannel, const [-12,1]}
          Shear.Y {type=MovieSceneScriptingFloatChannel, const [0,1]}
         ]
        }]
      }
      {
       type=MovieSceneWidgetMaterialTrack
       display_name=Font.FontMaterial
       sections [{type=MovieSceneParameterSection, channels [AnimatetoFill {type=MovieSceneScriptingFloatChannel, keys [[0,0],[3,1]]}]}]
      }
      {
       type=MovieSceneFloatTrack
       display_name=Render Opacity
       sections [{type=MovieSceneFloatSection, channels [None {type=MovieSceneScriptingFloatChannel, keys [[0,1],[2,0],[3,0],[30,1]]}]}]
      }
      {
       type=MovieSceneColorTrack
       display_name=Color and Opacity
       sections [{
         type=MovieSceneColorSection
         end_frame=19
         start_frame=0
         channels [
          Color.R {type=MovieSceneScriptingFloatChannel, keys [[0,0],[19,1]]}
          Color.G {type=MovieSceneScriptingFloatChannel, keys [[0,0.0118],[19,1]]}
          Color.B {type=MovieSceneScriptingFloatChannel, keys [[0,0.0667],[19,1]]}
          Color.A {type=MovieSceneScriptingFloatChannel, const [1,2]}
         ]
        }]
      }
     ]
    }
    FundingPercent_Overlay {
     id=#4
     possessed_class=Overlay
     tracks [
      {
       type=MovieSceneFloatTrack
       display_name=Render Opacity
       sections [{type=MovieSceneFloatSection, channels [None {type=MovieSceneScriptingFloatChannel, keys [[10,0],[13,1]]}]}]
      }
      {
       type=MovieScene2DTransformTrack
       sections [{
         type=MovieScene2DTransformSection
         channels [
          Scale.X {type=MovieSceneScriptingFloatChannel, keys [[10,1.25],[13,0.95],[18,1]]}
          Scale.Y {type=MovieSceneScriptingFloatChannel, keys [[10,1.25],[13,0.95],[18,1]]}
         ]
        }]
      }
     ]
    }
    BottomStrip {
     id=#5
     possessed_class=Image
     tracks [
      {
       type=MovieSceneFloatTrack
       display_name=Render Opacity
       sections [{type=MovieSceneFloatSection, channels [None {type=MovieSceneScriptingFloatChannel, keys [[0,0],[18.001,1]]}]}]
      }
      {
       type=MovieScene2DTransformTrack
       sections [{
         type=MovieScene2DTransformSection
         channels [
          Scale.X {type=MovieSceneScriptingFloatChannel, keys [[0,0],[12,1]]}
          Scale.Y {type=MovieSceneScriptingFloatChannel, const [1,1]}
         ]
        }]
      }
     ]
    }
    Text_SecondLine_L {
     id=#14
     possessed_class=CommonTextBlock
     tracks [
      {
       type=MovieSceneWidgetMaterialTrack
       display_name=Font.FontMaterial
       sections [
        {type=MovieSceneParameterSection, channels [AnimatetoFill {type=MovieSceneScriptingFloatChannel, keys [[0,0],[3,1]]}]}
        {
         type=MovieSceneParameterSection
         channels [
          AnimatetoFill {type=MovieSceneScriptingFloatChannel, keys [[0,0],[3,0],[4,1]]}
          FillColorR {type=MovieSceneScriptingFloatChannel, keys [[0,0.025],[19,1]]}
          FillColorG {type=MovieSceneScriptingFloatChannel, keys [[0,0.8375],[19,1]]}
          FillColorB {type=MovieSceneScriptingFloatChannel, const [1,2]}
          FillColorA {type=MovieSceneScriptingFloatChannel, const [1,2]}
         ]
        }
       ]
      }
      {
       type=MovieSceneWidgetMaterialTrack
       display_name=Font.Outline Settings.OutlineMaterial
       sections [{type=MovieSceneParameterSection, channels [AnimateOff {type=MovieSceneScriptingFloatChannel, keys [[0,0.25],[2,1]]}]}]
      }
      {
       type=MovieSceneFloatTrack
       display_name=Render Opacity
       sections [{type=MovieSceneFloatSection, channels [None {type=MovieSceneScriptingFloatChannel, keys [[0,1],[2,0],[3,0],[30,1]]}]}]
      }
      {
       type=MovieScene2DTransformTrack
       sections [{
         type=MovieScene2DTransformSection
         channels [
          Translation.X {type=MovieSceneScriptingFloatChannel, const [0,1]}
          Translation.Y {type=MovieSceneScriptingFloatChannel, keys [[2,25],[3,0]]}
          Angle {type=MovieSceneScriptingFloatChannel, const [0,1]}
          Scale.X {type=MovieSceneScriptingFloatChannel, keys [[2,0.75],[3,1]]}
          Scale.Y {type=MovieSceneScriptingFloatChannel, keys [[2,0.75],[3,1]]}
          Shear.X {type=MovieSceneScriptingFloatChannel, const [-12,1]}
          Shear.Y {type=MovieSceneScriptingFloatChannel, const [0,1]}
         ]
        }]
      }
      {
       type=MovieSceneColorTrack
       display_name=Color and Opacity
       sections [{
         type=MovieSceneColorSection
         end_frame=19
         start_frame=0
         channels [
          Color.R {type=MovieSceneScriptingFloatChannel, keys [[0,0],[19,1]]}
          Color.G {type=MovieSceneScriptingFloatChannel, keys [[0,0.0118],[19,1]]}
          Color.B {type=MovieSceneScriptingFloatChannel, keys [[0,0.0667],[19,1]]}
          Color.A {type=MovieSceneScriptingFloatChannel, const [1,2]}
         ]
        }]
      }
     ]
    }
    Ruler {
     id=#7
     possessed_class=Image
     tracks [{
       type=MovieSceneWidgetMaterialTrack
       sections [{type=MovieSceneParameterSection, channels [AnimateOff {type=MovieSceneScriptingFloatChannel, keys [[0,0],[3,1]]}]}]
      }]
    }
    FundStrip {
     id=#8
     possessed_class=SizeBox
     tracks [
      {
       type=MovieSceneFloatTrack
       display_name=Render Opacity
       sections [{type=MovieSceneFloatSection, channels [None {type=MovieSceneScriptingFloatChannel, keys [[2,0],[8,1]]}]}]
      }
      {
       type=MovieScene2DTransformTrack
       sections [{
         type=MovieScene2DTransformSection
         channels [
          Translation.X {type=MovieSceneScriptingFloatChannel, keys [[2,-319.9999],[17,0]]}
          Scale.X {type=MovieSceneScriptingFloatChannel, keys [[2,0.9],[17,1]]}
          Scale.Y {type=MovieSceneScriptingFloatChannel, keys [[2,0.9],[17,1]]}
         ]
        }]
      }
     ]
    }
    ProgressInfo {
     id=#9
     possessed_class=VerticalBox
     tracks [
      {
       type=MovieScene2DTransformTrack
       sections [{
         type=MovieScene2DTransformSection
         channels [
          Translation.X {type=MovieSceneScriptingFloatChannel, const [0,1]}
          Translation.Y {type=MovieSceneScriptingFloatChannel, keys [[0,24],[12,0]]}
          Angle {type=MovieSceneScriptingFloatChannel, const [0,1]}
          Scale.X {type=MovieSceneScriptingFloatChannel, const [1,1]}
          Scale.Y {type=MovieSceneScriptingFloatChannel, const [1,1]}
          Shear.X {type=MovieSceneScriptingFloatChannel, const [0,1]}
          Shear.Y {type=MovieSceneScriptingFloatChannel, const [0,1]}
         ]
        }]
      }
      {
       type=MovieSceneFloatTrack
       display_name=Render Opacity
       sections [{type=MovieSceneFloatSection, end_frame=6, start_frame=0, channels [None {type=MovieSceneScriptingFloatChannel, keys [[0,0],[6,1]]}]}]
      }
     ]
    }
   ]
   display_rate {numerator=60}
  }
 }
 Anim_Selected_Content_Left {
  binding_count=8
  length_seconds=0.5
  object=&self:Anim_Selected_Content_Left
  sequence {
   bindings [
    BlueprintPercentText {
     id=#2
     possessed_class=HorizontalBox
     tracks [{
       type=MovieSceneFloatTrack
       display_name=Render Opacity
       sections [{type=MovieSceneFloatSection, channels [None {type=MovieSceneScriptingFloatChannel, keys [[0,1],[3,0]]}]}]
      }]
    }
    Text_TitleBar {
     id=#3
     possessed_class=CommonTextBlock
     tracks [
      {
       type=MovieSceneWidgetMaterialTrack
       display_name=Font.Outline Settings.OutlineMaterial
       sections [{type=MovieSceneParameterSection, channels [AnimateOff {type=MovieSceneScriptingFloatChannel, keys [[0,0],[2,1]]}]}]
      }
      {
       type=MovieScene2DTransformTrack
       sections [{
         type=MovieScene2DTransformSection
         channels [
          Translation.X {type=MovieSceneScriptingFloatChannel, const [0,1]}
          Translation.Y {type=MovieSceneScriptingFloatChannel, keys [[2,45],[3,25]]}
          Angle {type=MovieSceneScriptingFloatChannel, const [0,1]}
          Scale.X {type=MovieSceneScriptingFloatChannel, keys [[2,0.75],[3,1]]}
          Scale.Y {type=MovieSceneScriptingFloatChannel, keys [[2,0.75],[3,1]]}
          Shear.X {type=MovieSceneScriptingFloatChannel, const [-12,1]}
          Shear.Y {type=MovieSceneScriptingFloatChannel, const [0,1]}
         ]
        }]
      }
      {
       type=MovieSceneWidgetMaterialTrack
       display_name=Font.FontMaterial
       sections [{type=MovieSceneParameterSection, channels [AnimatetoFill {type=MovieSceneScriptingFloatChannel, keys [[0,0],[3,1]]}]}]
      }
      {
       type=MovieSceneFloatTrack
       display_name=Render Opacity
       sections [{type=MovieSceneFloatSection, channels [None {type=MovieSceneScriptingFloatChannel, keys [[0,1],[2,0],[3,0],[30,1]]}]}]
      }
      {
       type=MovieSceneColorTrack
       display_name=Color and Opacity
       sections [{
         type=MovieSceneColorSection
         end_frame=19
         start_frame=0
         channels [
          Color.R {type=MovieSceneScriptingFloatChannel, keys [[0,0],[19,1]]}
          Color.G {type=MovieSceneScriptingFloatChannel, keys [[0,0.0118],[19,1]]}
          Color.B {type=MovieSceneScriptingFloatChannel, keys [[0,0.0667],[19,1]]}
          Color.A {type=MovieSceneScriptingFloatChannel, const [1,2]}
         ]
        }]
      }
     ]
    }
    Text_SecondLine_L {
     id=#15
     possessed_class=CommonTextBlock
     tracks [
      {
       type=MovieSceneWidgetMaterialTrack
       display_name=Font.FontMaterial
       sections [
        {type=MovieSceneParameterSection, channels [AnimatetoFill {type=MovieSceneScriptingFloatChannel, keys [[0,0],[3,1]]}]}
        {
         type=MovieSceneParameterSection
         channels [
          AnimatetoFill {type=MovieSceneScriptingFloatChannel, keys [[0,0],[3,0],[4,1]]}
          FillColorR {type=MovieSceneScriptingFloatChannel, keys [[0,0.025],[19,1]]}
          FillColorG {type=MovieSceneScriptingFloatChannel, keys [[0,0.8375],[19,1]]}
          FillColorB {type=MovieSceneScriptingFloatChannel, const [1,2]}
          FillColorA {type=MovieSceneScriptingFloatChannel, const [1,2]}
         ]
        }
       ]
      }
      {
       type=MovieSceneWidgetMaterialTrack
       display_name=Font.Outline Settings.OutlineMaterial
       sections [{type=MovieSceneParameterSection, channels [AnimateOff {type=MovieSceneScriptingFloatChannel, keys [[0,0.25],[2,1]]}]}]
      }
      {
       type=MovieSceneFloatTrack
       display_name=Render Opacity
       sections [{type=MovieSceneFloatSection, channels [None {type=MovieSceneScriptingFloatChannel, keys [[0,1],[2,0],[3,0],[30,1]]}]}]
      }
      {
       type=MovieScene2DTransformTrack
       sections [{
         type=MovieScene2DTransformSection
         channels [
          Translation.X {type=MovieSceneScriptingFloatChannel, const [0,1]}
          Translation.Y {type=MovieSceneScriptingFloatChannel, keys [[2,25],[3,0]]}
          Angle {type=MovieSceneScriptingFloatChannel, const [0,1]}
          Scale.X {type=MovieSceneScriptingFloatChannel, keys [[2,0.75],[3,1]]}
          Scale.Y {type=MovieSceneScriptingFloatChannel, keys [[2,0.75],[3,1]]}
          Shear.X {type=MovieSceneScriptingFloatChannel, const [-12,1]}
          Shear.Y {type=MovieSceneScriptingFloatChannel, const [0,1]}
         ]
        }]
      }
      {
       type=MovieSceneColorTrack
       display_name=Color and Opacity
       sections [{
         type=MovieSceneColorSection
         end_frame=19
         start_frame=0
         channels [
          Color.R {type=MovieSceneScriptingFloatChannel, keys [[0,0],[19,1]]}
          Color.G {type=MovieSceneScriptingFloatChannel, keys [[0,0.0118],[19,1]]}
          Color.B {type=MovieSceneScriptingFloatChannel, keys [[0,0.0667],[19,1]]}
          Color.A {type=MovieSceneScriptingFloatChannel, const [1,2]}
         ]
        }]
      }
     ]
    }
    FundingPercent_Overlay {
     id=#4
     possessed_class=Overlay
     tracks [
      {
       type=MovieSceneFloatTrack
       display_name=Render Opacity
       sections [{type=MovieSceneFloatSection, channels [None {type=MovieSceneScriptingFloatChannel, keys [[10,0],[13,1]]}]}]
      }
      {
       type=MovieScene2DTransformTrack
       sections [{
         type=MovieScene2DTransformSection
         channels [
          Scale.X {type=MovieSceneScriptingFloatChannel, keys [[10,1.25],[13,0.95],[18,1]]}
          Scale.Y {type=MovieSceneScriptingFloatChannel, keys [[10,1.25],[13,0.95],[18,1]]}
         ]
        }]
      }
     ]
    }
    BottomStrip {
     id=#5
     possessed_class=Image
     tracks [
      {
       type=MovieSceneFloatTrack
       display_name=Render Opacity
       sections [{type=MovieSceneFloatSection, channels [None {type=MovieSceneScriptingFloatChannel, keys [[0,0],[18,1]]}]}]
      }
      {
       type=MovieScene2DTransformTrack
       sections [{
         type=MovieScene2DTransformSection
         channels [
          Scale.X {type=MovieSceneScriptingFloatChannel, keys [[0,0],[12,1]]}
          Scale.Y {type=MovieSceneScriptingFloatChannel, const [1,1]}
         ]
        }]
      }
     ]
    }
    Ruler {
     id=#7
     possessed_class=Image
     tracks [{
       type=MovieSceneWidgetMaterialTrack
       sections [{type=MovieSceneParameterSection, channels [AnimateOff {type=MovieSceneScriptingFloatChannel, keys [[0,0],[3,1]]}]}]
      }]
    }
    FundStrip {
     id=#8
     possessed_class=SizeBox
     tracks [
      {
       type=MovieSceneFloatTrack
       display_name=Render Opacity
       sections [{type=MovieSceneFloatSection, channels [None {type=MovieSceneScriptingFloatChannel, keys [[2,0],[8,1]]}]}]
      }
      {
       type=MovieScene2DTransformTrack
       sections [{
         type=MovieScene2DTransformSection
         channels [
          Translation.X {type=MovieSceneScriptingFloatChannel, keys [[2,319.9999],[17,0]]}
          Scale.X {type=MovieSceneScriptingFloatChannel, keys [[2,0.9],[17,1]]}
          Scale.Y {type=MovieSceneScriptingFloatChannel, keys [[2,0.9],[17,1]]}
         ]
        }]
      }
     ]
    }
    ProgressInfo {
     id=#9
     possessed_class=VerticalBox
     tracks [
      {
       type=MovieScene2DTransformTrack
       sections [{
         type=MovieScene2DTransformSection
         channels [
          Translation.X {type=MovieSceneScriptingFloatChannel, const [0,1]}
          Translation.Y {type=MovieSceneScriptingFloatChannel, keys [[0,24],[12,0]]}
          Angle {type=MovieSceneScriptingFloatChannel, const [0,1]}
          Scale.X {type=MovieSceneScriptingFloatChannel, const [1,1]}
          Scale.Y {type=MovieSceneScriptingFloatChannel, const [1,1]}
          Shear.X {type=MovieSceneScriptingFloatChannel, const [0,1]}
          Shear.Y {type=MovieSceneScriptingFloatChannel, const [0,1]}
         ]
        }]
      }
      {
       type=MovieSceneFloatTrack
       display_name=Render Opacity
       sections [{type=MovieSceneFloatSection, end_frame=6, start_frame=0, channels [None {type=MovieSceneScriptingFloatChannel, keys [[0,0],[6,1]]}]}]
      }
     ]
    }
   ]
   display_rate {numerator=60}
  }
 }
]
graphs [
 EventGraph {nodes [
   n1 = OnLazyImageLoadingStateChanged!() -> (OutputDelegate, then, bIsLoading:bool=~) {is=title K2Node_CustomEvent}
   > SetVisibility(
    self=Get LazyImage_Spinner() -> (LazyImage_Spinner:Image Object Reference)
    InVisibility=Select(Option 0=Collapsed, Option 1=SelfHitTestInvisible, Index=@n1.bIsLoading)
   )
   Event Construct()
   > n2 = Cast To Athena_PlayerController!(execute, Object=GetOwningPlayer()) -> (then, CastFailed, AsAthena Player Controller:Athena Player Controller Object Reference) {is=title K2Node_DynamicCast}
   > Bind Event to On Client Donation Success Delegate(self:Fort Player Controller Athena Object Reference=@n2, Delegate=Create Event()) {is=title K2Node_AddDelegate}
   n3 = Event OnProgressUpdated!() -> (OutputDelegate, then, CurrentFundingFraction:float) {is=title K2Node_Event}
   > n4 = Set CurrentFundingProgress(CurrentFundingProgress:double=@n3.CurrentFundingFraction) -> (Output_Get:double)
   > "SetText (Text)"(
    self=Get Text_ProgressPercent_2() -> (Text_ProgressPercent_2:Common Text Object Reference)
    InText=As Percent!(
     Value:float=@n4
     RoundingMode:ERoundingMode Enum=ToPositiveInfinity
     bAlwaysSign:bool=false
     bUseGrouping:bool=true
     MinimumIntegralDigits:int=1
     MaximumIntegralDigits:int=324
     MinimumFractionalDigits:int=0
     MaximumFractionalDigits:int=2
    ) -> (ReturnValue:Text)
   )
   Event PreConstruct()
   > n5 = GetDynamicMaterial(self=Get BottomStrip() -> (BottomStrip:Image Object Reference))
   > SetScalarParameterValue(self=@n5, ParameterName=FlipCrop, Value=Get CropDirection() -> (CropDirection:double))
   > SetScalarParameterValue(self=@n5, ParameterName=OffsetCrop, Value=Get CropOffset() -> (CropOffset:double))
   > Set IsLeftSideSelection(IsLeftSideSelection:bool=Get IsLeftSideSelection() -> (IsLeftSideSelection:bool)) -> (Output_Get:bool)
   Event Destruct()
   > n6 = Cast To Athena_PlayerController(Object=GetOwningPlayer())
   > Unbind Event from On Client Donation Success Delegate(self:Fort Player Controller Athena Object Reference=@n6, Delegate=Create Event()) {is=title K2Node_RemoveDelegate}
   Event OnIntializationComplete() {is=title K2Node_Event}
   > n7 = GetDataForKey!(execute, self:Fort Item Fund Brief Object Reference, Key:String=CreatorName_, bUseIndexAdjustment:bool=true) -> (then, ReturnValue:Text)
   > "SetText (Text)"(self=Get Text_SecondLine_R() -> (Text_SecondLine_R:Common Text Object Reference), InText=@n7)
   > n8 = GetDataForKey(Key=CalendarEvent, bUseIndexAdjustment=false)
   > Set EventTimeRemaining(EventTimeRemaining:Text=@n8) -> (Output_Get:Text)
   > Call On Initialization Completed!(
    execute:Exec
    self:Fort Multi Item Vote Brief Entry Style B Object Reference
    VoteEntry:Fort Multi Item Vote Brief Entry Style B Object Reference=Self-Reference()
   ) -> (then:Exec) {is=title K2Node_CallDelegate}
   Event OnInitialized()
   > Assign On Loading State Changed!(
    execute:Exec
    self:Common Lazy Image Object Reference=Get LazyImage_Icon() -> (LazyImage_Icon:Fort Lazy Image Object Reference)
    Delegate:"Delegate (by ref)"=@n1
   ) -> (then:Exec) {is=title K2Node_AssignDelegate}
   > SetMaterialTextureParamName!(execute, self:Common Lazy Image Object Reference=Get LazyImage_Icon(), TextureParamName:Name=Thumbnail) -> (then)
  ]}
 GetTileData {nodes [
   GetTileData() {is=title K2Node_FunctionEntry}
   > Return Node!(
    execute
    CurrentProgress:double=Get CurrentFundingProgress() -> (CurrentFundingProgress:double)
    ItemName:Text="GetText (Text)"!(self:Text Object Reference=Get Text_TitleBar() -> (Text_TitleBar:Common Text Object Reference)) -> (ReturnValue:Text)
   )
  ]}
 OnDonateSuccess {nodes [
   OnDonateSuccess() {is=title K2Node_FunctionEntry}
   > Branch(Condition=Get bSelected() -> (bSelected:bool))
   > PlayAnimation(InAnimation=Get Anim_Donated() -> (Anim_Donated:Widget Animation Object Reference))
   > n1 = Cast To FortTransactionStripNew!(execute, Object=Get TransactionStrip_Main() -> (TransactionStrip_Main:Fort Transaction Strip Object Reference)) -> (then, CastFailed, AsFort Transaction Strip New:Fort Transaction Strip New Object Reference) {is=title K2Node_DynamicCast}
   > DonationAnim(self:Fort Transaction Strip New Object Reference=@n1)
  ]}
 OnInitializationCompleted {nodes [OnInitializationCompleted!() -> (then, VoteEntry:Fort Multi Item Vote Brief Entry Style B Object Reference) {is=title K2Node_FunctionEntry}]}
 ResetAnimStates {nodes [
   ResetAnimStates() {is=title K2Node_FunctionEntry}
   n1 = Get Anim_Selected() -> (Anim_Selected:Widget Animation Object Reference)
   > PlayAnimation(InAnimation=@n1, StartAtTime=GetEndTime(self=@n1), PlayMode=Reverse)
   n2 = Get Anim_Selected_Content() -> (Anim_Selected_Content:Widget Animation Object Reference)
   > PlayAnimation(InAnimation=@n2, StartAtTime=GetEndTime(self=@n2), PlayMode=Reverse)
  ]}
 SetIntroState {nodes [
   n1 = SetIntroState!() -> (then, Selected:bool=~) {is=title K2Node_FunctionEntry}
   > Branch(Condition="Not Equal (Boolean)"!(A:bool=@n1, B:bool=Get bSelected()) -> (ReturnValue:bool) {is=title K2Node_CallFunction})
   > n2 = Set bSelected(bSelected:bool=false) -> (Output_Get:bool)
   > n3 = Branch(Condition=@n2)
   n4 = Get Anim_Selected()
   > PlayAnimation(InAnimation=@n4, StartAtTime=GetEndTime(self=@n4), PlaybackSpeed=10)
   n5 = Get Anim_Selected()
   PlayAnimation(execute=@n3.else, InAnimation=@n5, StartAtTime=GetEndTime(self=@n5), PlayMode=Reverse, PlaybackSpeed=10)
  ]}
 SetSelectedVisuals {nodes [
   n1 = SetSelectedVisuals!() -> (then, InSelected:bool=~) {is=title K2Node_FunctionEntry}
   > Branch(Condition="Not Equal (Boolean)"(A=@n1, B=Get bSelected()) -> (ReturnValue=~) {is=title K2Node_PromotableOperator})
   > n2 = Set bSelected(bSelected=@n1)
   > n3 = Branch(Condition=@n2)
   > PlayAnimationForward(InAnimation=Get Anim_Selected())
   > PlayAnimationForward(InAnimation=Get Anim_Selected_Content())
   n4 = Get Anim_Selected_Content()
   PlayAnimation(execute=@n3.else, InAnimation=@n4, StartAtTime=GetEndTime(self=@n4), PlayMode=Reverse)
   > PlayAnimationReverse(InAnimation=Get Anim_Selected(), PlaybackSpeed=2)
  ]}
]
variables [
 bSelected {type=boolean, default=true}
 CropDirection {type=number}
 CropOffset {type=number}
 CurrentFundingProgress {type=number}
 DebugItem {type=FortItem, default=null}
 EventTimeRemaining {type=string, default=""}
 FlipShadow {type=number}
 IsLeftSideSelection {type=boolean}
 OffsetShadow {type=number}
 OnInitializationCompleted {}
]
widgets {
 info {inheritedWidgetCount=6, parentClass=/Script/FortniteConversationUI.FortItemFundBrief, rootWidgetClass=&ScaleBox, widgetCount=33}
 widgets [
  ItemScaleBox {widgetClassPath=&ScaleBox, bIsVariable=true, widget=ItemScaleBox, properties {stretch=UserSpecified}}
  DisplayOverlay {
   widgetClassPath=&Overlay
   bIsVariable=true
   parent=ItemScaleBox
   widget=DisplayOverlay
   slot {class=ScaleBoxSlot, properties {horizontalAlignment=HAlign_Fill, verticalAlignment=VAlign_Fill}}
  }
  LazyImage_Icon {
   widgetClassPath=&FortLazyImage
   bInherited=true
   bIsVariable=true
   parent=DisplayOverlay
   widget=LazyImage_Icon
   properties {
    clipping=ClipToBoundsWithoutIntersecting
    visibility=Visible
    brush {
     bIsDynamicallyLoaded=false
     drawAs=Image
     imageType=NoImage
     mirroring=NoMirror
     resourceName=None
     resourceObject=/FortniteConversationUI/Conversation/Materials/MaterialInstances/MI_UI_DonationFundTile_Thumbnail
     tiling=NoTile
     imageSize=<512 436>
     margin=<ltrb 0 0 0 0>
     outlineSettings {
      bUseBrushTransparency=false
      roundingType=HalfHeightRadius
      width=0
      color {colorUseRule=UseColor_Specified, specifiedColor=<rgba 0 0 0 0>}
      cornerRadii=<0 0 0 0>
     }
     tintColor {colorUseRule=UseColor_Specified, specifiedColor=<rgba 1 1 1 1>}
     uVRegion {bIsValid=false, max=<0 0>, min=<0 0>}
    }
   }
   slot {class=OverlaySlot, properties {horizontalAlignment=HAlign_Fill, verticalAlignment=VAlign_Bottom}}
  }
  Ruler {
   widgetClassPath=&Image
   bIsVariable=true
   parent=DisplayOverlay
   widget=Ruler
   properties {
    renderOpacity=0
    visibility=SelfHitTestInvisible
    brush {
     bIsDynamicallyLoaded=false
     drawAs=Image
     imageType=NoImage
     mirroring=NoMirror
     resourceName=None
     resourceObject=/FortniteConversationUI/Conversation/Materials/M_UI_FundBlueprint_Ruler
     tiling=NoTile
     imageSize=<256 16>
     margin=<ltrb 0 0 0 0>
     outlineSettings {
      bUseBrushTransparency=false
      roundingType=HalfHeightRadius
      width=0
      color {colorUseRule=UseColor_Specified, specifiedColor=<rgba 0 0 0 0>}
      cornerRadii=<0 0 0 0>
     }
     tintColor {colorUseRule=UseColor_Specified, specifiedColor=<rgba 1 1 1 1>}
     uVRegion {bIsValid=false, max=<0 0>, min=<0 0>}
    }
    renderTransform {angle=0, scale=<0.9 0.9>, shear=<0 0>, translation=<0 0>}
    renderTransformPivot=<0 1>
   }
   slot {class=OverlaySlot, properties {verticalAlignment=VAlign_Bottom, padding=<ltrb 64 0 0 80>}}
  }
  LazyImage_Spinner {
   widgetClassPath=&Image
   bIsVariable=true
   parent=DisplayOverlay
   widget=LazyImage_Spinner
   properties {
    visibility=Collapsed
    brush {
     bIsDynamicallyLoaded=false
     drawAs=Image
     imageType=NoImage
     mirroring=NoMirror
     resourceName=None
     resourceObject=/FortUILibrary/UIKit/Throbber/MI_UIKit_Monochrome_Throbber
     tiling=NoTile
     imageSize=<64 64>
     margin=<ltrb 0 0 0 0>
     outlineSettings {
      bUseBrushTransparency=false
      roundingType=HalfHeightRadius
      width=0
      color {colorUseRule=UseColor_Specified, specifiedColor=<rgba 0 0 0 0>}
      cornerRadii=<0 0 0 0>
     }
     tintColor {colorUseRule=UseColor_Specified, specifiedColor=<rgba 1 1 1 1>}
     uVRegion {bIsValid=false, max=<0 0>, min=<0 0>}
    }
   }
   slot {class=OverlaySlot, properties {horizontalAlignment=HAlign_Center, verticalAlignment=VAlign_Center}}
  }
  Content {
   widgetClassPath=&VerticalBox
   parent=DisplayOverlay
   widget=Content
   slot {class=OverlaySlot, properties {horizontalAlignment=HAlign_Fill, verticalAlignment=VAlign_Fill}}
  }
  Overlay_0 {
   widgetClassPath=&Overlay
   parent=Content
   widget=Overlay_0
   slot {class=VerticalBoxSlot, properties {size {sizeRule=Fill, value=1}}}
  }
  Overlay_2 {
   widgetClassPath=&Overlay
   parent=Overlay_0
   widget=Overlay_2
   properties {clipping=ClipToBoundsWithoutIntersecting}
   slot {class=OverlaySlot, properties {horizontalAlignment=HAlign_Center, verticalAlignment=VAlign_Bottom, padding=<ltrb 0 0 0 72>}}
  }
  NamePercent {
   widgetClassPath=&SizeBox
   parent=Overlay_2
   widget=NamePercent
   properties {bOverride_WidthOverride=true, clipping=ClipToBoundsWithoutIntersecting, widthOverride=704}
   slot {class=OverlaySlot, properties {horizontalAlignment=HAlign_Fill, verticalAlignment=VAlign_Fill}}
  }
  ScaleBox_1 {
   widgetClassPath=&ScaleBox
   parent=NamePercent
   widget=ScaleBox_1
   properties {clipping=ClipToBoundsWithoutIntersecting, stretch=ScaleToFitX, stretchDirection=DownOnly}
   slot {class=SizeBoxSlot, properties {horizontalAlignment=HAlign_Center}}
  }
  Overlay_3 {
   widgetClassPath=&Overlay
   parent=ScaleBox_1
   widget=Overlay_3
   properties {clipping=ClipToBoundsWithoutIntersecting}
   slot {class=ScaleBoxSlot}
  }
  Text_TitleBar {
   widgetClassPath=&CommonTextBlock
   bInherited=true
   bIsVariable=true
   parent=Overlay_3
   widget=Text_TitleBar
   properties {
    clipping=ClipToBoundsWithoutIntersecting
    justification=Center
    lineHeightPercentage=0.9
    text=Weapon Name
    textTransformPolicy=ToUpper
    colorAndOpacity {colorUseRule=UseColor_Specified, specifiedColor=<rgba 0 0.0118 0.0667 1>}
    font {
     bForceMonospaced=false
     bMaterialIsStencil=false
     fontMaterial=/FortniteConversationUI/Conversation/Materials/M_Fund_UI_BlueprintText_Default
     fontObject=&Burbank
     letterSpacing=0
     monospacedWidth=1
     size=40
     skewAmount=0
     typefaceFontName=BurbankRegular-Black
     outlineSettings {
      bApplyOutlineToDropShadows=true
      bMiteredCorners=false
      bSeparateFillAlpha=true
      outlineMaterial=/FortniteConversationUI/Conversation/Materials/M_UI_FundBlueprint_Texture_Font
      outlineSize=2
      outlineColor=<rgba 1 1 1 1>
     }
    }
    margin=<ltrb 0 5 0 10>
    renderTransform {angle=0, scale=<0.75 0.75>, shear=<-12 0>, translation=<0 45>}
    shadowOffset=<0 0>
   }
   slot {class=OverlaySlot, properties {horizontalAlignment=HAlign_Fill, verticalAlignment=VAlign_Fill}}
  }
  Text_SecondLine_L {
   widgetClassPath=&CommonTextBlock
   bIsVariable=true
   parent=Overlay_3
   widget=Text_SecondLine_L
   properties {
    clipping=ClipToBoundsWithoutIntersecting
    justification=Center
    lineHeightPercentage=0.9
    text=Created by
    textTransformPolicy=ToUpper
    colorAndOpacity {colorUseRule=UseColor_Specified, specifiedColor=<rgba 0.0015 0.013 0.0685 1>}
    font {
     bForceMonospaced=false
     bMaterialIsStencil=false
     fontMaterial=&M_Fund_UI_BlueprintText_Default
     fontObject=&BurbankSmall
     letterSpacing=0
     monospacedWidth=1
     size=16
     skewAmount=0
     typefaceFontName=Black
     outlineSettings {
      bApplyOutlineToDropShadows=true
      bMiteredCorners=false
      bSeparateFillAlpha=true
      outlineMaterial=&M_UI_FundBlueprint_Texture_Font
      outlineSize=2
      outlineColor=<rgba 1 1 1 1>
     }
    }
    margin=<ltrb 0 5 0 10>
    renderTransform {angle=0, scale=<0.75 0.75>, shear=<-12 0>, translation=<0 25>}
    shadowOffset=<0 0>
   }
   slot {class=OverlaySlot, properties {horizontalAlignment=HAlign_Fill, verticalAlignment=VAlign_Fill}}
  }
  Text_SecondLine_R {
   widgetClassPath=&CommonTextBlock
   bIsVariable=true
   parent=Overlay_3
   widget=Text_SecondLine_R
   properties {
    clipping=ClipToBoundsWithoutIntersecting
    justification=Center
    lineHeightPercentage=0.9
    text=Creator's Name
    textTransformPolicy=ToUpper
    visibility=Collapsed
    colorAndOpacity {colorUseRule=UseColor_Specified, specifiedColor=<rgba 0.0015 0.013 0.0685 1>}
    font {
     bForceMonospaced=false
     bMaterialIsStencil=false
     fontMaterial=&M_Fund_UI_BlueprintText_Default
     fontObject=&Burbank
     letterSpacing=0
     monospacedWidth=1
     size=40
     skewAmount=0
     typefaceFontName=BurbankRegular-Black
     outlineSettings {
      bApplyOutlineToDropShadows=true
      bMiteredCorners=false
      bSeparateFillAlpha=true
      outlineMaterial=&M_UI_FundBlueprint_Texture_Font
      outlineSize=2
      outlineColor=<rgba 1 1 1 1>
     }
    }
    margin=<ltrb 0 5 0 10>
    renderTransform {angle=0, scale=<0.6 0.6>, shear=<-12 0>, translation=<0 25>}
    shadowOffset=<0 0>
   }
   slot {class=OverlaySlot, properties {horizontalAlignment=HAlign_Fill, verticalAlignment=VAlign_Fill}}
  }
  BlueprintPercentText {
   widgetClassPath=&HorizontalBox
   parent=Overlay_2
   widget=BlueprintPercentText
   properties {visibility=Collapsed, renderTransform {angle=-2, scale=<1 1>, shear=<0 0>, translation=<0 -40>}}
   slot {class=OverlaySlot, properties {horizontalAlignment=HAlign_Right, verticalAlignment=VAlign_Bottom, padding=<ltrb 0 0 40 0>}}
  }
  Text_ProgressPercent_2 {
   widgetClassPath=&CommonTextBlock
   bIsVariable=true
   parent=BlueprintPercentText
   widget=Text_ProgressPercent_2
   properties {
    text=99%
    font {
     bForceMonospaced=false
     bMaterialIsStencil=false
     fontMaterial=&M_UI_FundBlueprint_Texture_Font
     fontObject=&Burbank
     letterSpacing=0
     monospacedWidth=1
     size=24
     skewAmount=0
     typefaceFontName=BurbankRegular-Black
     outlineSettings {
      bApplyOutlineToDropShadows=true
      bMiteredCorners=false
      bSeparateFillAlpha=true
      outlineMaterial=null
      outlineSize=0
      outlineColor=<rgba 0 0 0 0>
     }
    }
    shadowOffset=<0 0>
   }
   slot {class=HorizontalBoxSlot, properties {horizontalAlignment=HAlign_Right, verticalAlignment=VAlign_Bottom}}
  }
  Text_ProgressPercent_4 {
   widgetClassPath=&CommonTextBlock
   bIsVariable=true
   parent=BlueprintPercentText
   widget=Text_ProgressPercent_4
   properties {
    text="FUNDED!"
    font {
     bForceMonospaced=false
     bMaterialIsStencil=false
     fontMaterial=&M_UI_FundBlueprint_Texture_Font
     fontObject=&BurbankSmall
     letterSpacing=0
     monospacedWidth=1
     size=18
     skewAmount=0
     typefaceFontName=Black
     outlineSettings {
      bApplyOutlineToDropShadows=true
      bMiteredCorners=false
      bSeparateFillAlpha=true
      outlineMaterial=null
      outlineSize=0
      outlineColor=<rgba 0 0 0 0>
     }
    }
    shadowOffset=<0 0>
   }
   slot {class=HorizontalBoxSlot, properties {verticalAlignment=VAlign_Center, padding=<ltrb 5 0 0 0>}}
  }
  FundingPercent_Overlay {
   widgetClassPath=&Overlay
   bIsVariable=true
   parent=Overlay_2
   widget=FundingPercent_Overlay
   properties {clipping=ClipToBoundsWithoutIntersecting, visibility=Collapsed, renderTransform {angle=0, scale=<1 1>, shear=<0 0>, translation=<8 -36>}}
   slot {class=OverlaySlot, properties {horizontalAlignment=HAlign_Right, padding=<ltrb 0 -8 40 0>}}
  }
  Image_93 {
   widgetClassPath=&Image
   bIsVariable=true
   parent=FundingPercent_Overlay
   widget=Image_93
   properties {
    visibility=SelfHitTestInvisible
    brush {
     bIsDynamicallyLoaded=false
     drawAs=Image
     imageType=NoImage
     mirroring=NoMirror
     resourceName=None
     resourceObject=/FortUILibrary/Foundation/Art_ToolKit/AngledBoxes/M_UI_FundCenter_PercentBacking_Default
     tiling=NoTile
     imageSize=<32 32>
     margin=<ltrb 0 0 0 0>
     outlineSettings {
      bUseBrushTransparency=false
      roundingType=HalfHeightRadius
      width=0
      color {colorUseRule=UseColor_Specified, specifiedColor=<rgba 0 0 0 0>}
      cornerRadii=<0 0 0 0>
     }
     tintColor {colorUseRule=UseColor_Specified, specifiedColor=<rgba 1 1 1 1>}
     uVRegion {bIsValid=false, max=<0 0>, min=<0 0>}
    }
   }
   slot {class=OverlaySlot, properties {horizontalAlignment=HAlign_Fill, verticalAlignment=VAlign_Fill}}
  }
  HorizontalBox_1 {
   widgetClassPath=&HorizontalBox
   parent=FundingPercent_Overlay
   widget=HorizontalBox_1
   properties {renderTransform {angle=0, scale=<1 1>, shear=<-16 0>, translation=<0 0>}}
   slot {class=OverlaySlot, properties {horizontalAlignment=HAlign_Center, verticalAlignment=VAlign_Center, padding=<ltrb 18 10 18 10>}}
  }
  Text_ProgressPercent {
   widgetClassPath=&CommonTextBlock
   bInherited=true
   bIsVariable=true
   parent=HorizontalBox_1
   widget=Text_ProgressPercent
   properties {
    text=99%
    font {
     bForceMonospaced=false
     bMaterialIsStencil=false
     fontMaterial=null
     fontObject=&Burbank
     letterSpacing=0
     monospacedWidth=1
     size=16
     skewAmount=0
     typefaceFontName=BurbankRegular-Black
     outlineSettings {
      bApplyOutlineToDropShadows=true
      bMiteredCorners=false
      bSeparateFillAlpha=true
      outlineMaterial=null
      outlineSize=0
      outlineColor=<rgba 0 0 0 0>
     }
    }
    shadowOffset=<0 0>
   }
   slot {class=HorizontalBoxSlot}
  }
  Text_ProgressPercent_1 {
   widgetClassPath=&CommonTextBlock
   bIsVariable=true
   parent=HorizontalBox_1
   widget=Text_ProgressPercent_1
   properties {
    text="FUNDED!"
    font {
     bForceMonospaced=false
     bMaterialIsStencil=false
     fontMaterial=null
     fontObject=&BurbankSmall
     letterSpacing=0
     monospacedWidth=1
     size=16
     skewAmount=0
     typefaceFontName=Black
     outlineSettings {
      bApplyOutlineToDropShadows=true
      bMiteredCorners=false
      bSeparateFillAlpha=true
      outlineMaterial=null
      outlineSize=0
      outlineColor=<rgba 0 0 0 0>
     }
    }
    shadowOffset=<0 0>
   }
   slot {class=HorizontalBoxSlot, properties {verticalAlignment=VAlign_Center, padding=<ltrb 6 0 0 0>}}
  }
  ProgressInfo {
   widgetClassPath=&VerticalBox
   parent=Overlay_0
   widget=ProgressInfo
   properties {visibility=Collapsed, renderTransform {angle=0, scale=<1 1>, shear=<-4 0>, translation=<0 0>}}
   slot {class=OverlaySlot, properties {horizontalAlignment=HAlign_Center, verticalAlignment=VAlign_Bottom, padding=<ltrb 0 0 0 44>}}
  }
  Border_0 {
   widgetClassPath=&Border
   parent=ProgressInfo
   widget=Border_0
   properties {
    background {
     bIsDynamicallyLoaded=false
     drawAs=Box
     imageType=NoImage
     mirroring=NoMirror
     resourceName=None
     resourceObject=/FortniteConversationUI/Conversation/Materials/M_UI_MultiFundProgress_BarBacking_Default
     tiling=NoTile
     imageSize=<32 32>
     margin=<ltrb 0.5 0 0.5 0>
     outlineSettings {
      bUseBrushTransparency=false
      roundingType=HalfHeightRadius
      width=0
      color {colorUseRule=UseColor_Specified, specifiedColor=<rgba 0 0 0 0>}
      cornerRadii=<0 0 0 0>
     }
     tintColor {colorUseRule=UseColor_Specified, specifiedColor=<rgba 1 1 1 1>}
     uVRegion {bIsValid=false, max=<0 0>, min=<0 0>}
    }
    padding=<ltrb 10 10 12 10>
   }
   slot {class=VerticalBoxSlot, properties {horizontalAlignment=HAlign_Center, verticalAlignment=VAlign_Bottom}}
  }
  ProgressBarSize {
   widgetClassPath=&SizeBox
   parent=Border_0
   widget=ProgressBarSize
   properties {bOverride_HeightOverride=true, bOverride_WidthOverride=true, heightOverride=16, widthOverride=512}
   slot {class=BorderSlot, properties {padding=<ltrb 10 10 12 10>}}
  }
  Progress_Funding {
   widgetClassPath=/Game/Athena/HUD/SimpleMaterialProgressBar
   bInherited=true
   bIsVariable=true
   parent=ProgressBarSize
   widget=Progress_Funding
   properties {
    bOverrideDefaultColorA=true
    bOverrideDefaultColorB=true
    bOverrideDefaultColorBackground=true
    bOverrideDefaultFillEdgeSoftness=true
    bOverrideDefaultOutlineScale=true
    bOverrideDefaultSegments=true
    bUseStroke=false
    designTime_Progress=0.75
    fillEdgeSoftness=0.0005
    segmentEdge=0.0002
    segments=1
    cachedColorA=<rgba 0 1 1 1>
    cachedColorB=<rgba 1 1 1 1>
    cachedColorBackground=<rgba 0 0.3125 0.75 1>
    renderTransform {angle=0, scale=<1 1>, shear=<-8 0>, translation=<0 0>}
   }
   slot {class=SizeBoxSlot}
  }
  HBox_ProgressText {
   widgetClassPath=&HorizontalBox
   bIsVariable=true
   parent=ProgressInfo
   widget=HBox_ProgressText
   properties {visibility=Collapsed}
   slot {class=VerticalBoxSlot, properties {padding=<ltrb -5 -20 0 0>}}
  }
  Image_CurrencyIcon {
   widgetClassPath=&Image
   bIsVariable=true
   parent=HBox_ProgressText
   widget=Image_CurrencyIcon
   properties {brush {
     bIsDynamicallyLoaded=false
     drawAs=Image
     imageType=NoImage
     mirroring=NoMirror
     resourceName=None
     resourceObject=/Game/UI/Foundation/Textures/Icons/Athena/T-T-Icon-BR-GoldBars-UI-Icon
     tiling=NoTile
     imageSize=<42 42>
     margin=<ltrb 0 0 0 0>
     outlineSettings {
      bUseBrushTransparency=false
      roundingType=HalfHeightRadius
      width=0
      color {colorUseRule=UseColor_Specified, specifiedColor=<rgba 0 0 0 0>}
      cornerRadii=<0 0 0 0>
     }
     tintColor {colorUseRule=UseColor_Specified, specifiedColor=<rgba 1 1 1 1>}
     uVRegion {bIsValid=false, max=<0 0>, min=<0 0>}
    }}
   slot {class=HorizontalBoxSlot, properties {horizontalAlignment=HAlign_Center, verticalAlignment=VAlign_Center, padding=<ltrb 3 0 3 0>}}
  }
  Text_ProgressCount {
   widgetClassPath=&CommonTextBlock
   bInherited=true
   bIsVariable=true
   parent=HBox_ProgressText
   widget=Text_ProgressCount
   properties {
    text="500,000 / 1,000,000"
    font {
     bForceMonospaced=false
     bMaterialIsStencil=false
     fontMaterial=null
     fontObject=&Burbank
     letterSpacing=0
     monospacedWidth=1
     size=16
     skewAmount=0
     typefaceFontName=BurbankRegular-Black
     outlineSettings {
      bApplyOutlineToDropShadows=true
      bMiteredCorners=false
      bSeparateFillAlpha=true
      outlineMaterial=null
      outlineSize=2
      outlineColor=<rgba 0 0 0 1>
     }
    }
    renderTransform {angle=0, scale=<1 1>, shear=<-8 0>, translation=<0 0>}
    shadowOffset=<0 0>
   }
   slot {class=HorizontalBoxSlot, properties {verticalAlignment=VAlign_Center}}
  }
  Button_BriefEntry {
   widgetClassPath=/Game/Athena/UI/Matchmaking/SimpleCommonButton
   bIsVariable=true
   parent=Overlay_0
   widget=Button_BriefEntry
   properties {
    onButtonBaseClicked=(&self:WidgetTree.Button_BriefEntry.HandleOnButtonClicked)
    onButtonBaseDoubleClicked=(&self:WidgetTree.Button_BriefEntry.HandleOnButtonDoubleClicked)
    onButtonBaseHovered=(&self:WidgetTree.Button_BriefEntry.HandleOnButtonHovered)
    onButtonBaseUnhovered=(&self:WidgetTree.Button_BriefEntry.HandleOnButtonUnhovered)
    onSelectedChangedBase=(&self:WidgetTree.Button_BriefEntry.HandleOnSelectedChanged)
    renderOpacity=0
    style=null
   }
   slot {class=OverlaySlot, properties {horizontalAlignment=HAlign_Fill, verticalAlignment=VAlign_Fill}}
  }
  BottomStrip {
   widgetClassPath=&Image
   bIsVariable=true
   parent=DisplayOverlay
   widget=BottomStrip
   properties {
    renderOpacity=0
    visibility=SelfHitTestInvisible
    brush {
     bIsDynamicallyLoaded=false
     drawAs=Image
     imageType=NoImage
     mirroring=NoMirror
     resourceName=None
     resourceObject=/FortniteConversationUI/Conversation/Materials/M_UI_DonationFundTile_Background_Inst
     tiling=NoTile
     imageSize=<32 16>
     margin=<ltrb 0 0 0 0>
     outlineSettings {
      bUseBrushTransparency=false
      roundingType=HalfHeightRadius
      width=0
      color {colorUseRule=UseColor_Specified, specifiedColor=<rgba 0 0 0 0>}
      cornerRadii=<0 0 0 0>
     }
     tintColor {colorUseRule=UseColor_Specified, specifiedColor=<rgba 1 1 1 1>}
     uVRegion {bIsValid=false, max=<0 0>, min=<0 0>}
    }
    renderTransformPivot=<0.5 1>
   }
   slot {class=OverlaySlot, properties {horizontalAlignment=HAlign_Fill, verticalAlignment=VAlign_Fill}}
  }
  FundStrip {
   widgetClassPath=&SizeBox
   parent=DisplayOverlay
   widget=FundStrip
   properties {
    bOverride_MinDesiredWidth=true
    minDesiredWidth=320
    widthOverride=320
    renderTransform {angle=0, scale=<1 1>, shear=<0 0>, translation=<0 36>}
   }
   slot {class=OverlaySlot, properties {horizontalAlignment=HAlign_Center, verticalAlignment=VAlign_Bottom}}
  }
  TransactionStrip_Main {
   widgetClassPath=&/FortTransactionStripNew
   bInherited=true
   bIsVariable=true
   parent=FundStrip
   widget=TransactionStrip_Main
   properties {bFundingVersion=true, transactionString=Buy}
   slot {class=SizeBoxSlot, properties {verticalAlignment=VAlign_Center}}
  }
 ]
}
