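# Conversation UI widget for an NPC funding vote, showing three item choices, NPC dialogue and time until funding closes.
# On construct it plays the intro, binds donation-success and initialization delegates, and caches entry arrays.
# Selection changes highlight an entry using a reversed index (2 minus selected) and refresh NPC text.
# A successful donation plays the Funded animation, then copies progress into vote-info rows, reveals them and flags the highest as winner.
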WidgetBlueprint /FortniteConversationUI/Conversation/FortMultiItemVoteBriefWidget
animations [
 Anim_SelectLeft {
  binding_count=1
  length_seconds=0.4
  object=&self:Anim_SelectLeft
  sequence {
   bindings [Image_Arrow {
     id=#1
     possessed_class=Image
     tracks [{
       type=MovieScene2DTransformTrack
       sections [{
         type=MovieScene2DTransformSection
         channels [
          Scale.X {type=MovieSceneScriptingFloatChannel, thinned_from=7, keys [[0,1],[3,1],[8,1.05],[16,-1.05],[21,-1],[24,-1]]}
          Scale.Y {type=MovieSceneScriptingFloatChannel, const [1,3]}
         ]
        }]
      }]
    }]
   display_rate {numerator=60}
  }
 }
 Anim_SelectRight {
  binding_count=1
  length_seconds=0.4
  object=&self:Anim_SelectRight
  sequence {
   bindings [Image_Arrow {
     id=#2
     possessed_class=Image
     tracks [{
       type=MovieScene2DTransformTrack
       sections [{
         type=MovieScene2DTransformSection
         channels [
          Scale.X {type=MovieSceneScriptingFloatChannel, thinned_from=7, keys [[0,-1],[3,-1],[8,-1.05],[16,1.05],[21,1],[24,1]]}
          Scale.Y {type=MovieSceneScriptingFloatChannel, const [1,3]}
         ]
        }]
      }]
    }]
   display_rate {numerator=60}
  }
 }
 Anim_SelectRight_Bounce {
  binding_count=4
  length_seconds=0.4
  object=&self:Anim_SelectRight_Bounce
  sequence {
   bindings [
    Image_Shadow {
     id=#3
     possessed_class=Image
     tracks [
      {
       type=MovieScene2DTransformTrack
       sections [{
         type=MovieScene2DTransformSection
         end_frame=24
         start_frame=0
         channels [
          Translation.X {type=MovieSceneScriptingFloatChannel, const [0,1]}
          Translation.Y {type=MovieSceneScriptingFloatChannel, const [0,1]}
          Angle {type=MovieSceneScriptingFloatChannel, const [0,1]}
          Scale.X {type=MovieSceneScriptingFloatChannel, keys [[0,1],[3,0.95],[6,1.05],[12,1]]}
          Scale.Y {type=MovieSceneScriptingFloatChannel, keys [[0,1],[3,0.95],[6,1.05],[12,1]]}
          Shear.X {type=MovieSceneScriptingFloatChannel, const [0,1]}
          Shear.Y {type=MovieSceneScriptingFloatChannel, const [0,1]}
         ]
        }]
      }
      {
       type=MovieSceneWidgetMaterialTrack
       sections [{
         type=MovieSceneParameterSection
         end_frame=24
         start_frame=0
         channels [
          ShadowAmount {type=MovieSceneScriptingFloatChannel, keys [[0,0.333],[3,1],[6,0.25],[12,0.333]]}
          ShadowScale {type=MovieSceneScriptingFloatChannel, keys [[0,1],[3,0.9],[6,1.125],[12,1]]}
         ]
        }]
      }
     ]
    }
    Image_Circle {
     id=#4
     possessed_class=Image
     tracks [{
       type=MovieScene2DTransformTrack
       sections [{
         type=MovieScene2DTransformSection
         end_frame=24
         start_frame=0
         channels [
          Scale.X {type=MovieSceneScriptingFloatChannel, keys [[2,1],[5,0.95],[8,1.05],[14,1]]}
          Scale.Y {type=MovieSceneScriptingFloatChannel, keys [[2,1],[5,0.95],[8,1.05],[14,1]]}
         ]
        }]
      }]
    }
    Image_Arrow {
     id=#2
     possessed_class=Image
     tracks [{
       type=MovieScene2DTransformTrack
       sections [{
         type=MovieScene2DTransformSection
         channels [
          Scale.X {type=MovieSceneScriptingFloatChannel, thinned_from=7, keys [[0,-1],[3,-1],[8,-1.05],[16,1.05],[21,1],[24,1]]}
          Scale.Y {type=MovieSceneScriptingFloatChannel, const [1,3]}
         ]
        }]
      }]
    }
    CommonTextBlock_82 {
     id=#5
     possessed_class=CommonTextBlock
     tracks [{
       type=MovieScene2DTransformTrack
       sections [{
         type=MovieScene2DTransformSection
         end_frame=24
         start_frame=0
         channels [
          Translation.X {type=MovieSceneScriptingFloatChannel}
          Translation.Y {type=MovieSceneScriptingFloatChannel}
          Angle {type=MovieSceneScriptingFloatChannel}
          Scale.X {type=MovieSceneScriptingFloatChannel, keys [[0,1],[3,0.95],[6,1.05],[12,1]]}
          Scale.Y {type=MovieSceneScriptingFloatChannel, keys [[0,1],[3,0.95],[6,1.05],[12,1]]}
          Shear.X {type=MovieSceneScriptingFloatChannel}
          Shear.Y {type=MovieSceneScriptingFloatChannel}
         ]
        }]
      }]
    }
   ]
   display_rate {numerator=60}
  }
 }
 Anim_SelectLeft_Bounce {
  binding_count=4
  length_seconds=0.4
  object=&self:Anim_SelectLeft_Bounce
  sequence {
   bindings [
    Image_Arrow {
     id=#1
     possessed_class=Image
     tracks [{
       type=MovieScene2DTransformTrack
       sections [{
         type=MovieScene2DTransformSection
         channels [
          Scale.X {type=MovieSceneScriptingFloatChannel, thinned_from=7, keys [[0,1],[3,1],[8,1.05],[16,-1.05],[21,-1],[24,-1]]}
          Scale.Y {type=MovieSceneScriptingFloatChannel, const [1,3]}
         ]
        }]
      }]
    }
    Image_Circle {
     id=#6
     possessed_class=Image
     tracks [{
       type=MovieScene2DTransformTrack
       sections [{
         type=MovieScene2DTransformSection
         end_frame=24
         start_frame=0
         channels [
          Scale.X {type=MovieSceneScriptingFloatChannel, keys [[2,1],[5,0.95],[8,1.05],[14,1]]}
          Scale.Y {type=MovieSceneScriptingFloatChannel, keys [[2,1],[5,0.95],[8,1.05],[14,1]]}
         ]
        }]
      }]
    }
    CommonTextBlock_82 {
     id=#7
     possessed_class=CommonTextBlock
     tracks [{
       type=MovieScene2DTransformTrack
       sections [{
         type=MovieScene2DTransformSection
         end_frame=24
         start_frame=0
         channels [
          Translation.X {type=MovieSceneScriptingFloatChannel}
          Translation.Y {type=MovieSceneScriptingFloatChannel}
          Angle {type=MovieSceneScriptingFloatChannel}
          Scale.X {type=MovieSceneScriptingFloatChannel, keys [[0,1],[3,0.95],[6,1.05],[12,1]]}
          Scale.Y {type=MovieSceneScriptingFloatChannel, keys [[0,1],[3,0.95],[6,1.05],[12,1]]}
          Shear.X {type=MovieSceneScriptingFloatChannel}
          Shear.Y {type=MovieSceneScriptingFloatChannel}
         ]
        }]
      }]
    }
    Image_Shadow {
     id=#8
     possessed_class=Image
     tracks [
      {
       type=MovieScene2DTransformTrack
       sections [{
         type=MovieScene2DTransformSection
         end_frame=24
         start_frame=0
         channels [
          Translation.X {type=MovieSceneScriptingFloatChannel, const [0,1]}
          Translation.Y {type=MovieSceneScriptingFloatChannel, const [0,1]}
          Angle {type=MovieSceneScriptingFloatChannel, const [0,1]}
          Scale.X {type=MovieSceneScriptingFloatChannel, keys [[0,1],[3,0.95],[6,1.05],[12,1]]}
          Scale.Y {type=MovieSceneScriptingFloatChannel, keys [[0,1],[3,0.95],[6,1.05],[12,1]]}
          Shear.X {type=MovieSceneScriptingFloatChannel, const [0,1]}
          Shear.Y {type=MovieSceneScriptingFloatChannel, const [0,1]}
         ]
        }]
      }
      {
       type=MovieSceneWidgetMaterialTrack
       sections [{
         type=MovieSceneParameterSection
         end_frame=24
         start_frame=0
         channels [
          ShadowAmount {type=MovieSceneScriptingFloatChannel, keys [[0,0.333],[3,1],[6,0.25],[12,0.333]]}
          ShadowScale {type=MovieSceneScriptingFloatChannel, keys [[0,1],[3,0.9],[6,1.125],[12,1]]}
         ]
        }]
      }
     ]
    }
   ]
   display_rate {numerator=60}
  }
 }
 Intro {
  binding_count=6
  length_seconds=0.55
  object=&self:Intro
  sequence {bindings [
    Image_Circle {
     id=#9
     possessed_class=Image
     tracks [{
       type=MovieScene2DTransformTrack
       sections [{
         type=MovieScene2DTransformSection
         channels [
          Translation.X {type=MovieSceneScriptingFloatChannel, const [0,1]}
          Translation.Y {type=MovieSceneScriptingFloatChannel, const [0,1]}
          Angle {type=MovieSceneScriptingFloatChannel, const [0,1]}
          Scale.X {type=MovieSceneScriptingFloatChannel, keys [[1,1],[4,1.1],[7,0.95],[9.0003,1]]}
          Scale.Y {type=MovieSceneScriptingFloatChannel, keys [[1,1],[4,1.1],[7,0.95],[9.0003,1]]}
          Shear.X {type=MovieSceneScriptingFloatChannel, const [0,1]}
          Shear.Y {type=MovieSceneScriptingFloatChannel, const [0,1]}
         ]
        }]
      }]
    }
    Overlay_15 {
     id=#10
     possessed_class=Overlay
     tracks [
      {
       type=MovieSceneFloatTrack
       display_name=Render Opacity
       sections [{type=MovieSceneFloatSection, channels [None {type=MovieSceneScriptingFloatChannel, keys [[0,0],[2,1]]}]}]
      }
      {
       type=MovieScene2DTransformTrack
       sections [{
         type=MovieScene2DTransformSection
         channels [
          Translation.X {type=MovieSceneScriptingFloatChannel, const [0,1]}
          Translation.Y {type=MovieSceneScriptingFloatChannel, keys [[0,128],[5,0]]}
          Angle {type=MovieSceneScriptingFloatChannel, const [0,1]}
          Scale.X {type=MovieSceneScriptingFloatChannel, const [1,1]}
          Scale.Y {type=MovieSceneScriptingFloatChannel, const [1,1]}
          Shear.X {type=MovieSceneScriptingFloatChannel, const [0,1]}
          Shear.Y {type=MovieSceneScriptingFloatChannel, const [0,1]}
         ]
        }]
      }
     ]
    }
    GradientBacking {
     id=#11
     possessed_class=Image
     tracks [{
       type=MovieSceneFloatTrack
       display_name=Render Opacity
       sections [{type=MovieSceneFloatSection, channels [None {type=MovieSceneScriptingFloatChannel, keys [[0,0],[10,1]]}]}]
      }]
    }
    Darken {
     id=#12
     possessed_class=Image
     tracks [{
       type=MovieSceneFloatTrack
       display_name=Render Opacity
       sections [{type=MovieSceneFloatSection, channels [None {type=MovieSceneScriptingFloatChannel, keys [[3,0],[7,1]]}]}]
      }]
    }
    GlobalEffort {
     id=#13
     possessed_class=Overlay
     tracks [
      {
       type=MovieScene2DTransformTrack
       sections [{
         type=MovieScene2DTransformSection
         channels [
          Translation.X {type=MovieSceneScriptingFloatChannel}
          Translation.Y {type=MovieSceneScriptingFloatChannel}
          Angle {type=MovieSceneScriptingFloatChannel}
          Scale.X {type=MovieSceneScriptingFloatChannel, keys [[6,1.25],[8,0.95],[11,1]]}
          Scale.Y {type=MovieSceneScriptingFloatChannel, keys [[6,1.25],[8,0.95],[11,1]]}
          Shear.X {type=MovieSceneScriptingFloatChannel}
          Shear.Y {type=MovieSceneScriptingFloatChannel}
         ]
        }]
      }
      {
       type=MovieSceneFloatTrack
       display_name=Render Opacity
       sections [{type=MovieSceneFloatSection, channels [None {type=MovieSceneScriptingFloatChannel, keys [[6,0],[8,1]]}]}]
      }
     ]
    }
    Overlay_1 {
     id=#14
     possessed_class=Overlay
     tracks [
      {
       type=MovieSceneFloatTrack
       display_name=Render Opacity
       sections [{type=MovieSceneFloatSection, channels [None {type=MovieSceneScriptingFloatChannel, keys [[0,1],[1,0],[4,0],[6,1]]}]}]
      }
      {
       type=MovieScene2DTransformTrack
       sections [{
         type=MovieScene2DTransformSection
         channels [
          Translation.X {type=MovieSceneScriptingFloatChannel, const [0,1]}
          Translation.Y {type=MovieSceneScriptingFloatChannel, const [16,1]}
          Angle {type=MovieSceneScriptingFloatChannel, const [0,1]}
          Scale.X {type=MovieSceneScriptingFloatChannel, keys [[1,1.5],[4,1.25],[6,0.95],[9,1]]}
          Scale.Y {type=MovieSceneScriptingFloatChannel, keys [[1,1.5],[4,1.25],[6,0.95],[9,1]]}
          Shear.X {type=MovieSceneScriptingFloatChannel, const [0,1]}
          Shear.Y {type=MovieSceneScriptingFloatChannel, const [0,1]}
         ]
        }]
      }
     ]
    }
   ]}
 }
 Funded {
  binding_count=8
  length_seconds=1.1
  object=&self:Funded
  sequence {
   bindings [
    Image_Circle {
     id=#15
     possessed_class=Image
     tracks [{
       type=MovieScene2DTransformTrack
       sections [{
         type=MovieScene2DTransformSection
         channels [
          Scale.X {type=MovieSceneScriptingFloatChannel, keys [[5.999,1],[11,0.85],[23,1]]}
          Scale.Y {type=MovieSceneScriptingFloatChannel, keys [[6,1],[11,0.85],[23,1]]}
         ]
        }]
      }]
    }
    Image_Arrow {
     id=#16
     display_name=AnimBoundFundedBorder
     possessed_class=Image
     tracks [
      {
       type=MovieSceneFloatTrack
       display_name=Render Opacity
       sections [{type=MovieSceneFloatSection, channels [None {type=MovieSceneScriptingFloatChannel, keys [[0,1],[3,1],[9,0],[48,0],[54,1]]}]}]
      }
      {
       type=MovieScene2DTransformTrack
       sections [{
         type=MovieScene2DTransformSection
         channels [
          Scale.X {type=MovieSceneScriptingFloatChannel, keys [[0,1],[3,1.125],[9,0],[48,0],[54,1.125],[66,1]]}
          Scale.Y {type=MovieSceneScriptingFloatChannel, keys [[0,1],[3,1.125],[9,0],[48,0],[54,1.125],[66,1]]}
         ]
        }]
      }
     ]
    }
    Image_229 {
     id=#17
     possessed_class=Image
     tracks [{
       type=MovieScene2DTransformTrack
       sections [{type=MovieScene2DTransformSection, channels [Angle {type=MovieSceneScriptingFloatChannel, keys [[0,-8],[27,352]]}]}]
      }]
    }
    Image_Check {
     id=#18
     possessed_class=Image
     tracks [
      {
       type=MovieSceneFloatTrack
       display_name=Render Opacity
       sections [{type=MovieSceneFloatSection, channels [None {type=MovieSceneScriptingFloatChannel, keys [[3.001,0],[9,0],[15,1],[42,1],[48,0]]}]}]
      }
      {
       type=MovieScene2DTransformTrack
       sections [{
         type=MovieScene2DTransformSection
         channels [
          Scale.X {type=MovieSceneScriptingFloatChannel, keys [[3,1],[9,1],[15,1.25],[27,1],[36,1],[42,1.125],[48,0]]}
          Scale.Y {type=MovieSceneScriptingFloatChannel, keys [[3,1],[9,1],[15,1.25],[27,1],[36,1],[42,1.125],[48,0]]}
         ]
        }]
      }
     ]
    }
    CommonTextBlock_82 {
     id=#19
     possessed_class=CommonTextBlock
     tracks [{
       type=MovieScene2DTransformTrack
       sections [{
         type=MovieScene2DTransformSection
         channels [
          Translation.X {type=MovieSceneScriptingFloatChannel, const [0,1]}
          Translation.Y {type=MovieSceneScriptingFloatChannel, keys [[6,88],[9,80],[15,92],[27,88]]}
          Scale.X {type=MovieSceneScriptingFloatChannel, keys [[6,1],[9,0.95],[15,1.05],[27,1]]}
          Scale.Y {type=MovieSceneScriptingFloatChannel, keys [[6,1],[9,0.95],[15,1.05],[27,1]]}
         ]
        }]
      }]
    }
    Image_Shadow {
     id=#20
     possessed_class=Image
     tracks [
      {
       type=MovieScene2DTransformTrack
       sections [{
         type=MovieScene2DTransformSection
         channels [
          Scale.X {type=MovieSceneScriptingFloatChannel, keys [[6,1],[9,0.9],[15,1.1],[27,1]]}
          Scale.Y {type=MovieSceneScriptingFloatChannel, keys [[6,1],[9,0.9],[15,1.1],[27,1]]}
         ]
        }]
      }
      {
       type=MovieSceneWidgetMaterialTrack
       sections [{type=MovieSceneParameterSection, channels [DonateColorTransition {type=MovieSceneScriptingFloatChannel, keys [[15,0],[18,0.666],[42,0]]}]}]
      }
     ]
    }
    Image {
     id=#21
     possessed_class=Image
     tracks [{
       type=MovieScene2DTransformTrack
       sections [{type=MovieScene2DTransformSection, channels [Angle {type=MovieSceneScriptingFloatChannel, keys [[0,8],[27,-352]]}]}]
      }]
    }
    Image_Burst {
     id=#22
     possessed_class=Image
     tracks [
      {
       type=MovieSceneFloatTrack
       display_name=Render Opacity
       sections [{type=MovieSceneFloatSection, channels [None {type=MovieSceneScriptingFloatChannel, keys [[0,0],[3,1],[24,0]]}]}]
      }
      {
       type=MovieScene2DTransformTrack
       sections [{
         type=MovieScene2DTransformSection
         channels [
          Scale.X {type=MovieSceneScriptingFloatChannel, keys [[0,0.5],[33,1.5]]}
          Scale.Y {type=MovieSceneScriptingFloatChannel, keys [[0,0.5],[33,1.5]]}
         ]
        }]
      }
     ]
    }
   ]
   display_rate {numerator=60}
  }
 }
]
graphs [
 EventGraph {nodes [
   OnDonationSuccess() {is=title K2Node_CustomEvent}
   > PlayAnimation(InAnimation=Get Funded() -> (Funded:Widget Animation Object Reference))
   > Delay(Duration=0.1)
   > UpdateVoteData()
   n1 = Event UpdateSelectedVisuals!() -> (OutputDelegate, then, SelectedIndex:int=0) {is=title K2Node_Event}
   > Branch(Condition="Not Equal (Integer)"!(A:int=@n1.SelectedIndex, B:int=Get CachedSelectedIndex() -> (CachedSelectedIndex:int=0)) -> (ReturnValue:bool) {is=title K2Node_CallFunction})
   > n2 = Set CachedSelectedIndex(CachedSelectedIndex:int=@n1.SelectedIndex) -> (Output_Get:int=0)
   > UpdateSelectedState!(execute, self, SelectedIndex:int=@n2) -> (then)
   > n3 = Cast To NPCTextDisplay!(execute, Object=Get NPCTextDisplay() -> (NPCTextDisplay:Fort NPCText Display Object Reference)) -> (then, CastFailed, AsNPCText Display:NPCText Display Object Reference) {is=title K2Node_DynamicCast}
   > NewItemUpdate(self:NPCText Display Object Reference=@n3)
   Event Construct()
   > PlayAnimation(InAnimation=Get Intro())
   > n4 = Cast To Athena_PlayerController!(execute, Object=GetOwningPlayer()) -> (then, CastFailed, AsAthena Player Controller:Athena Player Controller Object Reference) {is=title K2Node_DynamicCast}
   > Bind Event to On Client Donation Success Delegate(self:Fort Player Controller Athena Object Reference=@n4, Delegate=Create Event()) {is=title K2Node_AddDelegate}
   > n5 = Cast To FortMultiItemVoteBriefEntry_StyleB!(execute, Object=Get Item_Entry() -> (Item_Entry:Fort Item Fund Brief Object Reference)) -> (then, CastFailed, AsFort Multi Item Vote Brief Entry Style B:Fort Multi Item Vote Brief Entry Style B Object Reference) {is=title K2Node_DynamicCast}
   > Bind Event to On Initialization Completed(self:Fort Multi Item Vote Brief Entry Style B Object Reference=@n5, Delegate=Create Event()) {is=title K2Node_AddDelegate}
   > n6 = Cast To NPCTextDisplay(Object=Get NPCTextDisplay())
   > NewItemUpdate(self=@n6)
   > Set ItemEntryArray(ItemEntryArray:Array of Fort Item Fund Brief Object References=Make Array!(
     "[0]":Fort Item Fund Brief Object Reference=Get Item_Entry()
     "[1]":Fort Item Fund Brief Object Reference=Get Item_EntrySecondary() -> (Item_EntrySecondary:Fort Item Fund Brief Object Reference)
     "[2]":Fort Item Fund Brief Object Reference=Get Item_EntryTertiary() -> (Item_EntryTertiary:Fort Item Fund Brief Object Reference)
    ) -> (Array:Array of Fort Item Fund Brief Object References)) -> (Output_Get:Array of Fort Item Fund Brief Object References)
   > Set VoteInfoEntryArray(VoteInfoEntryArray:Array of B Vote Info Entry Object References=Make Array(
     "[0]":B Vote Info Entry Object Reference=Get B_VoteInfoEntry_0() -> (B_VoteInfoEntry_0:B Vote Info Entry Object Reference)
     "[1]":B Vote Info Entry Object Reference=Get B_VoteInfoEntry_1() -> (B_VoteInfoEntry_1:B Vote Info Entry Object Reference)
     "[2]":B Vote Info Entry Object Reference=Get B_VoteInfoEntry_2() -> (B_VoteInfoEntry_2:B Vote Info Entry Object Reference)
    ) -> (Array:Array of B Vote Info Entry Object References)) -> (Output_Get:Array of B Vote Info Entry Object References)
   n7 = Get VoteDetails_Switcher() -> (VoteDetails_Switcher:Common Animated Switcher Object Reference)
   > SetDisableTransitionAnimation!(execute, self:Common Animated Switcher Object Reference=@n7, bDisableAnimation:bool=true) -> (then)
   > SetActiveWidgetIndex(self=@n7, Index=0)
   > UpdateSelectedState(SelectedIndex=0)
   n8 = Event PreConstruct()
   > Branch(Condition=@n8.IsDesignTime)
  ]}
 FundedAnim {nodes [
   FundedAnim() {is=title K2Node_FunctionEntry}
   > PlayAnimation(InAnimation=Get Funded())
   > UpdateVoteData()
  ]}
 HandleEntryInitializationComplete {nodes [
   n1 = HandleEntryInitializationComplete!() -> (then, VoteEntry:Fort Multi Item Vote Brief Entry Style B Object Reference) {is=title K2Node_FunctionEntry}
   > UpdateTimeLeft!(
    execute
    self
    timeLeft:Text=Get EventTimeRemaining!(self:Fort Multi Item Vote Brief Entry Style B Object Reference=@n1) -> (EventTimeRemaining:Text)
   ) -> (then)
  ]}
 HideBriefText {nodes [
   HideBriefText() {is=title K2Node_FunctionEntry}
   > SetVisibility(
    self=[
     Get TimeRemainingDetails() -> (TimeRemainingDetails:Overlay Object Reference)
     Get VotingDetails() -> (VotingDetails:Overlay Object Reference)
    ]
    InVisibility=Hidden
   )
   > RemoveFromParent(self=Get NPCTextDisplay())
   > n1 = For Each Loop(Array:Array of Fort Item Fund Brief Object References=Get ItemEntryArray() -> (ItemEntryArray:Array of Fort Item Fund Brief Object References)) -> (Array Element:Fort Item Fund Brief Object Reference)
   > n2 = Cast To FortMultiItemVoteBriefEntry_StyleB(Object=@n1)
   > SetVisibility(
    self=[
     Get Text_SecondLine_L!(self:Fort Multi Item Vote Brief Entry Style B Object Reference=@n2) -> (Text_SecondLine_L:Common Text Object Reference)
     Get Text_SecondLine_R!(self:Fort Multi Item Vote Brief Entry Style B Object Reference=@n2) -> (Text_SecondLine_R:Common Text Object Reference)
     Get Text_CostDescription!(self:Fort Transaction Strip Object Reference=Get TransactionStrip_Main!(self:Fort Item Fund Brief Object Reference=@n2) -> (TransactionStrip_Main:Fort Transaction Strip Object Reference)) -> (Text_CostDescription:Common Text Object Reference)
    ]
    InVisibility=Hidden
   )
  ]}
 HideVoteDetails {nodes [
   HideVoteDetails() {is=title K2Node_FunctionEntry}
   > SetVisibility(self=Get VotingDetails(), InVisibility=Hidden)
  ]}
 UpdateSelectedState {nodes [
   UpdateSelectedState!() -> (then, SelectedIndex:int) {is=title K2Node_FunctionEntry}
   > n1 = Cast To FortMultiItemVoteBriefEntry_StyleB(Object=Get Item_Entry())
   n2 = int - int(A=2, B=Get SelectedIndex() -> (SelectedIndex:int=0)) {is=title K2Node_CallFunction}
   > SetSelectedVisuals!(
    execute
    self:Fort Multi Item Vote Brief Entry Style B Object Reference=@n1
    InSelected:bool="Equal (Integer)"(A=@n2, B=0) {is=title K2Node_CallFunction}
   ) -> (then)
   > n3 = Cast To FortMultiItemVoteBriefEntry_StyleB(Object=Get Item_EntrySecondary())
   > SetSelectedVisuals(self=@n3, InSelected="Equal (Integer)"(A=@n2, B=1))
   > n4 = Cast To FortMultiItemVoteBriefEntry_StyleB(Object=Get Item_EntryTertiary())
   > SetSelectedVisuals(self=@n4, InSelected="Equal (Integer)"(A=@n2, B=2) {is=title K2Node_PromotableOperator})
  ]}
 UpdateTimeLeft {nodes [
   n1 = UpdateTimeLeft!() -> (then, timeLeft:Text) {is=title K2Node_FunctionEntry}
   > SetText(
    self=Get CommonRichTextBlock_156() -> (CommonRichTextBlock_156:Common Rich Text Block Object Reference)
    InText=Format Text!(Format="LOCTEXT(\"Funding closes in <large>{0}</> \")", 0:Text=@n1) -> (Result)
   )
  ]}
 UpdateVoteData {nodes [
   UpdateVoteData() {is=title K2Node_FunctionEntry}
   > n1 = For Each Loop(Array=Get ItemEntryArray())
   > n2 = Cast To FortMultiItemVoteBriefEntry_StyleB(Object=@n1)
   > n3 = GetTileData!(execute, self:Fort Multi Item Vote Brief Entry Style B Object Reference=@n2) -> (then, CurrentProgress:double, ItemName:Text)
   > n4 = Set CurrentProgressValue(CurrentProgressValue:double=@n3) -> (Output_Get:double)
   > SetData!(
    execute
    self:B Vote Info Entry Object Reference="Get (a copy)"(
     Array:Array of B Vote Info Entry Object References=Get VoteInfoEntryArray() -> (VoteInfoEntryArray:Array of B Vote Info Entry Object References)
     Dimension 1=@n1.Array Index
    ) -> (Output:B Vote Info Entry Object Reference)
    Progress:double=@n4
    Name:Text=@n3.ItemName
   ) -> (then)
   > Branch(Condition="float > float"(A=Get CurrentProgressValue() -> (CurrentProgressValue:double), B=Get HighestProgressValue() -> (HighestProgressValue:double)))
   > Set HighestProgressValue(HighestProgressValue:double=Get CurrentProgressValue()) -> (Output_Get:double)
   n5 = Get VoteDetails_Switcher()
   SetDisableTransitionAnimation(execute=@n1.Completed, self=@n5, bDisableAnimation=false)
   > SetActiveWidgetIndex(self=@n5, Index=1)
   > n6 = For Each Loop(Array:Array of B Vote Info Entry Object References=Get VoteInfoEntryArray()) -> (Array Element:B Vote Info Entry Object Reference)
   n7 = GetCurrentProgress!(self:B Vote Info Entry Object Reference=@n6) -> (CurrentProgress:double)
   n8 = Get HighestProgressValue()
   > SetWinner!(
    execute
    self:B Vote Info Entry Object Reference=@n6
    IsWinner:bool=OR Boolean(A="Nearly Equal (Float)"(A=@n7, B=@n8, ErrorTolerance=0), B="float > float"(A=@n7, B=@n8))
   ) -> (then)
  ]}
]
variables [
 CachedSelectedIndex {type=integer, default=-1}
 DebugItem {type=FortItem, default=null}
 ItemEntryArray {type="array<FortItemFundBrief>", default []}
 VoteInfoEntryArray {type="array<B_VoteInfoEntry_C>", default []}
]
widgets {
 info {
  inheritedWidgetCount=4
  parentClass=/Script/FortniteConversationUI.FortMultiItemFundBrief
  rootWidgetClass=&Overlay
  widgetCount=42
 }
 widgets [
  Overlay_13 {widgetClassPath=&Overlay, widget=Overlay_13}
  GradientBacking {
   widgetClassPath=&Image
   bIsVariable=true
   parent=Overlay_13
   widget=GradientBacking
   properties {
    visibility=SelfHitTestInvisible
    brush {
     bIsDynamicallyLoaded=false
     drawAs=Image
     imageType=NoImage
     mirroring=NoMirror
     resourceName=None
     resourceObject=/FortniteConversationUI/Conversation/Materials/M_UI_Donation_FadedBG
     tiling=NoTile
     imageSize=<32 448>
     margin=<ltrb 0 0 0 0>
     outlineSettings {
      bUseBrushTransparency=false
      roundingType=HalfHeightRadius
      width=0
      color {colorUseRule=UseColor_Specified, specifiedColor=<rgba 0 0 0 0>}
      cornerRadii=<0 0 0 0>
     }
     tintColor {colorUseRule=UseColor_Specified, specifiedColor=<rgba 1 1 1 1>}
     uVRegion {bIsValid=false, max=<0 0>, min=<0 0>}
    }
    renderTransform {angle=0, scale=<2 4>, shear=<0 0>, translation=<0 0>}
   }
   slot {class=OverlaySlot, properties {horizontalAlignment=HAlign_Fill, verticalAlignment=VAlign_Center, padding=<ltrb 0 0 0 20>}}
  }
  NPCTextDisplay {
   widgetClassPath=&/NPCTextDisplay
   bInherited=true
   bIsVariable=true
   parent=Overlay_13
   widget=NPCTextDisplay
   properties {bIsMultiVersion=true}
   slot {class=OverlaySlot, properties {verticalAlignment=VAlign_Center, padding=<ltrb 64 0 0 384>}}
  }
  Overlay_15 {
   widgetClassPath=&Overlay
   parent=Overlay_13
   widget=Overlay_15
   slot {class=OverlaySlot, properties {horizontalAlignment=HAlign_Fill, verticalAlignment=VAlign_Center, padding=<ltrb 0 0 0 80>}}
  }
  SizeBox_1 {
   widgetClassPath=&SizeBox
   parent=Overlay_15
   widget=SizeBox_1
   properties {bOverride_HeightOverride=true, bOverride_WidthOverride=true, heightOverride=448, widthOverride=5000}
   slot {class=OverlaySlot, properties {horizontalAlignment=HAlign_Fill, verticalAlignment=VAlign_Center}}
  }
  Overlay_0 {widgetClassPath=&Overlay, parent=SizeBox_1, widget=Overlay_0, slot {class=SizeBoxSlot}}
  Darken {
   widgetClassPath=&Image
   bIsVariable=true
   parent=Overlay_0
   widget=Darken
   properties {colorAndOpacity=<rgba 0 0 0 0>}
   slot {class=OverlaySlot, properties {horizontalAlignment=HAlign_Fill, verticalAlignment=VAlign_Fill, padding=<ltrb 0 56 0 0>}}
  }
  HorizontalBox_22 {
   widgetClassPath=&HorizontalBox
   parent=Overlay_0
   widget=HorizontalBox_22
   slot {class=OverlaySlot, properties {horizontalAlignment=HAlign_Center, verticalAlignment=VAlign_Fill}}
  }
  Item_Entry {
   widgetClassPath=&/FortMultiItemVoteBriefEntry_StyleB
   bInherited=true
   bIsVariable=true
   parent=HorizontalBox_22
   widget=Item_Entry
   properties {renderTransform {angle=0, scale=<1 1>, shear=<0 0>, translation=<0 0>}}
   slot {class=HorizontalBoxSlot, properties {size {sizeRule=Fill, value=1}}}
  }
  Item_EntrySecondary {
   widgetClassPath=&FortMultiItemVoteBriefEntry_StyleB
   bInherited=true
   bIsVariable=true
   parent=HorizontalBox_22
   widget=Item_EntrySecondary
   properties {renderTransform {angle=0, scale=<1 1>, shear=<0 0>, translation=<0 0>}}
   slot {class=HorizontalBoxSlot, properties {size {sizeRule=Fill, value=1}}}
  }
  Item_EntryTertiary {
   widgetClassPath=&FortMultiItemVoteBriefEntry_StyleB
   bInherited=true
   bIsVariable=true
   parent=HorizontalBox_22
   widget=Item_EntryTertiary
   properties {renderTransform {angle=0, scale=<1 1>, shear=<0 0>, translation=<0 0>}}
   slot {class=HorizontalBoxSlot, properties {size {sizeRule=Fill, value=1}}}
  }
  GlobalEffort {
   widgetClassPath=&Overlay
   parent=Overlay_15
   widget=GlobalEffort
   properties {visibility=Collapsed}
   slot {class=OverlaySlot, properties {horizontalAlignment=HAlign_Center, padding=<ltrb 8 32 0 0>}}
  }
  Image_1 {
   widgetClassPath=&Image
   bIsVariable=true
   parent=GlobalEffort
   widget=Image_1
   properties {
    visibility=SelfHitTestInvisible
    brush {
     bIsDynamicallyLoaded=false
     drawAs=Box
     imageType=NoImage
     mirroring=NoMirror
     resourceName=None
     resourceObject=/FortUILibrary/Foundation/Art_ToolKit/AngledBoxes/M_UI_FundCenter_AnnouncementBacking_Shadow
     tiling=NoTile
     imageSize=<128 32>
     margin=<ltrb 0.5 0 0.5 0>
     outlineSettings {
      bUseBrushTransparency=false
      roundingType=HalfHeightRadius
      width=0
      color {colorUseRule=UseColor_Specified, specifiedColor=<rgba 0 0 0 0>}
      cornerRadii=<0 0 0 0>
     }
     tintColor {colorUseRule=UseColor_Specified, specifiedColor=<rgba 1 1 1 1>}
     uVRegion {bIsValid=false, max=<0 0>, min=<0 0>}
    }
    renderTransform {angle=0, scale=<1.125 1.5>, shear=<-8 0>, translation=<0 8>}
   }
   slot {class=OverlaySlot, properties {horizontalAlignment=HAlign_Fill, verticalAlignment=VAlign_Fill}}
  }
  Image_93 {
   widgetClassPath=&Image
   bIsVariable=true
   parent=GlobalEffort
   widget=Image_93
   properties {
    brush {
     bIsDynamicallyLoaded=false
     drawAs=Box
     imageType=NoImage
     mirroring=NoMirror
     resourceName=None
     resourceObject=/FortUILibrary/Foundation/Art_ToolKit/AngledBoxes/M_UI_FundCenter_AnnouncementBacking
     tiling=NoTile
     imageSize=<128 32>
     margin=<ltrb 0.5 0 0.5 0>
     outlineSettings {
      bUseBrushTransparency=false
      roundingType=HalfHeightRadius
      width=0
      color {colorUseRule=UseColor_Specified, specifiedColor=<rgba 0 0 0 0>}
      cornerRadii=<0 0 0 0>
     }
     tintColor {colorUseRule=UseColor_Specified, specifiedColor=<rgba 1 1 1 1>}
     uVRegion {bIsValid=false, max=<0 0>, min=<0 0>}
    }
    renderTransform {angle=0, scale=<1 1>, shear=<-8 0>, translation=<0 0>}
   }
   slot {class=OverlaySlot, properties {horizontalAlignment=HAlign_Fill, verticalAlignment=VAlign_Fill}}
  }
  HorizontalBox_0 {
   widgetClassPath=&HorizontalBox
   parent=GlobalEffort
   widget=HorizontalBox_0
   slot {class=OverlaySlot, properties {horizontalAlignment=HAlign_Center, verticalAlignment=VAlign_Center, padding=<ltrb 18 16 20 16>}}
  }
  Image_229 {
   widgetClassPath=&Image
   bIsVariable=true
   parent=HorizontalBox_0
   widget=Image_229
   properties {
    renderOpacity=0.9
    brush {
     bIsDynamicallyLoaded=false
     drawAs=Image
     imageType=NoImage
     mirroring=NoMirror
     resourceName=None
     resourceObject=/Game/UI/Foundation/Textures/Icons/Stars/Stars-Full-64
     tiling=NoTile
     imageSize=<24 24>
     margin=<ltrb 0 0 0 0>
     outlineSettings {
      bUseBrushTransparency=false
      roundingType=HalfHeightRadius
      width=0
      color {colorUseRule=UseColor_Specified, specifiedColor=<rgba 0 0 0 0>}
      cornerRadii=<0 0 0 0>
     }
     tintColor {colorUseRule=UseColor_Specified, specifiedColor=<rgba 1 1 1 1>}
     uVRegion {bIsValid=false, max=<0 0>, min=<0 0>}
    }
    colorAndOpacity=<rgba 0.6939 0.0343 0.0242 1>
    renderTransform {angle=-8, scale=<1 1>, shear=<0 0>, translation=<0 0>}
   }
   slot {class=HorizontalBoxSlot, properties {horizontalAlignment=HAlign_Center, verticalAlignment=VAlign_Center}}
  }
  SizeBox_0 {
   widgetClassPath=&SizeBox
   parent=HorizontalBox_0
   widget=SizeBox_0
   properties {bOverride_MaxDesiredWidth=true, clipping=OnDemand, maxDesiredWidth=400, minDesiredWidth=400, widthOverride=400}
   slot {class=HorizontalBoxSlot, properties {verticalAlignment=VAlign_Center}}
  }
  ScaleBox_0 {
   widgetClassPath=&ScaleBox
   parent=SizeBox_0
   widget=ScaleBox_0
   properties {clipping=OnDemand, stretch=ScaleToFitX, stretchDirection=DownOnly}
   slot {class=SizeBoxSlot}
  }
  Text_TitleBar {
   widgetClassPath=&CommonTextBlock
   bIsVariable=true
   parent=ScaleBox_0
   widget=Text_TitleBar
   properties {
    clipping=OnDemand
    justification=Center
    lineHeightPercentage=0.9
    text="Contribute to the global effort!"
    textTransformPolicy=ToUpper
    colorAndOpacity {colorUseRule=UseColor_Specified, specifiedColor=<rgba 0.1195 0.1144 0.0545 1>}
    font {
     bForceMonospaced=false
     bMaterialIsStencil=false
     fontMaterial=null
     fontObject=&Burbank
     letterSpacing=0
     monospacedWidth=1
     size=15
     skewAmount=0
     typefaceFontName=BurbankRegular-Black
     outlineSettings {
      bApplyOutlineToDropShadows=true
      bMiteredCorners=false
      bSeparateFillAlpha=false
      outlineMaterial=null
      outlineSize=0
      outlineColor=<rgba 0.0144 0.0194 0.0065 0>
     }
    }
    margin=<ltrb 8 -2 8 0>
    renderTransform {angle=0, scale=<1 1>, shear=<-12 0>, translation=<0 0>}
    shadowOffset=<0 0>
   }
   slot {class=ScaleBoxSlot}
  }
  Image {
   widgetClassPath=&Image
   bIsVariable=true
   parent=HorizontalBox_0
   widget=Image
   properties {
    renderOpacity=0.9
    brush {
     bIsDynamicallyLoaded=false
     drawAs=Image
     imageType=NoImage
     mirroring=NoMirror
     resourceName=None
     resourceObject=&Stars-Full-64
     tiling=NoTile
     imageSize=<24 24>
     margin=<ltrb 0 0 0 0>
     outlineSettings {
      bUseBrushTransparency=false
      roundingType=HalfHeightRadius
      width=0
      color {colorUseRule=UseColor_Specified, specifiedColor=<rgba 0 0 0 0>}
      cornerRadii=<0 0 0 0>
     }
     tintColor {colorUseRule=UseColor_Specified, specifiedColor=<rgba 1 1 1 1>}
     uVRegion {bIsValid=false, max=<0 0>, min=<0 0>}
    }
    colorAndOpacity=<rgba 0.6939 0.0343 0.0242 1>
    renderTransform {angle=8, scale=<1 1>, shear=<0 0>, translation=<0 0>}
   }
   slot {class=HorizontalBoxSlot, properties {horizontalAlignment=HAlign_Center, verticalAlignment=VAlign_Center}}
  }
  Overlay_1 {
   widgetClassPath=&Overlay
   parent=Overlay_15
   widget=Overlay_1
   properties {visibility=Collapsed, renderTransform {angle=0, scale=<1 1>, shear=<0 0>, translation=<0 16>}}
   slot {class=OverlaySlot, properties {horizontalAlignment=HAlign_Center, verticalAlignment=VAlign_Center}}
  }
  Overlay_2 {
   widgetClassPath=&Overlay
   parent=Overlay_1
   widget=Overlay_2
   slot {class=OverlaySlot, properties {horizontalAlignment=HAlign_Center, verticalAlignment=VAlign_Center}}
  }
  Image_Shadow {
   widgetClassPath=&Image
   bIsVariable=true
   parent=Overlay_2
   widget=Image_Shadow
   properties {
    renderOpacity=0.5
    brush {
     bIsDynamicallyLoaded=false
     drawAs=Image
     imageType=NoImage
     mirroring=NoMirror
     resourceName=None
     resourceObject=/FortniteConversationUI/Conversation/Materials/M_UI_DonationFund_BadgeCircle_Default_Secondary
     tiling=NoTile
     imageSize=<168 168>
     margin=<ltrb 0 0 0 0>
     outlineSettings {
      bUseBrushTransparency=false
      roundingType=HalfHeightRadius
      width=0
      color {colorUseRule=UseColor_Specified, specifiedColor=<rgba 0 0 0 0>}
      cornerRadii=<0 0 0 0>
     }
     tintColor {colorUseRule=UseColor_Specified, specifiedColor=<rgba 1 1 1 1>}
     uVRegion {bIsValid=false, max=<0 0>, min=<0 0>}
    }
   }
   slot {class=OverlaySlot, properties {horizontalAlignment=HAlign_Fill, verticalAlignment=VAlign_Fill}}
  }
  Image_Burst {
   widgetClassPath=&Image
   bIsVariable=true
   parent=Overlay_2
   widget=Image_Burst
   properties {
    brush {
     bIsDynamicallyLoaded=false
     drawAs=Image
     imageType=NoImage
     mirroring=NoMirror
     resourceName=None
     resourceObject=/FortniteConversationUI/Conversation/Materials/M_UI_DonationFund_StripBacking_GreenBurst
     tiling=NoTile
     imageSize=<32 32>
     margin=<ltrb 0 0 0 0>
     outlineSettings {
      bUseBrushTransparency=false
      roundingType=HalfHeightRadius
      width=0
      color {colorUseRule=UseColor_Specified, specifiedColor=<rgba 0 0 0 0>}
      cornerRadii=<0 0 0 0>
     }
     tintColor {colorUseRule=UseColor_Specified, specifiedColor=<rgba 1 1 1 1>}
     uVRegion {bIsValid=false, max=<0 0>, min=<0 0>}
    }
    renderTransform {angle=0, scale=<0.5 0.5>, shear=<0 0>, translation=<0 0>}
   }
   slot {class=OverlaySlot, properties {horizontalAlignment=HAlign_Fill, verticalAlignment=VAlign_Fill}}
  }
  Image_Circle {
   widgetClassPath=&Image
   bIsVariable=true
   parent=Overlay_2
   widget=Image_Circle
   properties {brush {
     bIsDynamicallyLoaded=false
     drawAs=Image
     imageType=NoImage
     mirroring=NoMirror
     resourceName=None
     resourceObject=/FortniteConversationUI/Conversation/Materials/M_UI_DonationFund_BadgeCircle_Default
     tiling=NoTile
     imageSize=<112 112>
     margin=<ltrb 0 0 0 0>
     outlineSettings {
      bUseBrushTransparency=false
      roundingType=HalfHeightRadius
      width=0
      color {colorUseRule=UseColor_Specified, specifiedColor=<rgba 0 0 0 0>}
      cornerRadii=<0 0 0 0>
     }
     tintColor {colorUseRule=UseColor_Specified, specifiedColor=<rgba 1 1 1 1>}
     uVRegion {bIsValid=false, max=<0 0>, min=<0 0>}
    }}
   slot {class=OverlaySlot, properties {horizontalAlignment=HAlign_Center, verticalAlignment=VAlign_Center}}
  }
  AnimBoundFundedBorder {
   widgetClassPath=&Border
   parent=Overlay_2
   widget=AnimBoundFundedBorder
   properties {
    background {
     bIsDynamicallyLoaded=false
     drawAs=NoDrawType
     imageType=NoImage
     mirroring=NoMirror
     resourceName=None
     resourceObject=null
     tiling=NoTile
     imageSize=<32 32>
     margin=<ltrb 0 0 0 0>
     outlineSettings {
      bUseBrushTransparency=false
      roundingType=HalfHeightRadius
      width=0
      color {colorUseRule=UseColor_Specified, specifiedColor=<rgba 0 0 0 0>}
      cornerRadii=<0 0 0 0>
     }
     tintColor {colorUseRule=UseColor_Specified, specifiedColor=<rgba 1 1 1 1>}
     uVRegion {bIsValid=false, max=<0 0>, min=<0 0>}
    }
    padding=<ltrb 0 0 0 0>
   }
   slot {class=OverlaySlot, properties {horizontalAlignment=HAlign_Center, verticalAlignment=VAlign_Center}}
  }
  Image_Arrow {
   widgetClassPath=&Image
   bIsVariable=true
   parent=AnimBoundFundedBorder
   widget=Image_Arrow
   properties {
    brush {
     bIsDynamicallyLoaded=false
     drawAs=Image
     imageType=NoImage
     mirroring=NoMirror
     resourceName=None
     resourceObject=/FortniteCoreUI/UI/Login/Art/M_Arrow_DF
     tiling=NoTile
     imageSize=<90 90>
     margin=<ltrb 0 0 0 0>
     outlineSettings {
      bUseBrushTransparency=false
      roundingType=HalfHeightRadius
      width=0
      color {colorUseRule=UseColor_Specified, specifiedColor=<rgba 0 0 0 0>}
      cornerRadii=<0 0 0 0>
     }
     tintColor {colorUseRule=UseColor_Specified, specifiedColor=<rgba 1 1 1 1>}
     uVRegion {bIsValid=false, max=<0 0>, min=<0 0>}
    }
    colorAndOpacity=<rgba 0 0.9131 1 1>
   }
   slot {class=BorderSlot, properties {padding=<ltrb 0 0 0 0>}}
  }
  Image_Check {
   widgetClassPath=&Image
   bIsVariable=true
   parent=Overlay_2
   widget=Image_Check
   properties {
    renderOpacity=0
    visibility=SelfHitTestInvisible
    brush {
     bIsDynamicallyLoaded=false
     drawAs=Image
     imageType=NoImage
     mirroring=NoMirror
     resourceName=None
     resourceObject=/Game/UI/Foundation/Textures/Icons/Checks/T-Icon-Check-64
     tiling=NoTile
     imageSize=<64 64>
     margin=<ltrb 0 0 0 0>
     outlineSettings {
      bUseBrushTransparency=false
      roundingType=HalfHeightRadius
      width=0
      color {colorUseRule=UseColor_Specified, specifiedColor=<rgba 0 0 0 0>}
      cornerRadii=<0 0 0 0>
     }
     tintColor {colorUseRule=UseColor_Specified, specifiedColor=<rgba 1 1 1 1>}
     uVRegion {bIsValid=false, max=<0 0>, min=<0 0>}
    }
    colorAndOpacity=<rgba 0 0.7529 0.2502 1>
   }
   slot {class=OverlaySlot, properties {horizontalAlignment=HAlign_Center, verticalAlignment=VAlign_Center}}
  }
  CommonTextBlock_82 {
   widgetClassPath=&CommonTextBlock
   parent=Overlay_1
   widget=CommonTextBlock_82
   properties {
    justification=Center
    text=OR
    visibility=Collapsed
    colorAndOpacity {colorUseRule=UseColor_Specified, specifiedColor=<rgba 0.0144 0.0194 0.0065 1>}
    font {
     bForceMonospaced=false
     bMaterialIsStencil=false
     fontMaterial=null
     fontObject=&Burbank
     letterSpacing=0
     monospacedWidth=1
     size=32
     skewAmount=0
     typefaceFontName=BurbankRegular-Black
     outlineSettings {
      bApplyOutlineToDropShadows=false
      bMiteredCorners=false
      bSeparateFillAlpha=false
      outlineMaterial=null
      outlineSize=4
      outlineColor=<rgba 1 0.9375 0.625 1>
     }
    }
    renderTransform {angle=0, scale=<1 1>, shear=<-12 0>, translation=<0 88>}
    shadowOffset=<0 0>
   }
   slot {class=OverlaySlot, properties {horizontalAlignment=HAlign_Center, verticalAlignment=VAlign_Center}}
  }
  SizeBox_2 {
   widgetClassPath=&SizeBox
   parent=Overlay_13
   widget=SizeBox_2
   properties {bOverride_WidthOverride=true, widthOverride=625}
   slot {class=OverlaySlot, properties {verticalAlignment=VAlign_Center, padding=<ltrb 50 350 0 0>}}
  }
  VotingDetails {widgetClassPath=&Overlay, bIsVariable=true, parent=SizeBox_2, widget=VotingDetails, slot {class=SizeBoxSlot}}
  Image_118 {
   widgetClassPath=&Image
   bIsVariable=true
   parent=VotingDetails
   widget=Image_118
   properties {colorAndOpacity=<rgba 0 0 0 0.5>}
   slot {class=OverlaySlot, properties {horizontalAlignment=HAlign_Fill, verticalAlignment=VAlign_Fill}}
  }
  VoteDetails_Switcher {
   widgetClassPath=/Script/CommonUI.CommonAnimatedSwitcher
   bIsVariable=true
   parent=VotingDetails
   widget=VoteDetails_Switcher
   properties {transitionCurveType=QuadOut}
   slot {class=OverlaySlot, properties {horizontalAlignment=HAlign_Fill}}
  }
  CommonTextBlock_149 {
   widgetClassPath=&CommonTextBlock
   parent=VoteDetails_Switcher
   widget=CommonTextBlock_149
   properties {
    autoWrapText=true
    justification=Center
    text=Fund a building to see who is in the lead
    textTransformPolicy=ToUpper
    font {
     bForceMonospaced=false
     bMaterialIsStencil=false
     fontMaterial=null
     fontObject=&BurbankSmall
     letterSpacing=0
     monospacedWidth=1
     size=20
     skewAmount=0
     typefaceFontName=Bold
     outlineSettings {
      bApplyOutlineToDropShadows=false
      bMiteredCorners=false
      bSeparateFillAlpha=false
      outlineMaterial=null
      outlineSize=0
      outlineColor=<rgba 0 0 0 1>
     }
    }
    shadowOffset=<0 0>
   }
   slot {class=WidgetSwitcherSlot, properties {padding=<ltrb 10 10 10 10>}}
  }
  VerticalBox_120 {
   widgetClassPath=&VerticalBox
   parent=VoteDetails_Switcher
   widget=VerticalBox_120
   slot {class=WidgetSwitcherSlot, properties {padding=<ltrb 10 10 10 10>}}
  }
  B_VoteInfoEntry_0 {
   widgetClassPath=&/B_VoteInfoEntry
   bIsVariable=true
   parent=VerticalBox_120
   widget=B_VoteInfoEntry_0
   slot {class=VerticalBoxSlot, properties {padding=<ltrb 0 0 0 5>}}
  }
  B_VoteInfoEntry_1 {
   widgetClassPath=&B_VoteInfoEntry
   bIsVariable=true
   parent=VerticalBox_120
   widget=B_VoteInfoEntry_1
   slot {class=VerticalBoxSlot, properties {padding=<ltrb 0 0 0 5>}}
  }
  B_VoteInfoEntry_2 {
   widgetClassPath=&B_VoteInfoEntry
   bIsVariable=true
   parent=VerticalBox_120
   widget=B_VoteInfoEntry_2
   slot {class=VerticalBoxSlot}
  }
  SizeBox {
   widgetClassPath=&SizeBox
   parent=Overlay_13
   widget=SizeBox
   properties {visibility=HitTestInvisible, widthOverride=450}
   slot {class=OverlaySlot, properties {horizontalAlignment=HAlign_Right, verticalAlignment=VAlign_Center, padding=<ltrb 50 0 50 400>}}
  }
  TimeRemainingDetails {widgetClassPath=&Overlay, bIsVariable=true, parent=SizeBox, widget=TimeRemainingDetails, slot {class=SizeBoxSlot}}
  Image_2 {
   widgetClassPath=&Image
   bIsVariable=true
   parent=TimeRemainingDetails
   widget=Image_2
   properties {colorAndOpacity=<rgba 0 0 0 0.5>}
   slot {class=OverlaySlot, properties {horizontalAlignment=HAlign_Fill, verticalAlignment=VAlign_Fill}}
  }
  CommonRichTextBlock_156 {
   widgetClassPath=&CommonRichTextBlock
   bIsVariable=true
   parent=TimeRemainingDetails
   widget=CommonRichTextBlock_156
   properties {
    inlineIconDisplayMode=TextOnly
    justification=Right
    text="FUNDING TIME <large>LEFT</> SHOWN HERE"
    textStyleSet=&/VoteBrief_RichTextStyle
    defaultTextStyleOverride {
     overflowPolicy=Clip
     transformPolicy=None
     colorAndOpacity {colorUseRule=UseColor_Specified, specifiedColor=<rgba 1 1 1 1>}
     font {
      bForceMonospaced=false
      bMaterialIsStencil=false
      fontMaterial=null
      fontObject=&NotoSans
      letterSpacing=0
      monospacedWidth=1
      size=24
      skewAmount=0
      typefaceFontName=NotoSans-Regular
      outlineSettings {
       bApplyOutlineToDropShadows=false
       bMiteredCorners=false
       bSeparateFillAlpha=false
       outlineMaterial=null
       outlineSize=0
       outlineColor=<rgba 0 0 0 1>
      }
     }
     highlightColor {colorUseRule=UseColor_Specified, specifiedColor=<rgba 0 0 0 1>}
     highlightShape {
      bIsDynamicallyLoaded=false
      drawAs=Image
      imageType=NoImage
      mirroring=NoMirror
      resourceName=None
      resourceObject=null
      tiling=NoTile
      imageSize=<32 32>
      margin=<ltrb 0 0 0 0>
      outlineSettings {
       bUseBrushTransparency=false
       roundingType=HalfHeightRadius
       width=0
       color {colorUseRule=UseColor_Specified, specifiedColor=<rgba 0 0 0 0>}
       cornerRadii=<0 0 0 0>
      }
      tintColor {colorUseRule=UseColor_Specified, specifiedColor=<rgba 1 1 1 1>}
      uVRegion {bIsValid=false, max=<0 0>, min=<0 0>}
     }
     selectedBackgroundColor {colorUseRule=UseColor_Foreground, specifiedColor=<rgba 1 0 1 1>}
     shadowColorAndOpacity=<rgba 0 0 0 1>
     shadowOffset=<0 0>
     strikeBrush {
      bIsDynamicallyLoaded=false
      drawAs=Image
      imageType=NoImage
      mirroring=NoMirror
      resourceName=None
      resourceObject=null
      tiling=NoTile
      imageSize=<32 32>
      margin=<ltrb 0 0 0 0>
      outlineSettings {
       bUseBrushTransparency=false
       roundingType=HalfHeightRadius
       width=0
       color {colorUseRule=UseColor_Specified, specifiedColor=<rgba 0 0 0 0>}
       cornerRadii=<0 0 0 0>
      }
      tintColor {colorUseRule=UseColor_Specified, specifiedColor=<rgba 1 1 1 1>}
      uVRegion {bIsValid=false, max=<0 0>, min=<0 0>}
     }
     underlineBrush {
      bIsDynamicallyLoaded=false
      drawAs=Image
      imageType=NoImage
      mirroring=NoMirror
      resourceName=None
      resourceObject=null
      tiling=NoTile
      imageSize=<32 32>
      margin=<ltrb 0 0 0 0>
      outlineSettings {
       bUseBrushTransparency=false
       roundingType=HalfHeightRadius
       width=0
       color {colorUseRule=UseColor_Specified, specifiedColor=<rgba 0 0 0 0>}
       cornerRadii=<0 0 0 0>
      }
      tintColor {colorUseRule=UseColor_Specified, specifiedColor=<rgba 1 1 1 1>}
      uVRegion {bIsValid=false, max=<0 0>, min=<0 0>}
     }
    }
   }
   slot {class=OverlaySlot, properties {horizontalAlignment=HAlign_Right, padding=<ltrb 10 10 10 10>}}
  }
 ]
}
