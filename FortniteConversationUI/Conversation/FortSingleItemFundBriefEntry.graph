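# Widget Blueprint on FortItemFundBrief that displays one fundable item in conversation UI, with icon, name, funded percentage, progress bar and transaction strip.
# Construct sets the icon's texture parameter name, plays Intro and binds to the player controller's client donation success delegate, which Destruct unbinds.
# OnDonationSuccess triggers the strip's DonationAnim and the Funded animation, which animates material shine and bar scale.
# The global-effort banner and currency count start collapsed.
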WidgetBlueprint /FortniteConversationUI/Conversation/FortSingleItemFundBriefEntry
animations [
 Intro {
  binding_count=5
  length_seconds=0.55
  object=&self:Intro
  sequence {bindings [
    SizeBox_2 {
     id=#1
     possessed_class=SizeBox
     tracks [
      {
       type=MovieSceneFloatTrack
       display_name=Render Opacity
       sections [{type=MovieSceneFloatSection, channels [None {type=MovieSceneScriptingFloatChannel, keys [[0,0],[4.3333,0],[8.3333,1]]}]}]
      }
      {
       type=MovieScene2DTransformTrack
       sections [{
         type=MovieScene2DTransformSection
         channels [
          Translation.X {type=MovieSceneScriptingFloatChannel, keys [[4.3333,-128],[8.3333,0]]}
          Translation.Y {type=MovieSceneScriptingFloatChannel, const [24,1]}
          Angle {type=MovieSceneScriptingFloatChannel, const [0,1]}
          Scale.X {type=MovieSceneScriptingFloatChannel, const [1,1]}
          Scale.Y {type=MovieSceneScriptingFloatChannel, const [1,1]}
          Shear.X {type=MovieSceneScriptingFloatChannel, const [0,1]}
          Shear.Y {type=MovieSceneScriptingFloatChannel, const [0,1]}
         ]
        }]
      }
     ]
    }
    GlobalEffort {
     id=#2
     possessed_class=Overlay
     tracks [
      {
       type=MovieSceneFloatTrack
       display_name=Render Opacity
       sections [{type=MovieSceneFloatSection, channels [None {type=MovieSceneScriptingFloatChannel, keys [[6,0],[8,1]]}]}]
      }
      {
       type=MovieScene2DTransformTrack
       sections [{
         type=MovieScene2DTransformSection
         channels [
          Translation.X {type=MovieSceneScriptingFloatChannel}
          Translation.Y {type=MovieSceneScriptingFloatChannel}
          Angle {type=MovieSceneScriptingFloatChannel}
          Scale.X {type=MovieSceneScriptingFloatChannel, keys [[6,1.25],[8,0.95],[11,1]]}
          Scale.Y {type=MovieSceneScriptingFloatChannel, keys [[6,1.25],[8,0.95],[11,1]]}
          Shear.X {type=MovieSceneScriptingFloatChannel}
          Shear.Y {type=MovieSceneScriptingFloatChannel}
         ]
        }]
      }
     ]
    }
    FundingPercent_Overlay {
     id=#3
     possessed_class=Overlay
     tracks [
      {
       type=MovieSceneFloatTrack
       display_name=Render Opacity
       sections [{type=MovieSceneFloatSection, channels [None {type=MovieSceneScriptingFloatChannel, keys [[3.3333,0],[4.3333,1]]}]}]
      }
      {
       type=MovieScene2DTransformTrack
       sections [{
         type=MovieScene2DTransformSection
         channels [
          Scale.X {type=MovieSceneScriptingFloatChannel, keys [[3.3333,1.25],[4.3333,0.95],[6.3333,1]]}
          Scale.Y {type=MovieSceneScriptingFloatChannel, keys [[3.3333,1.25],[4.3333,0.95],[6.3333,1]]}
         ]
        }]
      }
     ]
    }
    Darken {
     id=#4
     possessed_class=Image
     tracks [{
       type=MovieSceneFloatTrack
       display_name=Render Opacity
       sections [{type=MovieSceneFloatSection, channels [None {type=MovieSceneScriptingFloatChannel, keys [[3,0],[7,1]]}]}]
      }]
    }
    Overlay_0 {
     id=#5
     possessed_class=Overlay
     tracks [
      {
       type=MovieSceneFloatTrack
       display_name=Render Opacity
       sections [{type=MovieSceneFloatSection, channels [None {type=MovieSceneScriptingFloatChannel, keys [[0,0],[2,1]]}]}]
      }
      {
       type=MovieScene2DTransformTrack
       sections [{
         type=MovieScene2DTransformSection
         channels [
          Translation.X {type=MovieSceneScriptingFloatChannel, const [0,1]}
          Translation.Y {type=MovieSceneScriptingFloatChannel, keys [[0,128],[5,0]]}
          Angle {type=MovieSceneScriptingFloatChannel, const [0,1]}
          Scale.X {type=MovieSceneScriptingFloatChannel, const [1,1]}
          Scale.Y {type=MovieSceneScriptingFloatChannel, const [1,1]}
          Shear.X {type=MovieSceneScriptingFloatChannel, const [0,1]}
          Shear.Y {type=MovieSceneScriptingFloatChannel, const [0,1]}
         ]
        }]
      }
     ]
    }
   ]}
 }
 Funded {
  binding_count=4
  length_seconds=1
  object=&self:Funded
  sequence {bindings [
    Big_BG_Strip {
     id=#6
     possessed_class=Image
     tracks [{
       type=MovieSceneWidgetMaterialTrack
       sections [{type=MovieSceneParameterSection, channels [OffsetShine {type=MovieSceneScriptingFloatChannel, keys [[0,0],[20,1]]}]}]
      }]
    }
    Border_0 {
     id=#7
     possessed_class=Border
     tracks [
      {
       type=MovieSceneWidgetMaterialTrack
       sections [{
         type=MovieSceneParameterSection
         channels [
          BackingAlpha {type=MovieSceneScriptingFloatChannel, keys [[0,0.5],[2,1],[10,0.5]]}
          OffsetShine {type=MovieSceneScriptingFloatChannel, keys [[0,0],[15,1.475]]}
         ]
        }]
      }
      {
       type=MovieScene2DTransformTrack
       sections [{
         type=MovieScene2DTransformSection
         channels [
          Scale.X {type=MovieSceneScriptingFloatChannel, keys [[0.0003,1],[1,0.95],[2.6667,1.15],[6.6667,1]]}
          Scale.Y {type=MovieSceneScriptingFloatChannel, keys [[0.0003,1],[1,0.95],[2.6667,1.15],[6.6667,1]]}
         ]
        }]
      }
     ]
    }
    Image_229 {
     id=#8
     possessed_class=Image
     tracks [{
       type=MovieScene2DTransformTrack
       sections [{type=MovieScene2DTransformSection, channels [Angle {type=MovieSceneScriptingFloatChannel, keys [[0,-8],[10,352]]}]}]
      }]
    }
    Image_2 {
     id=#9
     possessed_class=Image
     tracks [{
       type=MovieScene2DTransformTrack
       sections [{type=MovieScene2DTransformSection, channels [Angle {type=MovieSceneScriptingFloatChannel, keys [[0,8],[10,-352]]}]}]
      }]
    }
   ]}
 }
]
graphs [
 EventGraph {nodes [
   Event Destruct()
   > n1 = Cast To Athena_PlayerController!(execute, Object=GetOwningPlayer()) -> (then, CastFailed, AsAthena Player Controller:Athena Player Controller Object Reference) {is=title K2Node_DynamicCast}
   > Unbind Event from On Client Donation Success Delegate(self:Fort Player Controller Athena Object Reference=@n1, Delegate=Create Event()) {is=title K2Node_RemoveDelegate}
   Event Construct()
   > SetMaterialTextureParamName!(
    execute
    self:Common Lazy Image Object Reference=Get LazyImage_Icon() -> (LazyImage_Icon:Fort Lazy Image Object Reference)
    TextureParamName:Name=IconImage
   ) -> (then)
   > PlayAnimation(InAnimation=Get Intro())
   > n2 = Cast To Athena_PlayerController(Object=GetOwningPlayer())
   > Bind Event to On Client Donation Success Delegate(self:Fort Player Controller Athena Object Reference=@n2, Delegate=Create Event()) {is=title K2Node_AddDelegate}
   Event PreConstruct()
  ]}
 OnDonationSuccess {nodes [
   OnDonationSuccess() {is=title K2Node_FunctionEntry}
   > n1 = Cast To FortTransactionStripNew!(execute, Object=Get TransactionStrip_Main() -> (TransactionStrip_Main:Fort Transaction Strip Object Reference)) -> (then, CastFailed, AsFort Transaction Strip New:Fort Transaction Strip New Object Reference) {is=title K2Node_DynamicCast}
   > DonationAnim(self:Fort Transaction Strip New Object Reference=@n1)
   > PlayAnimation(InAnimation=Get Funded() -> (Funded:Widget Animation Object Reference))
  ]}
]
variables [DebugItem {type=FortItem, default=null}]
widgets {
 info {inheritedWidgetCount=6, parentClass=/Script/FortniteConversationUI.FortItemFundBrief, rootWidgetClass=&Overlay, widgetCount=37}
 widgets [
  Overlay_2 {widgetClassPath=&Overlay, widget=Overlay_2}
  Overlay_0 {
   widgetClassPath=&Overlay
   parent=Overlay_2
   widget=Overlay_0
   slot {class=OverlaySlot, properties {horizontalAlignment=HAlign_Fill, verticalAlignment=VAlign_Center}}
  }
  SizeBox_3 {
   widgetClassPath=&SizeBox
   parent=Overlay_0
   widget=SizeBox_3
   properties {bOverride_HeightOverride=true, bOverride_WidthOverride=true, heightOverride=448, minDesiredWidth=2050, widthOverride=5000}
   slot {class=OverlaySlot, properties {horizontalAlignment=HAlign_Fill, verticalAlignment=VAlign_Center}}
  }
  CanBuy {widgetClassPath=&Overlay, bIsVariable=true, parent=SizeBox_3, widget=CanBuy, slot {class=SizeBoxSlot}}
  Darken {
   widgetClassPath=&Image
   bIsVariable=true
   parent=CanBuy
   widget=Darken
   properties {colorAndOpacity=<rgba 0 0 0 0>}
   slot {class=OverlaySlot, properties {horizontalAlignment=HAlign_Fill, verticalAlignment=VAlign_Fill, padding=<ltrb 0 56 0 0>}}
  }
  Big_BG_Strip {
   widgetClassPath=&Image
   bIsVariable=true
   parent=CanBuy
   widget=Big_BG_Strip
   properties {brush {
     bIsDynamicallyLoaded=false
     drawAs=Image
     imageType=NoImage
     mirroring=NoMirror
     resourceName=None
     resourceObject=/FortniteConversationUI/Conversation/Materials/M_UI_DonationFund_BlueprintBacking_DarkBlue
     tiling=NoTile
     imageSize=<32 32>
     margin=<ltrb 0 0 0 0>
     outlineSettings {
      bUseBrushTransparency=false
      roundingType=HalfHeightRadius
      width=0
      color {colorUseRule=UseColor_Specified, specifiedColor=<rgba 0 0 0 0>}
      cornerRadii=<0 0 0 0>
     }
     tintColor {colorUseRule=UseColor_Specified, specifiedColor=<rgba 1 1 1 1>}
     uVRegion {bIsValid=false, max=<0 0>, min=<0 0>}
    }}
   slot {class=OverlaySlot, properties {horizontalAlignment=HAlign_Fill, verticalAlignment=VAlign_Fill}}
  }
  LazyImage_Icon {
   widgetClassPath=&FortLazyImage
   bInherited=true
   bIsVariable=true
   parent=CanBuy
   widget=LazyImage_Icon
   properties {brush {
     bIsDynamicallyLoaded=false
     drawAs=Image
     imageType=NoImage
     mirroring=NoMirror
     resourceName=None
     resourceObject=/FortniteConversationUI/Conversation/Materials/T_UI_WarEffort_MountedTurret
     tiling=NoTile
     imageSize=<1024 512>
     margin=<ltrb 0 0 0 0>
     outlineSettings {
      bUseBrushTransparency=false
      roundingType=HalfHeightRadius
      width=0
      color {colorUseRule=UseColor_Specified, specifiedColor=<rgba 0 0 0 0>}
      cornerRadii=<0 0 0 0>
     }
     tintColor {colorUseRule=UseColor_Specified, specifiedColor=<rgba 1 1 1 1>}
     uVRegion {bIsValid=false, max=<0 0>, min=<0 0>}
    }}
   slot {class=OverlaySlot, properties {horizontalAlignment=HAlign_Right, verticalAlignment=VAlign_Fill}}
  }
  BG_Strip_1 {
   widgetClassPath=&Image
   bIsVariable=true
   parent=CanBuy
   widget=BG_Strip_1
   properties {
    visibility=Collapsed
    brush {
     bIsDynamicallyLoaded=false
     drawAs=Image
     imageType=NoImage
     mirroring=NoMirror
     resourceName=None
     resourceObject=/FortUILibrary/Foundation/Art_ToolKit/AngledBoxes/M_UI_FundCenter_BottomStrip_1
     tiling=NoTile
     imageSize=<930 15>
     margin=<ltrb 0 0 0 0>
     outlineSettings {
      bUseBrushTransparency=false
      roundingType=HalfHeightRadius
      width=0
      color {colorUseRule=UseColor_Specified, specifiedColor=<rgba 0 0 0 0>}
      cornerRadii=<0 0 0 0>
     }
     tintColor {colorUseRule=UseColor_Specified, specifiedColor=<rgba 1 1 1 1>}
     uVRegion {bIsValid=false, max=<0 0>, min=<0 0>}
    }
    renderTransformPivot=<0.5 1>
   }
   slot {class=OverlaySlot, properties {horizontalAlignment=HAlign_Fill, verticalAlignment=VAlign_Bottom}}
  }
  VerticalBox_2 {
   widgetClassPath=&VerticalBox
   parent=CanBuy
   widget=VerticalBox_2
   slot {class=OverlaySlot, properties {horizontalAlignment=HAlign_Fill, verticalAlignment=VAlign_Fill}}
  }
  HorizontalBox_0 {
   widgetClassPath=&HorizontalBox
   parent=VerticalBox_2
   widget=HorizontalBox_0
   slot {class=VerticalBoxSlot, properties {size {sizeRule=Fill, value=1}}}
  }
  ProgressInfo {
   widgetClassPath=&VerticalBox
   parent=HorizontalBox_0
   widget=ProgressInfo
   slot {
    class=HorizontalBoxSlot
    properties {horizontalAlignment=HAlign_Left, verticalAlignment=VAlign_Center, padding=<ltrb 128 24 0 0>, size {sizeRule=Fill, value=0.5}}
   }
  }
  VBox_NameFunded {
   widgetClassPath=&VerticalBox
   parent=ProgressInfo
   widget=VBox_NameFunded
   slot {class=VerticalBoxSlot, properties {horizontalAlignment=HAlign_Center, padding=<ltrb 0 0 0 16>}}
  }
  FundingPercent_Overlay {
   widgetClassPath=&Overlay
   bIsVariable=true
   parent=VBox_NameFunded
   widget=FundingPercent_Overlay
   slot {class=VerticalBoxSlot, properties {horizontalAlignment=HAlign_Right, padding=<ltrb 0 0 32 0>}}
  }
  Image_93 {
   widgetClassPath=&Image
   bIsVariable=true
   parent=FundingPercent_Overlay
   widget=Image_93
   properties {brush {
     bIsDynamicallyLoaded=false
     drawAs=Image
     imageType=NoImage
     mirroring=NoMirror
     resourceName=None
     resourceObject=/FortUILibrary/Foundation/Art_ToolKit/AngledBoxes/M_UI_FundCenter_PercentBacking_Default
     tiling=NoTile
     imageSize=<32 32>
     margin=<ltrb 0 0 0 0>
     outlineSettings {
      bUseBrushTransparency=false
      roundingType=HalfHeightRadius
      width=0
      color {colorUseRule=UseColor_Specified, specifiedColor=<rgba 0 0 0 0>}
      cornerRadii=<0 0 0 0>
     }
     tintColor {colorUseRule=UseColor_Specified, specifiedColor=<rgba 1 1 1 1>}
     uVRegion {bIsValid=false, max=<0 0>, min=<0 0>}
    }}
   slot {class=OverlaySlot, properties {horizontalAlignment=HAlign_Fill, verticalAlignment=VAlign_Fill}}
  }
  HorizontalBox_1 {
   widgetClassPath=&HorizontalBox
   parent=FundingPercent_Overlay
   widget=HorizontalBox_1
   properties {renderTransform {angle=0, scale=<1 1>, shear=<-16 0>, translation=<0 0>}}
   slot {class=OverlaySlot, properties {horizontalAlignment=HAlign_Center, verticalAlignment=VAlign_Center, padding=<ltrb 18 10 18 10>}}
  }
  Text_ProgressPercent {
   widgetClassPath=&CommonTextBlock
   bInherited=true
   bIsVariable=true
   parent=HorizontalBox_1
   widget=Text_ProgressPercent
   properties {
    text=99%
    font {
     bForceMonospaced=false
     bMaterialIsStencil=false
     fontMaterial=null
     fontObject=&Burbank
     letterSpacing=0
     monospacedWidth=1
     size=16
     skewAmount=0
     typefaceFontName=BurbankRegular-Black
     outlineSettings {
      bApplyOutlineToDropShadows=true
      bMiteredCorners=false
      bSeparateFillAlpha=true
      outlineMaterial=null
      outlineSize=0
      outlineColor=<rgba 0 0 0 0>
     }
    }
    shadowOffset=<0 0>
   }
   slot {class=HorizontalBoxSlot}
  }
  Text_ProgressText {
   widgetClassPath=&CommonTextBlock
   bIsVariable=true
   parent=HorizontalBox_1
   widget=Text_ProgressText
   properties {
    text="FUNDED!"
    font {
     bForceMonospaced=false
     bMaterialIsStencil=false
     fontMaterial=null
     fontObject=&BurbankSmall
     letterSpacing=0
     monospacedWidth=1
     size=16
     skewAmount=0
     typefaceFontName=Black
     outlineSettings {
      bApplyOutlineToDropShadows=true
      bMiteredCorners=false
      bSeparateFillAlpha=true
      outlineMaterial=null
      outlineSize=0
      outlineColor=<rgba 0 0 0 0>
     }
    }
    shadowOffset=<0 0>
   }
   slot {class=HorizontalBoxSlot, properties {verticalAlignment=VAlign_Center, padding=<ltrb 6 0 0 0>}}
  }
  SizeBox {
   widgetClassPath=&SizeBox
   parent=VBox_NameFunded
   widget=SizeBox
   properties {bOverride_WidthOverride=true, clipping=OnDemand, widthOverride=800, renderTransform {angle=0, scale=<1 1>, shear=<-12 0>, translation=<0 0>}}
   slot {class=VerticalBoxSlot, properties {verticalAlignment=VAlign_Bottom}}
  }
  ScaleBox_1 {
   widgetClassPath=&ScaleBox
   parent=SizeBox
   widget=ScaleBox_1
   properties {clipping=OnDemand, stretch=ScaleToFitX, stretchDirection=DownOnly}
   slot {class=SizeBoxSlot, properties {horizontalAlignment=HAlign_Center}}
  }
  Text_TitleBar {
   widgetClassPath=&CommonTextBlock
   bInherited=true
   bIsVariable=true
   parent=ScaleBox_1
   widget=Text_TitleBar
   properties {
    clipping=OnDemand
    lineHeightPercentage=0.9
    text=WEAPON NAME
    textTransformPolicy=ToUpper
    font {
     bForceMonospaced=false
     bMaterialIsStencil=false
     fontMaterial=null
     fontObject=&Burbank
     letterSpacing=0
     monospacedWidth=1
     size=72
     skewAmount=0
     typefaceFontName=BurbankRegular-Black
     outlineSettings {
      bApplyOutlineToDropShadows=true
      bMiteredCorners=false
      bSeparateFillAlpha=true
      outlineMaterial=null
      outlineSize=0
      outlineColor=<rgba 0 0 0 0>
     }
    }
    margin=<ltrb 8 -4 8 10>
    shadowOffset=<0 0>
   }
   slot {class=ScaleBoxSlot}
  }
  Border_0 {
   widgetClassPath=&Border
   parent=ProgressInfo
   widget=Border_0
   properties {
    visibility=SelfHitTestInvisible
    background {
     bIsDynamicallyLoaded=false
     drawAs=Box
     imageType=NoImage
     mirroring=NoMirror
     resourceName=None
     resourceObject=/FortniteConversationUI/Conversation/Materials/M_UI_MultiFundProgress_BarBacking_Default
     tiling=NoTile
     imageSize=<32 32>
     margin=<ltrb 0.5 0 0.5 0>
     outlineSettings {
      bUseBrushTransparency=false
      roundingType=HalfHeightRadius
      width=0
      color {colorUseRule=UseColor_Specified, specifiedColor=<rgba 0 0 0 0>}
      cornerRadii=<0 0 0 0>
     }
     tintColor {colorUseRule=UseColor_Specified, specifiedColor=<rgba 1 1 1 1>}
     uVRegion {bIsValid=false, max=<0 0>, min=<0 0>}
    }
    brushColor=<rgba 1 1 1 0.5>
    padding=<ltrb 10 10 12 10>
    renderTransform {angle=0, scale=<1 1>, shear=<-4 0>, translation=<0 0>}
   }
   slot {class=VerticalBoxSlot, properties {horizontalAlignment=HAlign_Center, verticalAlignment=VAlign_Bottom}}
  }
  ProgressBarSize {
   widgetClassPath=&SizeBox
   parent=Border_0
   widget=ProgressBarSize
   properties {bOverride_HeightOverride=true, bOverride_WidthOverride=true, heightOverride=16, widthOverride=512}
   slot {class=BorderSlot, properties {padding=<ltrb 10 10 12 10>}}
  }
  Progress_Funding {
   widgetClassPath=/Game/Athena/HUD/SimpleMaterialProgressBar
   bInherited=true
   bIsVariable=true
   parent=ProgressBarSize
   widget=Progress_Funding
   properties {
    bOverrideDefaultColorA=true
    bOverrideDefaultColorB=true
    bOverrideDefaultColorBackground=true
    bOverrideDefaultFillEdgeSoftness=true
    bOverrideDefaultOutlineScale=true
    bOverrideDefaultSegments=true
    designTime_Progress=0.75
    fillEdgeSoftness=0.0005
    outlineScale=3
    segmentEdge=0.0002
    segments=1
    cachedColorA=<rgba 0 1 1 1>
    cachedColorB=<rgba 1 1 1 1>
    cachedColorBackground=<rgba 0 0.3125 0.75 1>
    renderTransform {angle=0, scale=<1 1>, shear=<-8 0>, translation=<0 0>}
   }
   slot {class=SizeBoxSlot}
  }
  HBox_ProgressText {
   widgetClassPath=&HorizontalBox
   bIsVariable=true
   parent=ProgressInfo
   widget=HBox_ProgressText
   properties {visibility=Collapsed}
   slot {class=VerticalBoxSlot, properties {padding=<ltrb 5 -20 0 0>}}
  }
  Image_CurrencyIcon {
   widgetClassPath=&Image
   bIsVariable=true
   parent=HBox_ProgressText
   widget=Image_CurrencyIcon
   properties {brush {
     bIsDynamicallyLoaded=false
     drawAs=Image
     imageType=NoImage
     mirroring=NoMirror
     resourceName=None
     resourceObject=/Game/UI/Foundation/Textures/Icons/Athena/T-T-Icon-BR-GoldBars-UI-Icon
     tiling=NoTile
     imageSize=<55 55>
     margin=<ltrb 0 0 0 0>
     outlineSettings {
      bUseBrushTransparency=false
      roundingType=HalfHeightRadius
      width=0
      color {colorUseRule=UseColor_Specified, specifiedColor=<rgba 0 0 0 0>}
      cornerRadii=<0 0 0 0>
     }
     tintColor {colorUseRule=UseColor_Specified, specifiedColor=<rgba 1 1 1 1>}
     uVRegion {bIsValid=false, max=<0 0>, min=<0 0>}
    }}
   slot {class=HorizontalBoxSlot, properties {horizontalAlignment=HAlign_Center, verticalAlignment=VAlign_Center, padding=<ltrb 3 0 3 0>}}
  }
  Text_ProgressCount {
   widgetClassPath=&CommonTextBlock
   bInherited=true
   bIsVariable=true
   parent=HBox_ProgressText
   widget=Text_ProgressCount
   properties {
    text="500,000 / 1,000,000"
    font {
     bForceMonospaced=false
     bMaterialIsStencil=false
     fontMaterial=null
     fontObject=&Burbank
     letterSpacing=0
     monospacedWidth=1
     size=25
     skewAmount=0
     typefaceFontName=BurbankRegular-Black
     outlineSettings {
      bApplyOutlineToDropShadows=true
      bMiteredCorners=false
      bSeparateFillAlpha=true
      outlineMaterial=null
      outlineSize=2
      outlineColor=<rgba 0 0 0 1>
     }
    }
    shadowOffset=<0 0>
   }
   slot {class=HorizontalBoxSlot, properties {verticalAlignment=VAlign_Center}}
  }
  SizeBox_2 {
   widgetClassPath=&SizeBox
   parent=CanBuy
   widget=SizeBox_2
   properties {
    bOverride_MinDesiredWidth=true
    minDesiredWidth=320
    widthOverride=320
    renderTransform {angle=0, scale=<1 1>, shear=<0 0>, translation=<0 26>}
   }
   slot {class=OverlaySlot, properties {verticalAlignment=VAlign_Bottom, padding=<ltrb 360 0 0 0>}}
  }
  TransactionStrip_Main {
   widgetClassPath=&/FortTransactionStripNew
   bInherited=true
   bIsVariable=true
   parent=SizeBox_2
   widget=TransactionStrip_Main
   properties {bFundingVersion=true, transactionString=Buy}
   slot {class=SizeBoxSlot}
  }
  GlobalEffort {
   widgetClassPath=&Overlay
   parent=CanBuy
   widget=GlobalEffort
   properties {visibility=Collapsed}
   slot {class=OverlaySlot, properties {padding=<ltrb 320 32 0 0>}}
  }
  Image {
   widgetClassPath=&Image
   bIsVariable=true
   parent=GlobalEffort
   widget=Image
   properties {
    visibility=SelfHitTestInvisible
    brush {
     bIsDynamicallyLoaded=false
     drawAs=Box
     imageType=NoImage
     mirroring=NoMirror
     resourceName=None
     resourceObject=/FortUILibrary/Foundation/Art_ToolKit/AngledBoxes/M_UI_FundCenter_AnnouncementBacking_Shadow
     tiling=NoTile
     imageSize=<128 32>
     margin=<ltrb 0.5 0 0.5 0>
     outlineSettings {
      bUseBrushTransparency=false
      roundingType=HalfHeightRadius
      width=0
      color {colorUseRule=UseColor_Specified, specifiedColor=<rgba 0 0 0 0>}
      cornerRadii=<0 0 0 0>
     }
     tintColor {colorUseRule=UseColor_Specified, specifiedColor=<rgba 1 1 1 1>}
     uVRegion {bIsValid=false, max=<0 0>, min=<0 0>}
    }
    renderTransform {angle=0, scale=<1.125 1.5>, shear=<-8 0>, translation=<0 8>}
   }
   slot {class=OverlaySlot, properties {horizontalAlignment=HAlign_Fill, verticalAlignment=VAlign_Fill}}
  }
  Image_1 {
   widgetClassPath=&Image
   bIsVariable=true
   parent=GlobalEffort
   widget=Image_1
   properties {
    brush {
     bIsDynamicallyLoaded=false
     drawAs=Box
     imageType=NoImage
     mirroring=NoMirror
     resourceName=None
     resourceObject=/FortUILibrary/Foundation/Art_ToolKit/AngledBoxes/M_UI_FundCenter_AnnouncementBacking
     tiling=NoTile
     imageSize=<128 32>
     margin=<ltrb 0.5 0 0.5 0>
     outlineSettings {
      bUseBrushTransparency=false
      roundingType=HalfHeightRadius
      width=0
      color {colorUseRule=UseColor_Specified, specifiedColor=<rgba 0 0 0 0>}
      cornerRadii=<0 0 0 0>
     }
     tintColor {colorUseRule=UseColor_Specified, specifiedColor=<rgba 1 1 1 1>}
     uVRegion {bIsValid=false, max=<0 0>, min=<0 0>}
    }
    renderTransform {angle=0, scale=<1 1>, shear=<-8 0>, translation=<0 0>}
   }
   slot {class=OverlaySlot, properties {horizontalAlignment=HAlign_Fill, verticalAlignment=VAlign_Fill}}
  }
  HorizontalBox {
   widgetClassPath=&HorizontalBox
   parent=GlobalEffort
   widget=HorizontalBox
   slot {class=OverlaySlot, properties {horizontalAlignment=HAlign_Center, verticalAlignment=VAlign_Center, padding=<ltrb 18 16 20 16>}}
  }
  Image_229 {
   widgetClassPath=&Image
   bIsVariable=true
   parent=HorizontalBox
   widget=Image_229
   properties {
    renderOpacity=0.9
    brush {
     bIsDynamicallyLoaded=false
     drawAs=Image
     imageType=NoImage
     mirroring=NoMirror
     resourceName=None
     resourceObject=/Game/UI/Foundation/Textures/Icons/Stars/Stars-Full-64
     tiling=NoTile
     imageSize=<24 24>
     margin=<ltrb 0 0 0 0>
     outlineSettings {
      bUseBrushTransparency=false
      roundingType=HalfHeightRadius
      width=0
      color {colorUseRule=UseColor_Specified, specifiedColor=<rgba 0 0 0 0>}
      cornerRadii=<0 0 0 0>
     }
     tintColor {colorUseRule=UseColor_Specified, specifiedColor=<rgba 1 1 1 1>}
     uVRegion {bIsValid=false, max=<0 0>, min=<0 0>}
    }
    colorAndOpacity=<rgba 0.6939 0.0343 0.0242 1>
    renderTransform {angle=-8, scale=<1 1>, shear=<0 0>, translation=<0 0>}
   }
   slot {class=HorizontalBoxSlot, properties {horizontalAlignment=HAlign_Center, verticalAlignment=VAlign_Center}}
  }
  SizeBox_0 {
   widgetClassPath=&SizeBox
   parent=HorizontalBox
   widget=SizeBox_0
   properties {bOverride_MaxDesiredWidth=true, clipping=OnDemand, maxDesiredWidth=400}
   slot {class=HorizontalBoxSlot, properties {verticalAlignment=VAlign_Center}}
  }
  ScaleBox_0 {
   widgetClassPath=&ScaleBox
   parent=SizeBox_0
   widget=ScaleBox_0
   properties {clipping=OnDemand, stretch=ScaleToFitX, stretchDirection=DownOnly}
   slot {class=SizeBoxSlot}
  }
  Text_TitleBar_1 {
   widgetClassPath=&CommonTextBlock
   bIsVariable=true
   parent=ScaleBox_0
   widget=Text_TitleBar_1
   properties {
    clipping=OnDemand
    justification=Center
    lineHeightPercentage=0.9
    text="Contribute to the global effort!"
    textTransformPolicy=ToUpper
    colorAndOpacity {colorUseRule=UseColor_Specified, specifiedColor=<rgba 0.1195 0.1144 0.0545 1>}
    font {
     bForceMonospaced=false
     bMaterialIsStencil=false
     fontMaterial=null
     fontObject=&Burbank
     letterSpacing=0
     monospacedWidth=1
     size=15
     skewAmount=0
     typefaceFontName=BurbankRegular-Black
     outlineSettings {
      bApplyOutlineToDropShadows=true
      bMiteredCorners=false
      bSeparateFillAlpha=true
      outlineMaterial=null
      outlineSize=0
      outlineColor=<rgba 1 1 1 1>
     }
    }
    margin=<ltrb 8 -2 8 0>
    renderTransform {angle=0, scale=<1 1>, shear=<-12 0>, translation=<0 0>}
    shadowOffset=<0 0>
   }
   slot {class=ScaleBoxSlot}
  }
  Image_2 {
   widgetClassPath=&Image
   bIsVariable=true
   parent=HorizontalBox
   widget=Image_2
   properties {
    renderOpacity=0.9
    brush {
     bIsDynamicallyLoaded=false
     drawAs=Image
     imageType=NoImage
     mirroring=NoMirror
     resourceName=None
     resourceObject=&Stars-Full-64
     tiling=NoTile
     imageSize=<24 24>
     margin=<ltrb 0 0 0 0>
     outlineSettings {
      bUseBrushTransparency=false
      roundingType=HalfHeightRadius
      width=0
      color {colorUseRule=UseColor_Specified, specifiedColor=<rgba 0 0 0 0>}
      cornerRadii=<0 0 0 0>
     }
     tintColor {colorUseRule=UseColor_Specified, specifiedColor=<rgba 1 1 1 1>}
     uVRegion {bIsValid=false, max=<0 0>, min=<0 0>}
    }
    colorAndOpacity=<rgba 0.6939 0.0343 0.0242 1>
    renderTransform {angle=8, scale=<1 1>, shear=<0 0>, translation=<0 0>}
   }
   slot {class=HorizontalBoxSlot, properties {horizontalAlignment=HAlign_Center, verticalAlignment=VAlign_Center}}
  }
 ]
}
