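# Conversation UI widget that shows a brief for a single fundable item alongside NPC text, derived from native FortSingleItemFundBrief.
# An overlay layers a faded donation-background material behind two inherited child widgets, Item_Entry and NPCTextDisplay.
# Construct plays Intro, which slides and fades in the item entry while fading in the backing, then casts NPCTextDisplay and calls its NewItemUpdate.
# PreConstruct is empty and the DebugItem variable is unused in the graph.
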WidgetBlueprint /FortniteConversationUI/Conversation/FortSingleItemFundBrief
animations [Intro {
  binding_count=2
  length_seconds=0.5
  object=&self:Intro
  sequence {bindings [
    Item_Entry {
     id=#1
     possessed_class=FortSingleItemFundBriefEntry_C
     tracks [
      {
       type=MovieSceneFloatTrack
       display_name=Render Opacity
       sections [{type=MovieSceneFloatSection, channels [None {type=MovieSceneScriptingFloatChannel, keys [[0,0],[2,1]]}]}]
      }
      {
       type=MovieScene2DTransformTrack
       sections [{
         type=MovieScene2DTransformSection
         channels [
          Translation.X {type=MovieSceneScriptingFloatChannel, const [0,1]}
          Translation.Y {type=MovieSceneScriptingFloatChannel, keys [[0,128],[5,0]]}
          Angle {type=MovieSceneScriptingFloatChannel, const [0,1]}
          Scale.X {type=MovieSceneScriptingFloatChannel, const [1,1]}
          Scale.Y {type=MovieSceneScriptingFloatChannel, const [1,1]}
          Shear.X {type=MovieSceneScriptingFloatChannel, const [0,1]}
          Shear.Y {type=MovieSceneScriptingFloatChannel, const [0,1]}
         ]
        }]
      }
     ]
    }
    GradientBacking {
     id=#2
     possessed_class=Image
     tracks [{
       type=MovieSceneFloatTrack
       display_name=Render Opacity
       sections [{type=MovieSceneFloatSection, channels [None {type=MovieSceneScriptingFloatChannel, keys [[0,0],[10,1]]}]}]
      }]
    }
   ]}
 }]
graphs [EventGraph {nodes [
   Event Construct()
   > PlayAnimation(InAnimation=Get Intro())
   > n1 = Cast To NPCTextDisplay!(execute, Object=Get NPCTextDisplay() -> (NPCTextDisplay:Fort NPCText Display Object Reference)) -> (then, CastFailed, AsNPCText Display:NPCText Display Object Reference) {is=title K2Node_DynamicCast}
   > NewItemUpdate(self:NPCText Display Object Reference=@n1)
   Event PreConstruct()
  ]}]
variables [DebugItem {type=FortItem, default=null}]
widgets {
 info {
  inheritedWidgetCount=2
  parentClass=/Script/FortniteConversationUI.FortSingleItemFundBrief
  rootWidgetClass=&Overlay
  widgetCount=4
 }
 widgets [
  Overlay_2 {widgetClassPath=&Overlay, widget=Overlay_2}
  GradientBacking {
   widgetClassPath=&Image
   bIsVariable=true
   parent=Overlay_2
   widget=GradientBacking
   properties {
    visibility=SelfHitTestInvisible
    brush {
     bIsDynamicallyLoaded=false
     drawAs=Image
     imageType=NoImage
     mirroring=NoMirror
     resourceName=None
     resourceObject=/FortniteConversationUI/Conversation/Materials/M_UI_Donation_FadedBG
     tiling=NoTile
     imageSize=<32 448>
     margin=<ltrb 0 0 0 0>
     outlineSettings {
      bUseBrushTransparency=false
      roundingType=HalfHeightRadius
      width=0
      color {colorUseRule=UseColor_Specified, specifiedColor=<rgba 0 0 0 0>}
      cornerRadii=<0 0 0 0>
     }
     tintColor {colorUseRule=UseColor_Specified, specifiedColor=<rgba 1 1 1 1>}
     uVRegion {bIsValid=false, max=<0 0>, min=<0 0>}
    }
    renderTransform {angle=0, scale=<2 4>, shear=<0 0>, translation=<0 0>}
   }
   slot {class=OverlaySlot, properties {horizontalAlignment=HAlign_Fill, verticalAlignment=VAlign_Center, padding=<ltrb 0 0 0 20>}}
  }
  Item_Entry {
   widgetClassPath=&/FortSingleItemFundBriefEntry
   bInherited=true
   bIsVariable=true
   parent=Overlay_2
   widget=Item_Entry
   slot {class=OverlaySlot, properties {horizontalAlignment=HAlign_Fill, verticalAlignment=VAlign_Center, padding=<ltrb 0 0 0 50>}}
  }
  NPCTextDisplay {
   widgetClassPath=&/NPCTextDisplay
   bInherited=true
   bIsVariable=true
   parent=Overlay_2
   widget=NPCTextDisplay
   slot {class=OverlaySlot, properties {verticalAlignment=VAlign_Center, padding=<ltrb 64 0 0 384>}}
  }
 ]
}
