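# M_ConfirmCircle is a translucent UI material for a circle with an outline, part of the Conversation UI and, judging by its name, a confirm indicator.
# The MF_UI_Circle function generates the shape procedurally from UVs, with Size and Thickness parameters and a constant outline alignment.
# Its Outline output serves as the mask that blends between the Fill and Outline color parameters.
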Material /FortniteConversationUI/Conversation/Materials/M_ConfirmCircle
graph {
 properties {bAutomaticallySetUsageInEditor=false, blendMode=BLEND_Translucent, materialDomain=MD_UI, usedShadingModels=Default Lit}
 nodes [LinearInterpolate(
   A=VectorParameter() {parameterName=Fill, defaultValue=<rgba 1 1 1 0>}.RGB
   B=VectorParameter() {parameterName=Outline, defaultValue=<rgba 0.6038 0.6038 0.6038 1>}.RGB
   Alpha=MaterialFunctionCall(
    UVs=TextureCoordinate()
    Size=ScalarParameter() {defaultValue=1, parameterName=Size}
    OutlineThickness=ScalarParameter() {defaultValue=0.04, parameterName=Thickness}
    OutlineAlignment=Constant() {r=-0.5}
   ) {materialFunction=&MF_UI_Circle}.Outline
  )]
}
