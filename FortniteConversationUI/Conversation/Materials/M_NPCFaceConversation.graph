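# Translucent UI material for an NPC face portrait in the conversation UI, with a circular mask on the image's alpha.
# Centered UV length feeds an inverted SmoothStep that forms a radial falloff, which passes through the DPI anti-aliasing function.
# That function's Outer result multiplies the alpha of the Texture parameter, which defaults to a SynthContactM portrait asset.
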Material /FortniteConversationUI/Conversation/Materials/M_NPCFaceConversation
graph {
 properties {bAutomaticallySetUsageInEditor=false, blendMode=BLEND_Translucent, materialDomain=MD_UI, usedShadingModels=Default Lit}
 nodes [
  n1 = Length(None=ConstantBiasScale(None=TextureCoordinate()) {bias=-0.5, scale=2}) {is=class MaterialExpressionLength}
  Multiply(
   A=MaterialFunctionCall!(DistanceField=SmoothStep(Value=@n1) {constMax=0.7, constMin=1}, Threshold) {materialFunction=/FortUILibrary/MaterialFunctions/Shapes/Modifiers/MF_UI_AntiAliasing_DPI}.Outer
   B=TextureSample(Tex=TextureObjectParameter() {parameterName=Texture, texture=&/M_UI_Figment_Donation_NPC_Portrait_SynthContactM}).A
  )
 ]
}
