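# Radial list widget for the mobile conversation UI that arranges entry widgets around the center of a root Overlay.
# AddEntry creates a widget of EntryClass, then rebuilds every child, placing each through slot padding converted from a computed 2D offset.
# Offsets combine angle (full circle or limited arc, sine-based harmonic distortion) with distance (fixed or size-derived, plus spiral), then apply rotated axis scaling.
# PreConstruct fills placeholder entries at design time for preview.
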WidgetBlueprint /FortniteConversationUI/Conversation/MobileConversationComponents/BriefRadialList
graphs [
 AddEntry {nodes [
   AddEntry() {is=title K2Node_FunctionEntry}
   > n1 = Create Widget!(
    execute:Exec
    Class:User Widget Class Reference=Get EntryClass() -> (EntryClass:User Widget Class Reference)
    OwningPlayer:Player Controller Object Reference
   ) -> (then:Exec, ReturnValue:User Widget Object Reference)
   > n2 = Is Valid(InputObject=@n1)
   Return Node!(execute=@n2.Is Not Valid, New Entry:User Widget Object Reference, Success:bool=false)
   > Add(
    TargetArray:Array of User Widget Object References=Get Entries() -> (Entries:Array of User Widget Object References)
    NewItem:"User Widget Object Reference (by ref)"=@n1
   ) {is=title K2Node_CallArrayFunction}
   > Set Entry(Entry:User Widget Object Reference=@n1) -> (Output_Get:User Widget Object Reference)
   > ClearChildren(self=Get Root() -> (Root:Overlay Object Reference))
   > n3 = For Each Loop(Array:Array of User Widget Object References=Get Entries()) -> (Array Element:User Widget Object Reference)
   Return Node(execute=@n3.Completed, New Entry=Get Entry() -> (Entry:User Widget Object Reference), Success=false)
   > n4 = AddChildToOverlay!(execute, self:Overlay Object Reference=Get Root(), Content:Widget Object Reference=@n3) -> (then, ReturnValue:Overlay Slot Object Reference)
   > SetHorizontalAlignment(self=@n4, InHorizontalAlignment=HAlign_Center)
   > SetVerticalAlignment(self=@n4, InVerticalAlignment=VAlign_Center)
   > SetPadding!(
    execute
    self:Overlay Slot Object Reference=@n4
    InPadding:Margin Structure=Translate2DVecToMargin!(InVec:Vector 2D Structure=CalculateItemPosition!(Index:int=@n3.Array Index) -> (ReturnValue:Vector 2D Structure)) -> (Margin:Margin Structure)
   ) -> (then)
  ]}
 ApplyPreview {nodes [
   n1 = ApplyPreview!() -> (then, NumDesignerPreview:int) {is=title K2Node_FunctionEntry}
   > ClearEntries()
   > For Loop(LastIndex=int - int(A=@n1))
   > AddEntry!(execute, self) -> (then, New Entry:User Widget Object Reference, Success:bool) {is=title K2Node_CallFunction}
   Get NumDesignerPreview() -> (NumDesignerPreview:int=0)
  ]}
 CalculateAngleCorrection {nodes [
   n1 = CalculateAngleCorrection!() -> (then, Angle:double=~) {is=title K2Node_FunctionEntry}
   n2 = Break Harmonic Distortion!(HarmonicDistortion:"Harmonic Distortion Structure (by ref)"=Get HarmonicDistortion() -> (HarmonicDistortion:Harmonic Distortion Structure)) -> (
    Frequency_7_#1:double=~
    Amplitude_8_#2:double=~
    Rotataion_9_#3:double=~
   ) {is=title K2Node_BreakStruct}
   n3 = float + float(A=@n1, B=@n2.Rotataion_9_#3)
   n4 = float * float(A="Sin (Degrees)"!(A:double=float * float(A=@n3, B=@n2)) -> (ReturnValue:double), B=@n2.Amplitude_8_#2) {is=title K2Node_PromotableOperator}
   > Return Node!(execute, ReturnValue:double=float + float(A=@n1, B=@n4), Harmonic:double=@n4)
  ]}
 CalculateItemAngleByLenght {nodes [
   n1 = CalculateItemAngleByLenght!() -> (then, Index:int, Lenght:int) {is=title K2Node_FunctionEntry}
   n2 = Get DistriuteItemsEvenly() -> (DistriuteItemsEvenly:bool)
   n3 = NOT Boolean(A=@n2)
   n4 = SafeDivide(
    A=SelectFloat(A=360, B=Get MaxAngleLenght() -> (MaxAngleLenght:double), bPickA=@n2)
    B="To Float (Integer)"(InInt=int - int(A=@n1.Lenght, B:bool=@n3))
   )
   n5 = "To Float (Boolean)"(InBool=Get Clockwise() -> (Clockwise:bool))
   n6 = float * float(A=float - float(A=float * float(A=@n5, B=2), B=1), B=-1)
   n7 = CalculateAngleCorrection!(Angle:double=float + float(A=float * float!(A:int=@n1, B=@n4, C:double=@n6) -> (ReturnValue), B=Get StartAngle() -> (StartAngle:double))) -> (ReturnValue:double, Harmonic:double) {is=title K2Node_CallFunction}
   n8 = GetRotated2D!(A:Vector 2D Structure=MakeVector2D(X=0, Y=-1), AngleDeg:float=@n7) -> (ReturnValue:Vector 2D Structure)
   n9 = Break Harmonic Distortion(HarmonicDistortion=Get HarmonicDistortion()) -> (
    Frequency_7_#1=~
    Amplitude_8_#2=~
    Rotataion_9_#3=~
   )
   n10 = float * float!(
    A:double=@n9.Amplitude_8_#2
    B="Sin (Degrees)"(A=float * float!(A=@n9.Rotataion_9_#3, B=@n9) -> (ReturnValue))
    C=-1
    D:bool="float < float"(A="% (Float)"!(A:double=@n9, B:double=2) -> (ReturnValue:double), B=1)
   ) -> (ReturnValue)
   > Return Node!(
    execute
    Angle:double=@n7
    Direction:Vector 2D Structure=Select(Option 0:Vector 2D Structure=GetRotated2D(A=@n8, AngleDeg=@n10), Option 1:Vector 2D Structure=@n8, Index=Get DistriuteItemsEvenly()) -> (ReturnValue:Vector 2D Structure)
   )
  ]}
 CalculateItemPosition {nodes [
   n1 = CalculateItemPosition!() -> (then, Index:int) {is=title K2Node_FunctionEntry}
   n2 = Break Radial List Transform!(RadialListTransform:"Radial List Transform Structure (by ref)"=Get RadialTransform() -> (RadialTransform:Radial List Transform Structure)) -> (AxisScale_7_#1:Vector 2D Structure, AxisScaleRotation_4_#2:double=~) {is=title K2Node_BreakStruct}
   n3 = GetRotated2D(
    A=vector2d * vector2d(
     A:Vector 2D Structure=GetItemAngle!(Index:int=@n1) -> (Angle:double, Direction:Vector 2D Structure).Direction
     B=GetItemDistance!(Index:int=@n1) -> (RadiusDistance:double)
    ) -> (ReturnValue:Vector 2D Structure) {is=title K2Node_PromotableOperator}
    AngleDeg=float * float(A=@n2.AxisScaleRotation_4_#2, B=-1)
   )
   > Return Node!(
    execute
    ReturnValue:Vector 2D Structure=GetRotated2D(A=vector2d * vector2d(A=@n3, B:Vector 2D Structure=@n2), AngleDeg=@n2.AxisScaleRotation_4_#2)
   )
  ]}
 ClearEntries {nodes [
   ClearEntries() {is=title K2Node_FunctionEntry}
   > ClearChildren(self=Get Root())
   > Clear(TargetArray:Array of User Widget Object References=Get Entries())
  ]}
 EventGraph {nodes [
   n1 = Event PreConstruct()
   > Branch(Condition=@n1.IsDesignTime)
   > ApplyPreview!(execute, self, NumDesignerPreview:int=Get NumDesignerPreview() -> (NumDesignerPreview=0)) -> (then) {is=title K2Node_CallFunction}
  ]}
 GetAllEntries {nodes [
   GetAllEntries() {is=title K2Node_FunctionEntry}
   > Return Node!(execute, AllEntries:Array of User Widget Object References=Get Entries())
  ]}
 GetEntryWidget {nodes [
   n1 = GetEntryWidget!() -> (then, Index:int) {is=title K2Node_FunctionEntry}
   > Return Node!(
    execute
    UserWidget:User Widget Object Reference="Get (a copy)"(Array:Array of User Widget Object References=Get Entries(), Dimension 1=@n1) -> (Output:User Widget Object Reference)
   )
  ]}
 GetItemAngle {nodes [
   n1 = GetItemAngle!() -> (then, Index:int) {is=title K2Node_FunctionEntry}
   n2 = CalculateItemAngleByLenght!(Index:int=@n1, Lenght:int=Length(TargetArray:Array of User Widget Object References=Get Entries())) -> (Angle:double, Direction:Vector 2D Structure) {is=title K2Node_CallFunction}
   > Return Node!(execute, Angle:double=@n2, Direction:Vector 2D Structure=@n2.Direction)
  ]}
 GetItemDistance {nodes [
   n1 = GetItemDistance!() -> (then, Index:int) {is=title K2Node_FunctionEntry}
   n2 = GetTickSpaceGeometry!(self:Widget Object Reference=Get Root()) -> (ReturnValue:"Geometry Structure (by ref)")
   n3 = float * float(
    A=Vector2D Length!(A:Vector 2D Structure=GetAbsoluteSize!(Geometry:"Geometry Structure (by ref)"=@n2) -> (ReturnValue:Vector 2D Structure)) -> (ReturnValue:double)
    B=0.25
   )
   n4 = Get RadiusDistance() -> (RadiusDistance:double)
   n5 = Get FillDistanceBySize() -> (FillDistanceBySize:bool)
   n6 = Break Radial List Transform!(RadialListTransform=Get RadialTransform()) -> (SpiralFactor_6_#3:double=~)
   > Return Node!(execute, RadiusDistance:double=float + float(A=SelectFloat(A=@n3, B=@n4, bPickA=@n5), B=float * float(A=@n6, B:int=@n1)))
  ]}
 GetRadialListOverlay {nodes [
   GetRadialListOverlay() {is=title K2Node_FunctionEntry}
   > Return Node!(execute, ReturnValue:Overlay Object Reference=Get Root())
  ]}
 Translate2DVecToMargin {nodes [
   n1 = Translate2DVecToMargin!() -> (then, InVec:Vector 2D Structure) {is=title K2Node_FunctionEntry}
   n2 = float * float(A=Get RadiusDistance(), B:double=2)
   n3 = BreakVector2D(InVec=@n1)
   n4 = float * float(A=@n3, B=-1)
   n5 = float * float(A=@n3.Y, B=-1)
   > Return Node!(
    execute
    Margin:Margin Structure=Make Margin(
     Left="Clamp (Float)"(Value=@n3, Min=0, Max=@n2)
     Top="Clamp (Float)"(Value=@n3.Y, Min=0, Max=@n2)
     Right="Clamp (Float)"(Value=@n4, Min=0, Max=@n2)
     Bottom="Clamp (Float)"(Value=@n5, Min=0, Max=@n2)
    )
   )
  ]}
]
variables [
 Clockwise {type=boolean}
 DistriuteItemsEvenly {type=boolean, default=true}
 Entries {type="array<UserWidget>", default []}
 EntryClass {type=Class, default=/Game/Athena/HUD/ContextMenus/RadialMenuTestEntry.RadialMenuTestEntry_C}
 FillDistanceBySize {type=boolean}
 HarmonicDistortion {type=HarmonicDistortion, default {amplitude=0, frequency=0, rotataion=0}}
 MaxAngleLenght {type=number, default=180}
 NumDesignerPreview {type=integer, default=12}
 RadialTransform {type=RadialListTransform, default {axisScaleRotation=0, spiralFactor=0, axisScale=<1 1>}}
 RadiusDistance {type=number, default=80}
 StartAngle {type=number}
 TurnOnPreviewConstruct {type=boolean}
]
widgets {
 info {
  inheritedWidgetCount=0
  parentClass=/Script/FortniteConversationUI.FortConversationBriefRadialList
  rootWidgetClass=&Overlay
  widgetCount=1
 }
 widgets [Root {widgetClassPath=&Overlay, bIsVariable=true, widget=Root, properties {bOverride_Cursor=true}}]
}
