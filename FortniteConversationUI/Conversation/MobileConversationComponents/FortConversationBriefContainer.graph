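# Mobile conversation UI button that hosts one brief widget in a switcher and broadcasts BriefClicked with it when clicked.
# AddBriefChild inserts the brief and sets chat tail orientation, while SetBriefAlignment stores Direction and derives alignment from it.
# FocusPreviewCard plays CardHovered forward or reversed as a timeline for Anim_CardSelected.
# That event samples in/out curves to nudge along Direction, scale from an input-dependent unselected size, and update selected visuals.
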WidgetBlueprint /FortniteConversationUI/Conversation/MobileConversationComponents/FortConversationBriefContainer
animations [CardHovered {
  binding_count=0
  length_seconds=1
  object=&self:CardHovered
  sequence {bindings [], tracks [{type=MovieSceneEventTrack, sections [{type=MovieSceneEventRepeaterSection, end_frame=20, start_frame=0}]}]}
 }]
graphs [
 BriefClicked {nodes [BriefClicked!() -> (then, BriefWidget:User Widget Object Reference) {is=title K2Node_FunctionEntry}]}
 EventGraph {nodes [
   "On Clicked (BriefButton)"() {is=title K2Node_ComponentBoundEvent}
   > Call Brief Clicked!(
    execute:Exec
    self:Fort Conversation Brief Container Object Reference
    BriefWidget:User Widget Object Reference=Get Brief Widget() -> (Brief Widget:User Widget Object Reference)
   ) -> (then:Exec) {is=title K2Node_CallDelegate}
   Anim_CardSelected() {is=title K2Node_CustomEvent}
   > n1 = GetAnimLenght!(execute, self, Animation:Widget Animation Object Reference=Get CardHovered() -> (CardHovered:Widget Animation Object Reference)) -> (then, ReturnValue:double, PlayingForward:bool)
   n2 = GetFloatValue!(
    self:Curve Float Object Reference=Select(
     Option 0:Curve Float Object Reference=&/PreviewHover_Out
     Option 1:Curve Float Object Reference=&/PreviewHover_In
     Index=@n1.PlayingForward
    ) -> (ReturnValue:Curve Float Object Reference)
    InTime:float=@n1
   ) -> (ReturnValue:float)
   > SetRenderTranslation(
    self=Get Switcher() -> (Switcher:Common Visibility Switcher Object Reference)
    Translation=vector2d * vector2d!(A:Vector 2D Structure=Get Direction() -> (Direction:Vector 2D Structure), B:float=@n2, C:float=30) -> (ReturnValue:Vector 2D Structure) {is=title K2Node_PromotableOperator}
   )
   n3 = Select(
    Option 0:double=Get UnselectedScale_Core() -> (UnselectedScale_Core:double)
    Option 1:double=Get UnselectedScale_Touch() -> (UnselectedScale_Touch:double)
    Index=IsUsingInputType!(InputType:ECommonInputType Enum=Touch, PlayerController:Player Controller Object Reference=GetOwningPlayer()) -> (ReturnValue:bool)
   ) -> (ReturnValue:double)
   n4 = Get SelectedScale() -> (SelectedScale:double)
   > SetRenderScale!(
    execute
    self:Widget Object Reference=Get Switcher()
    Scale:Vector 2D Structure=To Vector2D!(InDouble:double=Lerp(A=@n3, B=@n4, Alpha=@n2)) -> (ReturnValue:Vector 2D Structure)
   ) -> (then)
   > SetBriefSelectedStates!(execute, self, IsSelected:bool=@n1.PlayingForward) -> (then)
   n5 = AddBriefChild!() -> (OutputDelegate, then, BriefWidget:User Widget Object Reference, TailDirection:bool=~) {is=title K2Node_CustomEvent}
   > Set Brief Widget(Brief Widget:User Widget Object Reference=@n5.BriefWidget) -> (Output_Get:User Widget Object Reference)
   n6 = Get Brief Widget()
   > n7 = Is Valid(InputObject=@n6)
   SetVisibility(execute=@n7.Is Not Valid, self=Self-Reference(), InVisibility=Collapsed)
   n8 = Get Switcher()
   AddChild(execute=@n7.Is Valid, self=@n8, Content=@n6)
   > SetActiveWidget(self=@n8, Widget=@n6)
   n9 = SlotAsWidgetSwitcherSlot!(Widget:Widget Object Reference=@n6) -> (ReturnValue:Widget Switcher Slot Object Reference)
   > SetVerticalAlignment(self:Widget Switcher Slot Object Reference=@n9, InVerticalAlignment=VAlign_Center)
   > SetHorizontalAlignment(self:Widget Switcher Slot Object Reference=@n9, InHorizontalAlignment=HAlign_Right)
   > SetVisibility(self=Self-Reference(), InVisibility=SelfHitTestInvisible)
   > SetBriefSelectedStates(IsSelected=false)
   n10 = Get UnselectedScale_Core()
   n11 = Get UnselectedScale_Touch()
   n12 = IsUsingInputType(InputType=Touch, PlayerController=GetOwningPlayer())
   > SetRenderScale(self=Get Switcher(), Scale=To Vector2D(InDouble=Select(Option 0=@n10, Option 1=@n11, Index=@n12)))
   > n13 = Cast To FortChatBrief!(execute, Object=Get Brief Widget()) -> (then, CastFailed, AsFort Chat Brief:Fort Chat Brief Object Reference) {is=title K2Node_DynamicCast}
   > SetTailOrientation!(execute, self:Fort Chat Brief Object Reference=@n13, isPointingToRight:bool=@n5.TailDirection) -> (then)
   n14 = FocusPreviewCard!() -> (OutputDelegate, then, IsHovered:bool=~) {is=title K2Node_CustomEvent}
   > Branch(Condition="Not Equal (Boolean)"(A=Get Is Selected() -> (Is Selected:bool), B=@n14.IsHovered) -> (ReturnValue=~))
   > Set Is Selected(Is Selected:bool=@n14.IsHovered) -> (Output_Get:bool)
   > PlayAnimation(
    InAnimation=Get CardHovered()
    PlayMode=Select(Option 0:EUMGSequencePlayMode Enum=Reverse, Option 1:EUMGSequencePlayMode Enum, Index=Get Is Selected()) -> (ReturnValue:EUMGSequencePlayMode Enum)
    PlaybackSpeed=5
   )
   Event Construct()
   > SetRenderTranslation(self=Get Switcher())
   n15 = Event PreConstruct()
   > Branch(Condition=@n15.IsDesignTime)
   > SetBriefAlignment!(execute, self, Angle:double=0, Direction:Vector 2D Structure=Get Direction()) -> (then)
  ]}
 GetAnimLenght {nodes [
   n1 = GetAnimLenght!() -> (then, Animation:Widget Animation Object Reference) {is=title K2Node_FunctionEntry}
   > n2 = IsAnimationPlayingForward!(execute, self:User Widget Object Reference, InAnimation:Widget Animation Object Reference=@n1) -> (then, ReturnValue:bool)
   > Return Node!(
    execute
    ReturnValue:double=SafeDivide(
     A=GetAnimationCurrentTime!(self:User Widget Object Reference, InAnimation:Widget Animation Object Reference=@n1) -> (ReturnValue:float)
     B=GetEndTime(self=@n1)
    )
    PlayingForward:bool=@n2
   )
  ]}
 Sequencer Events {links [], nodes []}
 SetBriefAlignment {nodes [
   n1 = SetBriefAlignment!() -> (then, Angle:double=~, Direction:Vector 2D Structure) {is=title K2Node_FunctionEntry}
   > Set Angle(Angle:double=@n1) -> (Output_Get:double)
   > Set Direction(Direction:Vector 2D Structure=@n1.Direction) -> (Output_Get:Vector 2D Structure)
   > n2 = Sequence!(execute) -> (then_0, then_1, then_2:Exec)
   n3 = BreakVector2D(InVec=@n1.Direction)
   > n4 = Branch(Condition="float > float"(A=@n3.Y, B=0.3))
   n5 = Get Switcher()
   n6 = SlotAsSizeBoxSlot!(Widget:Widget Object Reference=@n5) -> (ReturnValue:Size Box Slot Object Reference)
   > SetVerticalAlignment(self:Size Box Slot Object Reference=@n6, InVerticalAlignment=VAlign_Top)
   n7 = Branch(execute=@n4.else, Condition="float < float"(A=@n3.Y, B=-0.33))
   > SetVerticalAlignment(self=@n6, InVerticalAlignment=VAlign_Bottom)
   SetVerticalAlignment(execute=@n7.else, self=@n6, InVerticalAlignment=VAlign_Center)
   n8 = Branch(execute=@n2.then_1, Condition="float > float"(A=@n3, B=0.33))
   n9 = SlotAsSizeBoxSlot(Widget=@n5)
   > SetHorizontalAlignment(self:Size Box Slot Object Reference=@n9, InHorizontalAlignment=HAlign_Left)
   n10 = Branch(execute=@n8.else, Condition="float < float"(A=@n3, B=-0.33))
   > SetHorizontalAlignment(self=@n9, InHorizontalAlignment=HAlign_Right)
   SetHorizontalAlignment(execute=@n10.else, self=@n9, InHorizontalAlignment=HAlign_Center)
   Return Node!(execute=@n2.then_2)
  ]}
 SetBriefSelectedStates {nodes [
   n1 = SetBriefSelectedStates!() -> (then, IsSelected:bool=~) {is=title K2Node_FunctionEntry}
   > SetSelectedVisuals!(execute:Exec, self:Object Reference=Get Brief Widget(), IsSelected:bool=@n1) -> (then:Exec) {is=title K2Node_Message}
  ]}
]
variables [
 Angle {type=number}
 Brief Widget {type=UserWidget, default=null}
 BriefClicked {}
 DefaultUnfocusOpacity {type=number, default=1}
 Direction {type=Vector2D, default=<0 0>}
 Is Selected {type=boolean}
 SelectedScale {type=number, default=1}
 UnselectedScale_Core {type=number, default=0.6}
 UnselectedScale_Touch {type=number, default=0.8}
]
widgets {
 info {inheritedWidgetCount=0, parentClass=&UserWidget, rootWidgetClass=&Button, widgetCount=5}
 widgets [
  BriefButton {
   widgetClassPath=&Button
   widget=BriefButton
   properties {widgetStyle {
     clickedSlateSound {resourceObject=null}
     disabled {
      bIsDynamicallyLoaded=false
      drawAs=NoDrawType
      imageType=NoImage
      mirroring=NoMirror
      resourceName=None
      resourceObject=null
      tiling=NoTile
      imageSize=<64 64>
      margin=<ltrb 0 0 0 0>
      outlineSettings {
       bUseBrushTransparency=false
       roundingType=HalfHeightRadius
       width=0
       color {colorUseRule=UseColor_Specified, specifiedColor=<rgba 0 0 0 0>}
       cornerRadii=<0 0 0 0>
      }
      tintColor {colorUseRule=UseColor_Specified, specifiedColor=<rgba 1 1 1 0>}
      uVRegion {bIsValid=false, max=<0 0>, min=<0 0>}
     }
     disabledForeground {colorUseRule=UseColor_Specified, specifiedColor=<rgba 0.5271 0.5271 0.5271 1>}
     hovered {
      bIsDynamicallyLoaded=false
      drawAs=NoDrawType
      imageType=NoImage
      mirroring=NoMirror
      resourceName=None
      resourceObject=null
      tiling=NoTile
      imageSize=<72 72>
      margin=<ltrb 0 0 0 0>
      outlineSettings {
       bUseBrushTransparency=true
       roundingType=FixedRadius
       width=1
       color {colorUseRule=UseColor_Specified, specifiedColor=<rgba 0.7243 0.7243 0.7243 1>}
       cornerRadii=<4 4 4 4>
      }
      tintColor {colorUseRule=UseColor_Specified, specifiedColor=<rgba 1 1 1 0>}
      uVRegion {bIsValid=false, max=<0 0>, min=<0 0>}
     }
     hoveredForeground {colorUseRule=UseColor_Specified, specifiedColor=<rgba 0 1 1 1>}
     hoveredSlateSound {resourceObject=null}
     normal {
      bIsDynamicallyLoaded=false
      drawAs=NoDrawType
      imageType=NoImage
      mirroring=NoMirror
      resourceName=None
      resourceObject=null
      tiling=NoTile
      imageSize=<72 72>
      margin=<ltrb 0 0 0 0>
      outlineSettings {
       bUseBrushTransparency=true
       roundingType=FixedRadius
       width=1
       color {colorUseRule=UseColor_Specified, specifiedColor=<rgba 0.6951 0.6951 0.6951 1>}
       cornerRadii=<4 4 4 4>
      }
      tintColor {colorUseRule=UseColor_Specified, specifiedColor=<rgba 1 1 1 0>}
      uVRegion {bIsValid=false, max=<0 0>, min=<0 0>}
     }
     normalForeground {colorUseRule=UseColor_Specified, specifiedColor=<rgba 0 1 1 1>}
     normalPadding=<ltrb 0 0 0 0>
     pressed {
      bIsDynamicallyLoaded=false
      drawAs=NoDrawType
      imageType=NoImage
      mirroring=NoMirror
      resourceName=None
      resourceObject=null
      tiling=NoTile
      imageSize=<72 72>
      margin=<ltrb 0 0 0 0>
      outlineSettings {
       bUseBrushTransparency=true
       roundingType=FixedRadius
       width=1
       color {colorUseRule=UseColor_Specified, specifiedColor=<rgba 0.7243 0.7243 0.7243 1>}
       cornerRadii=<4 4 4 4>
      }
      tintColor {colorUseRule=UseColor_Specified, specifiedColor=<rgba 1 1 1 0>}
      uVRegion {bIsValid=false, max=<0 0>, min=<0 0>}
     }
     pressedForeground {colorUseRule=UseColor_Specified, specifiedColor=<rgba 0 1 1 1>}
     pressedPadding=<ltrb 0 0 0 0>
     pressedSlateSound {resourceObject=null}
    }}
  }
  Switcher {
   widgetClassPath=&CommonVisibilitySwitcher
   bIsVariable=true
   parent=BriefButton
   widget=Switcher
   slot {class=ButtonSlot, properties {padding=<ltrb 0 0 0 0>}}
  }
  PreviewCardForMeasurements {
   widgetClassPath=&Overlay
   parent=Switcher
   widget=PreviewCardForMeasurements
   slot {class=CommonVisibilitySwitcherSlot, properties {horizontalAlignment=HAlign_Center, verticalAlignment=VAlign_Center}}
  }
  Image_37 {
   widgetClassPath=&Image
   parent=PreviewCardForMeasurements
   widget=Image_37
   properties {brush {
     bIsDynamicallyLoaded=false
     drawAs=Image
     imageType=NoImage
     mirroring=NoMirror
     resourceName=None
     resourceObject=null
     tiling=NoTile
     imageSize=<460 195>
     margin=<ltrb 0 0 0 0>
     outlineSettings {
      bUseBrushTransparency=false
      roundingType=HalfHeightRadius
      width=0
      color {colorUseRule=UseColor_Specified, specifiedColor=<rgba 0 0 0 0>}
      cornerRadii=<0 0 0 0>
     }
     tintColor {colorUseRule=UseColor_Specified, specifiedColor=<rgba 1 1 1 1>}
     uVRegion {bIsValid=false, max=<0 0>, min=<0 0>}
    }}
   slot {class=OverlaySlot, properties {horizontalAlignment=HAlign_Center, verticalAlignment=VAlign_Center}}
  }
  measure {
   widgetClassPath=&UEFN_TextBlock
   parent=PreviewCardForMeasurements
   widget=measure
   properties {
    text=Preview - 460x195
    colorAndOpacity {colorUseRule=UseColor_Specified, specifiedColor=<rgba 0 0 0 1>}
    font {
     bForceMonospaced=false
     bMaterialIsStencil=false
     fontMaterial=null
     fontObject=&BurbankSmall
     letterSpacing=0
     monospacedWidth=1
     size=13.5
     skewAmount=0
     typefaceFontName=Black
     outlineSettings {
      bApplyOutlineToDropShadows=false
      bMiteredCorners=false
      bSeparateFillAlpha=false
      outlineMaterial=null
      outlineSize=0
      outlineColor=<rgba 0 0 0 1>
     }
    }
   }
   slot {class=OverlaySlot, properties {horizontalAlignment=HAlign_Center, verticalAlignment=VAlign_Bottom, padding=<ltrb 0 0 0 20>}}
  }
 ]
}
