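# Conversation UI widget, built on FortConversationResourceWidget, that shows the player's count of a resource, with gold bars as the default icon.
# On Construct it sets the icon from the ResourceItemDefinition's large preview image via a lazy texture load.
# OnResourceCountChanged writes the new amount directly to the currency text, ignoring the delta, and plays Transaction, a brief scale pulse on the root overlay.
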WidgetBlueprint /FortniteConversationUI/Conversation/WBP_UI_ConversationCurrencyWidget
animations [Transaction {
  binding_count=1
  length_seconds=0.3
  object=&self:Transaction
  sequence {bindings [Overlay_31 {
     id=#1
     possessed_class=Overlay
     tracks [{
       type=MovieScene2DTransformTrack
       sections [{
         type=MovieScene2DTransformSection
         end_frame=6
         start_frame=0
         channels [
          Scale.X {type=MovieSceneScriptingFloatChannel, keys [[0,1],[3,1.4],[6,1]]}
          Scale.Y {type=MovieSceneScriptingFloatChannel, keys [[0,1],[3,1.4],[6,1]]}
         ]
        }]
      }]
    }]}
 }]
graphs [
 EventGraph {nodes [
   Event Construct()
   > SetResouceIcon()
   n1 = Event OnResourceCountChanged!() -> (OutputDelegate, then, NewResourceAmount:int=0, DeltaAmount:int=0) {is=title K2Node_Event}
   > "SetText (Text)"(
    self=Get TXT_Currency() -> (TXT_Currency:Common Numeric Text Block Object Reference)
    InText="To Text (Integer)"(Value=@n1.NewResourceAmount)
   )
   > PlayAnimation(InAnimation=Get Transaction() -> (Transaction:Widget Animation Object Reference))
  ]}
 SetResouceIcon {nodes [
   SetResouceIcon() {is=title K2Node_FunctionEntry}
   > SetBrushFromLazyTexture!(
    execute
    self:Common Lazy Image Object Reference=Get ICO_Gold() -> (ICO_Gold:Fort Lazy Image Object Reference)
    LazyTexture:"Texture 2D Soft Object Reference (by ref)"=GetLargePreviewImage!(ItemDefinition:Item Definition Base Object Reference=Get ResourceItemDefinition() -> (ResourceItemDefinition:Fort Resource Item Definition Object Reference)) -> (ReturnValue:Texture 2D Soft Object Reference)
    bMatchSize:bool=false
   ) -> (then)
  ]}
]
variables []
widgets {
 info {
  inheritedWidgetCount=0
  parentClass=/Script/FortniteConversationUI.FortConversationResourceWidget
  rootWidgetClass=&Overlay
  widgetCount=8
 }
 widgets [
  Overlay_31 {widgetClassPath=&Overlay, widget=Overlay_31}
  IMG_DarkShadow {
   widgetClassPath=&Image
   bIsVariable=true
   parent=Overlay_31
   widget=IMG_DarkShadow
   properties {
    visibility=SelfHitTestInvisible
    brush {
     bIsDynamicallyLoaded=false
     drawAs=Image
     imageType=FullColor
     mirroring=NoMirror
     resourceName=None
     resourceObject=/FortUILibrary/MaterialInstances/MI_UI_RadialGradientTransparent
     tiling=NoTile
     imageSize=<75 75>
     margin=<ltrb 0 0 0 0>
     outlineSettings {
      bUseBrushTransparency=false
      roundingType=HalfHeightRadius
      width=0
      color {colorUseRule=UseColor_Specified, specifiedColor=<rgba 0 0 0 0>}
      cornerRadii=<0 0 0 0>
     }
     tintColor {colorUseRule=UseColor_Specified, specifiedColor=<rgba 0 0 0 0.9>}
     uVRegion {bIsValid=false, max=<0 0>, min=<0 0>}
    }
   }
   slot {class=OverlaySlot, properties {horizontalAlignment=HAlign_Center, padding=<ltrb 0 -25 0 0>}}
  }
  IMG_ResourceBG {
   widgetClassPath=&Image
   bIsVariable=true
   parent=Overlay_31
   widget=IMG_ResourceBG
   properties {
    visibility=SelfHitTestInvisible
    brush {
     bIsDynamicallyLoaded=false
     drawAs=Image
     imageType=FullColor
     mirroring=NoMirror
     resourceName=None
     resourceObject=null
     tiling=NoTile
     imageSize=<32 40>
     margin=<ltrb 0 0 0 0>
     outlineSettings {
      bUseBrushTransparency=false
      roundingType=HalfHeightRadius
      width=0
      color {colorUseRule=UseColor_Specified, specifiedColor=<rgba 0 0 0 0>}
      cornerRadii=<0 0 0 0>
     }
     tintColor {colorUseRule=UseColor_Specified, specifiedColor=<rgba 1 1 1 1>}
     uVRegion {bIsValid=false, max=<0 0>, min=<0 0>}
    }
   }
   slot {class=OverlaySlot, properties {horizontalAlignment=HAlign_Fill, verticalAlignment=VAlign_Center, padding=<ltrb 0 0 0 -4>}}
  }
  SizeBox_0 {
   widgetClassPath=&SizeBox
   parent=Overlay_31
   widget=SizeBox_0
   properties {
    bOverride_HeightOverride=true
    bOverride_WidthOverride=true
    heightOverride=80
    maxDesiredWidth=70
    minDesiredHeight=50
    minDesiredWidth=50
    widthOverride=80
   }
   slot {class=OverlaySlot, properties {horizontalAlignment=HAlign_Center, verticalAlignment=VAlign_Center}}
  }
  Overlay_103 {widgetClassPath=&Overlay, parent=SizeBox_0, widget=Overlay_103, properties {bOverride_Cursor=true}, slot {class=SizeBoxSlot}}
  ICO_Gold {
   widgetClassPath=&FortLazyImage
   bIsVariable=true
   parent=Overlay_103
   widget=ICO_Gold
   properties {
    bOverride_Cursor=true
    brush {
     bIsDynamicallyLoaded=false
     drawAs=Image
     imageType=FullColor
     mirroring=NoMirror
     resourceName=None
     resourceObject=/Game/UI/Foundation/Textures/Icons/Athena/T-T-Icon-BR-GoldBars-UI-Icon
     tiling=NoTile
     imageSize=<50 50>
     margin=<ltrb 0 0 0 0>
     outlineSettings {
      bUseBrushTransparency=false
      roundingType=HalfHeightRadius
      width=0
      color {colorUseRule=UseColor_Specified, specifiedColor=<rgba 0 0 0 0>}
      cornerRadii=<0 0 0 0>
     }
     tintColor {colorUseRule=UseColor_Specified, specifiedColor=<rgba 1 1 1 1>}
     uVRegion {bIsValid=false, max=<0 0>, min=<0 0>}
    }
   }
   slot {class=OverlaySlot, properties {horizontalAlignment=HAlign_Center, padding=<ltrb 0 -20 0 0>}}
  }
  ScaleBox_0 {
   widgetClassPath=&ScaleBox
   parent=Overlay_103
   widget=ScaleBox_0
   properties {stretch=ScaleToFitX, stretchDirection=DownOnly}
   slot {class=OverlaySlot, properties {horizontalAlignment=HAlign_Fill, verticalAlignment=VAlign_Center, padding=<ltrb 5 0 5 -4.5>}}
  }
  TXT_Currency {
   widgetClassPath=/Script/CommonUI.CommonNumericTextBlock
   bIsVariable=true
   parent=ScaleBox_0
   widget=TXT_Currency
   properties {
    applyLineHeightToBottomLine=false
    bIsScrollingEnabled=false
    justification=Center
    text=0
    colorAndOpacity {colorUseRule=UseColor_Specified, specifiedColor=<rgba 0 0 0 1>}
    font {
     bForceMonospaced=false
     bMaterialIsStencil=false
     fontMaterial=null
     fontObject=&BurbankSmall
     letterSpacing=0
     monospacedWidth=1
     size=21
     skewAmount=0
     typefaceFontName=Black
     outlineSettings {
      bApplyOutlineToDropShadows=false
      bMiteredCorners=false
      bSeparateFillAlpha=false
      outlineMaterial=null
      outlineSize=0
      outlineColor=<rgba 0 0 0 1>
     }
    }
    shadowOffset=<0 1>
   }
   slot {class=ScaleBoxSlot}
  }
 ]
}
