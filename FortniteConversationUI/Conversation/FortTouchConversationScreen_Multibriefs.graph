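# Touch NPC conversation screen on MobileConversationScreen that shows every option's brief card in a radial list beside a segmented wheel.
# Conversation updates set wheel segments and card spacing, then loaded briefs fill angled containers whose clicks select the matching radial item.
# Selection drives material-based arrow and highlight angles, focuses the card and updates the confirm button.
# It also async-loads the NPC portrait, shows currency, and ducks audio while active.
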WidgetBlueprint /FortniteConversationUI/Conversation/FortTouchConversationScreen_Multibriefs
animations [
 NewOptions {
  binding_count=3
  length_seconds=0.4
  object=&self:NewOptions
  sequence {bindings [
    Overlay_Wheel {
     id=#1
     possessed_class=Overlay
     tracks [{
       type=MovieScene2DTransformTrack
       sections [{
         type=MovieScene2DTransformSection
         channels [
          Translation.X {type=MovieSceneScriptingFloatChannel}
          Translation.Y {type=MovieSceneScriptingFloatChannel}
          Angle {type=MovieSceneScriptingFloatChannel}
          Scale.X {type=MovieSceneScriptingFloatChannel, keys [[0,0.8711],[4.615,1]]}
          Scale.Y {type=MovieSceneScriptingFloatChannel, keys [[0,0.8711],[4.615,1]]}
          Shear.X {type=MovieSceneScriptingFloatChannel}
          Shear.Y {type=MovieSceneScriptingFloatChannel}
         ]
        }]
      }]
    }
    RadialMenu_DialogOptions {
     id=#2
     possessed_class=FortSlottedRadialMenu_C
     tracks [{
       type=MovieScene2DTransformTrack
       sections [{
         type=MovieScene2DTransformSection
         channels [
          Translation.X {type=MovieSceneScriptingFloatChannel}
          Translation.Y {type=MovieSceneScriptingFloatChannel}
          Angle {type=MovieSceneScriptingFloatChannel}
          Scale.X {type=MovieSceneScriptingFloatChannel, keys [[0,0.8211],[7.615,1]]}
          Scale.Y {type=MovieSceneScriptingFloatChannel, keys [[0,0.8211],[7.615,1]]}
          Shear.X {type=MovieSceneScriptingFloatChannel}
          Shear.Y {type=MovieSceneScriptingFloatChannel}
         ]
        }]
      }]
    }
    BriefList {
     id=#3
     possessed_class=BriefRadialList_C
     tracks [
      {
       type=MovieScene2DTransformTrack
       sections [{
         type=MovieScene2DTransformSection
         channels [
          Translation.X {type=MovieSceneScriptingFloatChannel, keys [[0,30],[5,0]]}
          Translation.Y {type=MovieSceneScriptingFloatChannel}
          Angle {type=MovieSceneScriptingFloatChannel}
          Scale.X {type=MovieSceneScriptingFloatChannel, keys [[0,0.9],[8,1]]}
          Scale.Y {type=MovieSceneScriptingFloatChannel, keys [[0,0.9],[8,1]]}
          Shear.X {type=MovieSceneScriptingFloatChannel}
          Shear.Y {type=MovieSceneScriptingFloatChannel}
         ]
        }]
      }
      {
       type=MovieSceneFloatTrack
       display_name=Render Opacity
       sections [{type=MovieSceneFloatSection, end_frame=5, start_frame=0, channels [None {type=MovieSceneScriptingFloatChannel, keys [[0,0],[5,1]]}]}]
      }
     ]
    }
   ]}
 }
 OptionFocused {
  binding_count=1
  length_seconds=0.35
  object=&self:OptionFocused
  sequence {bindings [Image_RadialHighlight {
     id=#4
     possessed_class=Image
     tracks [
      {
       type=MovieScene2DTransformTrack
       sections [{
         type=MovieScene2DTransformSection
         end_frame=7
         start_frame=0
         channels [
          Scale.X {type=MovieSceneScriptingFloatChannel, keys [[0,0.8296],[4,1],[7,1]]}
          Scale.Y {type=MovieSceneScriptingFloatChannel, keys [[0,0.8296],[4,1],[7,1]]}
         ]
        }]
      }
      {
       type=MovieSceneFloatTrack
       display_name=Render Opacity
       sections [{type=MovieSceneFloatSection, end_frame=3, start_frame=0, channels [None {type=MovieSceneScriptingFloatChannel, keys [[0,0],[3,1]]}]}]
      }
      {
       type=MovieSceneWidgetMaterialTrack
       sections [{
         type=MovieSceneParameterSection
         end_frame=7
         start_frame=0
         channels [Thickness {type=MovieSceneScriptingFloatChannel, keys [[0,0.66],[5,0.55],[7,0.55]]}]
        }]
      }
     ]
    }]}
 }
 OptionRight {
  binding_count=1
  length_seconds=0.35
  object=&self:OptionRight
  sequence {bindings [Overlay_Details {
     id=#5
     possessed_class=Overlay
     tracks [{
       type=MovieScene2DTransformTrack
       sections [{
         type=MovieScene2DTransformSection
         channels [
          Translation.X {type=MovieSceneScriptingFloatChannel, keys [[0,-40],[7,0]]}
          Translation.Y {type=MovieSceneScriptingFloatChannel}
          Angle {type=MovieSceneScriptingFloatChannel}
          Scale.X {type=MovieSceneScriptingFloatChannel}
          Scale.Y {type=MovieSceneScriptingFloatChannel}
          Shear.X {type=MovieSceneScriptingFloatChannel}
          Shear.Y {type=MovieSceneScriptingFloatChannel}
         ]
        }]
      }]
    }]}
 }
 OptionIntro {
  binding_count=1
  length_seconds=0.2
  object=&self:OptionIntro
  sequence {bindings [Overlay_Details {
     id=#6
     possessed_class=Overlay
     tracks [{
       type=MovieSceneFloatTrack
       display_name=Render Opacity
       sections [{type=MovieSceneFloatSection, channels [None {type=MovieSceneScriptingFloatChannel, keys [[0,0],[4,1]]}]}]
      }]
    }]}
 }
 OptionLeft {
  binding_count=1
  length_seconds=0.35
  object=&self:OptionLeft
  sequence {bindings [Overlay_Details {
     id=#5
     possessed_class=Overlay
     tracks [{
       type=MovieScene2DTransformTrack
       sections [{
         type=MovieScene2DTransformSection
         channels [
          Translation.X {type=MovieSceneScriptingFloatChannel, keys [[0,40],[7,0]]}
          Translation.Y {type=MovieSceneScriptingFloatChannel}
          Angle {type=MovieSceneScriptingFloatChannel}
          Scale.X {type=MovieSceneScriptingFloatChannel}
          Scale.Y {type=MovieSceneScriptingFloatChannel}
          Shear.X {type=MovieSceneScriptingFloatChannel}
          Shear.Y {type=MovieSceneScriptingFloatChannel}
         ]
        }]
      }]
    }]}
 }
 Intro {
  binding_count=2
  length_seconds=0.5017
  object=&self:Intro
  sequence {bindings [
    Image_Background {
     id=#7
     possessed_class=Image
     tracks [{
       type=MovieSceneFloatTrack
       display_name=Render Opacity
       sections [{type=MovieSceneFloatSection, end_frame=10, start_frame=0, channels [None {type=MovieSceneScriptingFloatChannel, keys [[0,0],[10.034,1]]}]}]
      }]
    }
    Image_Radial {
     id=#8
     possessed_class=Image
     tracks [{
       type=MovieSceneFloatTrack
       display_name=Render Opacity
       sections [{type=MovieSceneFloatSection, end_frame=6, start_frame=0, channels [None {type=MovieSceneScriptingFloatChannel, keys [[0,0],[6,1]]}]}]
      }]
    }
   ]}
 }
 TapClue {
  binding_count=1
  length_seconds=0.55
  object=&self:TapClue
  sequence {bindings [TouchIconClue {
     id=#9
     possessed_class=Image
     tracks [{
       type=MovieScene2DTransformTrack
       sections [{
         type=MovieScene2DTransformSection
         end_frame=11
         start_frame=0
         channels [
          Scale.X {type=MovieSceneScriptingFloatChannel, keys [[0,1],[3,0.97],[11,1]]}
          Scale.Y {type=MovieSceneScriptingFloatChannel, keys [[0,1],[3,1.15],[11,1]]}
         ]
        }]
      }]
    }]}
 }
 ConfirmationButtonIntro {
  binding_count=1
  length_seconds=0.25
  object=&self:ConfirmationButtonIntro
  sequence {bindings [SizeBox_0 {
     id=#10
     possessed_class=SizeBox
     tracks [
      {
       type=MovieSceneFloatTrack
       display_name=Render Opacity
       sections [{type=MovieSceneFloatSection, end_frame=5, start_frame=0, channels [None {type=MovieSceneScriptingFloatChannel, keys [[0,0],[3,0],[5,1]]}]}]
      }
      {
       type=MovieScene2DTransformTrack
       sections [{
         type=MovieScene2DTransformSection
         end_frame=5
         start_frame=0
         channels [
          Translation.X {type=MovieSceneScriptingFloatChannel, keys [[0,70],[5,0]]}
          Translation.Y {type=MovieSceneScriptingFloatChannel, const [0,2]}
         ]
        }]
      }
     ]
    }]}
 }
 ConfirmHover {
  binding_count=1
  length_seconds=0.2
  object=&self:ConfirmHover
  sequence {bindings [CircleBG {
     id=#11
     possessed_class=Image
     tracks [{
       type=MovieSceneColorTrack
       display_name=Color and Opacity
       sections [{
         type=MovieSceneColorSection
         end_frame=4
         start_frame=0
         channels [
          Color.R {type=MovieSceneScriptingFloatChannel, keys [[0,1],[1,0.02],[4,1]]}
          Color.G {type=MovieSceneScriptingFloatChannel, keys [[0,1],[1,0.02],[4,1]]}
          Color.B {type=MovieSceneScriptingFloatChannel, keys [[0,1],[1,0.02],[4,1]]}
          Color.A {type=MovieSceneScriptingFloatChannel, const [1,2]}
         ]
        }]
      }]
    }]}
 }
]
graphs [
 BriefSelected {nodes [
   n1 = BriefSelected!() -> (then, BriefWidget:User Widget Object Reference) {is=title K2Node_FunctionEntry}
   > n2 = Set ItemIndex(ItemIndex:int=FindRadialIndexByBriefWidget!(
     self:Fort Conversation Screen Object Reference
     AllBriefsMap:Map of Integers to User Widget Object References=Get All Briefs Loaded() -> (All Briefs Loaded:Map of Integers to User Widget Object References)
     BriefWidget:User Widget Object Reference=@n1
    ) -> (ReturnValue:int=0)) -> (Output_Get:int=0)
   > Branch(Condition="integer >= integer"(A=@n2))
   > SelectItemAtIndex!(
    execute
    self:Fort Radial Menu Object Reference=GetSlottedRadialMenu!(self:Fort Conversation Screen Object Reference) -> (ReturnValue:Fort Slotted Radial Menu Object Reference)
    Index:int=Get ItemIndex() -> (ItemIndex:int=0)
   ) -> (then)
   > OnOptionSelectionAttempt(self:Mobile Conversation Screen Object Reference)
  ]}
 EventGraph {nodes [
   "On Clicked (BTN_RightSideRadial_Close)"() {is=title K2Node_ComponentBoundEvent}
   > EndConversation(self:Fort Conversation Screen Object Reference)
   Play Intro Audio() {is=title K2Node_CustomEvent}
   > PlaySound2D(Sound=/FortniteConversationUI/Sounds/Conversation_PopUp_Cue)
   n1 = AsyncLoad NPC Icon!() -> (OutputDelegate, then, Asset:Soft Object Reference) {is=title K2Node_CustomEvent}
   > n2 = Async Load Asset(Asset=@n1.Asset)
   n3 = Cast To Texture2D(execute=@n2.Completed, Object=@n2)
   > n4 = GetDynamicMaterial(self=Get Image_NPCPortrait() -> (Image_NPCPortrait:Image Object Reference))
   > SetTextureParameterValue(self=@n4, ParameterName=Texture, Value=@n3)
   > SetVisibility(self=Get Overlay_Portrait() -> (Overlay_Portrait:Overlay Object Reference), InVisibility=HitTestInvisible)
   HandleNPCPortrait() {is=title K2Node_CustomEvent}
   > n5 = ShouldLoadNPCPortrait!(execute, self) -> (then, Shoud Show:bool, Icon Soft Object:Soft Object Reference)
   > n6 = Branch(Condition=@n5)
   > AsyncLoad NPC Icon!(execute, self, Asset:Soft Object Reference=@n5.Icon Soft Object) -> (then) {is=title K2Node_CallFunction}
   SetVisibility(execute=@n6.else, self=Get Overlay_Portrait(), InVisibility=Collapsed)
   > SetVisibility(self=Get MobileButtonConfirm() -> (MobileButtonConfirm:Button Object Reference), InVisibility=HitTestInvisible)
   > SetConfirmationAppearance!(execute, self, ConfirmationEnabled:bool=false) -> (then)
   UpdateCurrency() {is=title K2Node_CustomEvent}
   > n7 = Cast To FortResourceItemDefinition!(
    execute
    Object="Get (a copy)"(Array:Array of Fort World Item Definition Object References=GetCurrentResourcesUsedInConversationBriefs!(self:Fort Conversation Screen Object Reference) -> (ReturnValue:Array of Fort World Item Definition Object References)) -> (Output:Fort World Item Definition Object Reference)
   ) -> (then, CastFailed, AsFort Resource Item Definition:Fort Resource Item Definition Object Reference) {is=title K2Node_DynamicCast}
   n8 = Get Currencies() -> (Currencies:Array of Fort Resource Item Definition Object References)
   > n9 = Branch(Condition=Contains Item(
     TargetArray:Array of Fort Resource Item Definition Object References=@n8
     ItemToFind:"Fort Resource Item Definition Object Reference (by ref)"=@n7
    ))
   n10 = Get Switcher_Currency() -> (Switcher_Currency:Common Visibility Switcher Object Reference)
   > SetActiveWidgetIndex(
    self:Common Visibility Switcher Object Reference=@n10
    Index=Find Item!(
     TargetArray:Array of Fort Resource Item Definition Object References=@n8
     ItemToFind:"Fort Resource Item Definition Object Reference (by ref)"=@n7
    ) -> (ReturnValue)
   )
   > SetVisibility(self=@n10, InVisibility=SelfHitTestInvisible)
   n11 = Create Widget!(
    execute:Exec=@n9.else
    Class:User Widget Class Reference=&/WBP_UI_ConversationCurrencyWidget.WBP_UI_ConversationCurrencyWidget_C
    OwningPlayer:Player Controller Object Reference
   ) -> (then:Exec, ReturnValue:WBP UI Conversation Currency Widget Object Reference)
   > Set ResourceItemDefinition!(
    execute
    ResourceItemDefinition:Fort Resource Item Definition Object Reference=@n7
    self:Fort Conversation Resource Widget Object Reference=@n11
   ) -> (then, Output_Get:Fort Resource Item Definition Object Reference)
   > n12 = Add(
    TargetArray:Array of Fort Resource Item Definition Object References=Get Currencies()
    NewItem:"Fort Resource Item Definition Object Reference (by ref)"=@n7
   ) {is=title K2Node_CallArrayFunction}
   n13 = Get Switcher_Currency()
   > AddChild(self=@n13, Content=@n11)
   > SetActiveWidgetIndex(self=@n13, Index=@n12)
   > SetVisibility(self=@n13, InVisibility=SelfHitTestInvisible)
   n14 = Event PreConstruct()
   > Branch(Condition=@n14.IsDesignTime)
   > SetWheelAppearance()
   > SetHighlightOnOptionIndex!(execute, self, Index:int=Get Original Index() -> (Original Index:int=0)) -> (then)
   > ApplyPreview!(
    execute
    self:Brief Radial List Object Reference=Get BriefList() -> (BriefList:Brief Radial List Object Reference)
    NumDesignerPreview:int=Get NumOptions() -> (NumOptions:int=0)
   ) -> (then)
   Event OnOptionsPopulated() {is=title K2Node_Event}
   > PlayAnimation(InAnimation=Get NewOptions() -> (NewOptions:Widget Animation Object Reference))
   > PlayAnimation(InAnimation=Get TapClue() -> (TapClue:Widget Animation Object Reference), NumLoopsToPlay=0, PlaybackSpeed=0.8)
   > SetConfirmationAppearance(ConfirmationEnabled=false)
   > PlayAnimation(InAnimation=Get ConfirmationButtonIntro() -> (ConfirmationButtonIntro:Widget Animation Object Reference))
   n15 = Event OnSelectedItemAvailabilityChanged!() -> (OutputDelegate, then, bIsAvailable:bool) {is=title K2Node_Event}
   > SetContentColorAndOpacity!(
    execute
    self:Border Object Reference=Get KeybindBorder() -> (KeybindBorder:Border Object Reference)
    InContentColorAndOpacity:Linear Color Structure=Select(Option 0:Linear Color Structure=(R=0.0273,G=0.0273,B=0.0273,A=1), Option 1:Linear Color Structure=(R=1,G=1,B=1,A=1), Index=@n15.bIsAvailable) -> (ReturnValue:Linear Color Structure)
   ) -> (then)
   > SetConfirmationAppearance(ConfirmationEnabled=@n15.bIsAvailable)
   n16 = Event OnItemSelected!() -> (OutputDelegate, then, OriginalIndex:int=0) {is=title K2Node_Event}
   > n17 = Sequence()
   n18 = Get OptionFocused() -> (OptionFocused:Widget Animation Object Reference)
   > StopAnimation(InAnimation=@n18)
   > PlayAnimation(InAnimation=@n18)
   > PlayAnimation(InAnimation=Get ConfirmationButtonIntro())
   > SetHighlightOnOptionIndex(Index=@n16.OriginalIndex)
   n19 = Branch(execute=@n17.then_1, Condition=ShouldBlockOptionIntroAnim!(self:Fort Conversation Screen Object Reference) -> (ReturnValue:bool))
   PlayAnimation(execute=@n19.else, InAnimation=Get OptionIntro() -> (OptionIntro:Widget Animation Object Reference))
   Event Construct()
   > Play Intro Audio() {is=title K2Node_CallFunction}
   > PlayAnimation(InAnimation=Get Intro())
   > SetVisibility(self=Get Switcher_Currency(), InVisibility=Collapsed)
   Event On Activated()
   > PushSoundMixModifier!(execute, InSoundMixModifier:Sound Mix Object Reference=/FortniteConversationUI/Sounds/Conversation_Suppress_MixMod) -> (then)
   Event On Deactivated()
   > PopSoundMixModifier!(execute, InSoundMixModifier:Sound Mix Object Reference=&Conversation_Suppress_MixMod) -> (then)
   Event OnConversationOptionUnavailable!() -> (OutputDelegate, then, SelectedEntry:Fort Radial Slot Object Reference) {is=title K2Node_Event}
   > PlaySound2D(Sound=/Game/Athena/Sounds/NPCs/Conversation/Conversation_Error_Cue)
   > SetConfirmationForInvalidOption()
   n20 = Event OnHandleConversationUpdated!() -> (OutputDelegate, then, AmountOfEquippedEntries:"Integer (by ref)"=0, MessagePayload:"Client Conversation Message Payload Structure (by ref)") {is=title K2Node_Event}
   > Set NumOptions(NumOptions:int=@n20.AmountOfEquippedEntries) -> (Output_Get:int=0)
   > SetWheelAppearance()
   > SetVisibility(
    self=[Get Image_Arrow() -> (Image_Arrow:Image Object Reference), Get Image_RadialHighlight() -> (Image_RadialHighlight:Image Object Reference)]
    InVisibility=Hidden
   )
   > ClearEntries(self:Brief Radial List Object Reference=Get BriefList())
   > HandleNPCPortrait() {is=title K2Node_CallFunction}
   n21 = Event OnAllBriefsLoaded!() -> (OutputDelegate, then, AllBriefsLoaded:Map of Integers to User Widget Object References) {is=title K2Node_Event}
   > Set All Briefs Loaded(All Briefs Loaded:Map of Integers to User Widget Object References=@n21.AllBriefsLoaded) -> (Output_Get:Map of Integers to User Widget Object References)
   > ClearEntries(self=Get BriefList())
   > n22 = For Loop(LastIndex=int - int(A=Get NumOptions() -> (NumOptions=0)))
   UpdateCurrency(execute=@n22.Completed) {is=title K2Node_CallFunction}
   n23 = AddEntry!(execute=@n22.LoopBody, self:Brief Radial List Object Reference=Get BriefList()) -> (then, New Entry:User Widget Object Reference, Success:bool)
   > n24 = Cast To FortConversationBriefContainer!(execute, Object=@n23) -> (then, CastFailed, AsFort Conversation Brief Container:Fort Conversation Brief Container Object Reference) {is=title K2Node_DynamicCast}
   n25 = CalculateItemAngleByLenght!(self:Brief Radial List Object Reference=Get BriefList(), Index:int=@n22, Lenght:int=Get NumOptions() -> (NumOptions=0)) -> (Angle:double, Direction:Vector 2D Structure)
   > SetBriefAlignment!(execute, self:Fort Conversation Brief Container Object Reference=@n24, Angle:double=@n25, Direction:Vector 2D Structure=@n25.Direction) -> (then)
   n26 = Find(TargetMap:Map of Integers to User Widget Object References=Get All Briefs Loaded(), Key:"Integer (by ref)"=@n22) -> (Value:User Widget Object Reference)
   > AddBriefChild!(
    execute
    self:Fort Conversation Brief Container Object Reference=@n24
    BriefWidget:User Widget Object Reference=Select(Option 0:User Widget Object Reference, Option 1:User Widget Object Reference=@n26, Index=@n26.ReturnValue) -> (ReturnValue:User Widget Object Reference)
    TailDirection:bool=true
   ) -> (then)
   > Bind Event to Brief Clicked(self:Fort Conversation Brief Container Object Reference=@n24, Delegate=Create Event()) {is=title K2Node_AddDelegate}
   n27 = Get ConfirmHover() -> (ConfirmHover:Widget Animation Object Reference)
   PlayAnimation(InAnimation=@n27, PlayMode=Reverse)
   PlayAnimation(InAnimation=@n27)
   "On Hovered (MobileButtonConfirm)"() {is=title K2Node_ComponentBoundEvent}
   "On Unhovered (MobileButtonConfirm)"() {is=title K2Node_ComponentBoundEvent}
  ]}
 ForceOptionSelect {nodes [
   n1 = ForceOptionSelect!() -> (then, Index:int) {is=title K2Node_FunctionEntry}
   n2 = Get EntryBox_ContextRadial!(self:Fort Radial Menu Object Reference=GetSlottedRadialMenu()) -> (EntryBox_ContextRadial:Dynamic Entry Box Object Reference)
   > n3 = Cast To CommonButtonBase!(
    execute
    Object="Get (a copy)"(
     Array:Array of User Widget Object References=GetAllEntries!(self:Dynamic Entry Box Base Object Reference=@n2) -> (ReturnValue:Array of User Widget Object References)
     Dimension 1=@n1
    ) -> (Output:User Widget Object Reference)
   ) -> (then, CastFailed, AsCommon Button Base:Common Button Base Object Reference) {is=title K2Node_DynamicCast}
   > SetIsSelected!(execute, self:Common Button Base Object Reference=@n3, InSelected:bool=true, bGiveClickFeedback:bool=true) -> (then)
  ]}
 GetAngleCompensation {nodes [
   GetAngleCompensation() {is=title K2Node_FunctionEntry}
   n1 = Get NumOptions() -> (NumOptions=0)
   > Return Node!(
    execute
    AngleCompensation:double=SelectFloat(A=Get AngleCompensation() -> (AngleCompensation:double), B=0, bPickA="integer > integer"(A=@n1, B=1))
   )
  ]}
 GetItemAngle {nodes [
   GetItemAngle() {is=title K2Node_FunctionEntry}
   n1 = Get NumOptions() -> (NumOptions=0)
   > Return Node!(execute, ReturnValue:double=SafeDivide(A=360, B="To Float (Integer)"(InInt=@n1)))
  ]}
 GetNPCToastIcon {nodes [
   n1 = GetNPCToastIcon!() -> (then, ConversationParticipants:"Conversation Participants Structure (by ref)") {is=title K2Node_FunctionEntry}
   n2 = Break Conversation Participants!(ConversationParticipants:"Conversation Participants Structure (by ref)"=@n1) -> (List:Array of Conversation Participant Entry Structures) {is=title K2Node_BreakStruct}
   > n3 = Branch(Condition=Is Not Empty!(TargetArray:Array of Conversation Participant Entry Structures=@n2) -> (ReturnValue:bool))
   > n4 = For Each Loop with Break!(Exec:Exec, Array:Array of Conversation Participant Entry Structures=@n2, Break:Exec=@n8.then) -> (LoopBody:Exec, Array Element:Conversation Participant Entry Structure, Array Index:int, Completed:Exec)
   n5 = FindConversationComponent!(execute=@n4.Completed, Actor:Actor Object Reference=Get NPC() -> (NPC:Actor Object Reference)) -> (then, ReturnValue:Conversation Participant Component Object Reference)
   > n6 = Cast To FortNPCConversationParticipantComponent!(execute, Object=@n5) -> (then, CastFailed, AsFort NPCConversation Participant Component:Fort NPCConversation Participant Component Object Reference) {is=title K2Node_DynamicCast}
   > Return Node!(
    execute
    Toast Icon:Soft Object Reference=Get ToastIcon!(self:Fort Tandem Character Data Object Reference=GetCharacterData!(self:Fort NPCConversation Participant Component Object Reference=@n6) -> (ReturnValue:Fort Tandem Character Data Object Reference)) -> (ToastIcon:Soft Object Reference)
    Successful:bool=true
   )
   Return Node(execute=@n6.CastFailed, Successful=false)
   n7 = Break Conversation Participant Entry!(ConversationParticipantEntry:"Conversation Participant Entry Structure (by ref)"=@n4) -> (Actor:Actor Object Reference, ParticipantID:Gameplay Tag Structure) {is=title K2Node_BreakStruct}
   Branch(
    execute=@n4.LoopBody
    Condition="Not Equal (Object)"(A:Object Reference=GetPlayerPawn!(PlayerIndex:int=0) -> (ReturnValue:Pawn Object Reference), B:Object Reference=@n7) -> (ReturnValue:bool=~) {is=title K2Node_PromotableOperator}
   )
   > n8 = Set NPC(NPC:Actor Object Reference=@n7) -> (Output_Get:Actor Object Reference)
   Return Node(execute=@n3.else, Successful=false)
  ]}
 RemoveContentFromWidgetHost {nodes [
   RemoveContentFromWidgetHost!() -> (
    then
    WidgetHostTag:"Gameplay Tag Structure (by ref)"
    ContentWidget:Widget Object Reference
    RemovalType:EWidgetSlottingRemovalType Enum=Unparent
   ) {is=title K2Node_FunctionEntry}
   > Return Node!(execute, ReturnValue:bool=false)
  ]}
 SetArrowRotationToIndex {nodes [
   SetArrowRotationToIndex!() -> (then, Index:int) {is=title K2Node_FunctionEntry}
   > n1 = GetDynamicMaterial(self=Get Image_Radial() -> (Image_Radial:Image Object Reference))
   > n2 = Get Scalar Parameter Value!(execute, self:Material Instance Dynamic Object Reference=@n1, ParameterName:Name=Spacing) -> (then, ReturnValue:float) {is=title K2Node_CallFunction}
   > n3 = GetDynamicMaterial(self=Get Image_Arrow())
   n4 = GetItemAngle!() -> (ReturnValue:double) {is=title K2Node_CallFunction}
   n5 = SafeDivide(A=@n4, B=360)
   n6 = float * float(A=@n4, B:int="Min (Integer)"(A=Get NumOptions() -> (NumOptions=0), B=Get Index() -> (Index:int=0)) -> (ReturnValue=0)) {is=title K2Node_PromotableOperator}
   n7 = float + float(A=float * float(A=SafeDivide(A=@n6, B=360), B:double=0.5), B=float * float(A=@n5, B=0.25))
   > SetScalarParameterValue(self=@n3, ParameterName=ArrowAngle, Value=@n7)
   n8 = float * float(A=@n5, B=0.5)
   n9 = float - float(A=1, B=@n2)
   > SetScalarParameterValue(self=@n3, ParameterName=RadialMask, Value=float - float(A=@n8, B=float * float(A=@n9, B=0.25)))
   > n10 = GetDynamicMaterial(self=Get Image_RadialHighlight())
   > SetScalarParameterValue(self=@n10, ParameterName=Angle, Value=@n7)
   > SetScalarParameterValue(self=@n10, ParameterName=RadialMask, Value=@n8)
  ]}
 SetConfirmationAppearance {nodes [
   n1 = SetConfirmationAppearance!() -> (then, ConfirmationEnabled:bool=~) {is=title K2Node_FunctionEntry}
   > SetBrushTintColor!(
    execute
    self:Image Object Reference=Get Confirmation_Icon() -> (Confirmation_Icon:Image Object Reference)
    TintColor:Slate Color Structure=Select(
     Option 0:Slate Color Structure=Make Slate Color(SpecifiedColor=(R=0.0437,G=0.0423,B=0.0452,A=1))
     Option 1:Slate Color Structure=Make Slate Color(SpecifiedColor=(R=0.0999,G=0.6445,B=0.1046,A=1))
     Index=@n1
    ) -> (ReturnValue:Slate Color Structure)
   ) -> (then)
   > SetColorAndOpacity(
    self:Image Object Reference=Get CircleBG() -> (CircleBG:Image Object Reference)
    InColorAndOpacity:Linear Color Structure=Select(Option 0:Linear Color Structure=(R=0.0296,G=0.0242,B=0.0307,A=1), Option 1:Linear Color Structure=(R=1,G=1,B=1,A=1), Index=@n1) -> (ReturnValue:Linear Color Structure)
   )
  ]}
 SetConfirmationForInvalidOption {nodes [
   SetConfirmationForInvalidOption() {is=title K2Node_FunctionEntry}
   > n1 = Branch(Condition=DoesCurrentConversationNPCHaveIcon!(self:Fort Conversation Screen Object Reference) -> (ReturnValue:bool))
   > AsyncLoad NPC Icon(Asset=GetCurrentConversationNPCIcon!(self:Fort Conversation Screen Object Reference) -> (ReturnValue:Soft Object Reference))
   > SetVisibility(self=Get MobileButtonConfirm(), InVisibility=Hidden)
   SetVisibility(execute=@n1.else, self=Get MobileButtonConfirm(), InVisibility=Visible)
   > SetConfirmationAppearance(ConfirmationEnabled=false) {is=title K2Node_CallFunction}
  ]}
 SetFocusOnPreview {nodes [
   n1 = SetFocusOnPreview!() -> (then, Index:int, OnFocus:bool=~) {is=title K2Node_FunctionEntry}
   n2 = GetEntryWidget!(self:Brief Radial List Object Reference=Get BriefList(), Index:int=@n1) -> (UserWidget:User Widget Object Reference)
   > Is Valid(InputObject=@n2)
   n3 = Cast To FortConversationBriefContainer!(Object=@n2) -> (AsFort Conversation Brief Container, bSuccess:bool=~)
   > FocusPreviewCard!(execute, self:Fort Conversation Brief Container Object Reference=@n3, IsHovered:bool=@n1.OnFocus) -> (then)
   > n4 = Branch(Condition=@n1.OnFocus)
   > SetBriefContainerToHighestZOrder!(execute, self:Fort Conversation Brief Radial List Object Reference=Get BriefList(), BriefContainer:User Widget Object Reference=@n3) -> (then)
   ResetBriefContainerZOrder!(execute=@n4.else, self:Fort Conversation Brief Radial List Object Reference=Get BriefList(), BriefContainer:User Widget Object Reference=@n3) -> (then)
  ]}
 SetHighlightOnOptionIndex {nodes [
   SetHighlightOnOptionIndex!() -> (then, Index:int) {is=title K2Node_FunctionEntry}
   > n1 = Branch(Condition="Not Equal (Integer)"(A=Get Index() -> (Index=0), B=-1))
   > SetFocusOnPreview!(execute, self, Index:int=Get Original Index() -> (Original Index=0), OnFocus:bool=false) -> (then) {is=title K2Node_CallFunction}
   > SetArrowRotationToIndex!(execute, self, Index:int=Get Index() -> (Index=0)) -> (then) {is=title K2Node_CallFunction}
   > n2 = Set Original Index(Original Index:int=Get Index() -> (Index=0)) -> (Output_Get:int=0)
   > SetFocusOnPreview(Index=@n2, OnFocus=true)
   > SetVisibility(self=Get Overlay_Portrait(), InVisibility=Hidden)
   > SetVisibility(self=[Get Image_RadialHighlight(), Get Image_Arrow()], InVisibility=SelfHitTestInvisible)
   > SetVisibility(self=Get MobileButtonConfirm(), InVisibility=Visible)
   SetVisibility(execute=@n1.else, self=[Get Image_Arrow(), Get Image_RadialHighlight()], InVisibility=Hidden)
   > PrintString!(execute, InString=Hide, bPrintToScreen, bPrintToLog, TextColor, Duration, Key) -> (then)
   > SetConfirmationForInvalidOption() {is=title K2Node_CallFunction}
  ]}
 SetRadialListOptionDistribution {nodes [
   SetRadialListOptionDistribution!() -> (then, Num:int) {is=title K2Node_FunctionEntry}
   n1 = Make Literal Float(Value=180)
   n2 = Get Num() -> (Num:int=0)
   n3 = float - float(
    A=SafeDivide(A=@n1, B="To Float (Integer)"(InInt=@n2))
    B=GetAngleCompensation!() -> (AngleCompensation:double) {is=title K2Node_CallFunction}
   )
   n4 = Get BriefList()
   > Set MaxAngleLenght!(execute, MaxAngleLenght:double=float - float(A=@n1, B=@n3), self:Brief Radial List Object Reference=@n4) -> (then, Output_Get:double)
   > Set StartAngle!(execute, StartAngle:double=float * float(A=@n3, B=-0.5), self:Brief Radial List Object Reference=@n4) -> (then, Output_Get:double)
   n5 = "integer < integer"(A=Get Num() -> (Num=0), B=5)
   > Set!(
    execute
    HarmonicDistortion_Frequency_7_#12:double=2
    HarmonicDistortion_Amplitude_8_#13:double=SelectFloat(A=6, B=12, bPickA=@n5)
    HarmonicDistortion_Rotataion_9_#14:double=0
    self:Brief Radial List Object Reference=Get BriefList()
   ) -> (then, Output_Get:Harmonic Distortion Structure)
   > Set AngleCompensation(AngleCompensation:double=SelectFloat(A=0, B=20, bPickA=@n5)) -> (Output_Get:double)
   > Return Node!(execute)
  ]}
 SetWheelAppearance {nodes [
   SetWheelAppearance() {is=title K2Node_FunctionEntry}
   n1 = Get NumOptions() -> (NumOptions=0)
   > SetRadialListOptionDistribution!(execute, self, Num:int=@n1) -> (then) {is=title K2Node_CallFunction}
   > n2 = GetDynamicMaterial(self=Get Image_Radial())
   > SetScalarParameterValue(self=@n2, ParameterName=Segment Count, Value="To Float (Integer)"(InInt=@n1))
  ]}
 ShouldLoadNPCPortrait {nodes [
   ShouldLoadNPCPortrait() {is=title K2Node_FunctionEntry}
   > n1 = GetConsoleVariableBoolValue!(execute, VariableName:String=Fort.Rollback.AutoSelectWhenOnlyOneOptionIsAvailable) -> (then, ReturnValue:bool)
   > n2 = Branch(Condition=@n1)
   n3 = GetIsAutoSelectingOption!(self:Mobile Conversation Screen Object Reference) -> (ReturnValue:bool)
   > Return Node!(
    execute
    Shoud Show:bool=AND Boolean(A=NOT Boolean(A=@n3), B=DoesCurrentConversationNPCHaveIcon())
    Icon Soft Object:Soft Object Reference=GetCurrentConversationNPCIcon()
   )
   Return Node(execute=@n2.else, Shoud Show=DoesCurrentConversationNPCHaveIcon(), Icon Soft Object=GetCurrentConversationNPCIcon())
  ]}
]
variables [
 All Briefs Loaded {default {}}
 AngleCompensation {type=number, default=20}
 Currencies {type="array<FortResourceItemDefinition>", default [/Game/Items/ResourcePickups/Athena_WadsItemData]}
 NumOptions {type=integer, default=5}
 Original Index {type=integer, default=-1}
]
widgets {
 info {
  inheritedWidgetCount=14
  parentClass=/Script/FortniteConversationUI.MobileConversationScreen
  rootWidgetClass=&Overlay
  widgetCount=49
 }
 widgets [
  Overlay_0 {widgetClassPath=&Overlay, widget=Overlay_0}
  HorizontalBox_0 {
   widgetClassPath=&HorizontalBox
   parent=Overlay_0
   widget=HorizontalBox_0
   properties {bOverride_Cursor=true}
   slot {class=OverlaySlot, properties {horizontalAlignment=HAlign_Fill, verticalAlignment=VAlign_Fill}}
  }
  Spacer_83 {
   widgetClassPath=&Spacer
   parent=HorizontalBox_0
   widget=Spacer_83
   slot {class=HorizontalBoxSlot, properties {size {sizeRule=Fill, value=0.5}}}
  }
  Overlay_3 {
   widgetClassPath=&Overlay
   parent=HorizontalBox_0
   widget=Overlay_3
   slot {class=HorizontalBoxSlot, properties {size {sizeRule=Fill, value=0.5}}}
  }
  Image_Background {
   widgetClassPath=&Image
   bInherited=true
   parent=Overlay_3
   widget=Image_Background
   properties {
    visibility=SelfHitTestInvisible
    brush {
     bIsDynamicallyLoaded=false
     drawAs=Image
     imageType=NoImage
     mirroring=NoMirror
     resourceName=None
     resourceObject=/FortniteConversationUI/Conversation/Materials/M_UI_MobileConversationBG_Inst1
     tiling=NoTile
     imageSize=<32 32>
     margin=<ltrb 0 0 0 0>
     outlineSettings {
      bUseBrushTransparency=false
      roundingType=HalfHeightRadius
      width=0
      color {colorUseRule=UseColor_Specified, specifiedColor=<rgba 0 0 0 0>}
      cornerRadii=<0 0 0 0>
     }
     tintColor {colorUseRule=UseColor_Specified, specifiedColor=<rgba 1 1 1 1>}
     uVRegion {bIsValid=false, max=<0 0>, min=<0 0>}
    }
    colorAndOpacity=<rgba 1 1 1 0.9>
    renderTransform {angle=0, scale=<1 -1>, shear=<0 0>, translation=<0 0>}
   }
   slot {class=OverlaySlot, properties {horizontalAlignment=HAlign_Fill, verticalAlignment=VAlign_Fill}}
  }
  VisibilityBorder_Touch {
   widgetClassPath=&CommonHardwareVisibilityBorder
   parent=Overlay_3
   widget=VisibilityBorder_Touch
   properties {visibilityQuery {
     autoDescription=" ANY( Input.Touch )"
     tokenStreamVersion=0
     userDescription=if using touch
     queryTokenStream [0,1,1,1,0]
     tagDictionary [{tagName=Input.Touch}]
    }}
   slot {class=OverlaySlot, properties {horizontalAlignment=HAlign_Fill, verticalAlignment=VAlign_Fill}}
  }
  CloseButtonArea {
   widgetClassPath=&Button
   bInherited=true
   parent=VisibilityBorder_Touch
   widget=CloseButtonArea
   properties {widgetStyle {
     clickedSlateSound {resourceObject=null}
     disabled {
      bIsDynamicallyLoaded=false
      drawAs=NoDrawType
      imageType=NoImage
      mirroring=NoMirror
      resourceName=None
      resourceObject=null
      tiling=NoTile
      imageSize=<0 0>
      margin=<ltrb 0 0 0 0>
      outlineSettings {
       bUseBrushTransparency=false
       roundingType=HalfHeightRadius
       width=0
       color {colorUseRule=UseColor_Specified, specifiedColor=<rgba 0 0 0 0>}
       cornerRadii=<0 0 0 0>
      }
      tintColor {colorUseRule=UseColor_Specified, specifiedColor=<rgba 1 1 1 1>}
      uVRegion {bIsValid=false, max=<0 0>, min=<0 0>}
     }
     disabledForeground {colorUseRule=UseColor_Specified, specifiedColor=<rgba 0.5271 0.5271 0.5271 1>}
     hovered {
      bIsDynamicallyLoaded=false
      drawAs=NoDrawType
      imageType=NoImage
      mirroring=NoMirror
      resourceName=None
      resourceObject=null
      tiling=NoTile
      imageSize=<32 32>
      margin=<ltrb 0 0 0 0>
      outlineSettings {
       bUseBrushTransparency=true
       roundingType=FixedRadius
       width=1
       color {colorUseRule=UseColor_Specified, specifiedColor=<rgba 0.7243 0.7243 0.7243 1>}
       cornerRadii=<4 4 4 4>
      }
      tintColor {colorUseRule=UseColor_Specified, specifiedColor=<rgba 0.7243 0.7243 0.7243 1>}
      uVRegion {bIsValid=false, max=<0 0>, min=<0 0>}
     }
     hoveredForeground {colorUseRule=UseColor_Specified, specifiedColor=<rgba 0 1 1 1>}
     hoveredSlateSound {resourceObject=null}
     normal {
      bIsDynamicallyLoaded=false
      drawAs=NoDrawType
      imageType=NoImage
      mirroring=NoMirror
      resourceName=None
      resourceObject=null
      tiling=NoTile
      imageSize=<32 32>
      margin=<ltrb 0 0 0 0>
      outlineSettings {
       bUseBrushTransparency=true
       roundingType=FixedRadius
       width=1
       color {colorUseRule=UseColor_Specified, specifiedColor=<rgba 0.6951 0.6951 0.6951 1>}
       cornerRadii=<4 4 4 4>
      }
      tintColor {colorUseRule=UseColor_Specified, specifiedColor=<rgba 0.4955 0.4955 0.4955 1>}
      uVRegion {bIsValid=false, max=<0 0>, min=<0 0>}
     }
     normalForeground {colorUseRule=UseColor_Specified, specifiedColor=<rgba 0 1 1 1>}
     normalPadding=<ltrb 12 1.5 12 1.5>
     pressed {
      bIsDynamicallyLoaded=false
      drawAs=NoDrawType
      imageType=NoImage
      mirroring=NoMirror
      resourceName=None
      resourceObject=null
      tiling=NoTile
      imageSize=<32 32>
      margin=<ltrb 0 0 0 0>
      outlineSettings {
       bUseBrushTransparency=true
       roundingType=FixedRadius
       width=1
       color {colorUseRule=UseColor_Specified, specifiedColor=<rgba 0.7243 0.7243 0.7243 1>}
       cornerRadii=<4 4 4 4>
      }
      tintColor {colorUseRule=UseColor_Specified, specifiedColor=<rgba 0.3843 0.3843 0.3843 1>}
      uVRegion {bIsValid=false, max=<0 0>, min=<0 0>}
     }
     pressedForeground {colorUseRule=UseColor_Specified, specifiedColor=<rgba 0 1 1 1>}
     pressedPadding=<ltrb 12 1.5 12 1.5>
     pressedSlateSound {resourceObject=null}
    }}
   slot {class=BorderSlot, properties {padding=<ltrb 0 0 0 0>}}
  }
  SafeZone_CloseTouch {
   widgetClassPath=&SafeZone
   bIsVariable=true
   parent=Overlay_0
   widget=SafeZone_CloseTouch
   properties {visibility=HitTestInvisible}
   slot {class=OverlaySlot, properties {horizontalAlignment=HAlign_Fill, verticalAlignment=VAlign_Fill}}
  }
  Overlay_2 {widgetClassPath=&Overlay, parent=SafeZone_CloseTouch, widget=Overlay_2, slot {class=SafeZoneSlot}}
  Fake_CloseButton {
   widgetClassPath=/FortUILibrary/Foundation/Buttons/CloseButton
   bIsVariable=true
   parent=Overlay_2
   widget=Fake_CloseButton
   properties {
    holdData=&CommonUI_HoldData_Base_C
    onButtonBaseClicked=(&self:WidgetTree.Fake_CloseButton.HandleOnButtonClicked)
    onButtonBaseDoubleClicked=(&self:WidgetTree.Fake_CloseButton.HandleOnButtonDoubleClicked)
    onButtonBaseHovered=(&self:WidgetTree.Fake_CloseButton.HandleOnButtonHovered)
    onButtonBaseUnhovered=(&self:WidgetTree.Fake_CloseButton.HandleOnButtonUnhovered)
    onSelectedChangedBase=(&self:WidgetTree.Fake_CloseButton.HandleOnSelectedChanged)
   }
   slot {class=OverlaySlot, properties {horizontalAlignment=HAlign_Right, padding=<ltrb 0 35 35 0>}}
  }
  Switcher_Currency {
   widgetClassPath=&CommonVisibilitySwitcher
   bIsVariable=true
   parent=Overlay_2
   widget=Switcher_Currency
   properties {visibility=Collapsed}
   slot {class=OverlaySlot, properties {horizontalAlignment=HAlign_Right, padding=<ltrb 0 180 35 0>}}
  }
  WBP_UI_ConversationCurrencyWidget {
   widgetClassPath=&/WBP_UI_ConversationCurrencyWidget
   bIsVariable=true
   parent=Switcher_Currency
   widget=WBP_UI_ConversationCurrencyWidget
   slot {class=CommonVisibilitySwitcherSlot}
  }
  Overlay_Details {
   widgetClassPath=&Overlay
   bInherited=true
   parent=Overlay_0
   widget=Overlay_Details
   properties {visibility=Collapsed}
   slot {class=OverlaySlot, properties {horizontalAlignment=HAlign_Fill, verticalAlignment=VAlign_Bottom}}
  }
  SizeBox_Details {
   widgetClassPath=&SizeBox
   parent=Overlay_Details
   widget=SizeBox_Details
   properties {bOverride_WidthOverride=true, visibility=Collapsed, widthOverride=852}
   slot {class=OverlaySlot, properties {horizontalAlignment=HAlign_Right, verticalAlignment=VAlign_Bottom, padding=<ltrb 0 0 0 100>}}
  }
  Switcher_Details {
   widgetClassPath=&CommonVisibilitySwitcher
   bInherited=true
   parent=SizeBox_Details
   widget=Switcher_Details
   properties {activeWidgetIndex=1}
   slot {class=SizeBoxSlot, properties {horizontalAlignment=HAlign_Center}}
  }
  Spacer_94 {widgetClassPath=&Spacer, parent=Switcher_Details, widget=Spacer_94, slot {class=CommonVisibilitySwitcherSlot}}
  BasicBrief_Main {
   widgetClassPath=&/FortBasicBrief
   bInherited=true
   parent=Switcher_Details
   widget=BasicBrief_Main
   slot {class=CommonVisibilitySwitcherSlot}
  }
  Switcher_CenteredDetails {
   widgetClassPath=&CommonVisibilitySwitcher
   bInherited=true
   parent=Overlay_Details
   widget=Switcher_CenteredDetails
   properties {activeWidgetIndex=1, visibility=Collapsed}
   slot {class=OverlaySlot, properties {horizontalAlignment=HAlign_Fill, verticalAlignment=VAlign_Fill}}
  }
  Spacer {widgetClassPath=&Spacer, parent=Switcher_CenteredDetails, widget=Spacer, slot {class=CommonVisibilitySwitcherSlot}}
  BasicBrief_Main_1 {
   widgetClassPath=&FortBasicBrief
   parent=Switcher_CenteredDetails
   widget=BasicBrief_Main_1
   slot {class=CommonVisibilitySwitcherSlot}
  }
  Overlay_ConversationContainer {
   widgetClassPath=&Overlay
   bIsVariable=true
   parent=Overlay_0
   widget=Overlay_ConversationContainer
   slot {class=OverlaySlot, properties {horizontalAlignment=HAlign_Right, verticalAlignment=VAlign_Center}}
  }
  Overlay_WheelContainer {
   widgetClassPath=&Overlay
   bInherited=true
   parent=Overlay_ConversationContainer
   widget=Overlay_WheelContainer
   slot {class=OverlaySlot, properties {horizontalAlignment=HAlign_Center, verticalAlignment=VAlign_Bottom}}
  }
  SizeBox_RadialMenu {
   widgetClassPath=&SizeBox
   bInherited=true
   parent=Overlay_WheelContainer
   widget=SizeBox_RadialMenu
   properties {bOverride_HeightOverride=true, bOverride_WidthOverride=true, heightOverride=500, widthOverride=500}
   slot {class=OverlaySlot, properties {horizontalAlignment=HAlign_Center, verticalAlignment=VAlign_Bottom}}
  }
  Overlay_Wheel {widgetClassPath=&Overlay, parent=SizeBox_RadialMenu, widget=Overlay_Wheel, slot {class=SizeBoxSlot}}
  Image_Radial {
   widgetClassPath=&Image
   bIsVariable=true
   parent=Overlay_Wheel
   widget=Image_Radial
   properties {
    visibility=HitTestInvisible
    brush {
     bIsDynamicallyLoaded=false
     drawAs=Image
     imageType=FullColor
     mirroring=NoMirror
     resourceName=None
     resourceObject=/FortniteConversationUI/Conversation/Materials/MaterialInstances/M_ConversationSection_Inst
     tiling=NoTile
     imageSize=<128 128>
     margin=<ltrb 0 0 0 0>
     outlineSettings {
      bUseBrushTransparency=false
      roundingType=HalfHeightRadius
      width=0
      color {colorUseRule=UseColor_Specified, specifiedColor=<rgba 0 0 0 0>}
      cornerRadii=<0 0 0 0>
     }
     tintColor {colorUseRule=UseColor_Specified, specifiedColor=<rgba 1 1 1 1>}
     uVRegion {bIsValid=false, max=<0 0>, min=<0 0>}
    }
   }
   slot {class=OverlaySlot, properties {horizontalAlignment=HAlign_Fill, verticalAlignment=VAlign_Fill, padding=<ltrb 15 15 15 15>}}
  }
  Image_Arrow {
   widgetClassPath=&Image
   bIsVariable=true
   parent=Overlay_Wheel
   widget=Image_Arrow
   properties {
    visibility=HitTestInvisible
    brush {
     bIsDynamicallyLoaded=false
     drawAs=Image
     imageType=FullColor
     mirroring=NoMirror
     resourceName=None
     resourceObject=/FortniteConversationUI/Conversation/Materials/MaterialInstances/M_PointerArrow_Inst
     tiling=NoTile
     imageSize=<200 200>
     margin=<ltrb 0 0 0 0>
     outlineSettings {
      bUseBrushTransparency=false
      roundingType=HalfHeightRadius
      width=0
      color {colorUseRule=UseColor_Specified, specifiedColor=<rgba 0 0 0 0>}
      cornerRadii=<0 0 0 0>
     }
     tintColor {colorUseRule=UseColor_Specified, specifiedColor=<rgba 1 1 1 1>}
     uVRegion {bIsValid=false, max=<0 0>, min=<0 0>}
    }
   }
   slot {class=OverlaySlot, properties {horizontalAlignment=HAlign_Center, verticalAlignment=VAlign_Center}}
  }
  Image_RadialHighlight {
   widgetClassPath=&Image
   bInherited=true
   parent=Overlay_Wheel
   widget=Image_RadialHighlight
   properties {
    visibility=SelfHitTestInvisible
    brush {
     bIsDynamicallyLoaded=false
     drawAs=Image
     imageType=FullColor
     mirroring=NoMirror
     resourceName=None
     resourceObject=/FortniteConversationUI/Conversation/Materials/MaterialInstances/MI_HighlightSection
     tiling=NoTile
     imageSize=<32 32>
     margin=<ltrb 0 0 0 0>
     outlineSettings {
      bUseBrushTransparency=false
      roundingType=HalfHeightRadius
      width=0
      color {colorUseRule=UseColor_Specified, specifiedColor=<rgba 0 0 0 0>}
      cornerRadii=<0 0 0 0>
     }
     tintColor {colorUseRule=UseColor_Specified, specifiedColor=<rgba 1 1 1 1>}
     uVRegion {bIsValid=false, max=<0 0>, min=<0 0>}
    }
   }
   slot {class=OverlaySlot, properties {horizontalAlignment=HAlign_Fill, verticalAlignment=VAlign_Fill}}
  }
  RadialMenu_DialogOptions {
   widgetClassPath=/Game/Athena/HUD/ContextMenus/FortSlottedRadialMenu
   bInherited=true
   bIsVariable=true
   parent=Overlay_Wheel
   widget=RadialMenu_DialogOptions
   properties {
    bUseAsManySlotsAsItems=true
    entryWidgetClass=&/FortConversationOption.FortConversationOption_C
    joystickDeadZone=0.4
    mouseDeadZoneRadiusPercentage=0.3
    navigation=&self:WidgetTree.RadialMenu_DialogOptions.WidgetNavigation_0
    numSlotsSupported=5
    orderingMode=Clockwise
    visibility=Visible
    customOrdering [N,NNE,NE,ENE,E]
    radialBoxSettings {angleBetweenItems=0, bClockwiseOrder=false, bDistributeItemsEvenly=true, marginSize=0, sectorCentralAngle=180, startingAngle=90}
   }
   slot {class=OverlaySlot, properties {horizontalAlignment=HAlign_Fill, verticalAlignment=VAlign_Fill, padding=<ltrb 35 35 35 35>}}
  }
  Image_LoadingSpinner {
   widgetClassPath=&Image
   bInherited=true
   parent=Overlay_WheelContainer
   widget=Image_LoadingSpinner
   properties {
    visibility=Hidden
    brush {
     bIsDynamicallyLoaded=false
     drawAs=Image
     imageType=NoImage
     mirroring=NoMirror
     resourceName=None
     resourceObject=/FortUILibrary/UIKit/Throbber/MI_UIKit_Monochrome_Throbber
     tiling=NoTile
     imageSize=<50 50>
     margin=<ltrb 0 0 0 0>
     outlineSettings {
      bUseBrushTransparency=false
      roundingType=HalfHeightRadius
      width=0
      color {colorUseRule=UseColor_Specified, specifiedColor=<rgba 0 0 0 0>}
      cornerRadii=<0 0 0 0>
     }
     tintColor {colorUseRule=UseColor_Specified, specifiedColor=<rgba 1 1 1 1>}
     uVRegion {bIsValid=false, max=<0 0>, min=<0 0>}
    }
    colorAndOpacity=<rgba 1 1 1 0.4>
   }
   slot {class=OverlaySlot, properties {horizontalAlignment=HAlign_Center, verticalAlignment=VAlign_Center}}
  }
  VisibilityBorder_Touch_1 {
   widgetClassPath=&CommonHardwareVisibilityBorder
   parent=Overlay_WheelContainer
   widget=VisibilityBorder_Touch_1
   properties {
    padding=<ltrb 260 0 0 0>
    visibilityQuery {
     autoDescription=" ANY( Input.Touch )"
     tokenStreamVersion=0
     userDescription=if using touch
     queryTokenStream [0,1,1,1,0]
     tagDictionary [{tagName=Input.Touch}]
    }
   }
   slot {class=OverlaySlot, properties {horizontalAlignment=HAlign_Fill, verticalAlignment=VAlign_Fill}}
  }
  BTN_RightSideRadial_Close {
   widgetClassPath=&Button
   parent=VisibilityBorder_Touch_1
   widget=BTN_RightSideRadial_Close
   properties {widgetStyle {
     clickedSlateSound {resourceObject=null}
     disabled {
      bIsDynamicallyLoaded=false
      drawAs=NoDrawType
      imageType=NoImage
      mirroring=NoMirror
      resourceName=None
      resourceObject=null
      tiling=NoTile
      imageSize=<0 0>
      margin=<ltrb 0 0 0 0>
      outlineSettings {
       bUseBrushTransparency=false
       roundingType=HalfHeightRadius
       width=0
       color {colorUseRule=UseColor_Specified, specifiedColor=<rgba 0 0 0 0>}
       cornerRadii=<0 0 0 0>
      }
      tintColor {colorUseRule=UseColor_Specified, specifiedColor=<rgba 1 1 1 1>}
      uVRegion {bIsValid=false, max=<0 0>, min=<0 0>}
     }
     disabledForeground {colorUseRule=UseColor_Specified, specifiedColor=<rgba 0.5271 0.5271 0.5271 1>}
     hovered {
      bIsDynamicallyLoaded=false
      drawAs=NoDrawType
      imageType=NoImage
      mirroring=NoMirror
      resourceName=None
      resourceObject=null
      tiling=NoTile
      imageSize=<32 32>
      margin=<ltrb 0 0 0 0>
      outlineSettings {
       bUseBrushTransparency=true
       roundingType=FixedRadius
       width=1
       color {colorUseRule=UseColor_Specified, specifiedColor=<rgba 0.7243 0.7243 0.7243 1>}
       cornerRadii=<4 4 4 4>
      }
      tintColor {colorUseRule=UseColor_Specified, specifiedColor=<rgba 0.7243 0.7243 0.7243 1>}
      uVRegion {bIsValid=false, max=<0 0>, min=<0 0>}
     }
     hoveredForeground {colorUseRule=UseColor_Specified, specifiedColor=<rgba 0 1 1 1>}
     hoveredSlateSound {resourceObject=null}
     normal {
      bIsDynamicallyLoaded=false
      drawAs=NoDrawType
      imageType=NoImage
      mirroring=NoMirror
      resourceName=None
      resourceObject=null
      tiling=NoTile
      imageSize=<32 32>
      margin=<ltrb 0 0 0 0>
      outlineSettings {
       bUseBrushTransparency=true
       roundingType=FixedRadius
       width=1
       color {colorUseRule=UseColor_Specified, specifiedColor=<rgba 0.6951 0.6951 0.6951 1>}
       cornerRadii=<4 4 4 4>
      }
      tintColor {colorUseRule=UseColor_Specified, specifiedColor=<rgba 0.4955 0.4955 0.4955 1>}
      uVRegion {bIsValid=false, max=<0 0>, min=<0 0>}
     }
     normalForeground {colorUseRule=UseColor_Specified, specifiedColor=<rgba 0 1 1 1>}
     normalPadding=<ltrb 12 1.5 12 1.5>
     pressed {
      bIsDynamicallyLoaded=false
      drawAs=NoDrawType
      imageType=NoImage
      mirroring=NoMirror
      resourceName=None
      resourceObject=null
      tiling=NoTile
      imageSize=<32 32>
      margin=<ltrb 0 0 0 0>
      outlineSettings {
       bUseBrushTransparency=true
       roundingType=FixedRadius
       width=1
       color {colorUseRule=UseColor_Specified, specifiedColor=<rgba 0.7243 0.7243 0.7243 1>}
       cornerRadii=<4 4 4 4>
      }
      tintColor {colorUseRule=UseColor_Specified, specifiedColor=<rgba 0.3843 0.3843 0.3843 1>}
      uVRegion {bIsValid=false, max=<0 0>, min=<0 0>}
     }
     pressedForeground {colorUseRule=UseColor_Specified, specifiedColor=<rgba 0 1 1 1>}
     pressedPadding=<ltrb 12 1.5 12 1.5>
     pressedSlateSound {resourceObject=null}
    }}
   slot {class=BorderSlot, properties {padding=<ltrb 260 0 0 0>}}
  }
  MobileButtonConfirm {
   widgetClassPath=&Button
   bInherited=true
   parent=Overlay_WheelContainer
   widget=MobileButtonConfirm
   properties {
    visibility=SelfHitTestInvisible
    widgetStyle {
     clickedSlateSound {resourceObject=null}
     disabled {
      bIsDynamicallyLoaded=false
      drawAs=NoDrawType
      imageType=NoImage
      mirroring=NoMirror
      resourceName=None
      resourceObject=null
      tiling=NoTile
      imageSize=<64 64>
      margin=<ltrb 0 0 0 0>
      outlineSettings {
       bUseBrushTransparency=false
       roundingType=HalfHeightRadius
       width=0
       color {colorUseRule=UseColor_Specified, specifiedColor=<rgba 0 0 0 0>}
       cornerRadii=<0 0 0 0>
      }
      tintColor {colorUseRule=UseColor_Specified, specifiedColor=<rgba 1 1 1 0>}
      uVRegion {bIsValid=false, max=<0 0>, min=<0 0>}
     }
     disabledForeground {colorUseRule=UseColor_Specified, specifiedColor=<rgba 0.5271 0.5271 0.5271 1>}
     hovered {
      bIsDynamicallyLoaded=false
      drawAs=NoDrawType
      imageType=NoImage
      mirroring=NoMirror
      resourceName=None
      resourceObject=null
      tiling=NoTile
      imageSize=<72 72>
      margin=<ltrb 0 0 0 0>
      outlineSettings {
       bUseBrushTransparency=true
       roundingType=FixedRadius
       width=1
       color {colorUseRule=UseColor_Specified, specifiedColor=<rgba 0.7243 0.7243 0.7243 1>}
       cornerRadii=<4 4 4 4>
      }
      tintColor {colorUseRule=UseColor_Specified, specifiedColor=<rgba 1 1 1 0>}
      uVRegion {bIsValid=false, max=<0 0>, min=<0 0>}
     }
     hoveredForeground {colorUseRule=UseColor_Specified, specifiedColor=<rgba 0 1 1 1>}
     hoveredSlateSound {resourceObject=null}
     normal {
      bIsDynamicallyLoaded=false
      drawAs=NoDrawType
      imageType=NoImage
      mirroring=NoMirror
      resourceName=None
      resourceObject=null
      tiling=NoTile
      imageSize=<72 72>
      margin=<ltrb 0 0 0 0>
      outlineSettings {
       bUseBrushTransparency=true
       roundingType=FixedRadius
       width=1
       color {colorUseRule=UseColor_Specified, specifiedColor=<rgba 0.6951 0.6951 0.6951 1>}
       cornerRadii=<4 4 4 4>
      }
      tintColor {colorUseRule=UseColor_Specified, specifiedColor=<rgba 1 1 1 0>}
      uVRegion {bIsValid=false, max=<0 0>, min=<0 0>}
     }
     normalForeground {colorUseRule=UseColor_Specified, specifiedColor=<rgba 0 1 1 1>}
     normalPadding=<ltrb 12 1.5 12 1.5>
     pressed {
      bIsDynamicallyLoaded=false
      drawAs=NoDrawType
      imageType=NoImage
      mirroring=NoMirror
      resourceName=None
      resourceObject=null
      tiling=NoTile
      imageSize=<72 72>
      margin=<ltrb 0 0 0 0>
      outlineSettings {
       bUseBrushTransparency=true
       roundingType=FixedRadius
       width=1
       color {colorUseRule=UseColor_Specified, specifiedColor=<rgba 0.7243 0.7243 0.7243 1>}
       cornerRadii=<4 4 4 4>
      }
      tintColor {colorUseRule=UseColor_Specified, specifiedColor=<rgba 1 1 1 0>}
      uVRegion {bIsValid=false, max=<0 0>, min=<0 0>}
     }
     pressedForeground {colorUseRule=UseColor_Specified, specifiedColor=<rgba 0 1 1 1>}
     pressedPadding=<ltrb 4 7 4 1.5>
     pressedSlateSound {resourceObject=null}
    }
   }
   slot {class=OverlaySlot, properties {horizontalAlignment=HAlign_Center, verticalAlignment=VAlign_Center}}
  }
  SizeBox_0 {
   widgetClassPath=&SizeBox
   parent=MobileButtonConfirm
   widget=SizeBox_0
   properties {bOverride_HeightOverride=true, bOverride_WidthOverride=true, heightOverride=140, renderOpacity=0, widthOverride=140}
   slot {class=ButtonSlot, properties {padding=<ltrb 0 0 0 0>}}
  }
  Overlay_1 {widgetClassPath=&Overlay, parent=SizeBox_0, widget=Overlay_1, slot {class=SizeBoxSlot}}
  CircleBG {
   widgetClassPath=&Image
   bIsVariable=true
   parent=Overlay_1
   widget=CircleBG
   properties {brush {
     bIsDynamicallyLoaded=false
     drawAs=Image
     imageType=NoImage
     mirroring=NoMirror
     resourceName=None
     resourceObject=/FortUILibrary/Foundation/Art_ToolKit/Tools/MI_UI_Circle_ConversationBacchusConfirm
     tiling=NoTile
     imageSize=<32 32>
     margin=<ltrb 0 0 0 0>
     outlineSettings {
      bUseBrushTransparency=false
      roundingType=HalfHeightRadius
      width=0
      color {colorUseRule=UseColor_Specified, specifiedColor=<rgba 0 0 0 0>}
      cornerRadii=<0 0 0 0>
     }
     tintColor {colorUseRule=UseColor_Specified, specifiedColor=<rgba 1 1 1 1>}
     uVRegion {bIsValid=false, max=<0 0>, min=<0 0>}
    }}
   slot {class=OverlaySlot, properties {horizontalAlignment=HAlign_Fill, verticalAlignment=VAlign_Fill}}
  }
  Confirmation_Icon {
   widgetClassPath=&Image
   bIsVariable=true
   parent=Overlay_1
   widget=Confirmation_Icon
   properties {brush {
     bIsDynamicallyLoaded=false
     drawAs=Image
     imageType=NoImage
     mirroring=NoMirror
     resourceName=None
     resourceObject=/Game/Athena/HUD/Art/BacchusIcons/Confirm_Normal_XL
     tiling=NoTile
     imageSize=<95 85>
     margin=<ltrb 0 0 0 0>
     outlineSettings {
      bUseBrushTransparency=false
      roundingType=HalfHeightRadius
      width=0
      color {colorUseRule=UseColor_Specified, specifiedColor=<rgba 0 0 0 0>}
      cornerRadii=<0 0 0 0>
     }
     tintColor {colorUseRule=UseColor_Specified, specifiedColor=<rgba 0.1033 0.6302 0.1096 1>}
     uVRegion {bIsValid=false, max=<0 0>, min=<0 0>}
    }}
   slot {class=OverlaySlot, properties {horizontalAlignment=HAlign_Center, verticalAlignment=VAlign_Center}}
  }
  MobileButtonTouchInformation {
   widgetClassPath=&Button
   bInherited=true
   parent=Overlay_WheelContainer
   widget=MobileButtonTouchInformation
   properties {widgetStyle {
     clickedSlateSound {resourceObject=null}
     disabled {
      bIsDynamicallyLoaded=false
      drawAs=NoDrawType
      imageType=NoImage
      mirroring=NoMirror
      resourceName=None
      resourceObject=null
      tiling=NoTile
      imageSize=<64 64>
      margin=<ltrb 0 0 0 0>
      outlineSettings {
       bUseBrushTransparency=false
       roundingType=HalfHeightRadius
       width=0
       color {colorUseRule=UseColor_Specified, specifiedColor=<rgba 0 0 0 0>}
       cornerRadii=<0 0 0 0>
      }
      tintColor {colorUseRule=UseColor_Specified, specifiedColor=<rgba 0.5 0.5 0.5 1>}
      uVRegion {bIsValid=false, max=<0 0>, min=<0 0>}
     }
     disabledForeground {colorUseRule=UseColor_Specified, specifiedColor=<rgba 0.5271 0.5271 0.5271 1>}
     hovered {
      bIsDynamicallyLoaded=false
      drawAs=Image
      imageType=NoImage
      mirroring=NoMirror
      resourceName=None
      resourceObject=null
      tiling=NoTile
      imageSize=<128 128>
      margin=<ltrb 0 0 0 0>
      outlineSettings {
       bUseBrushTransparency=true
       roundingType=FixedRadius
       width=1
       color {colorUseRule=UseColor_Specified, specifiedColor=<rgba 0.7243 0.7243 0.7243 1>}
       cornerRadii=<4 4 4 4>
      }
      tintColor {colorUseRule=UseColor_Specified, specifiedColor=<rgba 1 1 1 0>}
      uVRegion {bIsValid=false, max=<0 0>, min=<0 0>}
     }
     hoveredForeground {colorUseRule=UseColor_Specified, specifiedColor=<rgba 0 1 1 1>}
     hoveredSlateSound {resourceObject=null}
     normal {
      bIsDynamicallyLoaded=false
      drawAs=NoDrawType
      imageType=NoImage
      mirroring=NoMirror
      resourceName=None
      resourceObject=null
      tiling=NoTile
      imageSize=<128 128>
      margin=<ltrb 0 0 0 0>
      outlineSettings {
       bUseBrushTransparency=true
       roundingType=FixedRadius
       width=1
       color {colorUseRule=UseColor_Specified, specifiedColor=<rgba 0.6951 0.6951 0.6951 1>}
       cornerRadii=<4 4 4 4>
      }
      tintColor {colorUseRule=UseColor_Specified, specifiedColor=<rgba 1 1 1 0>}
      uVRegion {bIsValid=false, max=<0 0>, min=<0 0>}
     }
     normalForeground {colorUseRule=UseColor_Specified, specifiedColor=<rgba 0 1 1 1>}
     normalPadding=<ltrb 12 1.5 12 1.5>
     pressed {
      bIsDynamicallyLoaded=false
      drawAs=NoDrawType
      imageType=NoImage
      mirroring=NoMirror
      resourceName=None
      resourceObject=null
      tiling=NoTile
      imageSize=<128 128>
      margin=<ltrb 0 0 0 0>
      outlineSettings {
       bUseBrushTransparency=true
       roundingType=FixedRadius
       width=1
       color {colorUseRule=UseColor_Specified, specifiedColor=<rgba 0.7243 0.7243 0.7243 1>}
       cornerRadii=<4 4 4 4>
      }
      tintColor {colorUseRule=UseColor_Specified, specifiedColor=<rgba 1 1 1 0>}
      uVRegion {bIsValid=false, max=<0 0>, min=<0 0>}
     }
     pressedForeground {colorUseRule=UseColor_Specified, specifiedColor=<rgba 0 1 1 1>}
     pressedPadding=<ltrb 12 1.5 12 1.5>
     pressedSlateSound {resourceObject=null}
    }}
   slot {class=OverlaySlot, properties {horizontalAlignment=HAlign_Center, verticalAlignment=VAlign_Center, padding=<ltrb 150 0 0 0>}}
   bIsVariable=~
  }
  SizeBox_1 {
   widgetClassPath=&SizeBox
   parent=MobileButtonTouchInformation
   widget=SizeBox_1
   properties {bOverride_HeightOverride=true, bOverride_WidthOverride=true, heightOverride=110, visibility=Collapsed, widthOverride=110}
   slot {class=ButtonSlot, properties {padding=<ltrb 0 0 0 0>}}
  }
  Overlay_81 {
   widgetClassPath=&Overlay
   parent=SizeBox_1
   widget=Overlay_81
   properties {renderTransform {angle=-90, scale=<-1 1>, shear=<0 0>, translation=<0 0>}}
   slot {class=SizeBoxSlot}
  }
  SectionBG {
   widgetClassPath=&Image
   parent=Overlay_81
   widget=SectionBG
   properties {
    brush {
     bIsDynamicallyLoaded=false
     drawAs=Image
     imageType=NoImage
     mirroring=NoMirror
     resourceName=None
     resourceObject=/Game/UI/Items/Materials/MI-BacchusNCPAnim
     tiling=NoTile
     imageSize=<32 32>
     margin=<ltrb 0 0 0 0>
     outlineSettings {
      bUseBrushTransparency=false
      roundingType=HalfHeightRadius
      width=0
      color {colorUseRule=UseColor_Specified, specifiedColor=<rgba 0 0 0 0>}
      cornerRadii=<0 0 0 0>
     }
     tintColor {colorUseRule=UseColor_Specified, specifiedColor=<rgba 1 1 1 1>}
     uVRegion {bIsValid=false, max=<0 0>, min=<0 0>}
    }
    colorAndOpacity=<rgba 0.0273 0.0273 0.0273 0.6>
   }
   slot {class=OverlaySlot, properties {horizontalAlignment=HAlign_Fill, verticalAlignment=VAlign_Fill}}
  }
  Section {
   widgetClassPath=&Image
   parent=Overlay_81
   widget=Section
   properties {
    brush {
     bIsDynamicallyLoaded=false
     drawAs=Image
     imageType=NoImage
     mirroring=NoMirror
     resourceName=None
     resourceObject=/Game/UI/Items/Materials/MI-BacchusNCPAnimSec
     tiling=NoTile
     imageSize=<32 32>
     margin=<ltrb 0 0 0 0>
     outlineSettings {
      bUseBrushTransparency=false
      roundingType=HalfHeightRadius
      width=0
      color {colorUseRule=UseColor_Specified, specifiedColor=<rgba 0 0 0 0>}
      cornerRadii=<0 0 0 0>
     }
     tintColor {colorUseRule=UseColor_Specified, specifiedColor=<rgba 1 1 1 1>}
     uVRegion {bIsValid=false, max=<0 0>, min=<0 0>}
    }
    colorAndOpacity=<rgba 0.0273 0.0273 0.0273 0.8>
   }
   slot {class=OverlaySlot, properties {horizontalAlignment=HAlign_Fill, verticalAlignment=VAlign_Fill}}
  }
  TouchIconClue {
   widgetClassPath=&Image
   parent=Overlay_81
   widget=TouchIconClue
   properties {
    brush {
     bIsDynamicallyLoaded=false
     drawAs=Image
     imageType=NoImage
     mirroring=NoMirror
     resourceName=None
     resourceObject=/Game/UI/Icons/Icon-Settings/T-Icon-TouchMotion-64
     tiling=NoTile
     imageSize=<50 50>
     margin=<ltrb 0 0 0 0>
     outlineSettings {
      bUseBrushTransparency=false
      roundingType=HalfHeightRadius
      width=0
      color {colorUseRule=UseColor_Specified, specifiedColor=<rgba 0 0 0 0>}
      cornerRadii=<0 0 0 0>
     }
     tintColor {colorUseRule=UseColor_Specified, specifiedColor=<rgba 1 1 1 1>}
     uVRegion {bIsValid=false, max=<0 0>, min=<0 0>}
    }
    renderTransform {angle=20, scale=<1 1>, shear=<0 0>, translation=<-9 9>}
   }
   slot {class=OverlaySlot, properties {horizontalAlignment=HAlign_Center, verticalAlignment=VAlign_Center, padding=<ltrb 4 0 0 20>}}
  }
  Overlay_Portrait {
   widgetClassPath=&Overlay
   bIsVariable=true
   parent=Overlay_WheelContainer
   widget=Overlay_Portrait
   properties {visibility=Hidden}
   slot {class=OverlaySlot, properties {horizontalAlignment=HAlign_Center, verticalAlignment=VAlign_Center}}
  }
  Image_PortraitBackground {
   widgetClassPath=&Image
   parent=Overlay_Portrait
   widget=Image_PortraitBackground
   properties {
    brush {
     bIsDynamicallyLoaded=false
     drawAs=Image
     imageType=FullColor
     mirroring=NoMirror
     resourceName=None
     resourceObject=/FortniteConversationUI/Conversation/Materials/M_ConfirmCircle
     tiling=NoTile
     imageSize=<140 140>
     margin=<ltrb 0 0 0 0>
     outlineSettings {
      bUseBrushTransparency=false
      roundingType=HalfHeightRadius
      width=0
      color {colorUseRule=UseColor_Specified, specifiedColor=<rgba 0 0 0 0>}
      cornerRadii=<0 0 0 0>
     }
     tintColor {colorUseRule=UseColor_Specified, specifiedColor=<rgba 1 1 1 1>}
     uVRegion {bIsValid=false, max=<0 0>, min=<0 0>}
    }
    colorAndOpacity=<rgba 0 0 0 0.4>
   }
   slot {class=OverlaySlot, properties {horizontalAlignment=HAlign_Center, verticalAlignment=VAlign_Center}}
  }
  Image_NPCPortrait {
   widgetClassPath=&Image
   bIsVariable=true
   parent=Overlay_Portrait
   widget=Image_NPCPortrait
   properties {brush {
     bIsDynamicallyLoaded=false
     drawAs=Image
     imageType=FullColor
     mirroring=NoMirror
     resourceName=None
     resourceObject=/FortniteConversationUI/Conversation/Materials/M_NPCFaceConversation
     tiling=NoTile
     imageSize=<130 130>
     margin=<ltrb 0 0 0 0>
     outlineSettings {
      bUseBrushTransparency=false
      roundingType=HalfHeightRadius
      width=0
      color {colorUseRule=UseColor_Specified, specifiedColor=<rgba 0 0 0 0>}
      cornerRadii=<0 0 0 0>
     }
     tintColor {colorUseRule=UseColor_Specified, specifiedColor=<rgba 1 1 1 1>}
     uVRegion {bIsValid=false, max=<0 0>, min=<0 0>}
    }}
   slot {class=OverlaySlot, properties {horizontalAlignment=HAlign_Center, verticalAlignment=VAlign_Center}}
  }
  VisibilityBorder_Keybind_NonTouch {
   widgetClassPath=&CommonHardwareVisibilityBorder
   parent=Overlay_WheelContainer
   widget=VisibilityBorder_Keybind_NonTouch
   properties {visibilityQuery {
     autoDescription=" NONE( Input.Touch )"
     tokenStreamVersion=0
     userDescription=""
     queryTokenStream [0,1,3,1,0]
     tagDictionary [{tagName=Input.Touch}]
    }}
   slot {class=OverlaySlot, properties {horizontalAlignment=HAlign_Center, verticalAlignment=VAlign_Center}}
  }
  KeybindBorder {
   widgetClassPath=&Border
   bIsVariable=true
   parent=VisibilityBorder_Keybind_NonTouch
   widget=KeybindBorder
   properties {
    background {
     bIsDynamicallyLoaded=false
     drawAs=Image
     imageType=NoImage
     mirroring=NoMirror
     resourceName=None
     resourceObject=null
     tiling=NoTile
     imageSize=<24 24>
     margin=<ltrb 0 0 0 0>
     outlineSettings {
      bUseBrushTransparency=false
      roundingType=HalfHeightRadius
      width=0
      color {colorUseRule=UseColor_Specified, specifiedColor=<rgba 0 0 0 0>}
      cornerRadii=<0 0 0 0>
     }
     tintColor {colorUseRule=UseColor_Specified, specifiedColor=<rgba 1 1 1 1>}
     uVRegion {bIsValid=false, max=<0 0>, min=<0 0>}
    }
    brushColor=<rgba 1 1 1 0>
    padding=<ltrb 0 0 0 0>
    renderTransform {angle=0, scale=<1 1>, shear=<0 0>, translation=<0 -59>}
   }
   slot {class=BorderSlot, properties {padding=<ltrb 0 0 0 0>}}
  }
  Keybind_Confirm {
   widgetClassPath=/FortUILibrary/Input/KeybindWidgetDark
   bInherited=true
   parent=KeybindBorder
   widget=Keybind_Confirm
   properties {boundAction=Use}
   slot {class=BorderSlot, properties {padding=<ltrb 0 0 0 0>}}
  }
  BriefList {
   widgetClassPath=/FortniteConversationUI/Conversation/MobileConversationComponents/BriefRadialList
   bIsVariable=true
   parent=Overlay_0
   widget=BriefList
   properties {
    clockwise=true
    distriuteItemsEvenly=false
    entryClass=/FortniteConversationUI/Conversation/MobileConversationComponents/FortConversationBriefContainer.FortConversationBriefContainer_C
    maxAngleLenght=160.1517
    numDesignerPreview=5
    radiusDistance=300
    startAngle=-11.5909
    harmonicDistortion {amplitude=12, frequency=2, rotataion=0}
    radialTransform {axisScaleRotation=0, spiralFactor=0, axisScale=<1 1.1162>}
   }
   slot {class=OverlaySlot, properties {horizontalAlignment=HAlign_Right, verticalAlignment=VAlign_Center, padding=<ltrb 0 0 100 0>}}
  }
 ]
}
