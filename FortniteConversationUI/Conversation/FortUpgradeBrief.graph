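# Widget Blueprint for the weapon upgrade brief in the conversation UI, switching between a can-upgrade panel and a cannot-upgrade message.
# Events from the C++ parent fill the icon, shadow and rarity names, tinting old and new text and a triangle effect from rarity data.
# The failure state always shows "Equip an upgradable weapon", ignoring the reported cause.
# Selection visuals are forwarded to the inherited transaction strip.
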WidgetBlueprint /FortniteConversationUI/Conversation/FortUpgradeBrief
graphs [EventGraph {nodes [
   n1 = Event OnTransactionInfoReceived!() -> (OutputDelegate, then, Item:Fort Item Object Reference, TransactionInfo:"Fort Service Transaction Info Structure (by ref)") {is=title K2Node_Event}
   > SetActiveWidget(
    self=Get StateVisibilitySwitcher() -> (StateVisibilitySwitcher:Common Visibility Switcher Object Reference)
    Widget=Get CanUpgrade() -> (CanUpgrade:Overlay Object Reference)
   )
   > SetBrushFromLazyDisplayAsset!(
    execute
    self:Common Lazy Image Object Reference=[Get Icon() -> (Icon:Fort Lazy Image Object Reference), Get IconShadow() -> (IconShadow:Fort Lazy Image Object Reference)]
    LazyObject:"Soft Object Reference (by ref)"=GetSmallPreviewImage!(self:Fort Item Object Reference=@n1.Item) -> (ReturnValue:Texture 2D Soft Object Reference)
    bMatchTextureSize:bool=false
   ) -> (then)
   > n2 = GetRarityData!(execute, Rarity:EFortRarity Enum=GetRarity!(self:Fort Item Object Reference=@n1.Item) -> (ReturnValue:EFortRarity Enum=Common)) -> (then, ReturnValue:Fort Rarity Item Data Structure)
   n3 = Break Fort Rarity Item Data!(FortRarityItemData:"Fort Rarity Item Data Structure (by ref)"=@n2) -> (
    Name:Text
    Color1:Linear Color Structure
    Color2:Linear Color Structure
    Color3:Linear Color Structure
    Color4:Linear Color Structure
    Color5:Linear Color Structure
    Radius:float=~
    Falloff:float=~
    Brightness:float=~
    Roughness:float=~
    Glow:float=~
   ) {is=title K2Node_BreakStruct}
   n4 = Get Text_WeaponRarity_New() -> (Text_WeaponRarity_New:Common Text Object Reference)
   > "SetText (Text)"(self=@n4, InText=@n3)
   > SetColorAndOpacity(self=@n4, InColorAndOpacity=Make Slate Color(SpecifiedColor=@n3.Color1))
   > SetColorAndOpacity(
    self:Image Object Reference=Get ImageTriangleL() -> (ImageTriangleL:Image Object Reference)
    InColorAndOpacity:Linear Color Structure=@n3.Color1
   )
   n5 = Event OnUpgradeInfoReceived!() -> (OutputDelegate, then, CurrentWeapon:Fort Item Object Reference) {is=title K2Node_Event}
   > n6 = GetRarityData(Rarity=GetRarity(self=@n5.CurrentWeapon) -> (ReturnValue=Common))
   n7 = Break Fort Rarity Item Data(FortRarityItemData=@n6) -> (Radius=~, Falloff=~, Brightness=~, Roughness=~, Glow=~)
   n8 = Get Text_WeaponRarity_Old() -> (Text_WeaponRarity_Old:Common Text Object Reference)
   > "SetText (Text)"(self=@n8, InText=@n7)
   > SetColorAndOpacity(self:Text Object Reference=@n8, InColorAndOpacity:Slate Color Structure=Make Slate Color(SpecifiedColor=@n7.Color1))
   Event OnUnableToUpgrade!() -> (OutputDelegate, then, CannotUpgradeCause:EFortWeaponUpgradeInteractionResult Enum=Upgradable) {is=title K2Node_Event}
   > SetActiveWidget(self=Get StateVisibilitySwitcher(), Widget=Get CannotUpGrade() -> (CannotUpGrade:Overlay Object Reference))
   > SetErrorMessage!(
    execute
    self:Brief Error Message New Object Reference=Get BriefErrorMessage() -> (BriefErrorMessage:Brief Error Message New Object Reference)
    ErrorText:Text="LOCTEXT(\"Equip an upgradable weapon\")"
   ) -> (then)
   n9 = Event SetSelectedVisuals!() -> (OutputDelegate, then, IsSelected:bool) {is=title K2Node_Event}
   > n10 = Cast To FortTransactionStripNew!(execute, Object=Get TransactionStrip_Main() -> (TransactionStrip_Main:Fort Transaction Strip Object Reference)) -> (then, CastFailed, AsFort Transaction Strip New:Fort Transaction Strip New Object Reference) {is=title K2Node_DynamicCast}
   > SetSelectedVisuals!(execute, self:Fort Transaction Strip New Object Reference=@n10, IsSelected:bool=@n9.IsSelected) -> (then)
   Event Construct()
  ]}]
variables []
widgets {
 info {inheritedWidgetCount=1, parentClass=/Script/FortniteConversationUI.FortUpgradeBrief, rootWidgetClass=&SizeBox, widgetCount=26}
 widgets [
  SizeBox_36 {widgetClassPath=&SizeBox, widget=SizeBox_36, properties {heightOverride=150, widthOverride=518.7535}}
  StateVisibilitySwitcher {
   widgetClassPath=&CommonVisibilitySwitcher
   bIsVariable=true
   parent=SizeBox_36
   widget=StateVisibilitySwitcher
   slot {class=SizeBoxSlot}
  }
  CanUpgrade {
   widgetClassPath=&Overlay
   bIsVariable=true
   parent=StateVisibilitySwitcher
   widget=CanUpgrade
   slot {class=CommonVisibilitySwitcherSlot}
  }
  VerticalBox_2 {widgetClassPath=&VerticalBox, parent=CanUpgrade, widget=VerticalBox_2, slot {class=OverlaySlot}}
  Border_168 {
   widgetClassPath=&Border
   parent=VerticalBox_2
   widget=Border_168
   properties {horizontalAlignment=HAlign_Center, brushColor=<rgba 0 0 0 0.446>, padding=<ltrb 0 0 0 0>}
   slot {class=VerticalBoxSlot}
  }
  Overlay_1 {
   widgetClassPath=&Overlay
   parent=Border_168
   widget=Overlay_1
   slot {class=BorderSlot, properties {horizontalAlignment=HAlign_Center, padding=<ltrb 0 0 0 0>}}
  }
  ImageTriangleL {
   widgetClassPath=&Image
   bIsVariable=true
   parent=Overlay_1
   widget=ImageTriangleL
   properties {
    visibility=HitTestInvisible
    brush {
     bIsDynamicallyLoaded=false
     drawAs=Image
     imageType=NoImage
     mirroring=NoMirror
     resourceName=None
     resourceObject=/Game/UI/HUD/Materials/M_FX_TriangleScroll-Burst_Fast_Vertical
     tiling=NoTile
     imageSize=<163.2664 257.5856>
     margin=<ltrb 0 0 0 0>
     outlineSettings {
      bUseBrushTransparency=false
      roundingType=HalfHeightRadius
      width=0
      color {colorUseRule=UseColor_Specified, specifiedColor=<rgba 0 0 0 0>}
      cornerRadii=<0 0 0 0>
     }
     tintColor {colorUseRule=UseColor_Specified, specifiedColor=<rgba 1 1 1 1>}
     uVRegion {bIsValid=false, max=<0 0>, min=<0 0>}
    }
    colorAndOpacity=<rgba 0.7018 0 1 1>
    renderTransform {angle=0, scale=<1 1>, shear=<0 0>, translation=<-20 0>}
   }
   slot {class=OverlaySlot, properties {verticalAlignment=VAlign_Bottom, padding=<ltrb 0 -300 -200 0>}}
  }
  Overlay_0 {widgetClassPath=&Overlay, parent=Overlay_1, widget=Overlay_0, slot {class=OverlaySlot, properties {padding=<ltrb 8 8 8 8>}}}
  HorizontalBox_2 {
   widgetClassPath=&HorizontalBox
   parent=Overlay_0
   widget=HorizontalBox_2
   slot {class=OverlaySlot, properties {horizontalAlignment=HAlign_Center}}
  }
  Overlay_ItemIcon {
   widgetClassPath=&Overlay
   parent=HorizontalBox_2
   widget=Overlay_ItemIcon
   properties {renderTransform {angle=0, scale=<1 1>, shear=<0 0>, translation=<-12 0>}}
   slot {class=HorizontalBoxSlot, properties {padding=<ltrb 0 -64 0 -20>}}
  }
  IconShadow {
   widgetClassPath=&FortLazyImage
   bIsVariable=true
   parent=Overlay_ItemIcon
   widget=IconShadow
   properties {
    brush {
     bIsDynamicallyLoaded=false
     drawAs=Image
     imageType=NoImage
     mirroring=NoMirror
     resourceName=None
     resourceObject=/Game/UI/Foundation/Textures/Icons/Weapons/Items/T-Icon-Weapons-SK-AutoHigh-Suppressed
     tiling=NoTile
     imageSize=<160 160>
     margin=<ltrb 0 0 0 0>
     outlineSettings {
      bUseBrushTransparency=false
      roundingType=HalfHeightRadius
      width=0
      color {colorUseRule=UseColor_Specified, specifiedColor=<rgba 0 0 0 0>}
      cornerRadii=<0 0 0 0>
     }
     tintColor {colorUseRule=UseColor_Specified, specifiedColor=<rgba 1 1 1 1>}
     uVRegion {bIsValid=false, max=<0 0>, min=<0 0>}
    }
    colorAndOpacity=<rgba 0.0052 0.0052 0.0056 0.3>
    renderTransform {angle=0, scale=<1 1>, shear=<0 0>, translation=<6 13>}
   }
   slot {class=OverlaySlot}
  }
  Icon {
   widgetClassPath=&FortLazyImage
   bIsVariable=true
   parent=Overlay_ItemIcon
   widget=Icon
   properties {brush {
     bIsDynamicallyLoaded=false
     drawAs=Image
     imageType=NoImage
     mirroring=NoMirror
     resourceName=None
     resourceObject=&T-Icon-Weapons-SK-AutoHigh-Suppressed
     tiling=NoTile
     imageSize=<160 160>
     margin=<ltrb 0 0 0 0>
     outlineSettings {
      bUseBrushTransparency=false
      roundingType=HalfHeightRadius
      width=0
      color {colorUseRule=UseColor_Specified, specifiedColor=<rgba 0 0 0 0>}
      cornerRadii=<0 0 0 0>
     }
     tintColor {colorUseRule=UseColor_Specified, specifiedColor=<rgba 1 1 1 1>}
     uVRegion {bIsValid=false, max=<0 0>, min=<0 0>}
    }}
   slot {class=OverlaySlot}
  }
  Image_1 {
   widgetClassPath=&Image
   bIsVariable=true
   parent=Overlay_ItemIcon
   widget=Image_1
   properties {
    brush {
     bIsDynamicallyLoaded=false
     drawAs=Image
     imageType=NoImage
     mirroring=NoMirror
     resourceName=None
     resourceObject=/FortniteConversationUI/Conversation/Materials/T_UI_NPC_Upgrade
     tiling=NoTile
     imageSize=<64 64>
     margin=<ltrb 0 0 0 0>
     outlineSettings {
      bUseBrushTransparency=false
      roundingType=HalfHeightRadius
      width=0
      color {colorUseRule=UseColor_Specified, specifiedColor=<rgba 0 0 0 0>}
      cornerRadii=<0 0 0 0>
     }
     tintColor {colorUseRule=UseColor_Specified, specifiedColor=<rgba 1 1 1 1>}
     uVRegion {bIsValid=false, max=<0 0>, min=<0 0>}
    }
    renderTransform {angle=0, scale=<1 1>, shear=<0 0>, translation=<-33 -12>}
   }
   slot {class=OverlaySlot, properties {horizontalAlignment=HAlign_Right, verticalAlignment=VAlign_Bottom}}
  }
  VerticalBox_4 {
   widgetClassPath=&VerticalBox
   parent=HorizontalBox_2
   widget=VerticalBox_4
   properties {renderTransform {angle=0, scale=<1 1>, shear=<-8 0>, translation=<-20 0>}}
   slot {class=HorizontalBoxSlot, properties {verticalAlignment=VAlign_Center}}
  }
  SizeBox_1 {
   widgetClassPath=&SizeBox
   parent=VerticalBox_4
   widget=SizeBox_1
   properties {bOverride_MinDesiredWidth=true, minDesiredWidth=384}
   slot {class=VerticalBoxSlot}
  }
  HorizontalBox_3 {widgetClassPath=&HorizontalBox, parent=SizeBox_1, widget=HorizontalBox_3, slot {class=SizeBoxSlot}}
  Text_WeaponRarity_Old {
   widgetClassPath=&CommonTextBlock
   bIsVariable=true
   parent=HorizontalBox_3
   widget=Text_WeaponRarity_Old
   properties {
    text="{OldRarity}"
    textTransformPolicy=ToUpper
    font {
     bForceMonospaced=false
     bMaterialIsStencil=false
     fontMaterial=null
     fontObject=&Burbank
     letterSpacing=0
     monospacedWidth=1
     size=18
     skewAmount=0
     typefaceFontName=BurbankRegular-Black
     outlineSettings {
      bApplyOutlineToDropShadows=false
      bMiteredCorners=false
      bSeparateFillAlpha=false
      outlineMaterial=null
      outlineSize=0
      outlineColor=<rgba 0 0 0 1>
     }
    }
    renderTransform {angle=0, scale=<1 1>, shear=<-8 0>, translation=<0 0>}
    shadowOffset=<0 0>
   }
   slot {class=HorizontalBoxSlot}
  }
  Image {
   widgetClassPath=&Image
   bIsVariable=true
   parent=HorizontalBox_3
   widget=Image
   properties {
    brush {
     bIsDynamicallyLoaded=false
     drawAs=Image
     imageType=NoImage
     mirroring=NoMirror
     resourceName=None
     resourceObject=/Game/Athena/HUD/Materials/M_UI_FocusBorderPulse_ArrowOnly
     tiling=NoTile
     imageSize=<128 128>
     margin=<ltrb 0 0 0 0>
     outlineSettings {
      bUseBrushTransparency=false
      roundingType=HalfHeightRadius
      width=0
      color {colorUseRule=UseColor_Specified, specifiedColor=<rgba 0 0 0 0>}
      cornerRadii=<0 0 0 0>
     }
     tintColor {colorUseRule=UseColor_Specified, specifiedColor=<rgba 1 1 1 1>}
     uVRegion {bIsValid=false, max=<0 0>, min=<0 0>}
    }
    renderTransform {angle=-90, scale=<1 1>, shear=<0 0>, translation=<-47 0>}
   }
   slot {class=HorizontalBoxSlot, properties {horizontalAlignment=HAlign_Center, verticalAlignment=VAlign_Center, padding=<ltrb -40 -64 -40 -64>}}
  }
  Text_WeaponRarity_New {
   widgetClassPath=&CommonTextBlock
   bIsVariable=true
   parent=HorizontalBox_3
   widget=Text_WeaponRarity_New
   properties {
    text="{NewRarity}"
    textTransformPolicy=ToUpper
    font {
     bForceMonospaced=false
     bMaterialIsStencil=false
     fontMaterial=null
     fontObject=&Burbank
     letterSpacing=0
     monospacedWidth=1
     size=18
     skewAmount=0
     typefaceFontName=BurbankRegular-Black
     outlineSettings {
      bApplyOutlineToDropShadows=false
      bMiteredCorners=false
      bSeparateFillAlpha=false
      outlineMaterial=null
      outlineSize=0
      outlineColor=<rgba 0 0 0 1>
     }
    }
    renderTransform {angle=0, scale=<1 1>, shear=<-8 0>, translation=<0 0>}
    shadowOffset=<0 0>
   }
   slot {class=HorizontalBoxSlot}
  }
  SizeBox_0 {
   widgetClassPath=&SizeBox
   parent=VerticalBox_4
   widget=SizeBox_0
   properties {heightOverride=120}
   slot {class=VerticalBoxSlot, properties {verticalAlignment=VAlign_Top}}
  }
  MoneyReward {widgetClassPath=&Overlay, bIsVariable=true, parent=SizeBox_0, widget=MoneyReward, slot {class=SizeBoxSlot}}
  SizeBox_38 {widgetClassPath=&SizeBox, parent=MoneyReward, widget=SizeBox_38, properties {widthOverride=150}, slot {class=OverlaySlot}}
  HorizontalBox_0 {widgetClassPath=&HorizontalBox, parent=SizeBox_38, widget=HorizontalBox_0, slot {class=SizeBoxSlot}}
  TransactionStrip_Main {
   widgetClassPath=&/FortTransactionStripNew
   bInherited=true
   bIsVariable=true
   parent=VerticalBox_2
   widget=TransactionStrip_Main
   properties {transactionString=Upgrade}
   slot {class=VerticalBoxSlot}
  }
  CannotUpGrade {
   widgetClassPath=&Overlay
   bIsVariable=true
   parent=StateVisibilitySwitcher
   widget=CannotUpGrade
   slot {class=CommonVisibilitySwitcherSlot, properties {verticalAlignment=VAlign_Bottom}}
  }
  BriefErrorMessage {
   widgetClassPath=&/BriefErrorMessageNew
   bIsVariable=true
   parent=CannotUpGrade
   widget=BriefErrorMessage
   properties {fontSize=20}
   slot {class=OverlaySlot}
  }
 ]
}
