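# M_ConversationSection is a translucent UI material for the conversation interface that draws a segmented circular ring section with a shoulder line.
# Rotated UVs, a circle SDF outline and a radial-segments function supply the ring, segment and half-split divider masks through named reroutes.
# The shoulder line comes from subtracting two smoothstepped SDF bands, scaled by an opacity parameter.
# Section and shoulder colors and alphas are blended by a mask, with further masks shaping the opacity.
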Material /FortniteConversationUI/Conversation/Materials/M_ConversationSection
graph {
 properties {bAutomaticallySetUsageInEditor=false, blendMode=BLEND_Translucent, materialDomain=MD_UI, usedShadingModels=Default Lit}
 nodes [
  n1 = VectorParameter() {parameterName=SectionColor, defaultValue=<rgba 0 0 0 0.4>}
  n2 = VectorParameter() {parameterName=ShoulderColor, defaultValue=<rgba 1 1 1 0.4>}
  n3 = NamedRerouteUsage()
  LinearInterpolate(A=@n1.RGB, B=@n2.RGB, Alpha=@n3)
  n4 = LinearInterpolate(A=@n1.A, B=@n2.A, Alpha=@n3)
  n5 = NamedRerouteUsage()
  Multiply(A=Subtract(A=Multiply(A=@n4, B=@n5), B=NamedRerouteUsage()), B=NamedRerouteUsage())
  NamedRerouteDeclaration(None=MaterialFunctionCall(UVs=TextureCoordinate(), "Rotation Angle (0-1)"=ScalarParameter() {parameterName=RotateAngle}) {materialFunction=&CustomRotator}.Rotated Values) {name=TexCoord, nodeColor=<rgba 0 0.9297 1 1>}
  n6 = ScalarParameter() {defaultValue=0.95, parameterName=Spacing}
  n7 = MaterialFunctionCall!(
   UV=NamedRerouteUsage()
   Segment Count=Multiply(A=ScalarParameter() {defaultValue=5, parameterName=Segment Count}) {constB=2}
   Segment Spacing=@n6
   Fixed Spacing
  ) {materialFunction=/FortUILibrary/MaterialFunctions/Basics/MF_UI_RadialSegments}
  NamedRerouteDeclaration(None=SmoothStep(Max=Constant() {r=0.1}, Value=Subtract(A=@n7.Segmented Lines Gradient) {constB=0})) {name=Segments, nodeColor=<rgba 1 0 0.7969 1>}
  n8 = Constant() {r=0.01}
  n9 = MaterialFunctionCall(
   UVs=NamedRerouteUsage()
   Size=ScalarParameter() {defaultValue=0.9, parameterName=Size}
   OutlineThickness=Multiply(A=ScalarParameter() {defaultValue=0.4, parameterName=Thickness}) {constB=0.5}
   OutlineAlignment=Constant() {r=-1}
  ) {materialFunction=&MF_UI_Circle}
  n10 = Saturate(None=@n9.SDF)
  n11 = Constant() {r=1}
  n12 = Subtract(A=@n6, B=ScalarParameter() {parameterName=ShoulderThickness}) {constB=-0.02}
  n13 = Add(A=@n9.SDF, B=Multiply(A=Subtract(A=@n11, B=@n12) {constB=-0.02}) {constB=0.5})
  n14 = Subtract(A=OneMinus(None=SmoothStep(Max=@n8, Value=@n10) {constMax=0.05}), B=OneMinus(None=SmoothStep(Max=@n8, Value=@n13) {constMax=0.05}))
  n15 = ScalarParameter() {defaultValue=1, parameterName=ShoulderOutlineOpacity}
  NamedRerouteDeclaration(None=Saturate(None=Multiply(A=@n14, B=@n15))) {name=ShoulderLineOpc, nodeColor=<rgba 0 1 0.8828 1>}
  NamedRerouteDeclaration(None=@n9.Outline) {name=Outline, nodeColor=<rgba 1 0.5625 0 1>}
  n16 = ComponentMask(None=NamedRerouteUsage()) {r=true}
  NamedRerouteDeclaration(None=OneMinus(None=Round(None=@n16))) {name=Divider, nodeColor=<rgba 1 0.5625 0 1>}
 ]
}
