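# Translucent UI master material, named for blueprint text, that blends from a hatched, noisy look to a filled one.
# Rotated sine stripes, distorted by scaled ice noise, combine with that noise into lifted hatch shading for alpha.
# AnimatetoFill blends that alpha toward FillAlpha, or a vertical gradient via UseFadedText, and shifts Color to FillColor.
# A few UV and power nodes appear unused.
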Material /FortniteConversationUI/Conversation/Materials/M_Fund_UI_BlueprintText_Master
graph {
 properties {blendMode=BLEND_Translucent, materialDomain=MD_UI, pixelDepthOffsetMode=PDOM_Legacy, usedShadingModels=Default Lit}
 nodes [
  n1 = ScalarParameter() {defaultValue=0.25, parameterName=Noise_Lift}
  n2 = TextureCoordinate()
  n3 = TextureSample(UVs=MaterialFunctionCall(UVs=@n2, Texture Scale=ScalarParameter() {defaultValue=6, group=Noise, parameterName=Noise_Scale}) {materialFunction=&ScaleUVsByCenter}.UVs) {texture=/Game/Global/Textures/Default/Noise/T_UI_Ice_Noise}
  n4 = ScalarParameter() {defaultValue=3, group=Noise, parameterName=Noise_Intensity}
  n5 = Multiply(A=@n3.R, B=@n4) {constB=16}
  n6 = ScalarParameter() {defaultValue=0.5, group=ShadeLines, parameterName=ShadeLines_Lift}
  n7 = ScalarParameter() {defaultValue=1, parameterName=ShadeLinesOpacity}
  n8 = Add(
   A=MaterialFunctionCall(UVs=@n2, "Rotation Angle (0-1)"=ScalarParameter() {defaultValue=0.075, group=ShadeLines, parameterName=ShadeLines_Rotation}) {materialFunction=&CustomRotator}.Rotated Values
   B=Multiply(A=@n3.R, B=ScalarParameter() {defaultValue=0.0125, group=ShadeLines, parameterName=ShadeLines_Distortion})
  )
  n9 = ScalarParameter() {defaultValue=40, group=ShadeLines, parameterName=ShadeLines_Amount}
  n10 = Power(Base=Sine(None=Multiply(A=@n8, B=@n9)), Exp=ScalarParameter() {defaultValue=3, group=ShadeLines, parameterName=ShadeLines_Thickness})
  n11 = Saturate(None=MaterialFunctionCall(Float2=Saturate(None=@n10)) {materialFunction=&BreakOutFloat2Components}.R)
  n12 = Saturate(None=DotProduct(A=Saturate(None=@n5), B=LinearInterpolate(A=@n6, B=@n7, Alpha=@n11)))
  n13 = ScalarParameter() {defaultValue=1, parameterName=FillAlpha}
  n14 = TextureCoordinate()
  n15 = DotProduct(A=@n13, B=LinearInterpolate(A=ScalarParameter() {defaultValue=0.5, parameterName=FadedBottomAlpha}, Alpha=ComponentMask(None=@n14) {g=true}))
  LinearInterpolate(
   A=Saturate(None=LinearInterpolate(A=@n1, Alpha=@n12))
   B=LinearInterpolate(A=@n13, B=Saturate(None=@n15), Alpha=ScalarParameter() {parameterName=UseFadedText})
   Alpha=ScalarParameter() {parameterName=AnimatetoFill}
  )
  LinearInterpolate(
   A=VectorParameter() {parameterName=Color, defaultValue=<rgba 1 1 1 1>}.RGB
   B=VectorParameter() {parameterName=FillColor, defaultValue=<rgba 0.0123 0.006 0 1>}.RGB
   Alpha=ScalarParameter() {parameterName=AnimatetoFill}
  )
  MaterialFunctionCall() {materialFunction=&GetUserInterfaceUV}
  MaterialFunctionCall(Texture Dimensions=@n2) {materialFunction=&ScreenAlignedUVs}
  Power(Base=@n5, Exp=Multiply(A=@n4) {constB=0.5})
 ]
}
