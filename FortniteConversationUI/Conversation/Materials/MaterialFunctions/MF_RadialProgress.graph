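# Material function for Conversation UI materials that outputs an antialiased radial progress wedge mask, driven by UV, Angle, Progress and Width inputs.
# CustomRotator sets the start angle, then a sine/cosine rotation of the centered UVs creates the second, progress-driven edge.
# Screen-space derivatives scale the SmoothStep edge softness.
# The two edges combine by Min or Max around half progress, with Ceil terms masking small and zero progress.
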MaterialFunction /FortniteConversationUI/Conversation/Materials/MaterialFunctions/MF_RadialProgress
graph {
 properties {previewMaterialDomain=MD_UI}
 nodes [
  n1 = NamedRerouteUsage()
  n2 = DDX(Value=@n1)
  n3 = DDY(Value=@n1)
  n4 = Max(A=DotProduct(A=@n2, B=@n2), B=DotProduct(A=@n3, B=@n3))
  n5 = Multiply(
   A=NamedRerouteDeclaration(None=SquareRoot(None=@n4)) {name=FilterWidth, nodeColor=<rgba 0 1 0.5078 1>}
   B=FunctionInput(Preview=Constant() {r=2}) {bUsePreviewValueAsDefault=true, inputName=Width, inputType=FunctionInput_Scalar, sortPriority=5}
  ) {constB=2}
  n6 = NamedRerouteUsage()
  MaterialExpressionNamedRerouteDeclaration_3 = NamedRerouteDeclaration(None=Subtract(A=MaterialFunctionCall(UVs=@n1, "Rotation Angle (0-1)"=@n6) {materialFunction=&CustomRotator}.Rotated Values) {constB=0.5}) {name=UVs, nodeColor=<rgba 0.8281 0 1 1>}
  n7 = Convert(None=@MaterialExpressionNamedRerouteDeclaration_3) {
   nodeName=Break Float2
   convertInputs [{
     type=Vector2
     expressionInput {
      expression=&self:MaterialExpressionNamedRerouteDeclaration_3
      inputName=None
      mask=0
      maskA=0
      maskB=0
      maskG=0
      maskR=0
      outputIndex=0
     }
    }]
   convertOutputs [{type=Scalar}, {type=Scalar}]
  }
  n8 = Saturate(None=NamedRerouteUsage())
  n9 = Cosine(None=@n8)
  n10 = Sine(None=@n8)
  n11 = Multiply(A=@n10) {constB=-1}
  n12 = Add(A=Multiply(A=@n7, B=AppendVector(A=@n9, B=@n11)), B=Multiply(A=@n7, B=AppendVector(A=@n10, B=@n9)))
  n13 = SmoothStep(
   Min=Multiply(A=@n5) {constB=-1}
   Value=ComponentMask(None=NamedRerouteDeclaration(None=@n12) {name=RotatedUVs, nodeColor=<rgba 1 0.1875 0 1>}) {r=true}
  ) {constMax=0}
  n14 = NamedRerouteUsage()
  n15 = SmoothStep(Min=@n5, Value=ComponentMask(None=@n14) {r=true}) {constMax=0}
  n16 = If(A=NamedRerouteUsage(), "A > B"=Max(A=@n13, B=@n15), "A < B"=Min(A=@n13, B=@n15)) {constB=0.5}
  n17 = NamedRerouteUsage()
  n18 = Ceil(None=Max(A=ComponentMask(None=@n14) {g=true}, B=Add(A=@n17) {constB=-0.25}))
  FunctionOutput(None=Min(A=Multiply(A=@n16, B=@n18), B=Ceil(None=@n17)))
  NamedRerouteDeclaration(None=FunctionInput(Preview=TextureCoordinate()) {bUsePreviewValueAsDefault=true, inputName=UV, inputType=FunctionInput_Vector2}) {name=Coord, nodeColor=<rgba 1 0 0.2109 1>}
  NamedRerouteDeclaration(None=FunctionInput(Preview=Constant() {r=0.5}) {bUsePreviewValueAsDefault=true, inputName=Angle, inputType=FunctionInput_Scalar, sortPriority=1}) {name=Angle, nodeColor=<rgba 0 0.5312 1 1>}
  NamedRerouteDeclaration(None=FunctionInput(Preview=Constant() {r=0.6}) {inputName=Progress, inputType=FunctionInput_Scalar, sortPriority=2}) {name=Progress, nodeColor=<rgba 0.8516 1 0 1>}
 ]
}
