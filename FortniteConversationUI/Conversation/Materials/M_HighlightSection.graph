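# Translucent UI material for the Fortnite conversation UI that draws a highlight arc, a section of a ring, with a soft shadow.
# MF_UI_Circle builds the ring outlines, and MF_RadialProgress on rotated UVs masks them to a section set by Angle and RadialMask.
# The shadow repeats this with offset UVs and its own size, extra thickness, glow and opacity, then is added to the highlight.
# Named reroutes share coordinates, thickness and outlines between branches.
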Material /FortniteConversationUI/Conversation/Materials/M_HighlightSection
graph {
 properties {bAutomaticallySetUsageInEditor=false, blendMode=BLEND_Translucent, materialDomain=MD_UI, usedShadingModels=Default Lit}
 nodes [
  n1 = MaterialFunctionCall(
   UVs=NamedRerouteUsage()
   "Rotation Angle (0-1)"=Add(A=ScalarParameter() {defaultValue=0.25, group=Section, parameterName=Angle}) {constB=0.5}
  ) {materialFunction=&CustomRotator}
  n2 = ScalarParameter() {defaultValue=0.15, group=Section, parameterName=RadialMask}
  n3 = Add(A=Add(A=Constant(), B=Multiply(A=@n2) {constB=0.5})) {constB=0}
  n4 = Multiply(A=Constant() {r=200}, B=ScalarParameter() {defaultValue=0.01, group=Shadow, parameterName=CircleGlow, sortPriority=2})
  n5 = NamedRerouteUsage()
  n6 = Constant()
  n7 = ScalarParameter() {defaultValue=0.15, group=Section, parameterName=RadialMask}
  n8 = Multiply(A=@n7) {constB=0.5}
  n9 = MaterialFunctionCall!(
   UV=NamedRerouteUsage()
   Angle=Add(A=Add(A=@n6, B=@n8)) {constB=0}
   Progress=@n7
   Width=Multiply(A=Constant() {r=200}, B=ScalarParameter() {defaultValue=0.1, group=Shadow, parameterName=ShadowGlow, sortPriority=2})
  ) {materialFunction=/FortniteConversationUI/Conversation/Materials/MaterialFunctions/MF_RadialProgress}
  Add(
   A=Multiply(
    A=NamedRerouteUsage()
    B=MaterialFunctionCall(UV=@n1.Rotated Values, Angle=@n3, Progress=@n2, Width=@n4) {materialFunction=&MF_RadialProgress}.Result
   )
   B=Multiply(A=Multiply(A=@n5, B=@n9.Result), B=ScalarParameter() {defaultValue=0.5, group=Shadow, parameterName=ShadowOpacity, sortPriority=3}) {constB=0.5}
  )
  NamedRerouteDeclaration(None=TextureCoordinate()) {name=TexCoord, nodeColor=<rgba 0 0.9297 1 1>}
  NamedRerouteDeclaration(None=Multiply(A=ScalarParameter() {defaultValue=0.4, group=Section, parameterName=Thickness}) {constB=0.5}) {name=Thickness, nodeColor=<rgba 1 0 0 1>}
  n10 = VectorParameter() {group=Shadow Offset, parameterName=ShadowOffset, sortPriority=1, defaultValue=<rgba -0.05 -0.05 0 0>}
  n11 = Multiply(A=@n10.RGB, B=ScalarParameter() {defaultValue=0.2, group=Shadow Offset, parameterName=OffsetScale, sortPriority=1})
  n12 = MaterialFunctionCall(
   UVs=Add(A=NamedRerouteUsage(), B=ComponentMask(None=@n11) {g=true, r=true})
   "Rotation Angle (0-1)"=Add(A=ScalarParameter() {defaultValue=0.25, group=Section, parameterName=Angle}) {constB=0.5}
  ) {materialFunction=&CustomRotator}
  n13 = NamedRerouteUsage()
  n14 = ScalarParameter() {group=Shadow, parameterName=ShadowExtraThickness, sortPriority=1}
  NamedRerouteDeclaration(None=MaterialFunctionCall(
    UVs=NamedRerouteDeclaration(None=@n12.Rotated Values) {name=TexShadowOffset, nodeColor=<rgba 0 1 0.3203 1>}
    Size=ScalarParameter() {defaultValue=0.9, group=Shadow, parameterName=ShadoeSize, sortPriority=0}
    OutlineThickness=Add(A=@n13, B=@n14)
    OutlineAlignment=Constant() {r=-1}
    GlowMax=ScalarParameter() {defaultValue=0.1, group=Shadow, parameterName=ShadowGlow, sortPriority=2}
   ) {materialFunction=&MF_UI_Circle}.Outline) {name=ShadowOutline, nodeColor=<rgba 1 0 0.5625 1>}
  NamedRerouteDeclaration(None=MaterialFunctionCall(
    Size=ScalarParameter() {defaultValue=0.9, group=Section, parameterName=Size}
    OutlineThickness=NamedRerouteUsage()
    OutlineAlignment=Constant() {r=-1}
    GlowMax=ScalarParameter() {defaultValue=0.01, group=Shadow, parameterName=CircleGlow, sortPriority=2}
   ) {materialFunction=&MF_UI_Circle}.Outline) {name=Outline, nodeColor=<rgba 0.6406 1 0 1>}
 ]
}
