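# WIP UI material function that builds a signed distance field for an isosceles triangle from UVs and a Size (base width, height), marked not for use until exposed in the library.
# UVs are remapped around the center and shifted by height.
# Squared distances to the two sides and base are combined with Min before one square root, with a Sign-based inner-surface term.
# It outputs Fill, Outline and two-segment-only fields, with debug nodes left in.
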MaterialFunction /FortUILibrary/MaterialFunctions/Shapes/SDFs/MF_SDF_Triangle_Isosceles
graph {
 properties {description=WIP : DO NOT USE UNTIL IT IS EXPOSED IN THE LIBRARY CATEGORY, previewMaterialDomain=MD_UI, libraryCategoriesText []}
 nodes [
  n1 = NamedRerouteUsage()
  n2 = DotProduct(A=@n1, B=@n1)
  n3 = Divide(A=DotProduct(A=NamedRerouteUsage(), B=@n1), B=@n2)
  n4 = NamedRerouteUsage()
  n5 = Subtract(A=NamedRerouteUsage(), B=Multiply(A=Saturate(None=@n3), B=@n4))
  n6 = Divide(A=DotProduct(A=@n1, B=NamedRerouteUsage()), B=@n2)
  n7 = Subtract(A=NamedRerouteUsage(), B=Multiply(A=@n4, B=Saturate(None=@n6)))
  n8 = Min(
   A=NamedRerouteDeclaration(None=DotProduct(A=@n5, B=@n5)) {name=Right Segment, nodeColor=<rgba 1 0 0.2344 1>}
   B=NamedRerouteDeclaration(None=DotProduct(A=@n7, B=@n7)) {name=Left Segment, nodeColor=<rgba 1 0 0.2344 1>}
  )
  FunctionOutput(None=Saturate(None=SquareRoot(None=@n8))) {outputName=SDF 2 Segments Only, sortPriority=2}
  n9 = Divide(A=NamedRerouteUsage(), B=ComponentMask(None=NamedRerouteUsage()) {desc="Base :\r\nTotal Width between 0 and 1", r=true}) {desc="This is defining the total triangle base length.\r\nSize.x = 1 : the gradient from the center (0) reaches 1 at the very left and right.\r\nSize.x =0.5 : the gradient from the center reaches 1 at the half of the distance to the very left and right.\r\nLike that we can observe the total width of the triangle base segment by additionning the distance from\r\nfrom the center to the left, and the distance from the center to the right."}
  n10 = Multiply(A=AppendVector(A=Saturate(None=@n9), B=Constant() {desc=Full y, r=1}), B=NamedRerouteUsage())
  n11 = Subtract(A=NamedRerouteUsage(), B=@n10)
  n12 = Subtract(A=NamedRerouteUsage(), B=@n10)
  n13 = Min(A=DotProduct(A=@n11, B=@n11), B=DotProduct(A=@n12, B=@n12))
  n14 = SquareRoot(None=Min(A=@n8, B=NamedRerouteDeclaration(None=@n13) {name=Base Segment, nodeColor=<rgba 1 0 0.2344 1>}))
  FunctionOutput(None=NamedRerouteDeclaration(None=Saturate(None=@n14)) {name=Distance Field Segments, nodeColor=<rgba 1 0 0.2344 1>}) {outputName=SDF Outline, sortPriority=1}
  FunctionOutput(None=Multiply(A=NamedRerouteUsage(), B=NamedRerouteUsage())) {description=Fill + Oultine, outputName=SDF Fill}
  MaterialFunctionCall() {materialFunction=&DebugScalarValues}
  MaterialFunctionCall!(Vector2, MaximumNumberOfDigits, UVs, DebugTextLocation RG_UpperRight BA_LowerLeft, Component Spacing) {materialFunction=/Engine/Functions/Engine_MaterialFunctions02/Utility/DebugFloat2Values}
  MaterialFunctionCall!(X range Y range, Gradient Height, ShowGrid?, Thickness, Color, "f(x)") {materialFunction=/Engine/Functions/Engine_MaterialFunctions01/Debug/PlotFunctionOnGraph}
  n15 = ConstantBiasScale(None=FunctionInput(Preview=TextureCoordinate()) {bUsePreviewValueAsDefault=true, inputName=UVs, inputType=FunctionInput_Vector2}) {bias=-0.5, desc="UVs Shifted to the center\r\nAnd rescaled 0 to 1", scale=2}
  n16 = ComponentMask(None=@n15) {r=true}
  n17 = Add(A=ComponentMask(None=@n15) {g=true}, B=Multiply(A=NamedRerouteUsage()) {constB=0.5}) {constB=0, desc="V Shift depending Height\r\nto keep the shape always\r\ncentered wathever the size"}
  NamedRerouteDeclaration(None=AppendVector(A=@n16, B=@n17)) {name=UVs, nodeColor=<rgba 1 0 0.1172 1>}
  NamedRerouteDeclaration(None=AppendVector(A=Multiply(A=@n16) {constB=-1, desc=Invert u}, B=@n17)) {name=UVs Mirrored, nodeColor=<rgba 1 0 0.1172 1>}
  n18 = NamedRerouteUsage()
  n19 = NamedRerouteUsage()
  n20 = NamedRerouteUsage()
  n21 = ComponentMask(None=NamedRerouteUsage()) {g=true}
  n22 = Multiply(
   A=@n18
   B=Subtract(
    A=Multiply(A=@n18, B=@n19) {desc="Influence U (both direction left && right)\r\nDepending size.y"}
    B=Multiply(A=@n20, B=@n21) {desc="Influence V\r\nDepending size.x"}
   ) {desc="SDF with negative values\r\n-1 to 1"}
  )
  n23 = NamedRerouteUsage()
  n24 = Multiply(A=Subtract(A=@n21, B=@n23) {desc="SDF with negative values\r\n-1 to 1."}, B=@n23) {desc="Influence V\r\nDepending size.y\r\nto determine the triangle \r\nbase height postion"}
  NamedRerouteDeclaration(None=Sign(None=Max(A=@n22, B=@n24)) {desc="-1, 0 or 1"}) {name=Triangle Inner Surface, nodeColor=<rgba 0.005 0.2181 0.5538 1>}
  NamedRerouteDeclaration(None=Abs(None=@n16)) {name=Abs. UVs, nodeColor=<rgba 1 0 0.1172 1>}
  n25 = FunctionInput(Preview=Constant2Vector() {g=1, r=1}) {
   bUsePreviewValueAsDefault=true
   desc="X= Base (Total width)\r\nY= Height from base center"
   description="X= Base (Total width); Y= Height from base center"
   inputName=Size
   inputType=FunctionInput_Vector2
   sortPriority=1
   previewValue=<0.4 0.4 0 1>
  }
  n26 = ComponentMask(None=@n25) {r=true}
  n27 = Max(A=Multiply(A=ComponentMask(None=@n25) {g=true}) {constB=2, desc="Double the height to fill the "}) {constB=0, desc="Avoid Full white when\r\nheight = 0\r\nneed to investigate"}
  NamedRerouteDeclaration(None=AppendVector(
    A=NamedRerouteDeclaration(None=@n26) {name=Width from Base Center, nodeColor=<rgba 0.1564 1 0.4924 1>}
    B=NamedRerouteDeclaration(None=@n27) {name=Height from Base, nodeColor=<rgba 1 0.3391 0.5398 1>}
   )) {name=Size, nodeColor=<rgba 0 0.4375 1 1>}
  Sign(None=@n23)
  Sign(None=@n18) {desc=Is it useful to get -1 for any negative height?}
 ]
}
