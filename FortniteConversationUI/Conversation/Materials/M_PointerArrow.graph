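# Translucent UI material for the conversation UI that draws a pointer arrow on a circular outline, aimed by ArrowAngle.
# An isosceles triangle SDF, rotated, offset and scaled by parameters, gets derivative-based smoothstep edges.
# The circle's fill is subtracted from the triangle and its outline added.
# A radial progress mask, sized by RadialMask and rotated from ArrowAngle, multiplies the combined shape to limit it to an arc.
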Material /FortniteConversationUI/Conversation/Materials/M_PointerArrow
graph {
 expressions {
  MaterialExpressionAdd_0 {
   class=MaterialExpressionAdd
   inputs [
    A {from=MaterialExpressionSaturate_4}
    B {from=MaterialExpressionReroute_3}
   ]
  }
  MaterialExpressionAdd_11 {
   class=MaterialExpressionAdd
   inputs [
    A {from=MaterialExpressionMultiply_10}
    B {from=MaterialExpressionConstant_10}
   ]
  }
  MaterialExpressionAdd_3 {
   class=MaterialExpressionAdd
   inputs [
    A {from=MaterialExpressionConstant_7}
    B {from=MaterialExpressionMultiply_6}
   ]
  }
  MaterialExpressionAdd_4 {
   class=MaterialExpressionAdd
   inputs [
    A {from=MaterialExpressionAdd_3}
    B {from=""}
   ]
   properties {constB=0}
  }
  MaterialExpressionAdd_5 {
   class=MaterialExpressionAdd
   inputs [
    A {from=MaterialExpressionScalarParameter_2}
    B {from=""}
   ]
   properties {constB=0.5}
  }
  MaterialExpressionAdd_9 {
   class=MaterialExpressionAdd
   inputs [
    A {from=MaterialExpressionMaterialFunctionCall_31, output=Rotated Values}
    B {from=MaterialExpressionSubtract_2}
   ]
  }
  MaterialExpressionConstant_0 {class=MaterialExpressionConstant, properties {r=0.85}}
  MaterialExpressionConstant_1 {class=MaterialExpressionConstant, properties {r=-1}}
  MaterialExpressionConstant_10 {class=MaterialExpressionConstant, properties {r=0.2}}
  MaterialExpressionConstant_6 {class=MaterialExpressionConstant, properties {r=1}}
  MaterialExpressionConstant_7 {class=MaterialExpressionConstant}
  MaterialExpressionConvert_0 {
   class=MaterialExpressionConvert
   inputs [
    None {from=""}
    None {from=MaterialExpressionScalarParameter_8}
   ]
   properties {
    nodeName=Make Vector2
    convertInputs [
     {type=Scalar, expressionInput {expression=null, inputName=None, mask=0, maskA=0, maskB=0, maskG=0, maskR=0, outputIndex=0}}
     {
      type=Scalar
      defaultValue=<rgba 0.5 0 0 1>
      expressionInput {
       expression=&self:MaterialExpressionScalarParameter_8
       inputName=None
       mask=0
       maskA=0
       maskB=0
       maskG=0
       maskR=0
       outputIndex=0
      }
     }
    ]
    convertOutputs [{type=Vector2}]
   }
  }
  MaterialExpressionConvert_4 {
   class=MaterialExpressionConvert
   inputs [
    None {from=""}
    None {from=MaterialExpressionMultiply_7}
   ]
   properties {
    nodeName=Make Vector2
    convertInputs [
     {type=Scalar, expressionInput {expression=null, inputName=None, mask=0, maskA=0, maskB=0, maskG=0, maskR=0, outputIndex=0}}
     {
      type=Scalar
      defaultValue=<rgba -0.05 0 0 1>
      expressionInput {
       expression=&self:MaterialExpressionMultiply_7
       inputName=None
       mask=0
       maskA=0
       maskB=0
       maskG=0
       maskR=0
       outputIndex=0
      }
     }
    ]
    convertOutputs [{type=Vector2}]
   }
  }
  MaterialExpressionConvert_5 {
   class=MaterialExpressionConvert
   inputs [
    None {from=MaterialExpressionScalarParameter_5}
    None {from=""}
   ]
   properties {
    nodeName=Make Vector2
    convertInputs [
     {
      type=Scalar
      expressionInput {
       expression=&self:MaterialExpressionScalarParameter_5
       inputName=None
       mask=0
       maskA=0
       maskB=0
       maskG=0
       maskR=0
       outputIndex=0
      }
     }
     {
      type=Scalar
      defaultValue=<rgba 1 0 0 1>
      expressionInput {expression=null, inputName=None, mask=0, maskA=0, maskB=0, maskG=0, maskR=0, outputIndex=0}
     }
    ]
    convertOutputs [{type=Vector2}]
   }
  }
  MaterialExpressionDDX_0 {class=MaterialExpressionDDX, inputs [Value {from=MaterialExpressionNamedRerouteUsage_4, output=""}], properties {}}
  MaterialExpressionDivide_0 {
   class=MaterialExpressionDivide
   inputs [
    A {from=MaterialExpressionScalarParameter_4}
    B {from=""}
   ]
  }
  MaterialExpressionLength_0 {class=MaterialExpressionLength, inputs [None {from=MaterialExpressionDDX_0, output=""}], properties {}}
  MaterialExpressionMaterialFunctionCall_1 {
   class=MaterialExpressionMaterialFunctionCall
   inputs [
    UVs {from=""}
    Size {from=MaterialExpressionConstant_0}
    OutlineThickness {from=MaterialExpressionMultiply_0}
    OutlineAlignment {from=MaterialExpressionConstant_1}
    GlowMin {from=""}
    GlowMax {from=""}
   ]
   properties {materialFunction=&MF_UI_Circle}
  }
  MaterialExpressionMaterialFunctionCall_19 {
   class=MaterialExpressionMaterialFunctionCall
   inputs [
    UV {from=MaterialExpressionNamedRerouteUsage_2}
    Angle {from=MaterialExpressionAdd_4}
    Progress {from=MaterialExpressionReroute_1}
    Width {from=""}
   ]
   properties {materialFunction=/FortniteConversationUI/Conversation/Materials/MaterialFunctions/MF_RadialProgress}
  }
  MaterialExpressionMaterialFunctionCall_21 {
   class=MaterialExpressionMaterialFunctionCall
   inputs [
    UVs {from=MaterialExpressionNamedRerouteDeclaration_1}
    Rotation Center {from=""}
    "Rotation Angle (0-1)" {from=MaterialExpressionAdd_5}
   ]
   properties {materialFunction=&CustomRotator}
  }
  MaterialExpressionMaterialFunctionCall_29 {
   class=MaterialExpressionMaterialFunctionCall
   inputs [
    UVs {from=MaterialExpressionAdd_9}
    Texture Scale {from=MaterialExpressionConvert_5}
   ]
   properties {materialFunction=&ScaleUVsByCenter}
  }
  MaterialExpressionMaterialFunctionCall_31 {
   class=MaterialExpressionMaterialFunctionCall
   inputs [
    UVs {from=MaterialExpressionNamedRerouteUsage_3}
    Rotation Center {from=""}
    "Rotation Angle (0-1)" {from=MaterialExpressionScalarParameter_3}
   ]
   properties {materialFunction=&CustomRotator}
  }
  MaterialExpressionMaterialFunctionCall_33 {
   class=MaterialExpressionMaterialFunctionCall
   inputs [
    UVs {from=MaterialExpressionMaterialFunctionCall_29, output=UVs}
    Size {from=MaterialExpressionScalarParameter_9}
   ]
   properties {materialFunction=/FortUILibrary/MaterialFunctions/Shapes/SDFs/MF_SDF_Triangle_Isosceles}
  }
  MaterialExpressionMultiply_0 {
   class=MaterialExpressionMultiply
   inputs [
    A {from=MaterialExpressionScalarParameter_0}
    B {from=""}
   ]
   properties {constB=0.5}
  }
  MaterialExpressionMultiply_10 {
   class=MaterialExpressionMultiply
   inputs [
    A {from=MaterialExpressionLength_0}
    B {from=""}
   ]
   properties {constB=4}
  }
  MaterialExpressionMultiply_4 {
   class=MaterialExpressionMultiply
   inputs [
    A {from=MaterialExpressionMaterialFunctionCall_19, output=Result}
    B {from=MaterialExpressionSaturate_5}
   ]
  }
  MaterialExpressionMultiply_6 {
   class=MaterialExpressionMultiply
   inputs [
    A {from=MaterialExpressionScalarParameter_1}
    B {from=""}
   ]
   properties {constB=0.5}
  }
  MaterialExpressionMultiply_7 {
   class=MaterialExpressionMultiply
   inputs [
    A {from=MaterialExpressionDivide_0}
    B {from=""}
   ]
   properties {constB=1.25}
  }
  MaterialExpressionNamedRerouteDeclaration_0 {
   class=MaterialExpressionNamedRerouteDeclaration
   inputs [None {from=MaterialExpressionMaterialFunctionCall_21, output=Rotated Values}]
   properties {name=TexCoord, nodeColor=<rgba 0 0.9297 1 1>}
  }
  MaterialExpressionNamedRerouteDeclaration_1 {
   class=MaterialExpressionNamedRerouteDeclaration
   inputs [None {from=MaterialExpressionTextureCoordinate_1}]
   properties {name=Tex, nodeColor=<rgba 0 0.4844 1 1>}
  }
  MaterialExpressionNamedRerouteUsage_2 {class=MaterialExpressionNamedRerouteUsage}
  MaterialExpressionNamedRerouteUsage_3 {class=MaterialExpressionNamedRerouteUsage}
  MaterialExpressionNamedRerouteUsage_4 {class=MaterialExpressionNamedRerouteUsage}
  MaterialExpressionOneMinus_1 {class=MaterialExpressionOneMinus, inputs [None {from=MaterialExpressionSmoothStep_4}]}
  MaterialExpressionReroute_1 {class=MaterialExpressionReroute, inputs [None {from=MaterialExpressionScalarParameter_1}]}
  MaterialExpressionReroute_3 {class=MaterialExpressionReroute, inputs [None {from=MaterialExpressionMaterialFunctionCall_1, output=Outline}]}
  MaterialExpressionSaturate_2 {class=MaterialExpressionSaturate, inputs [None {from=MaterialExpressionMultiply_4}]}
  MaterialExpressionSaturate_4 {class=MaterialExpressionSaturate, inputs [None {from=MaterialExpressionSubtract_4}]}
  MaterialExpressionSaturate_5 {class=MaterialExpressionSaturate, inputs [None {from=MaterialExpressionAdd_0}]}
  MaterialExpressionScalarParameter_0 {class=MaterialExpressionScalarParameter, properties {defaultValue=0.05, parameterName=Thickness}}
  MaterialExpressionScalarParameter_1 {class=MaterialExpressionScalarParameter, properties {defaultValue=0.12, parameterName=RadialMask}}
  MaterialExpressionScalarParameter_2 {class=MaterialExpressionScalarParameter, properties {defaultValue=0.25, parameterName=ArrowAngle}}
  MaterialExpressionScalarParameter_3 {class=MaterialExpressionScalarParameter, properties {defaultValue=0.25, parameterName=ArrowAngle}}
  MaterialExpressionScalarParameter_4 {class=MaterialExpressionScalarParameter, properties {defaultValue=0.08, parameterName=ArrowSize}}
  MaterialExpressionScalarParameter_5 {class=MaterialExpressionScalarParameter, properties {defaultValue=1.25, parameterName=ArrowWideScale}}
  MaterialExpressionScalarParameter_8 {class=MaterialExpressionScalarParameter, properties {defaultValue=0.5, parameterName=ArrowHeightPos}}
  MaterialExpressionScalarParameter_9 {class=MaterialExpressionScalarParameter, properties {defaultValue=0.08, parameterName=ArrowSize}}
  MaterialExpressionSmoothStep_4 {
   class=MaterialExpressionSmoothStep
   inputs [
    Min {from=MaterialExpressionConstant_10, output=""}
    Max {from=MaterialExpressionAdd_11, output=""}
    Value {from=MaterialExpressionSubtract_3, output=""}
   ]
   properties {constMax=0.99, constMin=1}
  }
  MaterialExpressionSubtract_2 {
   class=MaterialExpressionSubtract
   inputs [
    A {from=MaterialExpressionConvert_0}
    B {from=MaterialExpressionConvert_4}
   ]
  }
  MaterialExpressionSubtract_3 {
   class=MaterialExpressionSubtract
   inputs [
    A {from=MaterialExpressionMaterialFunctionCall_33, output=SDF Fill}
    B {from=""}
   ]
   properties {constB=-0.2}
  }
  MaterialExpressionSubtract_4 {
   class=MaterialExpressionSubtract
   inputs [
    A {from=MaterialExpressionOneMinus_1}
    B {from=MaterialExpressionMaterialFunctionCall_1, output=Fill}
   ]
  }
  MaterialExpressionTextureCoordinate_1 {class=MaterialExpressionTextureCoordinate}
 }
 properties {bAutomaticallySetUsageInEditor=false, blendMode=BLEND_Translucent, materialDomain=MD_UI, usedShadingModels=Default Lit}
}
