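# Translucent UI master material for donation fund tiles in the conversation UI, drawing a stylized icon over a procedural grid.
# Offset icon samples form an outline, which noise and rotated sine shade lines turn into a drawing effect.
# Scalar parameters animate icon transform, shine, shadow, crop and a blend to a radial color-dodge gradient with depth tint.
# HasIcon and JustIcon static switches select the layers, while edge and crop masks shape the alpha.
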Material /FortniteConversationUI/Conversation/Materials/M_UI_DonationFundTile_Master
graph {
 properties {blendMode=BLEND_Translucent, materialDomain=MD_UI, pixelDepthOffsetMode=PDOM_Legacy, usedShadingModels=Default Lit}
 nodes [
  n1 = NamedRerouteUsage()
  n2 = NamedRerouteUsage()
  n3 = Add(
   A=@n2
   B=AppendVector(
    A=ScalarParameter() {defaultValue=-0.005, group=Icon, parameterName=Outline4_ShiftX}
    B=ScalarParameter() {defaultValue=0.005, group=Icon, parameterName=Outline4_ShiftY}
   )
  )
  n4 = TextureObjectParameter() {group=Icon, parameterName=IconImage, texture=/Game/UI/Foundation/Textures/Icons/Weapons/Items/T-Icon-Weapons-SK-AssaultWoodForend-L}
  n5 = MaterialFunctionCall!(
   UVs=@n3
   Alpha=TextureSample(UVs=@n3, Tex=@n4) {texture=&T-Icon-Weapons-SK-AssaultWoodForend-L}.A
  ) {materialFunction=/Game/UI/Foundation/Items/ClampAlphaToTransparentBorder}
  n6 = DotProduct(A=Saturate(None=OneMinus(None=@n5.Result)), B=NamedRerouteUsage())
  n7 = Add(
   A=@n2
   B=AppendVector(
    A=ScalarParameter() {defaultValue=0.005, group=Icon, parameterName=Outline3_ShiftX}
    B=ScalarParameter() {defaultValue=-0.005, group=Icon, parameterName=Outline3_ShiftY}
   )
  )
  n8 = MaterialFunctionCall(
   UVs=@n7
   Alpha=TextureSample(UVs=@n7, Tex=@n4) {texture=&T-Icon-Weapons-SK-AssaultWoodForend-L}.A
  ) {materialFunction=&ClampAlphaToTransparentBorder}
  n9 = DotProduct(A=Saturate(None=OneMinus(None=@n8.Result)), B=NamedRerouteUsage())
  n10 = Saturate(None=Add(A=Saturate(None=@n6), B=Saturate(None=@n9)))
  n11 = NamedRerouteUsage()
  n12 = Add(
   A=@n11
   B=AppendVector(
    A=ScalarParameter() {defaultValue=0.005, group=Icon, parameterName=Outline2_ShiftX}
    B=ScalarParameter() {defaultValue=0.005, group=Icon, parameterName=Outline2_ShiftY}
   )
  )
  n13 = TextureSample(UVs=@n12, Tex=@n4) {texture=&T-Icon-Weapons-SK-AssaultWoodForend-L}
  n14 = OneMinus(None=MaterialFunctionCall(UVs=@n12, Alpha=@n13.A) {materialFunction=&ClampAlphaToTransparentBorder}.Result)
  n15 = Saturate(None=DotProduct(A=Saturate(None=@n14), B=NamedRerouteUsage()))
  n16 = Add(
   A=@n11
   B=AppendVector(
    A=ScalarParameter() {defaultValue=-0.005, group=Icon, parameterName=Outline1_ShiftX}
    B=ScalarParameter() {defaultValue=-0.005, group=Icon, parameterName=Outline1_ShiftY}
   )
  )
  n17 = TextureSample(UVs=@n16, Tex=@n4) {texture=&T-Icon-Weapons-SK-AssaultWoodForend-L}
  n18 = OneMinus(None=MaterialFunctionCall(UVs=@n16, Alpha=@n17.A) {materialFunction=&ClampAlphaToTransparentBorder}.Result)
  n19 = NamedRerouteUsage()
  n20 = TextureSample(UVs=@n19, Tex=@n4) {samplerSource=SSM_Clamp_WorldGroupSettings}
  n21 = MaterialFunctionCall(UVs=@n19, Alpha=@n20.A) {materialFunction=&ClampAlphaToTransparentBorder}
  n22 = ComponentMask(None=Add(A=TextureCoordinate(), B=ScalarParameter() {parameterName=ExtraTextureFadeOffset})) {g=true}
  n23 = OneMinus(None=Saturate(None=LinearInterpolate(Alpha=@n22) {constA=-1}))
  n24 = ScalarParameter() {defaultValue=1.25, parameterName=ExtraTextureFadeTaper}
  n25 = LinearInterpolate(B=Saturate(None=Power(Base=@n23, Exp=@n24)), Alpha=ScalarParameter() {parameterName=UseExtraFade}) {constA=1}
  n26 = NamedRerouteDeclaration(None=Saturate(None=DotProduct(A=@n21.Result, B=@n25))) {name=IconImageAlpha, nodeColor=<rgba 1 0.2344 0 1>}
  n27 = Saturate(None=DotProduct(A=Saturate(None=@n18), B=@n26))
  n28 = Add(A=@n10, B=Saturate(None=Add(A=@n15, B=@n27)))
  n29 = ScalarParameter() {defaultValue=6, parameterName=OutlineMultiply}
  n30 = Multiply(A=OneMinus(None=@n17.R), B=@n29) {constB=16}
  n31 = ScalarParameter() {defaultValue=3, parameterName=OutlineSubtract}
  n32 = Multiply(A=NamedRerouteUsage(), B=@n29) {constB=8}
  n33 = DotProduct(A=Saturate(None=Subtract(A=@n30, B=@n31) {constB=0.25}), B=Saturate(None=Subtract(A=@n32, B=@n31) {constB=0.25}))
  n34 = Saturate(None=Add(A=Saturate(None=@n28), B=Saturate(None=@n33)))
  n35 = TextureSample(UVs=MaterialFunctionCall(UVs=NamedRerouteUsage(), Texture Scale=ScalarParameter() {defaultValue=6, group=Noise, parameterName=Noise_Scale}) {materialFunction=&ScaleUVsByCenter}.UVs) {texture=/Game/Global/Textures/Default/Noise/T_UI_Ice_Noise}
  n36 = ScalarParameter() {defaultValue=3, group=Noise, parameterName=Noise_Intensity}
  n37 = Multiply(A=@n35.R, B=@n36) {constB=16}
  n38 = Saturate(None=@n37)
  n39 = NamedRerouteDeclaration(None=Saturate(None=DotProduct(A=@n34, B=@n38))) {name=Outline, nodeColor=<rgba 0.4062 0 1 1>}
  n40 = Multiply(A=NamedRerouteUsage(), B=ScalarParameter() {defaultValue=0.333, parameterName=FillAmount})
  n41 = ScalarParameter() {defaultValue=0.25, parameterName=Noise_Lift}
  n42 = MaterialFunctionCall() {materialFunction=&ScreenAlignedUVs}
  n43 = ScalarParameter() {defaultValue=0.075, group=ShadeLines, parameterName=ShadeLines_Rotation}
  n44 = ScalarParameter() {defaultValue=0.0125, group=ShadeLines, parameterName=ShadeLines_Distortion}
  n45 = Multiply(
   A=Add(
    A=MaterialFunctionCall(UVs=@n42."X 100%, Y Scale to Ratio", "Rotation Angle (0-1)"=@n43) {materialFunction=&CustomRotator}.Rotated Values
    B=Multiply(A=@n35.R, B=@n44)
   )
   B=ScalarParameter() {defaultValue=40, group=ShadeLines, parameterName=ShadeLines_Amount}
  )
  n46 = Saturate(None=Power(Base=Sine(None=@n45), Exp=ScalarParameter() {defaultValue=3, group=ShadeLines, parameterName=ShadeLines_Thickness}))
  n47 = LinearInterpolate(
   A=ScalarParameter() {defaultValue=0.5, group=ShadeLines, parameterName=ShadeLines_Lift}
   Alpha=Saturate(None=MaterialFunctionCall(Float2=@n46) {materialFunction=&BreakOutFloat2Components}.R)
  )
  n48 = Saturate(None=@n47)
  n49 = Saturate(None=DotProduct(A=Saturate(None=@n40), B=LinearInterpolate(A=@n41, Alpha=@n48)))
  n50 = NamedRerouteDeclaration(None=Saturate(None=Add(A=@n39, B=@n49))) {name=DrawingIcon, nodeColor=<rgba 1 0 0.8672 1>}
  n51 = LinearInterpolate(
   A=@n1
   B=VectorParameter() {parameterName=DrawingColor, defaultValue=<rgba 1 1 1 1>}.RGB
   Alpha=Multiply(A=@n50, B=ScalarParameter() {defaultValue=0.9, parameterName=DrawingIcon_Alpha})
  )
  n52 = Add(A=@n10, B=@n27)
  n53 = Saturate(None=Add(A=Saturate(None=@n52), B=@n15))
  Add(A=@n51, B=@n53)
  ComponentMask() {g=true, r=true}
  Constant() {r=0.1}
  DotProduct(A=@n38, B=@n47)
  LinearInterpolate()
  Multiply(A=ScalarParameter() {defaultValue=16, group=Background Gradient, parameterName=Spotlight_Hardness}) {constB=0.1}
  n54 = Constant()
  n55 = Add(
   A=MaterialFunctionCall(UVs=TextureCoordinate(), "Rotation Angle (0-1)"=ScalarParameter() {parameterName=TopAlpha_Rotation}) {materialFunction=&CustomRotator}.Rotated Values
   B=AppendVector(A=Constant(), B=ScalarParameter() {defaultValue=-0.125, parameterName=TopAlpha_Offset})
  )
  n56 = Saturate(None=Multiply(A=ComponentMask(None=@n55) {g=true}, B=ScalarParameter() {defaultValue=384, parameterName=EdgeHardness}))
  n57 = DotProduct(
   A=LinearInterpolate(
    A=ScalarParameter() {defaultValue=1, group=Grid, parameterName=GridBacking_Alpha}
    Alpha=StaticSwitchParameter(True=@n50, False=@n54) {defaultValue=true, parameterName=HasIcon}
   )
   B=@n56
  )
  n58 = Add(A=NamedRerouteUsage(), B=@n56)
  n59 = StaticSwitchParameter(
   True=LinearInterpolate(A=NamedRerouteUsage(), B=NamedRerouteUsage(), Alpha=NamedRerouteUsage())
   False=LinearInterpolate(A=Saturate(None=@n57), B=Saturate(None=@n58), Alpha=ScalarParameter() {group=Animate, parameterName=AnimateCrop})
  ) {parameterName=JustIcon}
  n60 = MaterialFunctionCall(
   UVs=TextureCoordinate() {coordinateIndex=4}
   "Rotation Angle (0-1)"=ScalarParameter() {defaultValue=-0.0125, group=Crop, parameterName=CropAngle}
  ) {materialFunction=&CustomRotator}
  n61 = LinearInterpolate(A=OneMinus(None=@n60.Rotated Values), B=@n60.Rotated Values, Alpha=ScalarParameter() {group=Crop, parameterName=FlipCrop})
  n62 = ScalarParameter() {defaultValue=-0.05, group=Crop, parameterName=OffsetCrop}
  n63 = Multiply(A=ComponentMask(None=Add(A=@n61, B=@n62)) {r=true}, B=ScalarParameter() {defaultValue=88, group=Crop, parameterName=CropEdge})
  n64 = Saturate(None=Power(Base=Saturate(None=@n63), Exp=ScalarParameter() {defaultValue=1, group=Crop, parameterName=CropTaper}))
  n65 = DotProduct(
   A=@n59
   B=LinearInterpolate(
    A=ScalarParameter() {group=Crop, parameterName=LiftCropAlpha}
    Alpha=NamedRerouteDeclaration(None=@n64) {name=Crop, nodeColor=<rgba 1 0.7969 0 1>}
   )
  )
  Multiply(
   A=LinearInterpolate(A=@n59, B=Saturate(None=@n65), Alpha=ScalarParameter() {group=Crop, parameterName=UseCrop})
   B=ScalarParameter() {defaultValue=1, parameterName=FinalALpha}
  )
  n66 = ScalarParameter() {defaultValue=16, group=Background Gradient, parameterName=Spotlight_Intensity}
  Multiply(A=@n66) {constB=0.01}
  n67 = MaterialFunctionCall() {materialFunction=&GetUserInterfaceUV}
  n68 = MaterialFunctionCall(Float2=@n67.Pixel Size) {materialFunction=&BreakOutFloat2Components}
  n69 = Divide(A=@n68.R, B=@n68.R)
  n70 = AppendVector(A=@n69, B=Constant() {r=1})
  n71 = Multiply(A=OneMinus(None=@n69)) {constB=0.5}
  n72 = Constant()
  NamedRerouteDeclaration(None=Add(A=Multiply(A=@n67.Normalized UV, B=@n70), B=AppendVector(A=@n71, B=@n72))) {name=XPrioritySquareUVS, nodeColor=<rgba 1 0 0.4687 1>}
  n73 = TextureCoordinate()
  n74 = LinearInterpolate(
   A=ScalarParameter() {defaultValue=0.5, group=Shine, parameterName=ShineOffsetStart}
   B=ScalarParameter() {defaultValue=-0.5, group=Shine, parameterName=ShineOffsetStop}
   Alpha=ScalarParameter() {group=Shine, parameterName=OffsetShine}
  )
  n75 = MaterialFunctionCall(
   UVs=Multiply(A=Add(A=@n73, B=@n74), B=ScalarParameter() {defaultValue=2, group=Shine, parameterName=ShineWidth})
   "Rotation Angle (0-1)"=ScalarParameter() {defaultValue=-0.0625, group=Shine, parameterName=ShineRotation}
  ) {materialFunction=&CustomRotator}
  n76 = OneMinus(None=@n75.Rotated Values)
  n77 = Multiply(
   A=DotProduct(A=ComponentMask(None=@n76) {r=true}, B=ComponentMask(None=@n75.Rotated Values) {r=true})
   B=ScalarParameter() {defaultValue=3, group=Shine, parameterName=ShineStrength}
  )
  NamedRerouteDeclaration(None=Power(Base=Saturate(None=@n77), Exp=ScalarParameter() {defaultValue=1.22, group=Shine, parameterName=ShineTaper})) {name=ShineSwipe, nodeColor=<rgba 0 0.9062 1 1>}
  n78 = MaterialFunctionCall(
   UVs=NamedRerouteUsage()
   "Rotation Angle (0-1)"=LinearInterpolate(
    A=ScalarParameter() {defaultValue=-0.1, group=Icon, parameterName=IconRotationStart}
    B=ScalarParameter() {group=Icon, parameterName=IconRotationStop}
    Alpha=ScalarParameter() {group=Animate, parameterName=AnimateRotation}
   )
  ) {materialFunction=&CustomRotator}
  n79 = AppendVector(
   A=LinearInterpolate(
    A=ScalarParameter() {defaultValue=0.125, group=Icon, parameterName=IconPositionX_Start}
    B=ScalarParameter() {group=Icon, parameterName=IconPositionX_Stop}
    Alpha=ScalarParameter() {group=Animate, parameterName=Animate}
   )
   B=LinearInterpolate(
    A=ScalarParameter() {defaultValue=0.125, group=Icon, parameterName=IconPositionY_Start}
    B=ScalarParameter() {group=Icon, parameterName=IconPositionY_Stop}
    Alpha=ScalarParameter() {group=Animate, parameterName=Animate}
   )
  )
  n80 = ScalarParameter() {defaultValue=0.8, group=Icon, parameterName=IconScale_Start}
  n81 = ScalarParameter() {defaultValue=0.825, group=Icon, parameterName=IconScale_Stop}
  n82 = ScalarParameter() {group=Animate, parameterName=Animate}
  NamedRerouteDeclaration(None=MaterialFunctionCall(UVs=Add(A=@n78.Rotated Values, B=@n79), Texture Scale=LinearInterpolate(A=@n80, B=@n81, Alpha=@n82)) {materialFunction=&ScaleUVsByCenter}.UVs) {name=UVAdjustment, nodeColor=<rgba 1 0.75 0 1>}
  NamedRerouteDeclaration(None=@n20.RGB) {name=IconRGB, nodeColor=<rgba 0 1 0.2031 1>}
  NamedRerouteDeclaration(None=Saturate(None=DotProduct(A=@n20.R, B=@n26))) {name=IconRedChannel, nodeColor=<rgba 1 0.3281 0 1>}
  n83 = LinearInterpolate(
   A=VectorParameter() {group=Grid, parameterName=BackingColorA, defaultValue=<rgba 0 0.4167 1 1>}.RGB
   B=VectorParameter() {group=Grid, parameterName=BackingColorB, defaultValue=<rgba 0 0.3 0.9 1>}.RGB
   Alpha=MaterialFunctionCall(
    UVs=NamedRerouteUsage()
    Radius=ScalarParameter() {defaultValue=0.5, group=Grid, parameterName=GridBacking_Radius}
    Density=ScalarParameter() {defaultValue=1.666, group=Grid, parameterName=GridBacking_Density}
   ) {materialFunction=&RadialGradientExponential}.RadialGradientExponential
  )
  n84 = LinearInterpolate(
   A=VectorParameter() {group=Grid, parameterName=GridColor, defaultValue=<rgba 0 0.5 1 1>}.RGB
   B=VectorParameter() {group=Grid, parameterName=GridColorShine, defaultValue=<rgba 0.25 1 1 1>}.RGB
   Alpha=NamedRerouteUsage()
  )
  n85 = Add(
   A=MaterialFunctionCall(
    UVs=MaterialFunctionCall() {materialFunction=&ScreenAlignedUVs}."X 100%, Y Scale to Ratio"
    "Rotation Angle (0-1)"=ScalarParameter() {defaultValue=0.0025, group=Grid, parameterName=Grid_Rotation}
   ) {materialFunction=&CustomRotator}.Rotated Values
   B=AppendVector(
    A=ScalarParameter() {defaultValue=-0.025, group=Grid, parameterName=ShiftGrid_X}
    B=ScalarParameter() {defaultValue=-0.025, group=Grid, parameterName=ShiftGrid_Y}
   )
  )
  n86 = ScalarParameter() {defaultValue=16, group=Grid, parameterName=GridAmount}
  n87 = AppendVector(A=Multiply(A=@n86, B=ScalarParameter() {defaultValue=1, group=Grid, parameterName=AdjustXGrid}), B=@n86)
  n88 = Sine(None=Multiply(A=@n85, B=@n87))
  n89 = ScalarParameter() {defaultValue=24, group=Grid, parameterName=GridThickness}
  n90 = MaterialFunctionCall(Float2=Saturate(None=Power(Base=@n88, Exp=@n89))) {materialFunction=&BreakOutFloat2Components}
  n91 = Saturate(None=Add(A=@n90.R, B=@n90.R))
  n92 = ScalarParameter() {defaultValue=0.5, parameterName=GridSmaller_Amount}
  n93 = AppendVector(
   A=ScalarParameter() {defaultValue=-0.0125, group=Grid, parameterName=ShiftGrid_Large_X}
   B=ScalarParameter() {defaultValue=-0.0125, group=Grid, parameterName=ShiftGrid_Large_Y}
  )
  n94 = Sine(None=Multiply(A=Add(A=@n85, B=@n93), B=Multiply(A=@n87) {constB=0.25}))
  n95 = Multiply(A=ScalarParameter() {defaultValue=24, group=Grid, parameterName=GridThickness}) {constB=4}
  n96 = MaterialFunctionCall(Float2=Saturate(None=Power(Base=@n94, Exp=@n95))) {materialFunction=&BreakOutFloat2Components}
  n97 = ScalarParameter() {defaultValue=16, group=Grid, parameterName=GridLarge_Hardness}
  n98 = Add(A=Saturate(None=Multiply(A=@n96.R, B=@n97)), B=Saturate(None=Multiply(A=@n96.G, B=@n97)))
  n99 = Saturate(None=Add(A=Multiply(A=@n91, B=@n92), B=Saturate(None=@n98)))
  NamedRerouteDeclaration(None=LinearInterpolate(
    A=VectorParameter() {parameterName=GrimeColor, defaultValue=<rgba 0 0.1671 0.2865 1>}.RGB
    B=LinearInterpolate(A=@n83, B=@n84, Alpha=@n99)
    Alpha=@n38
   )) {name=Grid, nodeColor=<rgba 0 0.0156 1 1>}
  Power(Base=@n37, Exp=Multiply(A=@n36) {constB=0.5})
  Saturate(None=DotProduct(A=OneMinus(None=@n13.R), B=NamedRerouteUsage()))
  ScalarParameter() {parameterName=HardOutline}
  n100 = NamedRerouteUsage()
  n101 = ComponentMask(None=Add(A=TextureCoordinate(), B=ScalarParameter() {defaultValue=-0.5, group=Fade, parameterName=OffsetFade})) {g=true}
  n102 = ScalarParameter() {defaultValue=2, group=Fade, parameterName=FadeHardness}
  n103 = Saturate(None=Power(Base=Multiply(A=@n101, B=@n102), Exp=ScalarParameter() {defaultValue=1.22, group=Fade, parameterName=FadeTaper}))
  n104 = LinearInterpolate(A=NamedRerouteUsage(), Alpha=@n103)
  n105 = NamedRerouteDeclaration(None=@n103) {name=Fade, nodeColor=<rgba 1 0.1406 0 1>}
  n106 = NamedRerouteUsage()
  n107 = NamedRerouteUsage()
  n108 = NamedRerouteUsage()
  n109 = Add(A=@n107, B=@n108)
  n110 = Saturate(None=LinearInterpolate(A=@n107, B=Saturate(None=@n109), Alpha=ScalarParameter() {parameterName=Sticker}))
  n111 = NamedRerouteUsage()
  n112 = Add(A=@n111, B=@n108)
  n113 = Saturate(None=LinearInterpolate(A=@n111, B=Saturate(None=@n112), Alpha=ScalarParameter() {parameterName=Sticker}))
  n114 = LinearInterpolate(
   A=LinearInterpolate(A=@n51, B=VectorParameter() {parameterName=OutlineColor, defaultValue=<rgba 1 1 1 1>}.RGB, Alpha=@n53)
   B=LinearInterpolate(A=LinearInterpolate(A=@n106, B=@n110, Alpha=@n113), B=@n51, Alpha=NamedRerouteUsage())
   Alpha=ScalarParameter() {group=Fade, parameterName=UseFade}
  )
  n115 = ScalarParameter() {parameterName=MixGridIn}
  n116 = ScalarParameter() {defaultValue=50, group=Background Gradient, parameterName=Gradient_Position_X}
  n117 = ScalarParameter() {defaultValue=45, group=Background Gradient, parameterName=Gradient_Position_Y}
  n118 = MaterialFunctionCall() {materialFunction=&GetUserInterfaceUV}
  n119 = MaterialFunctionCall(Float2=@n118.Pixel Size) {materialFunction=&BreakOutFloat2Components}
  n120 = Divide(A=@n119.G, B=@n119.G)
  n121 = AppendVector(A=Constant() {r=1}, B=@n120)
  n122 = Constant()
  n123 = Multiply(A=OneMinus(None=@n120)) {constB=0.5}
  n124 = NamedRerouteDeclaration(None=Add(A=Multiply(A=@n118.Normalized UV, B=@n121), B=AppendVector(A=@n122, B=@n123))) {name=YPrioritySquareUVS, nodeColor=<rgba 0.9922 1 0 1>}
  n125 = MaterialFunctionCall(
   UVs=@n124
   CenterPosition=AppendVector(A=Multiply(A=@n116) {constB=0.01}, B=Multiply(A=@n117) {constB=0.01})
   Density=Multiply(A=ScalarParameter() {defaultValue=2, group=Background Gradient, parameterName=Gradient_Hardness}) {constB=0.1}
  ) {materialFunction=&RadialGradientExponential}
  n126 = VectorParameter() {group=Background Gradient, parameterName=Background_Color_A, defaultValue=<rgba 0.0575 0.0625 0.0437 1>}
  n127 = AppendVector(
   A=Multiply(A=ScalarParameter() {defaultValue=50, group=Background Gradient, parameterName=Gradient_Position_X}) {constB=0.01}
   B=Multiply(A=ScalarParameter() {defaultValue=45, group=Background Gradient, parameterName=Gradient_Position_Y}) {constB=0.01}
  )
  n128 = Multiply(A=ScalarParameter() {defaultValue=66.67, group=Background Gradient, parameterName=Gradient_Size}) {constB=0.01}
  n129 = ScalarParameter() {defaultValue=2, group=Background Gradient, parameterName=Gradient_Hardness}
  n130 = MaterialFunctionCall!(
   Blend=Saturate(None=LinearInterpolate(B=@n66, Alpha=@n125.RadialGradientExponential))
   Base=LinearInterpolate(
    A=@n126.RGB
    B=VectorParameter() {group=Background Gradient, parameterName=Background_Color_B, defaultValue=<rgba 0.325 0.3047 0.2031 1>}.RGB
    Alpha=MaterialFunctionCall(UVs=@n124, CenterPosition=@n127, Radius=@n128, Density=@n129) {materialFunction=&RadialGradientExponential}.RadialGradientExponential
   )
  ) {materialFunction=/Engine/Functions/Engine_MaterialFunctions03/Blends/Blend_ColorDodge}
  n131 = NamedRerouteUsage()
  n132 = ComponentMask(None=TextureCoordinate()) {g=true}
  n133 = ScalarParameter() {defaultValue=1, parameterName=DepthFadeTaper}
  n134 = LinearInterpolate(
   A=@n131
   B=VectorParameter() {parameterName=DepthColor, defaultValue=<rgba 0.3234 0.3333 0.1346 1>}.RGB
   Alpha=Saturate(None=Power(Base=@n132, Exp=@n133))
  )
  n135 = ScalarParameter() {group=Animate, parameterName=AnimateDepthFade}
  n136 = LinearInterpolate(
   A=StaticSwitchParameter(True=LinearInterpolate(A=@n114, B=@n1, Alpha=@n115), False=NamedRerouteUsage()) {defaultValue=true, parameterName=HasIcon}
   B=LinearInterpolate(A=Saturate(None=@n130.Result), B=LinearInterpolate(A=@n134, B=@n131, Alpha=@n135), Alpha=NamedRerouteUsage())
   Alpha=ScalarParameter() {group=Animate, parameterName=AnimateVisual}
  )
  n137 = MaterialFunctionCall(
   UVs=TextureCoordinate() {coordinateIndex=4}
   "Rotation Angle (0-1)"=ScalarParameter() {defaultValue=-0.0125, group=Shadow, parameterName=ShadowAngle}
  ) {materialFunction=&CustomRotator}
  n138 = LinearInterpolate(A=OneMinus(None=@n137.Rotated Values), B=@n137.Rotated Values, Alpha=ScalarParameter() {group=Shadow, parameterName=FlipShadow})
  n139 = ScalarParameter() {defaultValue=-0.9, group=Shadow, parameterName=OffsetShadow}
  n140 = Multiply(A=ComponentMask(None=Add(A=@n138, B=@n139)) {r=true}, B=ScalarParameter() {defaultValue=2, group=Shadow, parameterName=ShadowEdge})
  n141 = LinearInterpolate(
   A=@n136
   B=LinearInterpolate(A=@n136, B=VectorParameter() {parameterName=ShadowColor, defaultValue=<rgba 0 0 0 1>}.RGB, Alpha=Saturate(None=@n140))
   Alpha=ScalarParameter() {group=Animate, parameterName=AnimateShadow}
  )
  n142 = VectorParameter() {group=Crop, parameterName=CropFadeColor, defaultValue=<rgba 0 0 0 1>}
  n143 = NamedRerouteUsage()
  StaticSwitchParameter(
   True=NamedRerouteDeclaration(None=LinearInterpolate(A=@n100, B=@n104, Alpha=@n105)) {name=Fadedicon, nodeColor=<rgba 0.8984 0 1 1>}
   False=LinearInterpolate(A=@n141, B=LinearInterpolate(A=@n142.RGB, B=@n141, Alpha=@n143), Alpha=ScalarParameter() {group=Crop, parameterName=UseCropFadeColor})
  ) {parameterName=JustIcon}
 ]
}
