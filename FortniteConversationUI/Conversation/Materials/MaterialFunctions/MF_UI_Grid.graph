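# Material function for Fortnite conversation UI that outputs a procedural line mask with primary and secondary grid layers.
# UVs are fitted by MF_UI_UVScaling, rotated, center-scaled and multiplied up, with a further multiplier on the primary layer.
# Each layer runs offset UVs through sine and a thickness exponent, then doubles only the R component, producing lines along one axis.
# Hardness weights both layers before they are summed and saturated.
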MaterialFunction /FortniteConversationUI/Conversation/Materials/MaterialFunctions/MF_UI_Grid
graph {nodes [
  n1 = FunctionInput(Preview=Constant() {r=0.25}) {bUsePreviewValueAsDefault=true, inputName=Primary Grid Multiplier, inputType=FunctionInput_Scalar, sortPriority=30}
  n2 = MaterialFunctionCall(UV=FunctionInput(Preview=TextureCoordinate()) {bUsePreviewValueAsDefault=true, inputName=UVs, inputType=FunctionInput_Vector2}) {materialFunction=&MF_UI_UVScaling}
  n3 = FunctionInput(Preview=Constant() {r=0.0025}) {bUsePreviewValueAsDefault=true, inputName=Rotation, inputType=FunctionInput_Scalar, sortPriority=32}
  n4 = Constant2Vector() {g=1, r=1}
  n5 = Multiply(A=MaterialFunctionCall(
    UVs=MaterialFunctionCall(UVs=@n2.Scaled to Fit, "Rotation Angle (0-1)"=@n3) {materialFunction=&CustomRotator}.Rotated Values
    Texture Scale=FunctionInput(Preview=@n4) {bUsePreviewValueAsDefault=true, inputName=Scale, inputType=FunctionInput_Scalar, sortPriority=31}
   ) {materialFunction=&ScaleUVsByCenter}.UVs) {constB=25}
  n6 = Constant2Vector() {g=0.05, r=0.05}
  n7 = Sine(None=Subtract(
    A=Multiply(A=@n1, B=@n5) {constA=0.25, constB=0.25}
    B=FunctionInput(Preview=@n6) {bUsePreviewValueAsDefault=true, inputName=Primary Grid Offset, inputType=FunctionInput_Vector2, sortPriority=5}
   ))
  n8 = FunctionInput(Preview=Constant() {r=96}) {bUsePreviewValueAsDefault=true, inputName=Primary Grid Thickness, inputType=FunctionInput_Scalar, sortPriority=1}
  n9 = MaterialFunctionCall(Float2=Saturate(None=Power(Base=@n7, Exp=@n8))) {materialFunction=&BreakOutFloat2Components}
  n10 = Multiply(
   A=Saturate(None=Add(A=@n9.R, B=@n9.R))
   B=FunctionInput(Preview=Constant() {r=16}) {bUsePreviewValueAsDefault=true, inputName=Primary Grid Hardness, inputType=FunctionInput_Scalar, sortPriority=3}
  )
  n11 = Constant2Vector()
  n12 = Sine(None=Subtract(
    A=@n5
    B=FunctionInput(Preview=@n11) {bUsePreviewValueAsDefault=true, inputName=Secondary Grid Offset, inputType=FunctionInput_Vector2, sortPriority=6}
   ))
  n13 = FunctionInput(Preview=Constant() {r=24}) {bUsePreviewValueAsDefault=true, inputName=Secondary Grid Thickness, inputType=FunctionInput_Scalar, sortPriority=1}
  n14 = MaterialFunctionCall(Float2=Saturate(None=Power(Base=@n12, Exp=@n13))) {materialFunction=&BreakOutFloat2Components}
  n15 = Multiply(
   A=Saturate(None=Add(A=@n14.R, B=@n14.R))
   B=FunctionInput(Preview=Constant() {r=0.5}) {bUsePreviewValueAsDefault=true, inputName=Secondary Grid Hardness, inputType=FunctionInput_Scalar, sortPriority=4}
  )
  FunctionOutput(None=Saturate(None=Add(A=@n10, B=@n15)))
 ]}
