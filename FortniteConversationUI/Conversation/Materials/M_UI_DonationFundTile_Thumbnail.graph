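# Translucent UI material for a donation fund tile thumbnail, compositing a fill-scaled image with stylized, parameterized color layers.
# A circle-SDF two-color gradient overlays the image, followed by tinted noise splash, a rotated two-scale grid with optional gradient masking, and a linear depth gradient.
# Named reroutes carry each layer into a chain of lerps, while a rotated-UV crop mask with adjustable edge sharpness provides slanted cropping.
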Material /FortniteConversationUI/Conversation/Materials/M_UI_DonationFundTile_Thumbnail
graph {
 properties {blendMode=BLEND_Translucent, materialDomain=MD_UI, pixelDepthOffsetMode=PDOM_Legacy, usedShadingModels=Default Lit}
 nodes [
  n1 = NamedRerouteUsage()
  n2 = LinearInterpolate(
   A=LinearInterpolate(
    A=NamedRerouteUsage()
    B=NamedRerouteUsage()
    Alpha=ScalarParameter() {group=2 - Overlay, parameterName=Overlay Opacity, sliderMax=1, sortPriority=2}
   )
   B=NamedRerouteUsage()
   Alpha=Multiply(A=OneMinus(None=@n1), B=ScalarParameter() {group=2 - Overlay, parameterName=Overlay Edge Opacity, sliderMax=1, sortPriority=3})
  )
  n3 = VectorParameter() {group=1 - Grunge, parameterName=Splash Color, sortPriority=1, defaultValue=<rgba 0 0.1686 0.2863 1>}
  n4 = Multiply(A=NamedRerouteUsage(), B=ScalarParameter() {defaultValue=1, group=1 - Grunge, parameterName=Splash Opacity, sliderMax=1, sortPriority=2})
  n5 = NamedRerouteUsage()
  n6 = ScalarParameter() {group=3 - Grid, parameterName=Grid Masked by Gradient, sliderMax=1, sortPriority=2}
  n7 = Multiply(A=NamedRerouteUsage(), B=OneMinus(None=Multiply(A=@n5, B=@n6)))
  n8 = ScalarParameter() {defaultValue=0.25, group=3 - Grid, parameterName=Grid Opacity, sliderMax=1, sortPriority=1}
  n9 = VectorParameter() {group=5 - Depth Gradient, parameterName=Depth Gradient Color, sortPriority=0, defaultValue=<rgba 1 0 0 1>}
  LinearInterpolate(
   A=LinearInterpolate(
    A=LinearInterpolate(A=@n2, B=@n3.RGB, Alpha=@n4)
    B=VectorParameter() {group=3 - Grid, parameterName=Grid Color, sortPriority=0, defaultValue=<rgba 1 1 1 1>}.RGB
    Alpha=Multiply(A=@n7, B=@n8)
   )
   B=@n9.RGB
   Alpha=NamedRerouteUsage()
  )
  n10 = VectorParameter() {group=2 - Overlay, parameterName=Gradient Color A, sortPriority=0, defaultValue=<rgba 0 0.1063 0.2552 1>}
  n11 = VectorParameter() {group=2 - Overlay, parameterName=Gradient Color B, sortPriority=1, defaultValue=<rgba 0 0.4072 1 1>}
  n12 = MaterialFunctionCall() {materialFunction=&MF_UI_UVScaling}
  n13 = SmoothStep(
   Min=ScalarParameter() {defaultValue=0.5, group=2 - Overlay, parameterName=Gradient Radius, sortPriority=4}
   Max=ScalarParameter() {defaultValue=-0.5, group=2 - Overlay, parameterName=Gradient Density, sortPriority=5}
   Value=MaterialFunctionCall(
    UVs=NamedRerouteDeclaration(None=@n12.Scaled to Fit) {name=Scaled UVs, nodeColor=<rgba 0 0 0 1>}
    Size=ScalarParameter() {defaultValue=0.5, group=2 - Overlay, parameterName=Gradient Scale, sortPriority=6}
   ) {materialFunction=&MF_UI_Circle}.SDF
  )
  NamedRerouteDeclaration(None=LinearInterpolate(A=@n10.RGB, B=@n11.RGB, Alpha=NamedRerouteDeclaration(None=@n13) {name=Gradient Mask, nodeColor=<rgba 0 0 0 1>})) {name=Gradient Overlay, nodeColor=<rgba 0 0 0 1>}
  n14 = TextureSample(
   UVs=MaterialFunctionCall(UVs=NamedRerouteUsage(), Texture Scale=ScalarParameter() {defaultValue=6, group=1 - Grunge, parameterName=Splash Scale, sortPriority=4}) {materialFunction=&ScaleUVsByCenter}.UVs
   Tex=TextureObjectParameter() {group=1 - Grunge, parameterName=Splash, sortPriority=0, texture=/Game/Global/Textures/Default/Noise/T_UI_Ice_Noise}
  ) {texture=&T_UI_Ice_Noise}
  n15 = ScalarParameter() {defaultValue=3, group=1 - Grunge, parameterName=Splash Intensity, sortPriority=3}
  NamedRerouteDeclaration(None=Saturate(None=Multiply(A=@n14.R, B=@n15) {constB=16})) {name=Splash, nodeColor=<rgba 0 0 0 1>}
  n16 = ScalarParameter() {defaultValue=0.05, group=3 - Grid, parameterName=Large Grid Offset X, sortPriority=6}
  n17 = ScalarParameter() {defaultValue=0.05, group=3 - Grid, parameterName=Large Grid Offset Y, sortPriority=7}
  n18 = ScalarParameter() {defaultValue=-0.025, group=3 - Grid, parameterName=Small Grid Offset X, sortPriority=8}
  n19 = ScalarParameter() {defaultValue=-0.025, group=3 - Grid, parameterName=Small Grid Offset Y, sortPriority=9}
  NamedRerouteDeclaration(None=MaterialFunctionCall!(
    UVs
    Primary Grid Thickness=ScalarParameter() {defaultValue=96, group=3 - Grid, parameterName=Large Grid Thickness, sortPriority=4}
    Secondary Grid Thickness=ScalarParameter() {defaultValue=24, group=3 - Grid, parameterName=Small Grid Thickness, sortPriority=5}
    Primary Grid Hardness
    Secondary Grid Hardness
    Primary Grid Offset=AppendVector(A=@n16, B=@n17)
    Secondary Grid Offset=AppendVector(A=@n18, B=@n19)
    Primary Grid Multiplier
    Scale=ScalarParameter() {defaultValue=1, group=3 - Grid, parameterName=Grid Scale, sortPriority=2}
    Rotation=ScalarParameter() {defaultValue=0.0025, group=3 - Grid, parameterName=Grid Rotation, sliderMax=1, sortPriority=3}
   ) {materialFunction=/FortniteConversationUI/Conversation/Materials/MaterialFunctions/MF_UI_Grid}.Result) {name=Grid, nodeColor=<rgba 0 0 0 1>}
  n20 = TextureObjectParameter() {group=0 - Thumbnail, parameterName=Thumbnail, sortPriority=0, texture=/Game/Athena/UI/Frontend/Spatial/Art/T_Ui_Empty}
  n21 = MaterialFunctionCall(TextureSize=TextureProperty(Tex=@n20)) {materialFunction=&MF_UI_UVScaling}
  n22 = ScalarParameter() {defaultValue=1, group=0 - Thumbnail, parameterName=Thumbnail Scale, sortPriority=1}
  NamedRerouteDeclaration(None=TextureSample(
    UVs=MaterialFunctionCall(UVs=@n21.Scaled To Fill, Texture Scale=@n22) {materialFunction=&ScaleUVsByCenter}.UVs
    Tex=@n20
   ).RGB) {name=Thumbnail, nodeColor=<rgba 0 0 0 1>}
  n23 = MaterialFunctionCall(UVs=@n12.ScaleX, "Rotation Angle (0-1)"=ScalarParameter() {defaultValue=-0.0063, group=4 - Crop Mask, parameterName=CropAngle}) {materialFunction=&CustomRotator}
  n24 = ScalarParameter() {defaultValue=-0.01, group=4 - Crop Mask, parameterName=OffsetCrop}
  n25 = Add(A=@n23.Rotated Values, B=@n24)
  n26 = ScalarParameter() {defaultValue=28, group=4 - Crop Mask, parameterName=CropEdge}
  n27 = OneMinus(None=Subtract(A=@n23.Rotated Values, B=@n24))
  n28 = Multiply(A=Multiply(A=ComponentMask(None=@n25) {r=true}, B=@n26), B=Multiply(A=@n26, B=ComponentMask(None=@n27) {r=true}))
  NamedRerouteDeclaration(None=LinearInterpolate(
    A=Constant() {r=1}
    B=Saturate(None=@n28)
    Alpha=ScalarParameter() {defaultValue=1, group=4 - Crop Mask, parameterName=Mask Opacity, sliderMax=1}
   )) {name=Crop Mask, nodeColor=<rgba 0 0 0 1>}
  n29 = Power(
   Base=MaterialFunctionCall(
    GradientStart=ScalarParameter() {group=5 - Depth Gradient, parameterName=Gradient Start, sortPriority=2}
    GradientEnd=ScalarParameter() {defaultValue=1, group=5 - Depth Gradient, parameterName=Gradient End, sortPriority=3}
    RotationAngle=Constant() {r=0.75}
    GradientOffset=ScalarParameter() {group=5 - Depth Gradient, parameterName=Gradient Offset, sortPriority=3}
   ) {materialFunction=&MF_UI_LinearGradient}.Result
   Exp=ScalarParameter() {defaultValue=2, group=5 - Depth Gradient, parameterName=Gradient Decay Rate, sortPriority=3}
  )
  NamedRerouteDeclaration(None=Multiply(
    A=Saturate(None=@n29)
    B=ScalarParameter() {defaultValue=1, group=5 - Depth Gradient, parameterName=Depth Opacity, sliderMax=1, sortPriority=1}
   )) {name=Depth Gradient, nodeColor=<rgba 0 0 0 1>}
  NamedRerouteUsage()
 ]
}
