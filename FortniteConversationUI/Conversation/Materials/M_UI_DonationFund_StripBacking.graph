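# Translucent UI material for the donation fund strip backing in the conversation UI, built around a two-tone circular shape.
# An inner exponential radial gradient, with adjustable radius and hardness, blends RingColor at the edge into FillColor at the center.
# A second radial gradient with its own hardness forms the outer circle mask.
# A texture parameter is also sampled, though no connection from it is shown.
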Material /FortniteConversationUI/Conversation/Materials/M_UI_DonationFund_StripBacking
graph {
 properties {blendMode=BLEND_Translucent, materialDomain=MD_UI, pixelDepthOffsetMode=PDOM_Legacy, usedShadingModels=Default Lit}
 nodes [
  n1 = ScalarParameter() {defaultValue=0.4, parameterName=InnerRadius}
  n2 = ScalarParameter() {defaultValue=24, parameterName=InnerHardness}
  LinearInterpolate(
   A=VectorParameter() {parameterName=RingColor, defaultValue=<rgba 1 0 0 1>}.RGB
   B=VectorParameter() {parameterName=FillColor, defaultValue=<rgba 0 1 0 1>}.RGB
   Alpha=Saturate(None=MaterialFunctionCall(Radius=@n1, Density=@n2) {materialFunction=&RadialGradientExponential}.RadialGradientExponential)
  )
  Saturate(None=MaterialFunctionCall(Density=ScalarParameter() {defaultValue=24, parameterName=OuterHardness}) {materialFunction=&RadialGradientExponential}.RadialGradientExponential)
  TextureSample(Tex=TextureObjectParameter() {parameterName=Param})
 ]
}
