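# Translucent UI master material for a faded background band, named for the donation screen in the conversation UI.
# A period-2 sine of the UV's vertical component forms a gradient that peaks mid-height and fades toward top and bottom.
# Multiply and Taper parameters scale and sharpen that falloff, which FinalAlpha then scales into a clamped mask.
# A separate Color parameter, black by default, supplies the tint.
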Material /FortniteConversationUI/Conversation/Materials/M_UI_Donation_FadedBG_Master
graph {
 properties {blendMode=BLEND_Translucent, materialDomain=MD_UI, pixelDepthOffsetMode=PDOM_Legacy, usedShadingModels=Default Lit}
 nodes [
  n1 = ComponentMask(None=Sine(None=TextureCoordinate()) {period=2}) {g=true}
  n2 = ScalarParameter() {defaultValue=1, parameterName=Multiply}
  n3 = Power(Base=Saturate(None=Multiply(A=@n1, B=@n2)), Exp=ScalarParameter() {defaultValue=1, parameterName=Taper})
  Saturate(None=Multiply(A=Saturate(None=@n3), B=ScalarParameter() {parameterName=FinalAlpha}))
  VectorParameter() {parameterName=Color, defaultValue=<rgba 0 0 0 1>}
 ]
}
