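# Translucent UI material for the conversation Fund Blueprint display, drawing noise-modulated blueprint guide lines or a progress bar.
# Ice noise sets a base mask, optionally multiplied by the guide-line texture's alpha through the UseTexture switch.
# With IsProgressBar on, a Progress-driven cutoff, screen-aligned rotated sine hash lines distorted by noise and a rounded-square outline shape that mask.
# FinalAlpha scales the result and AnimateOff fades it to zero.
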Material /FortniteConversationUI/Conversation/Materials/M_UI_FundBlueprint_Texture
graph {
 properties {blendMode=BLEND_Translucent, materialDomain=MD_UI, pixelDepthOffsetMode=PDOM_Legacy, usedShadingModels=Default Lit}
 nodes [
  n1 = MaterialFunctionCall() {materialFunction=&GetUserInterfaceUV}
  n2 = MaterialFunctionCall(Float2=@n1.Pixel Size) {materialFunction=&BreakOutFloat2Components}
  n3 = Divide(A=@n2.R, B=@n2.R)
  n4 = Constant() {r=1}
  n5 = OneMinus(None=@n3)
  Add(A=Multiply(A=@n1.Normalized UV, B=AppendVector(A=@n3, B=@n4)), B=AppendVector(A=Multiply(A=@n5) {constB=0.5}, B=Constant()))
  n6 = TextureSample(UVs=MaterialFunctionCall(UVs=TextureCoordinate(), Texture Scale=ScalarParameter() {defaultValue=6, group=Noise, parameterName=Noise_Scale}) {materialFunction=&ScaleUVsByCenter}.UVs) {texture=/Game/Global/Textures/Default/Noise/T_UI_Ice_Noise}
  n7 = LinearInterpolate(A=LinearInterpolate(Alpha=ScalarParameter() {parameterName=NoiseAmount}) {constA=0.5, constB=0}, Alpha=@n6.R)
  n8 = TextureSample() {texture=&/T_UI_BlueprintGuideLines}
  n9 = StaticSwitchParameter(True=Saturate(None=DotProduct(A=@n7, B=@n8.A)), False=@n7) {parameterName=UseTexture}
  n10 = MaterialFunctionCall() {materialFunction=&GetUserInterfaceUV}
  n11 = Add(A=Add(A=OneMinus(None=@n10.Normalized UV)) {constB=-1}, B=ScalarParameter() {defaultValue=0.25, parameterName=Progress}) {constB=-1}
  n12 = Saturate(None=Multiply(A=ComponentMask(None=@n11) {r=true}, B=ScalarParameter() {defaultValue=512, parameterName=ProgressEdge}))
  n13 = ScalarParameter() {defaultValue=0.5, group=ShadeLines, parameterName=ShadeLines_Lift}
  n14 = Add(
   A=MaterialFunctionCall(
    UVs=MaterialFunctionCall() {materialFunction=&ScreenAlignedUVs}."X 100%, Y 100%"
    "Rotation Angle (0-1)"=ScalarParameter() {defaultValue=0.075, group=ShadeLines, parameterName=ShadeLines_Rotation}
   ) {materialFunction=&CustomRotator}.Rotated Values
   B=Multiply(A=@n6.R, B=ScalarParameter() {defaultValue=0.0125, group=ShadeLines, parameterName=ShadeLines_Distortion})
  )
  n15 = ScalarParameter() {defaultValue=40, group=ShadeLines, parameterName=ShadeLines_Amount}
  n16 = Power(Base=Sine(None=Multiply(A=@n14, B=@n15)), Exp=ScalarParameter() {defaultValue=3, group=ShadeLines, parameterName=ShadeLines_Thickness})
  n17 = Saturate(None=MaterialFunctionCall(Float2=Saturate(None=@n16)) {materialFunction=&BreakOutFloat2Components}.R)
  n18 = Saturate(None=DotProduct(A=@n12, B=LinearInterpolate(A=@n13, Alpha=@n17)))
  n19 = ScalarParameter() {defaultValue=1, parameterName=UseHashLines}
  n20 = MaterialFunctionCall!(UV=TextureCoordinate(), Size=ScalarParameter() {defaultValue=1.125, parameterName=OutlineSize}, Curvature) {materialFunction=/Game/Effects/Fort_Effects/MaterialFunctions/MF_RoundedSquare}
  n21 = Multiply(A=OneMinus(None=Saturate(None=@n20.Linear Gradient))) {constB=48}
  n22 = Add(
   A=LinearInterpolate(A=ScalarParameter() {defaultValue=0.25, parameterName=ProgressBarEmptyFIllAmount}, Alpha=LinearInterpolate(A=@n12, B=@n18, Alpha=@n19))
   B=Multiply(A=Saturate(None=@n21), B=ScalarParameter() {defaultValue=0.25, parameterName=OutlineAmount})
  )
  n23 = Saturate(None=DotProduct(A=@n9, B=Saturate(None=@n22)))
  LinearInterpolate(
   A=Multiply(A=StaticSwitchParameter(True=@n23, False=@n9) {parameterName=IsProgressBar}, B=ScalarParameter() {defaultValue=1, parameterName=FinalAlpha})
   Alpha=ScalarParameter() {parameterName=AnimateOff}
  ) {constB=0}
  TextureCoordinate()
  VectorParameter() {parameterName=Color, defaultValue=<rgba 1 1 1 1>}
 ]
}
