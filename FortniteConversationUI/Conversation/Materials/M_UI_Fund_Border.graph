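# Translucent UI material for the conversation UI's fund border, drawing a backing shape with a shine swipe.
# Four rotated, offset UV edge gradients multiply into an adjustable quadrilateral mask, kept crisp by a resolution sharpness correction function.
# The shine is a tapered band from rotated UI UVs, moved by the OffsetShine parameter rather than time.
# That band blends backing and shine colors and alphas, with the shape masking the alpha.
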Material /FortniteConversationUI/Conversation/Materials/M_UI_Fund_Border
graph {
 properties {blendMode=BLEND_Translucent, materialDomain=MD_UI, pixelDepthOffsetMode=PDOM_Legacy, usedShadingModels=Default Lit}
 nodes [
  LinearInterpolate(
   A=VectorParameter() {parameterName=BackingColor, defaultValue=<rgba 1 1 1 1>}.RGB
   B=VectorParameter() {parameterName=ShineColor, defaultValue=<rgba 1 1 1 1>}.RGB
   Alpha=NamedRerouteUsage()
  )
  n1 = MaterialFunctionCall() {materialFunction=&GetUserInterfaceUV}
  n2 = LinearInterpolate(
   A=ScalarParameter() {defaultValue=0.5, group=Shine, parameterName=ShineOffsetStart}
   B=ScalarParameter() {defaultValue=-0.5, group=Shine, parameterName=ShineOffsetStop}
   Alpha=ScalarParameter() {group=Shine, parameterName=OffsetShine}
  )
  n3 = MaterialFunctionCall(
   UVs=Multiply(A=Add(A=@n1.Normalized UV, B=@n2), B=ScalarParameter() {defaultValue=2, group=Shine, parameterName=ShineWidth})
   "Rotation Angle (0-1)"=ScalarParameter() {defaultValue=-0.0625, group=Shine, parameterName=ShineRotation}
  ) {materialFunction=&CustomRotator}
  n4 = OneMinus(None=@n3.Rotated Values)
  n5 = Multiply(
   A=DotProduct(A=ComponentMask(None=@n4) {r=true}, B=ComponentMask(None=@n3.Rotated Values) {r=true})
   B=ScalarParameter() {defaultValue=3, group=Shine, parameterName=ShineStrength}
  )
  NamedRerouteDeclaration(None=Power(Base=Saturate(None=@n5), Exp=ScalarParameter() {defaultValue=1.22, group=Shine, parameterName=ShineTaper})) {name=ShineSwipe, nodeColor=<rgba 0 0.9062 1 1>}
  n6 = TextureCoordinate() {coordinateIndex=4}
  n7 = ScalarParameter() {defaultValue=-0.245, group=OverallShape, parameterName=Shape_Rotation_Top}
  n8 = ComponentMask(None=Add(
    A=Rotator(Coordinate=@n6, Time=@n7) {speed=6.2832}
    B=ScalarParameter() {defaultValue=-0.075, group=OverallShape, parameterName=Shape_Offset_Top}
   )) {r=true}
  n9 = MaterialFunctionCall!(Native_U_PixelSize=ScalarParameter() {defaultValue=128, group=ResolutionSharpness, parameterName=NativePixelSize}) {materialFunction=/FortUILibrary/Foundation/Art_ToolKit/Tools/M_UI_Resolution_Sharpness_Correction}
  n10 = Multiply(A=@n9.Result, B=ScalarParameter() {defaultValue=64, group=OverallShape, parameterName=Shape_Sharpness_V})
  n11 = Saturate(None=Saturate(None=Multiply(A=@n8, B=@n10) {constB=32}))
  n12 = TextureCoordinate() {coordinateIndex=4}
  n13 = ScalarParameter() {defaultValue=-0.245, group=OverallShape, parameterName=Shape_Rotation_Bottom}
  n14 = Add(
   A=OneMinus(None=Rotator(Coordinate=@n12, Time=@n13) {speed=6.2832})
   B=ScalarParameter() {defaultValue=-0.075, group=OverallShape, parameterName=Shape_Offset_Bottom}
  )
  n15 = Saturate(None=Multiply(A=ComponentMask(None=@n14) {r=true}, B=@n10) {constB=32})
  n16 = Add(
   A=Rotator(Coordinate=TextureCoordinate(), Time=ScalarParameter() {defaultValue=-0.025, group=OverallShape, parameterName=Shape_Rotation_Left}) {speed=6.2832}
   B=ScalarParameter() {defaultValue=-0.0875, group=OverallShape, parameterName=Shape_Offset_Left}
  )
  n17 = Multiply(A=@n9.Result, B=ScalarParameter() {defaultValue=64, group=OverallShape, parameterName=Shape_Sharpness_U})
  n18 = Saturate(None=Multiply(A=ComponentMask(None=@n16) {r=true}, B=@n17) {constB=32})
  n19 = TextureCoordinate()
  n20 = ScalarParameter() {defaultValue=-0.025, group=OverallShape, parameterName=Shape_Rotation_Right}
  n21 = Add(
   A=OneMinus(None=Rotator(Coordinate=@n19, Time=@n20) {speed=6.2832})
   B=ScalarParameter() {defaultValue=-0.0875, group=OverallShape, parameterName=Shape_Offset_Right}
  )
  n22 = Saturate(None=Multiply(A=ComponentMask(None=@n21) {r=true}, B=@n17) {constB=32})
  n23 = DotProduct(A=Saturate(None=DotProduct(A=@n11, B=@n15)), B=Saturate(None=DotProduct(A=@n18, B=@n22)))
  n24 = ScalarParameter() {defaultValue=1, group=Alpha, parameterName=BackingAlpha}
  n25 = ScalarParameter() {defaultValue=1, group=Alpha, parameterName=ShineAlpha}
  n26 = NamedRerouteUsage()
  Saturate(None=DotProduct(A=Saturate(None=@n23), B=LinearInterpolate(A=@n24, B=@n25, Alpha=@n26)))
  TextureCoordinate()
 ]
}
