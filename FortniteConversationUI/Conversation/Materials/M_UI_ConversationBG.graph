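# Translucent UI material for the conversation background that produces an adjustable gradient overlay, vertical by default.
# GradientScaleOffset scales the UVs around a point, with R as the scale and G as both coordinates of the pivot.
# A static Mask parameter selects the UV channel, green by default, as a texture-free gradient.
# That gradient multiplies the alpha of OverlayColorOpacity, whose default color is near-black.
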Material /FortniteConversationUI/Conversation/Materials/M_UI_ConversationBG
graph {
 properties {blendMode=BLEND_Translucent, materialDomain=MD_UI, pixelDepthOffsetMode=PDOM_Legacy, usedShadingModels=Default Lit}
 nodes [
  n1 = VectorParameter() {parameterName=GradientScaleOffset, defaultValue=<rgba 1 0.5 1 0>}
  n2 = TextureCoordinate()
  n3 = MaterialFunctionCall(R=@n1.G, G=@n1.G, B=Constant() {r=1}) {materialFunction=&AppendMany}
  Multiply(
   A=VectorParameter() {parameterName=OverlayColorOpacity, defaultValue=<rgba 0.0052 0.0052 0.0056 1>}.A
   B=StaticComponentMaskParameter(None=MaterialFunctionCall!(Texture Scale=@n1.R, UVs=@n2, Offset XY Texturesize Z=@n3.RGB) {materialFunction=/Engine/Functions/Engine_MaterialFunctions01/Texturing/ScaleUVsAroundPoint}.UVs) {defaultG=true, parameterName=Mask}
  )
 ]
}
