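# Translucent UI material that, by its name, supplies a radial background for the mobile conversation interface.
# An exponential radial gradient is multiplied by a steep inverted vertical ramp around mid-height, noted as the bottom-line gradient.
# A separate branch applies Sine, Ceil and OneMinus to vertical UV, forming a hard step mask for the radial gradient without a texture.
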Material /FortniteConversationUI/Conversation/Materials/M_UI_MobileConversationBG_Radial
graph {
 properties {blendMode=BLEND_Translucent, materialDomain=MD_UI, pixelDepthOffsetMode=PDOM_Legacy, usedShadingModels=Default Lit}
 nodes [
  n1 = Add(A=ComponentMask(None=TextureCoordinate() {desc=gradiant for the bottom line}) {g=true}, B=Constant() {r=-0.5})
  n2 = Constant() {r=8}
  Multiply(
   A=MaterialFunctionCall() {materialFunction=&RadialGradientExponential}.RadialGradientExponential
   B=OneMinus(None=Multiply(A=@n1, B=@n2))
  )
  n3 = Sine(None=ComponentMask(None=TextureCoordinate() {desc=mask for the radial gradient}) {g=true})
  OneMinus(None=Multiply(A=Ceil(None=@n3), B=Constant3Vector() {constant=<rgba 1 1 1 0>}))
 ]
}
