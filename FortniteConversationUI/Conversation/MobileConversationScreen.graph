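# Mobile NPC conversation screen: a five-slot half-pie radial menu of dialog options in the bottom-right, with a details brief and confirm button.
# Fades in with a pop-up cue, and pushes a sound-mix suppressor while active.
# Populated options trigger a wheel scale-in and looping tap clue; selection animates the highlight, confirm button and, unless blocked, details fade.
# Hardware visibility borders separate touch and keybind controls; unavailable options play an error sound.
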WidgetBlueprint /FortniteConversationUI/Conversation/MobileConversationScreen
animations [
 NewOptions {
  binding_count=2
  length_seconds=0.3808
  object=&self:NewOptions
  sequence {bindings [
    Overlay_Wheel {
     id=#1
     possessed_class=Overlay
     tracks [{
       type=MovieScene2DTransformTrack
       sections [{
         type=MovieScene2DTransformSection
         channels [
          Translation.X {type=MovieSceneScriptingFloatChannel}
          Translation.Y {type=MovieSceneScriptingFloatChannel}
          Angle {type=MovieSceneScriptingFloatChannel}
          Scale.X {type=MovieSceneScriptingFloatChannel, keys [[0,0.8711],[4.615,1]]}
          Scale.Y {type=MovieSceneScriptingFloatChannel, keys [[0,0.8711],[4.615,1]]}
          Shear.X {type=MovieSceneScriptingFloatChannel}
          Shear.Y {type=MovieSceneScriptingFloatChannel}
         ]
        }]
      }]
    }
    RadialMenu_DialogOptions {
     id=#2
     possessed_class=FortSlottedRadialMenu_C
     tracks [{
       type=MovieScene2DTransformTrack
       sections [{
         type=MovieScene2DTransformSection
         channels [
          Translation.X {type=MovieSceneScriptingFloatChannel}
          Translation.Y {type=MovieSceneScriptingFloatChannel}
          Angle {type=MovieSceneScriptingFloatChannel}
          Scale.X {type=MovieSceneScriptingFloatChannel, keys [[0,0.8211],[7.615,1]]}
          Scale.Y {type=MovieSceneScriptingFloatChannel, keys [[0,0.8211],[7.615,1]]}
          Shear.X {type=MovieSceneScriptingFloatChannel}
          Shear.Y {type=MovieSceneScriptingFloatChannel}
         ]
        }]
      }]
    }
   ]}
 }
 OptionFocused {
  binding_count=1
  length_seconds=0.3
  object=&self:OptionFocused
  sequence {bindings [Image_RadialHighlight {
     id=#3
     possessed_class=Image
     tracks [
      {
       type=MovieScene2DTransformTrack
       sections [{
         type=MovieScene2DTransformSection
         channels [
          Translation.X {type=MovieSceneScriptingFloatChannel}
          Translation.Y {type=MovieSceneScriptingFloatChannel}
          Angle {type=MovieSceneScriptingFloatChannel}
          Scale.X {type=MovieSceneScriptingFloatChannel, keys [[0,0.9463],[6,1.05]]}
          Scale.Y {type=MovieSceneScriptingFloatChannel, keys [[0,0.9463],[6,1.05]]}
          Shear.X {type=MovieSceneScriptingFloatChannel}
          Shear.Y {type=MovieSceneScriptingFloatChannel}
         ]
        }]
      }
      {
       type=MovieSceneWidgetMaterialTrack
       sections [{type=MovieSceneParameterSection, channels [InnerRadius_0-1 {type=MovieSceneScriptingFloatChannel, keys [[0,0.6271],[6,0.6071]]}]}]
      }
     ]
    }]}
 }
 OptionRight {
  binding_count=1
  length_seconds=0.35
  object=&self:OptionRight
  sequence {bindings [Overlay_Details {
     id=#4
     possessed_class=Overlay
     tracks [{
       type=MovieScene2DTransformTrack
       sections [{
         type=MovieScene2DTransformSection
         channels [
          Translation.X {type=MovieSceneScriptingFloatChannel, keys [[0,-40],[7,0]]}
          Translation.Y {type=MovieSceneScriptingFloatChannel}
          Angle {type=MovieSceneScriptingFloatChannel}
          Scale.X {type=MovieSceneScriptingFloatChannel}
          Scale.Y {type=MovieSceneScriptingFloatChannel}
          Shear.X {type=MovieSceneScriptingFloatChannel}
          Shear.Y {type=MovieSceneScriptingFloatChannel}
         ]
        }]
      }]
    }]}
 }
 OptionIntro {
  binding_count=1
  length_seconds=0.2
  object=&self:OptionIntro
  sequence {bindings [Overlay_Details {
     id=#5
     possessed_class=Overlay
     tracks [{
       type=MovieSceneFloatTrack
       display_name=Render Opacity
       sections [{type=MovieSceneFloatSection, channels [None {type=MovieSceneScriptingFloatChannel, keys [[0,0],[4,1]]}]}]
      }]
    }]}
 }
 OptionLeft {
  binding_count=1
  length_seconds=0.35
  object=&self:OptionLeft
  sequence {bindings [Overlay_Details {
     id=#4
     possessed_class=Overlay
     tracks [{
       type=MovieScene2DTransformTrack
       sections [{
         type=MovieScene2DTransformSection
         channels [
          Translation.X {type=MovieSceneScriptingFloatChannel, keys [[0,40],[7,0]]}
          Translation.Y {type=MovieSceneScriptingFloatChannel}
          Angle {type=MovieSceneScriptingFloatChannel}
          Scale.X {type=MovieSceneScriptingFloatChannel}
          Scale.Y {type=MovieSceneScriptingFloatChannel}
          Shear.X {type=MovieSceneScriptingFloatChannel}
          Shear.Y {type=MovieSceneScriptingFloatChannel}
         ]
        }]
      }]
    }]}
 }
 Intro {
  binding_count=4
  length_seconds=0.5
  object=&self:Intro
  sequence {bindings [
    bg_blocker {
     id=#6
     possessed_class=Image
     tracks [{
       type=MovieSceneFloatTrack
       display_name=Render Opacity
       sections [{
         type=MovieSceneFloatSection
         end_frame=10
         start_frame=-1
         channels [None {type=MovieSceneScriptingFloatChannel, keys [[-0.0267,0],[10.034,1]]}]
        }]
      }]
    }
    Image_Background {
     id=#7
     possessed_class=Image
     tracks [{
       type=MovieSceneFloatTrack
       display_name=Render Opacity
       sections [{type=MovieSceneFloatSection, end_frame=10, start_frame=0, channels [None {type=MovieSceneScriptingFloatChannel, keys [[0,0],[10.034,1]]}]}]
      }]
    }
    radial_bg {
     id=#8
     possessed_class=Image
     tracks [{
       type=MovieSceneFloatTrack
       display_name=Render Opacity
       sections [{
         type=MovieSceneFloatSection
         end_frame=10
         start_frame=-1
         channels [None {type=MovieSceneScriptingFloatChannel, keys [[-0.0587,0],[10.002,1]]}]
        }]
      }]
    }
    bottom_Bg {
     id=#9
     possessed_class=Image
     tracks [{
       type=MovieSceneFloatTrack
       display_name=Render Opacity
       sections [{type=MovieSceneFloatSection, end_frame=10, start_frame=0, channels [None {type=MovieSceneScriptingFloatChannel, keys [[0,0],[10.034,1]]}]}]
      }]
    }
   ]}
 }
 TapClue {
  binding_count=1
  length_seconds=0.55
  object=&self:TapClue
  sequence {bindings [TouchIconClue {
     id=#10
     possessed_class=Image
     tracks [{
       type=MovieScene2DTransformTrack
       sections [{
         type=MovieScene2DTransformSection
         end_frame=11
         start_frame=0
         channels [
          Translation.X {type=MovieSceneScriptingFloatChannel, const [0,1]}
          Translation.Y {type=MovieSceneScriptingFloatChannel, const [0,1]}
          Angle {type=MovieSceneScriptingFloatChannel, const [0,1]}
          Scale.X {type=MovieSceneScriptingFloatChannel, keys [[0,1],[3,0.97],[11,1]]}
          Scale.Y {type=MovieSceneScriptingFloatChannel, keys [[0,1],[3,1.15],[11,1]]}
          Shear.X {type=MovieSceneScriptingFloatChannel, const [0,1]}
          Shear.Y {type=MovieSceneScriptingFloatChannel, const [0,1]}
         ]
        }]
      }]
    }]}
 }
 ConfirmationButtonIntro {
  binding_count=1
  length_seconds=0.25
  object=&self:ConfirmationButtonIntro
  sequence {bindings [SizeBox_0 {
     id=#11
     possessed_class=SizeBox
     tracks [
      {
       type=MovieSceneFloatTrack
       display_name=Render Opacity
       sections [{type=MovieSceneFloatSection, end_frame=5, start_frame=0, channels [None {type=MovieSceneScriptingFloatChannel, keys [[0,0],[3,0],[5,1]]}]}]
      }
      {
       type=MovieScene2DTransformTrack
       sections [{
         type=MovieScene2DTransformSection
         end_frame=5
         start_frame=0
         channels [
          Translation.X {type=MovieSceneScriptingFloatChannel, const [0,2]}
          Translation.Y {type=MovieSceneScriptingFloatChannel, keys [[0,0],[5,-72.1515]]}
          Angle {type=MovieSceneScriptingFloatChannel, const [0,1]}
          Scale.X {type=MovieSceneScriptingFloatChannel, const [1,1]}
          Scale.Y {type=MovieSceneScriptingFloatChannel, const [1,1]}
          Shear.X {type=MovieSceneScriptingFloatChannel, const [0,1]}
          Shear.Y {type=MovieSceneScriptingFloatChannel, const [0,1]}
         ]
        }]
      }
     ]
    }]}
 }
 ConfirmHover {
  binding_count=1
  length_seconds=0.2
  object=&self:ConfirmHover
  sequence {bindings [CircleBG {
     id=#12
     possessed_class=Image
     tracks [{
       type=MovieSceneColorTrack
       display_name=Color and Opacity
       sections [{
         type=MovieSceneColorSection
         end_frame=4
         start_frame=0
         channels [
          Color.R {type=MovieSceneScriptingFloatChannel, keys [[0,0.0273],[4,1]]}
          Color.G {type=MovieSceneScriptingFloatChannel, keys [[0,0.0273],[4,1]]}
          Color.B {type=MovieSceneScriptingFloatChannel, keys [[0,0.0273],[4,1]]}
          Color.A {type=MovieSceneScriptingFloatChannel, const [1,2]}
         ]
        }]
      }]
    }]}
 }
]
graphs [EventGraph {nodes [
   "On Hovered (MobileButtonConfirm)"() {is=title K2Node_ComponentBoundEvent}
   n1 = Get ConfirmHover() -> (ConfirmHover:Widget Animation Object Reference)
   > PlayAnimation(InAnimation=@n1)
   "On Unhovered (MobileButtonConfirm)"() {is=title K2Node_ComponentBoundEvent}
   > PlayAnimation(InAnimation=@n1, PlayMode=Reverse)
   Play Intro Audio() {is=title K2Node_CustomEvent}
   > PlaySound2D(Sound=/FortniteConversationUI/Sounds/Conversation_PopUp_Cue)
   Event OnOptionsPopulated() {is=title K2Node_Event}
   > PlayAnimation(InAnimation=Get NewOptions() -> (NewOptions:Widget Animation Object Reference))
   > PlayAnimation(InAnimation=Get TapClue() -> (TapClue:Widget Animation Object Reference), NumLoopsToPlay=0, PlaybackSpeed=0.8)
   n2 = Event OnSelectedItemAvailabilityChanged!() -> (OutputDelegate, then, bIsAvailable:bool) {is=title K2Node_Event}
   > SetContentColorAndOpacity!(
    execute
    self:Border Object Reference=Get KeybindBorder() -> (KeybindBorder:Border Object Reference)
    InContentColorAndOpacity:Linear Color Structure=Select(Option 0:Linear Color Structure=(R=0.0273,G=0.0273,B=0.0273,A=1), Option 1:Linear Color Structure=(R=1,G=1,B=1,A=1), Index=@n2.bIsAvailable) -> (ReturnValue:Linear Color Structure)
   ) -> (then)
   Event OnItemSelected!() -> (OutputDelegate, then, OriginalIndex:int=0) {is=title K2Node_Event}
   > n3 = Sequence()
   n4 = Get OptionFocused() -> (OptionFocused:Widget Animation Object Reference)
   > StopAnimation(InAnimation=@n4)
   > PlayAnimation(InAnimation=@n4)
   > PlayAnimation(InAnimation=Get ConfirmationButtonIntro() -> (ConfirmationButtonIntro:Widget Animation Object Reference))
   StopAnimation(execute=@n3.then_1, InAnimation=Get OptionRight() -> (OptionRight:Widget Animation Object Reference))
   > StopAnimation(InAnimation=Get OptionLeft() -> (OptionLeft:Widget Animation Object Reference))
   > n5 = Branch(Condition=ShouldBlockOptionIntroAnim!(self:Fort Conversation Screen Object Reference) -> (ReturnValue:bool))
   PlayAnimation(execute=@n5.else, InAnimation=Get OptionIntro() -> (OptionIntro:Widget Animation Object Reference))
   Event Construct()
   > Play Intro Audio() {is=title K2Node_CallFunction}
   > PlayAnimation(InAnimation=Get Intro())
   Event On Activated()
   > PushSoundMixModifier!(execute, InSoundMixModifier:Sound Mix Object Reference=/FortniteConversationUI/Sounds/Conversation_Suppress_MixMod) -> (then)
   Event On Deactivated()
   > PopSoundMixModifier!(execute, InSoundMixModifier:Sound Mix Object Reference=&Conversation_Suppress_MixMod) -> (then)
   Event OnConversationOptionUnavailable!() -> (OutputDelegate, then, SelectedEntry:Fort Radial Slot Object Reference) {is=title K2Node_Event}
   > PlaySound2D(Sound=/Game/Athena/Sounds/NPCs/Conversation/Conversation_Error_Cue)
   Event OnConversationStarted() {is=title K2Node_Event}
  ]}]
variables []
widgets {
 info {
  inheritedWidgetCount=14
  parentClass=/Script/FortniteConversationUI.MobileConversationScreen
  rootWidgetClass=&Overlay
  widgetCount=46
 }
 widgets [
  Overlay_0 {widgetClassPath=&Overlay, widget=Overlay_0}
  bg_blocker {
   widgetClassPath=&Image
   parent=Overlay_0
   widget=bg_blocker
   properties {visibility=SelfHitTestInvisible, colorAndOpacity=<rgba 0 0 0 0.2>, renderTransform {angle=0, scale=<1 -1>, shear=<0 0>, translation=<0 0>}}
   slot {class=OverlaySlot, properties {horizontalAlignment=HAlign_Fill, verticalAlignment=VAlign_Fill}}
  }
  VerticalBox_0 {
   widgetClassPath=&VerticalBox
   parent=Overlay_0
   widget=VerticalBox_0
   slot {class=OverlaySlot, properties {horizontalAlignment=HAlign_Fill, verticalAlignment=VAlign_Fill}}
  }
  Image_Background {
   widgetClassPath=&Image
   bInherited=true
   parent=VerticalBox_0
   widget=Image_Background
   properties {
    visibility=SelfHitTestInvisible
    brush {
     bIsDynamicallyLoaded=false
     drawAs=Image
     imageType=FullColor
     mirroring=NoMirror
     resourceName=None
     resourceObject=/FortniteConversationUI/Conversation/Materials/M_UI_ConversationBG
     tiling=NoTile
     imageSize=<32 32>
     margin=<ltrb 0 0 0 0>
     outlineSettings {
      bUseBrushTransparency=false
      roundingType=HalfHeightRadius
      width=0
      color {colorUseRule=UseColor_Specified, specifiedColor=<rgba 0 0 0 0>}
      cornerRadii=<0 0 0 0>
     }
     tintColor {colorUseRule=UseColor_Specified, specifiedColor=<rgba 1 1 1 1>}
     uVRegion {bIsValid=false, max=<0 0>, min=<0 0>}
    }
    colorAndOpacity=<rgba 1 1 1 0.8>
    renderTransform {angle=0, scale=<1 -1>, shear=<0 0>, translation=<0 0>}
   }
   slot {class=VerticalBoxSlot, properties {size {sizeRule=Fill, value=0.25}}}
  }
  Spacer_113 {
   widgetClassPath=&Spacer
   parent=VerticalBox_0
   widget=Spacer_113
   slot {class=VerticalBoxSlot, properties {size {sizeRule=Fill, value=0.5}}}
  }
  bottom_Bg {
   widgetClassPath=&Image
   parent=VerticalBox_0
   widget=bottom_Bg
   properties {
    visibility=SelfHitTestInvisible
    brush {
     bIsDynamicallyLoaded=false
     drawAs=Image
     imageType=FullColor
     mirroring=NoMirror
     resourceName=None
     resourceObject=&M_UI_ConversationBG
     tiling=NoTile
     imageSize=<32 32>
     margin=<ltrb 0 0 0 0>
     outlineSettings {
      bUseBrushTransparency=false
      roundingType=HalfHeightRadius
      width=0
      color {colorUseRule=UseColor_Specified, specifiedColor=<rgba 0 0 0 0>}
      cornerRadii=<0 0 0 0>
     }
     tintColor {colorUseRule=UseColor_Specified, specifiedColor=<rgba 1 1 1 1>}
     uVRegion {bIsValid=false, max=<0 0>, min=<0 0>}
    }
    colorAndOpacity=<rgba 1 1 1 0.8>
   }
   slot {class=VerticalBoxSlot, properties {size {sizeRule=Fill, value=0.25}}}
  }
  SafeZone_CloseTouch {
   widgetClassPath=&SafeZone
   bIsVariable=true
   parent=Overlay_0
   widget=SafeZone_CloseTouch
   slot {class=OverlaySlot, properties {horizontalAlignment=HAlign_Fill, verticalAlignment=VAlign_Fill}}
  }
  Overlay_2 {widgetClassPath=&Overlay, parent=SafeZone_CloseTouch, widget=Overlay_2, slot {class=SafeZoneSlot}}
  Fake_CloseButton {
   widgetClassPath=/FortUILibrary/Foundation/Buttons/CloseButton
   bIsVariable=true
   parent=Overlay_2
   widget=Fake_CloseButton
   properties {
    holdData=&CommonUI_HoldData_Base_C
    onButtonBaseClicked=(&self:WidgetTree.Fake_CloseButton.HandleOnButtonClicked)
    onButtonBaseDoubleClicked=(&self:WidgetTree.Fake_CloseButton.HandleOnButtonDoubleClicked)
    onButtonBaseHovered=(&self:WidgetTree.Fake_CloseButton.HandleOnButtonHovered)
    onButtonBaseUnhovered=(&self:WidgetTree.Fake_CloseButton.HandleOnButtonUnhovered)
    onSelectedChangedBase=(&self:WidgetTree.Fake_CloseButton.HandleOnSelectedChanged)
   }
   slot {class=OverlaySlot, properties {horizontalAlignment=HAlign_Right, padding=<ltrb 0 35 35 0>}}
  }
  VisibilityBorder_Touch {
   widgetClassPath=&CommonHardwareVisibilityBorder
   parent=Overlay_2
   widget=VisibilityBorder_Touch
   properties {visibilityQuery {
     autoDescription=" ANY( Input.Touch )"
     tokenStreamVersion=0
     userDescription=if using touch
     queryTokenStream [0,1,1,1,0]
     tagDictionary [{tagName=Input.Touch}]
    }}
   slot {class=OverlaySlot, properties {horizontalAlignment=HAlign_Fill, verticalAlignment=VAlign_Fill}}
  }
  CloseButtonArea {
   widgetClassPath=&Button
   bInherited=true
   parent=VisibilityBorder_Touch
   widget=CloseButtonArea
   properties {widgetStyle {
     clickedSlateSound {resourceObject=null}
     disabled {
      bIsDynamicallyLoaded=false
      drawAs=NoDrawType
      imageType=NoImage
      mirroring=NoMirror
      resourceName=None
      resourceObject=null
      tiling=NoTile
      imageSize=<0 0>
      margin=<ltrb 0 0 0 0>
      outlineSettings {
       bUseBrushTransparency=false
       roundingType=HalfHeightRadius
       width=0
       color {colorUseRule=UseColor_Specified, specifiedColor=<rgba 0 0 0 0>}
       cornerRadii=<0 0 0 0>
      }
      tintColor {colorUseRule=UseColor_Specified, specifiedColor=<rgba 1 1 1 1>}
      uVRegion {bIsValid=false, max=<0 0>, min=<0 0>}
     }
     disabledForeground {colorUseRule=UseColor_Specified, specifiedColor=<rgba 0.5271 0.5271 0.5271 1>}
     hovered {
      bIsDynamicallyLoaded=false
      drawAs=NoDrawType
      imageType=NoImage
      mirroring=NoMirror
      resourceName=None
      resourceObject=null
      tiling=NoTile
      imageSize=<32 32>
      margin=<ltrb 0 0 0 0>
      outlineSettings {
       bUseBrushTransparency=true
       roundingType=FixedRadius
       width=1
       color {colorUseRule=UseColor_Specified, specifiedColor=<rgba 0.7243 0.7243 0.7243 1>}
       cornerRadii=<4 4 4 4>
      }
      tintColor {colorUseRule=UseColor_Specified, specifiedColor=<rgba 0.7243 0.7243 0.7243 1>}
      uVRegion {bIsValid=false, max=<0 0>, min=<0 0>}
     }
     hoveredForeground {colorUseRule=UseColor_Specified, specifiedColor=<rgba 0 1 1 1>}
     hoveredSlateSound {resourceObject=null}
     normal {
      bIsDynamicallyLoaded=false
      drawAs=NoDrawType
      imageType=NoImage
      mirroring=NoMirror
      resourceName=None
      resourceObject=null
      tiling=NoTile
      imageSize=<32 32>
      margin=<ltrb 0 0 0 0>
      outlineSettings {
       bUseBrushTransparency=true
       roundingType=FixedRadius
       width=1
       color {colorUseRule=UseColor_Specified, specifiedColor=<rgba 0.6951 0.6951 0.6951 1>}
       cornerRadii=<4 4 4 4>
      }
      tintColor {colorUseRule=UseColor_Specified, specifiedColor=<rgba 0.4955 0.4955 0.4955 1>}
      uVRegion {bIsValid=false, max=<0 0>, min=<0 0>}
     }
     normalForeground {colorUseRule=UseColor_Specified, specifiedColor=<rgba 0 1 1 1>}
     normalPadding=<ltrb 12 1.5 12 1.5>
     pressed {
      bIsDynamicallyLoaded=false
      drawAs=NoDrawType
      imageType=NoImage
      mirroring=NoMirror
      resourceName=None
      resourceObject=null
      tiling=NoTile
      imageSize=<32 32>
      margin=<ltrb 0 0 0 0>
      outlineSettings {
       bUseBrushTransparency=true
       roundingType=FixedRadius
       width=1
       color {colorUseRule=UseColor_Specified, specifiedColor=<rgba 0.7243 0.7243 0.7243 1>}
       cornerRadii=<4 4 4 4>
      }
      tintColor {colorUseRule=UseColor_Specified, specifiedColor=<rgba 0.3843 0.3843 0.3843 1>}
      uVRegion {bIsValid=false, max=<0 0>, min=<0 0>}
     }
     pressedForeground {colorUseRule=UseColor_Specified, specifiedColor=<rgba 0 1 1 1>}
     pressedPadding=<ltrb 12 1.5 12 1.5>
     pressedSlateSound {resourceObject=null}
    }}
   slot {class=BorderSlot, properties {padding=<ltrb 0 0 0 0>}}
  }
  OV_BG_Radial {
   widgetClassPath=&Overlay
   parent=Overlay_0
   widget=OV_BG_Radial
   slot {class=OverlaySlot, properties {horizontalAlignment=HAlign_Right, verticalAlignment=VAlign_Bottom, padding=<ltrb 0 0 225 -100>}}
  }
  Overlay_WheelContainer_1 {
   widgetClassPath=&Overlay
   parent=OV_BG_Radial
   widget=Overlay_WheelContainer_1
   slot {class=OverlaySlot, properties {horizontalAlignment=HAlign_Center, verticalAlignment=VAlign_Bottom}}
  }
  SizeBox_RadialMenu_1 {
   widgetClassPath=&SizeBox
   parent=Overlay_WheelContainer_1
   widget=SizeBox_RadialMenu_1
   properties {bOverride_HeightOverride=true, bOverride_WidthOverride=true, heightOverride=400, widthOverride=400}
   slot {class=OverlaySlot, properties {horizontalAlignment=HAlign_Center, verticalAlignment=VAlign_Bottom}}
  }
  radial_bg {
   widgetClassPath=&Image
   parent=SizeBox_RadialMenu_1
   widget=radial_bg
   properties {
    visibility=SelfHitTestInvisible
    brush {
     bIsDynamicallyLoaded=false
     drawAs=Image
     imageType=NoImage
     mirroring=NoMirror
     resourceName=None
     resourceObject=/FortniteConversationUI/Conversation/Materials/M_UI_MobileConversationBG_Radial
     tiling=NoTile
     imageSize=<32 32>
     margin=<ltrb 0 0 0 0>
     outlineSettings {
      bUseBrushTransparency=false
      roundingType=HalfHeightRadius
      width=0
      color {colorUseRule=UseColor_Specified, specifiedColor=<rgba 0 0 0 0>}
      cornerRadii=<0 0 0 0>
     }
     tintColor {colorUseRule=UseColor_Specified, specifiedColor=<rgba 1 1 1 0.4>}
     uVRegion {bIsValid=false, max=<0 0>, min=<0 0>}
    }
    colorAndOpacity=<rgba 0 0 0 1>
    renderTransform {angle=0, scale=<1.5 1.5>, shear=<0 0>, translation=<0 0>}
   }
   slot {class=SizeBoxSlot}
  }
  Overlay_Details {
   widgetClassPath=&Overlay
   bInherited=true
   parent=Overlay_0
   widget=Overlay_Details
   slot {class=OverlaySlot, properties {horizontalAlignment=HAlign_Fill, verticalAlignment=VAlign_Fill}}
  }
  SizeBox_Details {
   widgetClassPath=&SizeBox
   parent=Overlay_Details
   widget=SizeBox_Details
   properties {bOverride_WidthOverride=true, widthOverride=852}
   slot {class=OverlaySlot, properties {horizontalAlignment=HAlign_Right, verticalAlignment=VAlign_Bottom, padding=<ltrb 0 0 0 350>}}
  }
  Switcher_Details {
   widgetClassPath=&CommonVisibilitySwitcher
   bInherited=true
   parent=SizeBox_Details
   widget=Switcher_Details
   properties {activeWidgetIndex=1}
   slot {class=SizeBoxSlot, properties {horizontalAlignment=HAlign_Center}}
  }
  Spacer_94 {widgetClassPath=&Spacer, parent=Switcher_Details, widget=Spacer_94, slot {class=CommonVisibilitySwitcherSlot}}
  BasicBrief_Main {
   widgetClassPath=&/FortBasicBrief
   bInherited=true
   parent=Switcher_Details
   widget=BasicBrief_Main
   slot {class=CommonVisibilitySwitcherSlot}
  }
  Switcher_CenteredDetails {
   widgetClassPath=&CommonVisibilitySwitcher
   bInherited=true
   parent=Overlay_Details
   widget=Switcher_CenteredDetails
   properties {activeWidgetIndex=1}
   slot {class=OverlaySlot, properties {horizontalAlignment=HAlign_Fill, verticalAlignment=VAlign_Fill}}
  }
  Spacer {widgetClassPath=&Spacer, parent=Switcher_CenteredDetails, widget=Spacer, slot {class=CommonVisibilitySwitcherSlot}}
  BasicBrief_Main_1 {
   widgetClassPath=&FortBasicBrief
   parent=Switcher_CenteredDetails
   widget=BasicBrief_Main_1
   slot {class=CommonVisibilitySwitcherSlot}
  }
  Overlay_ConversationContainer {
   widgetClassPath=&Overlay
   parent=Overlay_0
   widget=Overlay_ConversationContainer
   slot {class=OverlaySlot, properties {horizontalAlignment=HAlign_Right, verticalAlignment=VAlign_Bottom, padding=<ltrb 0 0 225 -80>}}
  }
  Overlay_WheelContainer {
   widgetClassPath=&Overlay
   bInherited=true
   parent=Overlay_ConversationContainer
   widget=Overlay_WheelContainer
   slot {class=OverlaySlot, properties {horizontalAlignment=HAlign_Center, verticalAlignment=VAlign_Bottom}}
  }
  SizeBox_RadialMenu {
   widgetClassPath=&SizeBox
   bInherited=true
   parent=Overlay_WheelContainer
   widget=SizeBox_RadialMenu
   properties {bOverride_HeightOverride=true, bOverride_WidthOverride=true, heightOverride=400, widthOverride=400}
   slot {class=OverlaySlot, properties {horizontalAlignment=HAlign_Center, verticalAlignment=VAlign_Bottom}}
  }
  Overlay_Wheel {widgetClassPath=&Overlay, parent=SizeBox_RadialMenu, widget=Overlay_Wheel, slot {class=SizeBoxSlot}}
  Image_Radial_BG {
   widgetClassPath=&Image
   parent=Overlay_Wheel
   widget=Image_Radial_BG
   properties {
    visibility=Collapsed
    brush {
     bIsDynamicallyLoaded=false
     drawAs=Image
     imageType=NoImage
     mirroring=NoMirror
     resourceName=None
     resourceObject=/Game/UI/Items/Materials/MI-AthenaPieShadow_Half
     tiling=NoTile
     imageSize=<128 128>
     margin=<ltrb 0 0 0 0>
     outlineSettings {
      bUseBrushTransparency=false
      roundingType=HalfHeightRadius
      width=0
      color {colorUseRule=UseColor_Specified, specifiedColor=<rgba 0 0 0 0>}
      cornerRadii=<0 0 0 0>
     }
     tintColor {colorUseRule=UseColor_Specified, specifiedColor=<rgba 1 1 1 1>}
     uVRegion {bIsValid=false, max=<0 0>, min=<0 0>}
    }
    colorAndOpacity=<rgba 0.0052 0.0052 0.0056 0.4>
    renderTransform {angle=0, scale=<1 1>, shear=<0 0>, translation=<10 10>}
   }
   slot {class=OverlaySlot, properties {horizontalAlignment=HAlign_Fill, verticalAlignment=VAlign_Fill}}
  }
  Image_Radial {
   widgetClassPath=&Image
   parent=Overlay_Wheel
   widget=Image_Radial
   properties {
    visibility=HitTestInvisible
    brush {
     bIsDynamicallyLoaded=false
     drawAs=Image
     imageType=NoImage
     mirroring=NoMirror
     resourceName=None
     resourceObject=/Game/UI/Items/Materials/MI-AthenaHalfPieMenu
     tiling=NoTile
     imageSize=<128 128>
     margin=<ltrb 0 0 0 0>
     outlineSettings {
      bUseBrushTransparency=false
      roundingType=HalfHeightRadius
      width=0
      color {colorUseRule=UseColor_Specified, specifiedColor=<rgba 0 0 0 0>}
      cornerRadii=<0 0 0 0>
     }
     tintColor {colorUseRule=UseColor_Specified, specifiedColor=<rgba 1 1 1 1>}
     uVRegion {bIsValid=false, max=<0 0>, min=<0 0>}
    }
    colorAndOpacity=<rgba 0.0273 0.0273 0.0273 1>
   }
   slot {class=OverlaySlot, properties {horizontalAlignment=HAlign_Fill, verticalAlignment=VAlign_Fill}}
  }
  Image_RadialHighlight {
   widgetClassPath=&Image
   bInherited=true
   parent=Overlay_Wheel
   widget=Image_RadialHighlight
   properties {
    visibility=Hidden
    brush {
     bIsDynamicallyLoaded=false
     drawAs=Image
     imageType=NoImage
     mirroring=NoMirror
     resourceName=None
     resourceObject=/Game/UI/Items/Materials/MI-AthenaPieMenu_Selection_FiveSlots
     tiling=NoTile
     imageSize=<32 32>
     margin=<ltrb 0 0 0 0>
     outlineSettings {
      bUseBrushTransparency=false
      roundingType=HalfHeightRadius
      width=0
      color {colorUseRule=UseColor_Specified, specifiedColor=<rgba 0 0 0 0>}
      cornerRadii=<0 0 0 0>
     }
     tintColor {colorUseRule=UseColor_Specified, specifiedColor=<rgba 1 1 1 1>}
     uVRegion {bIsValid=false, max=<0 0>, min=<0 0>}
    }
   }
   slot {class=OverlaySlot, properties {horizontalAlignment=HAlign_Fill, verticalAlignment=VAlign_Fill}}
  }
  RadialMenu_DialogOptions {
   widgetClassPath=/Game/Athena/HUD/ContextMenus/FortSlottedRadialMenu
   bInherited=true
   parent=Overlay_Wheel
   widget=RadialMenu_DialogOptions
   properties {
    bUseAsManySlotsAsItems=true
    entryWidgetClass=&/FortConversationOption.FortConversationOption_C
    joystickDeadZone=0.4
    mouseDeadZoneRadiusPercentage=0.3
    navigation=&self:WidgetTree.RadialMenu_DialogOptions.WidgetNavigation_0
    numSlotsSupported=5
    orderingMode=Custom
    visibility=Visible
    customOrdering [N,NNE,NE,ENE,E]
    radialBoxSettings {angleBetweenItems=0, bClockwiseOrder=false, bDistributeItemsEvenly=true, marginSize=0, sectorCentralAngle=180, startingAngle=0}
   }
   slot {class=OverlaySlot, properties {horizontalAlignment=HAlign_Fill, verticalAlignment=VAlign_Fill, padding=<ltrb 5 5 5 5>}}
  }
  Image_LoadingSpinner {
   widgetClassPath=&Image
   bInherited=true
   parent=Overlay_WheelContainer
   widget=Image_LoadingSpinner
   properties {
    visibility=Hidden
    brush {
     bIsDynamicallyLoaded=false
     drawAs=Image
     imageType=NoImage
     mirroring=NoMirror
     resourceName=None
     resourceObject=/FortUILibrary/UIKit/Throbber/MI_UIKit_Monochrome_Throbber
     tiling=NoTile
     imageSize=<50 50>
     margin=<ltrb 0 0 0 0>
     outlineSettings {
      bUseBrushTransparency=false
      roundingType=HalfHeightRadius
      width=0
      color {colorUseRule=UseColor_Specified, specifiedColor=<rgba 0 0 0 0>}
      cornerRadii=<0 0 0 0>
     }
     tintColor {colorUseRule=UseColor_Specified, specifiedColor=<rgba 1 1 1 1>}
     uVRegion {bIsValid=false, max=<0 0>, min=<0 0>}
    }
    colorAndOpacity=<rgba 1 1 1 0.4>
   }
   slot {class=OverlaySlot, properties {horizontalAlignment=HAlign_Center, padding=<ltrb 0 20 0 0>}}
  }
  VisibilityBorder_Keybind_NonTouch {
   widgetClassPath=&CommonHardwareVisibilityBorder
   parent=Overlay_WheelContainer
   widget=VisibilityBorder_Keybind_NonTouch
   properties {visibilityQuery {
     autoDescription=" NONE( Input.Touch )"
     tokenStreamVersion=0
     userDescription=""
     queryTokenStream [0,1,3,1,0]
     tagDictionary [{tagName=Input.Touch}]
    }}
   slot {class=OverlaySlot, properties {horizontalAlignment=HAlign_Center, verticalAlignment=VAlign_Center, padding=<ltrb 0 30 0 0>}}
  }
  KeybindBorder {
   widgetClassPath=&Border
   bIsVariable=true
   parent=VisibilityBorder_Keybind_NonTouch
   widget=KeybindBorder
   properties {
    background {
     bIsDynamicallyLoaded=false
     drawAs=Image
     imageType=NoImage
     mirroring=NoMirror
     resourceName=None
     resourceObject=null
     tiling=NoTile
     imageSize=<24 24>
     margin=<ltrb 0 0 0 0>
     outlineSettings {
      bUseBrushTransparency=false
      roundingType=HalfHeightRadius
      width=0
      color {colorUseRule=UseColor_Specified, specifiedColor=<rgba 0 0 0 0>}
      cornerRadii=<0 0 0 0>
     }
     tintColor {colorUseRule=UseColor_Specified, specifiedColor=<rgba 1 1 1 1>}
     uVRegion {bIsValid=false, max=<0 0>, min=<0 0>}
    }
    brushColor=<rgba 1 1 1 0>
    padding=<ltrb 0 0 0 0>
    renderTransform {angle=0, scale=<1 1>, shear=<0 0>, translation=<0 -59>}
   }
   slot {class=BorderSlot, properties {padding=<ltrb 0 0 0 0>}}
  }
  Keybind_Confirm {
   widgetClassPath=/FortUILibrary/Input/KeybindWidgetDark
   bInherited=true
   parent=KeybindBorder
   widget=Keybind_Confirm
   properties {boundAction=Use}
   slot {class=BorderSlot, properties {padding=<ltrb 0 0 0 0>}}
  }
  MobileButtonConfirm {
   widgetClassPath=&Button
   bInherited=true
   parent=Overlay_WheelContainer
   widget=MobileButtonConfirm
   properties {
    visibility=SelfHitTestInvisible
    renderTransform {angle=0, scale=<1 1>, shear=<0 0>, translation=<0 70>}
    widgetStyle {
     clickedSlateSound {resourceObject=null}
     disabled {
      bIsDynamicallyLoaded=false
      drawAs=NoDrawType
      imageType=NoImage
      mirroring=NoMirror
      resourceName=None
      resourceObject=null
      tiling=NoTile
      imageSize=<64 64>
      margin=<ltrb 0 0 0 0>
      outlineSettings {
       bUseBrushTransparency=false
       roundingType=HalfHeightRadius
       width=0
       color {colorUseRule=UseColor_Specified, specifiedColor=<rgba 0 0 0 0>}
       cornerRadii=<0 0 0 0>
      }
      tintColor {colorUseRule=UseColor_Specified, specifiedColor=<rgba 1 1 1 0>}
      uVRegion {bIsValid=false, max=<0 0>, min=<0 0>}
     }
     disabledForeground {colorUseRule=UseColor_Specified, specifiedColor=<rgba 0.5271 0.5271 0.5271 1>}
     hovered {
      bIsDynamicallyLoaded=false
      drawAs=NoDrawType
      imageType=NoImage
      mirroring=NoMirror
      resourceName=None
      resourceObject=null
      tiling=NoTile
      imageSize=<72 72>
      margin=<ltrb 0 0 0 0>
      outlineSettings {
       bUseBrushTransparency=true
       roundingType=FixedRadius
       width=1
       color {colorUseRule=UseColor_Specified, specifiedColor=<rgba 0.7243 0.7243 0.7243 1>}
       cornerRadii=<4 4 4 4>
      }
      tintColor {colorUseRule=UseColor_Specified, specifiedColor=<rgba 1 1 1 0>}
      uVRegion {bIsValid=false, max=<0 0>, min=<0 0>}
     }
     hoveredForeground {colorUseRule=UseColor_Specified, specifiedColor=<rgba 0 1 1 1>}
     hoveredSlateSound {resourceObject=null}
     normal {
      bIsDynamicallyLoaded=false
      drawAs=NoDrawType
      imageType=NoImage
      mirroring=NoMirror
      resourceName=None
      resourceObject=null
      tiling=NoTile
      imageSize=<72 72>
      margin=<ltrb 0 0 0 0>
      outlineSettings {
       bUseBrushTransparency=true
       roundingType=FixedRadius
       width=1
       color {colorUseRule=UseColor_Specified, specifiedColor=<rgba 0.6951 0.6951 0.6951 1>}
       cornerRadii=<4 4 4 4>
      }
      tintColor {colorUseRule=UseColor_Specified, specifiedColor=<rgba 1 1 1 0>}
      uVRegion {bIsValid=false, max=<0 0>, min=<0 0>}
     }
     normalForeground {colorUseRule=UseColor_Specified, specifiedColor=<rgba 0 1 1 1>}
     normalPadding=<ltrb 12 1.5 12 1.5>
     pressed {
      bIsDynamicallyLoaded=false
      drawAs=NoDrawType
      imageType=NoImage
      mirroring=NoMirror
      resourceName=None
      resourceObject=null
      tiling=NoTile
      imageSize=<72 72>
      margin=<ltrb 0 0 0 0>
      outlineSettings {
       bUseBrushTransparency=true
       roundingType=FixedRadius
       width=1
       color {colorUseRule=UseColor_Specified, specifiedColor=<rgba 0.7243 0.7243 0.7243 1>}
       cornerRadii=<4 4 4 4>
      }
      tintColor {colorUseRule=UseColor_Specified, specifiedColor=<rgba 1 1 1 0>}
      uVRegion {bIsValid=false, max=<0 0>, min=<0 0>}
     }
     pressedForeground {colorUseRule=UseColor_Specified, specifiedColor=<rgba 0 1 1 1>}
     pressedPadding=<ltrb 4 7 4 1.5>
     pressedSlateSound {resourceObject=null}
    }
   }
   slot {class=OverlaySlot, properties {horizontalAlignment=HAlign_Center, verticalAlignment=VAlign_Center}}
   bIsVariable=~
  }
  SizeBox_0 {
   widgetClassPath=&SizeBox
   parent=MobileButtonConfirm
   widget=SizeBox_0
   properties {bOverride_HeightOverride=true, bOverride_WidthOverride=true, heightOverride=140, renderOpacity=0, widthOverride=140}
   slot {class=ButtonSlot, properties {padding=<ltrb 0 0 0 0>}}
  }
  Overlay_1 {widgetClassPath=&Overlay, parent=SizeBox_0, widget=Overlay_1, slot {class=SizeBoxSlot}}
  CircleBG {
   widgetClassPath=&Image
   parent=Overlay_1
   widget=CircleBG
   properties {
    brush {
     bIsDynamicallyLoaded=false
     drawAs=Image
     imageType=NoImage
     mirroring=NoMirror
     resourceName=None
     resourceObject=/FortUILibrary/Foundation/Art_ToolKit/Tools/MI_UI_Circle_ConversationBacchusConfirm
     tiling=NoTile
     imageSize=<32 32>
     margin=<ltrb 0 0 0 0>
     outlineSettings {
      bUseBrushTransparency=false
      roundingType=HalfHeightRadius
      width=0
      color {colorUseRule=UseColor_Specified, specifiedColor=<rgba 0 0 0 0>}
      cornerRadii=<0 0 0 0>
     }
     tintColor {colorUseRule=UseColor_Specified, specifiedColor=<rgba 1 1 1 1>}
     uVRegion {bIsValid=false, max=<0 0>, min=<0 0>}
    }
    colorAndOpacity=<rgba 0.0273 0.0273 0.0273 1>
   }
   slot {class=OverlaySlot, properties {horizontalAlignment=HAlign_Fill, verticalAlignment=VAlign_Fill}}
  }
  Confirmation_Icon {
   widgetClassPath=&Image
   parent=Overlay_1
   widget=Confirmation_Icon
   properties {brush {
     bIsDynamicallyLoaded=false
     drawAs=Image
     imageType=NoImage
     mirroring=NoMirror
     resourceName=None
     resourceObject=/Game/Athena/HUD/Art/BacchusIcons/Confirm_Normal_XL
     tiling=NoTile
     imageSize=<95 85>
     margin=<ltrb 0 0 0 0>
     outlineSettings {
      bUseBrushTransparency=false
      roundingType=HalfHeightRadius
      width=0
      color {colorUseRule=UseColor_Specified, specifiedColor=<rgba 0 0 0 0>}
      cornerRadii=<0 0 0 0>
     }
     tintColor {colorUseRule=UseColor_Specified, specifiedColor=<rgba 0.1033 0.6302 0.1096 1>}
     uVRegion {bIsValid=false, max=<0 0>, min=<0 0>}
    }}
   slot {class=OverlaySlot, properties {horizontalAlignment=HAlign_Center, verticalAlignment=VAlign_Center}}
  }
  MobileButtonTouchInformation {
   widgetClassPath=&Button
   bInherited=true
   parent=Overlay_WheelContainer
   widget=MobileButtonTouchInformation
   properties {
    visibility=SelfHitTestInvisible
    widgetStyle {
     clickedSlateSound {resourceObject=null}
     disabled {
      bIsDynamicallyLoaded=false
      drawAs=NoDrawType
      imageType=NoImage
      mirroring=NoMirror
      resourceName=None
      resourceObject=null
      tiling=NoTile
      imageSize=<64 64>
      margin=<ltrb 0 0 0 0>
      outlineSettings {
       bUseBrushTransparency=false
       roundingType=HalfHeightRadius
       width=0
       color {colorUseRule=UseColor_Specified, specifiedColor=<rgba 0 0 0 0>}
       cornerRadii=<0 0 0 0>
      }
      tintColor {colorUseRule=UseColor_Specified, specifiedColor=<rgba 0.5 0.5 0.5 1>}
      uVRegion {bIsValid=false, max=<0 0>, min=<0 0>}
     }
     disabledForeground {colorUseRule=UseColor_Specified, specifiedColor=<rgba 0.5271 0.5271 0.5271 1>}
     hovered {
      bIsDynamicallyLoaded=false
      drawAs=Image
      imageType=NoImage
      mirroring=NoMirror
      resourceName=None
      resourceObject=null
      tiling=NoTile
      imageSize=<128 128>
      margin=<ltrb 0 0 0 0>
      outlineSettings {
       bUseBrushTransparency=true
       roundingType=FixedRadius
       width=1
       color {colorUseRule=UseColor_Specified, specifiedColor=<rgba 0.7243 0.7243 0.7243 1>}
       cornerRadii=<4 4 4 4>
      }
      tintColor {colorUseRule=UseColor_Specified, specifiedColor=<rgba 1 1 1 0>}
      uVRegion {bIsValid=false, max=<0 0>, min=<0 0>}
     }
     hoveredForeground {colorUseRule=UseColor_Specified, specifiedColor=<rgba 0 1 1 1>}
     hoveredSlateSound {resourceObject=null}
     normal {
      bIsDynamicallyLoaded=false
      drawAs=NoDrawType
      imageType=NoImage
      mirroring=NoMirror
      resourceName=None
      resourceObject=null
      tiling=NoTile
      imageSize=<128 128>
      margin=<ltrb 0 0 0 0>
      outlineSettings {
       bUseBrushTransparency=true
       roundingType=FixedRadius
       width=1
       color {colorUseRule=UseColor_Specified, specifiedColor=<rgba 0.6951 0.6951 0.6951 1>}
       cornerRadii=<4 4 4 4>
      }
      tintColor {colorUseRule=UseColor_Specified, specifiedColor=<rgba 1 1 1 0>}
      uVRegion {bIsValid=false, max=<0 0>, min=<0 0>}
     }
     normalForeground {colorUseRule=UseColor_Specified, specifiedColor=<rgba 0 1 1 1>}
     normalPadding=<ltrb 12 1.5 12 1.5>
     pressed {
      bIsDynamicallyLoaded=false
      drawAs=NoDrawType
      imageType=NoImage
      mirroring=NoMirror
      resourceName=None
      resourceObject=null
      tiling=NoTile
      imageSize=<128 128>
      margin=<ltrb 0 0 0 0>
      outlineSettings {
       bUseBrushTransparency=true
       roundingType=FixedRadius
       width=1
       color {colorUseRule=UseColor_Specified, specifiedColor=<rgba 0.7243 0.7243 0.7243 1>}
       cornerRadii=<4 4 4 4>
      }
      tintColor {colorUseRule=UseColor_Specified, specifiedColor=<rgba 1 1 1 0>}
      uVRegion {bIsValid=false, max=<0 0>, min=<0 0>}
     }
     pressedForeground {colorUseRule=UseColor_Specified, specifiedColor=<rgba 0 1 1 1>}
     pressedPadding=<ltrb 12 1.5 12 1.5>
     pressedSlateSound {resourceObject=null}
    }
   }
   slot {class=OverlaySlot, properties {horizontalAlignment=HAlign_Center, verticalAlignment=VAlign_Center}}
   bIsVariable=~
  }
  SizeBox_1 {
   widgetClassPath=&SizeBox
   parent=MobileButtonTouchInformation
   widget=SizeBox_1
   properties {bOverride_HeightOverride=true, bOverride_WidthOverride=true, heightOverride=110, widthOverride=110}
   slot {class=ButtonSlot, properties {padding=<ltrb 0 0 0 0>}}
  }
  Overlay_81 {widgetClassPath=&Overlay, parent=SizeBox_1, widget=Overlay_81, slot {class=SizeBoxSlot}}
  SectionBG {
   widgetClassPath=&Image
   parent=Overlay_81
   widget=SectionBG
   properties {
    brush {
     bIsDynamicallyLoaded=false
     drawAs=Image
     imageType=NoImage
     mirroring=NoMirror
     resourceName=None
     resourceObject=/Game/UI/Items/Materials/MI-BacchusNCPAnim
     tiling=NoTile
     imageSize=<32 32>
     margin=<ltrb 0 0 0 0>
     outlineSettings {
      bUseBrushTransparency=false
      roundingType=HalfHeightRadius
      width=0
      color {colorUseRule=UseColor_Specified, specifiedColor=<rgba 0 0 0 0>}
      cornerRadii=<0 0 0 0>
     }
     tintColor {colorUseRule=UseColor_Specified, specifiedColor=<rgba 1 1 1 1>}
     uVRegion {bIsValid=false, max=<0 0>, min=<0 0>}
    }
    colorAndOpacity=<rgba 0.0273 0.0273 0.0273 0.6>
   }
   slot {class=OverlaySlot, properties {horizontalAlignment=HAlign_Fill, verticalAlignment=VAlign_Fill}}
  }
  Section {
   widgetClassPath=&Image
   parent=Overlay_81
   widget=Section
   properties {
    brush {
     bIsDynamicallyLoaded=false
     drawAs=Image
     imageType=NoImage
     mirroring=NoMirror
     resourceName=None
     resourceObject=/Game/UI/Items/Materials/MI-BacchusNCPAnimSec
     tiling=NoTile
     imageSize=<32 32>
     margin=<ltrb 0 0 0 0>
     outlineSettings {
      bUseBrushTransparency=false
      roundingType=HalfHeightRadius
      width=0
      color {colorUseRule=UseColor_Specified, specifiedColor=<rgba 0 0 0 0>}
      cornerRadii=<0 0 0 0>
     }
     tintColor {colorUseRule=UseColor_Specified, specifiedColor=<rgba 1 1 1 1>}
     uVRegion {bIsValid=false, max=<0 0>, min=<0 0>}
    }
    colorAndOpacity=<rgba 0.0273 0.0273 0.0273 0.8>
   }
   slot {class=OverlaySlot, properties {horizontalAlignment=HAlign_Fill, verticalAlignment=VAlign_Fill}}
  }
  TouchIconClue {
   widgetClassPath=&Image
   parent=Overlay_81
   widget=TouchIconClue
   properties {brush {
     bIsDynamicallyLoaded=false
     drawAs=Image
     imageType=NoImage
     mirroring=NoMirror
     resourceName=None
     resourceObject=/Game/UI/Icons/Icon-Settings/T-Icon-TouchMotion-64
     tiling=NoTile
     imageSize=<50 50>
     margin=<ltrb 0 0 0 0>
     outlineSettings {
      bUseBrushTransparency=false
      roundingType=HalfHeightRadius
      width=0
      color {colorUseRule=UseColor_Specified, specifiedColor=<rgba 0 0 0 0>}
      cornerRadii=<0 0 0 0>
     }
     tintColor {colorUseRule=UseColor_Specified, specifiedColor=<rgba 1 1 1 1>}
     uVRegion {bIsValid=false, max=<0 0>, min=<0 0>}
    }}
   slot {class=OverlaySlot, properties {horizontalAlignment=HAlign_Center, verticalAlignment=VAlign_Center, padding=<ltrb 4 0 0 20>}}
  }
 ]
}
