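# Conversation UI widget that shows a brief for a data-driven service: name, description, remaining stock and transaction strip.
# A visibility switcher swaps between that purchasable layout and an error-only layout that shows the reason on OnUnableToPurchase.
# OnWarningMessageUpdated shows a warning above the transaction strip, collapsing it when the text is empty.
# The native parent supplies the title, description, stock and strip widgets; SetSelectedVisuals is left empty here.
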WidgetBlueprint /FortniteConversationUI/Conversation/FortDataDrivenServiceBrief
graphs [EventGraph {nodes [
   Event OnAbleToPurchase() {is=title K2Node_Event}
   > SetActiveWidget(
    self=Get VisibilitySwitcher_PurchaseState() -> (VisibilitySwitcher_PurchaseState:Common Visibility Switcher Object Reference)
    Widget=Get Overlay_CanPurchase() -> (Overlay_CanPurchase:Overlay Object Reference)
   )
   n1 = Event OnUnableToPurchase!() -> (OutputDelegate, then, Reason:"Text (by ref)") {is=title K2Node_Event}
   > SetActiveWidget(self=Get VisibilitySwitcher_PurchaseState(), Widget=Get Overlay_CannotPurchase() -> (Overlay_CannotPurchase:Overlay Object Reference))
   > SetErrorMessage!(
    execute
    self:Brief Error Message New Object Reference=Get BriefErrorMessage() -> (BriefErrorMessage:Brief Error Message New Object Reference)
    ErrorText:Text=@n1.Reason
   ) -> (then)
   n2 = Event OnWarningMessageUpdated!() -> (OutputDelegate, then, WarningMessage:"Text (by ref)") {is=title K2Node_Event}
   > n3 = Branch(Condition=TextIsEmpty(InText=@n2.WarningMessage))
   n4 = Get Transaction_WarningMessage() -> (Transaction_WarningMessage:Brief Error Message New Object Reference)
   > SetVisibility(self=@n4, InVisibility=Collapsed)
   SetErrorMessage(execute=@n3.else, self=@n4, ErrorText=@n2.WarningMessage)
   > SetVisibility(self=@n4, InVisibility=HitTestInvisible)
   Event SetSelectedVisuals!() -> (OutputDelegate, then, IsSelected:bool) {is=title K2Node_Event}
  ]}]
variables []
widgets {
 info {
  inheritedWidgetCount=4
  parentClass=/Script/FortniteConversationUI.FortDataDrivenServiceBrief
  rootWidgetClass=&SizeBox
  widgetCount=17
 }
 widgets [
  SizeBox_28 {
   widgetClassPath=&SizeBox
   widget=SizeBox_28
   properties {bOverride_MaxDesiredWidth=true, heightOverride=350, maxDesiredWidth=900, widthOverride=550}
  }
  VisibilitySwitcher_PurchaseState {
   widgetClassPath=&CommonVisibilitySwitcher
   bIsVariable=true
   parent=SizeBox_28
   widget=VisibilitySwitcher_PurchaseState
   slot {class=SizeBoxSlot}
  }
  Overlay_CanPurchase {
   widgetClassPath=&Overlay
   bIsVariable=true
   parent=VisibilitySwitcher_PurchaseState
   widget=Overlay_CanPurchase
   slot {class=CommonVisibilitySwitcherSlot}
  }
  VerticalBox_0 {widgetClassPath=&VerticalBox, parent=Overlay_CanPurchase, widget=VerticalBox_0, slot {class=OverlaySlot}}
  Overlay_4 {widgetClassPath=&Overlay, parent=VerticalBox_0, widget=Overlay_4, slot {class=VerticalBoxSlot}}
  Border_168 {
   widgetClassPath=&Border
   parent=Overlay_4
   widget=Border_168
   properties {brushColor=<rgba 0 0 0 0.446>, padding=<ltrb 18 10 18 10>}
   slot {class=OverlaySlot, properties {horizontalAlignment=HAlign_Fill}}
  }
  Overlay_0 {widgetClassPath=&Overlay, parent=Border_168, widget=Overlay_0, slot {class=BorderSlot, properties {padding=<ltrb 18 10 18 10>}}}
  Text_TitleBar {
   widgetClassPath=&CommonTextBlock
   bInherited=true
   parent=Overlay_0
   widget=Text_TitleBar
   properties {
    text=ServiceName
    textTransformPolicy=ToUpper
    font {
     bForceMonospaced=false
     bMaterialIsStencil=false
     fontMaterial=null
     fontObject=&Burbank
     letterSpacing=0
     monospacedWidth=1
     size=18
     skewAmount=0
     typefaceFontName=BurbankRegular-Black
     outlineSettings {
      bApplyOutlineToDropShadows=false
      bMiteredCorners=false
      bSeparateFillAlpha=false
      outlineMaterial=null
      outlineSize=2
      outlineColor=<rgba 0.0052 0.0052 0.0056 0.6>
     }
    }
    renderTransform {angle=0, scale=<1 1>, shear=<-8 0>, translation=<0 -30>}
    shadowOffset=<0 0>
   }
   slot {class=OverlaySlot, properties {verticalAlignment=VAlign_Fill}}
  }
  Spacer_140 {widgetClassPath=&Spacer, parent=Overlay_0, widget=Spacer_140, properties {size=<1 48>}, slot {class=OverlaySlot}}
  SizeBox_0 {
   widgetClassPath=&SizeBox
   parent=Overlay_0
   widget=SizeBox_0
   properties {bOverride_MaxDesiredWidth=true, bOverride_MinDesiredWidth=true, maxDesiredWidth=640, minDesiredWidth=320}
   slot {class=OverlaySlot, properties {verticalAlignment=VAlign_Center}}
  }
  HorizontalBox_0 {
   widgetClassPath=&HorizontalBox
   parent=SizeBox_0
   widget=HorizontalBox_0
   properties {renderTransform {angle=0, scale=<1 1>, shear=<-8 0>, translation=<0 0>}}
   slot {class=SizeBoxSlot}
  }
  RichText_Description {
   widgetClassPath=&CommonRichTextBlock
   bInherited=true
   bIsVariable=true
   parent=HorizontalBox_0
   widget=RichText_Description
   properties {
    bOverrideDefaultStyle=true
    text=ServiceDescription
    textStyleSet=&/Conversation_RichTextStyleSet
    wrapTextAt=640
    defaultTextStyleOverride {
     overflowPolicy=Clip
     transformPolicy=None
     colorAndOpacity {colorUseRule=UseColor_Specified, specifiedColor=<rgba 1 1 1 1>}
     font {
      bForceMonospaced=false
      bMaterialIsStencil=false
      fontMaterial=null
      fontObject=&BurbankSmall
      letterSpacing=0
      monospacedWidth=1
      size=20
      skewAmount=0
      typefaceFontName=Bold
      outlineSettings {
       bApplyOutlineToDropShadows=false
       bMiteredCorners=false
       bSeparateFillAlpha=false
       outlineMaterial=null
       outlineSize=0
       outlineColor=<rgba 0 0 0 1>
      }
     }
     highlightColor {colorUseRule=UseColor_Specified, specifiedColor=<rgba 0 0 0 1>}
     highlightShape {
      bIsDynamicallyLoaded=false
      drawAs=Image
      imageType=NoImage
      mirroring=NoMirror
      resourceName=None
      resourceObject=null
      tiling=NoTile
      imageSize=<32 32>
      margin=<ltrb 0 0 0 0>
      outlineSettings {
       bUseBrushTransparency=false
       roundingType=HalfHeightRadius
       width=0
       color {colorUseRule=UseColor_Specified, specifiedColor=<rgba 0 0 0 0>}
       cornerRadii=<0 0 0 0>
      }
      tintColor {colorUseRule=UseColor_Specified, specifiedColor=<rgba 1 1 1 1>}
      uVRegion {bIsValid=false, max=<0 0>, min=<0 0>}
     }
     selectedBackgroundColor {colorUseRule=UseColor_Foreground, specifiedColor=<rgba 1 0 1 1>}
     shadowColorAndOpacity=<rgba 0 0 0 1>
     shadowOffset=<0 0>
     strikeBrush {
      bIsDynamicallyLoaded=false
      drawAs=Image
      imageType=NoImage
      mirroring=NoMirror
      resourceName=None
      resourceObject=null
      tiling=NoTile
      imageSize=<32 32>
      margin=<ltrb 0 0 0 0>
      outlineSettings {
       bUseBrushTransparency=false
       roundingType=HalfHeightRadius
       width=0
       color {colorUseRule=UseColor_Specified, specifiedColor=<rgba 0 0 0 0>}
       cornerRadii=<0 0 0 0>
      }
      tintColor {colorUseRule=UseColor_Specified, specifiedColor=<rgba 1 1 1 1>}
      uVRegion {bIsValid=false, max=<0 0>, min=<0 0>}
     }
     underlineBrush {
      bIsDynamicallyLoaded=false
      drawAs=Image
      imageType=NoImage
      mirroring=NoMirror
      resourceName=None
      resourceObject=null
      tiling=NoTile
      imageSize=<32 32>
      margin=<ltrb 0 0 0 0>
      outlineSettings {
       bUseBrushTransparency=false
       roundingType=HalfHeightRadius
       width=0
       color {colorUseRule=UseColor_Specified, specifiedColor=<rgba 0 0 0 0>}
       cornerRadii=<0 0 0 0>
      }
      tintColor {colorUseRule=UseColor_Specified, specifiedColor=<rgba 1 1 1 1>}
      uVRegion {bIsValid=false, max=<0 0>, min=<0 0>}
     }
    }
   }
   slot {class=HorizontalBoxSlot, properties {verticalAlignment=VAlign_Center}}
  }
  Text_StockRemaining {
   widgetClassPath=&CommonTextBlock
   bInherited=true
   bIsVariable=true
   parent=Overlay_4
   widget=Text_StockRemaining
   properties {
    justification=Center
    text=Stock: 1
    textTransformPolicy=ToUpper
    font {
     bForceMonospaced=false
     bMaterialIsStencil=false
     fontMaterial=null
     fontObject=&Burbank
     letterSpacing=0
     monospacedWidth=1
     size=14
     skewAmount=0
     typefaceFontName=Burbank-PlayerPower
     outlineSettings {
      bApplyOutlineToDropShadows=false
      bMiteredCorners=false
      bSeparateFillAlpha=false
      outlineMaterial=&M-FontShadow
      outlineSize=2
      outlineColor=<rgba 0.0052 0.0052 0.0056 0.8>
     }
    }
    renderTransform {angle=0, scale=<1 1>, shear=<-8 0>, translation=<0 0>}
    shadowOffset=<0 0>
   }
   slot {class=OverlaySlot, properties {horizontalAlignment=HAlign_Right, verticalAlignment=VAlign_Bottom, padding=<ltrb 0 0 4 -7.5>}}
  }
  Transaction_WarningMessage {
   widgetClassPath=&/BriefErrorMessageNew
   bIsVariable=true
   parent=VerticalBox_0
   widget=Transaction_WarningMessage
   properties {fontSize=16}
   slot {class=VerticalBoxSlot}
  }
  TransactionStrip_Main {
   widgetClassPath=&/FortTransactionStripNew
   bInherited=true
   bIsVariable=true
   parent=VerticalBox_0
   widget=TransactionStrip_Main
   slot {class=VerticalBoxSlot}
  }
  Overlay_CannotPurchase {
   widgetClassPath=&Overlay
   bIsVariable=true
   parent=VisibilitySwitcher_PurchaseState
   widget=Overlay_CannotPurchase
   slot {class=CommonVisibilitySwitcherSlot}
  }
  BriefErrorMessage {
   widgetClassPath=&BriefErrorMessageNew
   bIsVariable=true
   parent=Overlay_CannotPurchase
   widget=BriefErrorMessage
   properties {fontSize=20}
   slot {class=OverlaySlot}
  }
 ]
}
