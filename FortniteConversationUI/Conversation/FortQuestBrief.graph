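# Quest briefing widget for the conversation UI, built on native FortQuestBrief, showing a quest's objective, rewards, timer and warnings.
# On receiving quest info, it picks a title from the category tag, applies rarity colors and fills the embedded transaction strip, abbreviating XP in thousands.
# A switcher swaps to an error view when unavailable; separate handlers drive timer visibility and selection visuals.
# The rarity label and inline reward row are collapsed yet still updated.
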WidgetBlueprint /FortniteConversationUI/Conversation/FortQuestBrief
graphs [
 EventGraph {nodes [
   n1 = Get FortTransactionStrip() -> (FortTransactionStrip:Fort Transaction Strip New Object Reference)
   n2 = Event OnQuestInformationRecieved!() -> (
    OutputDelegate
    then
    Objective:"Text (by ref)"
    RarityData:"Fort Rarity Item Data Structure (by ref)"
    CategoryTag:"Gameplay Tag Structure (by ref)"
    XPReward:int=0
    MoneyReward:int=0
    TokenReward:int=0
    TokenRewardIcon:"Texture 2D Soft Object Reference (by ref)"
   ) {is=title K2Node_Event}
   n3 = Break Fort Rarity Item Data!(FortRarityItemData:"Fort Rarity Item Data Structure (by ref)"=@n2.RarityData) -> (
    Name:Text
    Color1:Linear Color Structure
    Color2:Linear Color Structure
    Color3:Linear Color Structure
    Color4:Linear Color Structure
    Color5:Linear Color Structure
    Radius:float=~
    Falloff:float=~
    Brightness:float=~
    Roughness:float=~
    Glow:float=~
   ) {is=title K2Node_BreakStruct}
   n4 = SetContentColorAndOpacity!(
    execute
    self:Border Object Reference=[
     Get Border_RewardColor!(self:Fort Transaction Strip New Object Reference=@n1) -> (Border_RewardColor:Border Object Reference)
     Get Border_MoneyRewardColor!(self:Fort Transaction Strip New Object Reference=@n1) -> (Border_MoneyRewardColor:Common Border Object Reference)
     Get Border_RewardColor!() -> (Border_RewardColor)
    ]
    InContentColorAndOpacity:Linear Color Structure=@n3.Color1
   ) -> (then)
   n5 = Make Slate Color(SpecifiedColor=@n3.Color1)
   SetColorAndOpacity(execute=[@n4.then, @n21.then], self=Get Text_Rarity() -> (Text_Rarity:Common Text Object Reference), InColorAndOpacity=@n5)
   n6 = Get Image_RarityGlow() -> (Image_RarityGlow:Image Object Reference)
   n7 = Make Slate Color(SpecifiedColor=@n3.Color3)
   > SetBrushTintColor!(execute, self:Image Object Reference=@n6, TintColor:Slate Color Structure=@n7) -> (then)
   > SetBrushTintColor(self=Get Image_QuestProviderImageGlow() -> (Image_QuestProviderImageGlow:Image Object Reference), TintColor=@n7)
   > UpdateCurrencyIcon(self:Fort Transaction Strip New Object Reference=Get FortTransactionStrip())
   n8 = Event OnQuestExpirationTimeSet!() -> (OutputDelegate, then, bDisplayExpirationTime:bool) {is=title K2Node_Event}
   > n9 = Branch(Condition=@n8.bDisplayExpirationTime)
   > SetVisibility(self=Get TimerContainer() -> (TimerContainer:Horizontal Box Object Reference), InVisibility=Visible)
   SetVisibility(execute=@n9.else, self=Get TimerContainer(), InVisibility=Collapsed)
   K2Node_Event_1 = EventUnavailable!() -> (OutputDelegate, then, UnavailableMessage:Text) {is=title K2Node_CustomEvent}
   > SetActiveWidget(
    self=Get StateSwitcher() -> (StateSwitcher:Common Visibility Switcher Object Reference)
    Widget=Get NotAvailable() -> (NotAvailable:Overlay Object Reference)
   )
   > SetErrorMessage!(
    execute
    self:Brief Error Message New Object Reference=Get BriefErrorMessage1() -> (BriefErrorMessage1:Brief Error Message New Object Reference)
    ErrorText:Text=@K2Node_Event_1.UnavailableMessage
   ) -> (then)
   n10 = Event SetSelectedVisuals!() -> (OutputDelegate, then, IsSelected:bool) {is=title K2Node_Event}
   > SetSelectedVisuals!(execute, self:Fort Transaction Strip New Object Reference=Get FortTransactionStrip(), IsSelected:bool=@n10.IsSelected) -> (then)
   SetBrushFromSoftTexture!(
    execute=@n23.then
    self:Image Object Reference=Get Image_TokenReward!(self:Fort Transaction Strip New Object Reference=Get FortTransactionStrip()) -> (Image_TokenReward:Image Object Reference)
    SoftTexture:Texture 2D Soft Object Reference=@n2.TokenRewardIcon
    bMatchSize:bool=false
   ) -> (then)
   n11 = Get FortTransactionStrip()
   SetVisibility(
    execute=@n15.then
    self=Get Reward_Custom!(self:Fort Transaction Strip New Object Reference=@n11) -> (Reward_Custom:Size Box Object Reference)
    InVisibility=Select(Option 0=Collapsed, Option 1=SelfHitTestInvisible, Index="Not Equal (Integer)"(A=@n2.TokenReward))
   )
   SetOpacity!(
    execute
    self:Image Object Reference=@n6
    InOpacity:float=Select!(
     Common:double=0
     Uncommon:double=0.1
     Rare:double=0.2
     Epic:double=0.25
     Legendary:double=1
     Mythic:double=1
     Transcendent:double=1
     Unattainable:double=1
     Index:EFortRarity Enum=Common
    ) -> (ReturnValue:double)
   ) -> (then)
   SetText(
    execute=@n16.then
    self=Get Text_Objective() -> (Text_Objective:Common Rich Text Block Object Reference)
    InText=Format Text!(Format="LOCTEXT(\"{Description}\")", Description:"Text (by ref)"=@n2.Objective) -> (Result)
   )
   n12 = "To Text (Integer)"(Value=@n2.XPReward)
   n13 = "integer > integer"(A=@n2.XPReward, B=999) {is=title K2Node_CallFunction}
   n14 = Format Text!(
    Format="LOCTEXT(\"{0}k\")"
    0:Text="To Text (Integer)"(Value=int / int!(A:int=@n2.XPReward, B:int=1000) -> (ReturnValue:int=0) {is=title K2Node_CallFunction})
   ) -> (Result)
   > "SetText (Text)"(
    self=[
     Get Text_XpReward!(self:Fort Transaction Strip New Object Reference=@n11) -> (Text_XpReward:Common Text Object Reference)
     Get Text_XpReward!() -> (Text_XpReward)
    ]
    InText=Format Text(Format="LOCTEXT(\"{0}\")", 0=Select!(Option 0:Text=@n12, Option 1:Text=@n14, Index:bool=@n13) -> (ReturnValue:Text))
   )
   > "SetText (Text)"(
    self=[
     Get Text_MoneyReward!(self:Fort Transaction Strip New Object Reference=@n11) -> (Text_MoneyReward:Common Text Object Reference)
     Get Text_MoneyReward!() -> (Text_MoneyReward)
    ]
    InText=Format Text!(Format="LOCTEXT(\"{MoneyValue}\")", MoneyValue:Text="To Text (Integer)"(Value=@n2.MoneyReward)) -> (Result)
   )
   > "SetText (Text)"(
    self=Get Text_CustomReward!(self:Fort Transaction Strip New Object Reference=@n11) -> (Text_CustomReward:Common Text Object Reference)
    InText="To Text (Integer)"(Value=@n2.TokenReward)
   )
   > SetVisibility(
    self=[
     Get Reward_Money!(self:Fort Transaction Strip New Object Reference=@n11) -> (Reward_Money:Size Box Object Reference)
     Get MoneyReward() -> (MoneyReward:Horizontal Box Object Reference)
    ]
    InVisibility=Select(
     Option 0:ESlateVisibility Enum=Collapsed
     Option 1:ESlateVisibility Enum=SelfHitTestInvisible
     Index="Not Equal (Integer)"!(A:int=@n2.MoneyReward, B:int=0) -> (ReturnValue:bool) {is=title K2Node_CallFunction}
    ) -> (ReturnValue:ESlateVisibility Enum)
   )
   > n15 = SetVisibility(
    self=[
     Get Reward_XP!(self:Fort Transaction Strip New Object Reference=@n11) -> (Reward_XP:Size Box Object Reference)
     Get XpReward() -> (XpReward:Overlay Object Reference)
    ]
    InVisibility=Select(Option 0=Collapsed, Option 1=SelfHitTestInvisible, Index="Not Equal (Integer)"(A=@n2.XPReward, B=0))
   )
   n16 = "SetText (Text)"(execute=@n18.then, self=Get Text_Rarity(), InText=Format Text!(Format="LOCTEXT(\"({rarity})\")", rarity:Text=@n3) -> (Result))
   n17 = Get Title!(CategoryTag:Gameplay Tag Structure=@n2.CategoryTag) -> (Title:Text) {is=title K2Node_CallFunction}
   n18 = "SetText (Text)"(execute=@n19.then, self=Get Title() -> (Title:Common Text Object Reference) {is=title K2Node_VariableGet}, InText=@n17)
   n19 = SetActiveWidget(execute=@n22.then_0, self=Get StateSwitcher(), Widget=Get Available() -> (Available:Overlay Object Reference))
   n20 = IsValidSoftObjectReference(SoftObjectReference=@n2.TokenRewardIcon)
   n21 = SetBrushTintColor(execute=@n22.then_1, self=Get Image_RarityRibbon() -> (Image_RarityRibbon:Image Object Reference), TintColor=@n5)
   n22 = Sequence!(execute=@n2.then) -> (then_0, then_1, then_2:Exec)
   n23 = Branch(execute=@n22.then_2, Condition=@n20)
  ]}
 Get Title {nodes [
   n1 = Get Title!() -> (then, CategoryTag:Gameplay Tag Structure) {is=title K2Node_FunctionEntry}
   > n2 = Switch on Gameplay Tag!(execute:Exec, Selection:Gameplay Tag Structure=@n1) -> (Default:Exec, Athena.Quests.Bounty.Hunter:Exec, Athena.Quests.Tournament:Exec) {is=title GameplayTagsK2Node_SwitchGameplayTag, pinTags [{tagName=Athena.Quests.Bounty.Hunter}, {tagName=Athena.Quests.Tournament}]}
   Return Node!(execute=@n2.Athena.Quests.Bounty.Hunter, Title:Text="LOCTEXT(\"BOUNTY BRIEFING\")")
   Return Node(execute=@n2.Athena.Quests.Tournament, Title="LOCTEXT(\"Versus Battle\")")
   > Return Node(Title="LOCTEXT(\"Quest\")")
  ]}
 SetWarningText {nodes [
   n1 = SetWarningText!() -> (then, WarningText:Text) {is=title K2Node_FunctionEntry}
   n2 = Get WarningMessage() -> (WarningMessage:Brief Error Message New Object Reference)
   > SetErrorMessage(self=@n2, ErrorText=@n1)
   > SetVisibility(self=@n2, InVisibility=SelfHitTestInvisible)
  ]}
]
variables []
widgets {
 info {inheritedWidgetCount=6, parentClass=/Script/FortniteConversationUI.FortQuestBrief, rootWidgetClass=&SizeBox, widgetCount=39}
 widgets [
  SizeBox_36 {widgetClassPath=&SizeBox, widget=SizeBox_36, properties {heightOverride=350, maxDesiredWidth=768, widthOverride=550}}
  StateSwitcher {
   widgetClassPath=&CommonVisibilitySwitcher
   bIsVariable=true
   parent=SizeBox_36
   widget=StateSwitcher
   slot {class=SizeBoxSlot}
  }
  Available {widgetClassPath=&Overlay, bIsVariable=true, parent=StateSwitcher, widget=Available, slot {class=CommonVisibilitySwitcherSlot}}
  VerticalBox_0 {
   widgetClassPath=&VerticalBox
   parent=Available
   widget=VerticalBox_0
   slot {class=OverlaySlot, properties {horizontalAlignment=HAlign_Fill}}
  }
  Overlay_1 {widgetClassPath=&Overlay, parent=VerticalBox_0, widget=Overlay_1, slot {class=VerticalBoxSlot}}
  Border_168 {
   widgetClassPath=&Border
   parent=Overlay_1
   widget=Border_168
   properties {brushColor=<rgba 0.0052 0.0052 0.0056 0.6>, padding=<ltrb 20 10 20 10>}
   slot {class=OverlaySlot, properties {horizontalAlignment=HAlign_Fill, verticalAlignment=VAlign_Fill}}
  }
  Overlay_0 {widgetClassPath=&Overlay, parent=Border_168, widget=Overlay_0, slot {class=BorderSlot, properties {padding=<ltrb 20 10 20 10>}}}
  Title {
   widgetClassPath=&CommonTextBlock
   bIsVariable=true
   parent=Overlay_0
   widget=Title
   properties {
    text=Quest
    textTransformPolicy=ToUpper
    font {
     bForceMonospaced=false
     bMaterialIsStencil=false
     fontMaterial=null
     fontObject=&Burbank
     letterSpacing=0
     monospacedWidth=1
     size=18
     skewAmount=0
     typefaceFontName=BurbankRegular-Black
     outlineSettings {
      bApplyOutlineToDropShadows=false
      bMiteredCorners=false
      bSeparateFillAlpha=false
      outlineMaterial=null
      outlineSize=2
      outlineColor=<rgba 0.0052 0.0052 0.0056 0.6>
     }
    }
    renderTransform {angle=0, scale=<1 1>, shear=<-8 0>, translation=<0 -30>}
    shadowOffset=<0 0>
   }
   slot {class=OverlaySlot, properties {horizontalAlignment=HAlign_Fill, verticalAlignment=VAlign_Fill}}
  }
  HorizontalBox_ImageAndDetails {
   widgetClassPath=&HorizontalBox
   parent=Overlay_0
   widget=HorizontalBox_ImageAndDetails
   slot {class=OverlaySlot, properties {horizontalAlignment=HAlign_Fill, verticalAlignment=VAlign_Fill}}
  }
  Overlay_QuestProvider {
   widgetClassPath=&Overlay
   bInherited=true
   parent=HorizontalBox_ImageAndDetails
   widget=Overlay_QuestProvider
   properties {visibility=Collapsed}
   slot {class=HorizontalBoxSlot, properties {horizontalAlignment=HAlign_Center, verticalAlignment=VAlign_Center}}
  }
  Image_QuestProviderImageGlow {
   widgetClassPath=&Image
   bIsVariable=true
   parent=Overlay_QuestProvider
   widget=Image_QuestProviderImageGlow
   properties {
    visibility=SelfHitTestInvisible
    brush {
     bIsDynamicallyLoaded=false
     drawAs=Image
     imageType=NoImage
     mirroring=NoMirror
     resourceName=None
     resourceObject=/Game/Athena/HUD/LTM/Wax/Materials/M_UI_TextGlow_Add
     tiling=NoTile
     imageSize=<128 128>
     margin=<ltrb 0 0 0 0>
     outlineSettings {
      bUseBrushTransparency=false
      roundingType=HalfHeightRadius
      width=0
      color {colorUseRule=UseColor_Specified, specifiedColor=<rgba 0 0 0 0>}
      cornerRadii=<0 0 0 0>
     }
     tintColor {colorUseRule=UseColor_Specified, specifiedColor=<rgba 0.0037 0.3613 1 1>}
     uVRegion {bIsValid=false, max=<0 0>, min=<0 0>}
    }
   }
   slot {class=OverlaySlot, properties {horizontalAlignment=HAlign_Center, verticalAlignment=VAlign_Center}}
  }
  LazyImage_QuestProviderImage {
   widgetClassPath=&FortLazyImage
   bInherited=true
   parent=Overlay_QuestProvider
   widget=LazyImage_QuestProviderImage
   properties {
    materialTextureParamName=Texture
    brush {
     bIsDynamicallyLoaded=false
     drawAs=Image
     imageType=NoImage
     mirroring=NoMirror
     resourceName=None
     resourceObject=/Game/Athena/HUD/Collections/Materials/M_UI_FishToast_Character
     tiling=NoTile
     imageSize=<128 128>
     margin=<ltrb 0 0 0 0>
     outlineSettings {
      bUseBrushTransparency=false
      roundingType=HalfHeightRadius
      width=0
      color {colorUseRule=UseColor_Specified, specifiedColor=<rgba 0 0 0 0>}
      cornerRadii=<0 0 0 0>
     }
     tintColor {colorUseRule=UseColor_Specified, specifiedColor=<rgba 1 1 1 1>}
     uVRegion {bIsValid=false, max=<0 0>, min=<0 0>}
    }
    shadowColorAndOpacity=<rgba 0.0052 0.0052 0.0056 0.5>
   }
   slot {class=OverlaySlot, properties {horizontalAlignment=HAlign_Center, verticalAlignment=VAlign_Center}}
  }
  VerticalBox_75 {
   widgetClassPath=&VerticalBox
   parent=HorizontalBox_ImageAndDetails
   widget=VerticalBox_75
   slot {class=HorizontalBoxSlot, properties {verticalAlignment=VAlign_Top}}
  }
  HorizontalBox_4 {
   widgetClassPath=&HorizontalBox
   parent=VerticalBox_75
   widget=HorizontalBox_4
   slot {class=VerticalBoxSlot, properties {verticalAlignment=VAlign_Top}}
  }
  SizeBox_0 {
   widgetClassPath=&SizeBox
   parent=HorizontalBox_4
   widget=SizeBox_0
   properties {bOverride_MaxDesiredWidth=true, bOverride_MinDesiredWidth=true, maxDesiredWidth=640, minDesiredWidth=320}
   slot {class=HorizontalBoxSlot}
  }
  HorizontalBox_0 {widgetClassPath=&HorizontalBox, parent=SizeBox_0, widget=HorizontalBox_0, slot {class=SizeBoxSlot}}
  Text_Objective {
   widgetClassPath=&CommonRichTextBlock
   bInherited=true
   bIsVariable=true
   parent=HorizontalBox_0
   widget=Text_Objective
   properties {
    bOverrideDefaultStyle=true
    text="Complete this quest! Here is some extra description"
    wrapTextAt=640
    defaultTextStyleOverride {
     overflowPolicy=Clip
     transformPolicy=None
     colorAndOpacity {colorUseRule=UseColor_Specified, specifiedColor=<rgba 1 1 1 1>}
     font {
      bForceMonospaced=false
      bMaterialIsStencil=false
      fontMaterial=null
      fontObject=&BurbankSmall
      letterSpacing=0
      monospacedWidth=1
      size=20
      skewAmount=0.2
      typefaceFontName=Bold
      outlineSettings {
       bApplyOutlineToDropShadows=false
       bMiteredCorners=false
       bSeparateFillAlpha=false
       outlineMaterial=null
       outlineSize=0
       outlineColor=<rgba 0 0 0 1>
      }
     }
     highlightColor {colorUseRule=UseColor_Specified, specifiedColor=<rgba 0 0 0 1>}
     highlightShape {
      bIsDynamicallyLoaded=false
      drawAs=Image
      imageType=NoImage
      mirroring=NoMirror
      resourceName=None
      resourceObject=null
      tiling=NoTile
      imageSize=<32 32>
      margin=<ltrb 0 0 0 0>
      outlineSettings {
       bUseBrushTransparency=false
       roundingType=HalfHeightRadius
       width=0
       color {colorUseRule=UseColor_Specified, specifiedColor=<rgba 0 0 0 0>}
       cornerRadii=<0 0 0 0>
      }
      tintColor {colorUseRule=UseColor_Specified, specifiedColor=<rgba 1 1 1 1>}
      uVRegion {bIsValid=false, max=<0 0>, min=<0 0>}
     }
     selectedBackgroundColor {colorUseRule=UseColor_Foreground, specifiedColor=<rgba 1 0 1 1>}
     shadowColorAndOpacity=<rgba 0 0 0 1>
     shadowOffset=<0 0>
     strikeBrush {
      bIsDynamicallyLoaded=false
      drawAs=Image
      imageType=NoImage
      mirroring=NoMirror
      resourceName=None
      resourceObject=null
      tiling=NoTile
      imageSize=<32 32>
      margin=<ltrb 0 0 0 0>
      outlineSettings {
       bUseBrushTransparency=false
       roundingType=HalfHeightRadius
       width=0
       color {colorUseRule=UseColor_Specified, specifiedColor=<rgba 0 0 0 0>}
       cornerRadii=<0 0 0 0>
      }
      tintColor {colorUseRule=UseColor_Specified, specifiedColor=<rgba 1 1 1 1>}
      uVRegion {bIsValid=false, max=<0 0>, min=<0 0>}
     }
     underlineBrush {
      bIsDynamicallyLoaded=false
      drawAs=Image
      imageType=NoImage
      mirroring=NoMirror
      resourceName=None
      resourceObject=null
      tiling=NoTile
      imageSize=<32 32>
      margin=<ltrb 0 0 0 0>
      outlineSettings {
       bUseBrushTransparency=false
       roundingType=HalfHeightRadius
       width=0
       color {colorUseRule=UseColor_Specified, specifiedColor=<rgba 0 0 0 0>}
       cornerRadii=<0 0 0 0>
      }
      tintColor {colorUseRule=UseColor_Specified, specifiedColor=<rgba 1 1 1 1>}
      uVRegion {bIsValid=false, max=<0 0>, min=<0 0>}
     }
    }
   }
   slot {class=HorizontalBoxSlot, properties {verticalAlignment=VAlign_Center, size {sizeRule=Fill, value=1}}}
  }
  Overlay_Rarity {
   widgetClassPath=&Overlay
   parent=HorizontalBox_0
   widget=Overlay_Rarity
   properties {visibility=Collapsed}
   slot {class=HorizontalBoxSlot, properties {padding=<ltrb 8 0 0 0>}}
  }
  Text_Rarity {
   widgetClassPath=&CommonTextBlock
   bIsVariable=true
   parent=Overlay_Rarity
   widget=Text_Rarity
   properties {
    text="{Rarity}"
    textTransformPolicy=ToUpper
    font {
     bForceMonospaced=false
     bMaterialIsStencil=false
     fontMaterial=null
     fontObject=&Burbank
     letterSpacing=0
     monospacedWidth=1
     size=20
     skewAmount=0.2
     typefaceFontName=BurbankRegular-Black
     outlineSettings {
      bApplyOutlineToDropShadows=false
      bMiteredCorners=false
      bSeparateFillAlpha=false
      outlineMaterial=null
      outlineSize=0
      outlineColor=<rgba 0 0 0 1>
     }
    }
    shadowOffset=<0 0>
   }
   slot {class=OverlaySlot}
  }
  Image_RarityGlow {
   widgetClassPath=&Image
   bIsVariable=true
   parent=Overlay_Rarity
   widget=Image_RarityGlow
   properties {
    renderOpacity=0.3
    brush {
     bIsDynamicallyLoaded=false
     drawAs=Image
     imageType=NoImage
     mirroring=NoMirror
     resourceName=None
     resourceObject=&M_UI_TextGlow_Add
     tiling=NoTile
     imageSize=<32 32>
     margin=<ltrb 0 0 0 0>
     outlineSettings {
      bUseBrushTransparency=false
      roundingType=HalfHeightRadius
      width=0
      color {colorUseRule=UseColor_Specified, specifiedColor=<rgba 0 0 0 0>}
      cornerRadii=<0 0 0 0>
     }
     tintColor {colorUseRule=UseColor_Specified, specifiedColor=<rgba 0.521 0.0781 0 1>}
     uVRegion {bIsValid=false, max=<0 0>, min=<0 0>}
    }
   }
   slot {class=OverlaySlot, properties {horizontalAlignment=HAlign_Fill, verticalAlignment=VAlign_Fill}}
  }
  CommonTextBlock {
   widgetClassPath=&CommonTextBlock
   parent=HorizontalBox_4
   widget=CommonTextBlock
   properties {
    text=Reward
    textTransformPolicy=ToUpper
    visibility=Collapsed
    font {
     bForceMonospaced=false
     bMaterialIsStencil=false
     fontMaterial=null
     fontObject=&Burbank
     letterSpacing=0
     monospacedWidth=1
     size=24
     skewAmount=0
     typefaceFontName=BurbankRegular-Black
     outlineSettings {
      bApplyOutlineToDropShadows=false
      bMiteredCorners=false
      bSeparateFillAlpha=false
      outlineMaterial=null
      outlineSize=0
      outlineColor=<rgba 0 0 0 1>
     }
    }
    renderTransform {angle=0, scale=<1 1>, shear=<-8 0>, translation=<0 0>}
    shadowOffset=<0 0>
   }
   slot {class=HorizontalBoxSlot}
  }
  HorizontalBox_125 {
   widgetClassPath=&HorizontalBox
   parent=HorizontalBox_4
   widget=HorizontalBox_125
   properties {visibility=Collapsed}
   slot {class=HorizontalBoxSlot, properties {padding=<ltrb 20 0 0 0>}}
  }
  MoneyReward {
   widgetClassPath=&HorizontalBox
   bIsVariable=true
   parent=HorizontalBox_125
   widget=MoneyReward
   slot {class=HorizontalBoxSlot, properties {padding=<ltrb 15 0 15 0>}}
  }
  Image_78 {
   widgetClassPath=&Image
   bIsVariable=true
   parent=MoneyReward
   widget=Image_78
   properties {brush {
     bIsDynamicallyLoaded=false
     drawAs=Image
     imageType=NoImage
     mirroring=NoMirror
     resourceName=None
     resourceObject=/Game/UI/Foundation/Textures/Icons/Items/T-Icon-S-Loot-Wads_01
     tiling=NoTile
     imageSize=<32 32>
     margin=<ltrb 0 0 0 0>
     outlineSettings {
      bUseBrushTransparency=false
      roundingType=HalfHeightRadius
      width=0
      color {colorUseRule=UseColor_Specified, specifiedColor=<rgba 0 0 0 0>}
      cornerRadii=<0 0 0 0>
     }
     tintColor {colorUseRule=UseColor_Specified, specifiedColor=<rgba 1 1 1 1>}
     uVRegion {bIsValid=false, max=<0 0>, min=<0 0>}
    }}
   slot {class=HorizontalBoxSlot, properties {horizontalAlignment=HAlign_Center, verticalAlignment=VAlign_Center}}
  }
  Text_MoneyReward {
   widgetClassPath=&CommonTextBlock
   bInherited=true
   parent=MoneyReward
   widget=Text_MoneyReward
   properties {
    text=500
    font {
     bForceMonospaced=false
     bMaterialIsStencil=false
     fontMaterial=null
     fontObject=&Burbank
     letterSpacing=0
     monospacedWidth=1
     size=24
     skewAmount=0
     typefaceFontName=Burbank-PlayerPower
     outlineSettings {
      bApplyOutlineToDropShadows=false
      bMiteredCorners=false
      bSeparateFillAlpha=false
      outlineMaterial=null
      outlineSize=2
      outlineColor=<rgba 0.0052 0.0052 0.0056 0.3>
     }
    }
    renderTransform {angle=0, scale=<1 1>, shear=<-8 0>, translation=<0 0>}
    shadowOffset=<0 0>
   }
   slot {class=HorizontalBoxSlot}
  }
  XpReward {
   widgetClassPath=&Overlay
   bIsVariable=true
   parent=HorizontalBox_125
   widget=XpReward
   slot {class=HorizontalBoxSlot, properties {padding=<ltrb 15 0 15 0>}}
  }
  Image {
   widgetClassPath=&Image
   bIsVariable=true
   parent=XpReward
   widget=Image
   properties {
    visibility=Collapsed
    brush {
     bIsDynamicallyLoaded=false
     drawAs=Image
     imageType=NoImage
     mirroring=NoMirror
     resourceName=None
     resourceObject=/Game/UI/Foundation/Textures/Icons/Items/T-FNBR-SeasonalXP
     tiling=NoTile
     imageSize=<64 64>
     margin=<ltrb 0 0 0 0>
     outlineSettings {
      bUseBrushTransparency=false
      roundingType=HalfHeightRadius
      width=0
      color {colorUseRule=UseColor_Specified, specifiedColor=<rgba 0 0 0 0>}
      cornerRadii=<0 0 0 0>
     }
     tintColor {colorUseRule=UseColor_Specified, specifiedColor=<rgba 1 1 1 1>}
     uVRegion {bIsValid=false, max=<0 0>, min=<0 0>}
    }
   }
   slot {class=OverlaySlot}
  }
  Border_RewardColor {
   widgetClassPath=&Border
   bIsVariable=true
   parent=XpReward
   widget=Border_RewardColor
   properties {
    background {
     bIsDynamicallyLoaded=false
     drawAs=NoDrawType
     imageType=NoImage
     mirroring=NoMirror
     resourceName=None
     resourceObject=null
     tiling=NoTile
     imageSize=<32 32>
     margin=<ltrb 0 0 0 0>
     outlineSettings {
      bUseBrushTransparency=false
      roundingType=HalfHeightRadius
      width=0
      color {colorUseRule=UseColor_Specified, specifiedColor=<rgba 0 0 0 0>}
      cornerRadii=<0 0 0 0>
     }
     tintColor {colorUseRule=UseColor_Specified, specifiedColor=<rgba 1 1 1 1>}
     uVRegion {bIsValid=false, max=<0 0>, min=<0 0>}
    }
    padding=<ltrb 0 0 0 0>
   }
   slot {class=OverlaySlot, properties {verticalAlignment=VAlign_Bottom}}
  }
  HorizontalBox_3 {
   widgetClassPath=&HorizontalBox
   parent=Border_RewardColor
   widget=HorizontalBox_3
   slot {class=BorderSlot, properties {padding=<ltrb 0 0 0 0>}}
  }
  Text_XpReward {
   widgetClassPath=&CommonTextBlock
   bInherited=true
   parent=HorizontalBox_3
   widget=Text_XpReward
   properties {
    text=50k
    font {
     bForceMonospaced=false
     bMaterialIsStencil=false
     fontMaterial=null
     fontObject=&Burbank
     letterSpacing=0
     monospacedWidth=1
     size=24
     skewAmount=0
     typefaceFontName=Burbank-PlayerPower
     outlineSettings {
      bApplyOutlineToDropShadows=false
      bMiteredCorners=false
      bSeparateFillAlpha=false
      outlineMaterial=null
      outlineSize=2
      outlineColor=<rgba 0.0052 0.0052 0.0056 0.3>
     }
    }
    renderTransform {angle=0, scale=<1 1>, shear=<-8 0>, translation=<0 0>}
    shadowOffset=<0 0>
   }
   slot {class=HorizontalBoxSlot}
  }
  Image_56 {
   widgetClassPath=&Image
   bIsVariable=true
   parent=HorizontalBox_3
   widget=Image_56
   properties {brush {
     bIsDynamicallyLoaded=false
     drawAs=Image
     imageType=NoImage
     mirroring=NoMirror
     resourceName=None
     resourceObject=/FortniteCoreUI/Athena/HUD/Quests/Art/M_UI_TabCategory_XP_White
     tiling=NoTile
     imageSize=<48 48>
     margin=<ltrb 0 0 0 0>
     outlineSettings {
      bUseBrushTransparency=false
      roundingType=HalfHeightRadius
      width=0
      color {colorUseRule=UseColor_Specified, specifiedColor=<rgba 0 0 0 0>}
      cornerRadii=<0 0 0 0>
     }
     tintColor {colorUseRule=UseColor_Specified, specifiedColor=<rgba 1 1 1 1>}
     uVRegion {bIsValid=false, max=<0 0>, min=<0 0>}
    }}
   slot {class=HorizontalBoxSlot, properties {horizontalAlignment=HAlign_Center, verticalAlignment=VAlign_Center, padding=<ltrb 0 -32 0 -32>}}
  }
  TimerContainer {
   widgetClassPath=&HorizontalBox
   bIsVariable=true
   parent=VerticalBox_75
   widget=TimerContainer
   slot {class=VerticalBoxSlot, properties {padding=<ltrb 0 4 0 0>}}
  }
  ImageTimer {
   widgetClassPath=&Image
   bIsVariable=true
   parent=TimerContainer
   widget=ImageTimer
   properties {brush {
     bIsDynamicallyLoaded=false
     drawAs=Image
     imageType=NoImage
     mirroring=NoMirror
     resourceName=None
     resourceObject=/FortniteCoreUI/Athena/HUD/Quests/Art/M_UI_Timer_Icon_Quest
     tiling=NoTile
     imageSize=<32 32>
     margin=<ltrb 0 0 0 0>
     outlineSettings {
      bUseBrushTransparency=false
      roundingType=HalfHeightRadius
      width=0
      color {colorUseRule=UseColor_Specified, specifiedColor=<rgba 0 0 0 0>}
      cornerRadii=<0 0 0 0>
     }
     tintColor {colorUseRule=UseColor_Specified, specifiedColor=<rgba 1 1 1 1>}
     uVRegion {bIsValid=false, max=<0 0>, min=<0 0>}
    }}
   slot {class=HorizontalBoxSlot, properties {horizontalAlignment=HAlign_Right, verticalAlignment=VAlign_Center, padding=<ltrb 0 -12 0 -12>}}
  }
  Text_ExpirationTime {
   widgetClassPath=&CommonTextBlock
   bInherited=true
   bIsVariable=true
   parent=TimerContainer
   widget=Text_ExpirationTime
   properties {
    text=60m
    textTransformPolicy=ToUpper
    font {
     bForceMonospaced=false
     bMaterialIsStencil=false
     fontMaterial=/FortniteCoreUI/Athena/HUD/Quests/Art/M_UI_Timer_Font_Quest
     fontObject=&BurbankSmall
     letterSpacing=0
     monospacedWidth=1
     size=15
     skewAmount=0
     typefaceFontName=Bold
     outlineSettings {
      bApplyOutlineToDropShadows=false
      bMiteredCorners=false
      bSeparateFillAlpha=false
      outlineMaterial=null
      outlineSize=0
      outlineColor=<rgba 0 0 0 1>
     }
    }
    shadowOffset=<0 0>
   }
   slot {class=HorizontalBoxSlot, properties {size {sizeRule=Fill, value=1}}}
  }
  Image_RarityRibbon {
   widgetClassPath=&Image
   bIsVariable=true
   parent=Overlay_1
   widget=Image_RarityRibbon
   properties {
    visibility=HitTestInvisible
    brush {
     bIsDynamicallyLoaded=false
     drawAs=Image
     imageType=NoImage
     mirroring=NoMirror
     resourceName=None
     resourceObject=null
     tiling=NoTile
     imageSize=<6 32>
     margin=<ltrb 0 0 0 0>
     outlineSettings {
      bUseBrushTransparency=false
      roundingType=HalfHeightRadius
      width=0
      color {colorUseRule=UseColor_Specified, specifiedColor=<rgba 0 0 0 0>}
      cornerRadii=<0 0 0 0>
     }
     tintColor {colorUseRule=UseColor_Specified, specifiedColor=<rgba 1 1 1 1>}
     uVRegion {bIsValid=false, max=<0 0>, min=<0 0>}
    }
   }
   slot {class=OverlaySlot, properties {verticalAlignment=VAlign_Fill}}
  }
  WarningMessage {
   widgetClassPath=&/BriefErrorMessageNew
   bIsVariable=true
   parent=VerticalBox_0
   widget=WarningMessage
   properties {fontSize=16, visibility=Collapsed, wrap Text At=640, contentColor=<rgba 1 0.9301 0 1>}
   slot {class=VerticalBoxSlot, properties {verticalAlignment=VAlign_Top}}
  }
  FortTransactionStrip {
   widgetClassPath=&/FortTransactionStripNew
   bIsVariable=true
   parent=VerticalBox_0
   widget=FortTransactionStrip
   properties {isReward=true, transactionString=Reward}
   slot {class=VerticalBoxSlot}
  }
  NotAvailable {widgetClassPath=&Overlay, bIsVariable=true, parent=StateSwitcher, widget=NotAvailable, slot {class=CommonVisibilitySwitcherSlot}}
  BriefErrorMessage1 {
   widgetClassPath=&BriefErrorMessageNew
   bIsVariable=true
   parent=NotAvailable
   widget=BriefErrorMessage1
   properties {fontSize=20}
   slot {class=OverlaySlot}
  }
 ]
}
