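# Buy brief widget on FortBuyBrief, apparently for expert sales in conversations, showing an item's icon, name, rarity, stack size and stock above a transaction strip.
# Transaction info async-loads the preview (large, scaled-down for ammo), then tints text and dynamic materials from rarity data.
# Purchase data and unavailability events switch between the buy view and an out-of-stock or can't-afford error.
# Selection visuals forward to the transaction strip.
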WidgetBlueprint /FortniteConversationUI/Conversation/WBP_FortBuyBrief_ExpertSale
graphs [EventGraph {nodes [
   n1 = Colorize!() -> (OutputDelegate, then, Rarity:EFortRarity Enum=Common) {is=title K2Node_CustomEvent}
   > n2 = GetRarityData!(execute, Rarity:EFortRarity Enum=@n1.Rarity) -> (then, ReturnValue:Fort Rarity Item Data Structure)
   > n3 = GetDynamicMaterial(self=Get Image_ItemBG() -> (Image_ItemBG:Image Object Reference))
   n4 = Break Fort Rarity Item Data!(FortRarityItemData:"Fort Rarity Item Data Structure (by ref)"=@n2) -> (
    Name:Text
    Color1:Linear Color Structure
    Color2:Linear Color Structure
    Color3:Linear Color Structure
    Color4:Linear Color Structure
    Color5:Linear Color Structure
    Radius:float=~
    Falloff:float=~
    Brightness:float=~
    Roughness:float=~
    Glow:float=~
   ) {is=title K2Node_BreakStruct}
   > SetVectorParameterValue(self=@n3, ParameterName=BGColor, Value=@n4.Color3)
   > SetVectorParameterValue(self=@n3, ParameterName=GradientColor, Value=@n4.Color2)
   > SetVectorParameterValue(self=@n3, ParameterName=OutlineColor, Value=@n4.Color1)
   > n5 = GetDynamicMaterial(self=Get Image_RarityGlow() -> (Image_RarityGlow:Image Object Reference))
   > SetVectorParameterValue(self=@n5, ParameterName=Color1, Value=@n4.Color1)
   > SetVectorParameterValue(self=@n5, ParameterName=Color2, Value=@n4.Color3)
   n6 = Set Icon And Name!() -> (OutputDelegate, then, Weapon Icon:Texture 2D Object Reference, Item:Fort Item Object Reference) {is=title K2Node_CustomEvent}
   > SetVisibility(InVisibility=HitTestInvisible)
   > SetBrushFromTexture!(
    execute
    self:Image Object Reference=[Get Image_Icon() -> (Image_Icon:Image Object Reference), Get Image_IconShadow() -> (Image_IconShadow:Image Object Reference)]
    Texture:Texture 2D Object Reference=@n6.Weapon Icon
    bMatchSize:bool=false
   ) -> (then)
   > "SetText (Text)"(
    self=Get Text_WeaponName() -> (Text_WeaponName:Common Text Object Reference)
    InText=GetDisplayName!(self:Fort Item Object Reference=@n6.Item) -> (ReturnValue:Text)
   )
   n7 = GetRarity!(self:Fort Item Object Reference=@n6.Item) -> (ReturnValue:EFortRarity Enum=Common)
   > Colorize!(execute, self, Rarity:EFortRarity Enum=@n7) -> (then) {is=title K2Node_CallFunction}
   > n8 = GetRarityData(Rarity=@n7)
   n9 = Break Fort Rarity Item Data(FortRarityItemData=@n8) -> (Radius=~, Falloff=~, Brightness=~, Roughness=~, Glow=~)
   n10 = Get Text_WeaponRarity() -> (Text_WeaponRarity:Common Text Object Reference)
   > "SetText (Text)"(self=@n10, InText=Format Text!(Format="LOCTEXT(\"({rarity})\")", rarity:Text=@n9) -> (Result))
   > SetColorAndOpacity(self=@n10, InColorAndOpacity=Make Slate Color(SpecifiedColor=@n9.Color1))
   n11 = Event OnTransactionInfoReceived!() -> (OutputDelegate, then, Item:Fort Item Object Reference, TransactionInfo:"Fort Service Transaction Info Structure (by ref)") {is=title K2Node_Event}
   n12 = "Equal (Enum)"(
    A:EFortItemType Enum=GetType!(self:Fort Item Object Reference=@n11.Item) -> (ReturnValue:EFortItemType Enum=WorldItem)
    B:EFortItemType Enum=Ammo
   )
   n13 = Select(
    Option 0:Texture 2D Soft Object Reference=GetSmallPreviewImage!(self:Fort Item Object Reference=@n11.Item) -> (ReturnValue:Texture 2D Soft Object Reference)
    Option 1:Texture 2D Soft Object Reference=GetLargePreviewImage!(self:Fort Item Object Reference=@n11.Item) -> (ReturnValue:Texture 2D Soft Object Reference)
    Index=@n12
   ) -> (ReturnValue:Texture 2D Soft Object Reference)
   > n14 = Async Load Asset(Asset=@n13)
   Set Icon And Name!(
    execute=@n14.Completed
    self
    Weapon Icon:Texture 2D Object Reference=Resolve Soft Reference!(Input:Texture 2D Soft Object Reference=@n13) -> (Output:Texture 2D Object Reference)
    Item:Fort Item Object Reference=@n11.Item
   ) -> (then) {is=title K2Node_CallFunction}
   > SetRenderScale!(
    execute
    self:Widget Object Reference=Get Overlay_ItemIcon() -> (Overlay_ItemIcon:Overlay Object Reference)
    Scale:Vector 2D Structure=Select(Option 0:Vector 2D Structure=(X=1,Y=1), Option 1:Vector 2D Structure=(X=0.7,Y=0.7), Index=@n12) -> (ReturnValue:Vector 2D Structure)
   ) -> (then)
   n15 = Event OnDisplayUnavailability!() -> (OutputDelegate, then, CannotBuyReason:ECannotBuyReason Enum=CannotAfford) {is=title K2Node_Event}
   > SetActiveWidget(
    self=Get StateVisibilitySwitcher() -> (StateVisibilitySwitcher:Common Visibility Switcher Object Reference)
    Widget=Get CannotBuy() -> (CannotBuy:Overlay Object Reference)
   )
   > SetErrorMessage!(
    execute
    self:Brief Error Message New Object Reference=Get BriefErrorMessage() -> (BriefErrorMessage:Brief Error Message New Object Reference)
    ErrorText:Text=Select!(CannotAfford:Text, OutOfStock:Text="LOCTEXT(\"Out Of Stock\")", Index:ECannotBuyReason Enum=@n15.CannotBuyReason) -> (ReturnValue:Text)
   ) -> (then)
   n16 = Event PreConstruct()
   > Branch(Condition=@n16.IsDesignTime)
   > Colorize(Rarity=Common)
   n17 = Event OnPurchaseDataReceived!() -> (OutputDelegate, then, StackSize:int=0, RemainingForSale:int=0) {is=title K2Node_Event}
   > SetActiveWidget(self=Get StateVisibilitySwitcher(), Widget=Get CanBuy() -> (CanBuy:Overlay Object Reference))
   n18 = Get StackSize() -> (StackSize:Common Text Object Reference)
   > SetVisibility(
    self=@n18
    InVisibility=Select!(
     Option 0:ESlateVisibility Enum=HitTestInvisible
     Option 1:ESlateVisibility Enum=Hidden
     Index:bool="Equal (Integer)"(A=@n17.StackSize, B=1) {is=title K2Node_CallFunction}
    ) -> (ReturnValue:ESlateVisibility Enum)
   )
   > "SetText (Text)"(self=@n18, InText=Format Text!(Format="LOCTEXT(\"x{StackSize}\")", StackSize:int=@n17.StackSize) -> (Result))
   > "SetText (Text)"(
    self=Get StockRemaining() -> (StockRemaining:Common Text Object Reference)
    InText=Format Text!(Format="LOCTEXT(\"Stock: {RemainingForSale}\")", RemainingForSale:int=@n17.RemainingForSale) -> (Result)
   )
   n19 = Event SetSelectedVisuals!() -> (OutputDelegate, then, IsSelected:bool) {is=title K2Node_Event}
   > n20 = Cast To FortTransactionStripNew!(execute, Object=Get TransactionStrip_Main() -> (TransactionStrip_Main:Fort Transaction Strip Object Reference)) -> (then, CastFailed, AsFort Transaction Strip New:Fort Transaction Strip New Object Reference) {is=title K2Node_DynamicCast}
   > SetSelectedVisuals!(execute, self:Fort Transaction Strip New Object Reference=@n20, IsSelected:bool=@n19.IsSelected) -> (then)
   n21 = Get DebugItem() -> (DebugItem:Fort Item Object Reference)
   Set Icon And Name(Weapon Icon=Resolve Soft Reference(Input=GetSmallPreviewImage(self=@n21)), Item=@n21)
  ]}]
variables [DebugItem {type=FortItem, default=null}]
widgets {
 info {inheritedWidgetCount=1, parentClass=/Script/FortniteConversationUI.FortBuyBrief, rootWidgetClass=&SizeBox, widgetCount=30}
 widgets [
  SizeBox_36 {
   widgetClassPath=&SizeBox
   widget=SizeBox_36
   properties {heightOverride=350, maxDesiredWidth=768, minDesiredWidth=384, widthOverride=350}
  }
  StateVisibilitySwitcher {
   widgetClassPath=&CommonVisibilitySwitcher
   bIsVariable=true
   parent=SizeBox_36
   widget=StateVisibilitySwitcher
   slot {class=SizeBoxSlot}
  }
  CanBuy {
   widgetClassPath=&Overlay
   bIsVariable=true
   parent=StateVisibilitySwitcher
   widget=CanBuy
   slot {class=CommonVisibilitySwitcherSlot}
  }
  VerticalBox_2 {
   widgetClassPath=&VerticalBox
   parent=CanBuy
   widget=VerticalBox_2
   slot {class=OverlaySlot, properties {horizontalAlignment=HAlign_Fill}}
  }
  Border_168 {
   widgetClassPath=&Border
   parent=VerticalBox_2
   widget=Border_168
   properties {brushColor=<rgba 0 0 0 0.446>, padding=<ltrb 4 4 4 4>}
   slot {class=VerticalBoxSlot}
  }
  Overlay_0 {widgetClassPath=&Overlay, parent=Border_168, widget=Overlay_0, slot {class=BorderSlot, properties {padding=<ltrb 4 4 4 4>}}}
  HorizontalBox_0 {widgetClassPath=&HorizontalBox, parent=Overlay_0, widget=HorizontalBox_0, slot {class=OverlaySlot}}
  Overlay_WeaponPreview {
   widgetClassPath=&Overlay
   parent=HorizontalBox_0
   widget=Overlay_WeaponPreview
   slot {class=HorizontalBoxSlot, properties {horizontalAlignment=HAlign_Left, verticalAlignment=VAlign_Top}}
  }
  Overlay_ItemBG {
   widgetClassPath=&Overlay
   parent=Overlay_WeaponPreview
   widget=Overlay_ItemBG
   slot {class=OverlaySlot, properties {verticalAlignment=VAlign_Center}}
  }
  Image_84 {
   widgetClassPath=&Image
   bIsVariable=true
   parent=Overlay_ItemBG
   widget=Image_84
   properties {colorAndOpacity=<rgba 0.0052 0.0052 0.0056 0.2>, renderTransform {angle=0, scale=<1 1>, shear=<0 0>, translation=<8 8>}}
   slot {class=OverlaySlot, properties {horizontalAlignment=HAlign_Fill, verticalAlignment=VAlign_Fill}}
  }
  Overlay_1 {
   widgetClassPath=&Overlay
   parent=Overlay_ItemBG
   widget=Overlay_1
   slot {class=OverlaySlot, properties {horizontalAlignment=HAlign_Fill, verticalAlignment=VAlign_Fill, padding=<ltrb 3 3 3 3>}}
  }
  Image_ItemBG {
   widgetClassPath=&Image
   bIsVariable=true
   parent=Overlay_1
   widget=Image_ItemBG
   properties {brush {
     bIsDynamicallyLoaded=false
     drawAs=Image
     imageType=NoImage
     mirroring=NoMirror
     resourceName=None
     resourceObject=/Game/Athena/Playlists/Goose/UI/Material/M_GooseItemBG_Inst
     tiling=NoTile
     imageSize=<90 90>
     margin=<ltrb 0 0 0 0>
     outlineSettings {
      bUseBrushTransparency=false
      roundingType=HalfHeightRadius
      width=0
      color {colorUseRule=UseColor_Specified, specifiedColor=<rgba 0 0 0 0>}
      cornerRadii=<0 0 0 0>
     }
     tintColor {colorUseRule=UseColor_Specified, specifiedColor=<rgba 1 1 1 1>}
     uVRegion {bIsValid=false, max=<0 0>, min=<0 0>}
    }}
   slot {class=OverlaySlot, properties {horizontalAlignment=HAlign_Fill, verticalAlignment=VAlign_Fill}}
  }
  Image_RarityGlow {
   widgetClassPath=&Image
   bIsVariable=true
   parent=Overlay_1
   widget=Image_RarityGlow
   properties {
    renderOpacity=0
    visibility=HitTestInvisible
    brush {
     bIsDynamicallyLoaded=false
     drawAs=Image
     imageType=NoImage
     mirroring=NoMirror
     resourceName=None
     resourceObject=/Game/Athena/HUD/Interaction/M_UI_RotatingBurst
     tiling=NoTile
     imageSize=<220 220>
     margin=<ltrb 0 0 0 0>
     outlineSettings {
      bUseBrushTransparency=false
      roundingType=HalfHeightRadius
      width=0
      color {colorUseRule=UseColor_Specified, specifiedColor=<rgba 0 0 0 0>}
      cornerRadii=<0 0 0 0>
     }
     tintColor {colorUseRule=UseColor_Specified, specifiedColor=<rgba 1 1 1 1>}
     uVRegion {bIsValid=false, max=<0 0>, min=<0 0>}
    }
   }
   slot {class=OverlaySlot, properties {horizontalAlignment=HAlign_Center, verticalAlignment=VAlign_Center, padding=<ltrb -200 -200 -200 -200>}}
  }
  SizeBox_0 {
   widgetClassPath=&SizeBox
   parent=Overlay_1
   widget=SizeBox_0
   properties {bOverride_HeightOverride=true, bOverride_WidthOverride=true, heightOverride=128, widthOverride=128}
   slot {class=OverlaySlot, properties {horizontalAlignment=HAlign_Fill, verticalAlignment=VAlign_Fill}}
  }
  Overlay_ItemIcon {widgetClassPath=&Overlay, bIsVariable=true, parent=SizeBox_0, widget=Overlay_ItemIcon, slot {class=SizeBoxSlot}}
  Image_IconShadow {
   widgetClassPath=&Image
   bIsVariable=true
   parent=Overlay_ItemIcon
   widget=Image_IconShadow
   properties {
    brush {
     bIsDynamicallyLoaded=false
     drawAs=Image
     imageType=NoImage
     mirroring=NoMirror
     resourceName=None
     resourceObject=/Game/UI/Foundation/Textures/Icons/Weapons/Items/T-Icon-Weapons-SK-AutoHigh-Suppressed
     tiling=NoTile
     imageSize=<128 128>
     margin=<ltrb 0 0 0 0>
     outlineSettings {
      bUseBrushTransparency=false
      roundingType=HalfHeightRadius
      width=0
      color {colorUseRule=UseColor_Specified, specifiedColor=<rgba 0 0 0 0>}
      cornerRadii=<0 0 0 0>
     }
     tintColor {colorUseRule=UseColor_Specified, specifiedColor=<rgba 1 1 1 1>}
     uVRegion {bIsValid=false, max=<0 0>, min=<0 0>}
    }
    colorAndOpacity=<rgba 0.0052 0.0052 0.0056 0.3>
    renderTransform {angle=0, scale=<1 1>, shear=<0 0>, translation=<6 12>}
   }
   slot {class=OverlaySlot, properties {horizontalAlignment=HAlign_Fill, verticalAlignment=VAlign_Fill}}
  }
  Image_Icon {
   widgetClassPath=&Image
   bIsVariable=true
   parent=Overlay_ItemIcon
   widget=Image_Icon
   properties {brush {
     bIsDynamicallyLoaded=false
     drawAs=Image
     imageType=NoImage
     mirroring=NoMirror
     resourceName=None
     resourceObject=&T-Icon-Weapons-SK-AutoHigh-Suppressed
     tiling=NoTile
     imageSize=<128 128>
     margin=<ltrb 0 0 0 0>
     outlineSettings {
      bUseBrushTransparency=false
      roundingType=HalfHeightRadius
      width=0
      color {colorUseRule=UseColor_Specified, specifiedColor=<rgba 0 0 0 0>}
      cornerRadii=<0 0 0 0>
     }
     tintColor {colorUseRule=UseColor_Specified, specifiedColor=<rgba 1 1 1 1>}
     uVRegion {bIsValid=false, max=<0 0>, min=<0 0>}
    }}
   slot {class=OverlaySlot, properties {horizontalAlignment=HAlign_Fill, verticalAlignment=VAlign_Fill}}
  }
  StackSize {
   widgetClassPath=&CommonTextBlock
   bIsVariable=true
   parent=Overlay_ItemIcon
   widget=StackSize
   properties {
    justification=Right
    text=x100
    font {
     bForceMonospaced=false
     bMaterialIsStencil=false
     fontMaterial=null
     fontObject=&Burbank
     letterSpacing=0
     monospacedWidth=1
     size=18
     skewAmount=0
     typefaceFontName=BurbankRegular-Black
     outlineSettings {
      bApplyOutlineToDropShadows=false
      bMiteredCorners=false
      bSeparateFillAlpha=false
      outlineMaterial=&M-FontShadow
      outlineSize=2
      outlineColor=<rgba 0.0052 0.0052 0.0056 0.8>
     }
    }
    shadowColorAndOpacity=<rgba 0.0052 0.0052 0.0056 0.687>
    shadowOffset=<5 3>
   }
   slot {class=OverlaySlot, properties {horizontalAlignment=HAlign_Right, verticalAlignment=VAlign_Bottom}}
  }
  VerticalBox_ItemText {
   widgetClassPath=&VerticalBox
   parent=HorizontalBox_0
   widget=VerticalBox_ItemText
   slot {class=HorizontalBoxSlot, properties {verticalAlignment=VAlign_Top, padding=<ltrb 10 0 10 0>}}
  }
  VerticalBox_1 {
   widgetClassPath=&VerticalBox
   parent=VerticalBox_ItemText
   widget=VerticalBox_1
   properties {renderTransform {angle=0, scale=<1 1>, shear=<0 0>, translation=<0 7>}}
   slot {class=VerticalBoxSlot, properties {horizontalAlignment=HAlign_Left, verticalAlignment=VAlign_Center}}
  }
  SizeBox_2 {
   widgetClassPath=&SizeBox
   parent=VerticalBox_1
   widget=SizeBox_2
   properties {bOverride_MaxDesiredWidth=true, bOverride_MinDesiredWidth=true, maxDesiredWidth=512, minDesiredWidth=320, widthOverride=345.8244}
   slot {class=VerticalBoxSlot, properties {size {sizeRule=Fill, value=1}}}
  }
  Text_WeaponName {
   widgetClassPath=&CommonTextBlock
   bIsVariable=true
   parent=SizeBox_2
   widget=Text_WeaponName
   properties {
    text=Weapon Name
    textTransformPolicy=ToUpper
    wrapTextAt=512
    font {
     bForceMonospaced=false
     bMaterialIsStencil=false
     fontMaterial=null
     fontObject=&Burbank
     letterSpacing=0
     monospacedWidth=1
     size=24
     skewAmount=0
     typefaceFontName=Burbank-Black
     outlineSettings {
      bApplyOutlineToDropShadows=true
      bMiteredCorners=false
      bSeparateFillAlpha=true
      outlineMaterial=&M-FontShadow
      outlineSize=3
      outlineColor=<rgba 0 0 0 0.5>
     }
    }
    renderTransform {angle=0, scale=<1 1>, shear=<-8 0>, translation=<0 0>}
    shadowOffset=<0 0>
   }
   slot {class=SizeBoxSlot}
  }
  Text_WeaponRarity {
   widgetClassPath=&CommonTextBlock
   bIsVariable=true
   parent=VerticalBox_1
   widget=Text_WeaponRarity
   properties {
    text="{Rarity}"
    textTransformPolicy=ToUpper
    font {
     bForceMonospaced=false
     bMaterialIsStencil=false
     fontMaterial=null
     fontObject=&Burbank
     letterSpacing=0
     monospacedWidth=1
     size=18
     skewAmount=0
     typefaceFontName=BurbankRegular-Black
     outlineSettings {
      bApplyOutlineToDropShadows=false
      bMiteredCorners=false
      bSeparateFillAlpha=false
      outlineMaterial=null
      outlineSize=0
      outlineColor=<rgba 0 0 0 1>
     }
    }
    renderTransform {angle=0, scale=<1 1>, shear=<-8 0>, translation=<0 0>}
    shadowOffset=<0 0>
   }
   slot {class=VerticalBoxSlot}
  }
  ItemCost {
   widgetClassPath=&Overlay
   bIsVariable=true
   parent=VerticalBox_ItemText
   widget=ItemCost
   properties {renderTransform {angle=0, scale=<1 1>, shear=<0 0>, translation=<0 6>}}
   slot {class=VerticalBoxSlot, properties {horizontalAlignment=HAlign_Left, verticalAlignment=VAlign_Center}}
  }
  HorizontalBox_1 {widgetClassPath=&HorizontalBox, parent=ItemCost, widget=HorizontalBox_1, slot {class=OverlaySlot}}
  CommonTextBlock {
   widgetClassPath=&CommonTextBlock
   parent=Overlay_0
   widget=CommonTextBlock
   properties {
    text=Buy
    textTransformPolicy=ToUpper
    visibility=Collapsed
    font {
     bForceMonospaced=false
     bMaterialIsStencil=false
     fontMaterial=null
     fontObject=&Burbank
     letterSpacing=0
     monospacedWidth=1
     size=18
     skewAmount=0
     typefaceFontName=BurbankRegular-Black
     outlineSettings {
      bApplyOutlineToDropShadows=false
      bMiteredCorners=false
      bSeparateFillAlpha=false
      outlineMaterial=null
      outlineSize=2
      outlineColor=<rgba 0.0052 0.0052 0.0056 0.6>
     }
    }
    renderTransform {angle=0, scale=<1 1>, shear=<-8 0>, translation=<0 -20>}
    shadowOffset=<0 0>
   }
   slot {class=OverlaySlot, properties {verticalAlignment=VAlign_Fill}}
  }
  StockRemaining {
   widgetClassPath=&CommonTextBlock
   bIsVariable=true
   parent=Overlay_0
   widget=StockRemaining
   properties {
    justification=Center
    mobileFontSizeMultiplier=1.4
    text=Stock: 1
    textTransformPolicy=ToUpper
    font {
     bForceMonospaced=false
     bMaterialIsStencil=false
     fontMaterial=null
     fontObject=&Burbank
     letterSpacing=0
     monospacedWidth=1
     size=14
     skewAmount=0
     typefaceFontName=Burbank-PlayerPower
     outlineSettings {
      bApplyOutlineToDropShadows=false
      bMiteredCorners=false
      bSeparateFillAlpha=false
      outlineMaterial=&M-FontShadow
      outlineSize=2
      outlineColor=<rgba 0.0052 0.0052 0.0056 0.8>
     }
    }
    renderTransform {angle=0, scale=<1 1>, shear=<-8 0>, translation=<0 0>}
    shadowOffset=<0 0>
   }
   slot {class=OverlaySlot, properties {horizontalAlignment=HAlign_Right, verticalAlignment=VAlign_Bottom, padding=<ltrb 0 0 4 -5>}}
  }
  TransactionStrip_Main {
   widgetClassPath=&/FortTransactionStripNew
   bInherited=true
   bIsVariable=true
   parent=VerticalBox_2
   widget=TransactionStrip_Main
   properties {transactionString=Buy}
   slot {class=VerticalBoxSlot}
  }
  CannotBuy {
   widgetClassPath=&Overlay
   bIsVariable=true
   parent=StateVisibilitySwitcher
   widget=CannotBuy
   slot {class=CommonVisibilitySwitcherSlot}
  }
  BriefErrorMessage {
   widgetClassPath=&/BriefErrorMessageNew
   bIsVariable=true
   parent=CannotBuy
   widget=BriefErrorMessage
   properties {fontSize=20}
   slot {class=OverlaySlot}
  }
 ]
}
